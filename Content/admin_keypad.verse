using { /Fortnite.com/Devices }
using { /Verse.org/Simulation }
using { /Fortnite.com/UI }
using { /UnrealEngine.com/Temporary/UI }
using { /UnrealEngine.com/Temporary/SpatialMath }
using { /UnrealEngine.com/Temporary/UI }
using { /Verse.org/Colors/NamedColors }
using { /Verse.org/Colors }

ZeroMessage <localizes>  : message = "0"
OneMessage <localizes>  : message = "1"
TwoMessage <localizes>  : message = "2"
ThreeMessage <localizes>  : message = "3"
FourMessage <localizes>  : message = "4"
FiveMessage <localizes>  : message = "5"
SixMessage <localizes>  : message = "6"
SevenMessage <localizes>  : message = "7"
EightMessage <localizes>  : message = "8"
NineMessage <localizes>  : message = "9"
# EnterMessage <localizes>  : message = "Enter"
# ClearMessage <localizes>  : message = "Clear"
# CancelMessage <localizes>  : message = "Cancel"
EnterMessage <localizes>  : message = "✔"
ClearMessage <localizes>  : message = "←"
CancelMessage <localizes>  : message = "ˣ"
DashMessage <localizes>  : message = "-"
HashTagMessage <localizes>  : message = "#"

DisclaimerMessage <localizes>  : message = "* All public codes can be found around the map."


admin_keypad_ui := class():
    var MaybeKeypadCanvas : ?canvas = false
    var TxtInput : text_block = text_block:
        DefaultTextColor := White
        DefaultShadowColor := Black
        DefaultShadowOffset := option . vector2{X:= 1.0, Y := 1.0}
        DefaultShadowOpacity := 0.85

    var MaybeButton0Subscribe : ?cancelable = false
    var MaybeButton1Subscribe : ?cancelable = false
    var MaybeButton2Subscribe : ?cancelable = false
    var MaybeButton3Subscribe : ?cancelable = false
    var MaybeButton4Subscribe : ?cancelable = false
    var MaybeButton5Subscribe : ?cancelable = false
    var MaybeButton6Subscribe : ?cancelable = false
    var MaybeButton7Subscribe : ?cancelable = false
    var MaybeButton8Subscribe : ?cancelable = false
    var MaybeButton9Subscribe : ?cancelable = false
    var MaybeButtonEnterSubscribe : ?cancelable = false
    var MaybeButtonClearSubscribe : ?cancelable = false
    var MaybeButtonCancelSubscribe : ?cancelable = false
    var MaybeButtonDashSubscribe : ?cancelable = false
    var MaybeButtonHashTagSubscribe : ?cancelable = false


    MaxCharCount : int = 15
    ButtonMargin : margin = margin{Left := 5.0, Top := 5.0, Right := 5.0, Bottom := 5.0}

    Button0 : button_loud = button_loud{DefaultText := ZeroMessage}
    Button1 : button_loud = button_loud{DefaultText := OneMessage}
    Button2 : button_loud = button_loud{DefaultText := TwoMessage}
    Button3 : button_loud = button_loud{DefaultText := ThreeMessage}
    Button4 : button_loud = button_loud{DefaultText := FourMessage}
    Button5 : button_loud = button_loud{DefaultText := FiveMessage}
    Button6 : button_loud = button_loud{DefaultText := SixMessage}
    Button7 : button_loud = button_loud{DefaultText := SevenMessage}
    Button8 : button_loud = button_loud{DefaultText := EightMessage}
    Button9 : button_loud = button_loud{DefaultText := NineMessage}
    ButtonEnter : button_regular = button_regular{DefaultText := EnterMessage}
    ButtonClear : button_regular = button_regular{DefaultText := ClearMessage}
    ButtonCancel : button_quiet = button_quiet{DefaultText := CancelMessage}
    ButtonDash : button_loud = button_loud{DefaultText := DashMessage}
    ButtonHashTag : button_loud = button_loud{DefaultText := HashTagMessage}

    UpdateDisplay(Message:widget_message, Char:char):void=

        CurrentLen := TxtInput.GetText().Length
        if:
            CurrentLen < MaxCharCount
        then:
            Print("Update Display")
            CurrentDisplayString := TxtInput.GetText()
            NewDisplayString := "{CurrentDisplayString}{Char}"
            TxtInput.SetText(StringToMessage(NewDisplayString))

    SubmitCode(Message:widget_message):void=
        Print("Submit Code")

    ClearCode(Message:widget_message):void=
        Print("Clear Code")
        TxtInput.SetText(StringToMessage(""))

    BackspaceCode(Message:widget_message):void=
        Print("Backspace Code")
        CurrentDisplayString := TxtInput.GetText()
        NewDisplayString := for(i -> aChar : CurrentDisplayString, i < (CurrentDisplayString.Length - 1)):
                aChar
        TxtInput.SetText(StringToMessage(NewDisplayString))


    CloseKeypad(Message:widget_message):void=
        Print("Removing Keypad")
        if:
            Player := Message.Player
            PlayerUI := GetPlayerUI[Player]
            KeypadWidget := MaybeKeypadCanvas?
        then:
            PlayerUI.RemoveWidget(KeypadWidget)
            set MaybeKeypadCanvas = false

            RemoveButtonSubscribes()

        else:
            Print("Failed to remove keypad")


       
    RemoveButtonSubscribes():void=
        if:
            TheClearButton := MaybeButtonClearSubscribe?
        then:
            TheClearButton.Cancel()
            set MaybeButtonClearSubscribe = false

        if:
            TheCloseButton := MaybeButtonCancelSubscribe?
        then:
            TheCloseButton.Cancel()
            set MaybeButtonCancelSubscribe = false

        if:
            TheButton0 := MaybeButton0Subscribe?
        then:
            TheButton0.Cancel()
            set MaybeButton0Subscribe = false

        if:
            TheButton1 := MaybeButton1Subscribe?
        then:
            TheButton1.Cancel()
            set MaybeButton1Subscribe = false

        if:
            TheButton2 := MaybeButton2Subscribe?
        then:
            TheButton2.Cancel()
            set MaybeButton2Subscribe = false

        if:
            TheButton3 := MaybeButton3Subscribe?
        then:
            TheButton3.Cancel()
            set MaybeButton3Subscribe = false

        if:
            TheButton4 := MaybeButton4Subscribe?
        then:
            TheButton4.Cancel()
            set MaybeButton4Subscribe = false

        if:
            TheButton5 := MaybeButton5Subscribe?
        then:
            TheButton5.Cancel()
            set MaybeButton5Subscribe = false

        if:
            TheButton6 := MaybeButton6Subscribe?
        then:
            TheButton6.Cancel()
            set MaybeButton6Subscribe = false

        if:
            TheButton7 := MaybeButton7Subscribe?
        then:
            TheButton7.Cancel()
            set MaybeButton7Subscribe = false

        if:
            TheButton8 := MaybeButton8Subscribe?
        then:
            TheButton8.Cancel()
            set MaybeButton8Subscribe = false

        if:
            TheButton9 := MaybeButton9Subscribe?
        then:
            TheButton9.Cancel()
            set MaybeButton9Subscribe = false

        if:
            TheButtonEnter := MaybeButtonEnterSubscribe?
        then:
            TheButtonEnter.Cancel()
            set MaybeButtonEnterSubscribe = false

        if:
            TheButtonDash := MaybeButtonDashSubscribe?
        then:
            TheButtonDash.Cancel()
            set MaybeButtonDashSubscribe = false

        if:
            TheButtonHashTag := MaybeButtonHashTagSubscribe?
        then:
            TheButtonHashTag.Cancel()
            set MaybeButtonHashTagSubscribe = false
    


    MakeKeypad():canvas=

        Btn0Subscribe := Button0.OnClick().SubscribeMessage(UpdateDisplay, '0')
        set MaybeButton0Subscribe = option{Btn0Subscribe}
        Btn1Subscribe := Button1.OnClick().SubscribeMessage(UpdateDisplay, '1')
        set MaybeButton1Subscribe = option{Btn1Subscribe}
        Btn2Subscribe := Button2.OnClick().SubscribeMessage(UpdateDisplay, '2')
        set MaybeButton2Subscribe = option{Btn2Subscribe}
        Btn3Subscribe := Button3.OnClick().SubscribeMessage(UpdateDisplay, '3')
        set MaybeButton3Subscribe = option{Btn3Subscribe}
        Btn4Subscribe := Button4.OnClick().SubscribeMessage(UpdateDisplay, '4')
        set MaybeButton4Subscribe = option{Btn4Subscribe}
        Btn5Subscribe := Button5.OnClick().SubscribeMessage(UpdateDisplay, '5')
        set MaybeButton5Subscribe = option{Btn5Subscribe}
        Btn6Subscribe := Button6.OnClick().SubscribeMessage(UpdateDisplay, '6')
        set MaybeButton6Subscribe = option{Btn6Subscribe}
        Btn7Subscribe := Button7.OnClick().SubscribeMessage(UpdateDisplay, '7')
        set MaybeButton7Subscribe = option{Btn7Subscribe}
        Btn8Subscribe := Button8.OnClick().SubscribeMessage(UpdateDisplay, '8')
        set MaybeButton8Subscribe = option{Btn8Subscribe}
        Btn9Subscribe := Button9.OnClick().SubscribeMessage(UpdateDisplay, '9')
        set MaybeButton9Subscribe = option{Btn9Subscribe}
        # ButtonEnter.OnClick().Subscribe(SubmitCode)

        BtnDashSubscribe := ButtonDash.OnClick().SubscribeMessage(UpdateDisplay, '-')
        set MaybeButtonDashSubscribe = option{BtnDashSubscribe}
        BtnHashTagSubscribe := ButtonHashTag.OnClick().SubscribeMessage(UpdateDisplay, '#')
        set MaybeButtonHashTagSubscribe = option{BtnHashTagSubscribe}

        # ClearButtonSubscribe := ButtonClear.OnClick().Subscribe(ClearCode)
        ClearButtonSubscribe := ButtonClear.OnClick().Subscribe(BackspaceCode)
        set MaybeButtonClearSubscribe = option{ClearButtonSubscribe}
        CloseButtonSubscribe := ButtonCancel.OnClick().Subscribe(CloseKeypad)
        set MaybeButtonCancelSubscribe = option{CloseButtonSubscribe}

        RoundBackgroundTextures := (
            _UI.UI_Textures.ui_rounded_top_left_border,
            _UI.UI_Textures.ui_rounded_top_middle_border,
            _UI.UI_Textures.ui_rounded_top_right_border,
            _UI.UI_Textures.ui_rounded_middle_left_border,
            _UI.UI_Textures.ui_solid_black_75,
            _UI.UI_Textures.ui_rounded_middle_right_border,
            _UI.UI_Textures.ui_rounded_bottom_left_border,
            _UI.UI_Textures.ui_rounded_bottom_middle_border,
            _UI.UI_Textures.ui_rounded_bottom_right_border
        )

        MyCanvas := canvas:
            Slots := array:
                canvas_slot:
                    Anchors := anchors{Minimum := vector2{X := 0.5, Y := 0.5}, Maximum := vector2{X := 0.5, Y := 0.5}}
                    Offsets := margin{Left := 0.0, Top := 0.0, Right := 0.0, Bottom := 0.0}
                    Alignment := vector2{X := 0.5, Y := 0.5}
                    SizeToContent := true
                    Widget := stack_box:
                        Orientation := orientation.Vertical
                        Slots := array:
                            stack_box_slot:
                                HorizontalAlignment := horizontal_alignment.Fill
                                VerticalAlignment := vertical_alignment.Fill
                                Padding := margin{Left := 0.0, Top := 0.0, Right := 0.0, Bottom := 0.0}
                                Widget := overlay:
                                    Slots := array:
                                        #background
                                        overlay_slot:
                                            HorizontalAlignment := horizontal_alignment.Fill
                                            VerticalAlignment := vertical_alignment.Fill
                                            Padding := margin{Left := 0.0, Top := 0.0, Right := 0.0, Bottom := 0.0}
                                            Widget := MakeRoundedOverlayNine(RoundBackgroundTextures, 32.0)

                                        #content
                                        overlay_slot:
                                            HorizontalAlignment := horizontal_alignment.Fill
                                            VerticalAlignment := vertical_alignment.Fill
                                            Padding := margin{Left := 8.0, Top := 8.0, Right := 8.0, Bottom := 8.0}
                                            Widget := overlay:
                                                Slots := array:
                                                    overlay_slot:
                                                        HorizontalAlignment := horizontal_alignment.Fill
                                                        VerticalAlignment := vertical_alignment.Fill
                                                        Padding := margin{Left := 0.0, Top := 0.0, Right := 0.0, Bottom := 0.0}
                                                        Widget := stack_box:
                                                            Orientation := orientation.Vertical
                                                            Slots := array:

                                                                # Combines logo and close, in overlay
                                                                stack_box_slot:
                                                                    HorizontalAlignment := horizontal_alignment.Fill
                                                                    VerticalAlignment := vertical_alignment.Top
                                                                    Padding := margin{Left := 0.0, Top := 0.0, Right := 0.0, Bottom := 0.0}
                                                                    Widget := overlay:
                                                                        Slots := array:

                                                                            # determin sie of overlay
                                                                            # overlay_slot:
                                                                            #     HorizontalAlignment := horizontal_alignment.Fill
                                                                            #     VerticalAlignment := vertical_alignment.Fill
                                                                            #     Padding := margin{Left := 0.0, Top := 0.0, Right := 0.0, Bottom := 0.0}
                                                                            #     Widget := color_block:
                                                                            #         DefaultColor := Pink
                                                                            #         DefaultOpacity := 0.0
                                                                            #         DefaultDesiredSize := vector2{X := 320.0, Y := 128.0}

                                                                            overlay_slot:
                                                                                HorizontalAlignment := horizontal_alignment.Center
                                                                                VerticalAlignment := vertical_alignment.Center
                                                                                Padding := margin{Left := 0.0, Top := 0.0, Right := 0.0, Bottom := 0.0}
                                                                                Widget := texture_block:
                                                                                    DefaultImage := Textures.fort_who_logo_transparent
                                                                                    DefaultDesiredSize := vector2{X := 128.0, Y := 128.0}

                                                                            # Close Button
                                                                            overlay_slot:
                                                                                HorizontalAlignment := horizontal_alignment.Right
                                                                                VerticalAlignment := vertical_alignment.Top
                                                                                Padding := margin{Left := 4.0, Top := 4.0, Right := 4.0, Bottom := 4.0}
                                                                                Widget := overlay:
                                                                                    Slots := array:
                                                                                        # overlay_slot:
                                                                                        #     HorizontalAlignment := horizontal_alignment.Fill
                                                                                        #     VerticalAlignment := vertical_alignment.Fill
                                                                                        #     Padding := margin{Left := 0.0, Top := 0.0, Right := 0.0, Bottom := 0.0}
                                                                                        #     Widget := color_block:
                                                                                        #         DefaultColor := Red
                                                                                        #         DefaultOpacity := 0.7
                                                                                        #         DefaultDesiredSize := vector2{X := 16.0, Y := 16.0}
                                                                                        overlay_slot:
                                                                                            HorizontalAlignment := horizontal_alignment.Fill
                                                                                            VerticalAlignment := vertical_alignment.Fill
                                                                                            Padding := margin{Left := 0.0, Top := 0.0, Right := 0.0, Bottom := 0.0}
                                                                                            Widget := ButtonCancel
                                                                        
                                                            

                                                                # Input
                                                                stack_box_slot:
                                                                    HorizontalAlignment := horizontal_alignment.Fill
                                                                    VerticalAlignment := vertical_alignment.Center
                                                                    Padding := margin{Left := 0.0, Top := 0.0, Right := 0.0, Bottom := 0.0}
                                                                    Widget := stack_box:
                                                                        Orientation := orientation.Horizontal
                                                                        Slots := array:
                                                                            stack_box_slot:
                                                                                Distribution := option{1.0}
                                                                                HorizontalAlignment := horizontal_alignment.Fill
                                                                                VerticalAlignment := vertical_alignment.Fill
                                                                                Padding := margin{Left := 0.0, Top := 0.0, Right := 0.0, Bottom := 0.0}
                                                                                Widget := overlay:
                                                                                    Slots := array:
                                                                                        # Input BG
                                                                                        overlay_slot:
                                                                                            HorizontalAlignment := horizontal_alignment.Fill
                                                                                            VerticalAlignment := vertical_alignment.Fill
                                                                                            Padding := ButtonMargin
                                                                                            Widget := MakeRoundedOverlayInner(MakeColorFromSRGB(0.28, 0.28, 0.28), ?TextureSize:=24.0)

                                                                                        #determin width of input
                                                                                        overlay_slot:
                                                                                            HorizontalAlignment := horizontal_alignment.Fill
                                                                                            VerticalAlignment := vertical_alignment.Fill
                                                                                            Padding := ButtonMargin
                                                                                            Widget := color_block:
                                                                                                DefaultColor := Pink
                                                                                                DefaultOpacity := 0.0
                                                                                                DefaultDesiredSize := vector2{X := 320.0, Y := 28.0}
                                                                                        
                                                                                        # Input Text
                                                                                        overlay_slot:
                                                                                            HorizontalAlignment := horizontal_alignment.Center
                                                                                            VerticalAlignment := vertical_alignment.Center
                                                                                            Padding := margin{Left := 16.0, Top := 0.0, Right := 16.0, Bottom := 0.0}
                                                                                            Widget := TxtInput
                                                                            stack_box_slot:
                                                                                HorizontalAlignment := horizontal_alignment.Right
                                                                                VerticalAlignment := vertical_alignment.Fill
                                                                                Padding := ButtonMargin
                                                                                Widget := ButtonClear
                                                                # Buttons
                                                                stack_box_slot:
                                                                    Distribution := option{1.0}
                                                                    HorizontalAlignment := horizontal_alignment.Fill
                                                                    VerticalAlignment := vertical_alignment.Fill
                                                                    Padding := margin{Left := 0.0, Top := 16.0, Right := 0.0, Bottom := 0.0}
                                                                    Widget := stack_box:
                                                                        Orientation := orientation.Vertical
                                                                        Slots := array:
                                                                            # number buttons
                                                                            stack_box_slot:
                                                                                HorizontalAlignment := horizontal_alignment.Fill
                                                                                VerticalAlignment := vertical_alignment.Fill
                                                                                Padding := margin{Left := 0.0, Top := 0.0, Right := 0.0, Bottom := 0.0}
                                                                                Widget := overlay:
                                                                                    Slots := array:

                                                                                        #used to make the buttons the same size. sets the overlay width
                                                                                        overlay_slot:
                                                                                            HorizontalAlignment := horizontal_alignment.Fill
                                                                                            VerticalAlignment := vertical_alignment.Fill
                                                                                            Padding := margin{Left := 0.0, Top := 0.0, Right := 0.0, Bottom := 0.0}
                                                                                            Widget := color_block:
                                                                                                DefaultColor := Green
                                                                                                DefaultOpacity := 0.0
                                                                                                DefaultDesiredSize := vector2{X := 270.0, Y := 28.0}
                                                                                        # overlay_slot:
                                                                                        #     HorizontalAlignment := horizontal_alignment.Fill
                                                                                        #     VerticalAlignment := vertical_alignment.Fill
                                                                                        #     Padding := margin{Left := 0.0, Top := 0.0, Right := 0.0, Bottom := 0.0}
                                                                                        #     Widget := MakeRoundedOverlayInner(UI_Pink)
                                                                                        overlay_slot:
                                                                                            HorizontalAlignment := horizontal_alignment.Fill
                                                                                            VerticalAlignment := vertical_alignment.Fill
                                                                                            Padding := margin{Left := 0.0, Top := 0.0, Right := 0.0, Bottom := 0.0}
                                                                                            Widget := stack_box:
                                                                                                Orientation := orientation.Vertical
                                                                                                Slots := array:
                                                                                                    stack_box_slot:
                                                                                                        HorizontalAlignment := horizontal_alignment.Fill
                                                                                                        VerticalAlignment := vertical_alignment.Fill
                                                                                                        Padding := margin{Left := 0.0, Top := 0.0, Right := 0.0, Bottom := 0.0}
                                                                                                        Widget := stack_box:
                                                                                                            Orientation := orientation.Horizontal
                                                                                                            Slots := array:
                                                                                                                stack_box_slot:
                                                                                                                    Distribution := option{1.0}
                                                                                                                    HorizontalAlignment := horizontal_alignment.Fill
                                                                                                                    VerticalAlignment := vertical_alignment.Fill
                                                                                                                    Padding := ButtonMargin
                                                                                                                    Widget := Button1
                                                                                                                stack_box_slot:
                                                                                                                    Distribution := option{1.0}
                                                                                                                    HorizontalAlignment := horizontal_alignment.Fill
                                                                                                                    VerticalAlignment := vertical_alignment.Fill
                                                                                                                    Padding := ButtonMargin
                                                                                                                    Widget := Button2
                                                                                                                stack_box_slot:
                                                                                                                    Distribution := option{1.0}
                                                                                                                    HorizontalAlignment := horizontal_alignment.Fill
                                                                                                                    VerticalAlignment := vertical_alignment.Fill
                                                                                                                    Padding := ButtonMargin
                                                                                                                    Widget := Button3
                                                                                                    stack_box_slot:
                                                                                                        HorizontalAlignment := horizontal_alignment.Fill
                                                                                                        VerticalAlignment := vertical_alignment.Fill
                                                                                                        Padding := margin{Left := 0.0, Top := 0.0, Right := 0.0, Bottom := 0.0}
                                                                                                        Widget := stack_box:
                                                                                                            Orientation := orientation.Horizontal
                                                                                                            Slots := array:
                                                                                                                stack_box_slot:
                                                                                                                    Distribution := option{1.0}
                                                                                                                    HorizontalAlignment := horizontal_alignment.Fill
                                                                                                                    VerticalAlignment := vertical_alignment.Fill
                                                                                                                    Padding := ButtonMargin
                                                                                                                    Widget := Button4
                                                                                                                stack_box_slot:
                                                                                                                    Distribution := option{1.0}
                                                                                                                    HorizontalAlignment := horizontal_alignment.Fill
                                                                                                                    VerticalAlignment := vertical_alignment.Fill
                                                                                                                    Padding := ButtonMargin
                                                                                                                    Widget := Button5
                                                                                                                stack_box_slot:
                                                                                                                    Distribution := option{1.0}
                                                                                                                    HorizontalAlignment := horizontal_alignment.Fill
                                                                                                                    VerticalAlignment := vertical_alignment.Fill
                                                                                                                    Padding := ButtonMargin
                                                                                                                    Widget := Button6
                                                                                                    stack_box_slot:
                                                                                                        HorizontalAlignment := horizontal_alignment.Fill
                                                                                                        VerticalAlignment := vertical_alignment.Fill
                                                                                                        Padding := margin{Left := 0.0, Top := 0.0, Right := 0.0, Bottom := 0.0}
                                                                                                        Widget := stack_box:
                                                                                                            Orientation := orientation.Horizontal
                                                                                                            Slots := array:
                                                                                                                stack_box_slot:
                                                                                                                    Distribution := option{1.0}
                                                                                                                    HorizontalAlignment := horizontal_alignment.Fill
                                                                                                                    VerticalAlignment := vertical_alignment.Fill
                                                                                                                    Padding := ButtonMargin
                                                                                                                    Widget := Button7
                                                                                                                stack_box_slot:
                                                                                                                    Distribution := option{1.0}
                                                                                                                    HorizontalAlignment := horizontal_alignment.Fill
                                                                                                                    VerticalAlignment := vertical_alignment.Fill
                                                                                                                    Padding := ButtonMargin
                                                                                                                    Widget := Button8
                                                                                                                stack_box_slot:
                                                                                                                    Distribution := option{1.0}
                                                                                                                    HorizontalAlignment := horizontal_alignment.Fill
                                                                                                                    VerticalAlignment := vertical_alignment.Fill
                                                                                                                    Padding := ButtonMargin
                                                                                                                    Widget := Button9
                                                                                                    stack_box_slot:
                                                                                                        HorizontalAlignment := horizontal_alignment.Fill
                                                                                                        VerticalAlignment := vertical_alignment.Fill
                                                                                                        Padding := margin{Left := 0.0, Top := 0.0, Right := 0.0, Bottom := 0.0}
                                                                                                        Widget := stack_box:
                                                                                                            Orientation := orientation.Horizontal
                                                                                                            Slots := array:
                                                                                                                stack_box_slot:
                                                                                                                    Distribution := option{1.0}
                                                                                                                    HorizontalAlignment := horizontal_alignment.Fill
                                                                                                                    VerticalAlignment := vertical_alignment.Fill
                                                                                                                    Padding := ButtonMargin
                                                                                                                    Widget := ButtonDash
                                                                                                                stack_box_slot:
                                                                                                                    Distribution := option{1.0}
                                                                                                                    HorizontalAlignment := horizontal_alignment.Fill
                                                                                                                    VerticalAlignment := vertical_alignment.Fill
                                                                                                                    Padding := ButtonMargin
                                                                                                                    Widget := Button0
                                                                                                                stack_box_slot:
                                                                                                                    Distribution := option{1.0}
                                                                                                                    HorizontalAlignment := horizontal_alignment.Fill
                                                                                                                    VerticalAlignment := vertical_alignment.Fill
                                                                                                                    Padding := ButtonMargin
                                                                                                                    Widget := ButtonHashTag
                                                                            # function buttons
                                                                            stack_box_slot:
                                                                                HorizontalAlignment := horizontal_alignment.Fill
                                                                                VerticalAlignment := vertical_alignment.Top
                                                                                Padding := margin{Left := 0.0, Top := 16.0, Right := 0.0, Bottom := 0.0}
                                                                                Widget := overlay:
                                                                                    Slots := array:                                                                        
                                                                                        overlay_slot:
                                                                                            HorizontalAlignment := horizontal_alignment.Fill
                                                                                            VerticalAlignment := vertical_alignment.Fill
                                                                                            Padding := margin{Left := 0.0, Top := 0.0, Right := 0.0, Bottom := 0.0}
                                                                                            Widget := stack_box:
                                                                                                Orientation := orientation.Vertical
                                                                                                Slots := array:
                                                                                                    # stack_box_slot:
                                                                                                    #     Distribution := option{1.0}
                                                                                                    #     HorizontalAlignment := horizontal_alignment.Fill
                                                                                                    #     VerticalAlignment := vertical_alignment.Fill
                                                                                                    #     Padding := ButtonMargin
                                                                                                    #     Widget := ButtonClear
                                                                                                    # stack_box_slot:
                                                                                                    #     Distribution := option{1.0}
                                                                                                    #     HorizontalAlignment := horizontal_alignment.Fill
                                                                                                    #     VerticalAlignment := vertical_alignment.Fill
                                                                                                    #     Padding := ButtonMargin
                                                                                                    #     Widget := ButtonCancel
                                                                                                    stack_box_slot:
                                                                                                        Distribution := option{1.0}
                                                                                                        HorizontalAlignment := horizontal_alignment.Fill
                                                                                                        VerticalAlignment := vertical_alignment.Fill
                                                                                                        Padding := ButtonMargin
                                                                                                        Widget := ButtonEnter

                canvas_slot:
                    Anchors := anchors{Minimum := vector2{X := 0.0, Y := 0.95}, Maximum := vector2{X := 0.0, Y := 0.95}}
                    Offsets := margin{Left := 16.0, Top := 0.0, Right := 0.0, Bottom := 0.0}
                    Alignment := vector2{X := 0.0, Y := 0.5}
                    SizeToContent := true
                    Widget := text_block:
                        DefaultText := DisclaimerMessage
                        DefaultTextColor := White
                        DefaultShadowColor := Black
                        DefaultShadowOffset := option . vector2{X:= 1.0, Y := 1.0}
                        DefaultShadowOpacity := 0.85
                      

                                                                                        

                                                                           


admin_keypad := class<concrete>():

    @editable Keycodes : []admin_unlockable = array{}
    ActualKeyCodes : []string = array:
        "1234" # Secret XP
        "5794-2200-8636" # 50 Gold (One Time Use)
        "88-542-###" # 10,000 Gold (Anyone can use carefull)
        "000" # Open Bunker (Moderator Only)
        "#12#25#" # Winter 24, Collect 50 Presents Instantly
        "999##" # LlamaHunter Filming Setup
    

    @editable ButtonOpen : button_device = button_device{}

    @editable HUDMessageDevice : hud_message_device = hud_message_device{}
    @editable LogBillboard : billboard_device = billboard_device{}

    var MaybeKeypadUI : [agent]admin_keypad_ui = map{}

    Init()<suspends>:void=
        # ButtonOpen.InteractedWithEvent.Subscribe(Self.HandleOpenKeypad)
        loop:
            Agent := ButtonOpen.InteractedWithEvent.Await()
            HandleOpenKeypad(Agent)

    HandleOpenKeypad(Agent:agent):void=
        Print("Keypad Opening")

        if:
            HasKeyPadUI := MaybeKeypadUI[Agent]
        then:
            # do nothing already created
            if:
                HasKeyPadUI.MaybeKeypadCanvas?
            then:
                # do nothing already open
            else:
                if:
                    Player := player[Agent]
                    PlayerUI := GetPlayerUI[Player]
                then:
                    KeypadCanvas := HasKeyPadUI.MakeKeypad()
                    set HasKeyPadUI.MaybeKeypadCanvas = option{KeypadCanvas}
                    PlayerUI.AddWidget(KeypadCanvas, player_ui_slot{ZOrder := 1000, InputMode:=ui_input_mode.All})
        else:
            if:
                Player := player[Agent]
                PlayerUI := GetPlayerUI[Player]
            then:
                NewKeypadUI := admin_keypad_ui{}
                KeyPadCanvas := NewKeypadUI.MakeKeypad()
                set NewKeypadUI.MaybeKeypadCanvas = option{KeyPadCanvas}
                PlayerUI.AddWidget(KeyPadCanvas, player_ui_slot{ZOrder := 1000, InputMode:=ui_input_mode.All})
                NewKeypadUI.ButtonEnter.OnClick().Subscribe(HandleSubmitCode)
                if . set MaybeKeypadUI[Agent] = NewKeypadUI

    HandleSubmitCode(Message:widget_message):void=
        Print("Submit Code")
        if:
            Player := Message.Player
            PlayerUI := GetPlayerUI[Player]
            KeypadWidget := MaybeKeypadUI[Player]
        then:
            CurrentCode := KeypadWidget.TxtInput.GetText()
            if:
                CodeStuff := TestCode[CurrentCode] # returns (Name, Code, Trigger, RequiredRole, NewRoleToAssign, RemoveRole?, GiveWinter24Sticker?)
            then:
                if(IsAtLeast[Player, CodeStuff(3)]):
                    LogBillboard.SetText(TwoLineMessage("Correct!","{CodeStuff(0)}"))
                    Print("Code Correct!\nDesc: {CodeStuff(0)}\nCode: {CodeStuff(1)}\nRequired Role: {CodeStuff(3).ToString()}")
                    CodeStuff(2).Trigger(Player)

                    if(CodeStuff(4) <> player_role.None):
                        if(not CodeStuff(5)?):
                            SetRole(Player, CodeStuff(4))
                            LogBillboard.SetText(TwoLineMessage("Role Assigned", "{CodeStuff(4).ToString()}"))
                            HUDMessageDevice.Show(Player, StringToMessage("Role {CodeStuff(4).ToString()} Assigned"), ?DisplayTime := 2.0)
                        else:
                            RemoveRole(Player, CodeStuff(4))
                            LogBillboard.SetText(TwoLineMessage("Role Removed", "{CodeStuff(4).ToString()}"))
                            HUDMessageDevice.Show(Player, StringToMessage("Role {CodeStuff(4).ToString()} Removed"), ?DisplayTime := 2.0)


                    if(CodeStuff(6)?, Agent := agent[Player]):
                        LogBillboard.SetText(TwoLineMessage("Winter 24 Sticker", "Unlocked"))
                        HUDMessageDevice.Show(Player, StringToMessage("Winter 24 Sticker Unlocked"), ?DisplayTime := 2.0)
                        SetCompletedWinter2024Quest(Agent)
                    
                else:
                    Print("Role not high enough")
                    LogBillboard.SetText(MessageIncorrectRole_EN)
                    HUDMessageDevice.Show(Player, StringToMessage("Role not high enough"), ?DisplayTime := 2.0)
            else:
                Print("Code Incorrect")
                LogBillboard.SetText(MessageIncorrectCode_EN)
                HUDMessageDevice.Show(Player, StringToMessage("Code Incorrect"), ?DisplayTime := 2.0)

                #clear the code
                KeypadWidget.TxtInput.SetText(StringToMessage(""))


            if:
                KeypadCanvas := KeypadWidget.MaybeKeypadCanvas?
            then:
                PlayerUI.RemoveWidget(KeypadCanvas)
                set KeypadWidget.MaybeKeypadCanvas = false

            KeypadWidget.RemoveButtonSubscribes()

                
    
    SetRole(Player:player, Role:player_role):void=
        AssignRole(Player, Role, true)

    RemoveRole(Player:player, Role:player_role):void=
        AssignRole(Player, Role, false)

    TestCode(Code:string)<transacts><decides>:tuple(string, string, trigger_device, player_role, player_role, logic, logic)=
        
        TheNames := for:
            Index -> Unlockable : Keycodes
            ActualKeyCodes[Index] = Code
        do:
            Unlockable.Name

        TheCodes := for:
            Index -> Unlockable : Keycodes
            ActualKeyCodes[Index] = Code
        do:
           Code

        TheTriggers := for:
            Index -> Unlockable : Keycodes
            ActualKeyCodes[Index] = Code
        do:
            Unlockable.Trigger

        TheRoles := for:
            Index -> Unlockable : Keycodes
            ActualKeyCodes[Index] = Code
        do:
            Unlockable.RequiredRole


        UpgradeRoles := for:
            Index -> Unlockable : Keycodes
            ActualKeyCodes[Index] = Code
        do:
            Unlockable.NewRoleToAssign

        AssignOrRemoveRoles := for:
            Index -> Unlockable : Keycodes
            ActualKeyCodes[Index] = Code
        do:
            Unlockable.RemoveRole

        GiveWinter24Sticker := for:
            Index -> Unlockable : Keycodes
            ActualKeyCodes[Index] = Code
        do:
            Unlockable.GiveWinter24Sticker

        (TheNames[0], TheCodes[0], TheTriggers[0], TheRoles[0], UpgradeRoles[0], AssignOrRemoveRoles[0], GiveWinter24Sticker[0])

        