using { /Fortnite.com/Devices }
using { /Verse.org/Simulation }


admin_unlockable := class<concrete>():

    @editable Name : string = ""
    # @editable Code : string = ""
    @editable Trigger : trigger_device = trigger_device{}
    @editable RequiredRole : player_role = player_role.Admin
    @editable NewRoleToAssign : player_role = player_role.Admin
    @editable RemoveRole : logic = false
    @editable GiveWinter24Sticker : logic = false