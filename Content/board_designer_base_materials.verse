using { /Verse.org/Assets }
using { /Verse.org/Colors/NamedColors }
using { /Verse.org/Colors }






# Base Materials
# BaseMaterialPlastic : material = NEW_ASSETS.Materials.M_Props
# BaseMaterialMetallic : material = NEW_ASSETS.Materials.M_Props
# BaseMaterialPutty : material = Materials.ClayMaterial.Material_Clay.M_Clay
# BaseMaterialGalaxy : material = NEW_ASSETS.Materials.M_Props ### TODO: Create Galaxy Material


# Base Material Textures
# Plastic
T_PlayBoard_C : texture = NEW_ASSETS.Textures.T_PlayBoard_C
T_PlayBoard_Base_ORM : texture = NEW_ASSETS.Textures.T_PlayBoard_Base_ORM
T_PlayBoard_N : texture = NEW_ASSETS.Textures.T_PlayBoard_N
T_PlayBoard_SecondaryColor_M : texture = NEW_ASSETS.Textures.T_PlayBoard_SecondaryColor_M
T_Lights_EM : texture = NEW_ASSETS.Textures.T_Lights_EM

T_CardBack_C : texture = NEW_ASSETS.Textures.T_CardBack_C
T_CardBack_ORM : texture = NEW_ASSETS.Textures.T_CardBack_ORM
T_CardBack_N : texture = NEW_ASSETS.Textures.T_CardBack_N

T_Platform_C : texture = NEW_ASSETS.Textures.T_Platform_C
T_Platform_ORM : texture = NEW_ASSETS.Textures.T_Platform_ORM
T_Platform_N : texture = NEW_ASSETS.Textures.T_Platform_N

T_YouStand_C : texture = NEW_ASSETS.Textures.T_YouStand_C
T_YouStand_ORM : texture = NEW_ASSETS.Textures.T_YouStand_ORM
T_YouStand_N : texture = NEW_ASSETS.Textures.T_YouStand_N
T_YouStand_SecondaryColor_M : texture = NEW_ASSETS.Textures.T_YouStand_SecondaryColor_M

# Metallic
# N/A Just use the default material

# Putty
# Place Prop Normal & ORM in BoardNormal and BoardORM Parameters when creating MI's from M_Clay
T_Clay_Pattern_01_GreySacle : texture = Materials.ClayMaterial.Texture_Clay.T_Clay_Pattern_01_GreySacle
T_Clay_Pattern_01_Normal : texture = Materials.ClayMaterial.Texture_Clay.T_Clay_Pattern_01_Normal
T_Clay_Normal : texture = Materials.ClayMaterial.Texture_Clay.T_Clay_Normal
# Putty Speckles
T_Clay_Dot : texture = Materials.ClayMaterial.Texture_Clay.T_Clay_Dot
T_Stars : texture = Materials.ClayMaterial.Texture_Clay.T_Stars 

# Galaxy
# TODO: Create Galaxy Material








CreatePlasticMaterialInstance(PlasticMaterialData:plastic_material_data):material=

    NewMI := NEW_ASSETS.Materials.M_Props_material{}

    set NewMI.BaseColor = PlasticMaterialData.BaseColor
    set NewMI.BaseHueShift = PlasticMaterialData.BaseHueShift
    set NewMI.ColorDesaturation = PlasticMaterialData.ColorDesaturation
    set NewMI.ColorTint = PlasticMaterialData.ColorTint
    set NewMI.SecondaryColorMask = PlasticMaterialData.SecondaryColorMask
    set NewMI.SecondaryColorTint = PlasticMaterialData.SecondaryColorTint
    set NewMI.SecondaryHueShift = PlasticMaterialData.SecondaryHueShift
    set NewMI.EmissiveFlickerSpeed = PlasticMaterialData.EmissiveFlickerSpeed
    set NewMI.FrameEmissiveStrength = PlasticMaterialData.FrameEmissiveStrength
    set NewMI.GlobalEmissiveStrength = PlasticMaterialData.GlobalEmissiveStrength
    # set NewMI.UseFlickeringEmissive = PlasticMaterialData.true # not supported in Verse Materials --- turn on by default
    set NewMI.EmissiveMask2 = PlasticMaterialData.EmissiveMask2
    set NewMI.EmissiveFlickerTint = PlasticMaterialData.EmissiveFlickerTint
    set NewMI.EmissiveTint = PlasticMaterialData.EmissiveTint
    set NewMI.FrameEmissiveTint = PlasticMaterialData.FrameEmissiveTint
    set NewMI.BaseMaskedTexture = PlasticMaterialData.BaseMaskedTexture
    set NewMI.MetallicAmount = PlasticMaterialData.MetallicAmount
    set NewMI.RoughnessMultiplier = PlasticMaterialData.RoughnessMultiplier
    set NewMI.BaseNormal = PlasticMaterialData.BaseNormal
    # set NewMI.MaskAO = (R=True,G=False,B=False,A=False) # not supported in Verse Materials --- turn on by default
    # set NewMI.MaskMetallic = (R=False,G=False,B=True,A=False) # not supported in Verse Materials --- turn on by default
    # set NewMI.MaskRoughness = (R=False,G=True,B=False,A=False) # not supported in Verse Materials --- turn on by default


    return NewMI

CreateMetallicMaterialInstance(PlasticMaterialData:plastic_material_data):material=
    CreatePlasticMaterialInstance(PlasticMaterialData)
    


# CreatePuttyMaterialInstance():material=
#     NewMI : Materials.ClayMaterial.Material_Clay.M_Clay_material = Materials.ClayMaterial.Material_Clay.M_Clay_material{}

#     set NewMI.GlobalTilling = 1.0 #spelled wrong in material
#     set NewMI.Metallic = 0.0
#     set NewMI.Roughness = 0.6
#     set NewMI.ColorClay1 = Pink
#     set NewMI.ColorClay2 = Blue
#     set NewMI.ColorClayDot = White
#     # set NewMI.Use2ColorClay = true # not supported in Verse Materials --- turn on by default
#     # set NewMI.UseBoardNormal = true # not supported in Verse Materials --- turn on by default
#     # set NewMI.UseBoardOrm = true # not supported in Verse Materials --- turn on by default
#     set NewMI.Clay2GreyScaleMap = T_Clay_Pattern_01_GreySacle
#     set NewMI.Clay2NormalMap = T_Clay_Pattern_01_Normal
#     set NewMI.TilingUClaySplite = 1.0 #wtf does splite mean?
#     set NewMI.TilingVClaySplite = 1.0
#     # set NewMI.UseClayDot = true # not supported in Verse Materials --- turn on by default
#     set NewMI.MapClayDot = T_Clay_Dot
#     set NewMI.NumberClayDot = 0.066667
#     set NewMI.TilingUClayDot = 5.0
#     set NewMI.TilingVClayDot = 5.0
#     set NewMI.ClayNormalMap = T_Clay_Normal
#     set NewMI.IntensityClayNormal = 1.0
#     set NewMI.TilingUClayNormal = 1.0
#     set NewMI.TilingVClayNormal = 1.0
#     # set NewMI.BaseEmissive = BlackTexture
#     set NewMI.EmissiveStrength = 1.0
#     # set NewMI.UseFresnel = true # not supported in Verse Materials --- turn on by default
#     set NewMI.BaseReflectFractionInFresnel = 0.0
#     set NewMI.ExponentInFresnel = 4.0
#     set NewMI.IntensityFresnel = 5.0
#     set NewMI.EmissiveFlickerSpeed = 1.0
#     set NewMI.FrameEmissiveStrength = 1.0
#     set NewMI.GlobalEmissiveStrength = 50.0
#     # set NewMI.HasBoardLights = true # not supported in Verse Materials --- turn on by default
#     # set NewMI.UseFlashingEmissive = true # not supported in Verse Materials --- turn on by default
#     set NewMI.BoardNormal = T_PlayBoard_N
#     set NewMI.BoardORM = T_PlayBoard_Base_ORM
#     set NewMI.EmissiveMask2 = T_Lights_EM
#     set NewMI.EmissiveFlickerTint = Pink
#     set NewMI.EmissiveTint = White
#     set NewMI.FrameEmissiveTint = White

    # return NewMI


CreatePuttyMaterialInstance(PuttyMaterialData:putty_material_data):material=
    NewMI : Materials.ClayMaterial.Material_Clay.M_Clay_material = Materials.ClayMaterial.Material_Clay.M_Clay_material{}

    set NewMI.GlobalTiling = PuttyMaterialData.GlobalTilling
    set NewMI.Metallic = PuttyMaterialData.Metallic
    set NewMI.Roughness = PuttyMaterialData.Roughness
    set NewMI.ColorClay1 = PuttyMaterialData.ColorClay1
    set NewMI.ColorClay2 = PuttyMaterialData.ColorClay2
    set NewMI.ColorClayDot = PuttyMaterialData.ColorClayDot
    set NewMI.Clay2GreyScaleMap = PuttyMaterialData.Clay2GreyScaleMap
    set NewMI.Clay2NormalMap = PuttyMaterialData.Clay2NormalMap
    set NewMI.TilingUClaySplite = PuttyMaterialData.TilingUClaySplite
    set NewMI.TilingVClaySplite = PuttyMaterialData.TilingVClaySplite
    set NewMI.MapClayDot = PuttyMaterialData.MapClayDot
    set NewMI.NumberClayDot = PuttyMaterialData.NumberClayDot
    set NewMI.TilingUClayDot = PuttyMaterialData.TilingUClayDot
    set NewMI.TilingVClayDot = PuttyMaterialData.TilingVClayDot
    set NewMI.ClayNormalMap = PuttyMaterialData.ClayNormalMap
    set NewMI.IntensityClayNormal = PuttyMaterialData.IntensityClayNormal
    set NewMI.TilingUClayNormal = PuttyMaterialData.TilingUClayNormal
    set NewMI.TilingVClayNormal = PuttyMaterialData.TilingVClayNormal
    set NewMI.EmissiveStrength = PuttyMaterialData.EmissiveStrength
    set NewMI.BaseReflectFractionInFresnel = PuttyMaterialData.BaseReflectFractionInFresnel
    set NewMI.ExponentInFresnel = PuttyMaterialData.ExponentInFresnel
    set NewMI.IntensityFresnel = PuttyMaterialData.IntensityFresnel
    set NewMI.EmissiveFlickerSpeed = PuttyMaterialData.EmissiveFlickerSpeed
    set NewMI.FrameEmissiveStrength = PuttyMaterialData.FrameEmissiveStrength
    set NewMI.GlobalEmissiveStrength = PuttyMaterialData.GlobalEmissiveStrength
    set NewMI.BoardNormal = PuttyMaterialData.BoardNormal
    set NewMI.BoardORM = PuttyMaterialData.BoardORM
    set NewMI.EmissiveMask2 = PuttyMaterialData.EmissiveMask2
    set NewMI.EmissiveFlickerTint = PuttyMaterialData.EmissiveFlickerTint
    set NewMI.EmissiveTint = PuttyMaterialData.EmissiveTint
    set NewMI.FrameEmissiveTint = PuttyMaterialData.FrameEmissiveTint

    return NewMI