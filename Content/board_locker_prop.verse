using { /Fortnite.com/Devices }
using { /Verse.org/Simulation }
using { /UnrealEngine.com/Temporary/SpatialMath }
using { /Fortnite.com/Devices/CreativeAnimation }

board_locker_prop := class<concrete>():
    @editable Prop : creative_prop = creative_prop{}
    @editable FinalRelativeTransform : transform = transform{}

    @editable BoardCards : ?[]game_board_row = false
    @editable YouStandCard : ?creative_prop = false

    Init():void=
        if:
            HasCardRows := BoardCards?
        then:
            for:
                Row : HasCardRows
                Card : Row.Tiles
            do: 
                set Card.FlipDirection = 1.0

    # SafeHiddenTransform : transform = transform:
    #     # Translation := vector3{X:=6324.0, Y:=1254.0, Z:=195722.0}
    #     # (X=-1352.000000,Y=2194.000000,Z=-3626.000000)
    #     Rotation := IdentityRotation()
    #     Scale := vector3{X:=0.25, Y:=0.25, Z:=0.25}

    Show(?Delay:float=0.0)<suspends>:void=
        Sleep(Delay)

        # DeltaTranslation := FinalTransform.Translation - Prop.GetTransform().Translation
        # # DeltaScale := FinalTransform.Scale - Prop.GetTransform().Scale
        # AngularDist := AngularDistance(Prop.GetTransform().Rotation, FinalTransform.Rotation)
        # DeltaRotation := MakeShortestRotationBetween(Prop.GetTransform().Rotation, FinalTransform.Rotation)
        # ScaleFactor := FinalTransform.Scale / Prop.GetTransform().Scale

        DeltaTranslation := FinalRelativeTransform.Translation
        DeltaRotation := FinalRelativeTransform.Rotation




        if:
            AnimationController := Prop.GetAnimationController[]
        then:
            Keyframes := array:
                keyframe_delta:
                    DeltaLocation := DeltaTranslation
                    DeltaRotation := DeltaRotation
                    DeltaScale := vector3{X:=1.0, Y:=1.0, Z:=1.0} #don't scale
                    Time := 1.0
                    Interpolation := InterpolationTypes.Linear

            AnimationController.SetAnimation(Keyframes, ?Mode:=animation_mode.OneShot)
            AnimationController.Play()
            AnimationController.MovementCompleteEvent.Await()
        else:
            Print("No Animation Controller found on prop")
            # Prop.MoveTo(FinalTransform, 0.5)
        
        Sleep(0.25)

        if(CardRows := BoardCards?):
            for(Row : CardRows):
                spawn . Row.FlipRow()

    Hide()<suspends>:void=

        DeltaTranslation := FinalRelativeTransform.Translation - vector3{X:=-1352.0, Y:=2194.0, Z:=-3626.0}
        DeltaRotation := FinalRelativeTransform.Rotation
        DeltaScale := vector3{X:=1.0, Y:=1.0, Z:=1.0} #don't scale
        
        # Prop.MoveTo(SafeHiddenTransform, 0.25)

        if:
            AnimationController := Prop.GetAnimationController[]
        then:
            Keyframes := array:
                keyframe_delta:
                    DeltaLocation := DeltaTranslation
                    DeltaRotation := DeltaRotation
                    DeltaScale := DeltaScale
                    Time := 0.25
                    Interpolation := InterpolationTypes.Linear

            AnimationController.SetAnimation(Keyframes, ?Mode:=animation_mode.OneShot)
            AnimationController.Play()
            AnimationController.MovementCompleteEvent.Await()
        else:
            Print("No Animation Controller found on prop")
            # Prop.MoveTo(FinalTransform, 0.5)


        Sleep(0.25)



        if(CardRows := BoardCards?):
            for(Row : CardRows):
                spawn . Row.FlipRow(?Instant:=true)