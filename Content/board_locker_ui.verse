using { /UnrealEngine.com/Temporary/SpatialMath }
using { /Verse.org/Simulation }
using { /UnrealEngine.com/Temporary/UI }
using { /Fortnite.com/UI }
# using { /Fortnite.com/Characters }
# using { /Fortnite.com/FortPlayerUtilities }
using { /Verse.org/Colors }
using { /Verse.org/Colors/NamedColors }
using { /Verse.org/Assets }
using { /Fortnite.com/Devices }
# using { /Verse.org/Concurrency }
using { /UnrealEngine.com/Temporary/Diagnostics }
using { lerping }


board_locker_ui := class<concrete>():

    ExitLockerUIEvent : event() = event(){}

    ButtonTabGameBoardContainer : overlay = overlay{}
    ButtonTabGameBoardQuiet : button_quiet = button_quiet{DefaultText := MessageBoard_EN}
    ButtonTabGameBoardRegular : button_regular = button_regular{DefaultText := MessageBoard_EN}
    ButtonTabGameBoardLoud : button_loud = button_loud{DefaultText := MessageBoard_EN}
    var ButtonTabGameBoardSubscription : ?cancelable = false
    
    ButtonTabCardsContainer : overlay = overlay{}
    ButtonTabCardsQuiet : button_quiet = button_quiet{DefaultText := MessageCards_EN}
    ButtonTabCardsRegular : button_regular = button_regular{DefaultText := MessageCards_EN}
    ButtonTabCardsLoud : button_loud = button_loud{DefaultText := MessageCards_EN}
    var ButtonTabCardsSubscription : ?cancelable = false

    ButtonTabYouStandContainer : overlay = overlay{}
    ButtonTabYouStandQuiet : button_quiet = button_quiet{DefaultText := MessageStand_EN}
    ButtonTabYouStandRegular : button_regular = button_regular{DefaultText := MessageStand_EN}
    ButtonTabYouStandLoud : button_loud = button_loud{DefaultText := MessageStand_EN}
    var ButtonTabYouStandSubscription : ?cancelable = false

    ButtonTabPlatformContainer : overlay = overlay{}
    ButtonTabPlatformQuiet : button_quiet = button_quiet{DefaultText := MessagePlatform_EN}
    ButtonTabPlatformRegular : button_regular = button_regular{DefaultText := MessagePlatform_EN}
    ButtonTabPlatformLoud : button_loud = button_loud{DefaultText := MessagePlatform_EN}
    var ButtonTabPlatformSubscription : ?cancelable = false

    ButtonFilterClear : button_quiet = button_quiet{DefaultText := MessageClearFilters_EN}
    var ButtonFilterClearSubscription : ?cancelable = false

    ButtonFilterPlasticContainer : overlay = overlay{}
    ButtonFilterPlasticQuiet : button_quiet = button_quiet{DefaultText := MessageMaterialTypePlastic_EN}
    # ButtonFilterPlasticRegular : button_regular = button_regular{DefaultText := MessageMaterialTypePlastic_EN}
    # ButtonFilterPlasticLoud : button_loud = button_loud{DefaultText := MessageMaterialTypePlastic_EN}
    var ButtonFilterPlasticSubscription : ?cancelable = false

    ButtonFilterMetallicContainer : overlay = overlay{}
    ButtonFilterMetallicQuiet : button_quiet = button_quiet{DefaultText := MessageMaterialTypeMetallic_EN}
    # ButtonFilterMetallicRegular : button_regular = button_regular{DefaultText := MessageMaterialTypeMetallic_EN}
    # ButtonFilterMetallicLoud : button_loud = button_loud{DefaultText := MessageMaterialTypeMetallic_EN}
    var ButtonFilterMetallicSubscription : ?cancelable = false

    ButtonFilterPuttyContainer : overlay = overlay{}
    ButtonFilterPuttyQuiet: button_quiet = button_quiet{DefaultText := MessageMaterialTypePutty_EN}
    # ButtonFilterPuttyRegular : button_regular = button_regular{DefaultText := MessageMaterialTypePutty_EN}
    # ButtonFilterPuttyLoud : button_loud = button_loud{DefaultText := MessageMaterialTypePutty_EN}
    var ButtonFilterPuttySubscription : ?cancelable = false
    
    ButtonFilterGalaxyContainer : overlay = overlay{}
    ButtonFilterGalaxyQuiet : button_quiet = button_quiet{DefaultText := MessageMaterialTypeGalaxy_EN}
    # ButtonFilterGalaxyRegular : button_regular = button_regular{DefaultText := MessageMaterialTypeGalaxy_EN}
    # ButtonFilterGalaxyLoud : button_loud = button_loud{DefaultText := MessageMaterialTypeGalaxy_EN}
    var ButtonFilterGalaxySubscription : ?cancelable = false

    FilterPlasticBGColorBlock : color_block = color_block{}
    FilterMetallicBGColorBlock : color_block = color_block{}
    FilterPuttyBGColorBlock : color_block = color_block{}
    FilterGalaxyBGColorBlock : color_block = color_block{}


    ButtonPageDown : button_loud = button_loud{DefaultText := MessageArrowDown_EN}
    var ButtonPageDownSubscription : ?cancelable = false

    ButtonPageUp : button_loud = button_loud{DefaultText := MessageArrowUp_EN}
    var ButtonPageUpSubscription : ?cancelable = false

    ButtonExit : button_loud = button_loud{DefaultText := MessageExit_EN}
    var ButtonExitSubscription : ?cancelable = false

    GoldTextBlock : text_block = text_block:
        DefaultText := TestGoldMessage_EN
        DefaultTextColor := White

    PaintTokensTextBlock : text_block = text_block{}

    BoardDesignerTitleTextureBlock : texture_block = texture_block{
        DefaultImage := _UI.UI_Textures.BoardDesignerTitle_EN
        DefaultDesiredSize := vector2{X:=1024.0, Y:=96.0}
    }

    NewNextPageTextureBlock : texture_block = texture_block:
        DefaultImage := _UI.UI_Textures.Notification_One
        DefaultDesiredSize := vector2{X:=48.0, Y:=48.0}

    NewPrevPageTextureBlock : texture_block = texture_block:
        DefaultImage := _UI.UI_Textures.Notification_One
        DefaultDesiredSize := vector2{X:=48.0, Y:=48.0}

    var CurrentTabIndex : int = 0 #Board, Cards, YouStand, Platform
    var CurrentFilterTags : []string = array{}
    var MaybeLoadingOverlayCanvas: ?canvas = false
    var MaybeLockerUICanvas : ?canvas = false
    var MaybeModal : ?canvas_slot = false
    var MaybeModalCloseSubscription : ?cancelable = false
    var MaybeModalBuySubscription : ?cancelable = false


    var PaintStyles : []paint_selection_data = array{}
    var CurrentActivePaintStyle : paint_ui_option = paint_ui_option{}

    @editable PlayerReference : player_reference_device = player_reference_device{}
    @editable Camera : gameplay_camera_fixed_point_device = gameplay_camera_fixed_point_device{}
    @editable TrackerHIDETRACKERS : tracker_device = tracker_device{}


    @editable PropGameBoard : board_locker_prop = board_locker_prop{}
    @editable PropYouStand : board_locker_prop = board_locker_prop{}
    @editable PropPlatform : board_locker_prop = board_locker_prop{}

    var TrackerPaintQuest : tracker_device = tracker_device{}
    var CurrentActiveTrackedQuestPaintStyle: paint_ui_option = paint_ui_option{}


    
    InitPaintStyles():void=
        #OLD
        # set PaintStyles = array:
        #     paint_selection_data{PaintIndex := 0, PaintName := "Pink", PaintStyleIndex := 0, Cost := 0, PaintBGColor := MakeColorFromHSV(320.0, 1.0, 1.0), PaintBGTexture := _UI.UI_Textures.board_locker_paint_option_bg, PaintStyleTexture := _UI.UI_Textures.board_locker_paint_style_plastic,                                                                                                                                                                                                                    IsActive := true,   IsOwned := true,  IsLocked := false, HasUsed := true, IsNew := false, FilterTags := array{"Plastic"}}
        #     paint_selection_data{PaintIndex := 1, PaintName := "Yellow", PaintStyleIndex := 0, Cost := 0, PaintBGColor := MakeColorFromHSV(60.0, 1.0, 1.0), PaintBGTexture := _UI.UI_Textures.board_locker_paint_option_bg, PaintStyleTexture := _UI.UI_Textures.board_locker_paint_style_plastic,                                                                                                                                                                                                                   IsActive := false,  IsOwned := true,  IsLocked := false, HasUsed := false, IsNew := true,  FilterTags := array{"Plastic"}}
        #     paint_selection_data{PaintIndex := 2, PaintName := "Orange", PaintStyleIndex := 0, Cost := 0, PaintBGColor := MakeColorFromHSV(30.0, 1.0, 1.0), PaintBGTexture := _UI.UI_Textures.board_locker_paint_option_bg, PaintStyleTexture := _UI.UI_Textures.board_locker_paint_style_plastic,                                                                                                                                                                                                                   IsActive := false,  IsOwned := true,  IsLocked := false, HasUsed := false, IsNew := true,  FilterTags := array{"Plastic"}}
        #     paint_selection_data{PaintIndex := 3, PaintName := "Red", PaintStyleIndex := 0, Cost := 0, PaintBGColor := MakeColorFromHSV(0.0, 1.0, 1.0), PaintBGTexture := _UI.UI_Textures.board_locker_paint_option_bg, PaintStyleTexture := _UI.UI_Textures.board_locker_paint_style_plastic,                                                                                                                                                                                                                       IsActive := false,  IsOwned := true,  IsLocked := false, HasUsed := false, IsNew := true,  FilterTags := array{"Plastic"}}
        #     paint_selection_data{PaintIndex := 4, PaintName := "Blue", PaintStyleIndex := 0, Cost := 0, PaintBGColor := MakeColorFromHSV(240.0, 1.0, 1.0), PaintBGTexture := _UI.UI_Textures.board_locker_paint_option_bg, PaintStyleTexture := _UI.UI_Textures.board_locker_paint_style_plastic,                                                                                                                                                                                                                    IsActive := false,  IsOwned := false, IsLocked := false, HasUsed := false, IsNew := true,  FilterTags := array{"Plastic"}}
        #     paint_selection_data{PaintIndex := 5, PaintName := "Light Blue", PaintStyleIndex := 0, Cost := 50, PaintBGColor := MakeColorFromHSV(210.0, 1.0, 1.0), PaintBGTexture := _UI.UI_Textures.board_locker_paint_option_bg, PaintStyleTexture := _UI.UI_Textures.board_locker_paint_style_plastic,    IsActive := false,  IsOwned := false, IsLocked := false, HasUsed := false, IsNew := true, FilterTags := array{"Plastic"}}
        #     paint_selection_data{PaintIndex := 6, PaintName := "Green", PaintStyleIndex := 0, Cost := 50, PaintBGColor := MakeColorFromHSV(120.0, 1.0, 1.0), PaintBGTexture := _UI.UI_Textures.board_locker_paint_option_bg, PaintStyleTexture := _UI.UI_Textures.board_locker_paint_style_plastic,                                                                                                                                                                                                                  IsActive := false,  IsOwned := false, IsLocked := false, HasUsed := false, IsNew := true, FilterTags := array{"Plastic"}}
        #     paint_selection_data{PaintIndex := 7, PaintName := "Purple", PaintStyleIndex := 0, Cost := 50, PaintBGColor := MakeColorFromHSV(270.0, 1.0, 1.0), PaintBGTexture := _UI.UI_Textures.board_locker_paint_option_bg, PaintStyleTexture := _UI.UI_Textures.board_locker_paint_style_plastic,                                                                                                                                                                                                                 IsActive := false,  IsOwned := false, IsLocked := false, HasUsed := false, IsNew := true, FilterTags := array{"Plastic"}}
            
        #     paint_selection_data{PaintIndex := 8, PaintName := "Black", PaintStyleIndex := 0, Cost := 50, PaintBGColor := MakeColorFromHSV(0.0, 0.0, 0.0), PaintBGTexture := _UI.UI_Textures.board_locker_paint_option_bg, PaintStyleTexture := _UI.UI_Textures.board_locker_paint_style_plastic,                                                                                                                                                                                                                    IsActive := false,  IsOwned := false, IsLocked := false, HasUsed := false, IsNew := false, FilterTags := array{"Plastic"}}
        #     paint_selection_data{PaintIndex := 9, PaintName := "Gray", PaintStyleIndex := 0, Cost := 50, PaintBGColor := MakeColorFromHSV(0.0, 0.0, 0.5), PaintBGTexture := _UI.UI_Textures.board_locker_paint_option_bg, PaintStyleTexture := _UI.UI_Textures.board_locker_paint_style_plastic,                                                                                                                                                                                                                     IsActive := false,  IsOwned := false, IsLocked := true,  HasUsed := false, IsNew := false, FilterTags := array{"Plastic"}, QuestID:=2}
            
        #     paint_selection_data{PaintIndex := 10, PaintName := "Bronze", PaintStyleIndex := 1, Cost := 50, PaintBGColor := MakeColorFromHSV(30.0, 0.5, 0.5), PaintBGTexture := _UI.UI_Textures.board_locker_paint_option_bg, PaintStyleTexture := _UI.UI_Textures.board_locker_paint_style_metallic,                                                                                                                                                                                                                IsActive := false,  IsOwned := true, IsLocked := false,  HasUsed := true, IsNew := false, FilterTags := array{"Metallic"}, QuestID:=3}
        #     paint_selection_data{PaintIndex := 11, PaintName := "Silver", PaintStyleIndex := 1, Cost := 50, PaintBGColor := MakeColorFromHSV(240.0, 0.5, 0.75), PaintBGTexture := _UI.UI_Textures.board_locker_paint_option_bg, PaintStyleTexture := _UI.UI_Textures.board_locker_paint_style_metallic,                                                                                                                                                                                                              IsActive := false,  IsOwned := false, IsLocked := true,  HasUsed := false, IsNew := false, FilterTags := array{"Metallic"}, QuestID:=4}
        #     paint_selection_data{PaintIndex := 12, PaintName := "Gold", PaintStyleIndex := 1, Cost := 50, PaintBGColor := MakeColorFromHSV(60.0, 1.0, 1.0), PaintBGTexture := _UI.UI_Textures.board_locker_paint_option_bg, PaintStyleTexture := _UI.UI_Textures.board_locker_paint_style_metallic,                                                                                                                                                                                                                  IsActive := false,  IsOwned := false, IsLocked := true,  HasUsed := false, IsNew := false, FilterTags := array{"Metallic"}, QuestID:=5}
        #     paint_selection_data{PaintIndex := 13, PaintName := "Pink Putty", PaintStyleIndex := 2, Cost := 50, PaintBGColor := MakeColorFromHSV(320.0, 1.0, 1.0),  PaintBGTexture := _UI.UI_Textures.board_locker_paint_option_bg, PaintStyleTexture := _UI.UI_Textures.board_locker_paint_style_putty,                                                                                                                                                                                                             IsActive := false,  IsOwned := true, IsLocked := false,  HasUsed := false, IsNew := true, FilterTags := array{"Putty"}, SecondaryBGColor:=option{MakeColorFromHSV(350.0, 1.0, 1.0)}, QuestID:=6}
        #     paint_selection_data{PaintIndex := 14, PaintName := "Yellow Putty", PaintStyleIndex := 2, Cost := 50, PaintBGColor := MakeColorFromHSV(60.0, 1.0, 1.0),  PaintBGTexture := _UI.UI_Textures.board_locker_paint_option_bg, PaintStyleTexture := _UI.UI_Textures.board_locker_paint_style_putty,                                                                                                                                                                                                            IsActive := false,  IsOwned := false, IsLocked := true,  HasUsed := false, IsNew := false, FilterTags := array{"Putty"}, SecondaryBGColor:=option{MakeColorFromHSV(90.0, 1.0, 1.0)}, QuestID:=7}
        #     paint_selection_data{PaintIndex := 15, PaintName := "Space Putty", PaintStyleIndex := 2, Cost := 50, PaintBGColor := MakeColorFromHSV(240.0, 1.0, 1.0),  PaintBGTexture := _UI.UI_Textures.board_locker_paint_option_bg, PaintStyleTexture := _UI.UI_Textures.board_locker_paint_style_putty,                                                                                                                                                                                                            IsActive := false,  IsOwned := false, IsLocked := true,  HasUsed := false, IsNew := false, FilterTags := array{"Putty"}, SecondaryBGColor:=option{MakeColorFromHSV(270.0, 1.0, 1.0)}, QuestID:=8}
        #     paint_selection_data{PaintIndex := 16, PaintName := "Orange Putty", PaintStyleIndex := 2, Cost := 50, PaintBGColor := MakeColorFromHSV(30.0, 1.0, 1.0),  PaintBGTexture := _UI.UI_Textures.board_locker_paint_option_bg, PaintStyleTexture := _UI.UI_Textures.board_locker_paint_style_putty,                                                                                                                                                                                                            IsActive := false,  IsOwned := false, IsLocked := true,  HasUsed := false, IsNew := false, FilterTags := array{"Putty"}, SecondaryBGColor:=option{MakeColorFromHSV(60.0, 1.0, 1.0)}, QuestID:=9}
            
        #     paint_selection_data{PaintIndex := 17, PaintName := "Red Putty", PaintStyleIndex := 2, Cost := 50, PaintBGColor := MakeColorFromHSV(0.0, 1.0, 1.0), PaintBGTexture := _UI.UI_Textures.board_locker_paint_option_bg, PaintStyleTexture := _UI.UI_Textures.board_locker_paint_style_putty,                                                                                                                                                                                                                 IsActive := false,  IsOwned := false, IsLocked := true,  HasUsed := false, IsNew := false, FilterTags := array{"Putty"}, SecondaryBGColor:=option{MakeColorFromHSV(30.0, 1.0, 1.0)}, QuestID:=10}
        #     paint_selection_data{PaintIndex := 18, PaintName := "Blue Putty", PaintStyleIndex := 2, Cost := 50, PaintBGColor := MakeColorFromHSV(240.0, 1.0, 1.0), PaintBGTexture := _UI.UI_Textures.board_locker_paint_option_bg, PaintStyleTexture := _UI.UI_Textures.board_locker_paint_style_putty,                                                                                                                                                                                                              IsActive := false,  IsOwned := false, IsLocked := true,  HasUsed := false, IsNew := false, FilterTags := array{"Putty"}, SecondaryBGColor:=option{MakeColorFromHSV(270.0, 1.0, 1.0)}, QuestID:=11}
        #     paint_selection_data{PaintIndex := 19, PaintName := "Light Blue Putty", PaintStyleIndex := 2, Cost := 50, PaintBGColor := MakeColorFromHSV(210.0, 1.0, 1.0), PaintBGTexture := _UI.UI_Textures.board_locker_paint_option_bg, PaintStyleTexture := _UI.UI_Textures.board_locker_paint_style_putty, IsActive := false,  IsOwned := false, IsLocked := true,  HasUsed := false, IsNew := false, FilterTags := array{"Putty"}, SecondaryBGColor:=option{MakeColorFromHSV(240.0, 1.0, 1.0)}, QuestID:=12}
        #     paint_selection_data{PaintIndex := 20, PaintName := "Green Putty", PaintStyleIndex := 2, Cost := 50, PaintBGColor := MakeColorFromHSV(120.0, 1.0, 1.0), PaintBGTexture := _UI.UI_Textures.board_locker_paint_option_bg, PaintStyleTexture := _UI.UI_Textures.board_locker_paint_style_putty,                                                                                                                                                                                                             IsActive := false,  IsOwned := false, IsLocked := true,  HasUsed := false, IsNew := false, FilterTags := array{"Putty"}, SecondaryBGColor:=option{MakeColorFromHSV(150.0, 1.0, 1.0)}, QuestID:=13}
            
        #NEW 
        #     MakeColorFromHSV(324.38913, 0.948145, 0.947917) # Pink
        #     MakeColorFromHSV(45.33017, 1.0, 0.947307) # Yellow
        #     MakeColorFromHSV(30.0, 1.0, 0.947307) # Orange
        #     MakeColorFromHSV(0.0, 1.0, 0.947307) # Red
        #     MakeColorFromHSV(300.0, 1.0, 0.947307) # Purple
        #     MakeColorFromHSV(240.0, 1.0, 0.947307) # Dark Blue
        #     MakeColorFromHSV(210.0, 1.0, 0.947307) # Light Blue
        #     MakeColorFromHSV(120.0, 1.0, 0.947307) # Green
        #     MakeColorFromHSV(0.0, 0.0, 0.947307) # White
        #     MakeColorFromHSV(0.0, 0.0, 0.5) # Grey
        #     MakeColorFromHSV(0.0, 0.0, 0.0) # Black
        #     MakeColorFromHSV(5.0, 0.6, 0.95) # Peach
        #     MakeColorFromHSV(30.0, 0.5, 0.5) # Bronze
        #     MakeColorFromHSV(0.0, 0.0, 0.75) # Silver
        #     MakeColorFromHSV(45.0, 0.5, 0.5) # Gold
        #     MakeColorFromHSV(324.38913, 0.648145, 0.947917) # Pink Clay
        #     MakeColorFromHSV(45.33017, 1.0, 0.947307) # Yellow Clay
        #     MakeColorFromHSV(270.0, 0.5, 0.5) # Space Clay        
        #     MakeColorFromHSV(30.0, 1.0, 0.947307) # Orange Clay
        set PaintStyles = array:
            paint_selection_data:
                PaintIndex := 0
                PaintName := "Pink Plastic"
                PaintStyleIndex := 0
                Cost := 0
                PaintBGColor := MakeColorFromHSV(324.38913, 0.948145, 0.947917)   
                PaintBGTexture := _UI.UI_Textures.board_locker_paint_option_bg
                PaintStyleTexture := _UI.UI_Textures.board_locker_paint_style_plastic
                IsActive := true
                IsOwned := true
                IsLocked := false
                HasUsed := true
                IsNew := false
                FilterTags := array{"Plastic"}
            paint_selection_data:
                PaintIndex := 1
                PaintName := "Yellow Plastic"
                PaintStyleIndex := 0
                Cost := 0
                PaintBGColor := MakeColorFromHSV(54.906708, 0.93, 0.968686)       
                PaintBGTexture := _UI.UI_Textures.board_locker_paint_option_bg
                PaintStyleTexture := _UI.UI_Textures.board_locker_paint_style_plastic
                IsActive := false
                IsOwned := true
                IsLocked := false
                HasUsed := true
                IsNew := false
                FilterTags := array{"Plastic"}
            paint_selection_data:
                PaintIndex := 2
                PaintName := "Orange Plastic"
                PaintStyleIndex := 0
                Cost := 0
                PaintBGColor := MakeColorFromHSV(30.0, 1.0, 0.947307)             
                PaintBGTexture := _UI.UI_Textures.board_locker_paint_option_bg
                PaintStyleTexture := _UI.UI_Textures.board_locker_paint_style_plastic
                IsActive := false
                IsOwned := true
                IsLocked := false
                HasUsed := false
                IsNew := true
                FilterTags := array{"Plastic"}
            paint_selection_data:
                PaintIndex := 3
                PaintName := "Red Plastic"
                PaintStyleIndex := 0
                Cost := 0
                PaintBGColor := MakeColorFromHSV(0.0, 1.0, 0.947307)
                PaintBGTexture := _UI.UI_Textures.board_locker_paint_option_bg
                PaintStyleTexture := _UI.UI_Textures.board_locker_paint_style_plastic
                IsActive := false
                IsOwned := true
                IsLocked := false
                HasUsed := false
                IsNew := true
                FilterTags := array{"Plastic"}
            paint_selection_data:
                PaintIndex := 4
                PaintName := "Dark Blue Plastic"
                PaintStyleIndex := 0
                Cost := 0
                PaintBGColor := MakeColorFromHSV(240.0, 1.0, 0.947307)            
                PaintBGTexture := _UI.UI_Textures.board_locker_paint_option_bg
                PaintStyleTexture := _UI.UI_Textures.board_locker_paint_style_plastic
                IsActive := false
                IsOwned := true
                IsLocked := false
                HasUsed := false
                IsNew := true
                FilterTags := array{"Plastic"}
            paint_selection_data:
                PaintIndex := 5
                PaintName := "Light Blue Plastic"
                PaintStyleIndex := 0
                Cost := 0
                PaintBGColor := MakeColorFromHSV(210.0, 1.0, 0.947307)
                PaintBGTexture := _UI.UI_Textures.board_locker_paint_option_bg
                PaintStyleTexture := _UI.UI_Textures.board_locker_paint_style_plastic
                IsActive := false
                IsOwned := true
                IsLocked := false
                HasUsed := false
                IsNew := true
                FilterTags := array{"Plastic"}
            paint_selection_data:
                PaintIndex := 6
                PaintName := "Green Plastic"
                PaintStyleIndex := 0
                Cost := 0
                PaintBGColor := MakeColorFromHSV(120.0, 1.0, 0.947307)            
                PaintBGTexture := _UI.UI_Textures.board_locker_paint_option_bg
                PaintStyleTexture := _UI.UI_Textures.board_locker_paint_style_plastic
                IsActive := false
                IsOwned := true
                IsLocked := false
                HasUsed := false
                IsNew := true
                FilterTags := array{"Plastic"}
            paint_selection_data:
                PaintIndex := 7
                PaintName := "White Plastic"
                PaintStyleIndex := 0
                Cost := 0
                PaintBGColor := MakeColorFromHSV(0.0, 0.0, 0.947307)              
                PaintBGTexture := _UI.UI_Textures.board_locker_paint_option_bg
                PaintStyleTexture := _UI.UI_Textures.board_locker_paint_style_plastic
                IsActive := false
                IsOwned := true
                IsLocked := false
                HasUsed := false
                IsNew := true
                FilterTags := array{"Plastic"}
            paint_selection_data:
                PaintIndex := 8
                PaintName := "Grey Plastic"
                PaintStyleIndex := 0
                Cost := 0
                PaintBGColor := MakeColorFromHSV(0.0, 0.0, 0.5)                   
                PaintBGTexture := _UI.UI_Textures.board_locker_paint_option_bg
                PaintStyleTexture := _UI.UI_Textures.board_locker_paint_style_plastic
                IsActive := false
                IsOwned := true
                IsLocked := true
                HasUsed := false
                IsNew := false
                FilterTags := array{"Plastic"}
                QuestID:=2
            paint_selection_data:
                PaintIndex := 9
                PaintName := "Black Plastic"
                PaintStyleIndex := 0
                Cost := 0
                PaintBGColor := MakeColorFromHSV(0.0, 0.0, 0.0)
                PaintBGTexture := _UI.UI_Textures.board_locker_paint_option_bg
                PaintStyleTexture := _UI.UI_Textures.board_locker_paint_style_plastic
                IsActive := false
                IsOwned := true
                IsLocked := true
                HasUsed := false
                IsNew := false
                FilterTags := array{"Plastic"}
                QuestID:=3
            paint_selection_data:
                PaintIndex := 10
                PaintName := "Peach Plastic"
                PaintStyleIndex := 0
                Cost := 0
                PaintBGColor := MakeColorFromHSV(5.0, 0.6, 0.95)
                PaintBGTexture := _UI.UI_Textures.board_locker_paint_option_bg
                PaintStyleTexture := _UI.UI_Textures.board_locker_paint_style_plastic
                IsActive := false
                IsOwned := true
                IsLocked := true
                HasUsed := false
                IsNew := false
                FilterTags := array{"Plastic"}
                QuestID:=4
            paint_selection_data:
                PaintIndex := 11
                PaintName := "Bronze Metallic"
                PaintStyleIndex := 1
                Cost := 50
                PaintBGColor := MakeColorFromHSV(22.203491, 1.0, 1.0)            
                PaintBGTexture := _UI.UI_Textures.board_locker_paint_option_bg
                PaintStyleTexture := _UI.UI_Textures.board_locker_paint_style_metallic
                IsActive := false
                IsOwned := false
                IsLocked := false
                HasUsed := true
                IsNew := false
                FilterTags := array{"Metallic"}
                QuestID:=5
            paint_selection_data:
                PaintIndex := 12
                PaintName := "Silver Metallic"
                PaintStyleIndex := 1
                Cost := 50
                PaintBGColor := MakeColorFromHSV(0.0, 0.0, 0.95)                 
                PaintBGTexture := _UI.UI_Textures.board_locker_paint_option_bg
                PaintStyleTexture := _UI.UI_Textures.board_locker_paint_style_metallic
                IsActive := false
                IsOwned := false
                IsLocked := true
                HasUsed := false
                IsNew := false
                FilterTags := array{"Metallic"}
                QuestID:=6
            paint_selection_data:
                PaintIndex := 13
                PaintName := "Gold Metallic"
                PaintStyleIndex := 1
                Cost := 50
                PaintBGColor := MakeColorFromHSV(44.86673, 1.0, 0.947307)        
                PaintBGTexture := _UI.UI_Textures.board_locker_paint_option_bg
                PaintStyleTexture := _UI.UI_Textures.board_locker_paint_style_metallic
                IsActive := false
                IsOwned := false
                IsLocked := true
                HasUsed := false
                IsNew := false
                FilterTags := array{"Metallic"}
                QuestID:=7
            paint_selection_data:
                PaintIndex := 14
                PaintName := "Pink Putty"
                PaintStyleIndex := 2
                Cost := 50
                PaintBGColor := MakeColorFromHSV(325.234528, 0.944199, 0.947307) 
                PaintBGTexture := _UI.UI_Textures.board_locker_paint_option_bg
                PaintStyleTexture := _UI.UI_Textures.board_locker_paint_style_putty
                IsActive := false
                IsOwned := false
                IsLocked := false
                HasUsed := false
                IsNew := true
                FilterTags := array{"Putty"}
                QuestID:=8
                SecondaryBGColor:=option{MakeColorFromHSV(325.234558, 0.8125, 0.947307)}
            paint_selection_data:
                PaintIndex := 15
                PaintName := "Yellow Putty"
                PaintStyleIndex := 2
                Cost := 50
                PaintBGColor := MakeColorFromHSV(54.906708, 0.93, 0.968686)      
                PaintBGTexture := _UI.UI_Textures.board_locker_paint_option_bg
                PaintStyleTexture := _UI.UI_Textures.board_locker_paint_style_putty
                IsActive := false
                IsOwned := false
                IsLocked := true
                HasUsed := false
                IsNew := false
                FilterTags := array{"Putty"}
                QuestID:=9
                SecondaryBGColor:=option{MakeColorFromHSV(41.906738, 0.990269, 1.0)}
            paint_selection_data:
                PaintIndex := 16
                PaintName := "Space Putty"
                PaintStyleIndex := 2
                Cost := 50
                PaintBGColor := MakeColorFromHSV(260.715973, 0.5, 0.057292)      
                PaintBGTexture := _UI.UI_Textures.board_locker_paint_option_bg
                PaintStyleTexture := _UI.UI_Textures.board_locker_paint_style_putty
                IsActive := false
                IsOwned := false
                IsLocked := true
                HasUsed := false
                IsNew := false
                FilterTags := array{"Putty"}
                QuestID:=10
                SecondaryBGColor:=option{MakeColorFromHSV(267.916809, 1.0, 0.015625)}
            paint_selection_data:
                PaintIndex := 17
                PaintName := "Orange Putty"
                PaintStyleIndex := 2
                Cost := 50
                PaintBGColor := MakeColorFromHSV(16.276245, 1.0, 1.0)            
                PaintBGTexture := _UI.UI_Textures.board_locker_paint_option_bg
                PaintStyleTexture := _UI.UI_Textures.board_locker_paint_style_putty
                IsActive := false
                IsOwned := false
                IsLocked := true
                HasUsed := false
                IsNew := false
                FilterTags := array{"Putty"}
                QuestID:=11
                SecondaryBGColor:=option{MakeColorFromHSV(28.033844, 0.908701, 1.0)}


    
    InitLockerUI(Player:player, CPlayer:custom_player, Tracker:tracker_device)<suspends>:void=

        set TrackerPaintQuest = Tracker
        
        InitPaintStyles() #needs to be done per player based on their unlocked styles

        PropGameBoard.Init()

        LoadingScreen := canvas:
            Slots := array:
                canvas_slot:
                    Anchors := anchors{Minimum := vector2{X:=0.0, Y:=0.0}, Maximum := vector2{X:=1.0, Y:=1.0}}
                    Offsets := margin{Top := 0.0, Left := 0.0, Right := 0.0, Bottom := 0.0}
                    Alignment := vector2{X:=0.5, Y:=0.5}
                    ZOrder := 9000
                    Widget := overlay:
                        Slots:=array:
                            overlay_slot:
                                HorizontalAlignment := horizontal_alignment.Fill
                                VerticalAlignment := vertical_alignment.Fill
                                Padding := margin{Top := 0.0, Left := 0.0, Right := 0.0, Bottom := 0.0}
                                Widget := texture_block:
                                    DefaultImage :=Textures.locker_loading_screen
                                    DefaultDesiredSize := vector2{X:=1920.0, Y:=1080.0}

        set MaybeLoadingOverlayCanvas = option{LoadingScreen}
        
        if(PlayerUI := GetPlayerUI[Player]):
            PlayerUI.AddWidget(LoadingScreen, player_ui_slot{ZOrder := 9000, InputMode := ui_input_mode.All})
            NewCanvas := CreateLockerUI(Player)
            set MaybeLockerUICanvas = option{NewCanvas}
            PlayerUI.AddWidget(NewCanvas, player_ui_slot{ZOrder := 8000, InputMode := ui_input_mode.All})

            Camera.AddTo(Player)
            PlayerReference.Register(Player)
            TrackerHIDETRACKERS.Assign(Player)

            spawn . AfterUILoaded(Player, CPlayer, Tracker)
            # spawn . PropGameBoard.Show(?Delay:=1.5)
            

    AfterUILoaded(Player:player, CPlayer:custom_player, Tracker:tracker_device)<suspends>:void=
        Sleep(2.0)

        if(LoadingScreen := MaybeLoadingOverlayCanvas?):
            if(PlayerUI := GetPlayerUI[Player]):
                PlayerUI.RemoveWidget(LoadingScreen)
                set MaybeLoadingOverlayCanvas = false

        Sleep(0.75)
        PropGameBoard.Show()
        
    CreatePaintStyleStackBoxSlot(PaintStyle:overlay):stack_box_slot=
        stack_box_slot:
            Distribution := option{1.0}
            HorizontalAlignment := horizontal_alignment.Left
            VerticalAlignment := vertical_alignment.Top
            Padding := margin{Top := 0.0, Left := 0.0, Right := 0.0, Bottom := 0.0}
            Widget := PaintStyle


    ExitLockerUI(Message:widget_message):void=
        if(HasLockerUI := MaybeLockerUICanvas?):
            if(PlayerUI := GetPlayerUI[Message.Player]):
                PlayerUI.RemoveWidget(HasLockerUI)
                set MaybeLockerUICanvas = false


            case(CurrentTabIndex):
                0 => spawn . PropGameBoard.Hide()
                1 => spawn . PropGameBoard.Hide()
                2 => spawn . PropYouStand.Hide()
                3 => spawn . PropPlatform.Hide()
                _ => Print("Invalid Tab Index")

            ExitLockerUIEvent.Signal()

            Camera.RemoveFrom(Message.Player)
            TrackerHIDETRACKERS.Remove(Message.Player)

            

        else:
            Print("Locker UI Canvas Not Found")

    HandleChangeTabClicked(Message: widget_message): void =
        var ClickedTabIndex: int = 0
    
        if (Message.Source = ButtonTabGameBoardLoud or Message.Source = ButtonTabGameBoardQuiet or Message.Source = ButtonTabGameBoardRegular):
            set ClickedTabIndex = 0
        else if (Message.Source = ButtonTabCardsLoud or Message.Source = ButtonTabCardsQuiet or Message.Source = ButtonTabCardsRegular):
            set ClickedTabIndex = 1
        else if (Message.Source = ButtonTabYouStandLoud or Message.Source = ButtonTabYouStandQuiet or Message.Source = ButtonTabYouStandRegular):
            set ClickedTabIndex = 2
        else if (Message.Source = ButtonTabPlatformLoud or Message.Source = ButtonTabPlatformQuiet or Message.Source = ButtonTabPlatformRegular):
            set ClickedTabIndex = 3
    
        if (ClickedTabIndex <> CurrentTabIndex):
            # Hide previous tab prop
            case (CurrentTabIndex):
                0 => spawn . PropGameBoard.Hide()
                1 => # Do Nothing
                2 => spawn . PropYouStand.Hide()
                3 => spawn . PropPlatform.Hide()
                _ => Print("Invalid Tab Index")
    
            # Show new tab prop
            case (ClickedTabIndex):
                0 => spawn . PropGameBoard.Show()
                1 => # Do Nothing
                2 => spawn . PropYouStand.Show()
                3 => spawn . PropPlatform.Show()
                _ => Print("Invalid Tab Index")
    
            set CurrentTabIndex = ClickedTabIndex
    
            # Update tab display
            UpdateTabDisplay(CurrentTabIndex)
            

    HandleFilterClicked(Message: widget_message): void =
        var FilterName: string = ""
    
        if (Message.Source = ButtonFilterPlasticQuiet):
            set FilterName = "Plastic"
            Print("Plastic Filter Clicked")
        else if (Message.Source = ButtonFilterMetallicQuiet):
            set FilterName = "Metallic"
            Print("Metallic Filter Clicked")
        else if (Message.Source = ButtonFilterPuttyQuiet):
            set FilterName = "Putty"
            Print("Putty Filter Clicked")
        else if (Message.Source = ButtonFilterGalaxyQuiet):
            set FilterName = "Galaxy"
            Print("Galaxy Filter Clicked")
    
        # Toggle filter selection
        if (TheTagIndex := CurrentFilterTags.Find[FilterName]):
            # Remove filter
            if:
                # NewFilters := CurrentFilterTags.RemoveFirstElement[FilterName]
                NewFilters := CurrentFilterTags.RemoveElement[TheTagIndex]
            then:
                Print("Filter Removed")
                set CurrentFilterTags = NewFilters
            else:
                Print("Failed to remove filter")
        else:
            # Add filter
            set CurrentFilterTags += array{FilterName}
    
        # Update filter display
        UpdateFilterDisplay()
    
        # Optionally, update the displayed paint options based on filters
        UpdatePaintOptions()
            

    CreateLockerUI(Player:player)<suspends>:canvas=

        if(not ButtonExitSubscription?):
            Sub := ButtonExit.OnClick().Subscribe(ExitLockerUI)
            set ButtonExitSubscription = option{Sub}

        if(not ButtonTabGameBoardSubscription?):
            Sub := ButtonTabGameBoardQuiet.OnClick().Subscribe(HandleChangeTabClicked)
            set ButtonTabGameBoardSubscription = option{Sub}

        if(not ButtonTabCardsSubscription?):
            Sub := ButtonTabCardsQuiet.OnClick().Subscribe(HandleChangeTabClicked)
            set ButtonTabCardsSubscription = option{Sub}

        if(not ButtonTabYouStandSubscription?):
            Sub := ButtonTabYouStandQuiet.OnClick().Subscribe(HandleChangeTabClicked)
            set ButtonTabYouStandSubscription = option{Sub}

        if(not ButtonTabPlatformSubscription?):
            Sub := ButtonTabPlatformQuiet.OnClick().Subscribe(HandleChangeTabClicked)
            set ButtonTabPlatformSubscription = option{Sub}

        if(not ButtonFilterClearSubscription?):
            Sub := ButtonFilterClear.OnClick().Subscribe(HandleFilterClearClicked)
            set ButtonFilterClearSubscription = option{Sub}

        if(not ButtonFilterPlasticSubscription?):
            Sub := ButtonFilterPlasticQuiet.OnClick().Subscribe(HandleFilterClicked)
            set ButtonFilterPlasticSubscription = option{Sub}

        if(not ButtonFilterMetallicSubscription?):
            Sub := ButtonFilterMetallicQuiet.OnClick().Subscribe(HandleFilterClicked)
            set ButtonFilterMetallicSubscription = option{Sub}

        if(not ButtonFilterPuttySubscription?):
            Sub := ButtonFilterPuttyQuiet.OnClick().Subscribe(HandleFilterClicked)
            set ButtonFilterPuttySubscription = option{Sub}

        if(not ButtonFilterGalaxySubscription?):
            Sub := ButtonFilterGalaxyQuiet.OnClick().Subscribe(HandleFilterClicked)
            set ButtonFilterGalaxySubscription = option{Sub}

        if(not ButtonPageDownSubscription?):
            Sub := ButtonPageDown.OnClick().Subscribe(HandlePageDownClicked)
            set ButtonPageDownSubscription = option{Sub}

        if(not ButtonPageUpSubscription?):
            Sub := ButtonPageUp.OnClick().Subscribe(HandlePageUpClicked)
            set ButtonPageUpSubscription = option{Sub}

    
        GoldTextBlock.SetText(IntToMessage(1000))

        # ButtonTabGameBoardQuiet.OnClick().Subscribe(HandleChangeTabClicked)
        # ButtonTabCardsQuiet.OnClick().Subscribe(HandleChangeTabClicked)
        # ButtonTabYouStandQuiet.OnClick().Subscribe(HandleChangeTabClicked)
        # ButtonTabPlatformQuiet.OnClick().Subscribe(HandleChangeTabClicked)
        # ButtonTabGameBoardRegular.OnClick().Subscribe(HandleChangeTabClicked)
        # ButtonTabCardsRegular.OnClick().Subscribe(HandleChangeTabClicked)
        # ButtonTabYouStandRegular.OnClick().Subscribe(HandleChangeTabClicked)
        # ButtonTabPlatformRegular.OnClick().Subscribe(HandleChangeTabClicked)
        # ButtonTabGameBoardLoud.OnClick().Subscribe(HandleChangeTabClicked)
        # ButtonTabCardsLoud.OnClick().Subscribe(HandleChangeTabClicked)
        # ButtonTabYouStandLoud.OnClick().Subscribe(HandleChangeTabClicked)
        # ButtonTabPlatformLoud.OnClick().Subscribe(HandleChangeTabClicked)

        # ButtonFilterPlasticQuiet.OnClick().Subscribe(HandleFilterClicked)
        # ButtonFilterPlasticRegular.OnClick().Subscribe(HandleFilterClicked)
        # ButtonFilterMetallicQuiet.OnClick().Subscribe(HandleFilterClicked)
        # ButtonFilterMetallicRegular.OnClick().Subscribe(HandleFilterClicked)
        # ButtonFilterPuttyQuiet.OnClick().Subscribe(HandleFilterClicked)
        # ButtonFilterPuttyRegular.OnClick().Subscribe(HandleFilterClicked)
        # ButtonFilterGalaxyQuiet.OnClick().Subscribe(HandleFilterClicked)
        # ButtonFilterGalaxyRegular.OnClick().Subscribe(HandleFilterClicked)
        # ButtonFilterClear.OnClick().Subscribe(HandleFilterClearClicked)
        


        PaintOptionWidgets := for(PaintStyle : PaintStyles):
            PaintOption := paint_ui_option{}
            PaintOption.Create(Player, PaintStyle, 3)

            if(PaintStyle.IsActive?):
                set CurrentActivePaintStyle = PaintOption
        
            PaintOption.Button.OnClick().SubscribeMessage(OnPaintStyleButtonClicked, PaintOption)

            PaintOption.Widget

        UpdateTabDisplay(CurrentTabIndex)
        UpdateFilterDisplay()



        #create a stack box for the paint options, 4 columns, 3 rows
        PaintOptionsOverlay := overlay:
            Slots := array:
                overlay_slot:
                    HorizontalAlignment := horizontal_alignment.Fill
                    VerticalAlignment := vertical_alignment.Fill
                    Padding := margin{Top := 0.0, Left := 0.0, Right := 0.0, Bottom := 0.0}
                    Widget := stack_box:
                        Orientation := orientation.Vertical
                        Slots := array:
                            stack_box_slot:
                                HorizontalAlignment := horizontal_alignment.Fill
                                VerticalAlignment := vertical_alignment.Top
                                Padding := margin{Top := 0.0, Left := 0.0, Right := 0.0, Bottom := 0.0}
                                Widget := stack_box:
                                    Orientation := orientation.Horizontal
                                    Slots := for(Index := 0 .. 3, Overlay := PaintOptionWidgets[Index]):
                                        CreatePaintStyleStackBoxSlot(Overlay)
                            stack_box_slot:
                                HorizontalAlignment := horizontal_alignment.Fill
                                VerticalAlignment := vertical_alignment.Top
                                Padding := margin{Top := 0.0, Left := 0.0, Right := 0.0, Bottom := 0.0}
                                Widget := stack_box:
                                    Orientation := orientation.Horizontal
                                    Slots := for(Index := 4 .. 7, Overlay := PaintOptionWidgets[Index]):
                                        CreatePaintStyleStackBoxSlot(Overlay)
                            stack_box_slot:
                                HorizontalAlignment := horizontal_alignment.Fill
                                VerticalAlignment := vertical_alignment.Top
                                Padding := margin{Top := 0.0, Left := 0.0, Right := 0.0, Bottom := 0.0}
                                Widget := stack_box:
                                    Orientation := orientation.Horizontal
                                    Slots := for(Index := 8 .. 11, Overlay := PaintOptionWidgets[Index]):
                                        CreatePaintStyleStackBoxSlot(Overlay)
                            stack_box_slot:
                                HorizontalAlignment := horizontal_alignment.Fill
                                VerticalAlignment := vertical_alignment.Top
                                Padding := margin{Top := 0.0, Left := 0.0, Right := 0.0, Bottom := 0.0}
                                Widget := stack_box:
                                    Orientation := orientation.Horizontal
                                    Slots := for(Index := 12 .. 15, Overlay := PaintOptionWidgets[Index]):
                                        CreatePaintStyleStackBoxSlot(Overlay)
                            stack_box_slot:
                                HorizontalAlignment := horizontal_alignment.Fill
                                VerticalAlignment := vertical_alignment.Top
                                Padding := margin{Top := 0.0, Left := 0.0, Right := 0.0, Bottom := 0.0}
                                Widget := overlay:
                                    Slots := array:
                                        overlay_slot:
                                            HorizontalAlignment := horizontal_alignment.Left
                                            VerticalAlignment := vertical_alignment.Fill
                                            Padding := margin{Top := 0.0, Left := 0.0, Right := 0.0, Bottom := 0.0}
                                            Widget := stack_box:
                                                Orientation := orientation.Horizontal
                                                Slots := for(Index := 16 .. 16, Overlay := PaintOptionWidgets[Index]):
                                                    CreatePaintStyleStackBoxSlot(Overlay)
                                        overlay_slot:
                                            HorizontalAlignment := horizontal_alignment.Fill
                                            VerticalAlignment := vertical_alignment.Center
                                            Padding := margin{Top := 0.0, Left := 0.0, Right := 0.0, Bottom := 50.0}
                                            Widget := texture_block:
                                                DefaultImage := _UI.UI_Textures.ui_shadow_top
                                                DefaultDesiredSize := vector2{X:=128.0, Y:=128.0}
                                        




        LayoutCanvas := canvas:
            Slots := array:

                # Currencies Top Left
                canvas_slot:
                    Anchors := anchors{Minimum := vector2{X:=0.0, Y:=0.0}, Maximum := vector2{X:=1.0, Y:=0.0}}
                    Offsets := margin{Top := 0.0, Left := 0.0, Right := 0.0, Bottom := 0.0}
                    Alignment := vector2{X:=0.0, Y:=0.0}
                    ZOrder := 8040
                    Widget := overlay:
                        Slots := array:
                            overlay_slot:
                                HorizontalAlignment := horizontal_alignment.Left
                                VerticalAlignment := vertical_alignment.Top
                                Padding := margin{Top := 20.0, Left := 20.0, Right := 0.0, Bottom := 0.0}
                                Widget := stack_box:
                                    Orientation := orientation.Vertical
                                    Slots := array:
                                        stack_box_slot:
                                            HorizontalAlignment := horizontal_alignment.Left
                                            VerticalAlignment := vertical_alignment.Top
                                            Padding := margin{Top := 0.0, Left := 0.0, Right := 0.0, Bottom := 0.0}
                                            Widget := overlay:
                                                Slots := array:
                                                    overlay_slot:
                                                        HorizontalAlignment := horizontal_alignment.Fill
                                                        VerticalAlignment := vertical_alignment.Fill
                                                        Padding := margin{Top := 0.0, Left := 0.0, Right := 0.0, Bottom := 0.0}
                                                        Widget := MakeRoundedOverlayQuietButton(MakeColorFromHSV(240.0,1.0,0.25), ?TextureSize:=24.0)
                                                    overlay_slot:
                                                        HorizontalAlignment := horizontal_alignment.Fill
                                                        VerticalAlignment := vertical_alignment.Fill
                                                        Padding := margin{Top := 20.0, Left := 50.0, Right := 50.0, Bottom := 20.0}
                                                        Widget := stack_box:
                                                            Orientation := orientation.Horizontal
                                                            Slots := array:
                                                                stack_box_slot:
                                                                    HorizontalAlignment := horizontal_alignment.Left
                                                                    VerticalAlignment := vertical_alignment.Center
                                                                    Padding := margin{Top := 0.0, Left := 0.0, Right := 0.0, Bottom := 0.0}
                                                                    Widget := texture_block:
                                                                        DefaultImage := _UI.UI_Textures.ui_gold_coin
                                                                        DefaultDesiredSize := vector2{X:=64.0, Y:=64.0}
                                                                stack_box_slot:
                                                                    Distribution := option{1.0}
                                                                    HorizontalAlignment := horizontal_alignment.Right
                                                                    VerticalAlignment := vertical_alignment.Center
                                                                    Padding := margin{Top := 0.0, Left := 30.0, Right := 0.0, Bottom := 0.0}
                                                                    Widget := GoldTextBlock

                            overlay_slot:
                                HorizontalAlignment := horizontal_alignment.Right
                                VerticalAlignment := vertical_alignment.Top
                                Padding := margin{Top := 20.0, Left := 20.0, Right := 20.0, Bottom := 20.0}
                                Widget := ButtonExit


                                                
                                        # stack_box_slot:
                                        #     HorizontalAlignment := horizontal_alignment.Left
                                        #     VerticalAlignment := vertical_alignment.Top
                                        #     Padding := margin{Top := 0.0, Left := 10.0, Right := 0.0, Bottom := 0.0}
                                        #     Widget := PaintTokensTextBlock
                # Paint Options
                canvas_slot:
                    Anchors := anchors{Minimum := vector2{X:=1.0, Y:=0.0}, Maximum := vector2{X:=1.0, Y:=1.0}}
                    Offsets := margin{Top := 0.0, Left := 0.0, Right := 0.0, Bottom := 0.0}
                    Alignment := vector2{X:=1.0, Y:=0.5}
                    ZOrder := 8030
                    Widget := overlay:
                        Slots := array:
                            overlay_slot:
                                HorizontalAlignment := horizontal_alignment.Right
                                VerticalAlignment := vertical_alignment.Fill
                                Padding := margin{Top := 0.0, Left := 0.0, Right := 20.0, Bottom := 0.0}
                                Widget := stack_box:
                                    Orientation := orientation.Vertical
                                    Slots := array:

                                        # Title
                                        stack_box_slot:
                                            HorizontalAlignment := horizontal_alignment.Left
                                            VerticalAlignment := vertical_alignment.Top
                                            Padding := margin{Top := 100.0, Left := 0.0, Right := 0.0, Bottom := 0.0}
                                            Widget := BoardDesignerTitleTextureBlock

                                        # Tabs
                                        stack_box_slot:
                                            HorizontalAlignment := horizontal_alignment.Fill
                                            VerticalAlignment := vertical_alignment.Top
                                            Padding := margin{Top := 0.0, Left := 0.0, Right := 0.0, Bottom := 0.0}
                                            Widget := overlay:
                                                Slots := array:
                                                    overlay_slot:
                                                        HorizontalAlignment := horizontal_alignment.Fill
                                                        VerticalAlignment := vertical_alignment.Fill
                                                        Padding := margin{Top := 0.0, Left := 0.0, Right := 0.0, Bottom := 10.0}
                                                        Widget := stack_box:
                                                            Orientation := orientation.Horizontal
                                                            Slots := array:
                                                                stack_box_slot:
                                                                    Distribution := option{2.0}
                                                                    HorizontalAlignment := horizontal_alignment.Fill
                                                                    VerticalAlignment := vertical_alignment.Fill
                                                                    Padding := margin{Top := 0.0, Left := 0.0, Right := 5.0, Bottom := 0.0}
                                                                    Widget := ButtonTabGameBoardContainer
                                                                stack_box_slot:
                                                                    Distribution := option{2.0}
                                                                    HorizontalAlignment := horizontal_alignment.Fill
                                                                    VerticalAlignment := vertical_alignment.Fill
                                                                    Padding := margin{Top := 0.0, Left := 5.0, Right := 5.0, Bottom := 0.0}
                                                                    Widget := ButtonTabCardsContainer
                                                                stack_box_slot:
                                                                    Distribution := option{2.0}
                                                                    HorizontalAlignment := horizontal_alignment.Fill
                                                                    VerticalAlignment := vertical_alignment.Fill
                                                                    Padding := margin{Top := 0.0, Left := 5.0, Right := 5.0, Bottom := 0.0}
                                                                    Widget := ButtonTabYouStandContainer
                                                                stack_box_slot:
                                                                    Distribution := option{2.0}
                                                                    HorizontalAlignment := horizontal_alignment.Fill
                                                                    VerticalAlignment := vertical_alignment.Fill
                                                                    Padding := margin{Top := 0.0, Left := 5.0, Right := 0.0, Bottom := 0.0}
                                                                    Widget := ButtonTabPlatformContainer
                                        # Filters
                                        stack_box_slot:
                                            Distribution := option{1.0}
                                            HorizontalAlignment := horizontal_alignment.Fill
                                            VerticalAlignment := vertical_alignment.Top
                                            Padding := margin{Top := 0.0, Left := 0.0, Right := 0.0, Bottom := 0.0}
                                            Widget := overlay:
                                                Slots := array:
                                                    overlay_slot:
                                                        HorizontalAlignment := horizontal_alignment.Fill
                                                        VerticalAlignment := vertical_alignment.Top
                                                        Padding := margin{Top := 0.0, Left := 0.0, Right := 0.0, Bottom := 0.0}
                                                        Widget := stack_box:
                                                            Orientation := orientation.Horizontal
                                                            Slots := array:
                                                                stack_box_slot:
                                                                    Distribution := option{2.0}
                                                                    HorizontalAlignment := horizontal_alignment.Fill
                                                                    VerticalAlignment := vertical_alignment.Fill
                                                                    Padding := margin{Top := 0.0, Left := 0.0, Right := 5.0, Bottom := 0.0}
                                                                    Widget := ButtonFilterPlasticContainer
                                                                stack_box_slot:
                                                                    Distribution := option{2.0}
                                                                    HorizontalAlignment := horizontal_alignment.Fill
                                                                    VerticalAlignment := vertical_alignment.Fill
                                                                    Padding := margin{Top := 0.0, Left := 5.0, Right := 5.0, Bottom := 0.0}
                                                                    Widget := ButtonFilterMetallicContainer
                                                                stack_box_slot:
                                                                    Distribution := option{2.0}
                                                                    HorizontalAlignment := horizontal_alignment.Fill
                                                                    VerticalAlignment := vertical_alignment.Fill
                                                                    Padding := margin{Top := 0.0, Left := 5.0, Right := 5.0, Bottom := 0.0}
                                                                    Widget := ButtonFilterPuttyContainer
                                                                stack_box_slot:
                                                                    Distribution := option{2.0}
                                                                    HorizontalAlignment := horizontal_alignment.Fill
                                                                    VerticalAlignment := vertical_alignment.Fill
                                                                    Padding := margin{Top := 0.0, Left := 5.0, Right := 0.0, Bottom := 0.0}
                                                                    Widget := ButtonFilterGalaxyContainer
                                                                stack_box_slot:
                                                                    HorizontalAlignment := horizontal_alignment.Right
                                                                    VerticalAlignment := vertical_alignment.Fill
                                                                    Padding := margin{Top := 0.0, Left := 20.0, Right := 0.0, Bottom := 0.0}
                                                                    Widget := ButtonFilterClear

                                                                    
                                        # Paint Options (Pagination)
                                        stack_box_slot:
                                            Distribution := option{16.0} #squishes the buttons above
                                            HorizontalAlignment := horizontal_alignment.Fill
                                            VerticalAlignment := vertical_alignment.Fill
                                            Padding := margin{Top := 10.0, Left := 0.0, Right := 0.0, Bottom := 0.0}
                                            Widget := overlay:
                                                Slots := array:
                                                    overlay_slot:
                                                        HorizontalAlignment := horizontal_alignment.Fill
                                                        VerticalAlignment := vertical_alignment.Fill
                                                        Padding := margin{Top := 0.0, Left := 0.0, Right := 0.0, Bottom := 0.0}
                                                        Widget := stack_box:
                                                            Orientation := orientation.Horizontal
                                                            Slots := array:
                                                                # Paint Options
                                                                stack_box_slot:
                                                                    Distribution := option{1.0}
                                                                    HorizontalAlignment := horizontal_alignment.Fill
                                                                    VerticalAlignment := vertical_alignment.Fill
                                                                    Padding := margin{Top := 0.0, Left := 0.0, Right := 0.0, Bottom := 0.0}
                                                                    Widget := PaintOptionsOverlay
                                                                # Page Buttons
                                                                stack_box_slot:
                                                                    HorizontalAlignment := horizontal_alignment.Fill
                                                                    VerticalAlignment := vertical_alignment.Fill
                                                                    Padding := margin{Top := 0.0, Left := 0.0, Right := 0.0, Bottom := 0.0}
                                                                    Widget := stack_box:
                                                                        Orientation := orientation.Vertical
                                                                        Slots := array:
                                                                            stack_box_slot:
                                                                                Distribution := option{2.0}
                                                                                HorizontalAlignment := horizontal_alignment.Fill
                                                                                VerticalAlignment := vertical_alignment.Fill
                                                                                Padding := margin{Top := 10.0, Left := 0.0, Right := 0.0, Bottom := 5.0}
                                                                                Widget := overlay:
                                                                                    Slots := array:
                                                                                        overlay_slot:
                                                                                            HorizontalAlignment := horizontal_alignment.Fill
                                                                                            VerticalAlignment := vertical_alignment.Fill
                                                                                            Padding := margin{Top := 0.0, Left := 0.0, Right := 0.0, Bottom := 0.0}
                                                                                            Widget := ButtonPageUp
                                                                                        overlay_slot:
                                                                                            HorizontalAlignment := horizontal_alignment.Right
                                                                                            VerticalAlignment := vertical_alignment.Top
                                                                                            Padding := margin{Top := -12.0, Left := 0.0, Right := -12.0, Bottom := 0.0}
                                                                                            Widget := NewPrevPageTextureBlock
                                                                            stack_box_slot:
                                                                                Distribution := option{2.0}
                                                                                HorizontalAlignment := horizontal_alignment.Fill
                                                                                VerticalAlignment := vertical_alignment.Fill
                                                                                Padding := margin{Top := 5.0, Left := 0.0, Right := 0.0, Bottom := 5.0}
                                                                                Widget := overlay:
                                                                                    Slots := array:
                                                                                        overlay_slot:
                                                                                            HorizontalAlignment := horizontal_alignment.Fill
                                                                                            VerticalAlignment := vertical_alignment.Fill
                                                                                            Padding := margin{Top := 0.0, Left := 0.0, Right := 0.0, Bottom := 0.0}
                                                                                            Widget := ButtonPageDown
                                                                                        overlay_slot:
                                                                                            HorizontalAlignment := horizontal_alignment.Right
                                                                                            VerticalAlignment := vertical_alignment.Top
                                                                                            Padding := margin{Top := -12.0, Left := 0.0, Right := -12.0, Bottom := 0.0}
                                                                                            Widget := NewNextPageTextureBlock


    UpdateTabDisplay(SelectedTabIndex: int):void=
        # First, clear all tab containers
        ButtonTabGameBoardContainer.ClearSlots()
        ButtonTabCardsContainer.ClearSlots()
        ButtonTabYouStandContainer.ClearSlots()
        ButtonTabPlatformContainer.ClearSlots()
    
        # Define arrays of tab buttons and containers
        TabButtonsQuiet := array{ButtonTabGameBoardQuiet, ButtonTabCardsQuiet, ButtonTabYouStandQuiet, ButtonTabPlatformQuiet}
        TabButtonsLoud := array{ButtonTabGameBoardLoud, ButtonTabCardsLoud, ButtonTabYouStandLoud, ButtonTabPlatformLoud}
        TabContainers := array{ButtonTabGameBoardContainer, ButtonTabCardsContainer, ButtonTabYouStandContainer, ButtonTabPlatformContainer}
    
        # Loop through tabs and set their display based on the selected index
        for (Index := 0 .. 3):
            if (Index = SelectedTabIndex, TabContainer := TabContainers[Index], TabButton := TabButtonsLoud[Index]):
                # Selected tab
                TabContainer.AddWidget(overlay_slot{
                    HorizontalAlignment := horizontal_alignment.Fill,
                    VerticalAlignment := vertical_alignment.Fill,
                    Widget := overlay:
                        Slots := array:
                            overlay_slot:
                                HorizontalAlignment := horizontal_alignment.Fill,
                                VerticalAlignment := vertical_alignment.Fill,
                                Widget := MakeRoundedOverlayQuietButton(BGColor := MakeColorFromHSV(59.0, 0.0, 91.0))  # Adjust BGColor as needed
                            overlay_slot:
                                HorizontalAlignment := horizontal_alignment.Fill,
                                VerticalAlignment := vertical_alignment.Fill,
                                Widget := TabButton
                })
                TabButton.SetEnabled(false)
            else:
                # Unselected tabs
                if:
                    TabContainer := TabContainers[Index]
                    TabButton := TabButtonsQuiet[Index]
                then:
                    TabContainer.AddWidget(overlay_slot{
                        HorizontalAlignment := horizontal_alignment.Fill,
                        VerticalAlignment := vertical_alignment.Fill,
                        Widget := TabButton
                    })
                    TabButton.SetEnabled(true)
                                                                                
    

    UpdateFilterDisplay(): void =
        # Define filter buttons, containers, and names
        FilterButtonsQuiet := array{ButtonFilterPlasticQuiet, ButtonFilterMetallicQuiet, ButtonFilterPuttyQuiet, ButtonFilterGalaxyQuiet}
        FilterContainers := array{ButtonFilterPlasticContainer, ButtonFilterMetallicContainer, ButtonFilterPuttyContainer, ButtonFilterGalaxyContainer}
        FilterNames := array{"Plastic", "Metallic", "Putty", "Galaxy"} # just filters that are available not currently selected
    
        # Loop through filters
        for (Index := 0 .. 3, Container := FilterContainers[Index],Button := FilterButtonsQuiet[Index],FilterName := FilterNames[Index]):
            # Clear the container
            Container.ClearSlots()

            var BGOverlay : overlay = overlay{}
    
            if (CurrentFilterTags.Find[FilterName]):
                Print("Selected Filter: {FilterName}")
                set BGOverlay = MakeRoundedOverlayQuietButton(BGColor := MakeColorFromHSV(12.0, 0.0, 1.0), ?TextureSize:=16.0)
            else:
                Print("Unselected Filter: {FilterName}")
                set BGOverlay = MakeRoundedOverlayQuietButton(BGColor := MakeColorFromHSV(240.0, 1.0, 0.34), ?TextureSize:=16.0)
    
            # Add the button and background to the container
            Container.AddWidget(overlay_slot{
                HorizontalAlignment := horizontal_alignment.Fill,
                VerticalAlignment := vertical_alignment.Fill,
                Widget := overlay:
                    Slots := array:
                        # Background overlay
                        overlay_slot:
                            HorizontalAlignment := horizontal_alignment.Fill,
                            VerticalAlignment := vertical_alignment.Fill,
                            Widget := BGOverlay
                        # Button on top
                        overlay_slot:
                            HorizontalAlignment := horizontal_alignment.Fill,
                            VerticalAlignment := vertical_alignment.Fill,
                            Widget := Button
            })
    
                    
    
    HandleFilterClearClicked(Message: widget_message): void =
        set CurrentFilterTags = array{}  # Clear all filters
        UpdateFilterDisplay()
        UpdatePaintOptions()
                    
    

    UpdatePaintOptions(): void =
        Print("TODO: Update paint options based on filters")


    OnUnlockInfoClicked(Message: widget_message, PaintInfoPanel: locked_paint_info_panel): void =
        
        # place the info panel over the current ui
        if(LockerUI := MaybeLockerUICanvas? , not MaybeModal?):
            set MaybeModal = option{PaintInfoPanel.Modal}
            LockerUI.AddWidget(PaintInfoPanel.Modal)
            Sub := PaintInfoPanel.ButtonClose.OnClick().Subscribe(OnCloseInfoClicked)
            set MaybeModalCloseSubscription = option{Sub}

        



    OnCloseInfoClicked(Message: widget_message): void =
        if(HasInfoModal := MaybeModal?):
            if(LockerUI := MaybeLockerUICanvas?):
                LockerUI.RemoveWidget(HasInfoModal.Widget)
                set MaybeModal = false

            if(ModalCloseSubscription := MaybeModalCloseSubscription?):
                ModalCloseSubscription.Cancel()
                set MaybeModalCloseSubscription = false

            ApplyPaintStyle(CurrentActivePaintStyle)

        else:
            Print("Info Modal Not Found")

    OnPaintStyleButtonClicked(Message: widget_message, PaintStyle: paint_ui_option): void =
        Print("Paint Style Button Clicked: {PaintStyle.PaintName}")

        if(not PaintStyle.Unlocked?):
            Print("Paint Style is Locked- Showing info panel")
            InfoPanel := locked_paint_info_panel{}
            InfoPanel.Create(Message.Player, PaintStyle, PropGameBoard)

            if(LockerUI := MaybeLockerUICanvas? , not MaybeModal?):
                set MaybeModal = option{InfoPanel.Modal}
                LockerUI.AddWidget(InfoPanel.Modal)
                Sub := InfoPanel.ButtonClose.OnClick().Subscribe(OnCloseInfoClicked)
                set MaybeModalCloseSubscription = option{Sub}

                TrackSub := InfoPanel.ButtonTrackQuest.OnClick().SubscribeMessage(OnTrackQuestClicked, PaintStyle)

        else if(not PaintStyle.Owned?):
            Print("Paint Style is not owned")
            if(PaintStyle.Cost > 0):
                # Show purchase confirmation dialog
                Print("Showing Purchase Panel")
                PurchasePanel := locked_paint_buy_panel{}
                PurchasePanel.Create(Message.Player, PaintStyle, PropGameBoard)

                if(LockerUI := MaybeLockerUICanvas? , not MaybeModal?):
                    set MaybeModal = option{PurchasePanel.Modal}
                    LockerUI.AddWidget(PurchasePanel.Modal)
                    Sub := PurchasePanel.ButtonClose.OnClick().Subscribe(OnCloseInfoClicked)
                    set MaybeModalCloseSubscription = option{Sub}

                    BuySub := PurchasePanel.ButtonBuy.OnClick().SubscribeMessage(OnBuyPaintStyleClicked, (PaintStyle, PurchasePanel))
                    set MaybeModalBuySubscription = option{BuySub}

            else:
                Print("Paint Style is free - Unlocking")
                PaintStyle.SetOwned(true)
        else:
            # Toggle active state
            if(PaintStyle.Active?):
                # Do nothing, always need to have at least one paint style active
                # PaintStyle.SetActive(false)
            else:
                CurrentActivePaintStyle.SetActive(false)
                PaintStyle.SetActive(true)
                set CurrentActivePaintStyle = PaintStyle

                spawn . DelayApplyPaintStyle()
        
    DelayApplyPaintStyle()<suspends>:void=
        Sleep(0.1)
        ApplyPaintStyle(CurrentActivePaintStyle)

    OnBuyPaintStyleClicked(Message: widget_message, PaintData:tuple(paint_ui_option, locked_paint_buy_panel)): void =
        Print("Buy Paint Style Clicked: {PaintData(0).PaintName}")
        if(PaintData(0).Cost > 0):
            # Check if player has enough currency ##################################### TODO #####################################
            # Deduct cost from player
            Print("Deducting cost from player")
            PaintData(0).SetOwned(true)

            Print("Paint Style Purchased: {PaintData(0).PaintName}")
            if(PaintData(1).AutoApply?):
                PaintData(0).SetActive(true)
                CurrentActivePaintStyle.SetActive(false)
                set CurrentActivePaintStyle = PaintData(0)
            else:
                PaintData(0).SetNew()

            ApplyPaintStyle(CurrentActivePaintStyle)

            if(HasModal := MaybeModal?):
                if(LockerUI := MaybeLockerUICanvas?):
                    LockerUI.RemoveWidget(HasModal.Widget)
                    set MaybeModal = false

                if(ModalCloseSubscription := MaybeModalCloseSubscription?):
                    ModalCloseSubscription.Cancel()
                    set MaybeModalCloseSubscription = false

                if(BuySubscription := MaybeModalBuySubscription?):
                    BuySubscription.Cancel()
                    set MaybeModalBuySubscription = false
        else:
            Print("Paint Style is free - How did you click this?")


    ApplyPaintStyle(PaintStyle: paint_ui_option): void =
        case(PaintStyle.PaintStyleIndex):
            0 => block:
                PaintMatData := plastic_material_data:
                    ColorTint := PaintStyle.PaintBGColor
                    EmissiveFlickerTint := PaintStyle.PaintBGColor
                NewMI := CreatePlasticMaterialInstance(PaintMatData)
                PropGameBoard.Prop.SetMaterial(NewMI, ?Index:=1)
                PropGameBoard.Prop.SetMaterial(NewMI, ?Index:=2)
                PropGameBoard.Prop.SetMaterial(NewMI, ?Index:=3)
            1 => block:
                PaintMatData := plastic_material_data:
                    ColorTint := PaintStyle.PaintBGColor
                    MetallicAmount := 1.0
                    EmissiveFlickerTint := PaintStyle.PaintBGColor
                NewMI := CreateMetallicMaterialInstance(PaintMatData)
                PropGameBoard.Prop.SetMaterial(NewMI, ?Index:=1)
                PropGameBoard.Prop.SetMaterial(NewMI, ?Index:=2)
                PropGameBoard.Prop.SetMaterial(NewMI, ?Index:=3)
            2 => block:
                if(SecondaryColor := PaintStyle.SecondaryBGColor?):
                    PaintMatData := putty_material_data:
                        ColorClay1 := PaintStyle.PaintBGColor
                        ColorClay2 := SecondaryColor
                        EmissiveFlickerTint := PaintStyle.PaintBGColor
                    NewPuttyMI := CreatePuttyMaterialInstance(PaintMatData)
                    PropGameBoard.Prop.SetMaterial(NewPuttyMI, ?Index:=1)
                    PropGameBoard.Prop.SetMaterial(NewPuttyMI, ?Index:=2)
                    PropGameBoard.Prop.SetMaterial(NewPuttyMI, ?Index:=3)
            _ => block:
                PaintMatData := plastic_material_data:
                    ColorTint := PaintStyle.PaintBGColor
                    EmissiveFlickerTint := PaintStyle.PaintBGColor
                NewMI := CreatePlasticMaterialInstance(PaintMatData)
                PropGameBoard.Prop.SetMaterial(NewMI, ?Index:=1)
                PropGameBoard.Prop.SetMaterial(NewMI, ?Index:=2)
                PropGameBoard.Prop.SetMaterial(NewMI, ?Index:=3)

    OnTrackQuestClicked(Message: widget_message, PaintStyle: paint_ui_option): void =
        Print("Track Quest Clicked: " + PaintStyle.PaintName)
        if(PaintStyle.QuestID > 0, QuestInfo := GetQuestFromID[PaintStyle.QuestID]):

            if(not CurrentActiveTrackedQuestPaintStyle.QuestID = PaintStyle.QuestID):
                Print("Tracking Quest: {PaintStyle.QuestID}")
                # TrackerPaintQuest.TrackQuest(PaintStyle.QuestID)
                PaintStyle.SetTrackedQuest(true)
                CurrentActiveTrackedQuestPaintStyle.SetTrackedQuest(false)
                set CurrentActiveTrackedQuestPaintStyle = PaintStyle
                TrackerPaintQuest.Remove(Message.Player)

                spawn . AssignNewQuestTracker(Message.Player, QuestInfo)

            ApplyPaintStyle(CurrentActivePaintStyle)

            # Close the modal
            if(HasInfoModal := MaybeModal?):
                if(LockerUI := MaybeLockerUICanvas?):
                    LockerUI.RemoveWidget(HasInfoModal.Widget)
                    set MaybeModal = false
    
                if(ModalCloseSubscription := MaybeModalCloseSubscription?):
                    ModalCloseSubscription.Cancel()
                    set MaybeModalCloseSubscription = false
        else:
            Print("No Quest ID to track")

    AssignNewQuestTracker(Player:player, QuestInfo:fortwho_quest)<suspends>:void=
        Sleep(0.25)
        TrackerPaintQuest.SetTarget(QuestInfo.Target)
        TrackerPaintQuest.SetTitleText(StringToMessage(QuestInfo.Name))
        TrackerPaintQuest.SetDescriptionText(StringToMessage(QuestInfo.Description))
        TrackerPaintQuest.Assign(Player)
        Sleep(0.01)
        TrackerPaintQuest.SetValue(QuestInfo.Progress)

    HandlePageDownClicked(Message: widget_message): void =
        Print("Page Down Clicked")

    HandlePageUpClicked(Message: widget_message): void =
        Print("Page Up Clicked")