using { /Verse.org/Simulation }
using { /Fortnite.com/Devices }
using { /UnrealEngine.com/Temporary/SpatialMath }


board_sticker_slot := class<concrete>():
    # @editable Transform : transform = transform{}
    @editable Slot : creative_prop = creative_prop{}
    var MaybeSticker : ?creative_prop = false

    GetTransform()<transacts>:transform=
        Slot.GetTransform()
    