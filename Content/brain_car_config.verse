
using { /Fortnite.com/Devices }
using { /Verse.org/Simulation }
using { /UnrealEngine.com/Temporary/Diagnostics }
using { /Verse.org/Random }
using { /UnrealEngine.com/Temporary/SpatialMath }
using { /Fortnite.com/Characters }
using { /Fortnite.com/Devices/CreativeAnimation }
using { /Verse.org/Assets }

Collabs <public> := module:
    BrainCar <public> := module:
        HC1H2D3B61EU001257_BB <public> := module:
        HC1H2D3B61EU010069_BB <public> := module:
        HC1H2D3B61EU010070_BB <public> := module:
        HC1H2D3B61EU011295_BB <public> := module:
        HC1H2D3B61EU011299_BB <public> := module:


brain_car_paint_style := enum:
    HC1H2D3B61EU001257 # Meta Black Metallic
    HC1H2D3B61EU010069 # Photon Orange Metallic
    HC1H2D3B61EU010070 # Atom Grey Matte
    HC1H2D3B61EU011295 # Laser Red Metallic
    HC1H2D3B61EU011299 # Digital White Metallic
    

brain_car_config := class(creative_device):

    CancelSpin : event() = event(){}
    DialogClosed : event() = event(){}
    PaintAppliedEvent : event() = event(){}

    # test

    @editable CarProp : creative_prop = creative_prop{}
    @editable PlatformProp : creative_prop = creative_prop{}
    @editable Button : button_device = button_device{}
    @editable Switch : switch_device = switch_device{}
    @editable DialogPopup : popup_dialog_device = popup_dialog_device{}
    @editable Camera : gameplay_camera_fixed_point_device = gameplay_camera_fixed_point_device{}
    @editable Volume : volume_device = volume_device{}
    @editable var ColorLabelProp : creative_prop = creative_prop{}
    @editable Accolade : accolades_device = accolades_device{}
    @editable HUD_HideTurnTimer : hud_message_device = hud_message_device{}
    @editable SFX_Spray : audio_player_device = audio_player_device{}
    @editable AnalyticDevices : []analytics_device = array{} # 1 for each car paint style
    @editable VolumeAnalytic : analytics_device = analytics_device{}
    @editable ZoomInputsToHide : []input_trigger_device = array{}
    @editable HUD_Alerts : []hud_message_device = array{}

    # (Pitch=0.000000,Yaw=-334.200331,Roll=0.000000)
    var ConfigRotation : rotation = MakeRotationFromYawPitchRollDegrees(25.799669, 0.000000, 0.000000)
    # (X=-4935.746078,Y=64.000008,Z=2002.377896)

    var ConfigLocation : vector3 = vector3{X:=-4935.746078,Y:=64.000008,Z:=2002.377896}
    # (X=1.333333,Y=1.333333,Z=1.333333)
    var ConfigScale : vector3 = vector3{X:=1.0,Y:=1.0,Z:=1.0}

    # var IsSpinning : logic = false
    var InUse : logic = false
    var RotationSpeed : float = 0.75
    var CarSpinAxis : vector3 = vector3{Z:=1.0}

    var CurrentPaintIndex : int = 1

    var WasAlreadyNearby : [agent]logic = map{}


    MessageAtomGreyMatte <localizes> : message = "Atom Grey Matte"
    MessageDigitalWhiteMetallic <localizes> : message = "Digital White Metallic"
    MessageLaserRedMetallic <localizes> : message = "Laser Red Metallic"
    MessageMetaBlackMetallic <localizes> : message = "Meta Black Metallic"
    MessagePhotonOrangeMetallic <localizes> : message = "Photon Orange Metallic"


    OnBegin<override>()<suspends>:void=
        # Assign random paint style to car
        # CarProp.SetMaterial(GetRandomBrainCarPaintStyle())

        # Button.InteractedWithEvent.Subscribe(HandleButtonInteracted)
        DialogPopup.RespondingButtonEvent.Subscribe(HandleRespondingButton)
        DialogPopup.SetTitleText(MessagePhotonOrangeMetallic)
        Volume.AgentExitsEvent.Subscribe(HandleAgentExitsVolume)
        Volume.AgentEntersEvent.Subscribe(HandleAgentEntersVolume)

        spawn . ShowCollabAlert()

        spawn . SpinCar()

        loop:
            Agent := Button.InteractedWithEvent.Await()

            # Show Dialog
            block:
                if(InUse?) then break
            
                set InUse = true
                Button.Disable()
                Switch.SetState(true)
                Camera.AddTo(Agent)

                OtherPlayerOnPlatform := FindPlayersWithRadius(512.0)
                for(Player : OtherPlayerOnPlatform, FortChar := Player.GetFortCharacter[]):
                    FortChar.Hide()

                # Hide UI things
                HUD_HideTurnTimer.Show(Agent)
                for(Input:ZoomInputsToHide):
                    Input.Unregister(Agent)

                if:
                    TheGlobals := Globals[GetSession()]
                    Player := player[Agent]
                    CPlayer := TheGlobals.CustomPlayerMap[Player]
                then:
                    CPlayer.GuessUI.UpdateReticle(_UI.UI_Textures.transparent_square, vector2{X:=64.0, Y:=64.0})


                # Actually Show Dialog
                DialogPopup.Show(Agent)

            race:
                Agent.AwaitNotActive()
                PaintAppliedEvent.Await()

            set InUse = false
            Button.Enable()
            Switch.SetState(false)
                

        # loop:
        #     SpinCar()
        #     CancelSpin.Await()
        #     StopCarSpin()

        #     # CurrentRotation := CarProp.GetTransform().Rotation
        #     # InitialRotation.RotateBy(MakeShortestRotationBetween(InitialRotation, FinalRotation)) = FinalRotation` and
        #     # RotateToConfig := MakeShortestRotationBetween(CurrentRotation, ConfigRotation)
        #     CarProp.MoveTo(ConfigLocation, ConfigRotation, 0.25)

        #     DialogClosed.Await()
        #     Sleep(2.0)

    # HandleButtonInteracted(Agent:agent):void=
    #     if(InUse?):
    #         return

    #     set InUse = true
    #     Button.Disable()
    #     Switch.SetState(true)
    #     # CancelSpin.Signal()
    #     Camera.AddTo(Agent)

    #     # if(FortChar := Agent.GetFortCharacter[]):
    #     #     FortChar.Hide()

        
    #     OtherPlayerOnPlatform := FindPlayersWithRadius(512.0)
    #     for(Player : OtherPlayerOnPlatform, FortChar := Player.GetFortCharacter[]):
    #         FortChar.Hide()

    #     DialogPopup.Show(Agent)

    ShowCollabAlert()<suspends>:void=

        Sleep(60.0)
        for(HUD_Alert : HUD_Alerts):
            HUD_Alert.Show()

        loop:
            for(HUD_Alert : HUD_Alerts):
                HUD_Alert.Show()
            Sleep(480.0)

    HandleRespondingButton(Agent:agent, ButtonIndex:int):void=
        
        DialogPopup.SetTitleText(GetTitleMessageForPaintStyleIndex(ButtonIndex))
        
        block:
            CarProp.SetMaterial(GetBrainCarPaintStyleFromIndex(ButtonIndex))
            PlatformProp.SetMaterial(GetBrainPlatformMaterialFromIndex(ButtonIndex), ?Index:=2)
            
        ColorLabelProp.SetMaterial(GetColorLabelMaterailFromIndex(ButtonIndex))

        DialogPopup.SetDescriptionText(StringToMessage("1.0"))
        spawn . ResetDescriptionText()
        

        if(ButtonIndex = 5): #Confirm
            
            if(Analytic := GetAnalyticDeviceForPaintStyleIndex[CurrentPaintIndex]):
                Analytic.Submit(Agent)

                Accolade.Award(Agent)
                # DialogClosed.Signal()
                
                if(FortChar := Agent.GetFortCharacter[]):
                    FortChar.Show()
                Camera.RemoveFrom(Agent)
                CarProp.SetMaterial(GetBrainCarPaintStyleFromIndex(CurrentPaintIndex))
                DialogPopup.SetTitleText(GetTitleMessageForPaintStyleIndex(CurrentPaintIndex))
                block:
                    ColorLabelProp.SetMaterial(GetColorLabelMaterailFromIndex(CurrentPaintIndex))
                    PlatformProp.SetMaterial(GetBrainPlatformMaterialFromIndex(CurrentPaintIndex), ?Index:=2)
                PaintAppliedEvent.Signal()
                HUD_HideTurnTimer.Hide(Agent)
                for(Input:ZoomInputsToHide):
                    Input.Register(Agent)

                if:
                    TheGlobals := Globals[GetSession()]
                    Player := player[Agent]
                    CPlayer := TheGlobals.CustomPlayerMap[Player]
                then:
                    CPlayer.GuessUI.UpdateReticle(CPlayer.GuessUI.GetDefaultReticleTexture(Agent, CPlayer.GameBoard))
                    Print("Reset Reticle")


        else:
            set CurrentPaintIndex = ButtonIndex
            DialogPopup.Show(Agent)

        SFX_Spray.Play()

    ResetDescriptionText()<suspends>:void=
        Sleep(0.5)
        DialogPopup.SetDescriptionText(StringToMessage("-1.0"))


    GetRandomBrainCarPaintStyle():material=
        RandomIndex := GetRandomInt(0, 6)
        return GetBrainCarPaintStyleFromIndex(RandomIndex)

    GetBrainCarPaintStyleFromIndex(Index:int):material=
        case(Index):
            0 => return Collabs.BrainCar.HC1H2D3B61EU001257_BB.MI_Car_Ext_HC1H2D3B61EU001257
            1 => return Collabs.BrainCar.HC1H2D3B61EU010069_BB.MI_Car_Ext_HC1H2D3B61EU010069
            2 => return Collabs.BrainCar.HC1H2D3B61EU010070_BB.MI_Car_Ext_HC1H2D3B61EU010070
            3 => return Collabs.BrainCar.HC1H2D3B61EU011295_BB.MI_Car_Ext_HC1H2D3B61EU011295
            4 => return Collabs.BrainCar.HC1H2D3B61EU011299_BB.MI_Car_Ext_HC1H2D3B61EU011299
            _ => return Collabs.BrainCar.HC1H2D3B61EU001257_BB.MI_Car_Ext_HC1H2D3B61EU001257

    GetBrainPlatformMaterialFromIndex(Index:int):material=
        case(Index):
            0 => return Collabs.BrainCar.MI_Platform_HC1H2D3B61EU001257
            1 => return Collabs.BrainCar.MI_Platform_HC1H2D3B61EU010069
            2 => return Collabs.BrainCar.MI_Platform_HC1H2D3B61EU010070
            3 => return Collabs.BrainCar.MI_Platform_HC1H2D3B61EU011295
            4 => return Collabs.BrainCar.MI_Platform_HC1H2D3B61EU011299
            _ => return Collabs.BrainCar.MI_Platform_HC1H2D3B61EU001257

    GetColorLabelMaterailFromIndex(Index:int):material=
        case(Index):
            0 => return Collabs.BrainCar.MI_BrainColor_Sign_MetaBlackMetallic
            1 => return Collabs.BrainCar.MI_BrainColor_Sign_PhotonOrangeMetallic
            2 => return Collabs.BrainCar.MI_BrainColor_Sign_AtomGrey
            3 => return Collabs.BrainCar.MI_BrainColor_Sign_LaserRedMetallic
            4 => return Collabs.BrainCar.MI_BrainColor_Sign_DigitalWhiteMetallic
            _ => return Collabs.BrainCar.MI_BrainColor_Sign_MetaBlackMetallic

    GetTitleMessageForPaintStyle(PaintStyle:brain_car_paint_style):message=
        case(PaintStyle):
            brain_car_paint_style.HC1H2D3B61EU001257 => return MessageMetaBlackMetallic
            brain_car_paint_style.HC1H2D3B61EU010069 => return MessagePhotonOrangeMetallic
            brain_car_paint_style.HC1H2D3B61EU010070 => return MessageAtomGreyMatte
            brain_car_paint_style.HC1H2D3B61EU011295 => return MessageLaserRedMetallic
            brain_car_paint_style.HC1H2D3B61EU011299 => return MessageDigitalWhiteMetallic

    GetTitleMessageForPaintStyleIndex(Index:int):message=
        case(Index):
            0 => return MessageMetaBlackMetallic
            1 => return MessagePhotonOrangeMetallic
            2 => return MessageAtomGreyMatte
            3 => return MessageLaserRedMetallic
            4 => return MessageDigitalWhiteMetallic
            _ => return MessageMetaBlackMetallic


    GetAnalyticDeviceForPaintStyleIndex(Index:int)<transacts><decides>:analytics_device=
        case(Index):
            0 => return AnalyticDevices[0]
            1 => return AnalyticDevices[1]
            2 => return AnalyticDevices[2]
            3 => return AnalyticDevices[3]
            4 => return AnalyticDevices[4]
            5 => return AnalyticDevices[5]
            _ => return AnalyticDevices[0]
        

    SpinCar()<suspends>:void=
        if:
            CarProp.IsValid[]
            AnimationController := CarProp.GetAnimationController[]
        then:
            Keyframes := array:
                keyframe_delta:
                    DeltaLocation := vector3{}
                    DeltaRotation := MakeRotation(CarSpinAxis, -PiFloat * 360.0 * RotationSpeed / 2.0)
                    Time := 15.0 / 2.0
                    Interpolation := InterpolationTypes.Linear
                keyframe_delta:
                    DeltaLocation := vector3{}
                    DeltaRotation := MakeRotation(CarSpinAxis, -PiFloat * 360.0 * RotationSpeed / 2.0)
                    Time := 15.0 / 2.0
                    Interpolation := InterpolationTypes.Linear

            AnimationController.SetAnimation(Keyframes, ?Mode := animation_mode.Loop)

            Print("Starting car spin")
            AnimationController.Play()

    StopCarSpin()<suspends>:void=
        if:
            CarProp.IsValid[]
            AnimationController := CarProp.GetAnimationController[]
        then:
            AnimationController.Stop()

    HandleAgentExitsVolume(Agent:agent):void=
        if(FortChar := Agent.GetFortCharacter[]):
            FortChar.Show()

    HandleAgentEntersVolume(Agent:agent):void=
        if(InUse?):
            if(FortChar := Agent.GetFortCharacter[]):
                FortChar.Hide()

        if:
            not WasAlreadyNearby[Agent]
        then:
            VolumeAnalytic.Submit(Agent)
            if . set WasAlreadyNearby[Agent] = true
        

    FindPlayersWithRadius(Radius:float):[]agent=
        Playspace := GetPlayspace()
        Players := Playspace.GetPlayers()

        PlayersOnPlatform := for:
            Player : Players
            FortChar := Player.GetFortCharacter[]
            # Distance(FortChar.GetTransform().Translation, CarProp.GetTransform().Translation) < Radius
            DistanceSquared(FortChar.GetTransform().Translation, CarProp.GetTransform().Translation) < Pow(Radius, 2.0)
        do:
            Player

        
