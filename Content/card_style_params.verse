using { /Verse.org/Assets }
using { /Verse.org/Colors }
using { /Verse.org/Simulation }

card_style_params := class():
    CardStyle : card_style
    CardMaterial : material

    FrameMaterial : material
    BoardMaterial : material