
using { /Fortnite.com/Devices }
using { /Verse.org/Simulation }
using { /UnrealEngine.com/Temporary/Diagnostics }
using { /UnrealEngine.com/Temporary/SpatialMath }
using { /Fortnite.com/Devices/CreativeAnimation }
# using { /Fortnite.com/Devices/Animation }
using { /Verse.org/Random }
using { /Fortnite.com/Characters }
using { /Verse.org/Colors/NamedColors }
using { /UnrealEngine.com/Temporary/UI }
using { /Fortnite.com/UI }
using { /Verse.org/Assets }
using { /Fortnite.com/FortPlayerUtilities }

card_tile := class<concrete><unique>():

    Draw:debug_draw := debug_draw{Channel:=debug_drawing}

    var Index : int = 0
    @editable Prop : creative_prop = creative_prop{}
    @editable Trigger : trigger_device = trigger_device{}
    var AnimController : ?animation_controller = false
    var State : card_state = card_state.Down

    var FlipDirection : float = 0.0
    var LocalForward : vector3  = vector3{X:=1.0, Y:=0.0, Z:=0.0}
    # var InitialRotation : rotation = rotation{}


    Init(aFlipDirection:float, aLocalForward:vector3):void=
        set Self.LocalForward = aLocalForward
        set Self.FlipDirection = aFlipDirection

    SetStyle(Style:card_style):void=
        Self.Prop.SetMaterial(MakeMaterialForCardStyle(Style), ?Index:=0)


    # HandleTriggerActivated(Agent:?agent):void=
    #     case(State):
    #         card_state.Up =>
    #             spawn{Self.KnockDown()}
    #             return
    #         card_state.Down =>
    #             spawn{Self.FlipUp()}
    #             return
    #         card_state.Animating =>
    #             return

    Reset(TargetIndex:int):void=

        # if(Self.Index = TargetIndex):
        #     return

       case(State):
            card_state.Up =>
                spawn{Self.KnockDown()}
                return
            card_state.Down =>
                return
            card_state.Animating =>
                return

    FlipUp()<suspends>:void=
        # Print("Flipping Up Card")
        set AnimController = option{Self.Prop.GetAnimationController[]}
            if (MyAnim := AnimController?, MyAnim.IsValid[], Prop.IsValid[]):

                if (MyAnim.GetState() = animation_controller_state.Playing):
                    Print("Animation Controller is already playing")
                    return
                    
                CurrentRotation := Self.Prop.GetTransform().Rotation
                AngleRadians := DegreesToRadians(120.0 * FlipDirection)
                LocalRot := MakeRotation(LocalForward, AngleRadians)
    
                
                KeyframeDelta := keyframe_delta{
                    DeltaLocation := vector3{X:=0.0, Y:=0.0, Z:=0.0},
                    DeltaRotation := LocalRot, 
                    Time := FlipSpeed,
                    Interpolation := InterpolationTypes.Linear
                }
    
                MyAnim.SetAnimation(array{KeyframeDelta}, ?Mode:=animation_mode.OneShot)
                set State = card_state.Animating
                MyAnim.Play()
                MyAnim.AwaitNextKeyframe()
                set State = card_state.Up
            else:
                Print("No Animation Controller Found")

    KnockDown()<suspends>:void=
        Print("Knocking Down Card")
        set AnimController = option{Self.Prop.GetAnimationController[]}
            if (MyAnim := AnimController?, MyAnim.IsValid[], Prop.IsValid[]):

                if (MyAnim.GetState() = animation_controller_state.Playing):
                    Print("Animation Controller is already playing")
                    return
    
                CurrentRotation := Self.Prop.GetTransform().Rotation
                AngleRadians := DegreesToRadians(120.0 * FlipDirection)    
                LocalRot := MakeRotation(LocalForward, -AngleRadians)


                KeyframeDelta := keyframe_delta{
                    DeltaLocation := vector3{X:=0.0, Y:=0.0, Z:=0.0},
                    DeltaRotation := LocalRot, 
                    Time := FlipSpeed,
                    Interpolation := InterpolationTypes.Linear
                }
    
                MyAnim.SetAnimation(array{KeyframeDelta}, ?Mode:=animation_mode.OneShot)            
                set State = card_state.Animating
                MyAnim.Play()                   

                MyAnim.AwaitNextKeyframe()                
                set State = card_state.Down


                TileTransform := Prop.GetTransform()
                LocalUp := TileTransform.Rotation.GetLocalRight()
                UpwardOffset := LocalUp * TileDimensions.Y / 2.0
                TilePosition := if(State = card_state.Up) then TileTransform.Translation + TilePivotOffset else TileTransform.Translation + vector3{Y:=32.0}
                BoxCenter := TilePosition + UpwardOffset

                Draw.DrawBox(BoxCenter, TileTransform.Rotation, ?Extent:=vector3{X:=TileDimensions.X / 2.0, Y:=TileDimensions.Y / 2.0, Z:=TileDimensions.Z / 2.0}, ?Color:=Red, ?Duration:=10.0)
            else:
                Print("No Animation Controller Found")

            

    Flip():void=
        if(State = card_state.Up):
            spawn{Self.KnockDown()}
        else:
            spawn{Self.FlipUp()}

    IsRotationAboveGround(CurrentRotation: rotation, AngleChange: float)<transacts><decides>:void=
        ExpectedXDegrees := CurrentRotation.GetYawPitchRollDegrees()[1] + AngleChange
        ThresholdDegrees := 180.0
        IsAbove := ExpectedXDegrees < ThresholdDegrees


    DebugDraw():void=
        # DrawBox<native><public>(Center:vector3, Rotation:rotation, ?Extent:vector3 = external {}, ?Color:color = external {}, ?Thickness:float = external {}, ?DrawDurationPolicy:debug_draw_duration_policy = external {}, ?Duration:float = external {})<transacts>:void
        Draw.DrawBox(Self.Prop.GetTransform().Translation, Self.Prop.GetTransform().Rotation, ?Extent:=vector3{X:=TileDimensions.X / 2.0, Y:=TileDimensions.Y / 2.0, Z:=TileDimensions.Z / 2.0}, ?Color:=Red, ?Thickness:=2.0, ?DrawDurationPolicy:=debug_draw_duration_policy.SingleFrame)