using { /Fortnite.com/Devices }
using { /Verse.org/Simulation }
using { /UnrealEngine.com/Temporary/Diagnostics }
using { /Fortnite.com/Characters }
using { /UnrealEngine.com/Temporary/SpatialMath }
using { /Verse.org/Colors/NamedColors }
using { /Fortnite.com/Devices/CreativeAnimation }
using { /Verse.org/Random }

cr_ray_trace := class<concrete>():
    @editable Top : creative_prop = creative_prop{}
    @editable Bottom : creative_prop = creative_prop{}

    CastFrom(FromTrans:transform)<suspends>:transform=

        var HitLocation : transform = transform{}
        
        FromLocation := FromTrans.Translation
        FromRot := FromTrans.Rotation
        FromScale := FromTrans.Scale

        if:
            Top.IsValid[]
            Top.TeleportTo[FromLocation, FromRot]
        then:
            Sleep(0.0)
            set HitLocation = Bottom.GetTransform()
        else:
            Print("Top is not valid or teleport failed")

        return HitLocation


    CastFrom(Translation:vector3, Rotation:rotation)<suspends>:transform=

        var HitLocation : transform = transform{}

        if:
            Top.IsValid[]
            Top.TeleportTo[Translation, Rotation]
        then:
            Sleep(0.0)
            set HitLocation = Bottom.GetTransform()
        else:
            Print("Top is not valid or teleport failed")

        return HitLocation

