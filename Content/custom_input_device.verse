using { /Fortnite.com/Devices }
using { /Verse.org/Simulation }
using { /UnrealEngine.com/Temporary/Diagnostics }
using { /Fortnite.com/Characters }
using { /UnrealEngine.com/Temporary/SpatialMath }
using { /Verse.org/Colors/NamedColors }
using { /Fortnite.com/Devices/CreativeAnimation }
using { /Verse.org/Random }


custom_input_device := class<concrete>():

    @editable Cursor_RayCast : cr_ray_trace = cr_ray_trace{}

    @editable Cursor_InputTrigger_Fire : input_trigger_device = input_trigger_device{}
    var FireSubscription : ?cancelable = false

    @editable Cursor_InputTrigger_Zoom : input_trigger_device = input_trigger_device{}
    @editable InputTrigger_Undo : input_trigger_device = input_trigger_device{}
    @editable InputTrigger_Ask_Question : input_trigger_device = input_trigger_device{}

    @editable CustomZoomCameraOLD : gameplay_camera_fixed_point_device = gameplay_camera_fixed_point_device{}
    # @editable CustomZoomCamera2 : gameplay_camera_fixed_angle_device = gameplay_camera_fixed_angle_device{}

    @editable CameraAddTrigger : trigger_device = trigger_device{}
    @editable CameraRemoveTrigger : trigger_device = trigger_device{}

    @editable InputTrigger_Forward : input_trigger_device = input_trigger_device{}
    @editable InputTrigger_Backward : input_trigger_device = input_trigger_device{}
    @editable InputTrigger_Left : input_trigger_device = input_trigger_device{}
    @editable InputTrigger_Right : input_trigger_device = input_trigger_device{}
