using { /UnrealEngine.com/Temporary/SpatialMath }
using { /Verse.org/Simulation }
using { /UnrealEngine.com/Temporary/UI }
using { /Fortnite.com/UI }
using { /Fortnite.com/Characters }
using { /Fortnite.com/FortPlayerUtilities }
using { /Verse.org/Simulation/Tags }
using { /Fortnite.com/Game }
using { /Fortnite.com/Devices }
using { /UnrealEngine.com/Temporary/Diagnostics }
using { /Verse.org/Concurrency }

player_state := enum{Active, Falling, OnGround, Flying, Jumping, Gliding, Respawning, Walking, Sprinting}



custom_player := class():

    Logger:log := log{Channel:=log_device}
    Draw:debug_draw := debug_draw{Channel:=debug_drawing}

    TeamIndex : int
    var Player : player
    var State : player_state = player_state.Active
    var Score : int = 0
    var MenuCategoryIndex : int = 0
    # var YesNoCine : cinematic_sequence_device = cinematic_sequence_device{}

    var GuessUI : guess_ui = guess_ui{}

    var GameBoard : game_board

    var HasVoted : logic = false

    var In1v1Area : logic = false
    var AwayFromGameplayArea : logic = false

    var LiveRayCastActive : logic = false
    var CancelableLiveRayCast : ?cancelable = false

    var MaybePointerConditional : ?conditional_button_device = false


    GetOpponent()<transacts>:?player=

        var TempPlayer : ?player = false

        if:
            TheGlobals := Globals[GetSession()]
            Players:= TheGlobals.AllPlayers
        then:
            for (aPlayer :Players, aPlayer <> Self.Player):
                set TempPlayer = option{aPlayer}
        else:
            Logger.Print("GetOpponent Failed")

        return TempPlayer


    GetVote()<transacts><decides>:int=
        TheGlobals := Globals[GetSession()]
        return TheGlobals.PlayerVoteIndex[Self.Player]


    AwaitExit1v1Area()<suspends>:void=
        loop:
            if(not In1v1Area?):
                break
            Sleep(0.1)


    Init(aPlayer:player):void=
        set Self.Player = aPlayer
        # set Self.YesNoCine = aYesNoCine
        if:           
            FortChar := Player.GetFortCharacter[]
        then:
            FortChar.EliminatedEvent().Subscribe(HandlePlayerEliminated)
            FortChar.DamagedEvent().Subscribe(HandlePlayerDamaged)
            FortChar.DamagedShieldEvent().Subscribe(HandlePlayerDamaged)
            

            # Self.GuessUI.InitStyleSelectionUI(aPlayer, Self)

            Self.GuessUI.InitCorrectIncorrectUI(aPlayer)
            Self.GuessUI.InitReticleUI(aPlayer)
            Self.GuessUI.CreateSelectionScreenOverlays()
            Self.GuessUI.CreateSelectionScreenOptions()
            Self.GuessUI.InitStyleData()
           
            Print("Init Custom Player")


    AwaitVoted()<suspends>:int=
        Print("Awaiting Player Vote")

        var VotedIndex : int = 0

        loop:
            if(Self.HasVoted?):
                break
            Sleep(0.1)
            
        if:
            TheGlobals := Globals[GetSession()]
            TheVote := TheGlobals.PlayerVoteIndex[Self.Player]
        then:
            set VotedIndex = TheVote

        return -1


    HandlePlayerEliminated(Result:elimination_result):void=
        Print("Player Eliminated")

    HandlePlayerDamaged(Result:damage_result):void=
        Print("Player Damaged {Result.Amount}")

    AddScore<public>(Amount:int):void=
        set Self.Score += Amount
        Print("Add Score! - Score: {Self.Score}")
        if:
            CheckForPlayerSavedData[Player]
            SourceData := PlayerSavedDataMap[Player]
        then:
            AddScore(Player, Amount)


    AddDeath<public>(Amount:int):void=
        # set Self.Deaths += Amount
        Print("Death")

    Heal<public>():void=
        if:
            FortChar := Player.GetFortCharacter[]
        then:
        
            MaxHealth := FortChar.GetMaxHealth()
            MaxShield := FortChar.GetMaxShield()

            if(FortChar.IsActive[]): 
                FortChar.SetHealth(MaxHealth)
                FortChar.SetShield(MaxShield)
        else:
            Print("Heal Failed")
            Self.Heal()


    Kill<public>():void=
        if:
            FortChar := Player.GetFortCharacter[]
        then:
            if(FortChar.IsActive[]): 
                FortChar.Damage(1000000.0)
        else:
            Print("Kill Failed")
            Self.Kill()


    Respawn<public>(Location:?vector3)<suspends>:void=
        if(LocationExists:=Location?):
            Player.Respawn(LocationExists, rotation{})
        else:
            Player.Respawn(vector3{X:=57.0, Y:=29104.0, Z:=2360.0}, rotation{})

    ChangeState<public>(NewState:player_state):void=
        case(NewState):
            player_state.Active => 
                Print("Active")
                set Self.State = NewState

            player_state.Falling => 
                Print("Falling")
                set Self.State = NewState

            player_state.OnGround => 
                Print("OnGround")
                set Self.State = NewState

            player_state.Jumping => 
                Print("Jumping")
                set Self.State = NewState

            player_state.Flying => 
                Print("Flying")
                set Self.State = NewState

            player_state.Respawning => 
                Print("Respawning")
                set Self.State = NewState

            player_state.Walking => 
                Print("Walking")
                set Self.State = NewState

            player_state.Sprinting => 
                Print("Sprinting")
                set Self.State = NewState

            _ => Print("State Error")