


custom_player_data := class<final><persistable>():
    
    DataVersion : int = 1

    IsAdmin : logic = false
    IsModerator : logic = false
    IsDeveloper : logic = false

    # Moderation
    IsBanned : logic = false
    BanReason : ?string = false
    BanExpiration : int = 0
    BanIssuedBy : ?string = false
    BansIssued : int = 0
    WarningsIssued : int = 0
    PastWarnings : []string = array{}

    Score : int
    Rank : int
    Level : int

    Gold : int
    OverrideTokens : int

    Wins : int
    Losses : int
    NoKnockWins : int

    CardsKnocked : int
    UndosPressed : int

    RoundsPlayed : int
    TimePlayed : int
    LastDayPlayed : string
    LastDayWon : string
    DaysPlayedStreak : int
    
    QuestionsAsked : int
    QuestionsAnswered : int
    AnsweredYes : int
    AnsweredNo : int
    AnsweredDontKnow : int

    UnlockedBonusStyles : []int

    UnlockedBoardMeshes : []int
    UnlockedBoardPaints : []int

    UnlockedCardMeshes : []int
    UnlockedCardBackPaints : []int

    UnlockedPlayerMeshes : []int
    UnlockedPlatformPaints : []int

    UnlockedYouStandMeshes : []int
    UnlockedYouStandPaints : []int

    UnlockedHologramPoses : []int

    UnlockedMusic : []int
    UnlockedKnockSFX : []int
    UnlockedWinSFX : []int
    UnlockedLoseSFX : []int


    

    
