
using { /Verse.org/Simulation }

# SecondsInHour : float = 3600.0
# SecondsInMinute : float= 60.0


HoursToSeconds(Hours:float):float=
    return Hours * 3600.0

MinsToSeconds(Mins:float):float=
    return Mins * 60.0

SecondsToHours(Seconds:float):float=
    return Seconds / 3600.0

SecondsToMins(Seconds:float):float=
    return Seconds / 60.0

(Timer:custom_timer).ToString():string=
    var HoursStr : string = "{Timer.Hours}"
    var MinutesStr : string = "{Timer.Minutes}"
    var SecondsStr : string = "{Timer.Seconds}"

    if(Timer.Hours <10):
        set HoursStr = "0{Timer.Hours}"

    if(Timer.Minutes <10):
        set MinutesStr = "0{Timer.Minutes}"

    if(Timer.Seconds <10):
        set SecondsStr = "0{Timer.Seconds}"        


    return "{HoursStr}:{MinutesStr}:{SecondsStr}"


    

custom_timer :=class():
    var <private>Hours : int = 0
    var <private>Minutes : int = 0
    var <private>Seconds : int = 0
    var <private>Milliseconds : int = 0
    var SecondsInHour : float = 3600.0
    var SecondsInMinute : float= 60.0

    var Running : logic = false
    var Canceled : logic = false

    SetTime(H:int, M:int, S:int):void=
        set Self.Hours = H
        set Self.Minutes = M
        set Self.Seconds = S

    AddHour(Amount:int):void=
        set Self.Hours += Amount

    AddMinute(Amount:int):void=
        set Self.Minutes += Amount
        if(Self.Minutes >= 60):
            if:
                Q := Floor(Self.Minutes / 60)
            then:
                Self.AddHour(Q)  # Add the quotient to Hours
                if(RoundedDown := Mod[Self.Minutes, 60]):
                    set Self.Minutes = RoundedDown  # Set the remainder as Minutes
                Print("Added {Q} hour(s), New Time: {Self.ToString()}")
    
    AddSecond(Amount:int):void=
        set Self.Seconds += Amount
        if(Self.Seconds >= 60):
            if:
                Q := Floor(Self.Seconds / 60)  # Corrected line
            then:
                if(Modu := Mod[Self.Seconds, 60]):
                    set Self.Seconds = Modu  # Set the remainder as Seconds
                Self.AddMinute(Q)  # Then add the quotient to Minutes
                Print("Added {Q} minute(s), New Time: {Self.ToString()}")
    
        
        
        
            

    SetHour(Amount:int):void=
        set Self.Hours = Amount

    SetMinute(Amount:int):void=
        set Self.Minutes = Amount
    
    SetSecond(Amount:int):void=
        set Self.Seconds = Amount

    SetSecondsInHour(NewValue: float):void=
        set Self.SecondsInHour = NewValue
    
    # Method to set SecondsInMinute for testing
    SetSecondsInMinute(NewValue: float):void=
        set Self.SecondsInMinute = NewValue

    ResetHours():void=
        set Self.Hours = 0

    ResetMinutes():void=
        set Self.Minutes = 0
    
    ResetSeconds():void=
        set Self.Seconds = 0

    Reset():void=
        # Sleep(1.0)
        set Self.Hours = 0
        set Self.Minutes = 0
        set Self.Seconds = 0
        set Self.Milliseconds = 0
        set Self.Running = false
        set Self.Canceled = false

    Pause():void=
        set Self.Running = false

    ToSeconds()<transacts>:float=
        return (Self.Hours * SecondsInHour) + (Self.Minutes * SecondsInMinute) + (Self.Seconds * 1.0)

    Cancel():void=
        set Self.Canceled = true
        Self.Reset()

    AwaitTime(Sec:int)<suspends>:void=
        Self.Reset()

        var TempSec : int = Sec

        set Self.Running = true

        # Convert total seconds into hours, minutes, and seconds
        if:
            TotalHours:= Floor[((Sec * 1.0 ) / Self.SecondsInHour)]
            set TempSec -= Floor[TotalHours * Self.SecondsInHour]
            TotalMinutes:= Floor[(Sec * 1.0 ) / Self.SecondsInMinute]
            set TempSec -= Floor[TotalMinutes * Self.SecondsInMinute]
        then:
            TotalSeconds:= TempSec
            # Set the timer
            Self.SetTime(TotalHours, TotalMinutes, TotalSeconds)

        # defer:
        #     Self.Reset()

        loop:
            if(Self.ToSeconds() <= 0.0):
                break

            if(Self.Canceled?):
                break
            
            set Self.Seconds -= 1
            if(Self.Seconds < 0):
                set Self.Seconds = 59
                set Self.Minutes -= 1
                if(Self.Minutes < 0):
                    set Self.Minutes = 59
                    set Self.Hours -= 1

            Sleep(1.0)
        
        Print("Stopwatch has finished!")
        # Self.Reset()