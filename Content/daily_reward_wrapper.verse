using { /Fortnite.com/Devices }
using { /Verse.org/Simulation }

daily_reward_wrapper := class<concrete>():
    @editable ClaimButton : button_device = button_device{}
    @editable HudReward : hud_message_device = hud_message_device{}
    @editable HudFail : hud_message_device = hud_message_device{}
    @editable Accolade : accolades_device = accolades_device{}
    @editable Granter : item_granter_device = item_granter_device{}
    @editable DebugBillboard : billboard_device = billboard_device{}