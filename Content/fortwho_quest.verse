using { /Verse.org/Simulation }

fortwho_quest := class():
    ID : int
    Name : string
    Description : string
    QuestType : int # QuestType: 0 = Main, 1 = Daily, 2 = Weekly, 3 = Monthly, 4 = Yearly, 5 = Board Designer, 6 = Social Media
    
    Target : int
    var Progress : int
    var Complete : logic

    RewardType : int # RewardType: 0 = None, 1 = Currency, 2 = Item, 3 = Character, 4 = Board, 5 = Card, 6 = Stand, 7 = Platform, 8 = Material, 9 = Paint, 10 = Accessory, 11 = Emote, 12 = Avatar, 13 = Title, 14 = Profile, 15 = Background, 16 = Music, 17 = Sound, 18 = Effect, 19 = Animation, 20 = Skin, 21 = Sticker, 22 = Decal, 23 = Spray, 24 = Tag, 25 = Quest, 26 = Challenge, 27 = Event, 28 = Season, 29 = Special, 30 = Community, 31 = Limited Time, 32 = Custom, 33 = Other
    RewardID : int
    RewardAmount : ?int

    RewardType2 : ?int
    RewardID2 : ?int
    RewardAmount2 : ?int

    RewardType3 : ?int
    RewardID3 : ?int
    RewardAmount3 : ?int

    CompletedEvent : event() = event(){}

    AwaitCompletion()<suspends>:void=
        loop:
            if(Complete?):
                break
            Sleep(1.0)
        CompletedEvent.Signal()

    CompleteQuest()<suspends>:void=
        set Complete = true

GetQuestFromID(QuestID:int)<transacts><decides>:fortwho_quest=
    Print("TEMP - GetQuestFromID - MAKE SURE TO IMPLEMENT THIS FUNCTION")

    case(QuestID):
        2 => fortwho_quest{ID := 2, Name := "Quest 2", Description := "Quest Description", QuestType := 6, Target := 1, Progress := 0, Complete := false, RewardType := 1, RewardID := 0, RewardAmount := option{100}, RewardType2 := option{0}, RewardID2 := option{0}, RewardAmount2 := option{0}, RewardType3 := false, RewardID3 := false, RewardAmount3 := false}
        3 => fortwho_quest{ID := 3, Name := "Quest 3", Description := "Quest Description", QuestType := 5, Target := 5, Progress := 0, Complete := false, RewardType := 1, RewardID := 0, RewardAmount := option{100}, RewardType2 := option{0}, RewardID2 := option{0}, RewardAmount2 := option{0}, RewardType3 := false, RewardID3 := false, RewardAmount3 := false}
        4 => fortwho_quest{ID := 4, Name := "Quest 4", Description := "Quest Description", QuestType := 5, Target := 10, Progress := 4, Complete := false, RewardType := 1, RewardID := 0, RewardAmount := option{100}, RewardType2 := option{0}, RewardID2 := option{0}, RewardAmount2 := option{0}, RewardType3 := false, RewardID3 := false, RewardAmount3 := false}
        5 => fortwho_quest{ID := 5, Name := "Quest 5", Description := "Quest Description", QuestType := 5, Target := 15, Progress := 0, Complete := false, RewardType := 1, RewardID := 0, RewardAmount := option{100}, RewardType2 := option{0}, RewardID2 := option{0}, RewardAmount2 := option{0}, RewardType3 := false, RewardID3 := false, RewardAmount3 := false}
        6 => fortwho_quest{ID := 6, Name := "Quest 6", Description := "Quest Description", QuestType := 5, Target := 20, Progress := 0, Complete := false, RewardType := 1, RewardID := 0, RewardAmount := option{100}, RewardType2 := option{0}, RewardID2 := option{0}, RewardAmount2 := option{0}, RewardType3 := false, RewardID3 := false, RewardAmount3 := false}
        7 => fortwho_quest{ID := 7, Name := "Quest 7", Description := "Quest Description", QuestType := 5, Target := 25, Progress := 12, Complete := false, RewardType := 1, RewardID := 0, RewardAmount := option{100}, RewardType2 := option{0}, RewardID2 := option{0}, RewardAmount2 := option{0}, RewardType3 := false, RewardID3 := false, RewardAmount3 := false}
        8 => fortwho_quest{ID := 8, Name := "Quest 8", Description := "Quest Description", QuestType := 5, Target := 30, Progress := 0, Complete := false, RewardType := 1, RewardID := 0, RewardAmount := option{100}, RewardType2 := option{0}, RewardID2 := option{0}, RewardAmount2 := option{0}, RewardType3 := false, RewardID3 := false, RewardAmount3 := false}
        9 => fortwho_quest{ID := 9, Name := "Quest 9", Description := "Quest Description", QuestType := 5, Target := 35, Progress := 0, Complete := false, RewardType := 1, RewardID := 0, RewardAmount := option{100}, RewardType2 := option{0}, RewardID2 := option{0}, RewardAmount2 := option{0}, RewardType3 := false, RewardID3 := false, RewardAmount3 := false}
        10 => fortwho_quest{ID := 10, Name := "Quest 10", Description := "Quest Description", QuestType := 5, Target := 40, Progress := 20, Complete := false, RewardType := 1, RewardID := 0, RewardAmount := option{100}, RewardType2 := option{0}, RewardID2 := option{0}, RewardAmount2 := option{0}, RewardType3 := false, RewardID3 := false, RewardAmount3 := false}
        11 => fortwho_quest{ID := 11, Name := "Quest 11", Description := "Quest Description", QuestType := 5, Target := 45, Progress := 0, Complete := false, RewardType := 1, RewardID := 0, RewardAmount := option{100}, RewardType2 := option{0}, RewardID2 := option{0}, RewardAmount2 := option{0}, RewardType3 := false, RewardID3 := false, RewardAmount3 := false}
        12 => fortwho_quest{ID := 12, Name := "Quest 12", Description := "Quest Description", QuestType := 5, Target := 50, Progress := 49, Complete := false, RewardType := 1, RewardID := 0, RewardAmount := option{100}, RewardType2 := option{0}, RewardID2 := option{0}, RewardAmount2 := option{0}, RewardType3 := false, RewardID3 := false, RewardAmount3 := false}
        13 => fortwho_quest{ID := 13, Name := "Quest 13", Description := "Quest Description", QuestType := 5, Target := 55, Progress := 22, Complete := false, RewardType := 1, RewardID := 0, RewardAmount := option{100}, RewardType2 := option{0}, RewardID2 := option{0}, RewardAmount2 := option{0}, RewardType3 := false, RewardID3 := false, RewardAmount3 := false}
        _ => fortwho_quest{ID := -1, Name := "Invalid Quest", Description := "Invalid Quest", QuestType := 0, Target := 0, Progress := 0, Complete := false, RewardType := 0, RewardID := 0, RewardAmount := option{0}, RewardType2 := option{0}, RewardID2 := option{0}, RewardAmount2 := option{0}, RewardType3 := false, RewardID3 := false, RewardAmount3 := false}
