# ##############################################################
# # OLD OLD OLD 
# # This is the old version of the Fort Who device
# # game_manager is the new one


# using { /Fortnite.com/Devices }
# using { /Verse.org/Simulation }
# using { /UnrealEngine.com/Temporary/Diagnostics }
# using { /UnrealEngine.com/Temporary/SpatialMath }
# using { /Fortnite.com/Devices/CreativeAnimation }
# # using { /Fortnite.com/Devices/Animation }
# using { /Verse.org/Random }
# using { /Fortnite.com/Characters }
# using { /Verse.org/Colors/NamedColors }
# using { /UnrealEngine.com/Temporary/UI }
# using { /Fortnite.com/UI }
# using { /Verse.org/Assets }
# using { /Fortnite.com/FortPlayerUtilities }

# fort_who_device := class(creative_device):

#     Logger:log := log{Channel:=log_device}
#     Draw:debug_draw := debug_draw{Channel:=debug_drawing}

#     @editable AdminKeypad : admin_keypad = admin_keypad{}

#     @editable TriggerRandomForKnockActivation : trigger_device = trigger_device{}

#     @editable MainPlayerCounter : player_counter_device = player_counter_device{}

#     @editable P1SpawnPad : player_spawner_device = player_spawner_device{}
#     @editable P2SpawnPad : player_spawner_device = player_spawner_device{}

#     @editable HUDLoadTimerNumbers : hud_message_device = hud_message_device{}

#     @editable HUDResponseNo : hud_message_device = hud_message_device{}
#     @editable HUDResponseYes : hud_message_device = hud_message_device{}
#     @editable HUDResponseMaybe : hud_message_device = hud_message_device{}

#     QuestionBankDevice : question_bank_device = question_bank_device{}

#     @editable GameBoards : []game_board = array{}
#     # @editable CardStyleCines : []cinematic_sequence_device = array{}
#     @editable EndGame : end_game_device = end_game_device{}
#     @editable WinnerStopWatch : timer_device = timer_device{}
#     @editable ScoreManager : score_manager_device = score_manager_device{}
#     @editable GoldGranter : item_granter_device = item_granter_device{}

#     @editable Teleporter_1v1_Entrance : teleporter_device = teleporter_device{}

#     @editable VotedSound : audio_player_device = audio_player_device{}

#     @editable TrackerLifeTimeWins : tracker_device = tracker_device{}

#     @editable Analytics_Styles : []analytics_device = array{}

#     # @editable TimerVotingTime : timer_device = timer_device{}
#     @editable TimerTurnTime : timer_device = timer_device{}

#     @editable RealTimeClock : rtc_device = rtc_device{}
#     @editable DailyReward : daily_reward_wrapper = daily_reward_wrapper{}

#     var GameInProgress : logic = false

#     var HasFoundOnlyOnePlayer : logic = false
                
#     DrawBoundingBoxAroundTile(Tile:card_tile):void=
#         Print("Drawing bounding box around tile {Tile.Index}")
#         BoundsPadding := 25.0
#         # TileDimensions : vector3 = vector3{X:=171.0, Y:=288.0, Z:=16.0}  # Assuming the dimensions of a tile
#         HalfHeight : float = TileDimensions.Y / 2.0  # Half of the box's height
    
#         # Get the current tile's position and rotation
#         TileTransform := Tile.Prop.GetTransform()
#         TilePosition := TileTransform.Translation
#         TileRotation := TileTransform.Rotation
    
#         # Obtain the local upward vector relative to the tile's rotation
#         LocalUp := TileRotation.GetLocalRight()
    
#         # Calculate the offset to move the center of the bounding box upwards
#         UpwardOffset := LocalUp * HalfHeight
    
#         # Translate the center position upwards by half of the box's height
#         BoxCenter : vector3 = TilePosition + UpwardOffset
    
#         # Now draw the bounding box with the new center position
#         Draw.DrawBox(BoxCenter, TileRotation, ?Extent:=TileDimensions/2.0, ?Color := Green, ?Thickness := 5.0, ?DrawDurationPolicy := debug_draw_duration_policy.FiniteDuration, ?Duration:=5.0)
    
#         Print("Drew bounding box around tile {Tile.Index} at location {BoxCenter}")    

#     HandleButtonDailyReward(Agent:agent):void=
#         Print("Daily Reward Button Pressed")

#         if:
#             LastDayPlayed := GetLastDayPlayed[Agent]
#         then:
#             DailyReward.DebugBillboard.SetText(
#                 FiveLineMessage(
#                     "Last Day Played: {LastDayPlayed.String()}",
#                     "Today: {GetCurrentDateTime().String()}", 
#                     "Days between: {DaysBetweenDates(LastDayPlayed, GetCurrentDateTime())}",
#                     "Consecutive Days: {GetCurrentConsecutiveDaysPlayed(Agent)}",
#                     "Most Consecutive Days: {GetMostConsecutiveDaysPlayed(Agent)}"
#                 )
#             )

#         if:
#             AllowDailyReward[Agent]
#         then:
#             Print("Allowing daily reward")
#             ClaimDailyReward(Agent)
#             TimeUntil := TimeUntilNextUTCMidnight(GetCurrentDateTime())
#             DailyReward.Accolade.Award(Agent)
#             DailyReward.HudReward.Show(Agent, MakeHrsMinsMessage(TimeUntil(0), TimeUntil(1)))
#             DailyReward.Granter.GrantItem(Agent)

#         else:
#             Print("Not allowing daily reward")
#             TimeUntil := TimeUntilNextUTCMidnight(GetCurrentDateTime())
#             Consec := GetCurrentConsecutiveDaysPlayed(Agent)
#             ConsecRecord :=  GetMostConsecutiveDaysPlayed(Agent)
            
#             Print("Time until next daily reward {TimeUntil(0)}hr {TimeUntil(1)}min")
#             #first line : 1hrs 2mins
#             #second line (current consecutive days) : 3 days
#             #third line (max consecutive days) : 5 days
#             HoursString := "{TimeUntil(0)}hrs {TimeUntil(1)}mins"
#             ConsecString := "{Consec} Days"
#             ConsecRecordString := "{ConsecRecord} Days"
#             DailyReward.HudFail.Show(Agent, ThreeLineMessage(HoursString, ConsecString, ConsecRecordString))


#     # Runs when the device is started in a running game
#     OnBegin<override>()<suspends>:void=
#         Print("Starting Fort Who")
        
        

#         if:
#             not Globals[GetSession()]
#         then:
#             if:
#                 set Globals[GetSession()] = globals{TheQuestionBank:=QuestionBankDevice}
#             then:
#                 Print("Globals Set Up")
#             else:
#                 Print("Globals Failed")
#         else:
#             Print("Globals Already Set Up")

#         RealTimeClock.Init()
        
#         # SetUpSpawns()

#         P1SpawnPad.SpawnedEvent.SubscribeAgent(Self.HandlePlayerRespawn, 0)
#         P2SpawnPad.SpawnedEvent.SubscribeAgent(Self.HandlePlayerRespawn, 1)
        
#         TrackerLifeTimeWins.SetTarget(9956432)

#         if:
#             PlayerOneBoard := GameBoards[0]
#             PlayerTwoBoard := GameBoards[1]
#             TheGlobals := Globals[GetSession()]
#         then:

#             Print("Initializing game boards")
#             P1Values := (1.0, vector3{X:=1.0, Y:=0.0, Z:=0.0}, TriggerRandomForKnockActivation)
#             spawn:
#                 PlayerOneBoard.Init(P1Values)

#             P2Values := (-1.0, vector3{X:=1.0, Y:=0.0, Z:=0.0}, TriggerRandomForKnockActivation)
#             spawn:
#                 PlayerTwoBoard.Init(P1Values)
#             Print("Game boards initialized")

#             Playspace := Self.GetPlayspace()

#             Playspace.PlayerAddedEvent().Subscribe(Self.HandlePlayerAdded)
#             Self.MainPlayerCounter.CountSucceedsEvent.Subscribe(Self.HandlePlayerCountSucceeds)

#             Self.Teleporter_1v1_Entrance.TeleportedEvent.Subscribe(Self.HandleTeleportedOutOf1v1)
#             Self.Teleporter_1v1_Entrance.EnterEvent.Subscribe(Self.HandleTeleportedInto1v1)


#             set Self.QuestionBankDevice.HUDResponseNo = Self.HUDResponseNo
#             set Self.QuestionBankDevice.HUDResponseYes = Self.HUDResponseYes
#             set Self.QuestionBankDevice.HUDResponseMaybe = Self.HUDResponseMaybe

#             Sleep(0.0)

#             spawn{ InitQuestionBank() }                    

#             if(HasOwner := PlayerOneBoard.Owner?):
#                 PlayerOneBoard.PaintPalette.Init(HasOwner, PlayerOneBoard)

#                 # HasOwner.GuessUI.GenerateScreenOptions()
#                 HasOwner.GuessUI.CreateSelectionScreenOverlays()
#                 HasOwner.GuessUI.CreateSelectionScreenOptions()


#                 # Print("AllSelectionScreenOptions P1: {HasOwner.GuessUI.AllSelectionScreenOptions.Length}")

#                 # for(Option:HasOwner.GuessUI.AllSelectionScreenOptions):
#                 #     Print("Option Subscribed")
#                 #     Option.Button.OnClick().SubscribeMessage(HasOwner.GuessUI.OnClickButtonStyle, Option.StyleIndex, Option.StyleTexture)

#                 set HasOwner.GuessUI.TurnTimer = TimerTurnTime
#             else:
#                 Print("No owner found, cannot initialize paint palette for player 1")

#             if(HasOwner := PlayerTwoBoard.Owner?):
#                 PlayerTwoBoard.PaintPalette.Init(HasOwner, PlayerTwoBoard)

#                 # HasOwner.GuessUI.GenerateScreenOptions()
#                 HasOwner.GuessUI.CreateSelectionScreenOverlays()
#                 HasOwner.GuessUI.CreateSelectionScreenOptions()

#                 # Print("AllSelectionScreenOptions P2: {HasOwner.GuessUI.AllSelectionScreenOptions.Length}")

#                 # for(Option:HasOwner.GuessUI.AllSelectionScreenOptions):
#                 #     Print("Option Subscribed")
#                 #     Option.Button.OnClick().SubscribeMessage(HasOwner.GuessUI.OnClickButtonStyle, Option.StyleIndex, Option.StyleTexture)

#                 set HasOwner.GuessUI.TurnTimer = TimerTurnTime
#             else:
#                 Print("No owner found, cannot initialize paint palette for player 2")

#             # persistant data
#             # if(Player1:=PlayerOneBoard.Owner?, Player2:=PlayerTwoBoard.Owner?):
                
#             #     ExistingStreak := CheckForExistingStreak(Player1.Player, Player2.Player)

#             #     if:
#             #         FriendStreak := ExistingStreak?
#             #         FriendStreak.CurrentStreak >= 1
#             #     then:
#             #         Print("Friend Streak Found: {FriendStreak.CurrentStreak}")
#             #         for(I := 0 .. FriendStreak.CurrentStreak):
#             #             Print("Awarding Friend Streak: {I}")
#             #             # FriendStreakAccolade.Award(Player1.Player) # awards to both players

#             #         NewP1StreakData := AddDayToFriendStreak(Player1.Player, FriendStreak)
#             #         UpdateFriendStreak(Player1.Player, FriendStreak, NewP1StreakData)
#             #         NewP2StreakData := AddDayToFriendStreak(Player2.Player, FriendStreak)
#             #         UpdateFriendStreak(Player2.Player, FriendStreak, NewP2StreakData)

#             #     else:
#             #         Print("No friend streak found - Creating new streak")
#             #         NewFriendStreak := CreateFriendStreak(Player1.Player, Player2.Player)

#             #         AddFriendStreak(Player1.Player, NewFriendStreak)
#             #         AddFriendStreak(Player2.Player, NewFriendStreak)


#             AdminKeypad.Init()
                
#             DailyReward.ClaimButton.InteractedWithEvent.Subscribe(HandleButtonDailyReward)
#             GameLoop(PlayerOneBoard, PlayerTwoBoard, TheGlobals)


#     HandleTeleportedOutOf1v1(Agent:agent):void=
#         if:
#             Player := player[Agent]
#             TheGlobals := Globals[GetSession()]
#             CPlayer := TheGlobals.CustomPlayerMap[Player]
#             GameBoard := CPlayer.GameBoard            
#             TheCurrentPlayer := TheGlobals.CurrentPlayer?
#         then:
#             if(TheCurrentPlayer = Player):
#                 # GameBoard.Loadout.GrantItem(Player)
#                 GameBoard.EnableInputs()
#                 CPlayer.GuessUI.UpdateReticle(UI.UI_Textures.T_Reticle_Default, vector2{X:=52.0, Y:=52.0})
#             else:
#                 Print("Not current player, not granting item")

#             set CPlayer.In1v1Area = false
#         else:
#             Print("Failed to handle teleport")
                

#     HandleTeleportedInto1v1(Agent:agent):void=
#         if:
#             Player := player[Agent]
#             TheGlobals := Globals[GetSession()]
#             CPlayer := TheGlobals.CustomPlayerMap[Player]
#             GameBoard := CPlayer.GameBoard
#             # TheCurrentPlayer := TheGlobals.CurrentPlayer?
#         then:
#             Print("Teleported into 1v1 area")
#             set CPlayer.In1v1Area = true

#             GameBoard.CustomInput.Cursor_InputTrigger_Zoom.Unregister(Player)
#             GameBoard.CustomInput.Cursor_InputTrigger_Fire.Unregister(Player)
#             GameBoard.CustomInput.InputTrigger_Undo.Unregister(Player)
#             GameBoard.CustomInput.InputTrigger_Ask_Question.Unregister(Player)

#             CPlayer.GuessUI.UpdateReticle(UI.UI_Textures.transparent_square, vector2{X:=52.0, Y:=52.0})
#         else:
#             Print("Failed to handle teleport")
           
#     HandlePlayerCountSucceeds():void=
#         Print("Player count succeeded", ?Duration:=10.0, ?Color:=Red)
#         set Self.HasFoundOnlyOnePlayer = true

#     ComparePlayerCountDelayed()<suspends>:void=
#         Sleep(0.5)
#         Self.MainPlayerCounter.CompareToTarget()

#     GameLoop(PlayerOneBoard:game_board, PlayerTwoBoard:game_board, TheGlobals:globals)<suspends>:void=

#         Print("Starting Game Loop, Waiting for players to join")
#         spawn{ Self.ComparePlayerCountDelayed() }
#         Self.MainPlayerCounter.CountFailsEvent.Await()

#         Print("found 2 players, starting game")

#         if:
#             CPlayerOne:= PlayerOneBoard.Owner?
#             CPlayerTwo:= PlayerTwoBoard.Owner?
#             PlayerOne := CPlayerOne.Player
#             PlayerTwo := CPlayerTwo.Player
#         then:
#             # Determine who goes first
#             set TheGlobals.CurrentTurn = GetRandomInt(0,1)

#             PlayerOneBoard.ItemRemover.Remove(CPlayerOne.Player)
#             PlayerTwoBoard.ItemRemover.Remove(CPlayerTwo.Player)

#             # Pick a random tile for each player
#             PlayerOneBoard.NewRandomTile()
#             Sleep(0.0)
#             PlayerTwoBoard.NewRandomTile()
#             Sleep(0.0)

#             # Set the current target indexes
#             Print("Setting Current Target Indexes {PlayerOneBoard.CurrentTargetIndex} --- {PlayerTwoBoard.CurrentTargetIndex}", ?Duration:=5.0, ?Color:=Red)
#             PlayerOneBoard.SetOpponentTargetIndex(PlayerTwoBoard.CurrentTargetIndex)
#             PlayerTwoBoard.SetOpponentTargetIndex(PlayerOneBoard.CurrentTargetIndex)
#             Print("Opponent Target Indexes Set {PlayerOneBoard.OpponentTargetIndex} --- {PlayerTwoBoard.OpponentTargetIndex}", ?Duration:=5.0, ?Color:=Red)
#             Print("Awaiting Votes")

#             CPlayerOne.GuessUI.InitStyleData()
#             CPlayerTwo.GuessUI.InitStyleData()

#             Sleep(0.0)
#             Print("Board 1: TargetIndex = {PlayerOneBoard.OpponentTargetIndex} --- OpponentTargetIndex = {PlayerOneBoard.OpponentTargetIndex}")
#             Print("Board 2: TargetIndex = {PlayerTwoBoard.OpponentTargetIndex} --- OpponentTargetIndex = {PlayerTwoBoard.OpponentTargetIndex}")

#             # HUDLoadTimerNumbers.Show() #load the timer numbers into a HUD so they don't flicker

#             set Self.HasFoundOnlyOnePlayer = false # Reset the flag before voting
#             # Wait for players to vote
#             TheVotes := race:
#                 block:
#                     Self.MainPlayerCounter.CountSucceedsEvent.Await() # Only one player left
#                     set Self.HasFoundOnlyOnePlayer = true
#                     (-9999,-9999) # arbitrary value to allow race to compile

#                 AwaitPlayerVotes(CPlayerOne, PlayerOne, CPlayerTwo, PlayerTwo) # Both players voted
#             # GetMaterialForCardStyle

#             if(HasFoundOnlyOnePlayer?):
#                 Print("Only one player found, moving player to lobby area")
#                 AllPlayers := GetPlayspace().GetPlayers()
#                 for(Player : AllPlayers):
#                     Player.Respawn(vector3{X:=1787.0,Y:=-279.0,Z:=2611.0}, IdentityRotation())
                
#                 HideEveryoneUI()
#                 return

#             Print("Votes Received: {TheVotes(0)} --- {TheVotes(1)}")
    
#             # save persistent data
#             # AddPreviousStyleChoice(PlayerOne, TheVotes(0))
#             # AddPreviousStyleChoice(PlayerTwo, TheVotes(1))

#             # Set the style for the game
#             set TheGlobals.CurrentStyleIndex = CompareResult(TheVotes(0), TheVotes(1))
#             set TheGlobals.CurrentStyle = GetCardStyleForIndex(TheGlobals.CurrentStyleIndex)

#             # Wait and show voting results
#             # Sleep(3.0)
            
#             # Hide the Voting UI
#             HideEveryoneUI()

#             # Initilize Question Bank with the chosen style
#             HandleChooseStyle(TheGlobals.CurrentStyleIndex)
#             GetAnalyticForStyleIndex(TheGlobals.CurrentStyleIndex, PlayerOne)
#             # GetSequenceForStyleIndex(TheGlobals.CurrentStyleIndex)

#             # Set the style for the game
#             PlayerOneBoard.SetStyle(TheGlobals.CurrentStyle)
#             PlayerOneBoard.YouTile.SetCard(TheGlobals.CurrentStyle, PlayerOneBoard.OpponentTargetIndex, PlayerOneBoard)
#             Sleep(0.0)
#             PlayerOneBoard.YouTile.Prop.SetMaterial(Cards.Materials.YouCards.MI_You_Card_Cover_P1, ?Index:=1)
#             PlayerOneBoard.YouTile.Peak(PlayerOne)
            
#             PlayerTwoBoard.SetStyle(TheGlobals.CurrentStyle)
#             PlayerTwoBoard.YouTile.SetCard(TheGlobals.CurrentStyle, PlayerTwoBoard.OpponentTargetIndex , PlayerTwoBoard)
#             Sleep(0.0)
#             PlayerTwoBoard.YouTile.Prop.SetMaterial(Cards.Materials.YouCards.MI_You_Card_Cover_P2, ?Index:=1)
#             PlayerTwoBoard.YouTile.Peak(PlayerTwo)

#             # Display Index Card if available for style
#             spawn{HandleIndexCard(TheGlobals, PlayerOneBoard, PlayerTwoBoard, CPlayerOne, CPlayerTwo)}

#             TimerTurnTime.Reset()
#             TimerTurnTime.SetActiveDuration(PLAYER_TURN_TIME + 2.0)
#             CPlayerOne.GuessUI.UpdateTimerTexture(CPlayerOne.GuessUI.TimerBGTexture, CPlayerOne.GuessUI.TimerTimeTexture)
#             CPlayerTwo.GuessUI.UpdateTimerTexture(CPlayerTwo.GuessUI.TimerBGTexture, CPlayerTwo.GuessUI.TimerTimeTexture)

#             Sleep(0.25)
  
#             # Flip over the cards
#             sync:
#                 PlayerOneBoard.FlipBoard(?Instant:=true)
#                 PlayerTwoBoard.FlipBoard(?Instant:=true)
            
#             set Self.GameInProgress = true

#             # start the game
#             # spawn { HandleTurns(PlayerOneBoard, PlayerTwoBoard, TheGlobals, CPlayerOne, CPlayerTwo) }

#             race:
#                 PlayerOneBoard.AwaitWin()
#                 PlayerTwoBoard.AwaitWin()
#                 HandleTurns(PlayerOneBoard, PlayerTwoBoard, TheGlobals, CPlayerOne, CPlayerTwo)


#             PlayerOneBoard.DisableInputs()
#             PlayerTwoBoard.DisableInputs()

            
#             set CPlayerOne.GuessUI.IsTimerActive = false
#             set CPlayerTwo.GuessUI.IsTimerActive = false
#             set CPlayerOne.LiveRayCastActive = false
#             set CPlayerTwo.LiveRayCastActive = false

            
#             Sleep(0.0)
#             PlayerOneBoard.ItemRemover.Remove(CPlayerOne.Player)
#             PlayerTwoBoard.ItemRemover.Remove(CPlayerTwo.Player)
#             PlayerOneBoard.ResetTileHighlight(CPlayerOne)
#             PlayerOneBoard.ResetButtonHighlights()
#             PlayerTwoBoard.ResetTileHighlight(CPlayerTwo)
#             PlayerTwoBoard.ResetButtonHighlights()

            


#             Sleep(0.1)
#             CPlayerOne.GuessUI.UpdateReticle(UI.UI_Textures.transparent_square, vector2{X:=52.0, Y:=52.0})
#             CPlayerTwo.GuessUI.UpdateReticle(UI.UI_Textures.transparent_square, vector2{X:=52.0, Y:=52.0})
#             Sleep(0.1)
            
#             # Reveal the opponent's cards
#             sync:
#                 PlayerTwoBoard.YouTile.Reveal()
#                 PlayerOneBoard.YouTile.Reveal()

            


#             Print("Someone won?!?!?!?!?!?")

#             if:
#                 TheWinner := TheGlobals.Winner?
#             then:
#                 Self.WinnerStopWatch.SetLapTime(TheWinner)
#                 Self.ScoreManager.SetScoreAward(1)
#                 Self.ScoreManager.Activate(TheWinner)
#                 Self.GoldGranter.GrantItem(TheWinner)

#                 Self.TrackerLifeTimeWins.Increment(TheWinner)

#                 RestartGame(PlayerOneBoard, PlayerTwoBoard, TheGlobals)
#             else:
#                 Print("No winner found")

#             Sleep(0.1)
#             CPlayerOne.GuessUI.UpdateReticle(UI.UI_Textures.T_Reticle_Default, vector2{X:=52.0, Y:=52.0})
#             CPlayerTwo.GuessUI.UpdateReticle(UI.UI_Textures.T_Reticle_Default, vector2{X:=52.0, Y:=52.0})
            
#             Print("Flipped the board")



#     RestartGame(PlayerOneBoard:game_board, PlayerTwoBoard:game_board, TheGlobals:globals)<suspends>:void=
#         set Self.GameInProgress = false
#         set Self.HasFoundOnlyOnePlayer = false
        
#         Sleep(0.0)
#         PlayerOneBoard.Reset()
#         PlayerTwoBoard.Reset()
#         Sleep(0.0)

#         GameLoop(PlayerOneBoard, PlayerTwoBoard, TheGlobals)


#     # HandleTurns(PlayerOneBoard:game_board, PlayerTwoBoard:game_board, TheGlobals:globals, CPlayerOne:custom_player, CPlayerTwo:custom_player)<suspends>:void=
#     #     Print("Handling Turns")
#     #     if(Self.GameInProgress?):
#     #         if(TheGlobals.CurrentTurn = 0):

#     #             set TheGlobals.CurrentPlayer = option{CPlayerOne.Player}

#     #             Print("Player One Turn")
#     #             PlayerOneBoard.EnableInputs()
#     #             PlayerTwoBoard.DisableInputs()

#     #             if(not CPlayerOne.In1v1Area?):
#     #                 PlayerOneBoard.Loadout.GrantItem(CPlayerOne.Player)

#     #             Sleep(0.0)

#     #             PlayerOneBoard.AwaitTurnPlayed(CPlayerOne)

#     #             set CPlayerOne.LiveRayCastActive = false
#     #             PlayerOneBoard.ResetTileHighlight(CPlayerOne)
#     #             PlayerOneBoard.ResetButtonHighlights()
#     #             Sleep(0.5)
#     #             CPlayerTwo.GuessUI.UpdateReticle(Textures.T_Reticle_Default)
#     #             CPlayerOne.GuessUI.UpdateReticle(Textures.T_Reticle_Default)


#     #             PlayerOneBoard.ItemRemover.Remove(CPlayerOne.Player)
#     #             set TheGlobals.CurrentTurn = 1
                
#     #         else:

#     #             set TheGlobals.CurrentPlayer = option{CPlayerTwo.Player}

#     #             Print("Player Two Turn")
#     #             PlayerOneBoard.DisableInputs()
#     #             PlayerTwoBoard.EnableInputs()

#     #             if(not CPlayerTwo.In1v1Area?):
#     #                 PlayerTwoBoard.Loadout.GrantItem(CPlayerTwo.Player)

#     #             Sleep(0.0)

#     #             PlayerTwoBoard.AwaitTurnPlayed(CPlayerTwo)
#     #             set CPlayerTwo.LiveRayCastActive = false
#     #             PlayerTwoBoard.ResetTileHighlight(CPlayerTwo)
#     #             PlayerTwoBoard.ResetButtonHighlights()
#     #             Sleep(0.5)
#     #             CPlayerOne.GuessUI.UpdateReticle(Textures.T_Reticle_Default)
#     #             CPlayerTwo.GuessUI.UpdateReticle(Textures.T_Reticle_Default)

#     #             PlayerTwoBoard.ItemRemover.Remove(CPlayerTwo.Player)
#     #             set TheGlobals.CurrentTurn = 0

#     #         Print("Turns Handled")

#     #         CPlayerOne.GuessUI.RemoveQuestionUI(CPlayerOne.Player)
#     #         CPlayerTwo.GuessUI.RemoveQuestionUI(CPlayerTwo.Player)

#     #         PlayerOneBoard.YesNoCine.GoToEndAndStop(CPlayerOne.Player)
#     #         PlayerTwoBoard.YesNoCine.GoToEndAndStop(CPlayerTwo.Player)

#     #         set CPlayerOne.GuessUI.IsTimerActive = false
#     #         set CPlayerTwo.GuessUI.IsTimerActive = false

#     #         TimerTurnTime.Reset()
#     #         TimerTurnTime.SetActiveDuration(PLAYER_TURN_TIME + 4.0)
#     #         CPlayerOne.GuessUI.UpdateTimerTexture(CPlayerOne.GuessUI.TimerBGTexture, CPlayerOne.GuessUI.TimerTimeTexture)
#     #         CPlayerTwo.GuessUI.UpdateTimerTexture(CPlayerTwo.GuessUI.TimerBGTexture, CPlayerTwo.GuessUI.TimerTimeTexture)

#     #         CPlayerOne.GuessUI.UpdateReticle(Textures.T_Reticle_Default)
#     #         CPlayerTwo.GuessUI.UpdateReticle(Textures.T_Reticle_Default)

#     #         Sleep(1.1)
#     #         HandleTurns(PlayerOneBoard, PlayerTwoBoard, TheGlobals, CPlayerOne, CPlayerTwo)

#     HandleTurns(PlayerOneBoard:game_board, PlayerTwoBoard:game_board, TheGlobals:globals, CPlayerOne:custom_player, CPlayerTwo:custom_player)<suspends>:void=
#         Print("Handling Turns")
#         loop:
#             set TheGlobals.OpponentAnswered = (false, false)
            
#             if(not Self.GameInProgress?):
#                 break
    
#             # Determine the current player and opponent based on the turn
#             CurrentPlayer := if(TheGlobals.CurrentTurn = 0) then CPlayerOne else CPlayerTwo
#             OpponentPlayer := if(TheGlobals.CurrentTurn = 0) then CPlayerTwo else CPlayerOne
#             CurrentBoard := if(TheGlobals.CurrentTurn = 0) then PlayerOneBoard else PlayerTwoBoard
#             OpponentBoard := if(TheGlobals.CurrentTurn = 0) then PlayerTwoBoard else PlayerOneBoard

#             set TheGlobals.CurrentPlayer = option{CurrentPlayer.Player}
    
#             # Enable current player's inputs and disable opponent's
#             if(not CurrentPlayer.In1v1Area?):
#                 CurrentBoard.EnableInputs()
                
#             OpponentBoard.DisableInputs()
    
#             # Grant item if not in 1v1 area
#             # if(not CurrentPlayer.In1v1Area?):
#             #     CurrentBoard.Loadout.GrantItem(CurrentPlayer.Player)
    
#             # Wait for the current player to play their turn
#             Print("Player {TheGlobals.CurrentTurn + 1} Turn")

#             race:
#                 CurrentBoard.AwaitTurnPlayed(CurrentPlayer)

#                 block:
#                     # Live Raycast for Card Highlighting
#                     set CurrentPlayer.LiveRayCastActive = true
#                     Print("Enabling live raycast")
#                     CurrentBoard.LiveRayCast(CurrentPlayer.Player)
            
#             CurrentBoard.DisableInputs()
            
#             if(not CurrentPlayer.In1v1Area?):
#                 set CurrentPlayer.LiveRayCastActive = false

#             CurrentBoard.ResetTileHighlight(CurrentPlayer)
#             CurrentBoard.ResetButtonHighlights()
    
#             # Update UI elements
#             Sleep(0.25)
#             OpponentPlayer.GuessUI.UpdateReticle(UI.UI_Textures.T_Reticle_Default, vector2{X:=52.0, Y:=52.0})
#             CurrentPlayer.GuessUI.UpdateReticle(UI.UI_Textures.T_Reticle_Default, vector2{X:=52.0, Y:=52.0})

#             if(not CurrentPlayer.In1v1Area?):
#                 CurrentBoard.ItemRemover.Remove(CurrentPlayer.Player)
    
#             # Swap turns
#             set TheGlobals.CurrentTurn = 1 - TheGlobals.CurrentTurn
    
#             # Cleanup and prepare for next turn
#             PrepareForNextTurn(CurrentPlayer, OpponentPlayer, CurrentBoard, OpponentBoard, TheGlobals)

#             Sleep(0.1)
    
#     HandleTurnEnd(CPlayer:custom_player, PlayerBoard:game_board, TheGlobals:globals)<suspends>:void=
#         spawn . CPlayer.GuessUI.RemoveQuestionUI(CPlayer.Player)
#         # PlayerBoard.YesNoCine.GoToEndAndStop(CPlayer.Player)
#         PlayerBoard.YesNoCamera.RemoveFrom(CPlayer.Player)
#         set CPlayer.GuessUI.IsTimerActive = false
#         set PlayerBoard.HoldingInput = false
#         set PlayerBoard.AwaitingResponse = false
#         TimerTurnTime.Reset()
#         TimerTurnTime.SetActiveDuration(PLAYER_TURN_TIME + 4.0)
#         CPlayer.GuessUI.UpdateTimerTexture(CPlayer.GuessUI.TimerBGTexture, CPlayer.GuessUI.TimerTimeTexture)
    
#     PrepareForNextTurn(CurrentPlayer:custom_player, OpponentPlayer:custom_player, CurrentBoard:game_board, OpponentBoard:game_board, TheGlobals:globals)<suspends>:void=
#         HandleTurnEnd(CurrentPlayer, CurrentBoard, TheGlobals)
#         HandleTurnEnd(OpponentPlayer, OpponentBoard, TheGlobals)
#         Sleep(0.1)
#         CurrentPlayer.GuessUI.UpdateReticle(UI.UI_Textures.T_Reticle_Default, vector2{X:=52.0, Y:=52.0})
#         OpponentPlayer.GuessUI.UpdateReticle(UI.UI_Textures.T_Reticle_Default, vector2{X:=52.0, Y:=52.0})

        
    

#     CompareResult(P1Vote:int, P2Vote:int):int=

#         if(P1Vote = -1):
#             return P2Vote
#         else if(P2Vote = -1):
#             return P1Vote
#         else if:
#             P1Vote = P2Vote
#         then:
#             Print("Agree")
#             return P1Vote
#         else:
#             Print("Disagree, Randomizing")
#             return CoinFlipVotes(P1Vote, P2Vote)


#     CoinFlipVotes(P1Vote:int, P2Vote:int):int=
#         RandInt := GetRandomInt(0,1)
#         case (RandInt):
#             0 => 
#                 Print("Heads")
#                 return P1Vote
#             1 => 
#                 Print("Tails")
#                 return P2Vote
#             _ =>
#                 Print("Coin Flip Failed, falling back to P1")
#                 return P1Vote


#     AwaitPlayerVotes(CPlayerOne:custom_player, PlayerOne:player, CPlayerTwo:custom_player, PlayerTwo:player)<suspends>:tuple(int,int)=
#         Print("Awaiting player votes")
#         var P1Vote : int = -1
#         var P2Vote : int = -1

#         TimerTurnTime.Reset()
#         TimerTurnTime.SetActiveDuration(90.0)
        
#         Print("Attempt To Init Style Selection UI")
#         sync:
#             CPlayerOne.GuessUI.InitStyleSelectionUI(PlayerOne, CPlayerOne)
#             CPlayerTwo.GuessUI.InitStyleSelectionUI(PlayerTwo, CPlayerTwo)
#         Print("Style Selection UI Initialized")

#         # spawn{CPlayerOne.GuessUI.AwaitStyleSelection(PlayerOne, CPlayerOne, VotedSound)}
#         # spawn{CPlayerTwo.GuessUI.AwaitStyleSelection(PlayerTwo, CPlayerTwo, VotedSound)}

#         # Wait for both players to vote or timeout
        
#         TimerTurnTime.Start()
#         TimerTurnTime.SetActiveDuration(PLAYER_TURN_TIME)

#         Votes := sync:
#             CPlayerOne.GuessUI.AwaitStyleSelection(PlayerOne, CPlayerOne, VotedSound)
#             CPlayerTwo.GuessUI.AwaitStyleSelection(PlayerTwo, CPlayerTwo, VotedSound)

#         Print("Both players voted or timed out")

#         if:
#             Votes(0) = Votes(1)
#         then:
#             Print("Both players voted the same")
#             return (Votes(0), Votes(1))


#         # Both players voted differently and lots of time left

#         QuickVoteTime := 25.0

#         if:
#             TimerTurnTime.GetActiveDuration() > QuickVoteTime
#         then:
#             TimerTurnTime.SetActiveDuration(QuickVoteTime)
#             TimerTurnTime.Start()


#             race:
#                 # if votes match, return
#                 loop:
#                     if:
#                         P1:=CPlayerOne.GetVote[]
#                         P2:=CPlayerTwo.GetVote[]
#                         P1 = P2
#                     then:
#                         set P1Vote = P1
#                         set P2Vote = P2
#                         Print("Both votes match")
#                         return (P1Vote, P2Vote)
#                     Sleep(0.5)

#                 Sleep(QuickVoteTime)

#         Print("Both players voted differently and time is up")
#         if:
#             P1:=CPlayerOne.GetVote[]
#             P2:=CPlayerTwo.GetVote[]
#         then:
#             set P1Vote = P1
#             set P2Vote = P2

#         set CPlayerOne.GuessUI.IsTimerActive = false
#         set CPlayerTwo.GuessUI.IsTimerActive = false


#         return (P1Vote, P2Vote)

#         ## THIS ALL WORKS
#         # # Both players voted differently and lots of time left
#         # if:
#         #     CPlayerOne.GuessUI.Timer.ToSeconds() > 10.0
#         # then:
#         #     # CPlayerOne.GuessUI.Timer.SetSecond(10)
#         #     CPlayerOne.GuessUI.Timer.Cancel()
#         #     CPlayerTwo.GuessUI.Timer.SetSecond(10)


#         # CPlayerOne.GuessUI.Timer.AwaitTime(10)

#         # loop:
#         #     if:
#         #         CPlayerOne.HasVoted?
#         #         CPlayerTwo.HasVoted?
#         #         P1VoteInt := CPlayerOne.GetVote[]
#         #         P2VoteInt := CPlayerTwo.GetVote[]
#         #     then:
#         #         set P1Vote = P1VoteInt
#         #         set P2Vote = P2VoteInt
#         #         Print("Both players voted")
#         #         break
#         #     else:
#         #         Sleep(0.1)
#         # Print("Both players voted")
#         # return (P1Vote, P2Vote)

#     HandleChooseStyle(Index:int):void=
#         Print("Choose Card Style: {Index}")
#         case(Index):
#             1 => 
#                 # spawn{SwapToCurrentStyle(Index)}
#                 spawn{QuestionBankDevice.Init(card_style.Classic)}
#             2 => 
#                 # spawn{SwapToCurrentStyle(Index)}
#                 spawn{QuestionBankDevice.Init(card_style.Icon)}
#             3 => 
#                 # spawn{SwapToCurrentStyle(Index)}
#                 spawn{QuestionBankDevice.Init(card_style.Animal)}
#             4 => 
#                 # spawn{SwapToCurrentStyle(Index)}
#                 spawn{QuestionBankDevice.Init(card_style.Utilities)}
#             5 => 
#                 # spawn{SwapToCurrentStyle(Index)}
#                 spawn{QuestionBankDevice.Init(card_style.Marvel)}
#             6 => 
#                 # spawn{SwapToCurrentStyle(Index)}
#                 spawn{QuestionBankDevice.Init(card_style.POI)}
#             7 => 
#                 # spawn{SwapToCurrentStyle(Index)}
#                 spawn{QuestionBankDevice.Init(card_style.Legends)}
#             8 => 
#                 # spawn{SwapToCurrentStyle(Index)}
#                 spawn{QuestionBankDevice.Init(card_style.Food)}
#             9 => 
#                 # spawn{SwapToCurrentStyle(Index)}
#                 spawn{QuestionBankDevice.Init(card_style.Star)}

#             10 => 
#                 # spawn{SwapToCurrentStyle(Index)}
#                 # spawn{QuestionBankDevice.Init(card_style.Spooky)}
#                 Print("Spooky not implemented yet, using FN")
#                 # spawn{SwapToCurrentStyle(Index)}
#                 spawn{QuestionBankDevice.Init(card_style.FN)}
#             11 =>
#                 # spawn{SwapToCurrentStyle(Index)}
#                 spawn{QuestionBankDevice.Init(card_style.OG)}
#             12 =>
#                 # spawn{SwapToCurrentStyle(Index)}
#                 spawn{QuestionBankDevice.Init(card_style.Anime)}

#             13 =>
#                 # spawn{SwapToCurrentStyle(Index)}
#                 spawn{QuestionBankDevice.Init(card_style.Winter)}
#             14 =>
#                 # spawn{SwapToCurrentStyle(Index)}
#                 spawn{QuestionBankDevice.Init(card_style.Weapons)}

#             15 =>
#                 # spawn{SwapToCurrentStyle(Index)}
#                 spawn{QuestionBankDevice.Init(card_style.Pickaxes)}
#             16 =>
#                 # spawn{SwapToCurrentStyle(Index)}
#                 spawn{QuestionBankDevice.Init(card_style.Vehicles)}
#             17 =>
#                 # spawn{SwapToCurrentStyle(Index)}
#                 spawn{QuestionBankDevice.Init(card_style.MoviesTV)}
#             18 =>
#                 # spawn{SwapToCurrentStyle(Index)}
#                 spawn{QuestionBankDevice.Init(card_style.Blocks)}
#             19 =>
#                 # spawn{SwapToCurrentStyle(Index)}
#                 spawn{QuestionBankDevice.Init(card_style.Chapter5)}
#             20 =>
#                 # spawn{SwapToCurrentStyle(Index)}
#                 spawn{QuestionBankDevice.Init(card_style.Lockers)}
#             21 =>
#                 # spawn{SwapToCurrentStyle(Index)}
#                 spawn{QuestionBankDevice.Init(card_style.Backblings)}
#             22 =>
#                 # spawn{SwapToCurrentStyle(Index)}
#                 spawn{QuestionBankDevice.Init(card_style.Emotes)}
#             23 =>
#                 # spawn{SwapToCurrentStyle(Index)}
#                 spawn{QuestionBankDevice.Init(card_style.Marvel1)}
#             24 =>
#                 # spawn{SwapToCurrentStyle(Index)}
#                 spawn{QuestionBankDevice.Init(card_style.Marvel2)}
#             _ =>
#                 Print("Invalid Style Index: {Index}")
#                 # spawn{SwapToCurrentStyle(0)}
#                 spawn{QuestionBankDevice.Init(card_style.FN)}

#     var CurrentStyle : card_style = card_style.FN


#     HandleIndexCard(TheGlobals:globals, PlayerOneBoard:game_board, PlayerTwoBoard:game_board, CPlayerOne:custom_player, CPlayerTwo:custom_player)<suspends>:void=
#         set PlayerOneBoard.StyleHasIndexCard = false
#         set PlayerTwoBoard.StyleHasIndexCard = false

#         if:
#             PlayerOneBoard.IndexCard.Card.IsValid[]
#             PlayerTwoBoard.IndexCard.Card.IsValid[]
#             P1CardRotation := PlayerOneBoard.IndexCard.Card.GetTransform().Rotation
#             P2CardRotation := PlayerTwoBoard.IndexCard.Card.GetTransform().Rotation
            
#             PlayerOneBoard.IndexCard.Card.TeleportTo[vector3{X:=-5344.0,Y:=1205.0,Z:=2471.0}, P1CardRotation]
#             PlayerTwoBoard.IndexCard.Card.TeleportTo[vector3{X:=-1256.0,Y:=-1194.0,Z:=2471.0}, P2CardRotation]
#         then:
#             Print("Index Card's Reset")

#         # Utilities Style = 4
#         # Food Style = 8
#         # Vehicles Style = 16
#         if:            
#             TheGlobals.CurrentStyleIndex = 4 or TheGlobals.CurrentStyleIndex = 8 or TheGlobals.CurrentStyleIndex = 16
#         then:
#             set PlayerOneBoard.StyleHasIndexCard = true
#             set PlayerTwoBoard.StyleHasIndexCard = true

#             IndexCardMaterial := GetIndexCardMaterialForStyleIndex(TheGlobals.CurrentStyleIndex)

#             if:
#                 PlayerOneBoard.IndexCard.Card.IsValid[]
#                 PlayerTwoBoard.IndexCard.Card.IsValid[]

#                 P1CardRotation := PlayerOneBoard.IndexCard.Card.GetTransform().Rotation
#                 P2CardRotation := PlayerTwoBoard.IndexCard.Card.GetTransform().Rotation

#             then:
#                 PlayerOneBoard.IndexCard.Card.SetMaterial(IndexCardMaterial, ?Index:=1)
#                 PlayerTwoBoard.IndexCard.Card.SetMaterial(IndexCardMaterial, ?Index:=1)

#                 PlayerOneBoard.IndexCard.Card.MoveTo(PlayerOneBoard.IndexCard.Card.GetTransform().Translation + vector3{X:=-128.0}, P1CardRotation, 0.15)
#                 PlayerTwoBoard.IndexCard.Card.MoveTo(PlayerTwoBoard.IndexCard.Card.GetTransform().Translation + vector3{X:=128.0}, P2CardRotation, 0.15)
#                 Print("Index Card's Shown")
#                 Sleep(1.0)
#                 PlayerOneBoard.IndexCard.SlideOut(1.0)
#                 PlayerTwoBoard.IndexCard.SlideOut(-1.0)
#             else:
#                 Print("Index Card's Failed to move")



#     HideEveryoneUI():void=
#         Print("HideEveryoneUI")
#         if:
#             TheGlobals := Globals[GetSession()]
#         then:
#             for (Player : TheGlobals.AllPlayers):
#                 if:
#                     CPlayer := TheGlobals.CustomPlayerMap[Player]
#                     PlayerUI := GetPlayerUI[Player]
#                 then:

#                     # Fully remove to disable input
#                     if(MaybeVoteUI := CPlayer.GuessUI.MaybeVoteUI?):
#                         PlayerUI.RemoveWidget(MaybeVoteUI)

#                     if(MaybeStyleSelectionUI := CPlayer.GuessUI.MaybeStyleSelectionUI?):
#                         PlayerUI.RemoveWidget(MaybeStyleSelectionUI)

#                     # Hide we'll use later
#                     if(MaybeYourTurnUI := CPlayer.GuessUI.MaybeYourTurnUI?):
#                         MaybeYourTurnUI.SetVisibility(widget_visibility.Hidden)

#                     if(MaybeQuestionUI := CPlayer.GuessUI.MaybeQuestionUI?):
#                         MaybeQuestionUI.SetVisibility(widget_visibility.Hidden)

#                     if(MaybeOtherPlayersTurnUI := CPlayer.GuessUI.MaybeOtherPlayersTurnUI?):
#                         MaybeOtherPlayersTurnUI.SetVisibility(widget_visibility.Hidden)

#                     if(MaybeShootToGuessUI := CPlayer.GuessUI.MaybeShootToGuessUI?):
#                         MaybeShootToGuessUI.SetVisibility(widget_visibility.Hidden)

#                     if(MaybeCorrectIncorrectUI := CPlayer.GuessUI.MaybeCorrectIncorrectUI?):
#                         MaybeCorrectIncorrectUI.SetVisibility(widget_visibility.Hidden)


#     HandlePlayerAdded(Agent:agent):void=
#         Print("Player Added")
#         if:
#             Self.HasFoundOnlyOnePlayer?
#         then:
#             Print("Someone joined, restart round")
#             spawn { DelayedEndGame(Agent) }
#         else:
#             Print("someone joined, but dont restart round")


#     DelayedEndGame(Agent:agent)<suspends>:void=
#         Sleep(5.0)
#         EndGame.Activate(Agent)


#     HandlePlayerRespawn(Agent:agent, Index:int):void=
#         Print("Player Spawning")

#         if:
#             Player := player[Agent]
#             TheGlobals := Globals[GetSession()]
#             not TheGlobals.CustomPlayerMap[Player]
#         then:
#             Print("Player Respawn - First Spawn")
#             spawn{PlayerJoined(Player, Index)}
#         else:
#             Print("Player Respawn - Already Init")


#     PlayerJoined(InPlayer:player, Index:int)<suspends>:void=
#         Print("Player Joined first time!")
#         if:
#             FortCharacter := InPlayer.GetFortCharacter[]
#             FortCharacter.IsActive[]
#             PlayerBoard := GameBoards[Index]
#             ThePlayer := custom_player{Player:=InPlayer, GameBoard:=PlayerBoard}
#             TheGlobals := Globals[GetSession()]
#             set TheGlobals.CustomPlayerMap[InPlayer] = ThePlayer
#         then:
#             set TheGlobals.AllPlayers += array{InPlayer}
#             ThePlayer.ChangeState(player_state.Active)
#             ThePlayer.Init(InPlayer)
#             PlayerBoard.SetOwner(ThePlayer)

#             # Create Persistence Data
#             if:
#                 CheckForPlayerSavedData[InPlayer] # auto creates if not found
#                 SourceData := PlayerSavedDataMap[InPlayer]
#             then:
#                 Print("Player Data Created/Found for Player")

#                 PlayerScore := ScoreManager.GetCurrentScore(InPlayer)

#                 ConvertOldScoreData(InPlayer, PlayerScore)
    
#                 Print("Player Score Converted")

#                 if:
#                     SourceData.DataVersion = 1
                
#                     set PlayerSavedDataMap[InPlayer] = player_saved_data:
#                         MakePlayerSavedData<constructor>(SourceData)
#                         DataVersion := 2
#                         FirstDayPlayed := GetCurrentDateTime()
#                         LastDayPlayed := GetCurrentDateTime()
#                         TotalDaysPlayed := 0
#                         TotalMinutesPlayed := 0
#                         NewPlayersPlayedWith := 0
#                         LastGuessCorrect := false
#                         FriendStreaks := array{}
#                         CardsKnocked := 0
#                         PreviousStyleChoices := array{}
#                 then: 
#                     Print("Player Data Updated to Version 2")
#                 else:
#                     Print("Player Data Version 2 Found")
                    
#                 CheckForConsecutiveDaysPlayed(InPlayer, GetCurrentDateTime())


#             else:
#                 Print("No player data found for Player")

#         else:
#             Print("Player Joined - Something failed..")


#     HandlePlayerRemoved(InAgent:agent):void=
#         Print("Player Has Left Playspace")
        

#     InitQuestionBank()<suspends>:void=
#         QuestionBankDevice.InitFNQuestions()
#         Sleep(0.01)
#         # QuestionBankDevice.InitClassicQuestions()
#         # Sleep(0.01)
#         QuestionBankDevice.InitAnimalQuestions()
#         Sleep(0.01)
#         QuestionBankDevice.InitIconQuestions()
#         Sleep(0.01)
#         QuestionBankDevice.InitUtilitiesQuestions()
#         Sleep(0.01)
#         # QuestionBankDevice.InitMarvelQuestions()
#         # Sleep(0.01)
#         # QuestionBankDevice.InitPOIQuestions()
#         # Sleep(0.01)
#         QuestionBankDevice.InitLegendsQuestions()
#         Sleep(0.01)
#         QuestionBankDevice.InitFoodQuestions()
#         Sleep(0.01)
#         QuestionBankDevice.InitStarQuestions()
#         Sleep(0.01)
#         QuestionBankDevice.InitSpookyQuestions()
#         Sleep(0.01)
#         QuestionBankDevice.InitOGQuestions()
#         Sleep(0.01)
#         QuestionBankDevice.InitAnimeQuestions()
#         Sleep(0.01)
#         QuestionBankDevice.InitWinterQuestions()
#         Sleep(0.01)
#         QuestionBankDevice.InitWeaponsQuestions()
#         Sleep(0.01)
#         QuestionBankDevice.InitPickaxeQuestions()
#         Sleep(0.01)
#         QuestionBankDevice.InitVehicleQuestions()
#         Sleep(0.01)
#         QuestionBankDevice.InitMoviesTVQuestions()
#         Sleep(0.01)
#         QuestionBankDevice.InitChapter5Questions()
#         Sleep(0.01)
#         QuestionBankDevice.InitLockersQuestions()
#         Sleep(0.01)
#         QuestionBankDevice.InitBackblingQuestions()
#         Sleep(0.01)
#         QuestionBankDevice.InitEmoteQuestions()
#         Sleep(0.01)
#         QuestionBankDevice.InitMarvel1Questions()
#         Sleep(0.01)
#         QuestionBankDevice.InitMarvel2Questions()


#     GetAnalyticForStyleIndex(Index:int, Player:player):void=
#         if:
#             Analytics := Analytics_Styles[Index]
#         then:
#             Analytics.Submit(Player)
#             Print("Style Analytics Submitted")
#         else:
#             Print("Style Analytics Failed to Submit")

#     # GetSequenceForStyleIndex(Index:int):void=
#     #     if:
#     #         Sequence := CardStyleCines[Index]
#     #     then:
#     #         Sequence.Play()
#     #         Print("Playing Sequencer for Board / Card Style")
#     #     else:
#     #         Print("Failed to play Sequencer for Board / Card Style")