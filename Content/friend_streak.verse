using { /Verse.org/Simulation }
using { /Verse.org/Random }

friend_streak := class<final><persistable>:
    ID : int = -1
    CurrentStreak : int = 0
    HighestStreak : int = 0
    TimesPlayed : int = 0
    LastPlayedDate : time_and_date = time_and_date{}
    CorrectGuessMap : [int]int = map{} # first day played in seconds since epoch, wins per player
    IncorrectGuessMap : [int]int = map{} # first day played in seconds since epoch, losses per player


MakeFriendStreak<constructor>(Src:friend_streak)<transacts> := friend_streak:
    ID := Src.ID
    CurrentStreak := Src.CurrentStreak
    HighestStreak := Src.HighestStreak
    TimesPlayed := Src.TimesPlayed
    LastPlayedDate := Src.LastPlayedDate
    CorrectGuessMap := Src.CorrectGuessMap
    IncorrectGuessMap := Src.IncorrectGuessMap


CreateFriendStreak(Agent1:agent, Agent2:agent)<transacts>:friend_streak=
    var NewStreak : friend_streak = friend_streak:
        ID := GenerateStreakID(Agent1, Agent2)
        CurrentStreak := 0
        HighestStreak := 0
        TimesPlayed := 0
        LastPlayedDate := GetCurrentDateTime()
        CorrectGuessMap := map{}
        IncorrectGuessMap := map{}

    CreateCorrectGuessMap(Agent1, Agent2, NewStreak)
    CreateIncorrectGuessMap(Agent1, Agent2, NewStreak)

    return NewStreak


# UpdateFriendStreakCorrectGuesses(Agent:agent, Streak:friend_streak, NewMap:[int]int)<transacts>:void=
#     if:
#         CheckForPlayerSavedData[Agent]
#         Player := player[Agent]
#         SourceData := PlayerSavedDataMap[Player]

#         var SomeFriendStreaks : []friend_streak = for(FriendStreak : SourceData.FriendStreaks, not SameStreakID[FriendStreak, Streak]):
#             FriendStreak

#         set SomeFriendStreaks += array{Streak}

#         set PlayerSavedDataMap[Player] = player_saved_data:
#             MakePlayerSavedData<constructor>(SourceData)
#             FriendStreaks := SomeFriendStreaks
            

#     then:
#         Print("Friend streak updated successfully: {SomeFriendStreaks.Length + 1}")
#     else:
#         Print("Player does not have saved data")


ReplaceCorrectGuessMap(Agent:agent, Streak:friend_streak, NewMap:[int]int)<transacts>:friend_streak=
    if:
        CheckForPlayerSavedData[Agent]
        Player := player[Agent]
        SourceData := PlayerSavedDataMap[Player]

        NewStreakData := friend_streak:
            MakeFriendStreak<constructor>(Streak)
            ID := Streak.ID
            CurrentStreak := Streak.CurrentStreak
            HighestStreak := Streak.HighestStreak
            TimesPlayed := Streak.TimesPlayed
            LastPlayedDate := Streak.LastPlayedDate
            CorrectGuessMap := NewMap
            IncorrectGuessMap := Streak.IncorrectGuessMap
    then:
        return NewStreakData

    else:
        Print("Player does not have saved data")
        return Streak

ReplaceIncorrectGuessMap(Agent:agent, Streak:friend_streak, NewMap:[int]int)<transacts>:friend_streak=
    if:
        CheckForPlayerSavedData[Agent]
        Player := player[Agent]
        SourceData := PlayerSavedDataMap[Player]

        NewStreakData := friend_streak:
            MakeFriendStreak<constructor>(Streak)
            ID := Streak.ID
            CurrentStreak := Streak.CurrentStreak
            HighestStreak := Streak.HighestStreak
            TimesPlayed := Streak.TimesPlayed
            LastPlayedDate := Streak.LastPlayedDate
            CorrectGuessMap := Streak.CorrectGuessMap
            IncorrectGuessMap := NewMap
    then:
        return NewStreakData

    else:
        Print("Player does not have saved data")
        return Streak
            

UpdateFriendStreakIncorrectGuesses(Agent:agent, Streak:friend_streak, IncorrectGuesses:int)<transacts>:void=
    if:
        CheckForPlayerSavedData[Agent]
        Player := player[Agent]
        SourceData := PlayerSavedDataMap[Player]

        NewStreakArr := for(FriendStreak : SourceData.FriendStreaks, not SameStreakID[FriendStreak, Streak]):
            FriendStreak

        set PlayerSavedDataMap[Player] = player_saved_data:
            MakePlayerSavedData<constructor>(SourceData)
            FriendStreaks := NewStreakArr + array{Streak}
    then:
        Print("Friend streak updated successfully: {NewStreakArr.Length + 1}")
    else:
        Print("Player does not have saved data")

UpdateFriendStreak(Agent:agent, OldStreak:friend_streak, NewStreak:friend_streak)<transacts>:void=

    if:
        CheckForPlayerSavedData[Agent]
        Player := player[Agent]
        SourceData := PlayerSavedDataMap[Player]

        NewStreakArr := for(FriendStreak : SourceData.FriendStreaks, not SameStreakID[FriendStreak, OldStreak]):
            FriendStreak

        set PlayerSavedDataMap[Player] = player_saved_data:
            MakePlayerSavedData<constructor>(SourceData)
            FriendStreaks := NewStreakArr + array{NewStreak}
    then:
        Print("Friend streak updated successfully: {NewStreakArr.Length + 1}")
    else:
        Print("Player does not have saved data")
        
AddFriendStreak(Agent:agent, NewStreak:friend_streak)<transacts>:void=
    if:
        CheckForPlayerSavedData[Agent]
        Player := player[Agent]
        SourceData := PlayerSavedDataMap[Player]
        FriendStreaks := SourceData.FriendStreaks + array{NewStreak}

        set PlayerSavedDataMap[Player] = player_saved_data:
            MakePlayerSavedData<constructor>(SourceData)
            FriendStreaks := FriendStreaks
    then:
        Print("Friend streak added successfully: {FriendStreaks.Length}")
    else:
        Print("Player does not have saved data")

CreateCorrectGuessMap(Agent1:agent, Agent2:agent, StreakData:friend_streak)<transacts>:void=
    if:
        CheckForPlayerSavedData[Agent1]
        Player1 := player[Agent1]
        SourceData1 := PlayerSavedDataMap[Player1]
        FirstDayPlayed1 := SourceData1.FirstDayPlayed
        Seconds1 := CalculatedTotalSecondsSinceEpoch(FirstDayPlayed1)

        CheckForPlayerSavedData[Agent2]
        Player2 := player[Agent2]
        SourceData2 := PlayerSavedDataMap[Player2]
        FirstDayPlayed2 := SourceData2.FirstDayPlayed
        Seconds2 := CalculatedTotalSecondsSinceEpoch(FirstDayPlayed2)

        var NewMap : [int]int = map{}

        set NewMap[Seconds1] = 0
        set NewMap[Seconds2] = 0

        ReplaceCorrectGuessMap(Agent1, StreakData, NewMap)
        ReplaceCorrectGuessMap(Agent2, StreakData, NewMap)
    else:
        Print("Failed to create correct guess map.")


AddCorrectGuessToStreakData(Agent:agent, Opponent:agent, StreakData:friend_streak)<transacts>:void=
    if:
        CheckForPlayerSavedData[Agent]
        Player := player[Agent]
        SourceData1 := PlayerSavedDataMap[Player]
        FirstDay1 := SourceData1.FirstDayPlayed
        Seconds1 := CalculatedTotalSecondsSinceEpoch(FirstDay1)

        Player2 := player[Opponent]
        SourceData2 := PlayerSavedDataMap[Player2]
        FirstDay2 := SourceData2.FirstDayPlayed
        Seconds2 := CalculatedTotalSecondsSinceEpoch(FirstDay2)

        var NewMap : [int]int = map{}

        set NewMap[Seconds1] = 0
        set NewMap[Seconds2] = 0

        NewStreakData := friend_streak:
            MakeFriendStreak<constructor>(StreakData)
            ID := StreakData.ID
            CurrentStreak := StreakData.CurrentStreak
            HighestStreak := StreakData.HighestStreak
            TimesPlayed := StreakData.TimesPlayed
            LastPlayedDate := StreakData.LastPlayedDate
            CorrectGuessMap := NewMap
            IncorrectGuessMap := StreakData.IncorrectGuessMap

        set PlayerSavedDataMap[Player] = player_saved_data:
            MakePlayerSavedData<constructor>(SourceData1)
            FriendStreaks := SourceData1.FriendStreaks + array{NewStreakData}

        set PlayerSavedDataMap[Player2] = player_saved_data:
            MakePlayerSavedData<constructor>(SourceData2)
            FriendStreaks := SourceData2.FriendStreaks + array{NewStreakData}

        
CreateIncorrectGuessMap(Agent1:agent, Agent2:agent, StreakData:friend_streak)<transacts>:void=
    if:
        CheckForPlayerSavedData[Agent1]
        Player1 := player[Agent1]
        SourceData1 := PlayerSavedDataMap[Player1]
        FirstDayPlayed1 := SourceData1.FirstDayPlayed
        Seconds1 := CalculatedTotalSecondsSinceEpoch(FirstDayPlayed1)

        CheckForPlayerSavedData[Agent2]
        Player2 := player[Agent2]
        SourceData2 := PlayerSavedDataMap[Player2]
        FirstDayPlayed2 := SourceData2.FirstDayPlayed
        Seconds2 := CalculatedTotalSecondsSinceEpoch(FirstDayPlayed2)

        var NewMap : [int]int = map{}

        set NewMap[Seconds1] = 0
        set NewMap[Seconds2] = 0

        ReplaceIncorrectGuessMap(Agent1, StreakData, NewMap)
        ReplaceIncorrectGuessMap(Agent2, StreakData, NewMap)
    else:
        Print("Failed to create incorrect guess map.")


CheckForExistingStreak(Agent1:agent, Agent2:agent)<transacts>:?friend_streak=
    var Player1Streaks : []friend_streak = array{}
    var Player2Streaks : []friend_streak = array{}

    var MaybeExistingStreak : ?friend_streak = false

    if:
        CheckForPlayerSavedData[Agent1]
        Player := player[Agent1]
        SourceData := PlayerSavedDataMap[Player]
    then:
        Print("Player 1 has saved data")
        set Player1Streaks = SourceData.FriendStreaks
    else:
        Print("Player 1 does not have saved data")

    if:
        CheckForPlayerSavedData[Agent2]
        Player := player[Agent2]
        SourceData := PlayerSavedDataMap[Player]
    then:
        Print("Player 2 has saved data")
        set Player2Streaks = SourceData.FriendStreaks
    else:
        Print("Player 2 does not have saved data")

        
    if:
        ExistingStreakID := StreakIDFoundInBoth[Player1Streaks, Player2Streaks]
        ExistingStreakID > 0
        ExistingStreak := GetExistingStreak[Agent1, ExistingStreakID]
    then:
        Print("Existing streak found.")
        set MaybeExistingStreak = option{ExistingStreak}


    return MaybeExistingStreak


SameStreakID(Streak1:friend_streak, Streak2:friend_streak)<transacts><decides>:void=
    Streak1.ID = Streak2.ID

StreakIDFoundInBoth(Streaks1:[]friend_streak, Streaks2:[]friend_streak)<transacts><decides>:int=
    var HasExistingStreak : logic = false
    var ExistingStreakID : int = -1

    loop:
        for(Streak1 : Streaks1):
            loop:
                for(Streak2 : Streaks2):
                    if(Streak1.ID = Streak2.ID):
                        set HasExistingStreak = true
                        set ExistingStreakID = Streak1.ID
                        break # Exit the inner loop as we found a match
                if(HasExistingStreak?):
                    break # Exit the outer loop as we found a match

        if(HasExistingStreak?):
            Print("Common streak ID found.")
            break
        else:
            Print("No common streak ID found.")
            break

    return ExistingStreakID



        
GetExistingStreak(Agent:agent, StreakID:int)<transacts><decides>:friend_streak=
    var PlayerStreaks : []friend_streak = array{}

    if:
        CheckForPlayerSavedData[Agent]
        Player := player[Agent]
        SourceData := PlayerSavedDataMap[Player]
    then:
        Print("Player has saved data")
        set PlayerStreaks = SourceData.FriendStreaks
    else:
        Print("Player does not have saved data")

    FoundStreaks := for(Streak : PlayerStreaks, Streak.ID = StreakID):
        Streak

    FoundStreaks[0]

GetWinsFromStreakData(Streak:friend_streak)<transacts>:tuple(int,int)=
    var P1Wins : int = 0
    var P2Wins : int = 0

    for (Value -> Day : Streak.CorrectGuessMap):
        if(Value = 0):
            set P1Wins = P1Wins + 1
        else:
            set P2Wins = P2Wins + 1

    (P1Wins, P2Wins)

# CheckForConsecutiveCorrectGuesses(Agent:agent):void=
#     if:
#         CheckForPlayerSavedData[Agent]
#         Player := player[Agent]
#         SourceData := PlayerSavedDataMap[Player]
#     then:
#         CurrentConsecutiveCorrectGuesses := SourceData.ConsecutiveCorrectGuesses
#         NewConsecutiveCorrectGuesses := CurrentConsecutiveCorrectGuesses + 1

#         # If the player guessed correctly, increment the current consecutive correct guesses
#         if(SourceData.LastGuessCorrect?):
#             SetConsecutiveCorrectGuesses(Player, NewConsecutiveCorrectGuesses)

#         else:
#             SetConsecutiveCorrectGuesses(Player, 1)

#         # check record
#         if(NewConsecutiveCorrectGuesses > SourceData.ConsecutiveCorrectGuessesRecord):
#             SetConsecutiveCorrectGuessesRecord(Player, NewConsecutiveCorrectGuesses)    

CheckForFriendStreak(Agent1:agent, Agent2:agent)<transacts>:?friend_streak=
    var Player1Streaks : []friend_streak = array{}
    var Player2Streaks : []friend_streak = array{}

    var MaybeFriendStreak : ?friend_streak = false

    if:
        CheckForPlayerSavedData[Agent1]
        Player := player[Agent1]
        SourceData := PlayerSavedDataMap[Player]
    then:
        Print("Player 1 has saved data")
        set Player1Streaks = SourceData.FriendStreaks
    else:
        Print("Player 1 does not have saved data")

    if:
        CheckForPlayerSavedData[Agent2]
        Player := player[Agent2]
        SourceData := PlayerSavedDataMap[Player]
    then:
        Print("Player 2 has saved data")
        set Player2Streaks = SourceData.FriendStreaks
    else:
        Print("Player 2 does not have saved data")

        
    if:
        ExistingStreakID := StreakIDFoundInBoth[Player1Streaks, Player2Streaks]
        ExistingStreakID > 0
        ExistingStreak := GetExistingStreak[Agent1, ExistingStreakID]
    then:
        Print("Existing streak found.")
        set MaybeFriendStreak = option{ExistingStreak} 

    return MaybeFriendStreak


AddDayToFriendStreak(Agent:agent, Streak:friend_streak)<transacts>:friend_streak=
    if:
        CheckForPlayerSavedData[Agent]
        Player := player[Agent]
        SourceData := PlayerSavedDataMap[Player]

        var UpdatedStreak : friend_streak = friend_streak:
            ID := Streak.ID
            CurrentStreak := Streak.CurrentStreak + 1
            HighestStreak := Streak.HighestStreak
            TimesPlayed := Streak.TimesPlayed + 1
            LastPlayedDate := GetCurrentDateTime()
            CorrectGuessMap := Streak.CorrectGuessMap
            IncorrectGuessMap := Streak.IncorrectGuessMap
           
    then:

        if:
            Streak.CurrentStreak + 1 > Streak.HighestStreak
        then:
            # Update the highest streak 
            NewlyUpdatedStreak := SetHighestFriendStreak(Agent, Streak, Streak.CurrentStreak + 1)

            

        return UpdatedStreak

    return Streak
        

SetHighestFriendStreak(Agent:agent, Streak:friend_streak, NewHighestStreak:int)<transacts>:friend_streak=
    if:
        CheckForPlayerSavedData[Agent]
        Player := player[Agent]
        SourceData := PlayerSavedDataMap[Player]

        UpdatedStreak := friend_streak:
            ID := Streak.ID
            CurrentStreak := Streak.CurrentStreak
            HighestStreak := NewHighestStreak
            TimesPlayed := Streak.TimesPlayed
            LastPlayedDate := GetCurrentDateTime()
            CorrectGuessMap := Streak.CorrectGuessMap
            IncorrectGuessMap := Streak.IncorrectGuessMap
            
    then:
        return UpdatedStreak

    return Streak

        

GenerateStreakID(Agent1:agent, Agent2:agent)<transacts>:int=
    var UniqueID:int = 0

    CurrentDate := GetCurrentDateTime()

    var Player1FirstDayPlayedSeconds : int = 0
    var Player2FirstDayPlayedSeconds : int = 0

    var Player1CorrectGuesses : int = 0
    var Player2CorrectGuesses : int = 0

    var Player1DaysPlayed : int = 0
    var Player2DaysPlayed : int = 0    

    # p1 saved data
    if:
        CheckForPlayerSavedData[Agent1]
        Player := player[Agent1]
        SourceData := PlayerSavedDataMap[Player]
    then:
        set Player1FirstDayPlayedSeconds = CalculatedTotalSecondsSinceEpoch(SourceData.FirstDayPlayed)
        set Player1CorrectGuesses = SourceData.CorrectGuesses
        set Player1DaysPlayed = SourceData.TotalDaysPlayed
    else:
        set Player1FirstDayPlayedSeconds = CalculatedTotalSecondsSinceEpoch(CurrentDate)
        set Player1CorrectGuesses = GetRandomInt(0, 1000)
        set Player1DaysPlayed = GetRandomInt(0, 1000)

    # p2 saved data
    if:
        CheckForPlayerSavedData[Agent2]
        Player := player[Agent2]
        SourceData := PlayerSavedDataMap[Player]
    then:
        set Player2FirstDayPlayedSeconds = CalculatedTotalSecondsSinceEpoch(SourceData.FirstDayPlayed)
        set Player2CorrectGuesses = SourceData.CorrectGuesses
        set Player2DaysPlayed = SourceData.TotalDaysPlayed
    else:
        set Player2FirstDayPlayedSeconds = CalculatedTotalSecondsSinceEpoch(CurrentDate)
        set Player2CorrectGuesses = GetRandomInt(0, 1000)
        set Player2DaysPlayed = GetRandomInt(0, 1000)


    # algorthim to generate unique id based on the above variables
    StringValues:= "{Player1FirstDayPlayedSeconds}{Player2FirstDayPlayedSeconds}{Player1CorrectGuesses}{Player2CorrectGuesses}{Player1DaysPlayed}{Player2DaysPlayed}"
    set UniqueID = SimpleStringHash(StringValues)
    Print("String Values: {StringValues}")
    Print("Unique ID: {UniqueID}")

    return UniqueID


SimpleStringHash(InputString: string)<transacts>:int=
    var Hash: int = 0
    var PrimeMultiplier: int = 31
    var MaxInt: int = 2147483647  # To handle overflow for a 32-bit signed integer

    for (Index -> Char : InputString):
        CharValue := CharCode(Char)
        set Hash = (Hash * PrimeMultiplier + CharValue)
        if(NewHash :=  Mod[Hash, MaxInt]):
            set Hash = NewHash

    return Hash

# Helper function to get the integer value of a character
CharCode(Char: char)<transacts>:int =
    CharCodes : [char]int = map{
        '\t' => 9,
        '\n' => 10,
        '\r' => 13,
        ' ' => 32,
        '!' => 33,
        '"' => 34,
        '#' => 35,
        '$' => 36,
        '%' => 37,
        '&' => 38,
        '\'' => 39,
        '(' => 40,
        ')' => 41,
        '*' => 42,
        '+' => 43,
        ',' => 44,
        '-' => 45,
        '.' => 46,
        '/' => 47,
        '0' => 48,
        '1' => 49,
        '2' => 50,
        '3' => 51,
        '4' => 52,
        '5' => 53,
        '6' => 54,
        '7' => 55,
        '8' => 56,
        '9' => 57,
        ':' => 58,
        ';' => 59,
        '<' => 60,
        '=' => 61,
        '>' => 62,
        '?' => 63,
        '@' => 64,
        'A' => 65,
        'B' => 66,
        'C' => 67,
        'D' => 68,
        'E' => 69,
        'F' => 70,
        'G' => 71,
        'H' => 72,
        'I' => 73,
        'J' => 74,
        'K' => 75,
        'L' => 76,
        'M' => 77,
        'N' => 78,
        'O' => 79,
        'P' => 80,
        'Q' => 81,
        'R' => 82,
        'S' => 83,
        'T' => 84,
        'U' => 85,
        'V' => 86,
        'W' => 87,
        'X' => 88,
        'Y' => 89,
        'Z' => 90,
        '[' => 91,
        '\' => 92,
        ']' => 93,
        '^' => 94,
        '_' => 95,
        '`' => 96,
        'a' => 97,
        'b' => 98,
        'c' => 99,
        'd' => 100,
        'e' => 101,
        'f' => 102,
        'g' => 103,
        'h' => 104,
        'i' => 105,
        'j' => 106,
        'k' => 107,
        'l' => 108,
        'm' => 109,
        'n' => 110,
        'o' => 111,
        'p' => 112,
        'q' => 113,
        'r' => 114,
        's' => 115,
        't' => 116,
        'u' => 117,
        'v' => 118,
        'w' => 119,
        'x' => 120,
        'y' => 121,
        'z' => 122,
        '{' => 123,
        '|' => 124,
        '}' => 125,
        '~' => 126
    }

    if(TheInt := CharCodes[Char]):
        return TheInt
    else:
        return -1