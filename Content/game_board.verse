using { /Fortnite.com/Devices }
using { /Verse.org/Simulation }
using { /UnrealEngine.com/Temporary/Diagnostics }
using { /UnrealEngine.com/Temporary/SpatialMath }
# using { /Fortnite.com/Devices/CreativeAnimation }
# using { /Fortnite.com/Devices/Animation }
using { /Verse.org/Random }
using { /Fortnite.com/Characters }
using { /Verse.org/Colors }
using { /Verse.org/Colors/NamedColors }
# using { /UnrealEngine.com/Temporary/UI }
# using { /Fortnite.com/UI }
using { /Verse.org/Assets }
# using { /Fortnite.com/FortPlayerUtilities }


game_board := class<concrete>():

    Logger:log := log{Channel:=log_device}
    Draw:debug_draw := debug_draw{Channel:=debug_drawing}

    var OnEndTurn <public> : event() = event(){}

    var OnCorrectGuess <public> : event() = event(){}
    var OnIncorrectGuess <public> : event() = event(){}
    
    var ManualCharacterSelected<public>: event(int) = event(int){}
    var InputCancelable: ?cancelable = false
    var IsManualCharacterSelection: logic = false

    var Owner : ?custom_player = false
    var NumTiles : int = 24
    var TilesKnocked : int = 0
    var AllKnocked : logic = false

    var KnockedTiles : []card_tile = array{}

    var <public> CurrentTargetIndex : int = 0
    var <public> OpponentTargetIndex : int = 0

    var TargetKnocked : logic = false

    var MakingFinalGuess : logic = false

    var AwaitingResponse : logic = false
    var QuestionBankOpen : logic = false

    var HoldingInput : logic = false

    var BoardLocalForward : float = 0.0

    @editable BoardProp : creative_prop = creative_prop{}
    @editable YouStandProp : creative_prop = creative_prop{}
    @editable PlatformProp : creative_prop = creative_prop{}
    @editable Teleporter : teleporter_device = teleporter_device{}
    
    @editable Rows : []game_board_row = array{}

    @editable MakeGuessButton : button_device = button_device{}
    @editable EndTurnButton : button_device = button_device{}
    @editable AskQuestionButton : button_device = button_device{}
    @editable UndoButton : button_device = button_device{}

    @editable FloorMakeGuessButton_Prop : creative_prop = creative_prop{}
    @editable FloorEndTurnButton_Prop : creative_prop = creative_prop{}
    @editable FloorAskQuestionButton_Prop : creative_prop = creative_prop{}
    @editable FloorUndoButton_Prop : creative_prop = creative_prop{}

    @editable MakeGuessButton_Prop : creative_prop = creative_prop{}
    @editable EndTurnButton_Prop : creative_prop = creative_prop{}
    @editable AskQuestionButton_Prop : creative_prop = creative_prop{}
    @editable UndoButton_Prop : creative_prop = creative_prop{}

    @editable HatProp : creative_prop = creative_prop{}


    @editable DialogConfirmFinalGuess : popup_dialog_device = popup_dialog_device{}
    @editable DialogStreamerSettings : popup_dialog_device = popup_dialog_device{}

    @editable YesNoCamera : gameplay_camera_fixed_point_device = gameplay_camera_fixed_point_device{}
    @editable CameraZoom1 : gameplay_camera_orbit_device = gameplay_camera_orbit_device{}
    @editable CameraZoom2 : gameplay_camera_orbit_device = gameplay_camera_orbit_device{}

    @editable YouTile : you_tile = you_tile{}

    @editable Beacons : []beacon_device = array{}

    @editable HUD_YourTurn : hud_message_device = hud_message_device{}
    @editable HUD_OtherPlayersTurn : hud_message_device = hud_message_device{}
    @editable HUD_YourTurn_Centered : hud_message_device = hud_message_device{}
    @editable HUD_OtherPlayersTurn_Centered : hud_message_device = hud_message_device{}

    @editable HUD_AwaitingResponse : hud_message_device = hud_message_device{}

    @editable ItemRemover : item_remover_device = item_remover_device{}

    @editable CardKnockSound : audio_player_device = audio_player_device{}
    @editable CardHighlightSound : audio_player_device = audio_player_device{}
    @editable ButtonPressSound : audio_player_device = audio_player_device{}


    @editable WinnerXP : accolades_device = accolades_device{}
    @editable WinnerOpponentXP : accolades_device = accolades_device{}
    @editable CardKnockXP : accolades_device = accolades_device{}

    @editable CustomInput : custom_input_device = custom_input_device{}

    @editable IndexCard : index_card = index_card{}
    var StyleHasIndexCard : logic = false

    @editable PaintPalette : paint_palette = paint_palette{}

    @editable NewPlayerAccolade : accolades_device = accolades_device{}

    var TriggerRandomForKnockActivation : trigger_device = trigger_device{}
    var LSTurnTime : cinematic_sequence_device = cinematic_sequence_device{}

    # var PlayerCardHighlightMaterial : Materials.BoardDesigner.CardBacks.MI_Cardback_Highlight_material = Materials.BoardDesigner.CardBacks.MI_Cardback_Highlight_material{}
    var PlayerCardHighlightMaterial : ?Materials.BoardDesigner.CardBacks.MI_Cardback_Highlight_material = false
    
    var SavedFrameColor : material = Materials.BoardDesigner.CardBacks.MI_Cardback_Yellow

    @editable BoardLocker : board_locker_ui = board_locker_ui{}

    @editable StickerSlots : []board_sticker_slot = array{}
    # @editable AddStickerCam : gameplay_camera_fixed_point_device = gameplay_camera_fixed_point_device{}
    var DefaultStickerPropAsset : creative_prop_asset = DefaultCreativePropAsset
    
    @editable TargetArea : target_area = target_area{}



    InitNew(TeamIndex:int, FlipDirection:float, Agent:agent, TurnTimer:timer_device, aTriggerRandomForKnockActivation:trigger_device, InputTriggerHackCamera:input_trigger_device, LevelSequenceTurnTime:cinematic_sequence_device, DefaultStickerAsset:creative_prop_asset)<suspends>:void=
        Logger.Print("Initializing game board and player ------ NEW")
        profile("Profile InitNew: AssignTileIndexies"):
            spawn . AssignTileIndexies()

        spawn . TargetArea.Init(PlatformProp)
        
            
        if:
            Player := player[Agent]
            NewCustomPlayer := custom_player{TeamIndex:=TeamIndex, Player:=Player, GameBoard:=Self}
            TheGlobals := Globals[GetSession()]
            set TheGlobals.CustomPlayerMap[Player] = NewCustomPlayer
        then:

            profile("Profile InitNew: First Stuff"):
                SetOwner(NewCustomPlayer)
                set TheGlobals.AllPlayers += array{Player}
                NewCustomPlayer.ChangeState(player_state.Active)
                NewCustomPlayer.Init(Player)

                set NewCustomPlayer.GuessUI.TurnTimer = TurnTimer
                set TriggerRandomForKnockActivation = aTriggerRandomForKnockActivation
                set LSTurnTime = LevelSequenceTurnTime
                set DefaultStickerPropAsset = DefaultStickerAsset
                # set SuctionDart = aSuctionDart

                set BoardLocalForward = FlipDirection
                LocalForward := BoardProp.GetTransform().Rotation.GetLocalForward()
                
                PaintPalette.Init(NewCustomPlayer, Self)

                for(Row : Self.Rows):
                    for(Tile : Row.Tiles):
                        Tile.Init(FlipDirection, LocalForward)

            profile("Profile InitNew : Subscriptions"):
                UndoButton.InteractedWithEvent.Subscribe(HandleUndoButton)
                EndTurnButton.InteractedWithEvent.Subscribe(HandleEndTurnButton)
                AskQuestionButton.InteractedWithEvent.Subscribe(HandleAskQuestionButton)
                MakeGuessButton.InteractedWithEvent.Subscribe(HandleMakeGuessButton)
                CustomInput.Cursor_InputTrigger_Zoom.PressedEvent.Subscribe(HandleCursorInputTriggerZoom)
                CustomInput.InputTrigger_Ask_Question.PressedEvent.Subscribe(HandleCursorInputTriggerAskQuestion)
                DialogConfirmFinalGuess.RespondingButtonEvent.Subscribe(HandleDialogFinalGuess)
                DialogStreamerSettings.RespondingButtonEvent.Subscribe(HandleDialogStreamerSettings)

            profile("Profile InitNew : Persistent Data"):

                #check for PersistentData
                if:
                    CheckForPlayerSavedData[Player] # auto creates new player data if none exists
                    SourceData := PlayerSavedDataMap[Player]
                then:
                    Logger.Print("Player Data Found/Created")

                    
                    HighlightCardColorTuple := SourceData.CardHighlightColor
                    Print("Highlight Card Color Tuple: {HighlightCardColorTuple(0)}, {HighlightCardColorTuple(1)}, {HighlightCardColorTuple(2)}")

                    if(not PlayerCardHighlightMaterial?):
                        set PlayerCardHighlightMaterial = option{Materials.BoardDesigner.CardBacks.MI_Cardback_Highlight_material{}}
                    
                    if(NewMaterial := PlayerCardHighlightMaterial?):                    
                        set NewMaterial.BaseColor = NEW_ASSETS.Textures.T_CardBack_C 
                        set NewMaterial.BaseMaskedTexture = NEW_ASSETS.Textures.T_CardBack_ORM
                        set NewMaterial.BaseNormal = NEW_ASSETS.Textures.T_CardBack_N
                        set NewMaterial.ColorDesaturation = 1.0
                        set NewMaterial.BaseEmissive = _UI.UI_Textures.ui_square_blank
                        set NewMaterial.ColorTint = MakeColorFromSRGB(HighlightCardColorTuple(0), HighlightCardColorTuple(1), HighlightCardColorTuple(2))
                        set NewMaterial.EmissiveTint = MakeColorFromSRGB(HighlightCardColorTuple(0), HighlightCardColorTuple(1), HighlightCardColorTuple(2))
                        set NewMaterial.EmissiveStrength = 1.0
                        
                        
                    

                    set SavedFrameColor = NewCustomPlayer.GameBoard.PaintPalette.GetMaterialFromIndexFrame(NewCustomPlayer.GameBoard.PaintPalette.FrameColorIndex)

                    if(SourceData.IsBetaTester?):
                        Logger.Print("Beta Tester Found - Assigning Camera Hack")
                        InputTriggerHackCamera.Register(Player)
                    else:
                        Logger.Print("Beta Tester Not Found")


                    #convert v1 data to v2
                    if:
                        SourceData.DataVersion = 1

                        set PlayerSavedDataMap[Player] = player_saved_data:
                            MakePlayerSavedData<constructor>(SourceData)
                            DataVersion := 2
                            FirstDayPlayed := GetCurrentDateTime()
                            LastDayPlayed := GetCurrentDateTime()
                            TotalDaysPlayed := 0
                            TotalMinutesPlayed := 0
                            NewPlayersPlayedWith := 0
                            LastGuessCorrect := false
                            FriendStreaks := array{}
                            CardsKnocked := 0
                            PreviousStyleChoices := array{}
                    then:
                        Logger.Print("Converted Player Data to Version 2")
                    else:
                        Logger.Print("Player Already Converted to Version 2")

                    if:
                        SourceData.DataVersion = 2

                        set PlayerSavedDataMap[Player] = player_saved_data:
                            MakePlayerSavedData<constructor>(SourceData)
                            DataVersion := 3
                            CardHighlightColor := (0.0,0.302997,1.0)
                            LastSessionTime := GetCurrentDateTime()
                            StreamerModeHideCharacter := false
                            StreamerModeVerticalUI := false
                            ManuallyChooseCharacter := false
                    then:
                        Logger.Print("Converted Player Data to Version 3")
                    else:
                        Logger.Print("Player Already Converted to Version 3")


                    if:
                        SourceData.DataVersion = 3

                        set PlayerSavedDataMap[Player] = player_saved_data:
                            MakePlayerSavedData<constructor>(SourceData)
                            DataVersion := 4
                            CompletedWinter2024Quest := false
                    then:
                        Logger.Print("Converted Player Data to Version 4")
                    else:
                        Logger.Print("Player Already Converted to Version 4")

                    if:
                        SourceData.DataVersion = 4

                        set PlayerSavedDataMap[Player] = player_saved_data:
                            MakePlayerSavedData<constructor>(SourceData)
                            DataVersion := 5
                            TotalFlyingTargetsHit := 0
                    then:
                        Logger.Print("Converted Player Data to Version 5")
                    else:
                        Logger.Print("Player Already Converted to Version 5")

                    if:
                        SourceData.DataVersion = 5

                        set PlayerSavedDataMap[Player] = player_saved_data:
                            MakePlayerSavedData<constructor>(SourceData)
                            DataVersion := 6
                            BeenAssignedPointer := false
                            CurrentPointer := -1
                    then:
                        Logger.Print("Converted Player Data to Version 6")
                    else:
                        Logger.Print("Player Already Converted to Version 6")

                    if:
                        SourceData.DataVersion = 6

                        set PlayerSavedDataMap[Player] = player_saved_data:
                            MakePlayerSavedData<constructor>(SourceData)
                            DataVersion := 7
                            UnlockedStickers := array{}
                            CurrentStickers := array{}
                    then:
                        Logger.Print("Converted Player Data to Version 7")
                    else:
                        Logger.Print("Player Already Converted to Version 7")
                        

                if:
                    SourceData := PlayerSavedDataMap[Player]
                then:
                    if:
                        SourceData.StreamerModeHideCharacter?
                    then:
                        spawn . YouTile.HideCharacterCard(Player)

                    if:
                        SourceData.ManuallyChooseCharacter?
                    then:
                        set IsManualCharacterSelection = true


            profile("Profile InitNew : Check For Consecutive Days"):
                CheckForConsecutiveDaysPlayed(Player, GetCurrentDateTime())

            profile("Profile InitNew : Start Raycast"):
                set NewCustomPlayer.LiveRayCastActive = true
                FireSubscript := CustomInput.Cursor_InputTrigger_Fire.PressedEvent.Subscribe(HandleCursorInputTriggerFire)
                set CustomInput.FireSubscription = option{FireSubscript}
                CustomInput.Cursor_InputTrigger_Fire.Register(Agent)
                spawn . LiveRayCast(Agent)

                spawn . MonitorWithinGameplayArea(TheGlobals, NewCustomPlayer, Agent)


            #load winter quest sticker if completed
            if(HasCompletedWinter2024Quest[Agent]):
                Logger.Print("Player has completed Winter 2024 Quest")
                if:
                    StickerID := TheGlobals.GetStickerID["Winter2024"]
                    not AlreadyUnlockedSticker[Agent, StickerID]
                then:
                    UnlockSticker(Agent, StickerID)

                    if:
                        AvailStickerSlot := GetAvailableStickerSlot[]
                    then:
                        # AddStickerCam.AddTo(Agent)
                        # NewStickerResult := SpawnProp(DefaultStickerPropAsset, vector3{}, rotation{})
                        # Winter2024StickerMaterial := Materials.Stickers.MI_Sticker_Base_Inst_material{}
                        # set Winter2024StickerMaterial.Sticker_Texture = Materials.Stickers.winter_2024
                        # set Winter2024StickerMaterial.UseSnowBase = 1.0
                        # if:
                        #     TheSticker := NewStickerResult(0)?
                        # then:
                        #     TheSticker.SetMaterial(Winter2024StickerMaterial)
                        #     spawn . AddSticker(Agent, NewCustomPlayer, AvailStickerSlot, StickerID, TheSticker)
                        SaveCurrentSticker(Agent, StickerID, AvailStickerSlot.GetTransform())
                    # else:
                    #     Logger.Print("No available sticker slot")


            # Place Saved Stickers
            if:
                SavedPlacedStickers := GetCurrentStickers[Agent]
            then:

                for:
                    aSticker : SavedPlacedStickers
                    NewStickerResult := SpawnProp(DefaultStickerPropAsset, vector3{}, rotation{})
                    StickerMaterial := GetStickerMaterial(aSticker.StickerID)
                    TheSticker := NewStickerResult(0)?
                    AvailStickerSlot := GetAvailableStickerSlot[]
                do:
                    TheSticker.SetMaterial(StickerMaterial)
                    AddSticker(Agent, NewCustomPlayer, AvailStickerSlot, aSticker.StickerID, TheSticker)



        else:
            Logger.Print("No Player Data Found")


    HackView(OpponentPlayer:agent, InputTrigger:input_trigger_device)<suspends>:void=
        YouTile.Peak(OpponentPlayer)
        if(FortChar := OpponentPlayer.GetFortCharacter[]):
            PlayerRotation := FortChar.GetTransform().Rotation
            YesNoCamera.AddTo(OpponentPlayer)
            InputTrigger.ReleasedEvent.Await()
            YesNoCamera.RemoveFrom(OpponentPlayer)
            PlayerLocation := FortChar.GetTransform().Translation
            Sleep(0.1)
            if:
                FortChar.TeleportTo[PlayerLocation + vector3{Z:=10.0}, PlayerRotation]
            then:
                Logger.Print("player flipped")
            else:
                Logger.Print("Failed to flip player")


    # Init(FlipDirection:float, LocalForward:vector3, aTriggerRandomForKnockActivation:trigger_device)<suspends>:void=
    #     Logger.Print("Initializing game board")

    #     set BoardLocalForward = FlipDirection

    #     set TriggerRandomForKnockActivation = aTriggerRandomForKnockActivation

    #     Self.UndoButton.InteractedWithEvent.Subscribe(HandleUndoButton)
    #       Self.EndTurnButton.InteractedWithEvent.Subscribe(HandleEndTurnButton)
    #     Self.AskQuestionButton.InteractedWithEvent.Subscribe(HandleAskQuestionButton)
    #     MakeGuessButton.InteractedWithEvent.Subscribe(Self.HandleMakeGuessButton)

    #     # Custom Input Triggers
    #     CustomInput.Cursor_InputTrigger_Zoom.PressedEvent.Subscribe(HandleCursorInputTriggerZoom)
    #     CustomInput.InputTrigger_Ask_Question.PressedEvent.Subscribe(HandleCursorInputTriggerAskQuestion)
        
    #     Self.DialogConfirmFinalGuess.RespondingButtonEvent.Subscribe(HandleDialogFinalGuess)

    #     Self.AssignTileIndexies()

    #     for(Row : Self.Rows):
    #         for(Tile : Row.Tiles):
    #             Tile.Init(FlipDirection, LocalForward)



    LiveRayCast(Agent:agent)<suspends>:void=
        defer:
            Logger.Print("defer - Live raycast ended")            
            ResetButtonHighlights()

            if:
                CPlayer := Owner?
                Player := CPlayer.Player
            then:
                ResetTileHighlight(CPlayer)
                set CPlayer.LiveRayCastActive = false
                CPlayer.GuessUI.UpdateReticle(CPlayer.GuessUI.GetDefaultReticleTexture(Player, Self))

        Sleep(0.25)



        if:
            CPlayer := Owner?
            FortChar := Agent.GetFortCharacter[]
            TheGlobals := Globals[GetSession()]
            GameBoard := CPlayer.GameBoard
            CheckForPlayerSavedData[Agent]
            SourceData := PlayerSavedDataMap[Agent]
        then:

            # race:
            #     block:
            #         OnEndTurn.Await()
            #         Logger.Print("Live raycast ended - OnEndTurn signal")
            loop:

                if(CPlayer.AwayFromGameplayArea?):
                    Print("Paused Live Raycast for Player: {CPlayer.TeamIndex}")

                    if:
                        GameBoard.CustomInput.Cursor_RayCast.Top.TeleportTo[vector3{}, rotation{}]
                    then:
                        Print("Teleported Top of Raycast to Safe Location")
                    # GameBoard.CustomInput.Cursor_RayCast.Bottom.TeleportTo[vector3{}, rotation{}]



                    AwaitReturnToGameplayArea(CPlayer)

                # Logger.Print("Live raycast ping for Player: {CPlayer.TeamIndex}")
                if(CPlayer.LiveRayCastActive = false):
                    # Logger.Print("live raycast - set to false for Player: {CPlayer.TeamIndex}")
                    return

                if(CPlayer.In1v1Area?):
                    # Logger.Print("live raycast - in 1v1 area for Player: {CPlayer.TeamIndex}")
                    return

                set CPlayer.LiveRayCastActive = true
                ViewRotation := FortChar.GetViewRotation()
                ViewForward := ViewRotation.GetLocalForward()
                RayEnd := FortChar.GetViewLocation() + (ViewForward * 8000.0)
                RayStart := FortChar.GetViewLocation() + (ViewForward * 200.0)

                HitLocation := GameBoard.CustomInput.Cursor_RayCast.CastFrom(RayStart, ViewRotation)

                if:
                    HitLocation.Translation = vector3{}
                    YawPitchRoll := HitLocation.Rotation.GetYawPitchRollDegrees()
                    YawPitchRoll[0] = 0.0
                    YawPitchRoll[1] = 0.0
                    YawPitchRoll[2] = 0.0
                then:
                    # failed if empty transform
                    Logger.Print("Live raycast failed - Transform NULL")
                    return


                if:
                    LookingAtCard := GetNearestCard[HitLocation, GameBoard]
                    HighlightMaterial := PlayerCardHighlightMaterial?
                then:
                    ResetTileHighlight(CPlayer)

                    # LookingAtCard.Prop.SetMaterial(HighlightMaterial, ?Index:=2)
                    LookingAtCard.Prop.SetMaterial(Materials.BoardDesigner.CardBacks.MI_Cardback_Highlight, ?Index:=2)
                    # CardHighlightSound.Play(Player)
                    if(not HoldingInput?):
                        if(MakingFinalGuess?):
                            CPlayer.GuessUI.UpdateReticle(CPlayer.GuessUI.FinalGuessHoverReticleTexture)
                        else:
                            CPlayer.GuessUI.UpdateReticle(CPlayer.GuessUI.OpenHandReticleTexture)

                else:
                    # Print("No card found")
                    ResetTileHighlight(CPlayer)

                    DefaultButtonSize := 60.0
                    if(not HoldingInput?):

                        if(FortChar.IsLookingAtSphere[UndoButton_Prop.GetTransform().Translation, DefaultButtonSize * UndoButton_Prop.GetTransform().Scale.X]):
                            ResetButtonHighlights()
                            UndoButton_Prop.SetMaterial(NEW_ASSETS.Materials.MI_Button_Highlight, ?Index:=0)
                            # CardHighlightSound.Play(Player)
                            CPlayer.GuessUI.UpdateReticle(CPlayer.GuessUI.OpenHandReticleTexture)
                        else if(IsMyTurn[CPlayer], FortChar.IsLookingAtSphere[EndTurnButton_Prop.GetTransform().Translation, DefaultButtonSize * EndTurnButton_Prop.GetTransform().Scale.X]):
                            ResetButtonHighlights()
                            EndTurnButton_Prop.SetMaterial(NEW_ASSETS.Materials.MI_Button_Highlight, ?Index:=0)
                            # CardHighlightSound.Play(Player)
                            CPlayer.GuessUI.UpdateReticle(CPlayer.GuessUI.OpenHandReticleTexture)
                        else if(IsMyTurn[CPlayer], FortChar.IsLookingAtSphere[AskQuestionButton_Prop.GetTransform().Translation, DefaultButtonSize * AskQuestionButton_Prop.GetTransform().Scale.X]):
                            ResetButtonHighlights()
                            AskQuestionButton_Prop.SetMaterial(NEW_ASSETS.Materials.MI_Button_Highlight, ?Index:=0)
                            # CardHighlightSound.Play(Player)
                            CPlayer.GuessUI.UpdateReticle(CPlayer.GuessUI.OpenHandReticleTexture)
                        else if(IsMyTurn[CPlayer], FortChar.IsLookingAtSphere[MakeGuessButton_Prop.GetTransform().Translation, DefaultButtonSize * MakeGuessButton_Prop.GetTransform().Scale.X]):
                            ResetButtonHighlights()
                            MakeGuessButton_Prop.SetMaterial(NEW_ASSETS.Materials.MI_Button_Highlight, ?Index:=0)
                            # CardHighlightSound.Play(Player)
                            CPlayer.GuessUI.UpdateReticle(CPlayer.GuessUI.OpenHandReticleTexture)
                        else if(FortChar.IsLookingAtSphere[FloorUndoButton_Prop.GetTransform().Translation, DefaultButtonSize * FloorUndoButton_Prop.GetTransform().Scale.X]):
                            ResetButtonHighlights()
                            FloorUndoButton_Prop.SetMaterial(NEW_ASSETS.Materials.MI_Button_Highlight, ?Index:=0)
                            # CardHighlightSound.Play(Player)
                            CPlayer.GuessUI.UpdateReticle(CPlayer.GuessUI.OpenHandReticleTexture)
                        else if(IsMyTurn[CPlayer], FortChar.IsLookingAtSphere[FloorEndTurnButton_Prop.GetTransform().Translation, DefaultButtonSize * FloorEndTurnButton_Prop.GetTransform().Scale.X]):
                            ResetButtonHighlights()
                            FloorEndTurnButton_Prop.SetMaterial(NEW_ASSETS.Materials.MI_Button_Highlight, ?Index:=0)
                            # CardHighlightSound.Play(Player)
                            CPlayer.GuessUI.UpdateReticle(CPlayer.GuessUI.OpenHandReticleTexture)
                        else if(IsMyTurn[CPlayer], FortChar.IsLookingAtSphere[FloorAskQuestionButton_Prop.GetTransform().Translation, DefaultButtonSize * FloorAskQuestionButton_Prop.GetTransform().Scale.X]):
                            ResetButtonHighlights()
                            FloorAskQuestionButton_Prop.SetMaterial(NEW_ASSETS.Materials.MI_Button_Highlight, ?Index:=0)
                            # CardHighlightSound.Play(Player)
                            CPlayer.GuessUI.UpdateReticle(CPlayer.GuessUI.OpenHandReticleTexture)
                        else if(IsMyTurn[CPlayer], FortChar.IsLookingAtSphere[FloorMakeGuessButton_Prop.GetTransform().Translation, DefaultButtonSize * FloorMakeGuessButton_Prop.GetTransform().Scale.X]):
                            ResetButtonHighlights()
                            FloorMakeGuessButton_Prop.SetMaterial(NEW_ASSETS.Materials.MI_Button_Highlight, ?Index:=0)
                            # CardHighlightSound.Play(Player)
                            CPlayer.GuessUI.UpdateReticle(CPlayer.GuessUI.OpenHandReticleTexture)


                        # Index Card Hover
                        else if:

                            Self.StyleHasIndexCard?
                            
                            # Index card size
                            RectCenter := Self.IndexCard.Card.GetTransform().Translation
                            RectWidth := 640.0
                            RectHeight := 640.0

                            ObjectRotation := Self.IndexCard.Card.GetTransform().Rotation
                            RectDir1 := ObjectRotation.GetLocalRight()
                            RectDir2 := ObjectRotation.GetLocalForward()
                            PlaneNormal := ObjectRotation.GetLocalUp()

                            FortChar.IsLookingAtRectangle[RectCenter, RectWidth, RectHeight, RectDir1, RectDir2, PlaneNormal]
                        then:
                            ResetButtonHighlights()
                            Self.IndexCard.Hover()
                            CPlayer.GuessUI.UpdateReticle(CPlayer.GuessUI.OpenHandReticleTexture)

                        # else if(FortChar.IsLookingAtSphere[Self.HatProp.GetTransform().Translation, 10.0 * Self.HatProp.GetTransform().Scale.X]):
                        #     Print("Looking at hat")
                        #     CPlayer.GuessUI.UpdateReticle(Textures.T_Fabric_Reticle_GrabOpen)
                    

                        # YouTile.CardHidden? Hover
                        else if:
                            YouTile.CardHidden?

                            FortChar.IsLookingAtSphere[YouTile.FlipButtonProp.GetTransform().Translation, DefaultButtonSize * YouTile.FlipButtonProp.GetTransform().Scale.X]
                        then:
                            ResetButtonHighlights()

                            YouTile.FlipButtonProp.SetMaterial(NEW_ASSETS.Materials.MI_Button_Highlight)
                            CPlayer.GuessUI.UpdateReticle(CPlayer.GuessUI.OpenHandReticleTexture)
                        else:
                            ResetButtonHighlights()
                Sleep(0.1)

            # Print("Live raycast race ended")

    SetStyle( Style:card_style):void=
        for(Row : Self.Rows):
            for(Tile : Row.Tiles):
                Tile.SetStyle(Style)

        YouTile.SetStyle(Style)


    ResetTileHighlight(CPlayer: custom_player): void =
        for (Row : Self.Rows):
            for (Tile : Row.Tiles):
                Tile.Prop.SetMaterial(SavedFrameColor, ?Index:=2)
        
        if (not HoldingInput?):
            CPlayer.GuessUI.UpdateReticle(CPlayer.GuessUI.GetDefaultReticleTexture(CPlayer.Player, Self))
        
        Self.IndexCard.Unhover(SavedFrameColor)

    ResetButtonHighlights():void=
        UndoButton_Prop.SetMaterial(NEW_ASSETS.Materials.MI_Button, ?Index:=0)
        EndTurnButton_Prop.SetMaterial(NEW_ASSETS.Materials.MI_Button, ?Index:=0)
        AskQuestionButton_Prop.SetMaterial(NEW_ASSETS.Materials.MI_Button, ?Index:=0)
        MakeGuessButton_Prop.SetMaterial(NEW_ASSETS.Materials.MI_Button, ?Index:=0)
        FloorUndoButton_Prop.SetMaterial(NEW_ASSETS.Materials.MI_Button, ?Index:=0)
        FloorEndTurnButton_Prop.SetMaterial(NEW_ASSETS.Materials.MI_Button, ?Index:=0)
        FloorAskQuestionButton_Prop.SetMaterial(NEW_ASSETS.Materials.MI_Button, ?Index:=0)
        FloorMakeGuessButton_Prop.SetMaterial(NEW_ASSETS.Materials.MI_Button, ?Index:=0)
        YouTile.FlipButtonProp.SetMaterial(NEW_ASSETS.Materials.MI_Button, ?Index:=0)

        


    HandleCursorInputTriggerUndo(Agent:agent):void=
        HandleUndoTrigger(option{Agent})

    HandleCursorInputTriggerAskQuestion(Agent:agent):void=
        HandleAskQuestionTrigger(option{Agent})

    HandleCursorInputTriggerZoom(Agent:agent):void=
        spawn { ActualCursorZoom(Agent) }

    ActualCursorZoom(Agent:agent)<suspends>:void=
        var NearestCard : ?card_tile = false
        var PlayerPosition : transform = transform{}
        var TeleportedToCard : logic = false

        CardTeleportOffest := vector3{X:=0.0, Y:=315.0, Z:=0.0}

        if:
            Player := player[Agent]
            FortChar := Agent.GetFortCharacter[]
            TheGlobals := Globals[GetSession()]
            CPlayer := TheGlobals.CustomPlayerMap[Player]
            GameBoard := CPlayer.GameBoard
            not CPlayer.In1v1Area?
        then:

            GameBoard.CustomInput.CameraAddTrigger.Trigger(Player)

            GameBoard.CustomInput.Cursor_InputTrigger_Zoom.ReleasedEvent.Await()
            GameBoard.CustomInput.CameraRemoveTrigger.Trigger(Player)

    HandleCursorInputTriggerFire(Agent:agent):void=
        spawn { ActualCursorTrigger(Agent) }

    ActualCursorTrigger(Agent:agent)<suspends>:void=
        Logger.Print("Fire Button Pressed")
        var NearestCard : ?card_tile = false

        if:
            Player := player[Agent]
            FortChar := Agent.GetFortCharacter[]
            TheGlobals := Globals[GetSession()]
            CPlayer := TheGlobals.CustomPlayerMap[Player]
            GameBoard := CPlayer.GameBoard
            SourceData := PlayerSavedDataMap[Player]
            not CPlayer.In1v1Area?
        then:
            ViewRotation := FortChar.GetViewRotation()
            ViewForward := ViewRotation.GetLocalForward()
            RayEnd := FortChar.GetViewLocation() + (ViewForward * 8000.0)
            RayStart := FortChar.GetViewLocation() + (ViewForward * 300.0)

            Draw.DrawArrow(RayStart, RayEnd, ?Color:=Red, ?DrawDurationPolicy:=debug_draw_duration_policy.FiniteDuration, ?Duration:=5.0, ?Thickness:=10.0)

            HitLocation := GameBoard.CustomInput.Cursor_RayCast.CastFrom(FortChar.GetViewLocation(), ViewRotation)

            # POINTER PROJECTILE ########################################################
            if:
                PointerConditional := CPlayer.MaybePointerConditional?
                PointerConditional.IsHoldingItem[Agent, 0]
            then:
                PointerInfo := TheGlobals.GetPointerData(SourceData.CurrentPointer)
                # DartProjectile := SpawnProp(PointerInfo.AmmoProp, HitLocation)
                DartProjectile := SpawnProp(PointerInfo.AmmoProp, RayStart, ViewRotation)
                if(DartProp := DartProjectile(0)?):
                    spawn . DartProp.DelayedDispose(10.0)
            ################################################################################

            DefaultButtonSize := 60.0

            if(FoundCard := GetNearestCard[HitLocation, GameBoard], TheGlobals.GameInProgress?):
                Logger.Print("Found card {FoundCard.Index}")
                CPlayer.GuessUI.UpdateReticle(CPlayer.GuessUI.ClosedHandReticleTexture)

                HandleTriggerActivated(option{Agent}, FoundCard)

                set Self.HoldingInput = true
                CustomInput.Cursor_InputTrigger_Fire.ReleasedEvent.Await()
                CPlayer.GuessUI.UpdateReticle(CPlayer.GuessUI.OpenHandReticleTexture)
                Sleep(0.1)
                set Self.HoldingInput = false
            else if(TheGlobals.GameInProgress?):
                Logger.Print("No card found")
                
                if(FortChar.IsLookingAtSphere[UndoButton_Prop.GetTransform().Translation, DefaultButtonSize * UndoButton_Prop.GetTransform().Scale.X]):
                    HandleUndoTrigger(option{Agent})
                    CPlayer.GuessUI.UpdateReticle(CPlayer.GuessUI.ClosedHandReticleTexture)
                    ButtonPressSound.Play(Player)
                    set Self.HoldingInput = true
                    CustomInput.Cursor_InputTrigger_Fire.ReleasedEvent.Await()
                    CPlayer.GuessUI.UpdateReticle(CPlayer.GuessUI.OpenHandReticleTexture)
                    Sleep(0.3)
                    set Self.HoldingInput = false

                else if(IsMyTurn[CPlayer], FortChar.IsLookingAtSphere[EndTurnButton_Prop.GetTransform().Translation, DefaultButtonSize * EndTurnButton_Prop.GetTransform().Scale.X]):
                    HandleEndTurnTrigger(option{Agent})
                    CPlayer.GuessUI.UpdateReticle(CPlayer.GuessUI.ClosedHandReticleTexture)
                    ButtonPressSound.Play(Player)
                    set Self.HoldingInput = true
                    CustomInput.Cursor_InputTrigger_Fire.ReleasedEvent.Await()
                    CPlayer.GuessUI.UpdateReticle(CPlayer.GuessUI.OpenHandReticleTexture)
                    Sleep(0.3)
                    set Self.HoldingInput = false

                else if(IsMyTurn[CPlayer], FortChar.IsLookingAtSphere[AskQuestionButton_Prop.GetTransform().Translation, DefaultButtonSize * AskQuestionButton_Prop.GetTransform().Scale.X]):
                    HandleAskQuestionTrigger(option{Agent})
                    CPlayer.GuessUI.UpdateReticle(CPlayer.GuessUI.ClosedHandReticleTexture)
                    ButtonPressSound.Play(Player)
                    set Self.HoldingInput = true
                    CustomInput.Cursor_InputTrigger_Fire.ReleasedEvent.Await()
                    CPlayer.GuessUI.UpdateReticle(CPlayer.GuessUI.OpenHandReticleTexture)
                    Sleep(0.3)
                    set Self.HoldingInput = false

                else if(IsMyTurn[CPlayer], FortChar.IsLookingAtSphere[MakeGuessButton_Prop.GetTransform().Translation, DefaultButtonSize * MakeGuessButton_Prop.GetTransform().Scale.X]):
                    OpenDialogTrigger(option{Agent})
                    CPlayer.GuessUI.UpdateReticle(CPlayer.GuessUI.ClosedHandReticleTexture)
                    ButtonPressSound.Play(Player)
                    set Self.HoldingInput = true
                    CustomInput.Cursor_InputTrigger_Fire.ReleasedEvent.Await()
                    CPlayer.GuessUI.UpdateReticle(CPlayer.GuessUI.OpenHandReticleTexture)
                    Sleep(0.3)
                    set Self.HoldingInput = false

                else if(FortChar.IsLookingAtSphere[FloorUndoButton_Prop.GetTransform().Translation, DefaultButtonSize * FloorUndoButton_Prop.GetTransform().Scale.X]):
                    HandleUndoTrigger(option{Agent})
                    CPlayer.GuessUI.UpdateReticle(CPlayer.GuessUI.ClosedHandReticleTexture)
                    ButtonPressSound.Play(Player)
                    set Self.HoldingInput = true
                    CustomInput.Cursor_InputTrigger_Fire.ReleasedEvent.Await()
                    CPlayer.GuessUI.UpdateReticle(CPlayer.GuessUI.OpenHandReticleTexture)
                    Sleep(0.3)
                    set Self.HoldingInput = false

                else if(IsMyTurn[CPlayer], FortChar.IsLookingAtSphere[FloorEndTurnButton_Prop.GetTransform().Translation, DefaultButtonSize * FloorEndTurnButton_Prop.GetTransform().Scale.X]):
                    HandleEndTurnTrigger(option{Agent})
                    CPlayer.GuessUI.UpdateReticle(CPlayer.GuessUI.ClosedHandReticleTexture)
                    ButtonPressSound.Play(Player)
                    set Self.HoldingInput = true
                    CustomInput.Cursor_InputTrigger_Fire.ReleasedEvent.Await()
                    CPlayer.GuessUI.UpdateReticle(CPlayer.GuessUI.OpenHandReticleTexture)
                    Sleep(0.3)
                    set Self.HoldingInput = false

                else if(IsMyTurn[CPlayer], FortChar.IsLookingAtSphere[FloorAskQuestionButton_Prop.GetTransform().Translation, DefaultButtonSize * FloorAskQuestionButton_Prop.GetTransform().Scale.X]):
                    HandleAskQuestionTrigger(option{Agent})
                    CPlayer.GuessUI.UpdateReticle(CPlayer.GuessUI.ClosedHandReticleTexture)
                    ButtonPressSound.Play(Player)
                    set Self.HoldingInput = true
                    CustomInput.Cursor_InputTrigger_Fire.ReleasedEvent.Await()
                    CPlayer.GuessUI.UpdateReticle(CPlayer.GuessUI.OpenHandReticleTexture)
                    Sleep(0.3)
                    set Self.HoldingInput = false

                else if(IsMyTurn[CPlayer], FortChar.IsLookingAtSphere[FloorMakeGuessButton_Prop.GetTransform().Translation, DefaultButtonSize * FloorMakeGuessButton_Prop.GetTransform().Scale.X]):
                    OpenDialogTrigger(option{Agent})
                    CPlayer.GuessUI.UpdateReticle(CPlayer.GuessUI.ClosedHandReticleTexture)
                    ButtonPressSound.Play(Player)
                    set Self.HoldingInput = true
                    CustomInput.Cursor_InputTrigger_Fire.ReleasedEvent.Await()
                    CPlayer.GuessUI.UpdateReticle(CPlayer.GuessUI.OpenHandReticleTexture)
                    Sleep(0.3)
                    set Self.HoldingInput = false

                else if:

                    Self.StyleHasIndexCard?

                    # Rectangle properties
                    RectCenter := Self.IndexCard.Card.GetTransform().Translation
                    RectWidth := 640.0  # Example width
                    RectHeight := 640.0  # Example height

                    ObjectRotation := Self.IndexCard.Card.GetTransform().Rotation
                    RectDir1 :=ObjectRotation.GetLocalRight()
                    RectDir2 := ObjectRotation.GetLocalForward()
                    PlaneNormal := ObjectRotation.GetLocalUp()
                    FortChar.IsLookingAtRectangle[RectCenter, RectWidth, RectHeight, RectDir1, RectDir2, PlaneNormal]

                    # FortChar.IsLookingAtPlane[GameBoard.IndexCard.Card.GetTransform().Translation, GameBoard.IndexCard.Card.GetTransform().Rotation.GetLocalForward()]
                then:
                    Logger.Print("Clicked index card")
                    CPlayer.GuessUI.UpdateReticle(_UI.UI_Textures.T_Reticle_GrabClosed, vector2{X:=52.0, Y:=52.0})
                    ButtonPressSound.Play(Player)
                    set Self.HoldingInput = true
                    CustomInput.Cursor_InputTrigger_Fire.ReleasedEvent.Await()

                    CPlayer.GuessUI.UpdateReticle(_UI.UI_Textures.T_Reticle_GrabOpen, vector2{X:=52.0, Y:=52.0})

                    Self.IndexCard.ToggleSlide(Self.BoardLocalForward)

                    Sleep(0.25)
                    set Self.HoldingInput = false

                else if:

                    YouTile.CardHidden?

                    FortChar.IsLookingAtSphere[YouTile.FlipButtonProp.GetTransform().Translation, DefaultButtonSize * YouTile.FlipButtonProp.GetTransform().Scale.X]
                    
                then:
                    CPlayer.GuessUI.UpdateReticle(CPlayer.GuessUI.ClosedHandReticleTexture)
                    YouTile.ShowCharacterCard(Agent)
                    SetStreamerModeHideCharacter(Agent, false)
                    ButtonPressSound.Play(Agent)
                    set Self.HoldingInput = true
                    CustomInput.Cursor_InputTrigger_Fire.ReleasedEvent.Await()
                    CPlayer.GuessUI.UpdateReticle(CPlayer.GuessUI.GetDefaultReticleTexture(Agent, Self))
                    Sleep(0.3)
                    set Self.HoldingInput = false
       

                # Hat stuff maybe
                # else if(FortChar.IsLookingAtSphere[Self.HatProp.GetTransform().Translation, 10.0 * Self.HatProp.GetTransform().Scale.X]):
                #     Print("Looking at hat")
                #     CPlayer.GuessUI.UpdateReticle(Textures.T_Fabric_Reticle_GrabClosed)
                #     DistanceToHat := Distance(FortChar.GetViewLocation(), Self.HatProp.GetTransform().Translation)

                #     race:
                #         CustomInput.Cursor_InputTrigger_Fire.ReleasedEvent.Await()
                #         loop:
                #             HatRayEnd := FortChar.GetViewLocation() + (ViewForward * DistanceToHat)
                #             # Self.HatProp.SetMaterial(Materials.MI_Frame_Highlight, ?Index:=0)
                #             Self.HatProp.MoveTo(HatRayEnd, IdentityRotation(), 0.033)
                #             Sleep(0.033)

                    

                #     CPlayer.GuessUI.UpdateReticle(Textures.T_Fabric_Reticle_GrabOpen)
                else:
                    # CPlayer.GuessUI.UpdateReticle(Textures.T_Reticle_Default)
                    Logger.Print("No button found")

        else:
            Print("Failed to Fire, or game not in progress")


    HandleTriggerActivated(Agent:?agent, Tile:card_tile):void=
        Logger.Print("Tile {Tile.Index} activated")

        if(Self.MakingFinalGuess?):
            if(Tile.Index = CurrentTargetIndex):
                if(TheGlobals := Globals[GetSession()], RealOwner := Self.Owner?, Opponent := RealOwner.GetOpponent()?, OpponentCP := TheGlobals.CustomPlayerMap[Opponent]):
                    set TheGlobals.Winner = option{RealOwner.Player}
                    RealOwner.GuessUI.UpdateCorrectIncorrectImage(RealOwner.Player,RealOwner.Player, true)
                    OpponentCP.GuessUI.UpdateCorrectIncorrectImage(RealOwner.Player, Opponent, true)
                    RealOwner.GuessUI.HideShootToGuessUI(RealOwner.Player)
                    OnCorrectGuess.Signal()

                    WinnerXP.Award(RealOwner.Player)

                    # # Save Persistent Data
                    # AddCorrectGuess(RealOwner.Player)
                    # ExistingStreak := CheckForExistingStreak(RealOwner.Player, Opponent)
                    # if:
                    #     ActualStreak := ExistingStreak?
                    # then:
                    #     AddCorrectGuessToStreakData(RealOwner.Player, Opponent, ActualStreak)
                        
                    # AddGamesPlayed(RealOwner.Player)
                    # AddGamesPlayed(Opponent)

                    # if(IsNewPlayer[RealOwner.Player] or IsNewPlayer[Opponent]):
                    #     AddGamesPlayedWithNewFortWhoPlayers(RealOwner.Player)
                    #     AddGamesPlayedWithNewFortWhoPlayers(Opponent)
                        
                    #     NewPlayerAccolade.Award(RealOwner.Player)
                return
            else:

                if(TheGlobals := Globals[GetSession()], RealOwner := Self.Owner?, Player:=RealOwner.Player, Opponent := RealOwner.GetOpponent()?, OpponentCP := TheGlobals.CustomPlayerMap[Opponent]):
                    set TheGlobals.Winner = option{Opponent}
                    RealOwner.GuessUI.UpdateCorrectIncorrectImage(RealOwner.Player,RealOwner.Player, false)
                    OpponentCP.GuessUI.UpdateCorrectIncorrectImage(RealOwner.Player, Opponent, false)
                    RealOwner.GuessUI.HideShootToGuessUI(RealOwner.Player)
                    OnIncorrectGuess.Signal()

                    WinnerOpponentXP.Award(Opponent)

                    # Save Persistent Data
                    # AddIncorrectGuess(RealOwner.Player)
                    # AddGamesPlayed(RealOwner.Player)
                    # AddGamesPlayed(Opponent)

                    # if(IsNewPlayer[RealOwner.Player] or IsNewPlayer[Opponent]):
                    #     AddGamesPlayedWithNewFortWhoPlayers(RealOwner.Player)
                    #     AddGamesPlayedWithNewFortWhoPlayers(Opponent)
                        
                    #     NewPlayerAccolade.Award(RealOwner.Player)

                return

        if(Tile.State = card_state.Up):
            spawn{ Tile.KnockDown() }
            set KnockedTiles += array{Tile}
            set TilesKnocked += 1
            Logger.Print("Knocked tile {Tile.Index}")
            if (Tile.Index = CurrentTargetIndex):
                set TargetKnocked = true
                Logger.Print("Current target knocked")

            if (TilesKnocked >= NumTiles):
                set AllKnocked = true
                Logger.Print("All tiles knocked")

            if(RealPlayer := player[Agent?]):
                CardKnockXP.Award(RealPlayer)
        else:
            Logger.Print("Tile is already down, undo")
            spawn{ Tile.FlipUp() }
            if (ElementToRemove :=KnockedTiles.RemoveElement[FindIndexOfTile[Tile]]):
                set KnockedTiles = ElementToRemove
            set TilesKnocked -= 1
            set AllKnocked = false
            Logger.Print("Undid tile {Tile.Index}")

        if(Player := player[Agent?]):
            CardKnockSound.Play(Player)            
            TriggerRandomForKnockActivation.Trigger(Player)
            
            # persistent data
            # AddKnockedCard(Player)

    FindIndexOfTile(Tile: card_tile)<transacts><decides>:int=
        return KnockedTiles.Find[Tile]
            

    Reset()<suspends>:void=
        Logger.Print("Resetting game board")
        if(RealOwner := Self.Owner?, TheGlobals := Globals[GetSession()]):
            set RealOwner.HasVoted = false
            set TheGlobals.Winner = false
            set TheGlobals.PlayerVoteIndex = TheGlobals.PlayerVoteIndex.RemoveKey(RealOwner.Player)
            set RealOwner.MenuCategoryIndex = 0
            RealOwner.GuessUI.PlayerOneTexture.SetImage(Cards.Textures.UI.verse_ui_coming_soon)
            RealOwner.GuessUI.PlayerTwoTexture.SetImage(Cards.Textures.UI.verse_ui_coming_soon)
            set IndexCard.IsHidden = true

            RealOwner.GuessUI.TheirVoteTexture.SetImage(Cards.Textures.UI.verse_ui_coming_soon)
            RealOwner.GuessUI.YourVoteTexture.SetImage(Cards.Textures.UI.verse_ui_coming_soon)

            # set RealOwner.LiveRayCastActive = false

            if(VoteIndex := TheGlobals.PlayerVoteIndex[RealOwner.Player]):
                Logger.Print("GameBoard Reser, Vote Index {VoteIndex}")
            else:
                Logger.Print("GameBoard Reser, Vote Index not found")

            if(RealOwner.HasVoted?):
                Logger.Print("Has voted still true")
            else:
                Logger.Print("Has voted actually set false")
        else:
            Logger.Print("Failed to reset game board, owner not found")

        for(Row : Self.Rows):
            for(Tile : Row.Tiles):
                Tile.Reset(GetLastKnockedIndex())
            Sleep(0.0)

        set Self.KnockedTiles = array{}
        set Self.TilesKnocked = 0
        set Self.AllKnocked = false
        set Self.TargetKnocked = false
        set Self.CurrentTargetIndex = -1
        set Self.OpponentTargetIndex = -1
        set Self.MakingFinalGuess = false


    HideIndexIfShown()<suspends>:void=
        if(not IndexCard.IsHidden?):
            Self.IndexCard.ToggleSlide(Self.BoardLocalForward)


    DisableInputs():void=
        Logger.Print("Disabling buttons")
        Self.MakeGuessButton.Disable()
        Self.EndTurnButton.Disable()
        Self.AskQuestionButton.Disable()
        Self.UndoButton.Disable()

        set Self.HoldingInput = false

        if:
            TheOwner := Self.Owner?
        then:
            # Self.CustomInput.Cursor_InputTrigger_Fire.Unregister(TheOwner.Player)
            Self.CustomInput.InputTrigger_Ask_Question.Unregister(TheOwner.Player)

        for (B : Self.Beacons):
            B.Disable()


    RemoveFireSubscription():void=
        if:
            FireSub := Self.CustomInput.FireSubscription?
        then:
            FireSub.Cancel()
            set Self.CustomInput.FireSubscription = false
    


    EnableInputs():void=
        Logger.Print("Enabling buttons")
        Self.MakeGuessButton.Enable()
        Self.EndTurnButton.Enable()
        Self.AskQuestionButton.Enable()
        Self.UndoButton.Enable()

        set Self.HoldingInput = false

        if:
            TheOwner := Self.Owner?
        then:
            # Self.CustomInput.Cursor_InputTrigger_Fire.Register(TheOwner.Player)
            Self.CustomInput.InputTrigger_Ask_Question.Register(TheOwner.Player)
            # Self.CustomInput.Cursor_InputTrigger_Zoom.Register(TheOwner.Player)

        for (B : Self.Beacons):
            B.Enable()
    

    HandleAskQuestionTrigger(Agent:?agent):void=
        Logger.Print("Ask Question")
        if:
            not Self.AwaitingResponse?
            not Self.QuestionBankOpen?
            ActualAgent := Agent?
            TheGlobals := Globals[GetSession()]
            QuestionBankDevice := TheGlobals.TheQuestionBank
        then:
            # if opponent has not answered
            if:
                not TheGlobals.OpponentAnswered(0)? #yes
                not TheGlobals.OpponentAnswered(1)? #no
            then:
                set Self.QuestionBankOpen = true
                spawn{QuestionBankDevice.ShowMenu(ActualAgent)}
            else:
                Logger.Print("Opponent already answered")
                # display previous answer

                if(TheGlobals.OpponentAnswered(0)?):
                    Logger.Print("Opponent answered yes")
                    QuestionBankDevice.HUDResponseYes.SetText(QuestionBankDevice.SelectedQuestion)
                    QuestionBankDevice.HUDResponseYes.Show()
                else:
                    Logger.Print("Opponent answered no")
                    QuestionBankDevice.HUDResponseNo.SetText(QuestionBankDevice.SelectedQuestion)
                    QuestionBankDevice.HUDResponseNo.Show()

        else:
            if(ActualAgent := Agent?):
                HUD_AwaitingResponse.Show(ActualAgent)


    HandleAskQuestionButton(Agent:agent):void=
        HandleAskQuestionTrigger(option{Agent})

    HandleEndTurnButton(Agent:agent):void=
        HandleEndTurnTrigger(option{Agent})

    HandleEndTurnTrigger(Agent:?agent):void=
        Logger.Print("End Turn")
        OnEndTurn.Signal()


    OpenDialogTrigger(Agent:?agent):void=
        if:
            ActualAgent := Agent?
        then:
            Self.DialogConfirmFinalGuess.Show(ActualAgent)

    HandleMakeGuessButton(Agent:agent):void=
        Self.DialogConfirmFinalGuess.Show(Agent)

    HandleDialogFinalGuess(Agent: agent, Index: int):void=
        if(Index = 0):
            set Self.MakingFinalGuess = true

            if:                
                CPlayer := Self.Owner?
                Player := player[Agent]
            then:
                CPlayer.GuessUI.UpdateReticle(CPlayer.GuessUI.FinalGuessReticleTexture)

    HandleDialogStreamerSettings(Agent:agent, Index:int):void=
        case(Index):
            0 => block:
                Logger.Print("Hide character")
                SetStreamerModeHideCharacter(Agent, true)
                spawn . YouTile.HideCharacterCard(Agent)
            1 => block:
                Logger.Print("Show character")
                SetStreamerModeHideCharacter(Agent, false)
                spawn . YouTile.ShowCharacterCard(Agent)
            2 => block:
                Logger.Print("Horizontal UI")
                SetStreamerModeVerticalUI(Agent, false)
                if:
                    CPlayer := Self.Owner?
                then:
                    set CPlayer.GuessUI.CurrentUILayout = ui_layout.Horizontal
                    
                    if:
                        TheGlobals := Globals[GetSession()]
                        TheGlobals.CurrentPlayer = option{Agent}
                    then:
                        HUD_YourTurn.Show(Agent)
                    else:
                        HUD_OtherPlayersTurn.Show(Agent)
            3 => block:
                Logger.Print("Vertical UI")
                SetStreamerModeVerticalUI(Agent, true)
                if:
                    CPlayer := Self.Owner?
                then:
                    set CPlayer.GuessUI.CurrentUILayout = ui_layout.Vertical

                    if:
                        TheGlobals := Globals[GetSession()]
                        TheGlobals.CurrentPlayer = option{Agent}
                    then:
                        HUD_YourTurn_Centered.Show(Agent)
                    else:
                        HUD_OtherPlayersTurn_Centered.Show(Agent)
            _ => block:
                Logger.Print("Invalid index")
                return

    NewRandomTile():void=
        set CurrentTargetIndex = ChooseRandomTileIndex()

        Self.DrawBoundingBoxAroundTiles()
        
        if:
            TheChosenTile := GetTileFromIndex[CurrentTargetIndex]
        then:
            # draw sphere at chosen tile
            Logger.Print("Drawing sphere at chosen tile {TheChosenTile.Index}")
            Draw.DrawSphere(TheChosenTile.Prop.GetTransform().Translation, ?Radius:=100.0, ?Color:=Red, ?Thickness:=5.0, ?DrawDurationPolicy:=debug_draw_duration_policy.Persistent)



    GetTileFromIndex(Idx:int)<transacts><decides>:card_tile=

        var TempCardTile : ?card_tile = false

        if (Idx < NumTiles):
            if (Tile := Self.Rows[Floor(Idx / 6)].Tiles[Mod[Idx,6]]):
                set TempCardTile = option{Tile}

        return TempCardTile?

    SetOwner(CPlayer:custom_player):void=
        set Owner = option{CPlayer}
    
    SetOpponentTargetIndex(Index:int):void=
        set Self.OpponentTargetIndex = Index
        Logger.Print("Opponent target index set to {Self.OpponentTargetIndex}")
        
    AwaitWin()<suspends>:void=
        race:
            Self.OnCorrectGuess.Await()
            Self.OnIncorrectGuess.Await()

        

        

    DrawBoundingBoxAroundTiles():void=
        Logger.Print("Drawing bounding boxes around tiles")
        BoundsPadding := 25.0
        # TileDimensions : vector3 = vector3{X:=171.0, Y:=288.0, Z:=16.0}  # Assuming the dimensions of a tile
        HalfHeight : float = TileDimensions.Y / 2.0  # Half of the box's height
    
        for(Row : Self.Rows):
            for(Tile : Row.Tiles):
                # Get the current tile's position and rotation
                TileTransform := Tile.Prop.GetTransform()
                TilePosition := TileTransform.Translation
                TileRotation := TileTransform.Rotation
    
                # Obtain the local upward vector relative to the tile's rotation
                LocalUp := TileRotation.GetLocalRight()
    
                # Calculate the offset to move the center of the bounding box upwards
                UpwardOffset := LocalUp * HalfHeight
    
                # Translate the center position upwards by half of the box's height
                BoxCenter : vector3 = TilePosition + UpwardOffset
    
                # Now draw the bounding box with the new center position
                Draw.DrawBox(BoxCenter, TileRotation, ?Extent:=TileDimensions/2.0, ?Color := Green, ?Thickness := 5.0, ?DrawDurationPolicy := debug_draw_duration_policy.Persistent)




    HandleUndoButton(Agent:agent):void=
        HandleUndoTrigger(option{Agent})

    HandleUndoTrigger(Agent:?agent):void=
        Logger.Print("Undo")
        if:
            LastKnocked := KnockedTiles[KnockedTiles.Length - 1]
            ElementToRemove := KnockedTiles.RemoveElement[KnockedTiles.Length - 1]
        then:
            set KnockedTiles = ElementToRemove
            spawn{LastKnocked.FlipUp()}
            set TilesKnocked -= 1
        else:
            Logger.Print("No last move")
        
            



    ChooseRandomTileIndex():int=        
        GetRandomInt(0, NumTiles - 1)

    AssignTileIndexies()<suspends>:void=
        var index : int = 0
        for(Row : Self.Rows):
            for(Tile : Row.Tiles):
                set Tile.Index = index
                set index += 1

    KnockTile(RowIndex:int, ColIndex:int)<suspends>:void=
        if (RowIndex < 4 and ColIndex < 6):
            if (Tile := Self.Rows[RowIndex].Tiles[ColIndex]):
                Tile.KnockDown()
                set KnockedTiles += array{Tile}
                set TilesKnocked  += 1
                Logger.Print("Knocked tile {RowIndex} {ColIndex}")
        else:
            Logger.Print("Invalid tile index")

    KnockTile(TileIndex:int)<suspends>:void=
        if (TileIndex < NumTiles):
            if (Tile := Self.Rows[Floor(TileIndex / 6)].Tiles[Mod[TileIndex,6]]):
                Tile.KnockDown()
                set KnockedTiles += array{Tile}
                set TilesKnocked += 1
                Logger.Print("Knocked tile {TileIndex}")
        else:
            Logger.Print("Invalid tile index")

    GetLastKnockedIndex():int=
        if (LastKnocked := KnockedTiles[KnockedTiles.Length - 1]):
            return LastKnocked.Index
        else:
            return -1


    AwaitCurrentTargetKnocked()<suspends>:void=
        loop:
            if (TilesKnocked < NumTiles):
                if (LastKnocked := KnockedTiles[KnockedTiles.Length - 1]):

                    if(Self.MakingFinalGuess?):
                        if:
                            LastKnocked.Index = CurrentTargetIndex
                        then:
                            set TargetKnocked = true

                            if(TheGlobals := Globals[GetSession()], RealOwner := Self.Owner?):
                                set TheGlobals.Winner = option{RealOwner.Player}
                                RealOwner.GuessUI.HideShootToGuessUI(RealOwner.Player)
                                # RealOwner.GuessUI.UpdateCorrectIncorrectImage(RealOwner.Player, true)
                            break
                        else:
                            if(TheGlobals := Globals[GetSession()], RealOwner := Self.Owner?, Player:=RealOwner.Player):
                                set TheGlobals.Winner = RealOwner.GetOpponent()
                                RealOwner.GuessUI.HideShootToGuessUI(RealOwner.Player)
                                # RealOwner.GuessUI.UpdateCorrectIncorrectImage(RealOwner.Player, false)
                            break

            else:
                break
            Sleep(1.0)
        Logger.Print("Current target knocked")

    AwaitAllTilesKnocked()<suspends>:void=
        loop:
            if (TilesKnocked < NumTiles):
                Sleep(1.0)
            else:
                set AllKnocked = true
                break
        Logger.Print("All tiles knocked")



    AwaitTurnPlayed(CPlayer:custom_player)<suspends>:void=
        if:
            TheGlobals := Globals[GetSession()]
        then:
            if:
                Opponent := CPlayer.GetOpponent()?
                OpponentCP := TheGlobals.CustomPlayerMap[Opponent]
            then:
                set CPlayer.GuessUI.IsTimerActive = true
                set OpponentCP.GuessUI.IsTimerActive = true

                CPlayer.GuessUI.TurnTimer.Start()
                CPlayer.GuessUI.TurnTimer.SetActiveDuration(PLAYER_TURN_TIME)

                branch:
                    HandleTimerUI(CPlayer, OpponentCP) # just moved from race to branch kyle

                race:

                    TurnAwaitPlayedOrTimer(CPlayer, OpponentCP)

                    TurnAwaitWinRace(OpponentCP)
                                 
            set TheGlobals.OpponentAnswered = (false, false) # reset opponent answered
            set MakingFinalGuess = false
            
            CPlayer.GuessUI.HideShootToGuessUI(CPlayer.Player)
            Logger.Print("Turn played")

    TurnAwaitPlayedOrTimer(CPlayer:custom_player,OpponentCP:custom_player)<suspends>:void=
        race:
            OnEndTurn.Await()
            CPlayer.GuessUI.TurnTimer.SuccessEvent.Await()
            TurnAwaitWinRace(OpponentCP)

    TurnAwaitWinRace(OpponentCP:custom_player)<suspends>:void=
        race:
            AwaitWin()
            OpponentCP.GameBoard.AwaitWin()

    HandleTimerUI(CPlayer:custom_player, OpponentCP:custom_player)<suspends>:void=
        

        defer:
            Logger.Print("Timer Defer")

        Logger.Print("Handling timer UI")
        # race:
        #     TurnAwaitWinRace(OpponentCP)

        #     UpdateTimerUILoop(CPlayer, OpponentCP)
        Sleep(0.0)
        LSTurnTime.SetPlaybackTime(0.0)
        Sleep(0.0)
        LSTurnTime.Play()
        block:
            HUD_YourTurn.Show(CPlayer.Player)
            HUD_OtherPlayersTurn.Show(OpponentCP.Player)
        TurnAwaitWinRace(OpponentCP)


    UpdateTimerUILoop(CPlayer:custom_player, OpponentCP:custom_player)<suspends>:void=
        var CurrentDisplayedTime : int = 90
        race:
            Sleep(90.0)
            loop:
                block:
                    ShowPlayerTurnTime(CPlayer, OpponentCP, CurrentDisplayedTime)
                    ShowOpponentTurnTime(OpponentCP, CPlayer, CurrentDisplayedTime)

                Sleep(1.0)
                set CurrentDisplayedTime -= 1



    ShowPlayerTurnTime(CPlayer:custom_player, OpponentCP:custom_player, CurrentDisplayedTime:int):void=
        
        if(CPlayer.GuessUI.CurrentUILayout = ui_layout.Horizontal):
            HUD_YourTurn.Show(CPlayer.Player, IntToMessage(CurrentDisplayedTime), ?DisplayTime:=0.0)
        else:
            HUD_YourTurn_Centered.Show(CPlayer.Player, IntToMessage(CurrentDisplayedTime), ?DisplayTime:=0.0)

    ShowOpponentTurnTime(OpponentCP:custom_player, CPlayer:custom_player, CurrentDisplayedTime:int):void=
        
        if(OpponentCP.GuessUI.CurrentUILayout = ui_layout.Horizontal):
            HUD_OtherPlayersTurn.Show(OpponentCP.Player, IntToMessage(CurrentDisplayedTime), ?DisplayTime:=0.0)
        else:
            HUD_OtherPlayersTurn_Centered.Show(OpponentCP.Player, IntToMessage(CurrentDisplayedTime), ?DisplayTime:=0.0)

    FlipBoard(?Instant:logic=false)<suspends>:void=
        Logger.Print("Attempting to flip board")
        if(Instant?):
            ArraySync(Rows, FlipARow)
        else:
            for(Row : Self.Rows):
                Row.FlipRow(?Instant:=true)

    FlipARow(Row:game_board_row)<suspends>:void=
        Row.FlipRow(?Instant:=true)

    KnockBoard(?Instant:logic=false)<suspends>:void=
        Logger.Print("Attempting to knock board")
        if(Instant?):
            ArraySync(Rows, KnockARow)
        else:
            for(Row : Self.Rows):
                Row.KnockRow(?Instant:=true)

    KnockARow(Row:game_board_row)<suspends>:void=
        Row.KnockRow(?Instant:=true)



    AllowManualCharacterSelection()<suspends>:int=
        Logger.Print("Allowing manual character selection")

        var SelectedIndex : int = -1
    
        defer:
            Logger.Print("Manual character selection ended")
            ResetButtonHighlights()
    
            if:
                CPlayer := Owner?
            then:
                ResetTileHighlight(CPlayer)
                CPlayer.GuessUI.UpdateReticle(_UI.UI_Textures.T_Reticle_Default, vector2{X:=52.0, Y:=52.0})
                # Unsubscribe from the input event after selection or exit
                # set CPlayer.LiveRayCastActive = false
                # if (OldInput := InputCancelable?):
                #     OldInput.Cancel()
                #     set InputCancelable = false

                # CustomInput.Cursor_InputTrigger_Fire.Unregister(CPlayer.Player)
                set HoldingInput = false
                
    
        if:
            CPlayer := Owner?
            Player := CPlayer.Player
            FortChar := Player.GetFortCharacter[]
            Agent := FortChar.GetAgent[]
            TheGlobals := Globals[GetSession()]
            GameBoard := CPlayer.GameBoard
            CheckForPlayerSavedData[Agent]
            SourceData := PlayerSavedDataMap[Agent]
        then:
            # set CPlayer.LiveRayCastActive = false
            Sleep(0.25)
            # Subscribe to the input event for clicking
            CustomInput.Cursor_InputTrigger_Fire.Register(Player)
            InputSub := CustomInput.Cursor_InputTrigger_Fire.PressedEvent.Subscribe(HandleManualSelectionClick)
            set InputCancelable = option{InputSub}
            set HoldingInput = false
            CPlayer.GuessUI.UpdateReticle(_UI.UI_Textures.T_Reticle_Default, vector2{X:=52.0, Y:=52.0})
            
            
            # branch:
            #     RayTraceLoop(CPlayer, GameBoard, FortChar)

            CardIndex := ManualCharacterSelected.Await()
            set SelectedIndex = CardIndex
            
            # spawn . ResetCharacterSelectedCard(CardIndex)
        
        return SelectedIndex

    # RayTraceLoop(CPlayer:custom_player, GameBoard:game_board, FortChar:fort_character)<suspends>:void=
    #     set CPlayer.LiveRayCastActive = true


    #     defer:
    #         Print("Ray trace loop ended")
    #         set CPlayer.LiveRayCastActive = false
    #         set HoldingInput = false
            


    #     loop:
    #         if (not CPlayer.LiveRayCastActive?):
    #             break

    #         # set CPlayer.LiveRayCastActive = true
    #         ViewRotation := FortChar.GetViewRotation()
    #         ViewForward := ViewRotation.GetLocalForward()
    #         RayEnd := FortChar.GetViewLocation() + (ViewForward * 8000.0)

    #         HitLocation := GameBoard.CustomInput.Cursor_RayCast.CastFrom(FortChar.GetViewLocation(), ViewRotation)
    #         if:
    #             LookingAtCard := GetNearestCard[HitLocation, GameBoard]
    #             HighlightMaterial := PlayerCardHighlightMaterial?
    #         then:
    #             ResetTileHighlight(CPlayer)
    #             # Highlight the card the player is looking at
    #             # LookingAtCard.Prop.SetMaterial(HighlightMaterial, ?Index:=2)
    #             LookingAtCard.Prop.SetMaterial(Materials.BoardDesigner.CardBacks.MI_Cardback_Highlight, ?Index:=2)

    #             if (not HoldingInput?):
    #                 CPlayer.GuessUI.UpdateReticle(_UI.UI_Textures.T_Reticle_GrabOpen, vector2{X:=52.0, Y:=52.0})
    #         else:
    #             ResetTileHighlight(CPlayer)

    #         Sleep(0.1)

    HandleManualSelectionClick(Agent: agent): void =
        spawn . ActualHandleManualSelectionClick()
    
    ActualHandleManualSelectionClick()<suspends>: void =
        if:
            CPlayer := Owner?
            Player := CPlayer.Player
            FortChar := Player.GetFortCharacter[]
            GameBoard := CPlayer.GameBoard
            not CPlayer.In1v1Area?
        then:
            ViewRotation := FortChar.GetViewRotation()
            ViewForward := ViewRotation.GetLocalForward()
            RayEnd := FortChar.GetViewLocation() + (ViewForward * 8000.0)
    
            HitLocation := GameBoard.CustomInput.Cursor_RayCast.CastFrom(FortChar.GetViewLocation(), ViewRotation)
    
            if:
                SelectedCard := GetNearestCard[HitLocation, GameBoard]
                HighlightMaterial := PlayerCardHighlightMaterial?
            then:
                Logger.Print("Selected card {SelectedCard.Index}")
    
                # Set the CurrentTargetIndex to the selected card
                set CurrentTargetIndex = SelectedCard.Index
                Logger.Print("Current target index set to {CurrentTargetIndex}")
    
                # Provide feedback to the player
                # SelectedCard.Prop.SetMaterial(HighlightMaterial, ?Index:=2)
                SelectedCard.Prop.SetMaterial(Materials.BoardDesigner.CardBacks.MI_Cardback_Highlight, ?Index:=2)
    
                # Signal that the selection is complete
                ManualCharacterSelected.Signal(CurrentTargetIndex)
    
                # Unsubscribe from the input event after selection
                if (OldInput := InputCancelable?):
                    OldInput.Cancel()
                    set InputCancelable = false

                # set CPlayer.LiveRayCastActive = false

            else:
                Logger.Print("No card selected")



    GetAvailableStickerSlot()<transacts><decides>:board_sticker_slot=
        # temp will make this better once i add more stickers
        Avail := for(SS : StickerSlots, not SS.MaybeSticker?):
            SS
        Avail[0]


    AddSticker(Agent:agent, CPlayer:custom_player, Slot:board_sticker_slot, StickerID:int, StickerProp:creative_prop)<suspends>:void=
        if:
            StickerProp.TeleportTo[Slot.GetTransform().Translation, Slot.GetTransform().Rotation]
        then:
            set Slot.MaybeSticker = option{StickerProp}
            Logger.Print("Sticker added to slot")
        else:
            Logger.Print("Failed to add sticker to slot")


        # Sleep(1.5)
        # Sticker := SpawnProp(PropAsset, Slot.GetTransform())
        # if(TheProp := Sticker(0)?):
        #     set  Slot.MaybeSticker = option{TheProp}
        # Sleep(2.5)
        # CPlayer.GameBoard.AddStickerCam.RemoveFrom(Agent)



    IsMyTurn(CPlayer:custom_player)<transacts><decides>:void=
        TheGlobals := Globals[GetSession()]
        CurrentTurn := TheGlobals.CurrentTurn
        CurrentTurn = CPlayer.TeamIndex


    AwaitReturnToGameplayArea(CPlayer:custom_player)<suspends>:void=
        Logger.Print("Awaiting return to gameplay area")
        loop:
            if(not CPlayer.AwayFromGameplayArea?):
                break

            Sleep(1.0)

    MonitorWithinGameplayArea(TheGlobals:globals, CPlayer:custom_player, Agent:agent)<suspends>:void=
        loop:
            if:
                FortChar :=Agent.GetFortCharacter[]
                not TheGlobals.GameManager.GameplayExtents.Contains[FortChar.GetViewLocation()]
            then:
                set CPlayer.AwayFromGameplayArea = true
            else:
                set CPlayer.AwayFromGameplayArea = false
            
            Sleep(1.0)