
using { /Fortnite.com/Devices }
using { /Verse.org/Simulation }
using { /UnrealEngine.com/Temporary/Diagnostics }
using { /UnrealEngine.com/Temporary/SpatialMath }
using { /Fortnite.com/Devices/CreativeAnimation }
# using { /Fortnite.com/Devices/Animation }
using { /Verse.org/Random }
using { /Fortnite.com/Characters }
using { /Verse.org/Colors/NamedColors }
using { /UnrealEngine.com/Temporary/UI }
using { /Fortnite.com/UI }
using { /Verse.org/Assets }
using { /Fortnite.com/FortPlayerUtilities }


game_board_row := class<concrete>():
    @editable Tiles : []card_tile = array{}

    FlipRow(?Instant:logic=false)<suspends>:void=
        # Print("Attempting to flip row")
        if(Instant?):
            sync:
                if(Tile0 := Self.Tiles[0]):
                    Tile0.FlipUp()
                if(Tile1 := Self.Tiles[1]):
                    Tile1.FlipUp()
                if(Tile2 := Self.Tiles[2]):
                    Tile2.FlipUp()
                if(Tile3 := Self.Tiles[3]):
                    Tile3.FlipUp()
                if(Tile4 := Self.Tiles[4]):
                    Tile4.FlipUp()
                if(Tile5 := Self.Tiles[5]):
                    Tile5.FlipUp()
            # ArraySync(Tiles, FlipATile)
        else:
            for(Tile : Self.Tiles):
                Tile.FlipUp()
                Sleep(0.0)

    # FlipATile(Tile:card_tile)<suspends>:void=
    #     Tile.FlipUp()

    KnockRow(?Instant:logic=false)<suspends>:void=

        if(Instant?):
            sync:
                if(Tile0 := Self.Tiles[0]):
                    Tile0.KnockDown()
                if(Tile1 := Self.Tiles[1]):
                    Tile1.KnockDown()
                if(Tile2 := Self.Tiles[2]):
                    Tile2.KnockDown()
                if(Tile3 := Self.Tiles[3]):
                    Tile3.KnockDown()
                if(Tile4 := Self.Tiles[4]):
                    Tile4.KnockDown()
                if(Tile5 := Self.Tiles[5]):
                    Tile5.KnockDown()
            # ArraySync(Tiles, KnockATile)
        else:
            for(Tile : Self.Tiles):
                Tile.KnockDown()
                Sleep(0.0)

    # KnockATile(Tile:card_tile)<suspends>:void=
    #     Tile.KnockDown()