using { /Fortnite.com/Devices }
using { /Verse.org/Simulation }
using { /UnrealEngine.com/Temporary/Diagnostics }
using { /UnrealEngine.com/Temporary/SpatialMath }
using { /Verse.org/Random }
using { /UnrealEngine.com/Temporary/UI }
using { /Fortnite.com/FortPlayerUtilities }
using { /Fortnite.com/Characters }
using { /Verse.org/Assets }
using { /Verse.org/Colors/NamedColors }

game_manager := class(creative_device):

    Logger:log := log{Channel:=log_device}
    Draw:debug_draw := debug_draw{Channel:=debug_drawing}

    P1CoverMaterial : Cards.Materials.YouCards.MI_You_Card_Cover_P1_material = Cards.Materials.YouCards.MI_You_Card_Cover_P1_material{}
    P2CoverMaterial : Cards.Materials.YouCards.MI_You_Card_Cover_P2_material = Cards.Materials.YouCards.MI_You_Card_Cover_P2_material{}

    @editable AdminKeypad : admin_keypad = admin_keypad{}
    @editable VerseCrashDetector : trigger_device = trigger_device{}
    @editable ButtonTestLocker : button_device = button_device{}
    @editable ButtonTestDate : button_device = button_device{}
    # @editable LockerCameraTest : gameplay_camera_fixed_point_device = gameplay_camera_fixed_point_device{}
    @editable HideHUDTest : hud_controller_device = hud_controller_device{}
    @editable BillboardTest : billboard_device = billboard_device{}

    @editable PortalForRematch : teleporter_device = teleporter_device{}

    @editable PlayerOneBoard : game_board = game_board{}
    @editable PlayerTwoBoard : game_board = game_board{}

    @editable PlayerOneSpawnPad : player_spawner_device = player_spawner_device{}
    @editable PlayerTwoSpawnPad : player_spawner_device = player_spawner_device{}

    @editable TimerAwaitingPlayers : timer_device = timer_device{}
    @editable TimerTurnTime : timer_device = timer_device{}
    @editable LevelSequenceTurnTime : cinematic_sequence_device = cinematic_sequence_device{}
    @editable VotedSound : audio_player_device = audio_player_device{}

    @editable MainPlayerCounter : player_counter_device = player_counter_device{}
    @editable TwoPlayerCounter : player_counter_device = player_counter_device{}
    @editable ScoreManager : score_manager_device = score_manager_device{}
    @editable GoldGranter : item_granter_device = item_granter_device{}
    @editable TrackerLifeTimeWins : tracker_device = tracker_device{}

    @editable Analytics_Styles : []analytics_device = array{}

    QuestionBankDevice : question_bank_device = question_bank_device{}
    @editable RealTimeClock : rtc_device = rtc_device{}
    @editable DailyReward : daily_reward_wrapper = daily_reward_wrapper{}

    @editable RoundSettings : round_settings_device = round_settings_device{}

    @editable TriggerRandomForKnockActivation : trigger_device = trigger_device{}
    @editable HUDGeneric : hud_message_device = hud_message_device{}
    @editable HUDResponseNo : hud_message_device = hud_message_device{}
    @editable HUDResponseYes : hud_message_device = hud_message_device{}
    @editable HUDResponseMaybe : hud_message_device = hud_message_device{}
    # @editable HUDLoadTimerNumbers : hud_message_device = hud_message_device{}
    @editable HUDNoOpponentFound : hud_message_device = hud_message_device{}
    @editable HUDChooseYourCharacter : hud_message_device = hud_message_device{}
    @editable HUDTimerSelectionScreen : hud_message_device = hud_message_device{}
    @editable EndGame : end_game_device = end_game_device{}

    @editable InputTriggerHackCamera : input_trigger_device = input_trigger_device{}

    @editable BarrierBlockPlayerView : barrier_device = barrier_device{}
    @editable BarrierBlockPlayerViewHollow : barrier_device = barrier_device{}
    @editable BarrierBlockPlayerViewParent : creative_prop = creative_prop{}
    @editable TimerManualCharacterSelection : timer_device = timer_device{}

    @editable AccoladeRematchBonus : accolades_device = accolades_device{}
    @editable AnalyticQuestionAsked : analytics_device = analytics_device{}
    @editable AnalyticQuestionAnswered : analytics_device = analytics_device{}
    @editable AnaltyicCharacterSelectedManual : analytics_device = analytics_device{}
    @editable AnaltyicCharacterSelectedRandom : analytics_device = analytics_device{}

    @editable TrackerPaintQuest : tracker_device = tracker_device{}

    @editable Kevin : kevin = kevin{}
    # @editable Iceberg : iceberg_movement = iceberg_movement{}
    # @editable Igloo : igloo_campfire = igloo_campfire{}

    @editable DefaultStickerPropAsset : creative_prop_asset = DefaultCreativePropAsset
    @editable WinterQuestSwitch : switch_device = switch_device{}

    @editable OGXPAccolade : accolades_device = accolades_device{}

    # @editable SuctionDart : creative_prop_asset = DefaultCreativePropAsset
    # @editable DialogEquipPointer : popup_dialog_device = popup_dialog_device{}
    # @editable GranterPointers : item_granter_device = item_granter_device{}
    # @editable ItemRemoverPointers : item_remover_device = item_remover_device{}
    # @editable VFXPointers : []visual_effect_powerup_device = array{}
    # @editable AmmoProps : custom_ammo_assets = custom_ammo_assets{}
    @editable PointerShop : pointer_shop = pointer_shop{}

    @editable GameplayExtentsProp : creative_prop = creative_prop{}
    var GameplayExtents : bounding_box = bounding_box:
        Center:=vector3{X:=0.0,Y:=0.0,Z:=0.0}
        Extents:=vector3{X:=0.0,Y:=0.0,Z:=0.0}

    var HasFoundOnlyOnePlayer : logic = false
    

    var TestCrash : logic = false

    var TheP1Index : int = -1
    var TheP2Index : int = -1

    var GameEverStarted : logic = false

    var P1Spawned : logic = false
    var P2Spawned : logic = false
    var SpawnedPlayers : []agent = array{}

    StartVerseCrashDetector()<suspends>:void=
        loop:
            VerseCrashDetector.Enable()
            Sleep(0.25)
            if(TestCrash?):
                Print("Crashing")
                break

    

    HandleButtonTest(Agent:agent):void=

        Print("Button Test Pressed")

        if:
            Player := player[Agent]
            TheGlobals := Globals[GetSession()]
            CPlayer := TheGlobals.CustomPlayerMap[Player]
            GameBoard := CPlayer.GameBoard
            SourceData := PlayerSavedDataMap[Player]
        then:
            # set CPlayer.AwayFromGameplayArea = true
            spawn . GameBoard.BoardLocker.InitLockerUI(Player, CPlayer, TrackerPaintQuest)
            HideHUDTest.UpdateAffectedTeam(Player)

            spawn . AwaitLockerUIClosed(Player, GameBoard.BoardLocker, CPlayer)

            # FirstDayPlayed := SourceData.FirstDayPlayed
            # MinsPlayed := SourceData.TotalMinutesPlayed
            # YourScore := SourceData.Score

            # DebugBillboard.SetText(ThreeLineMessage(FirstDayPlayed.String(), "{MinsPlayed}", "{YourScore}"))


    AwaitLockerUIClosed(Player:player, NewLockerUI:board_locker_ui, CPlayer:custom_player)<suspends>:void=
        NewLockerUI.ExitLockerUIEvent.Await()
        HideHUDTest.ResetAffectedTeam()
        # set CPlayer.AwayFromGameplayArea = false


    HandleTriggerHackCamera(Agent:agent):void=
        Logger.Print("Trigger Hack Camera")
        spawn . ActualHandleTriggerHackCamera(Agent)

    ActualHandleTriggerHackCamera(Agent:agent)<suspends>:void=
        OpponentBoard := if(PlayerOneBoard.Owner?.Player = Agent) then PlayerTwoBoard else PlayerOneBoard
        OpponentBoard.HackView(Agent, InputTriggerHackCamera)


# Runs when the device is started in a running game
    OnBegin<override>()<suspends>:void=
        Logger.Print("Profile: Starting Fort Who")
        # spawn . StartVerseCrashDetector()

        # TwoPlayerCounter.CountedEvent.Subscribe(HandlePlayerSpawnCounted)

        # PlayerOneSpawnPad.SpawnedEvent.SubscribeAgent(HandlePlayerSpawnCounted, 0)
        # PlayerTwoSpawnPad.SpawnedEvent.SubscribeAgent(HandlePlayerSpawnCounted, 1)


        # GetPlayspace().PlayerAddedEvent().Subscribe(HandlePlayerJoinedSession)
        GetPlayspace().PlayerRemovedEvent().Subscribe(HandlePlayerRemoved)

        ButtonTestLocker.InteractedWithEvent.Subscribe(HandleButtonTest)
        ButtonTestDate.InteractedWithEvent.Subscribe(HandleButtonTestDate)
        InputTriggerHackCamera.PressedEvent.Subscribe(HandleTriggerHackCamera)

        TimerAwaitingPlayers.SuccessEvent.Subscribe(NoSecondPlayerFoundYet)

        # DialogEquipPointer.RespondingButtonEvent.Subscribe(HandleEquipPointerResponse)
        # DialogEquipPointer.TimeOutEvent.Subscribe(HandleEquipPointerTimeOut)

        if(not Globals[GetSession()]):
            NewGlobals := globals{GameManager:=Self, TheQuestionBank:=QuestionBankDevice, DefaultStickerPropAsset:=DefaultStickerPropAsset}
            NewGlobals.InitPointers(PointerShop.AmmoProps)
            if . set Globals[GetSession()] = NewGlobals
        else:
            Logger.Print("Globals already set")

        spawn . RealTimeClock.Init()
        spawn . AdminKeypad.Init()
        spawn . InitDailyReward()
        spawn . PointerShop.Init()
        # spawn . Kevin.Init()
        # spawn . Iceberg.Init()
        # spawn . Igloo.Init()

        TimerAwaitingPlayers.Start()
        

        set GameplayExtents = GetExtentsOfBoundingProp(GameplayExtentsProp)
        Print("Gameplay Extents: {GameplayExtents.Center} - {GameplayExtents.Extents}")
        # DrawBox<native><public>(Center:vector3, Rotation:rotation, ?Extent:vector3 = external {}, ?Color:color = external {}, ?Thickness:float = external {}, ?DrawDurationPolicy:debug_draw_duration_policy = external {}, ?Duration:float = external {})<transacts>:void
        Draw.DrawSphere(GameplayExtents.Center, ?Radius:=5.0, ?Color:=Red, ?Thickness:=15.0, ?DrawDurationPolicy:=debug_draw_duration_policy.FiniteDuration, ?Duration:=90.0)
        Draw.DrawBox(GameplayExtents.Center, IdentityRotation(), ?Extent:=GameplayExtents.Extents, ?Color:=Red, ?Thickness:=15.0, ?DrawDurationPolicy:=debug_draw_duration_policy.FiniteDuration, ?Duration:=90.0)

        # spawn . HandlePlayerOneSpawn()
        # spawn . HandlePlayerTwoSpawn()
        
        # AwaitBothPlayersSpawned()

        Players := sync:
            block:
                Agent := PlayerOneSpawnPad.SpawnedEvent.Await()
                PlayerOneBoard.InitNew(0, 1.0, Agent, TimerTurnTime, TriggerRandomForKnockActivation, InputTriggerHackCamera, LevelSequenceTurnTime, DefaultStickerPropAsset)
                AssignPointer(Agent)
                Agent
            block:
                Agent := PlayerTwoSpawnPad.SpawnedEvent.Await()
                PlayerTwoBoard.InitNew(1, -1.0, Agent, TimerTurnTime, TriggerRandomForKnockActivation, InputTriggerHackCamera, LevelSequenceTurnTime, DefaultStickerPropAsset)
                AssignPointer(Agent)
                Agent

        # TwoPlayerCounter.CountSucceedsEvent.Await() # BOTH PLAYERS SPAWNED
        TimerAwaitingPlayers.Reset()
        
        RoundSettings.DisableMatchmaking()
        Sleep(1.0)

        # if:
        #     Player1 := SpawnedPlayers[0]
        #     Player2 := SpawnedPlayers[1]
        # then:

        CheckForRematchBonus(Players(0))
        CheckForRematchBonus(Players(1))


        spawn . InitQuestionBank()
            

        #award anyone who has this on with the sticker
        if:
            WinterQuestSwitch.GetCurrentState[Players(0)]
            not HasCompletedWinter2024Quest[Players(0)] # if they havent already completed it
            CPlayer := Globals[GetSession()].CustomPlayerMap[Players(0)]
            AvailStickerSlot := PlayerOneBoard.GetAvailableStickerSlot[]
            SetCompletedWinter2024Quest(Players(0))
        then:
            NewStickerResult := SpawnProp(DefaultStickerPropAsset, vector3{}, rotation{})
            Winter2024StickerMaterial := Materials.Stickers.MI_Sticker_Base_Inst_material{}
            set Winter2024StickerMaterial.Sticker_Texture = Materials.Stickers.winter_2024
            set Winter2024StickerMaterial.UseSnowBase = 1.0
            if:
                TheSticker := NewStickerResult(0)?
                TheGlobals := Globals[GetSession()]
                StickerID := TheGlobals.GetStickerID["Winter2024"]
            then:
                TheSticker.SetMaterial(Winter2024StickerMaterial)
                spawn . PlayerOneBoard.AddSticker(Players(0), CPlayer, AvailStickerSlot, StickerID, TheSticker)

        if:
            WinterQuestSwitch.GetCurrentState[Players(1)]
            not HasCompletedWinter2024Quest[Players(1)] # if they havent already completed it
            CPlayer := Globals[GetSession()].CustomPlayerMap[Players(1)]
            AvailStickerSlot := PlayerTwoBoard.GetAvailableStickerSlot[]
            SetCompletedWinter2024Quest(Players(1))
        then:
            NewStickerResult := SpawnProp(DefaultStickerPropAsset, vector3{}, rotation{})
            Winter2024StickerMaterial := Materials.Stickers.MI_Sticker_Base_Inst_material{}
            set Winter2024StickerMaterial.Sticker_Texture = Materials.Stickers.winter_2024
            set Winter2024StickerMaterial.UseSnowBase = 1.0
            if:
                TheSticker := NewStickerResult(0)?
                TheGlobals := Globals[GetSession()]
                StickerID := TheGlobals.GetStickerID["Winter2024"]
            then:
                TheSticker.SetMaterial(Winter2024StickerMaterial)
                spawn . PlayerTwoBoard.AddSticker(Players(1), CPlayer, AvailStickerSlot, StickerID, TheSticker)



        Logger.Print("Game Is Ready")

        GameLoop(Players(0), Players(1))
            # GameLoop(P1, P2)
        # else:
        #     Print("🛑 DID NOT FIND 2 PLAYERS TO BEGIN")


    NoSecondPlayerFoundYet(MaybeAgent:?agent):void=
        HUDNoOpponentFound.Show()

        if(Agent := MaybeAgent?):
            # Agent.Respawn(vector3{X:=1787.0,Y:=-279.0,Z:=2611.0}, IdentityRotation())
            PortalForRematch.Teleport(Agent)


    GameLoop(AgentOne:agent, AgentTwo:agent)<suspends>:void=
        HUDNoOpponentFound.Hide()
        if:
            TheGlobals := Globals[GetSession()]
        then:
            set TheGlobals.CurrentTurn = GetRandomInt(0, 1) # Player 1 or 2
            set GameEverStarted = true
            # Choose the Random Cards for each player
            block:
                PlayerOneBoard.NewRandomTile()
                Sleep(0.001)
                PlayerTwoBoard.NewRandomTile()
                Sleep(0.001)

                PlayerOneBoard.SetOpponentTargetIndex(PlayerTwoBoard.CurrentTargetIndex)
                PlayerTwoBoard.SetOpponentTargetIndex(PlayerOneBoard.CurrentTargetIndex)


                Logger.Print("Awaiting Category Selection")

            set HasFoundOnlyOnePlayer = false

            PlayerOneBoard.CameraZoom1.RemoveFrom(AgentOne)
            PlayerOneBoard.CameraZoom2.RemoveFrom(AgentOne)
            PlayerTwoBoard.CameraZoom1.RemoveFrom(AgentTwo)
            PlayerTwoBoard.CameraZoom2.RemoveFrom(AgentTwo)
            Sleep(0.03)

            TheVotes := race:
                block:
                    MainPlayerCounter.CountSucceedsEvent.Await() # Only one player left
                    set HasFoundOnlyOnePlayer = true
                    (-9999,-9999) # arbitrary value to allow race to compile

                AwaitPlayerVotes() # Both players voted

            if(HasFoundOnlyOnePlayer?):
                Logger.Print("Only one player left")
                AllPlayers := GetPlayspace().GetPlayers()
                for(Player: AllPlayers):
                    # Player.Respawn(vector3{X:=1787.0,Y:=-279.0,Z:=2611.0}, IdentityRotation())
                    PortalForRematch.Teleport(Player)
                    HUDNoOpponentFound.Show()
                    HideEveryoneUI()
                    return
                    
            else if(TheVotes = (-9999,-9999)):
                Logger.Print("Something Majorly Wrong, restarting game")
                if:
                    RandomPlayer := GetPlayspace().GetPlayers()[0]
                then:
                    EndGame.Activate(RandomPlayer)

            Logger.Print("Both players voted - {TheVotes(0)}:{TheVotes(1)}")

            set TheGlobals.CurrentStyleIndex = CompareResult(TheVotes(0), TheVotes(1))
            set TheGlobals.CurrentStyle = GetCardStyleForIndex(TheGlobals.CurrentStyleIndex)

            Sleep(0.0)
                
                    

            HideEveryoneUI()

            HandleChooseStyle(TheGlobals.CurrentStyleIndex)
            GetAnalyticForStyleIndex(TheGlobals.CurrentStyleIndex, AgentOne) #sends for one only

            if(OGSet := TheGlobals.CurrentStyleIndex = 11):
                OGXPAccolade.Award(AgentOne) #awards to both players

            block:
                PlayerOneBoard.SetStyle(TheGlobals.CurrentStyle)
                PlayerTwoBoard.SetStyle(TheGlobals.CurrentStyle)
                

            sync:
                PlayerOneBoard.FlipBoard(?Instant:=true)
                PlayerTwoBoard.FlipBoard(?Instant:=true)

            set TheP1Index = PlayerOneBoard.CurrentTargetIndex
            set TheP2Index = PlayerTwoBoard.CurrentTargetIndex


            ############# MANUAL CHARACTER SELECTION #############
            if:
                PlayerOneBoard.IsManualCharacterSelection? or PlayerTwoBoard.IsManualCharacterSelection?
            then:

                # Scale To 1.0
                if:
                    NewScale := vector3{X:=1.0, Y:=1.0, Z:=1.0}
                    NewTransform := transform{Translation:=BarrierBlockPlayerViewParent.GetTransform().Translation, Rotation:=BarrierBlockPlayerViewParent.GetTransform().Rotation, Scale:=NewScale}
                    BarrierBlockPlayerViewParent.IsValid[]
                    BarrierBlockPlayerViewParent.TeleportTo[NewTransform]
                then:
                    Print("BarrierBlockPlayerViewParent Scaled to 1.0")
                else:
                    Logger.Print("BarrierBlockPlayerViewParent not valid - Trying to teleport")
                    Sleep(0.0)
                    if:
                        BarrierBlockPlayerViewParent.IsValid[]
                        BarrierBlockPlayerViewParent.TeleportTo[transform{Translation:=BarrierBlockPlayerViewParent.GetTransform().Translation, Rotation:=BarrierBlockPlayerViewParent.GetTransform().Rotation, Scale:=vector3{X:=1.0, Y:=1.0, Z:=1.0}}]
                    else:
                        Logger.Print("BarrierBlockPlayerViewParent not valid - Trying to teleport")
                        Sleep(0.0)
                        if:
                            BarrierBlockPlayerViewParent.IsValid[]
                            BarrierBlockPlayerViewParent.TeleportTo[transform{Translation:=BarrierBlockPlayerViewParent.GetTransform().Translation, Rotation:=BarrierBlockPlayerViewParent.GetTransform().Rotation, Scale:=vector3{X:=1.0, Y:=1.0, Z:=1.0}}]

                # Teleport both players to their platforms so they cant cheat
                PlayerOneBoard.Teleporter.Teleport(AgentOne)
                PlayerTwoBoard.Teleporter.Teleport(AgentTwo)


                TimerManualCharacterSelection.Start()
                race:
                    TimerManualCharacterSelection.SuccessEvent.Await()

                    sync:
                        CheckForManualCharacterSelection(PlayerOneBoard, AgentOne, TheGlobals, 0)
                        CheckForManualCharacterSelection(PlayerTwoBoard, AgentTwo, TheGlobals, 1)
                                
                TimerManualCharacterSelection.Reset()

                # if:
                #     CPlayerOne := PlayerOneBoard.Owner?
                #     CPlayerTwo := PlayerTwoBoard.Owner?
                # then:
                #     set CPlayerOne.LiveRayCastActive = false
                #     set CPlayerTwo.LiveRayCastActive = false

                set PlayerOneBoard.CurrentTargetIndex = TheP2Index
                set PlayerTwoBoard.CurrentTargetIndex = TheP1Index

                PlayerOneBoard.SetOpponentTargetIndex(TheP1Index)
                PlayerTwoBoard.SetOpponentTargetIndex(TheP2Index)

            else:
                AnaltyicCharacterSelectedRandom.Submit(AgentOne)
                AnaltyicCharacterSelectedRandom.Submit(AgentTwo)

            ############# END MANUAL CHARACTER SELECTION #############

            PeakAndSetYouTile(PlayerOneBoard, AgentOne, TheGlobals, ?P1Cover := true)

            PeakAndSetYouTile(PlayerTwoBoard, AgentTwo, TheGlobals, ?P1Cover:=false)

            # reassign Fire Subscriptions if they are not already assigned
            # if(not PlayerOneBoard.CustomInput.FireSubscription?):
            #     P1FireSub := PlayerOneBoard.CustomInput.Cursor_InputTrigger_Fire.PressedEvent.Subscribe(PlayerOneBoard.HandleCursorInputTriggerFire)
            #     set PlayerOneBoard.CustomInput.FireSubscription = option{P1FireSub}
            
            # if(not PlayerTwoBoard.CustomInput.FireSubscription?):
            #     P2FireSub := PlayerTwoBoard.CustomInput.Cursor_InputTrigger_Fire.PressedEvent.Subscribe(PlayerTwoBoard.HandleCursorInputTriggerFire)
                # set PlayerTwoBoard.CustomInput.FireSubscription = option{P2FireSub}
            

            Sleep(0.03)

            if:
                PlayerOneBoard.IsManualCharacterSelection? or PlayerTwoBoard.IsManualCharacterSelection?
            then:
                Sleep(0.97)
                # scale to 0.001
                if:
                    NewScale := vector3{X:=1.0, Y:=1.0, Z:=0.001}
                    NewTransform := transform{Translation:=BarrierBlockPlayerViewParent.GetTransform().Translation, Rotation:=BarrierBlockPlayerViewParent.GetTransform().Rotation, Scale:=NewScale}
                    BarrierBlockPlayerViewParent.IsValid[]
                    BarrierBlockPlayerViewParent.TeleportTo[NewTransform]
                then:
                    Print("BarrierBlockPlayerViewParent Scaled to 0.001")
                else:
                    Logger.Print("BarrierBlockPlayerViewParent not valid - Trying to teleport")
                    Sleep(0.0)
                    if:
                        BarrierBlockPlayerViewParent.IsValid[]
                        BarrierBlockPlayerViewParent.TeleportTo[transform{Translation:=BarrierBlockPlayerViewParent.GetTransform().Translation, Rotation:=BarrierBlockPlayerViewParent.GetTransform().Rotation, Scale:=vector3{X:=1.0, Y:=1.0, Z:=0.001}}]
                    else:
                        Logger.Print("BarrierBlockPlayerViewParent not valid - Trying to teleport")
                        Sleep(0.0)
                        if:
                            BarrierBlockPlayerViewParent.IsValid[]
                            BarrierBlockPlayerViewParent.TeleportTo[transform{Translation:=BarrierBlockPlayerViewParent.GetTransform().Translation, Rotation:=BarrierBlockPlayerViewParent.GetTransform().Rotation, Scale:=vector3{X:=1.0, Y:=1.0, Z:=0.001}}]

            spawn . HandleIndexCard()

            TimerTurnTime.Reset()
            TimerTurnTime.SetActiveDuration(PLAYER_TURN_TIME + 2.0)
            
            Sleep(0.0)        

            set TheGlobals.GameInProgress = true
                
        
            # Main Game Loop - Handle Turns
            race:
                PlayerOneBoard.AwaitWin()
                PlayerTwoBoard.AwaitWin()
                HandleTurns(TheGlobals)
                block:
                    loop:
                        if(GetPlayspace().GetPlayers().Length < 2):
                            Logger.Print("Only one player in the game")
                            break
                        Sleep(1.0)

            if(GetPlayspace().GetPlayers().Length < 2):
                Logger.Print("Only one player in the game")
                HUDNoOpponentFound.Show()

                # teleport remaining player to matchmaking portal area
                if:
                    RandomPlayer := GetPlayspace().GetPlayers()[0]
                    CPlayer := TheGlobals.CustomPlayerMap[RandomPlayer]
                    GameBoard := CPlayer.GameBoard
                then:
                    # RandomPlayer.Respawn(vector3{X:=1787.0,Y:=-279.0,Z:=2611.0}, IdentityRotation())
                    PortalForRematch.Teleport(RandomPlayer)
                    LevelSequenceTurnTime.SetPlaybackTime(0.0)
                    LevelSequenceTurnTime.Play()
                    LevelSequenceTurnTime.Stop()
                    GameBoard.HUD_YourTurn.Hide(RandomPlayer)
                    GameBoard.HUD_YourTurn_Centered.Hide(RandomPlayer)
                    GameBoard.HUD_OtherPlayersTurn.Hide(RandomPlayer)
                    GameBoard.HUD_OtherPlayersTurn_Centered.Hide(RandomPlayer)
                    HideEveryoneUI()
                    return

                return

            PlayerOneBoard.DisableInputs()
            PlayerTwoBoard.DisableInputs()

            PlayerOneBoard.RemoveFireSubscription()
            PlayerTwoBoard.RemoveFireSubscription()

            LevelSequenceTurnTime.SetPlaybackTime(0.0)
            LevelSequenceTurnTime.Play()
            LevelSequenceTurnTime.Stop()

            if:
                CPlayerOne := PlayerOneBoard.Owner?
                CPlayerTwo := PlayerTwoBoard.Owner?
            then:
                set CPlayerOne.GuessUI.IsTimerActive = false
                set CPlayerTwo.GuessUI.IsTimerActive = false
                # set CPlayerOne.LiveRayCastActive = false
                # set CPlayerTwo.LiveRayCastActive = false
                Sleep(0.0)
                PlayerOneBoard.ResetTileHighlight(CPlayerOne)
                PlayerTwoBoard.ResetTileHighlight(CPlayerTwo)
                PlayerOneBoard.ResetButtonHighlights()
                PlayerTwoBoard.ResetButtonHighlights()
                Sleep(0.0)
                CPlayerOne.GuessUI.UpdateReticle(_UI.UI_Textures.transparent_square, vector2{X:=52.0, Y:=52.0})
                CPlayerTwo.GuessUI.UpdateReticle(_UI.UI_Textures.transparent_square, vector2{X:=52.0, Y:=52.0})


            # if(FireSub := PlayerOneBoard.CustomInput.FireSubscription?):                
            #     FireSub.Cancel()
            #     set PlayerOneBoard.CustomInput.FireSubscription = false
            
            # if(FireSub := PlayerTwoBoard.CustomInput.FireSubscription?):
            #     FireSub.Cancel()
            #     set PlayerTwoBoard.CustomInput.FireSubscription = false
            

            sync:
                PlayerOneBoard.YouTile.Reveal()
                PlayerTwoBoard.YouTile.Reveal()


            sync:
                PlayerOneBoard.HideIndexIfShown()
                PlayerTwoBoard.HideIndexIfShown()

            if:
                TheWinner := TheGlobals.Winner?
            then:
               ScoreManager.SetScoreAward(1)
               ScoreManager.Activate(TheWinner)
               GoldGranter.GrantItem(TheWinner)

               TrackerLifeTimeWins.Increment(TheWinner)

               RestartGame(TheGlobals)

            else:
                Logger.Print("No Winner")

            Sleep(0.0)
            if:
                CPlayerOne := PlayerOneBoard.Owner?
                CPlayerTwo := PlayerTwoBoard.Owner?
            then:
                CPlayerOne.GuessUI.UpdateReticle(_UI.UI_Textures.T_Reticle_Default, vector2{X:=52.0, Y:=52.0})
                CPlayerTwo.GuessUI.UpdateReticle(_UI.UI_Textures.T_Reticle_Default, vector2{X:=52.0, Y:=52.0})


            PlayerOneBoard.HUD_YourTurn.Show(AgentOne,StringToMessage(""), ?DisplayTime:=0.5)
            PlayerOneBoard.HUD_YourTurn.Hide(AgentOne)


            PlayerTwoBoard.HUD_YourTurn.Show(AgentTwo,StringToMessage(""), ?DisplayTime:=0.5)
            PlayerTwoBoard.HUD_YourTurn.Hide(AgentTwo)
            

    CheckForManualCharacterSelection(PlayerBoard:game_board, Agent:agent, TheGlobals:globals, ThePIndex:int)<suspends>:void=
        if(PlayerBoard.IsManualCharacterSelection?):
            Print("Player Manually Choosing Character")
            PlayerBoard.YouTile.SetStyle(TheGlobals.CurrentStyle)
            HUDChooseYourCharacter.Show(Agent)
            # PlayerBoard.Teleporter.Teleport(Agent)
            PIndex := PlayerBoard.AllowManualCharacterSelection()
            
            if(ThePIndex = 0):
                set TheP1Index = PIndex
            else:
                set TheP2Index = PIndex


            AnaltyicCharacterSelectedManual.Submit(Agent)

        else:                                
            PlayerBoard.SetOpponentTargetIndex(ThePIndex)
            PeakAndSetYouTile(PlayerBoard, Agent, TheGlobals, ?P1Cover := false)
            # HUDChooseYourCharacter.Show(Agent)

            AnaltyicCharacterSelectedRandom.Submit(Agent)

            Sleep(0.1)
        # Sleep(1.0)
        HUDChooseYourCharacter.Hide(Agent)

    
    PeakAndSetYouTile(PlayerBoard:game_board, Agent:agent, TheGlobals:globals, ?P1Cover : logic = false)<suspends>:void=
        
        if(P1Cover?):
            set P1CoverMaterial.CardImageGrid = GetCardGridTextureForCardStyle(TheGlobals.CurrentStyle)
            PlayerBoard.YouTile.SetCard(TheGlobals.CurrentStyle, PlayerBoard.OpponentTargetIndex, PlayerBoard, ?P1Cover := true)
            PlayerBoard.YouTile.Peak(Agent)
        else:
            set P2CoverMaterial.CardImageGrid = GetCardGridTextureForCardStyle(TheGlobals.CurrentStyle)
            PlayerBoard.YouTile.SetCard(TheGlobals.CurrentStyle, PlayerBoard.OpponentTargetIndex, PlayerBoard, ?P1Cover := false)
            PlayerBoard.YouTile.Peak(Agent)

    AwaitPlayerVotes()<suspends>:tuple(int,int)=
        Logger.Print("Awaiting Player Votes")
        var  P1Vote : int = -1
        var  P2Vote : int = -1

        TimerTurnTime.Reset()
        TimerTurnTime.SetActiveDuration(90.0)

        sync:
            ShowCategorySelection(PlayerOneBoard)
            ShowCategorySelection(PlayerTwoBoard)

        
        TimerTurnTime.Start()


        TimerTurnTime.SetActiveDuration(90.0)
        Logger.Print("Timer Started : {TimerTurnTime.GetActiveDuration()}")

        HUDTimerSelectionScreen.Show()

        Votes := sync:
            AwaitStyleSelection(PlayerOneBoard)
            AwaitStyleSelection(PlayerTwoBoard)

        Logger.Print("Both Players Voted or Timed Out")

        if(Votes(0) = Votes(1)):
            Logger.Print("Both players voted the same - {Votes(0)} - {Votes(1)}")
            return (Votes(0), Votes(1))

        Logger.Print("Different Votes")

        QuickVoteTime := 25.0
        if:
            TimerTurnTime.GetActiveDuration() > QuickVoteTime
        then:
            TimerTurnTime.SetActiveDuration(QuickVoteTime)
            TimerTurnTime.Start()
            race:
                # if votes ever match return the votes
                loop:
                    if:
                        P1 := PlayerOneBoard.Owner?.GetVote[]
                        P2 := PlayerTwoBoard.Owner?.GetVote[]
                        P1 = P2
                    then:
                        set P1Vote = P1
                        set P2Vote = P2
                        Logger.Print("Both votes matche")
                        return (P1Vote, P2Vote)
                    Sleep(0.5)
                Sleep(QuickVoteTime)

        Logger.Print("Both players voted differently and time is up")

        if:
            CPlayerOne := PlayerOneBoard.Owner?
            CPlayerTwo := PlayerTwoBoard.Owner?
            P1 := CPlayerOne.GetVote[]
            P2 := CPlayerTwo.GetVote[]
        then:
            set P1Vote = P1
            set P2Vote = P2
            set CPlayerOne.GuessUI.IsTimerActive = false
            set CPlayerTwo.GuessUI.IsTimerActive = false

        return (P1Vote, P2Vote)



            


    AwaitStyleSelection(GameBoard:game_board)<suspends>:int=
        var TempVote : int = -1
        if:
            CPlayer := GameBoard.Owner?
            Player := CPlayer.Player
        then:
            Vote := CPlayer.GuessUI.AwaitStyleSelection(Player, CPlayer, VotedSound)
            set TempVote = Vote

        return TempVote

    ShowCategorySelection(GameBoard:game_board)<suspends>:void=
        if:
            CPlayer := GameBoard.Owner?
            Player := CPlayer.Player
        then:
            CPlayer.GuessUI.InitStyleSelectionUI(Player, CPlayer)

    InitDailyReward()<suspends>:void=
        loop:
            Agent := DailyReward.ClaimButton.InteractedWithEvent.Await()
            HandleButtonDailyReward(Agent)
            Sleep(0.0)


    HandleButtonTestDate(Agent:agent):void=
        CurrentDay := GetCurrentDateTime()
        Logger.Print("CivilDate: {CurrentDay.String()} - Days: {CalculatedTotalDaysSinceEpoch(CurrentDay)}")

        if:
            Player := player[Agent]
            SourceData := PlayerSavedDataMap[Player]
        then:
            CurrentDate := GetCurrentDateTime()
            FirstDayPlayed := SourceData.FirstDayPlayed
            LastDayPlayed := SourceData.LastDayPlayed
    
            FirstDay := "FirstDayPlayed: {FirstDayPlayed.String()} Seconds: {CalculatedTotalSecondsSinceEpoch(FirstDayPlayed)} Days: {CalculatedTotalDaysSinceEpoch(FirstDayPlayed)}"
            LastDay := "LastDayPlayed: {LastDayPlayed.String()} Seconds: {CalculatedTotalSecondsSinceEpoch(LastDayPlayed)} Days: {CalculatedTotalDaysSinceEpoch(LastDayPlayed)}"
            Current := "Current: {CurrentDate.String()} Seconds: {CalculatedTotalSecondsSinceEpoch(CurrentDate)} Days: {CalculatedTotalDaysSinceEpoch(CurrentDate)}"
            DaysBetween := "Days between: {DaysBetweenDates(LastDayPlayed, CurrentDate)}"
            SecondsBetween := "Seconds between: {SecondsBetweenDates(LastDayPlayed, CurrentDate)}"
    
            FiveLines := FiveLineMessage(FirstDay, LastDay, Current, DaysBetween, SecondsBetween)
            DailyReward.DebugBillboard.SetText(FiveLines)
            


    HandleButtonDailyReward(Agent:agent):void=
        Logger.Print("Daily Reward Button Pressed")

        if:
            LastDayPlayed := GetLastDayPlayed[Agent]
        then:
            DailyReward.DebugBillboard.SetText(
                FiveLineMessage(
                    "Last Day Played: {LastDayPlayed.String()} Days:{CalculatedTotalDaysSinceEpoch(LastDayPlayed)}",
                    "Today: {GetCurrentDateTime().String()} Days:{CalculatedTotalDaysSinceEpoch(GetCurrentDateTime())}", 
                    "Days between: {DaysBetweenDates(LastDayPlayed, GetCurrentDateTime())}",
                    "Consecutive Days: {GetCurrentConsecutiveDaysPlayed(Agent)}",
                    "Most Consecutive Days: {GetMostConsecutiveDaysPlayed(Agent)}"
                )
            )

        if:
            AllowDailyReward[Agent]
        then:
            Logger.Print("Allowing daily reward")
            ClaimDailyReward(Agent)
            TimeUntil := TimeUntilNextUTCMidnight(GetCurrentDateTime())
            DailyReward.Accolade.Award(Agent)
            DailyReward.HudReward.Show(Agent, MakeHrsMinsMessage(TimeUntil(0), TimeUntil(1)))
            DailyReward.Granter.GrantItem(Agent)

        else:
            Logger.Print("Not allowing daily reward")
            TimeUntil := TimeUntilNextUTCMidnight(GetCurrentDateTime())
            Consec := GetCurrentConsecutiveDaysPlayed(Agent)
            ConsecRecord :=  GetMostConsecutiveDaysPlayed(Agent)
            
            Logger.Print("Time until next daily reward {TimeUntil(0)}hr {TimeUntil(1)}min")
            #first line : 1hrs 2mins
            #second line (current consecutive days) : 3 days
            #third line (max consecutive days) : 5 days
            HoursString := "{TimeUntil(0)}hrs {TimeUntil(1)}mins"
            ConsecString := "{Consec} Days"
            ConsecRecordString := "{ConsecRecord} Days"
            DailyReward.HudFail.Show(Agent, ThreeLineMessage(HoursString, ConsecString, ConsecRecordString))


    CheckForRematchBonus(Agent:agent):void=
        if:
            FortChar := Agent.GetFortCharacter[]
            FortChar.IsActive[]
            CheckForPlayerSavedData[Agent]
            Player := player[Agent]
            SourceData := PlayerSavedDataMap[Player]
        then:
            LastSessionTime := SourceData.LastSessionTime

            TheMinutesBetweenDates := MinutesBetweenDates(LastSessionTime, GetCurrentDateTime())
            if:
                TheMinutesBetweenDates <= 15
            then:
                Logger.Print("Player has played within 15 minutes")
                HUDGeneric.Show(Agent, MessageRematchBonusGold_EN)
                GoldGranter.GrantItem(Agent)
                AccoladeRematchBonus.Award(Agent)
                
            

    HandlePlayerRemoved(Agent:agent):void=
        Logger.Print("Player Quit")
        CurrentDate := GetCurrentDateTime()
        SetLastSessionTime(Agent, CurrentDate)

    # HandlePlayerJoinedSession(Agent:agent):void=
    #     Logger.Print("Player Joined")
    #     spawn . EmergencyResetRoundAfterJoinInProgess(Agent)

    # EmergencyResetRoundAfterJoinInProgess(Agent:agent)<suspends>:void=
    #     Sleep(5.0)
    #     if(not GameEverStarted?):
    #         EndGame.Activate(Agent)


    # NEEDS TO BE UPDATED WHEN NEW STYLES (SETS) ARE ADDED ######################################################################################################
    InitQuestionBank()<suspends>:void=
        QuestionBankDevice.InitFNQuestions()
        Sleep(0.01)
        # QuestionBankDevice.InitClassicQuestions()
        # Sleep(0.01)
        QuestionBankDevice.InitAnimalQuestions()
        Sleep(0.01)
        QuestionBankDevice.InitIconQuestions()
        Sleep(0.01)
        QuestionBankDevice.InitUtilitiesQuestions()
        Sleep(0.01)
        # QuestionBankDevice.InitMarvelQuestions()
        # Sleep(0.01)
        # QuestionBankDevice.InitPOIQuestions()
        # Sleep(0.01)
        QuestionBankDevice.InitLegendsQuestions()
        Sleep(0.01)
        QuestionBankDevice.InitFoodQuestions()
        Sleep(0.01)
        QuestionBankDevice.InitStarQuestions()
        Sleep(0.01)
        QuestionBankDevice.InitSpookyQuestions()
        Sleep(0.01)
        QuestionBankDevice.InitOGQuestions()
        Sleep(0.01)
        QuestionBankDevice.InitAnimeQuestions()
        Sleep(0.01)
        QuestionBankDevice.InitWinterQuestions()
        Sleep(0.01)
        QuestionBankDevice.InitWeaponsQuestions()
        Sleep(0.01)
        QuestionBankDevice.InitPickaxeQuestions()
        Sleep(0.01)
        QuestionBankDevice.InitVehicleQuestions()
        Sleep(0.01)
        QuestionBankDevice.InitMoviesTVQuestions()
        Sleep(0.01)
        QuestionBankDevice.InitChapter5Questions()
        Sleep(0.01)
        QuestionBankDevice.InitLockersQuestions()
        Sleep(0.01)
        QuestionBankDevice.InitBackblingQuestions()
        Sleep(0.01)
        QuestionBankDevice.InitEmoteQuestions()
        Sleep(0.01)
        QuestionBankDevice.InitMarvel1Questions()
        Sleep(0.01)
        QuestionBankDevice.InitMarvel2Questions()
        Sleep(0.01)
        QuestionBankDevice.InitUmbrellasQuestions()
        Sleep(0.01)
        QuestionBankDevice.InitPeelyFishstickQuestions()
        Sleep(0.01)
        QuestionBankDevice.InitRemixQuestions()
        Sleep(0.01)
        QuestionBankDevice.InitFireIceQuestions()
        Sleep(0.01)
        QuestionBankDevice.InitDCQuestions()
        Sleep(0.01)
        QuestionBankDevice.InitChapter6Questions()


        set QuestionBankDevice.HUDResponseNo = HUDResponseNo
        set QuestionBankDevice.HUDResponseYes = HUDResponseYes
        set QuestionBankDevice.HUDResponseMaybe = HUDResponseMaybe
        set QuestionBankDevice.AnalyticQuestionAsked = AnalyticQuestionAsked
        set QuestionBankDevice.AnalyticQuestionAnswered = AnalyticQuestionAnswered



    HideEveryoneUI():void=
        Logger.Print("HideEveryoneUI")
        if:
            TheGlobals := Globals[GetSession()]
        then:
            for (Player : TheGlobals.AllPlayers):
                if:
                    CPlayer := TheGlobals.CustomPlayerMap[Player]
                    PlayerUI := GetPlayerUI[Player]
                then:

                    # Fully remove to disable input
                    if(MaybeVoteUI := CPlayer.GuessUI.MaybeVoteUI?):
                        PlayerUI.RemoveWidget(MaybeVoteUI)

                    if(MaybeStyleSelectionUI := CPlayer.GuessUI.MaybeStyleSelectionUI?):
                        PlayerUI.RemoveWidget(MaybeStyleSelectionUI)

                    # Hide we'll use later
                    if(MaybeYourTurnUI := CPlayer.GuessUI.MaybeYourTurnUI?):
                        MaybeYourTurnUI.SetVisibility(widget_visibility.Hidden)

                    if(MaybeQuestionUI := CPlayer.GuessUI.MaybeQuestionUI?):
                        MaybeQuestionUI.SetVisibility(widget_visibility.Hidden)

                    if(MaybeOtherPlayersTurnUI := CPlayer.GuessUI.MaybeOtherPlayersTurnUI?):
                        MaybeOtherPlayersTurnUI.SetVisibility(widget_visibility.Hidden)

                    if(MaybeShootToGuessUI := CPlayer.GuessUI.MaybeShootToGuessUI?):
                        MaybeShootToGuessUI.SetVisibility(widget_visibility.Hidden)

                    if(MaybeCorrectIncorrectUI := CPlayer.GuessUI.MaybeCorrectIncorrectUI?):
                        MaybeCorrectIncorrectUI.SetVisibility(widget_visibility.Hidden)



    CompareResult(P1Vote:int, P2Vote:int):int=

        if(P1Vote = -1):
            return P2Vote
        else if(P2Vote = -1):
            return P1Vote
        else if:
            P1Vote = P2Vote
        then:
            Logger.Print("Agree")
            return P1Vote
        else:
            Logger.Print("Disagree, Randomizing")
            return CoinFlipVotes(P1Vote, P2Vote)

    CoinFlipVotes(P1Vote:int, P2Vote:int):int=
        RandInt := GetRandomInt(0,1)
        case (RandInt):
            0 => 
                Logger.Print("Heads")
                return P1Vote
            1 => 
                Logger.Print("Tails")
                return P2Vote
            _ =>
                Logger.Print("Coin Flip Failed, falling back to P1")
                return P1Vote


    # NEEDS TO BE UPDATED WHEN NEW STYLES (SETS) ARE ADDED ######################################################################################################
    HandleChooseStyle(Index:int):void=
        Logger.Print("Choose Card Style: {Index}")
        case(Index):
            1 => spawn{QuestionBankDevice.Init(card_style.Classic)}
            2 => spawn{QuestionBankDevice.Init(card_style.Icon)}
            3 => spawn{QuestionBankDevice.Init(card_style.Animal)}
            4 => spawn{QuestionBankDevice.Init(card_style.Utilities)}
            5 => spawn{QuestionBankDevice.Init(card_style.Marvel)}
            6 => spawn{QuestionBankDevice.Init(card_style.POI)}
            7 => spawn{QuestionBankDevice.Init(card_style.Legends)}
            8 => spawn{QuestionBankDevice.Init(card_style.Food)}
            9 => spawn{QuestionBankDevice.Init(card_style.Star)}

            10 => spawn{QuestionBankDevice.Init(card_style.Spooky)}

            11 => spawn{QuestionBankDevice.Init(card_style.OG)}
            12 => spawn{QuestionBankDevice.Init(card_style.Anime)}
            13 => spawn{QuestionBankDevice.Init(card_style.Winter)}
            14 => spawn{QuestionBankDevice.Init(card_style.Weapons)}
            15 => spawn{QuestionBankDevice.Init(card_style.Pickaxes)}
            16 => spawn{QuestionBankDevice.Init(card_style.Vehicles)}
            17 => spawn{QuestionBankDevice.Init(card_style.MoviesTV)}
            18 => spawn{QuestionBankDevice.Init(card_style.Blocks)}
            19 => spawn{QuestionBankDevice.Init(card_style.Chapter5)}
            20 => spawn{QuestionBankDevice.Init(card_style.Lockers)}
            21 => spawn{QuestionBankDevice.Init(card_style.Backblings)}
            22 => spawn{QuestionBankDevice.Init(card_style.Emotes)}
            23 => spawn{QuestionBankDevice.Init(card_style.Marvel1)}
            24 => spawn{QuestionBankDevice.Init(card_style.Marvel2)}
            25 => spawn{QuestionBankDevice.Init(card_style.Umbrellas)}
            26 => spawn{QuestionBankDevice.Init(card_style.PeelyFishstick)}
            27 => spawn{QuestionBankDevice.Init(card_style.Remix)}
            28 => spawn{QuestionBankDevice.Init(card_style.FireIce)}
            29 => spawn{QuestionBankDevice.Init(card_style.DC)}
            30 => spawn{QuestionBankDevice.Init(card_style.Chapter6)}

            _ => 
                Logger.Print("Invalid Style Index: {Index}")
                spawn{QuestionBankDevice.Init(card_style.FN)}


    GetAnalyticForStyleIndex(Index:int, Agent:agent):void=
        if:
            Analytics := Analytics_Styles[Index]
        then:
            Analytics.Submit(Agent)
            Logger.Print("Style Analytics Submitted")
        else:
            Logger.Print("Style Analytics Failed to Submit")


    # NEEDS TO BE UPDATED WHEN NEW STYLES (SETS) ARE ADDED (If has index card) ######################################################################################################
    HandleIndexCard()<suspends>:void=
        set PlayerOneBoard.StyleHasIndexCard = false
        set PlayerTwoBoard.StyleHasIndexCard = false

        if:
            PlayerOneBoard.IndexCard.Card.IsValid[]
            PlayerTwoBoard.IndexCard.Card.IsValid[]
            P1CardRotation := PlayerOneBoard.IndexCard.Card.GetTransform().Rotation
            P2CardRotation := PlayerTwoBoard.IndexCard.Card.GetTransform().Rotation
            
            PlayerOneBoard.IndexCard.Card.TeleportTo[vector3{X:=-5344.0,Y:=1205.0,Z:=2471.0}, P1CardRotation]
            PlayerTwoBoard.IndexCard.Card.TeleportTo[vector3{X:=-1256.0,Y:=-1194.0,Z:=2471.0}, P2CardRotation]
        then:
            Logger.Print("Index Card's Reset")

        # Utilities Style = 4
        # Food Style = 8
        # Vehicles Style = 16
        if:
            TheGlobals := Globals[GetSession()]
            TheGlobals.CurrentStyleIndex = 4 or TheGlobals.CurrentStyleIndex = 8 or TheGlobals.CurrentStyleIndex = 16
        then:
            set PlayerOneBoard.StyleHasIndexCard = true
            set PlayerTwoBoard.StyleHasIndexCard = true

            IndexCardMaterial := GetIndexCardMaterialForStyleIndex(TheGlobals.CurrentStyleIndex)

            if:
                PlayerOneBoard.IndexCard.Card.IsValid[]
                PlayerTwoBoard.IndexCard.Card.IsValid[]

                P1CardRotation := PlayerOneBoard.IndexCard.Card.GetTransform().Rotation
                P2CardRotation := PlayerTwoBoard.IndexCard.Card.GetTransform().Rotation

            then:
                PlayerOneBoard.IndexCard.Card.SetMaterial(IndexCardMaterial, ?Index:=1)
                PlayerTwoBoard.IndexCard.Card.SetMaterial(IndexCardMaterial, ?Index:=1)

                PlayerOneBoard.IndexCard.Card.MoveTo(PlayerOneBoard.IndexCard.Card.GetTransform().Translation + vector3{X:=-128.0}, P1CardRotation, 0.15)
                PlayerTwoBoard.IndexCard.Card.MoveTo(PlayerTwoBoard.IndexCard.Card.GetTransform().Translation + vector3{X:=128.0}, P2CardRotation, 0.15)
                Logger.Print("Index Card's Shown")
                Sleep(1.0)
                PlayerOneBoard.IndexCard.SlideOut(1.0)
                PlayerTwoBoard.IndexCard.SlideOut(-1.0)
            else:
                Logger.Print("Index Card's Failed to move")




    HandleTurns(TheGlobals:globals)<suspends>:void=
        loop:
            set TheGlobals.OpponentAnswered = (false, false)

            if(not TheGlobals.GameInProgress?) then break

            CurrentBoard := if (TheGlobals.CurrentTurn = 0) then PlayerOneBoard else PlayerTwoBoard
            OpponentBoard := if (TheGlobals.CurrentTurn = 0) then PlayerTwoBoard else PlayerOneBoard

            if:
                CurrentCPlayer := CurrentBoard.Owner?
                OpponentCPlayer := OpponentBoard.Owner?
            then:

                set CurrentCPlayer.GameBoard.QuestionBankOpen = false
                set OpponentCPlayer.GameBoard.QuestionBankOpen = false

                set CurrentCPlayer.GameBoard.AwaitingResponse = false
                set OpponentCPlayer.GameBoard.AwaitingResponse = false

                set TheGlobals.CurrentPlayer = option{CurrentCPlayer.Player}

                CurrentBoard.EnableInputs()
                OpponentBoard.DisableInputs()


                # reassign Fire Subscriptions if they are not already assigned
                if(not CurrentBoard.CustomInput.FireSubscription?):
                    FireSub := CurrentBoard.CustomInput.Cursor_InputTrigger_Fire.PressedEvent.Subscribe(CurrentBoard.HandleCursorInputTriggerFire)
                    set CurrentBoard.CustomInput.FireSubscription = option{FireSub}

                if(not OpponentBoard.CustomInput.FireSubscription?):
                    FireSub := OpponentBoard.CustomInput.Cursor_InputTrigger_Fire.PressedEvent.Subscribe(OpponentBoard.HandleCursorInputTriggerFire)
                    set OpponentBoard.CustomInput.FireSubscription = option{FireSub}

                CurrentBoard.AwaitTurnPlayed(CurrentCPlayer)
                # race:
                #     CurrentBoard.AwaitTurnPlayed(CurrentCPlayer)
                #     block:
                #         # enables raycasting for card highlighting
                #         set CurrentCPlayer.LiveRayCastActive = true
                #         Print("Enabling live raycast")
                #         CurrentBoard.LiveRayCast(CurrentCPlayer.Player)
                #         Sleep(Inf) #this never wins the race


                set TheGlobals.OpponentAnswered = (false, false) # reset for next turn backup

                CurrentBoard.DisableInputs()
                # set CurrentCPlayer.LiveRayCastActive = false

                CurrentBoard.ResetTileHighlight(CurrentCPlayer)
                CurrentBoard.ResetButtonHighlights()

                Sleep(0.25)
                OpponentCPlayer.GuessUI.UpdateReticle(_UI.UI_Textures.T_Reticle_Default, vector2{X:=52.0, Y:=52.0})
                CurrentCPlayer.GuessUI.UpdateReticle(_UI.UI_Textures.T_Reticle_Default, vector2{X:=52.0, Y:=52.0})

                set TheGlobals.CurrentTurn = 1 - TheGlobals.CurrentTurn

                PrepareForNextTurn(CurrentCPlayer, OpponentCPlayer, CurrentBoard, OpponentBoard, TheGlobals)

            Sleep(0.1)





    PrepareForNextTurn(CurrentCPlayer:custom_player, OpponentCPlayer:custom_player, CurrentBoard:game_board, OpponentBoard:game_board, TheGlobals:globals)<suspends>:void=
        HandleTurnEnd(CurrentCPlayer, CurrentBoard, TheGlobals)
        HandleTurnEnd(OpponentCPlayer, OpponentBoard, TheGlobals)
        Sleep(0.1)
        CurrentCPlayer.GuessUI.UpdateReticle(_UI.UI_Textures.T_Reticle_Default, vector2{X:=52.0, Y:=52.0})
        OpponentCPlayer.GuessUI.UpdateReticle(_UI.UI_Textures.T_Reticle_Default, vector2{X:=52.0, Y:=52.0})

    HandleTurnEnd(CPlayer:custom_player, PlayerBoard:game_board, TheGlobals:globals)<suspends>:void=
        spawn . CPlayer.GuessUI.RemoveQuestionUI(CPlayer.Player)
        PlayerBoard.YesNoCamera.RemoveFromAll()
        set CPlayer.GuessUI.IsTimerActive = false
        set PlayerBoard.HoldingInput = false
        set PlayerBoard.AwaitingResponse = false
        set TheGlobals.OpponentAnswered = (false, false)
        TimerTurnTime.Reset()
        TimerTurnTime.SetActiveDuration(PLAYER_TURN_TIME + 4.0)
        CPlayer.GuessUI.UpdateTimerTexture(CPlayer.GuessUI.TimerBGTexture, CPlayer.GuessUI.TimerTimeTexture)



    RestartGame(TheGlobals:globals)<suspends>:void=
        set TheGlobals.GameInProgress = false
        set HasFoundOnlyOnePlayer = false
        Sleep(0.0)
        sync:
            PlayerOneBoard.Reset()
            PlayerTwoBoard.Reset()
        Sleep(0.0)

        if:
            PlayerOne := PlayerOneBoard.Owner?.Player
            PlayerTwo := PlayerTwoBoard.Owner?.Player
            PlayerOne.IsActive[]
            PlayerTwo.IsActive[]
            # PlayerOneAgent := agent[PlayerOne]
            # PlayerTwoAgent := agent[PlayerTwo]
            # PlayerOneFortChar := PlayerOneAgent.GetFortCharacter[]
            # PlayerTwoFortChar := PlayerTwoAgent.GetFortCharacter[]
            # PlayerOneFortChar.IsActive[]
            # PlayerTwoFortChar.IsActive[]
        then:
            GameLoop(PlayerOne, PlayerTwo)
        else:
            Logger.Print("Failed to restart game")

    # HandleEquipPointerTimeOut(Agent:agent):void=
    #     HandleEquipPointerResponse(Agent, 0)

    # HandleEquipPointerResponse(Agent:agent, Index:int):void=
    #     #equip and assign new default pointer
        
    #     # DefaultPointer.InstantUnlock(Agent)
    #     if:
    #         TheGlobals := Globals[GetSession()]
    #         ThePointer := TheGlobals.GetPointerData(Index+1)
    #         PointerVFX := TheGlobals.GetPointerVFX[Index+1]
    #     then:
    #         ThePointer.InstantUnlock(Agent)
    #         ThePointer.Equip(Agent, GranterPointers, PointerVFX, ItemRemoverPointers)
    #         # DefaultPointer.Assign(Agent, GranterPointers, PointerVFX)
    
    AssignPointer(Agent:agent)<suspends>:void=
        if(not HasPointerBeenAssigned[Agent]):
            PointerShop.Dialog.Show(Agent)
        else:
            #show the saved pointer
            if:
                CurrentPointerIndex := GetCurrentPointer[Agent]
                # CurrentPointer := GetPointerForIndex(CurrentPointerIndex)
                TheGlobals := Globals[GetSession()]
                CurrentPointer := TheGlobals.GetPointerData(CurrentPointerIndex)
                PointerVFX := PointerShop.PointersVFX[CurrentPointerIndex]
                PointerGranter := PointerShop.Granters[CurrentPointerIndex]
                PointerConditional := PointerShop.ConditionalButtons[CurrentPointerIndex]
                InvIcon := PointerShop.HUDInventoryIcons[CurrentPointerIndex]
            then:
                CurrentPointer.Equip(Agent, PointerGranter, PointerVFX, PointerConditional, InvIcon, PointerShop.Remover)
    

    # HandlePlayerOneSpawn()<suspends>:void=
    #     P1 := PlayerOneSpawnPad.SpawnedEvent.Await()
    #     PlayerOneBoard.InitNew(0, 1.0, P1, TimerTurnTime, TriggerRandomForKnockActivation, InputTriggerHackCamera, LevelSequenceTurnTime, DefaultStickerPropAsset)
    #     AssignPointer(P1)
    #     set SpawnedPlayers += array{P1}
        
    # HandlePlayerTwoSpawn()<suspends>:void=
    #     P2 := PlayerTwoSpawnPad.SpawnedEvent.Await()
    #     PlayerTwoBoard.InitNew(1, -1.0, P2, TimerTurnTime, TriggerRandomForKnockActivation, InputTriggerHackCamera, LevelSequenceTurnTime, DefaultStickerPropAsset)
    #     AssignPointer(P2)
    #     set SpawnedPlayers += array{P2}

    # HandlePlayerSpawnCounted(Agent:agent):void=
    #     # Playspace := GetPlayspace()
    #     # TeamCollection := Playspace.GetTeamCollection()

    #     # if(TeamIndex := TeamCollection.GetTeamIndex[Agent] = 0):
    #     #     spawn . ActuallySetupPlayer(Agent, TeamIndex, 1.0)
    #     # else if(TeamIndex := TeamCollection.GetTeamIndex[Agent] = 1):
    #     #     spawn . ActuallySetupPlayer(Agent, TeamIndex, -1.0)
    #     # else:
    #     #     Logger.Print("Player Spawned but no board available")

    #     if(not P1Spawned?):
    #         spawn . ActuallySetupPlayer(Agent, 0, 1.0)
    #     else:
    #         spawn . ActuallySetupPlayer(Agent, 1, -1.0)


    HandlePlayerSpawnCounted(Agent:agent, TeamIndex:int):void=
        Dir := if(TeamIndex = 0) then 1.0 else -1.0
        if:
            Player := player[Agent]
            TheGlobals := Globals[GetSession()]
            not TheGlobals.CustomPlayerMap[Player]
        then:
            spawn . ActuallySetupPlayer(Agent, TeamIndex, Dir)

    ActuallySetupPlayer(Agent:agent, TeamIndex:int, BoardDir:float)<suspends>:void=

        if(TeamIndex =  0):
            PlayerOneBoard.InitNew(TeamIndex, BoardDir, Agent, TimerTurnTime, TriggerRandomForKnockActivation, InputTriggerHackCamera, LevelSequenceTurnTime, DefaultStickerPropAsset)
            AssignPointer(Agent)
            set SpawnedPlayers += array{Agent}
            set P1Spawned = true
        else if(TeamIndex = 1):
            PlayerTwoBoard.InitNew(TeamIndex, BoardDir, Agent, TimerTurnTime, TriggerRandomForKnockActivation, InputTriggerHackCamera, LevelSequenceTurnTime, DefaultStickerPropAsset)
            AssignPointer(Agent)
            set SpawnedPlayers += array{Agent}
            set P2Spawned = true

    AwaitBothPlayersSpawned()<suspends>:void=
        loop:
            if:
                P1Spawned? and P2Spawned?
            then:
                break

            Sleep(0.1)