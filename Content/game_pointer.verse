using { /Verse.org/Simulation }
using { /Fortnite.com/Devices }
using { /Verse.org/Assets }

custom_ammo <public> := enum{WoodenBall, SuctionDart, FoamDisc}

custom_ammo_assets := class<concrete>():
    @editable WoodenBall : creative_prop_asset = DefaultCreativePropAsset
    @editable SuctionDart : creative_prop_asset = DefaultCreativePropAsset
    @editable FoamDisc : creative_prop_asset = DefaultCreativePropAsset

game_pointer <public>:= class():
	ID <public>: int
	Name<public> : string
	Desc <public>: string
	Cost <public>: int
    Icon <public>: texture
    Rarity <public>: int
    Ammo <public> : custom_ammo
    AmmoProp <public> : creative_prop_asset


    Equip(Agent:agent, Granter:item_granter_device, PointerVFX:visual_effect_powerup_device, ConditionalButton:conditional_button_device, InvIcon:hud_message_device, ItemRemover:item_remover_device):void=
        ItemRemover.Remove(Agent) # Remove any existing pointer
        # Granter.GrantItemIndex(Agent, ID) # Use if Epic fixes Equiping by ID
        Granter.GrantItem(Agent)
        PointerVFX.Pickup(Agent)
        InvIcon.Show(Agent)
        SetCurrentPointer(Agent, ID)

        if:
            TheGlobals := Globals[GetSession()]
            Player := player[Agent]
            CPlayer := TheGlobals.CustomPlayerMap[Player]
        then:
            set CPlayer.MaybePointerConditional = option{ConditionalButton}

        Print("Setting pointer to ID: {ID}")


    Remove(Agent:agent, ItemRemover:item_remover_device, TriggerRemoveVFX:trigger_device):void=
        ItemRemover.Remove(Agent)
        TriggerRemoveVFX.Trigger(Agent)
        SetCurrentPointer(Agent, 0)


    IsOwned(Agent:agent)<transacts><decides>:void=
        CheckForPlayerSavedData[Agent]
        Player := player[Agent]
        SourceData := PlayerSavedDataMap[Player]
        FoundMatchingGamePointer := SourceData.UnlockedPointers.Find[ID]

    InstantUnlock(Agent:agent):void=
        AddUnlockedPointer(Agent, ID)

    IsPurchasedAllowed(Agent:agent)<transacts><decides>:void=
        CheckForPlayerSavedData[Agent]
        Player := player[Agent]
        SourceData := PlayerSavedDataMap[Player]
        HasEnoughGold := SourceData.Gold >= Cost

    Purchase(Agent:agent)<transacts>:void=
        if:
            IsPurchasedAllowed[Agent]
        then:
            AddUnlockedPointer(Agent, ID)
            DeductGold(Agent, Cost)