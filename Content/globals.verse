using { /Verse.org/Simulation }
using { /Verse.org/Colors }
using { /UnrealEngine.com/Temporary/SpatialMath }
using { /Fortnite.com/Devices }
using { /Fortnite.com/Characters }

SELECTION_SCREEN_TIME : int = 90

PLAYER_TURN_TIME : float = 90.0

PLAYER_RESPONSE_TIME : int = 15

UI_Yellow : color = MakeColorFromSRGBValues(247, 255, 25)
UI_Pink : color = MakeColorFromSRGBValues(237, 15, 185)

DEVICE_RETURN_LOCATION : vector3 = vector3{X:=0.0, Y:=0.0, Z:=-1000.0}


Cards <public> := module:
    Materials <public> := module:
        YouCards <public> := module:
        CardStyles <public> := module:
        IndexCards <public> := module:
    Textures <public> := module:
        CardGrids <public> := module:
        UI <public> := module:
            
_UI <public> := module:
    UI_Textures <public> := module:
        Numbers <public> := module:
    UI_Materials <public> := module:

Materials <public> := module:
    BoardDesigner <public> := module:
        Boards <public> := module:
        CardBacks <public> := module:
        Platforms <public> := module:
        YouStands <public> := module:
    ClayMaterial <public> := module:
        Material_Clay <public> := module:
        Texture_Clay <public> := module:
    Stickers <public> := module:
        
        

NEW_ASSETS <public> := module:
    Materials <public> := module:
    Textures <public> := module:
    Meshess <public> := module:
        Cards <public> := module:

FlyingTargets <public> := module:
    Materials <public> := module:
    Textures <public> := module:
    Meshes <public> := module:
    VFX <public> := module:


Textures <public> := module:


    

globals := class():

    GameManager : game_manager

    DefaultStickerPropAsset : creative_prop_asset

    var AllPlayers : []player = array{}
    var CustomPlayerMap : [player]custom_player = map{}

    var GameInProgress : logic = false
    var CurrentTurn : int = 0
    var OpponentAnswered : tuple(logic,logic) = (false,false) 
    var CurrentPlayer : ?player = false

    var CurrentStyle : card_style = card_style.FN
    var CurrentStyleIndex : int = 0
    var PlayerVoteIndex : [player]int = map{}

    var Winner : ?player = false

    var TheQuestionBank : question_bank_device

    var AllPointers : []game_pointer = array{}

    
    StickerIDs : [string]int = map:
        "Winter2024" => 0
        "Targets10000" => 1

    GetStickerID(Name:string)<transacts><decides>:int=
        StickerIDs[Name]
    
    InitPointers(AmmoPropAssets:custom_ammo_assets):void=
        set AllPointers += array:
            game_pointer:
                ID := 0
                Name := "Point Plank"
                Desc := "My Dad made this out of some old table."
                Cost := 0
                Icon := _UI.UI_Textures.T_Reticle_Default
                Rarity := 1
                Ammo := custom_ammo.WoodenBall
                AmmoProp := AmmoPropAssets.WoodenBall
        
        set AllPointers += array:
            game_pointer:
                ID := 1
                Name := "Who Shooter"
                Desc := "Do these suction cups even work?!"
                Cost := 0
                Icon := _UI.UI_Textures.T_Reticle_Default
                Rarity := 1
                Ammo := custom_ammo.SuctionDart
                AmmoProp := AmmoPropAssets.SuctionDart

        set AllPointers += array:
            game_pointer:
                ID := 2
                Name := "Foam Disc Launcher"
                Desc := "This thing is a blast!"
                Cost := 0
                Icon := _UI.UI_Textures.T_Reticle_Default
                Rarity := 1
                Ammo := custom_ammo.FoamDisc
                AmmoProp := AmmoPropAssets.FoamDisc
    
    
    GetPointerData<public>(ID:int)<transacts>:game_pointer=
        for:
            Pointer : AllPointers
            Pointer.ID = ID
        do:
            return Pointer
    
        Print("Pointer not found, returning default")
        return game_pointer:
            ID := 0
            Name := "Point Plank"
            Desc := "My Dad made this out of some old table."
            Cost := 0
            Icon := _UI.UI_Textures.T_Reticle_Default
            Rarity := 0
            Ammo := custom_ammo.WoodenBall
            AmmoProp := DefaultCreativePropAsset
       
    
    GetPointerVFX(Index:int)<transacts><decides>:visual_effect_powerup_device=
        case(Index-1):
            0 => return GameManager.PointerShop.PointersVFX[0]
            1 => return GameManager.PointerShop.PointersVFX[1]
            2 => return GameManager.PointerShop.PointersVFX[2]

            _ => return GameManager.PointerShop.PointersVFX[0]

    # var MaterialStyles : []card_style_params = array:
    #     card_style_params{CardStyle:=card_style.FN, CardMaterial:=Cards.Materials.MI_Card_Style_FN, FrameMaterial:=Materials.MI_P1_Frame, BoardMaterial:=Materials.M_Board_Master_Pink},
    #     card_style_params{CardStyle:=card_style.Classic, CardMaterial:=Cards.Materials.MI_Card_Style_Classic, FrameMaterial:=Materials.M_Board_Master_Pink, BoardMaterial:=Materials.MI_P1_Frame},
    #     card_style_params{CardStyle:=card_style.Icon, CardMaterial:=Cards.Materials.MI_Card_Style_Icon, FrameMaterial:=Materials.MI_P1_Frame, BoardMaterial:=Materials.M_Board_Master_Pink},
    #     card_style_params{CardStyle:=card_style.Animal, CardMaterial:=Cards.Materials.MI_Card_Style_Animal, FrameMaterial:=Materials.MI_P1_Frame, BoardMaterial:=Materials.M_Board_Master_Pink},
    #     card_style_params{CardStyle:=card_style.Utilities, CardMaterial:=Cards.Materials.MI_Card_Style_Utilities, FrameMaterial:=Materials.MI_P1_Frame, BoardMaterial:=Materials.M_Board_Master_Pink},
    #     card_style_params{CardStyle:=card_style.Marvel, CardMaterial:=Cards.Materials.MI_Card_Style_Marvel, FrameMaterial:=Materials.MI_P1_Frame, BoardMaterial:=Materials.M_Board_Master_Pink},
    #     card_style_params{CardStyle:=card_style.POI, CardMaterial:=Cards.Materials.MI_Card_Style_POI, FrameMaterial:=Materials.MI_P1_Frame, BoardMaterial:=Materials.M_Board_Master_Pink},
    #     card_style_params{CardStyle:=card_style.Legends, CardMaterial:=Cards.Materials.MI_Card_Style_Legends, FrameMaterial:=Materials.MI_P1_Frame, BoardMaterial:=Materials.M_Board_Master_Pink},
    #     card_style_params{CardStyle:=card_style.Food, CardMaterial:=Cards.Materials.MI_Card_Style_Food, FrameMaterial:=Materials.MI_P1_Frame, BoardMaterial:=Materials.M_Board_Master_Pink},
    #     card_style_params{CardStyle:=card_style.Star, CardMaterial:=Cards.Materials.MI_Card_Style_Star, FrameMaterial:=Materials.MI_P1_Frame, BoardMaterial:=Materials.M_Board_Master_Pink},
    #     card_style_params{CardStyle:=card_style.Spooky, CardMaterial:=Cards.Materials.MI_Card_Style_FN, FrameMaterial:=Materials.MI_P1_Frame, BoardMaterial:=Materials.M_Board_Master_Pink},
    #     card_style_params{CardStyle:=card_style.OG, CardMaterial:=Cards.Materials.MI_Card_Style_OG, FrameMaterial:=Materials.MI_P1_Frame, BoardMaterial:=Materials.M_Board_Master_Pink},
    #     card_style_params{CardStyle:=card_style.Anime, CardMaterial:=Cards.Materials.MI_Card_Style_Anime, FrameMaterial:=Materials.MI_P1_Frame, BoardMaterial:=Materials.M_Board_Master_Pink},
    #     card_style_params{CardStyle:=card_style.Winter, CardMaterial:=Cards.Materials.MI_Card_Style_Winter, FrameMaterial:=Materials.MI_P1_Frame, BoardMaterial:=Materials.M_Board_Master_Pink},
    #     card_style_params{CardStyle:=card_style.Weapons, CardMaterial:=Cards.Materials.MI_Card_Style_Weapons, FrameMaterial:=Materials.MI_P1_Frame, BoardMaterial:=Materials.M_Board_Master_Pink},
    #     card_style_params{CardStyle:=card_style.Pickaxes, CardMaterial:=Cards.Materials.MI_Card_Style_Pickaxes, FrameMaterial:=Materials.MI_P1_Frame, BoardMaterial:=Materials.M_Board_Master_Pink},
    #     card_style_params{CardStyle:=card_style.Vehicles, CardMaterial:=Cards.Materials.MI_Card_Style_Vehicles, FrameMaterial:=Materials.MI_P1_Frame, BoardMaterial:=Materials.M_Board_Master_Pink}

var Globals : weak_map(session, globals) = map{}

var PlayerSavedDataMap : weak_map(player, player_saved_data) = map{}


# GetCurrentDateTime()<transacts>:time_and_date=
#     MaybeClockProps := GetCreativeObjectsWithTag(clock_prop{})

#     var ClockProp : creative_prop = creative_prop{}

#     for(TaggedActor : MaybeClockProps, aProp := creative_prop[TaggedActor]):
#         set ClockProp = aProp

#     CurrentDate := MakeTimeAndDate(ClockProp)

GetCurrentDateTime()<transacts>:time_and_date=
    
    var TheDateTime : time_and_date = time_and_date{}

    if:
        Date := MakeDateTimeNowUtc[]
    then:
        set TheDateTime = Date
    
    return TheDateTime


CheckForPlayerSavedData(Agent:agent)<decides><transacts>:void=

    CurrentDate := GetCurrentDateTime()

    Player := player[Agent]
    Player.IsActive[]

    if(PlayerSavedDataMap[Player]):
        # Already have data do nothing
    else:
        if . set PlayerSavedDataMap[Player] = player_saved_data:
            IsAdmin := false
            IsModerator := false
            IsDeveloper := false
            IsBetaTester := false

            IsBanned := false
            BanReason := false
            BanExpiration := false
            BanIssuedBy := false
            BansIssued := 0
            WarningsIssued := 0
            PastWarnings := array{}

            XP := 0
            Rank := 0
            Score := 0
            Level := 0

            Gold := 0
            OverrideThemeTokens := 0
            ConvertedScoreToCorrectGuesses := false
            CorrectGuesses := 0
            IncorrectGuesses := 0
            CorrectYoloGuesses := 0
            ConsecutiveCorrectGuesses := 0
            ConsecutiveIncorrectGuesses := 0
            ConsecutiveCorrectGuessesRecord := 0
            ConsecutiveIncorrectGuessesRecord := 0
            GamesPlayed := 0
            GamesPlayedWithNewFortWhoPlayers := 0
            ConsecutiveDaysPlayed := 0
            ConsecutiveDaysPlayedRecord := 0
            LastPlayedDate:= ""
            FirstPlayedDate := ""
            LastDayWon := ""

            WinStreak := 0
            WinStreakRecord := 0
            LastStyleChoice := 0
            PrevousStyleChoices := array{}

            # Board Paints
            ConvertedUnlockedPaints := false
            UnlockedBoardPaints := array{}
            UnlockedCardBackPaints := array{}
            UnlockedYouStandPaints := array{}
            UnlockedPlatformPaints := array{}

            CurrentBoardPaint := board_paint_data:
                Rollers := 0
                Rim := 0
                Base := 0
                Underside := 0
                Mesh := 0


            CurrentCardBackPaint := card_back_paint_data:
                CardBack := 0

            CurrentYouStandPaint := you_stand_paint_data:
                Stand:=0
                CardHolder:=0
                Mesh:=0

            CurrentPlatformPaint := platform_paint_data:
                Base := 0
                Trim := 0
                Main := 0
                Mesh := 0

            CustomPaintSets := array:
                custom_paint_set:
                    Name := "Default"
                    BoardPaint := board_paint_data:
                        Rollers := 0
                        Rim := 0
                        Base := 0
                        Underside := 0
                        Mesh := 0
                    CardBackPaint := card_back_paint_data:
                        CardBack := 0
                    YouStandPaint := you_stand_paint_data:
                        Stand:=0
                        CardHolder:=0
                        Mesh:=0
                    PlatformPaint := platform_paint_data:
                        Base := 0
                        Trim := 0
                        Main := 0
                        Mesh := 0

            UnlockedBoardMeshes := array{}
            CurrentBoardMesh := 0

            UnlockedCardBackMeshes := array{}
            CurrentCardBackMesh := 0

            UnlockedYouStandMeshes := array{}
            CurrentYouStandMesh := 0

            UnlockedPlatformMeshes := array{}
            CurrentPlatformMesh := 0

            # Cosmetics

            UnlockedHats := array{}
            CurrentHat := 0

            UnlockedPoses := array{}
            CurrentPose := 0

            UnlockedPointers := array{}
            CurrentPointer := 0

            UnlockedTrails := array{}
            CurrentTrail := 0

            UnlockedTitles := array{}
            CurrentTitle := 0

            UnlockedBackgrounds := array{}
            CurrentBackground := 0

            # Audio

            UnlockedMusic := array{}
            CurrentMusic := 0

            UnlockedCardKnockSFX := array{}
            CurrentCardKnockSFX := 0

            UnlockedUndoSFX := array{}
            CurrentUndoSFX := 0

            UnlockedCorrectSFX := array{}
            CurrentCorrectSFX := 0

            UnlockedIncorrectSFX := array{}
            CurrentIncorrectSFX := 0

            # Quests

            CompletedQuests := array{}
            CurrentQuests := array{}
            CurrentPinnedQuests := array{}

            # added in v2

            FirstDayPlayed := CurrentDate
            LastDayPlayed := CurrentDate
            TotalDaysPlayed := 1
            TotalMinutesPlayed := 0
            NewPlayersPlayedWith := 0
            LastGuessCorrect := false
            FriendStreaks := array{}
            CardsKnocked := 0
            PreviousStyleChoices := array{}

            # added in v3
            CardHighlightColor := (0.0,0.302997,1.0)
            LastSessionTime := time_and_date{}

            StreamerModeHideCharacter := false
            StreamerModeVerticalUI := false

            ManuallyChooseCharacter := false

            # added in v4
            CompletedWinter2024Quest := false
            
            # added in v5
            TotalFlyingTargetsHit := 0

            # added in v6
            BeenAssignedPointer := false

            # added in v7
            UnlockedStickers := array{}
            CurrentStickers := array{}



# AllQuestions <public>:=module:

#     Categories <public> := class:

#         OGQuestions : []category = array:
#             category:
#                 Style:=card_style.OG
#                 Name:="Accessories"
#                 Questions:=array:
#                     question:
#                         TheQuestion:= "Is your character wearing a helmet?"
#                     question:
#                         TheQuestion := "Is your character wearing a hat?"
#                     question:
#                         TheQuestion := "Is your character wearing glasses/goggles?"
#                     question:
#                         TheQuestion := "Is your character wearing dogtags"
#                     question:
#                         TheQuestion := "Is your character wearing a bandana?"

#             category:
#                 Style:=card_style.OG
#                 Name:="Clothing"
#                 Questions:=array:
#                     question :
#                         TheQuestion := "Is your character wearing a jacket?"
#                     question :
#                         TheQuestion := "Is your character wearing a tank top?"
#                     question :
#                         TheQuestion := "Is your character wearing chainmail?"
#                     question :
#                         TheQuestion := "Is your character wearing a long sleeve shirt?"

#             category:
#                 Style:=card_style.OG
#                 Name:="Colors"
#                 Questions:=array:
#                     question :
#                         TheQuestion := "Does your character contain the color red? ❤️"
#                     question :
#                         TheQuestion := "Does your character contain the color blue? 💙"
#                     question :
#                         TheQuestion := "Does your character contain the color green? 💚"
#                     question :
#                         TheQuestion := "Does your character contain the color yellow? 💛"
#                     question :
#                         TheQuestion := "Does your character contain the color purple? 💜"
#                     question :
#                         TheQuestion := "Does your character contain the color orange? 🧡"
#                     question :
#                         TheQuestion := "Does your character contain the color pink? 💗"
#                     question :
#                         TheQuestion := "Does your character contain the color black or white? ⚫⚪"

#             category:
#                 Style:=card_style.OG
#                 Name:= "Other"
#                 Questions:=
#                     array:
#                         question:
#                             TheQuestion := "Is your character a Male? 👨"
#                         question:
#                             TheQuestion := "Is your charater a Female? 👩"
#                         question:
#                             TheQuestion:= "Does your character have a Tanned Skin Tone?"
#                         question:
#                             TheQuestion:= "Does your character have a Dark Skin Tone?"
#                         question:
#                             TheQuestion:= "Does your character have a Light Skin Tone?"
#                         question:
#                             TheQuestion:= "Does your character have Blue Eyes? 💙👀"
#                         question:
#                             TheQuestion:= "Does your character have Brown Eyes? 👀"
#                         question:
#                             TheQuestion:= "Does your character have Green Eyes? 💚👀"