using { /Fortnite.com/Devices }
using { /Verse.org/Simulation }
using { /UnrealEngine.com/Temporary/Diagnostics }
using { /UnrealEngine.com/Temporary/SpatialMath }
using { /Fortnite.com/Devices/CreativeAnimation }
# using { /Fortnite.com/Devices/Animation }
using { /Verse.org/Random }
using { /Fortnite.com/Characters }
using { /Verse.org/Colors/NamedColors }
using { /UnrealEngine.com/Temporary/UI }
using { /Fortnite.com/UI }
using { /Verse.org/Assets }
using { /Fortnite.com/FortPlayerUtilities }

ground_triggers := class<concrete>():
    @editable MakeGuessTrigger : trigger_device = trigger_device{}
    @editable EndTurnTrigger : trigger_device = trigger_device{}
    @editable AskQuestionTrigger : trigger_device = trigger_device{}
    @editable UndoTrigger : trigger_device = trigger_device{}
    