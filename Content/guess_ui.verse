using { /UnrealEngine.com/Temporary/SpatialMath }
using { /Verse.org/Simulation }
using { /UnrealEngine.com/Temporary/UI }
using { /Fortnite.com/UI }
# using { /Fortnite.com/Characters }
# using { /Fortnite.com/FortPlayerUtilities }
using { /Verse.org/Colors }
using { /Verse.org/Colors/NamedColors }
using { /Verse.org/Assets }
using { /Fortnite.com/Devices }
# using { /Verse.org/Concurrency }
using { /UnrealEngine.com/Temporary/Diagnostics }
using { lerping }

# NewMessage<localizes>(value:string) : message = "{value}"


ui_layout := enum{Horizontal, Vertical}

guess_ui := class<concrete>():

    Logger:log := log{Channel:=log_device}

    var Styles : []style_selection_data = array{}

    var MaybeStyleSelectionUI : ?canvas = false
    var OptionsOverlay : overlay = overlay{}
    var ExistingOptionsOverlay : ?widget = false
    var MaybePreloadedCanvas : ?canvas = false

    var ButtonPageLeft : button_loud = button_loud{}
    var ButtonPageRight : button_loud = button_loud{}
    var TotalPages : int = 0
    var CurrentPage : int = 0
    var HorizontalOptionsPerRow : int = 9
    var HorizontalRows : int = 3

    var CurrentUILayout : ui_layout = ui_layout.Horizontal
    var VerticalOptionsPerRow : int = 3
    var VerticalRows : int = 3
    var ButtonUILayoutSwitcher : button_quiet = button_quiet{}

    var FinalSortedStyles : []style_selection_data = array{}
    var OptionsPerPage : int = 0

    

    var ButtonChooseCharacterMethod : button_quiet = button_quiet{}
    var TextBlockChooseCharacterUpdated : text_block = text_block:
        DefaultTextColor := White
        DefaultJustification := text_justification.Right
        DefaultShadowColor := Black
        DefaultShadowOpacity := 0.8
        DefaultShadowOffset := option{vector2{X := 4.0, Y := 4.0}}
    

    var ButtonUILayoutSwitcherSubscription : ?cancelable = false
    var ButtonPageLeftSubscription : ?cancelable = false
    var ButtonPageRightSubscription : ?cancelable = false
    var ButtonChooseCharacterMethodSubscription : ?cancelable = false


    var LeftButtons : stack_box = stack_box:
        Orientation := orientation.Vertical
        Slots := array{}




    InitStyleData():void=
        set Styles = array:
            style_selection_data{StyleIndex := 0, StyleTexture := TextureFNStyle,           StyleName := StyleNames.MessageFN_EN,    CreationDate := "2023-03-04", LastUpdatedDate := "2024-11-14", SortOrder := -1, IsPinned := true, IsNew := false, IsUpdated := false, IsHidden := false, IsBeta := false, IsUnlockable := false, IsComingSoon := false, IsAd := false, AdStartDate := "9999-01-01", AdEndDate := "9999-12-12"}
            
            # Deprecated DO NOT USE
            style_selection_data{StyleIndex := 1, StyleTexture := TextureClassicStyle,      StyleName := StyleNames.MessageClassic_EN,     CreationDate := "2022-20-08", LastUpdatedDate := "2022-20-08", SortOrder := 0, IsPinned := false, IsNew := false, IsUpdated := false, IsHidden := true, IsBeta := false, IsUnlockable := false, IsComingSoon := false, IsAd := false, AdStartDate := "9999-01-01", AdEndDate := "9999-12-12"}    
            
            style_selection_data{StyleIndex := 2, StyleTexture := TextureIconStyle,         StyleName := StyleNames.MessageIcons_EN,        CreationDate := "2023-20-07", LastUpdatedDate := "2024-13-06", SortOrder := -1, IsPinned := false, IsNew := false, IsUpdated := false, IsHidden := false, IsBeta := false, IsUnlockable := false, IsComingSoon := false, IsAd := false, AdStartDate := "9999-01-01", AdEndDate := "9999-12-12"}            
            style_selection_data{StyleIndex := 3, StyleTexture := TextureAnimalStyle,       StyleName := StyleNames.MessageAnimals_EN,      CreationDate := "2023-04-07", LastUpdatedDate := "2023-04-07", SortOrder := 0, IsPinned := false, IsNew := false, IsUpdated := false, IsHidden := false, IsBeta := false, IsUnlockable := false, IsComingSoon := false, IsAd := false, AdStartDate := "9999-01-01", AdEndDate := "9999-12-12"}
            style_selection_data{StyleIndex := 4, StyleTexture := TextureUtilStyle,         StyleName := StyleNames.MessageUtilites_EN,        CreationDate := "2023-09-07", LastUpdatedDate := "2024-01-01", SortOrder := 0, IsPinned := false, IsNew := false, IsUpdated := false, IsHidden := false, IsBeta := false, IsUnlockable := false, IsComingSoon := false, IsAd := false, AdStartDate := "9999-01-01", AdEndDate := "9999-12-12"}
            style_selection_data{StyleIndex := 5, StyleTexture := TextureMarvelStyle,       StyleName := StyleNames.MessageMarvel_EN,      CreationDate := "2023-20-07", LastUpdatedDate := "2023-20-07", SortOrder := 0, IsPinned := false, IsNew := false, IsUpdated := false, IsHidden := true,  IsBeta := false, IsUnlockable := false, IsComingSoon := false, IsAd := false, AdStartDate := "9999-01-01", AdEndDate := "9999-12-12"}
            style_selection_data{StyleIndex := 6, StyleTexture := TexturePOIStyle,          StyleName := StyleNames.MessagePOI_EN,         CreationDate := "2023-25-07", LastUpdatedDate := "2023-25-07", SortOrder := 0, IsPinned := false, IsNew := false, IsUpdated := false, IsHidden := true,  IsBeta := false, IsUnlockable := false, IsComingSoon := false, IsAd := false, AdStartDate := "9999-01-01", AdEndDate := "9999-12-12"}            
            style_selection_data{StyleIndex := 7, StyleTexture := TextureLegendStyle,       StyleName := StyleNames.MessageLegends_EN,      CreationDate := "2023-01-08", LastUpdatedDate := "2024-27-06", SortOrder := 0, IsPinned := false, IsNew := false, IsUpdated := false,  IsHidden := false, IsBeta := false, IsUnlockable := false, IsComingSoon := false, IsAd := false, AdStartDate := "9999-01-01", AdEndDate := "9999-12-12"}
            style_selection_data{StyleIndex := 8, StyleTexture := TextureFoodStyle,         StyleName := StyleNames.MessageFood_EN,        CreationDate := "2023-05-08", LastUpdatedDate := "2023-05-08", SortOrder := 0, IsPinned := false, IsNew := false, IsUpdated := false, IsHidden := false, IsBeta := false, IsUnlockable := false, IsComingSoon := false, IsAd := false, AdStartDate := "9999-01-01", AdEndDate := "9999-12-12"}        
            
            style_selection_data{StyleIndex := 9, StyleTexture := TextureStarStyle,         StyleName := StyleNames.MessageStar_EN,   CreationDate := "2023-10-08", LastUpdatedDate := "2024-27-06", SortOrder := 0, IsPinned := false, IsNew := false, IsUpdated := true,  IsHidden := false, IsBeta := false, IsUnlockable := false, IsComingSoon := false, IsAd := false, AdStartDate := "9999-01-01", AdEndDate := "9999-12-12"}
            
            style_selection_data{StyleIndex := 10, StyleTexture := TextureSpookyStyle,      StyleName := StyleNames.MessageSpooky_EN,      CreationDate := "9999-31-10", LastUpdatedDate := "9999-31-10", SortOrder := 0, IsPinned := false, IsNew := false, IsUpdated := false, IsHidden := false,  IsBeta := false, IsUnlockable := false, IsComingSoon := false, IsAd := false, AdStartDate := "9999-01-01", AdEndDate := "9999-12-12"}
            
            style_selection_data{StyleIndex := 11, StyleTexture := TextureOGStyle,          StyleName := StyleNames.MessageOG_EN,          CreationDate := "2023-06-11", LastUpdatedDate := "2023-06-11", SortOrder := -2, IsPinned := false, IsNew := false, IsUpdated := false, IsHidden := false, IsBeta := false, IsUnlockable := false, IsComingSoon := false, IsAd := false, AdStartDate := "9999-01-01", AdEndDate := "9999-12-12"}
            
            style_selection_data{StyleIndex := 12, StyleTexture := TextureAnimeStyle,       StyleName := StyleNames.MessageAnime_EN,       CreationDate := "2023-10-11", LastUpdatedDate := "2023-10-11", SortOrder := 0, IsPinned := false, IsNew := false, IsUpdated := false, IsHidden := false, IsBeta := false, IsUnlockable := false, IsComingSoon := false, IsAd := false, AdStartDate := "9999-01-01", AdEndDate := "9999-12-12"}                        
            style_selection_data{StyleIndex := 13, StyleTexture := TextureWinterStyle,      StyleName := StyleNames.MessageWinter_EN,      CreationDate := "2023-01-12", LastUpdatedDate := "2023-01-12", SortOrder := 0, IsPinned := false, IsNew := false, IsUpdated := false, IsHidden := false, IsBeta := false, IsUnlockable := false, IsComingSoon := false, IsAd := false, AdStartDate := "9999-01-01", AdEndDate := "9999-12-12"}    
            style_selection_data{StyleIndex := 14, StyleTexture := TextureWeaponsStyle,     StyleName := StyleNames.MessageWeapons_EN,     CreationDate := "2023-10-12", LastUpdatedDate := "2023-10-12", SortOrder := 0, IsPinned := false, IsNew := false, IsUpdated := false, IsHidden := false, IsBeta := false, IsUnlockable := false, IsComingSoon := false, IsAd := false, AdStartDate := "9999-01-01", AdEndDate := "9999-12-12"}
            style_selection_data{StyleIndex := 15, StyleTexture := TexturePickaxesStyle,    StyleName := StyleNames.MessagePickaxes_EN,    CreationDate := "2023-15-12", LastUpdatedDate := "2023-15-12", SortOrder := 0, IsPinned := false, IsNew := false, IsUpdated := false, IsHidden := false, IsBeta := false, IsUnlockable := false, IsComingSoon := false, IsAd := false, AdStartDate := "9999-01-01", AdEndDate := "9999-12-12"}            
            style_selection_data{StyleIndex := 16, StyleTexture := TextureVehiclesStyle,    StyleName := StyleNames.MessageVehicles_EN,    CreationDate := "2023-20-12", LastUpdatedDate := "2023-20-12", SortOrder := 0, IsPinned := false, IsNew := false, IsUpdated := false, IsHidden := false, IsBeta := false, IsUnlockable := false, IsComingSoon := false, IsAd := false, AdStartDate := "9999-01-01", AdEndDate := "9999-12-12"}                                                
            
            style_selection_data{StyleIndex := 17, StyleTexture := TextureMoviesTVStyle,    StyleName := StyleNames.MessageMoviesTV_EN,     CreationDate := "2023-01-12", LastUpdatedDate := "2024-27-06", SortOrder := 0, IsPinned := false, IsNew := false, IsUpdated := true, IsHidden := false, IsBeta := false, IsUnlockable := false, IsComingSoon := false, IsAd := false, AdStartDate := "9999-01-01", AdEndDate := "9999-12-12"}
            
            style_selection_data{StyleIndex := 18, StyleTexture := TextureBlocksStyle,      StyleName := MessageBlank_EN,                CreationDate := "9999-01-01", LastUpdatedDate := "9999-01-01", SortOrder := 0, IsPinned := false, IsNew := false, IsUpdated := false, IsHidden := true,  IsBeta := false, IsUnlockable := false, IsComingSoon := false, IsAd := false, AdStartDate := "9999-01-01", AdEndDate := "9999-12-12"}
            style_selection_data{StyleIndex := 19, StyleTexture := TextureChapter5Style,    StyleName := StyleNames.MessageChapter5_EN,   CreationDate := "2023-05-12", LastUpdatedDate := "2023-05-12", SortOrder := 0, IsPinned := false, IsNew := false, IsUpdated := false, IsHidden := false, IsBeta := false, IsUnlockable := false, IsComingSoon := false, IsAd := false, AdStartDate := "9999-01-01", AdEndDate := "9999-12-12"}
            style_selection_data{StyleIndex := 20, StyleTexture := TextureLockersStyle,     StyleName := StyleNames.MessageBundles_EN,     CreationDate := "2024-10-03", LastUpdatedDate := "2024-10-03", SortOrder := 0, IsPinned := false, IsNew := false, IsUpdated := false, IsHidden := false, IsBeta := false, IsUnlockable := false, IsComingSoon := false, IsAd := false, AdStartDate := "9999-01-01", AdEndDate := "9999-12-12"}
            style_selection_data{StyleIndex := 21, StyleTexture := TextureBackblingsStyle,  StyleName := StyleNames.MessageBackblings_EN,  CreationDate := "2024-01-04", LastUpdatedDate := "2024-01-04", SortOrder := 0, IsPinned := false, IsNew := false, IsUpdated := false, IsHidden := false, IsBeta := false, IsUnlockable := false, IsComingSoon := false, IsAd := false, AdStartDate := "9999-01-01", AdEndDate := "9999-12-12"}
            style_selection_data{StyleIndex := 22, StyleTexture := TextureEmotesStyle,      StyleName := StyleNames.MessageEmotes_EN,      CreationDate := "2024-13-05", LastUpdatedDate := "2024-13-05", SortOrder := 0, IsPinned := false, IsNew := false, IsUpdated := false, IsHidden := false, IsBeta := false, IsUnlockable := false, IsComingSoon := false, IsAd := false, AdStartDate := "9999-01-01", AdEndDate := "9999-12-12"}
            style_selection_data{StyleIndex := 23, StyleTexture := TextureMarvelStyleSet1,  StyleName := StyleNames.MessageHeroes1_EN,CreationDate := "2023-20-07", LastUpdatedDate := "2024-13-06", SortOrder := 0, IsPinned := false, IsNew := false, IsUpdated := false, IsHidden := false, IsBeta := false, IsUnlockable := false, IsComingSoon := false, IsAd := false, AdStartDate := "9999-01-01", AdEndDate := "9999-12-12"}
            style_selection_data{StyleIndex := 24, StyleTexture := TextureMarvelStyleSet2,  StyleName := StyleNames.MessageHeroes2_EN,CreationDate := "2024-13-06", LastUpdatedDate := "2024-13-06", SortOrder := 0, IsPinned := false, IsNew := false, IsUpdated := false, IsHidden := false, IsBeta := false, IsUnlockable := false, IsComingSoon := false, IsAd := false, AdStartDate := "9999-01-01", AdEndDate := "9999-12-12"}
            style_selection_data{StyleIndex := 25, StyleTexture := TextureUmbrellasStyle,   StyleName := StyleNames.MessageUmbrellas_EN,   CreationDate := "2024-04-09", LastUpdatedDate := "2024-04-09", SortOrder := 0, IsPinned := false, IsNew := false, IsUpdated := false, IsHidden := false, IsBeta := false, IsUnlockable := false, IsComingSoon := false, IsAd := false, AdStartDate := "9999-01-01", AdEndDate := "9999-12-12"}
            style_selection_data{StyleIndex := 26, StyleTexture := TexturePeelyFishstickStyle, StyleName := StyleNames.MessagePeelyFishstick_EN, CreationDate := "2024-04-09", LastUpdatedDate := "2024-04-09", SortOrder := 0, IsPinned := false, IsNew := false, IsUpdated := false, IsHidden := false, IsBeta := false, IsUnlockable := false, IsComingSoon := false, IsAd := false, AdStartDate := "9999-01-01", AdEndDate := "9999-12-12"}
            style_selection_data{StyleIndex := 27, StyleTexture := TextureRemixStyle,       StyleName := StyleNames.MessageRemix_EN,       CreationDate := "2024-04-09", LastUpdatedDate := "2024-04-09", SortOrder := 0, IsPinned := false, IsNew := false, IsUpdated := false, IsHidden := false, IsBeta := false, IsUnlockable := false, IsComingSoon := false, IsAd := false, AdStartDate := "9999-01-01", AdEndDate := "9999-12-12"}            
            style_selection_data{StyleIndex := 28, StyleTexture := TextureFireIceStyle,       StyleName := StyleNames.MessageFireIce_EN,       CreationDate := "2024-12-01", LastUpdatedDate := "2024-12-01", SortOrder := 0, IsPinned := false, IsNew := false, IsUpdated := false, IsHidden := false, IsBeta := false, IsUnlockable := false, IsComingSoon := false, IsAd := false, AdStartDate := "9999-01-01", AdEndDate := "9999-12-12"}
            
            style_selection_data{StyleIndex := 29, StyleTexture := TextureDCStyle,          StyleName := StyleNames.MessageHeroes3_EN,          CreationDate := "2025-02-14", LastUpdatedDate := "2025-02-14", SortOrder := 1, IsPinned := false, IsNew := true, IsUpdated := false, IsHidden := false, IsBeta := false, IsUnlockable := false, IsComingSoon := false, IsAd := false, AdStartDate := "9999-01-01", AdEndDate := "9999-12-12"}

            style_selection_data{StyleIndex := 30, StyleTexture := TextureChapter6Style,    StyleName := StyleNames.MessageChapter6_EN,  CreationDate := "2025-02-27", LastUpdatedDate := "2025-02-27", SortOrder := 0, IsPinned := false, IsNew := true, IsUpdated := false, IsHidden := false, IsBeta := false, IsUnlockable := false, IsComingSoon := false, IsAd := false, AdStartDate := "9999-01-01", AdEndDate := "9999-12-12"}


    # TextureTGA : texture = _UI.UI_Textures.tga_nom

    TextureBackground : texture = _UI.UI_Textures.ui_bg_blurred
    TextureNewBadge : texture = _UI.UI_Textures.New_Badge_EN
    TextureNotificationIcon : texture = _UI.UI_Textures.Notification_One

    TextureComingSoon : texture = Cards.Textures.UI.verse_ui_coming_soon
    TextureFNStyle : texture = Cards.Textures.UI.verse_ui_fn
    TextureClassicStyle : texture = Cards.Textures.UI.verse_ui_classic
    TextureIconStyle : texture = Cards.Textures.UI.verse_ui_icon
    TextureAnimalStyle : texture = Cards.Textures.UI.verse_ui_animal
    TextureUtilStyle : texture = Cards.Textures.UI.verse_ui_util
    TextureMarvelStyle : texture = Cards.Textures.UI.verse_ui_marvel
    TexturePOIStyle : texture = Cards.Textures.UI.verse_ui_poi
    TextureLegendStyle : texture = Cards.Textures.UI.verse_ui_legend
    TextureFoodStyle : texture = Cards.Textures.UI.verse_ui_food
    TextureStarStyle : texture = Cards.Textures.UI.verse_ui_star
    TextureSpookyStyle : texture = Cards.Textures.UI.verse_ui_spooky
    TextureOGStyle : texture = Cards.Textures.UI.verse_ui_og
    TextureAnimeStyle : texture = Cards.Textures.UI.verse_ui_anime
    TextureWinterStyle : texture = Cards.Textures.UI.verse_ui_winter
    TextureWeaponsStyle : texture = Cards.Textures.UI.verse_ui_weapons
    TexturePickaxesStyle : texture = Cards.Textures.UI.verse_ui_pickaxes
    TextureVehiclesStyle : texture = Cards.Textures.UI.verse_ui_vehicles
    TextureMoviesTVStyle : texture = Cards.Textures.UI.verse_ui_movies_tv
    TextureBlocksStyle : texture = Cards.Textures.UI.verse_ui_blocks
    TextureChapter5Style : texture = Cards.Textures.UI.verse_ui_chapter5
    TextureLockersStyle : texture = Cards.Textures.UI.verse_ui_lockers
    TextureBackblingsStyle : texture = Cards.Textures.UI.verse_ui_backblings
    TextureEmotesStyle : texture = Cards.Textures.UI.verse_ui_emotes
    TextureMarvelStyleSet1 : texture = Cards.Textures.UI.verse_ui_marvel_set_1
    TextureMarvelStyleSet2 : texture = Cards.Textures.UI.verse_ui_marvel_set_2
    TextureUmbrellasStyle : texture = Cards.Textures.UI.verse_ui_umbrellas
    TexturePeelyFishstickStyle : texture = Cards.Textures.UI.verse_ui_peely_fishstick
    TextureRemixStyle : texture = Cards.Textures.UI.verse_ui_remix
    TextureFireIceStyle : texture = Cards.Textures.UI.verse_ui_fire_ice
    TextureDCStyle : texture = Cards.Textures.UI.verse_ui_dc
    TextureChapter6Style : texture = Cards.Textures.UI.verse_ui_chapter6

    TextureInstructions_SelectionScreen : texture = _UI.UI_Textures.SelectionInstructions_EN
    TextureInstructions_VoteScreen : texture = _UI.UI_Textures.WaitingInstructions_EN
    TextureTimeRemainingText : texture = _UI.UI_Textures.TimeRemaining_EN
    TextureYourVoteText : texture = _UI.UI_Textures.YourVote_EN
    TextureTheirVoteText : texture = _UI.UI_Textures.TheirVote_EN

    # TextureYourTurnText : texture = _UI.UI_Textures.YourTurn_EN
    # TextureRemoteIcon : texture = _UI.UI_Textures.T_Remote_Legendary

    # TextureOtherPlayersTurnText : texture = _UI.UI_Textures.OtherPlayersTurn_EN
    # TextureShootToGuessIcon : texture = _UI.UI_Textures.T_Reticle_Make_Guess
    TextureShootToGuessText : texture = _UI.UI_Textures.FinalGuess_EN

    TextureCorrectText : texture = _UI.UI_Textures.Correct_EN
    TextureIncorrectText : texture = _UI.UI_Textures.Incorrect_EN

    var MaybeYourTurnUI : ?canvas = false
    var MaybeVoteUI : ?canvas = false
    var QuestionCategoryText : text_block = text_block{DefaultTextColor:= MakeColorFromSRGBValues(255,255,255),DefaultShadowColor:=MakeColorFromHex("#000000"),DefaultJustification:= text_justification.Center }
    
    var MaybeQuestionUI : ?widget = false
    var AllQuestionButtonsOverlay : overlay = overlay{}
    var MaybeExistingQuestionButtons : ?overlay_slot = false

    var MaybeOtherPlayersTurnUI : ?widget = false
    var MaybeShootToGuessUI : ?widget = false

    var MaybePaintPaletteUI : ?canvas = false


    var MaybeCorrectIncorrectUI : ?widget = false
    var CorrectIncorrectTextureBlock : texture_block = texture_block:
        DefaultImage := _UI.UI_Textures.transparent_square
        DefaultDesiredSize := vector2{X := 800.0, Y := 195.33}
        
    var CorrectIncorrectGuesserText : text_block = text_block:
        DefaultTextColor := White
        DefaultShadowColor := Black
        DefaultJustification := text_justification.Center
        DefaultShadowOffset := option{vector2{X := 3.0, Y := 3.0}}

    var MaybeReticleUI : ?canvas = false
    var ReticleTextureBlock : texture_block = texture_block:
        DefaultImage := _UI.UI_Textures.T_Reticle_Default
        DefaultDesiredSize := vector2{X := 52.0, Y := 52.0}

    DefaultReticleTexture : tuple(texture, vector2) = (_UI.UI_Textures.T_Reticle_Default, vector2{X:=52.0, Y:=52.0})
    OpenHandReticleTexture : tuple(texture, vector2) = (_UI.UI_Textures.T_Reticle_GrabOpen, vector2{X:=52.0, Y:=52.0})
    ClosedHandReticleTexture : tuple(texture, vector2) = (_UI.UI_Textures.T_Reticle_GrabClosed, vector2{X:=52.0, Y:=52.0})
    FinalGuessReticleTexture : tuple(texture, vector2) = (_UI.UI_Textures.question_mark_reticle, vector2{X:=196.0, Y:=196.0})
    FinalGuessHoverReticleTexture : tuple(texture, vector2) = (_UI.UI_Textures.question_mark_reticle_hover, vector2{X:=196.0, Y:=196.0})
    

        
    
    var MaybePlayer : ?player = false

    var TurnTimer : timer_device = timer_device{}

    var PlayerOneVoteText : overlay = overlay{}
    var PlayerTwoVoteText : overlay = overlay{}

    var PlayerOneTexture : texture_block = texture_block:
        DefaultImage := _UI.UI_Textures.transparent_square
        DefaultDesiredSize := vector2{X := 72.0, Y := 72.0}

    var PlayerTwoTexture : texture_block = texture_block:
        DefaultImage := _UI.UI_Textures.transparent_square
        DefaultDesiredSize := vector2{X := 72.0, Y := 72.0}

    var PlayerOneVoteTextureBlock : texture_block = texture_block:
        DefaultImage := _UI.UI_Textures.YourVote_EN
        DefaultDesiredSize := vector2{X := 338.0, Y := 94.0}

    var PlayerTwoVoteTextureBlock : texture_block = texture_block:
        DefaultImage := _UI.UI_Textures.TheirVote_EN
        DefaultDesiredSize := vector2{X := 338.0, Y := 94.0}
        
    var WaitingTextBlock : text_block = text_block:
        DefaultTextColor := White
        DefaultShadowColor :=Black
        DefaultJustification := text_justification.Center

    # var Timer : custom_timer = custom_timer{}

    var TimerTextBlock : text_block = text_block:
        # DefaultShadowOffset := option{vector2{X := 3.0, Y := 3.0}}
        DefaultTextColor := White
        DefaultShadowColor :=Black
        DefaultJustification := text_justification.Center

    var TimerTextBlock2 : text_block = text_block:
        DefaultShadowOffset := option{vector2{X := 3.0, Y := 3.0}}
        DefaultTextColor := White
        DefaultShadowColor := Black        
        DefaultJustification := text_justification.Center

    

    var AllSelectionScreenOptions : []selection_screen_option = array{}

    

    YourVoteTexture : texture_block = texture_block:
        DefaultImage := Cards.Textures.UI.verse_ui_coming_soon
        DefaultDesiredSize := vector2{X := 72.0, Y := 72.0}

    TheirVoteTexture : texture_block = texture_block:
        DefaultImage := Cards.Textures.UI.verse_ui_coming_soon
        DefaultDesiredSize := vector2{X := 72.0, Y := 72.0}

    TimerBGTexture : texture_block = texture_block:
        DefaultImage := _UI.UI_Textures.ui_timer_bg
        DefaultDesiredSize := vector2{X := 150.0, Y := 150.0}

    TimerTimeTexture : texture_block = texture_block:
        DefaultImage := _UI.UI_Textures.Numbers.Number_90
        DefaultDesiredSize := vector2{X := 150.0, Y := 150.0}

    ResetUIState(Player: player)<suspends>: void =
        # Reset layout to default or based on player preference
        Print("Resetting UI State - Based on Player Preference")
        if:
            GetStreamerModeVerticalUI[Player]
        then:
            Print("UI Mode - Vertical")
            set CurrentUILayout = ui_layout.Vertical
            ButtonUILayoutSwitcher.SetText(MessageSwitchLayoutHorizontal_EN)
        else:
            Print("UI Mode - Horizontal")
            set CurrentUILayout = ui_layout.Horizontal
            ButtonUILayoutSwitcher.SetText(MessageSwitchLayoutVertical_EN)

        if:
            TheGlobals := Globals[GetSession()]
            CPlayer := TheGlobals.CustomPlayerMap[Player]
            GameBoard := CPlayer.GameBoard
        then:
            Print("Checking for Manual Character Selection")
            if:
                GameBoard.IsManualCharacterSelection?
            then:
                Print("Manually Choose Character")
                ButtonChooseCharacterMethod.SetText(MessageRandomCharacterIcon_EN)
            else:
                Print("Randomly Choose Character")
                ButtonChooseCharacterMethod.SetText(MessageChooseCharacterIcon_EN)
        else:
            Print("Failed to get GameBoard - Resetting UI State")
    
        # Reset page number
        set CurrentPage = 0
    
        # Remove existing UI elements
        if (PlayerUI := GetPlayerUI[Player]):
            if (ExistingUI := MaybeStyleSelectionUI?):
                PlayerUI.RemoveWidget(ExistingUI)
                set MaybeStyleSelectionUI = false
            if (ExistingUI := MaybePreloadedCanvas?):
                PlayerUI.RemoveWidget(ExistingUI)
                set MaybePreloadedCanvas = false
    

    InitStyleSelectionUI(Player: player, CPlayer: custom_player)<suspends>: void =
        # Reset UI state
        ResetUIState(Player)
        Sleep(0.0)
    
        # Initialize player option
        set MaybePlayer = option{Player}
    
        # Get the player UI
        if (PlayerUI := GetPlayerUI[Player]):
            # Create the selection screen UI based on the current layout
            TheCanvas := CreateSelectionScreenUI_NEW(Player)
            set MaybeStyleSelectionUI = option{TheCanvas}
            PlayerUI.AddWidget(TheCanvas, player_ui_slot{ZOrder := 2147483647, InputMode := ui_input_mode.All})
    
            # Create the selection screen options
            CreateSelectionScreenOptions()
    
        Sleep(0.25)
    
        # Preload textures if necessary
        if (PlayerUI := GetPlayerUI[Player]):
            NewCanvas := canvas:
                Slots := array:
                    PreloadTimerTextures()
    
            set MaybePreloadedCanvas = option{NewCanvas}
            PlayerUI.AddWidget(NewCanvas, player_ui_slot{ZOrder := 1, InputMode := ui_input_mode.None})
    

    ResetPagination(): void =
        set CurrentPage = 0
        set TotalPages = 0
            

    AwaitVote(CPlayer:custom_player, VotedSound:audio_player_device)<suspends>:void=
        loop:
            if(CPlayer.HasVoted?):
                break
            Sleep(0.0)
        VotedSound.Play()


    var IsTimerActive : logic = false

    AwaitStyleSelection(Player:player, CPlayer:custom_player, VotedSound:audio_player_device)<suspends>:int=

        var TempSelectionIndex : int = -1

        var TimedOut : logic = false

        

        
        spawn { UpdateTimerTextureLoop(TimerBGTexture, TimerTimeTexture) }
        

        race:
            AwaitVote(CPlayer, VotedSound)

            block:
                Sleep(SELECTION_SCREEN_TIME  * 1.0)
                set TimedOut = true

        if:
            TimedOut?
            TheGlobals := Globals[GetSession()]
            set TheGlobals.PlayerVoteIndex[Player] = -1
        then:
            set CPlayer.HasVoted = true

        if:
            TheVote := CPlayer.GetVote[]
            TimeRemaining := TurnTimer.GetActiveDuration()
            Secs := Floor[TimeRemaining]
        then:
            set TempSelectionIndex = TheVote
            TurnTimer.SetActiveDuration(Secs * 1.0)

        return TempSelectionIndex




    UpdateTimerTexture(BGTexture:texture_block, TextureBlock:texture_block):void=           
            Secs := TurnTimer.GetActiveDuration()
            if( Floor[Secs] <= 10):
                BGTexture.SetImage(_UI.UI_Textures.ui_timer_bg_red)
            else if( Floor[Secs] <= 30):
                BGTexture.SetImage(_UI.UI_Textures.ui_timer_bg_yellow)
            else:
                BGTexture.SetImage(_UI.UI_Textures.ui_timer_bg)

            if(IntSec := Floor[Secs]):
                # Print("Updating Timer Texture - {IntSec}")
                TextureBlock.SetImage(GetTextureForSeconds(IntSec))
            else:
                Logger.Print("Failed to Update Timer Texture")

    UpdateTimerTextureLoop(BGTexture:texture_block, TextureBlock:texture_block)<suspends>:void=
        Logger.Print("Starting Timer Texture Loop")
        set IsTimerActive = true
        loop:
            if(not IsTimerActive?):
                break

            Self.UpdateTimerTexture(BGTexture, TextureBlock)
            Sleep(1.0)

    GetTextureForSeconds(Secs:int):texture=
        case(Secs):
            90 => return _UI.UI_Textures.Numbers.Number_90
            89 => return _UI.UI_Textures.Numbers.Number_89
            88 => return _UI.UI_Textures.Numbers.Number_88
            87 => return _UI.UI_Textures.Numbers.Number_87
            86 => return _UI.UI_Textures.Numbers.Number_86
            85 => return _UI.UI_Textures.Numbers.Number_85
            84 => return _UI.UI_Textures.Numbers.Number_84
            83 => return _UI.UI_Textures.Numbers.Number_83
            82 => return _UI.UI_Textures.Numbers.Number_82
            81 => return _UI.UI_Textures.Numbers.Number_81
            80 => return _UI.UI_Textures.Numbers.Number_80
            79 => return _UI.UI_Textures.Numbers.Number_79
            78 => return _UI.UI_Textures.Numbers.Number_78
            77 => return _UI.UI_Textures.Numbers.Number_77
            76 => return _UI.UI_Textures.Numbers.Number_76
            75 => return _UI.UI_Textures.Numbers.Number_75
            74 => return _UI.UI_Textures.Numbers.Number_74
            73 => return _UI.UI_Textures.Numbers.Number_73
            72 => return _UI.UI_Textures.Numbers.Number_72
            71 => return _UI.UI_Textures.Numbers.Number_71
            70 => return _UI.UI_Textures.Numbers.Number_70
            69 => return _UI.UI_Textures.Numbers.Number_69
            68 => return _UI.UI_Textures.Numbers.Number_68
            67 => return _UI.UI_Textures.Numbers.Number_67
            66 => return _UI.UI_Textures.Numbers.Number_66
            65 => return _UI.UI_Textures.Numbers.Number_65
            64 => return _UI.UI_Textures.Numbers.Number_64
            63 => return _UI.UI_Textures.Numbers.Number_63
            62 => return _UI.UI_Textures.Numbers.Number_62
            61 => return _UI.UI_Textures.Numbers.Number_61
            60 => return _UI.UI_Textures.Numbers.Number_60
            59 => return _UI.UI_Textures.Numbers.Number_59
            58 => return _UI.UI_Textures.Numbers.Number_58
            57 => return _UI.UI_Textures.Numbers.Number_57
            56 => return _UI.UI_Textures.Numbers.Number_56
            55 => return _UI.UI_Textures.Numbers.Number_55
            54 => return _UI.UI_Textures.Numbers.Number_54
            53 => return _UI.UI_Textures.Numbers.Number_53
            52 => return _UI.UI_Textures.Numbers.Number_52
            51 => return _UI.UI_Textures.Numbers.Number_51
            50 => return _UI.UI_Textures.Numbers.Number_50
            49 => return _UI.UI_Textures.Numbers.Number_49
            48 => return _UI.UI_Textures.Numbers.Number_48
            47 => return _UI.UI_Textures.Numbers.Number_47
            46 => return _UI.UI_Textures.Numbers.Number_46
            45 => return _UI.UI_Textures.Numbers.Number_45
            44 => return _UI.UI_Textures.Numbers.Number_44
            43 => return _UI.UI_Textures.Numbers.Number_43
            42 => return _UI.UI_Textures.Numbers.Number_42
            41 => return _UI.UI_Textures.Numbers.Number_41
            40 => return _UI.UI_Textures.Numbers.Number_40
            39 => return _UI.UI_Textures.Numbers.Number_39
            38 => return _UI.UI_Textures.Numbers.Number_38
            37 => return _UI.UI_Textures.Numbers.Number_37
            36 => return _UI.UI_Textures.Numbers.Number_36
            35 => return _UI.UI_Textures.Numbers.Number_35
            34 => return _UI.UI_Textures.Numbers.Number_34
            33 => return _UI.UI_Textures.Numbers.Number_33
            32 => return _UI.UI_Textures.Numbers.Number_32
            31 => return _UI.UI_Textures.Numbers.Number_31
            30 => return _UI.UI_Textures.Numbers.Number_30
            29 => return _UI.UI_Textures.Numbers.Number_29
            28 => return _UI.UI_Textures.Numbers.Number_28
            27 => return _UI.UI_Textures.Numbers.Number_27
            26 => return _UI.UI_Textures.Numbers.Number_26
            25 => return _UI.UI_Textures.Numbers.Number_25
            24 => return _UI.UI_Textures.Numbers.Number_24
            23 => return _UI.UI_Textures.Numbers.Number_23
            22 => return _UI.UI_Textures.Numbers.Number_22
            21 => return _UI.UI_Textures.Numbers.Number_21
            20 => return _UI.UI_Textures.Numbers.Number_20
            19 => return _UI.UI_Textures.Numbers.Number_19
            18 => return _UI.UI_Textures.Numbers.Number_18
            17 => return _UI.UI_Textures.Numbers.Number_17
            16 => return _UI.UI_Textures.Numbers.Number_16
            15 => return _UI.UI_Textures.Numbers.Number_15
            14 => return _UI.UI_Textures.Numbers.Number_14
            13 => return _UI.UI_Textures.Numbers.Number_13
            12 => return _UI.UI_Textures.Numbers.Number_12
            11 => return _UI.UI_Textures.Numbers.Number_11
            10 => return _UI.UI_Textures.Numbers.Number_10
            9 => return _UI.UI_Textures.Numbers.Number_9
            8 => return _UI.UI_Textures.Numbers.Number_8
            7 => return _UI.UI_Textures.Numbers.Number_7
            6 => return _UI.UI_Textures.Numbers.Number_6
            5 => return _UI.UI_Textures.Numbers.Number_5
            4 => return _UI.UI_Textures.Numbers.Number_4
            3 => return _UI.UI_Textures.Numbers.Number_3
            2 => return _UI.UI_Textures.Numbers.Number_2
            1 => return _UI.UI_Textures.Numbers.Number_1
            0 => return _UI.UI_Textures.Numbers.Number_0
            _ => return _UI.UI_Textures.Numbers.Number_0

    PreloadTimerTextures():canvas_slot=
        # load all timer textures and hide off screen
        PreloadCanvas := canvas_slot:
                    Anchors := anchors{Minimum := vector2{X := 0.5, Y := 0.0}, Maximum := vector2{X := 0.5, Y := 0.0} }
                    Offsets := margin{Top := -2000.0, Left := 0.0, Right := 0.0, Bottom := 0.0}
                    Alignment := vector2{X := 0.5, Y := 0.5}
                    ZOrder := 999999999
                    SizeToContent := true
                    Widget := overlay:
                        Slots:= array:
                            overlay_slot:
                                HorizontalAlignment := horizontal_alignment.Center
                                VerticalAlignment := vertical_alignment.Center
                                Padding := margin{Top := 0.0, Left := 0.0, Right := 0.0, Bottom := 0.0}
                                Widget := StackAllSecondTextures()

        return PreloadCanvas


    StackAllSecondTextures():overlay=
        var Stack : overlay = overlay{}
        var Slots : []overlay_slot = array{}
        for(ImgIdx := 0 .. 99):
            set Slots = Slots + array:
                overlay_slot:
                    HorizontalAlignment := horizontal_alignment.Center
                    VerticalAlignment := vertical_alignment.Center
                    Padding := margin{Top := 0.0, Left := 0.0, Right := 0.0, Bottom := 0.0}
                    Widget := texture_block:
                        DefaultImage := Self.GetTextureForSeconds(ImgIdx)
                        DefaultDesiredSize := vector2{X := 8.0, Y := 8.0}


        set Stack = overlay{Slots := Slots}
        
        return Stack


    AddCanvasToUI(Player:player, Canvas:canvas, InputMode:ui_input_mode, ZOrder:type{_X:int where 0 <= _X, _X <= 2147483647}):void=
        if (PlayerUI := GetPlayerUI[Player]):
            PlayerUI.AddWidget(Canvas, player_ui_slot{ZOrder:= ZOrder, InputMode := InputMode})
    

    CreateUI() : canvas =     
        var ui : canvas = canvas{}

        set ui = canvas:
            Slots := array:
                canvas_slot:
                    Anchors := anchors{Minimum := vector2{X := 1.0, Y := 0.5}, Maximum := vector2{X := 1.0, Y := 0.5} }
                    Offsets := margin{Top := 0.0, Left := 0.0, Right := 0.0, Bottom := 0.0}
                    Alignment := vector2{X := 1.0, Y := 0.5}
                    SizeToContent := true
                    Widget := overlay:
                        Slots:= array:
                            overlay_slot:
                                HorizontalAlignment := horizontal_alignment.Fill
                                VerticalAlignment := vertical_alignment.Fill
                                Padding := margin{Top := 0.0, Left := 0.0, Right := 0.0, Bottom := 0.0}
                                Widget := color_block:
                                    DefaultColor := Red
                                    DefaultDesiredSize := vector2{X := 256.0, Y := 256.0}                                                                                         
        return ui      
        

    UpdateTimerTextBlock(TextWidget:text_block):void=
        # Print("Should Update Timer Text Block")
        TextWidget.SetText(StringToMessage("{TurnTimer.GetActiveDuration()}"))


    UpdatePlayerOneTexture(Img:texture):void=
        PlayerOneTexture.SetImage(Img)

    UpdatePlayerTwoTexture(Img:texture):void=
        PlayerTwoTexture.SetImage(Img)


    UpdateSecondVoter():void=
        Logger.Print("Updating Second Voter")
        if:
            TheGlobals := Globals[GetSession()]
            CPlayer := TheGlobals.CustomPlayerMap[Self.MaybePlayer?]
            P2 := CPlayer.GetOpponent()?
            CPlayer2 := TheGlobals.CustomPlayerMap[P2]
            ImgIndexP1 := TheGlobals.PlayerVoteIndex[Self.MaybePlayer?]
            ImgIndexP2 := TheGlobals.PlayerVoteIndex[P2]
            P2UI := CPlayer2.GuessUI
            TextuerBlok := P2UI.PlayerTwoTexture
        then:
            Logger.Print("Player 2 Voted For: {ImgIndexP2} Updating Texture")
            TextuerBlok.SetImage(GetTextureFromStyleIndex(ImgIndexP1))

            # if(ImgIndexP1 = ImgIndexP2):
            #     WaitingTextBlock.SetText(MessageVotesMatch_EN)
            #     P2UI.WaitingTextBlock.SetText(MessageVotesMatch_EN)
            # else:
            #     WaitingTextBlock.SetText(MessageVotesDontMatch_EN)
            #     P2UI.WaitingTextBlock.SetText(MessageVotesDontMatch_EN)
        else:
            Logger.Print("UpdateSecondVoter Failed: Player Not Found")


    UpdateWaitingTextBlock(Message:message):void=
        Logger.Print("Updating Waiting Text Block")
        if:
            TheGlobals := Globals[GetSession()]
            CPlayer := TheGlobals.CustomPlayerMap[Self.MaybePlayer?]
            P2 := CPlayer.GetOpponent()?
            CPlayer2 := TheGlobals.CustomPlayerMap[P2]
            ImgIndexP1 := TheGlobals.PlayerVoteIndex[Self.MaybePlayer?]
            ImgIndexP2 := TheGlobals.PlayerVoteIndex[P2]
            P2UI := CPlayer2.GuessUI
            TextuerBlok := P2UI.PlayerTwoTexture
        then:
            Logger.Print("Player 2 Voted For: {ImgIndexP2} Updating Texture")
            TextuerBlok.SetImage(GetTextureFromStyleIndex(ImgIndexP1))


    UpdateReticle(NewTexture:texture, TextureSize:vector2):void=
        ReticleTextureBlock.SetImage(NewTexture)
        ReticleTextureBlock.SetDesiredSize(TextureSize)


    GetDefaultReticleTexture(Agent:agent, GameBoard:game_board):tuple(texture, vector2)=
        if:
            GameBoard.MakingFinalGuess?
        then:
            return (FinalGuessReticleTexture(0), FinalGuessReticleTexture(1))
        else:
            return (DefaultReticleTexture(0), DefaultReticleTexture(1))


    CreateReticleUI():canvas=
        Print("Creating Reticle UI")
        var ReticleCanvas : canvas = canvas{}

        set ReticleCanvas = canvas:
            Slots := array:
                canvas_slot:
                    Anchors := anchors{Minimum := vector2{X := 0.5, Y := 0.5}, Maximum := vector2{X := 0.5, Y := 0.5} }
                    Offsets := margin{Top := 0.0, Left := 0.0, Right := 0.0, Bottom := 0.0}
                    Alignment := vector2{X := 0.5, Y := 0.5}
                    SizeToContent := true
                    Widget := ReticleTextureBlock

        return ReticleCanvas


    CreateSelectionScreenOverlays():[]overlay=
        Logger.Print("Creating Selection Screen Options")

        var aOptionsOverlay : []overlay = array{}

        for(Option:AllSelectionScreenOptions):
            var NewOption : overlay = overlay{}
            set NewOption = overlay:
                Slots := array:
                    overlay_slot:
                        HorizontalAlignment := horizontal_alignment.Fill
                        VerticalAlignment := vertical_alignment.Fill
                        Padding := margin{Top := 0.0, Left := 0.0, Right := 0.0, Bottom := 0.0}
                        Widget := Option.Widget

            set aOptionsOverlay += array{NewOption}

        return aOptionsOverlay



    CreateSelectionScreenOptions() :void =
        # Logger.Print("Creating Selection Screen Options for {Styles.Length} Styles")
    
        var OptionsPerRow : int = 0
        var Rows : int = 0
    
        # Set options per row and rows based on current UI layout
        if (CurrentUILayout = ui_layout.Horizontal):
            set OptionsPerRow = HorizontalOptionsPerRow
            set Rows = HorizontalRows
        else:
            set OptionsPerRow = VerticalOptionsPerRow
            set Rows = VerticalRows
    
        set OptionsPerPage = Rows * OptionsPerRow
        Logger.Print("Options Per Page: {OptionsPerPage}")
    
        # Sort and categorize styles as before
        SortedStyles := SortStylesByOrder(Styles)
    
        # Create arrays for different categories
        var FortniteSet: []style_selection_data = array{}
        var AllPinned: []style_selection_data = array{}
        var AllAds: []style_selection_data = array{}
        var AllNew: []style_selection_data = array{}
        var AllUpdated: []style_selection_data = array{}
        var AllNormal: []style_selection_data = array{}
    
        var PreMakeRows : []stack_box_slot = array{}
    
        # Categorize styles based on their attributes
        for (Style: Styles):
            if (Style.StyleIndex = 0):  # Fortnite set
                set FortniteSet += array{Style}
            else if (Style.IsAd?):
                set AllAds += array{Style}
            else if (Style.IsNew?):
                set AllNew += array{Style}
            else if (Style.IsUpdated?):
                set AllUpdated += array{Style}
            else if (Style.IsPinned?):
                set AllPinned += array{Style}
            else if (not Style.IsHidden?):
                set AllNormal += array{Style}
    
        # Sort the arrays by SortOrder
        set AllAds = SortStylesByOrder(AllAds)
        set AllNew = SortStylesByOrder(AllNew)
        set AllUpdated = SortStylesByOrder(AllUpdated)
        set AllPinned = SortStylesByOrder(AllPinned)
        set AllNormal = SortStylesByOrder(AllNormal)
    
        set FinalSortedStyles = Concatenate(
            FortniteSet,
            AllAds,
            AllNew,
            AllUpdated,
            AllPinned,
            AllNormal
        )
        
    
        # Logger.Print("Final Sorted Styles: {FinalSortedStyles.Length}")
    
        # Calculate total pages
        if:
            aTotalPages := Ceil(FinalSortedStyles.Length / OptionsPerPage)
        then:
            set TotalPages = aTotalPages
            Logger.Print("Total Pages: {TotalPages}")
    
        # Determine displayed options
        StartIndex := CurrentPage * OptionsPerPage
        EndIndex := Min((CurrentPage + 1) * OptionsPerPage, FinalSortedStyles.Length)
        Logger.Print("Start Index: {StartIndex}, End Index: {EndIndex}")
    
        if:
            DisplayedOptions := FinalSortedStyles.Slice[StartIndex, EndIndex]
        then:
            # Logger.Print("Displayed Options: {DisplayedOptions.Length}")
    
            # Build rows
            if:
                NumberOfRows := Ceil(DisplayedOptions.Length / OptionsPerRow)
            then:
                for (RowIndex := 0 .. NumberOfRows - 1):
                    StartIdx := RowIndex * OptionsPerRow
                    EndIdx := Min(StartIdx + OptionsPerRow, DisplayedOptions.Length)
                    if:
                        RowStyles := DisplayedOptions.Slice[StartIdx, EndIdx]
                    then:
                        set PreMakeRows += array{BuildRow(RowIndex, RowStyles)}
    
        # Create the overlay
        aOptionsOverlay := overlay:
            Slots := array:
                overlay_slot:
                    HorizontalAlignment := horizontal_alignment.Center
                    VerticalAlignment := vertical_alignment.Fill
                    Padding := margin{Top := 0.0, Left := 0.0, Right := 0.0, Bottom := 0.0}
                    Widget := stack_box:
                        Orientation := orientation.Vertical
                        Slots := PreMakeRows
    
        # Remove existing overlay if present
        if (HasOverlay := ExistingOptionsOverlay?):
            Logger.Print("Removing Existing Options Overlay")
            OptionsOverlay.RemoveWidget(HasOverlay)
    
        # Add the new overlay
        OverlaySlot := overlay_slot:
            HorizontalAlignment := horizontal_alignment.Fill
            VerticalAlignment := vertical_alignment.Fill
            Widget := aOptionsOverlay
    
        set ExistingOptionsOverlay = option{aOptionsOverlay}
        OptionsOverlay.AddWidget(OverlaySlot)

        for(StyleOption : AllSelectionScreenOptions):
            # StyleOption.Button.OnClick().SubscribeMessage(OnClickButtonStyle, StyleOption.StyleIndex, StyleOption.StyleTexture)
            StyleOption.Button.OnClick().SubscribeMessage(OnClickButtonStyleNew, StyleOption.StyleIndex)
            # StyleOption.Button.OnClick().Subscribe(OnClickButtonStyleTEST)
            # Logger.Print("Subscribed to Style Option Button")

    
        # Update pagination buttons
        UpdatePaginationButtons()


    
    


    BuildRow(RowIndex:int, RowStyles: []style_selection_data):stack_box_slot =
        var RowSlots : []stack_box_slot = array{}
    
        for (Style : RowStyles):



            set RowSlots += array{stack_box_slot:
                Distribution := option{1.0}
                HorizontalAlignment := horizontal_alignment.Fill
                VerticalAlignment := vertical_alignment.Fill
                Padding := margin{Top := 0.0, Left := 0.0, Right := 0.0, Bottom := 0.0}
                Widget := CreateStyleOption(Style).Widget
            }
    
        return stack_box_slot:
            # Padding := if(RowIndex > 0) then margin{Top := -40.0} else margin{Top := -20.0} #### If adding labels to options then use this padding
            HorizontalAlignment := horizontal_alignment.Fill
            VerticalAlignment := vertical_alignment.Fill
            Widget := stack_box:
                Orientation := orientation.Horizontal
                Slots := RowSlots

    UpdatePaginationButtons(): void =
        # Enable or disable buttons based on current page
        if (CurrentPage = 0):
            if (CurrentUILayout = ui_layout.Horizontal):
                ButtonPageLeft.SetEnabled(false)
                # ButtonPageLeft.SetVisibility(widget_visibility.Hidden)
            else:
                # ButtonPageUp.SetEnabled(false)
                ButtonPageLeft.SetEnabled(false)
                # ButtonPageLeft.SetVisibility(widget_visibility.Hidden)
        else:
            if (CurrentUILayout = ui_layout.Horizontal):
                ButtonPageLeft.SetEnabled(true)
                ButtonPageLeft.SetVisibility(widget_visibility.Visible)
            else:
                # ButtonPageUp.SetEnabled(true)
                ButtonPageLeft.SetEnabled(true)
                ButtonPageLeft.SetVisibility(widget_visibility.Visible)
    
        if ((CurrentPage + 1) * OptionsPerPage >= FinalSortedStyles.Length):
            if (CurrentUILayout = ui_layout.Horizontal):
                ButtonPageRight.SetEnabled(false)
                # ButtonPageRight.SetVisibility(widget_visibility.Hidden)
            else:
                # ButtonPageDown.SetEnabled(false)
                ButtonPageRight.SetEnabled(false)
                # ButtonPageRight.SetVisibility(widget_visibility.Hidden)
        else:
            if (CurrentUILayout = ui_layout.Horizontal):
                ButtonPageRight.SetEnabled(true)
                ButtonPageRight.SetVisibility(widget_visibility.Visible)
            else:
                # ButtonPageDown.SetEnabled(true)
                ButtonPageRight.SetEnabled(true)
                ButtonPageRight.SetVisibility(widget_visibility.Visible)
    
    
        
    

    OnClickButtonPageLeft(Message: widget_message): void =
        Logger.Print("Clicked Page Left")
        if (CurrentPage > 0):
            set CurrentPage = Max(CurrentPage - 1, 0)
            RefreshSelectionScreenOptions()
    
    OnClickButtonPageRight(Message: widget_message): void =
        Logger.Print("Clicked Page Right")
        if (CurrentPage + 1 < TotalPages):
            set CurrentPage = Min(CurrentPage + 1, TotalPages - 1)
            RefreshSelectionScreenOptions()
    
    OnClickButtonPageUp(Message: widget_message): void =
        Logger.Print("Clicked Page Up")
        if (CurrentPage > 0):
            set CurrentPage = Max(CurrentPage - 1, 0)
            RefreshSelectionScreenOptions()
    
    OnClickButtonPageDown(Message: widget_message): void =
        Logger.Print("Clicked Page Down")
        if (CurrentPage + 1 < TotalPages):
            set CurrentPage = Min(CurrentPage + 1, TotalPages - 1)
            RefreshSelectionScreenOptions()
    
    RefreshSelectionScreenOptions(): void =
        OptionsOverlay.SetVisibility(widget_visibility.Hidden)
        if (HasOverlay := ExistingOptionsOverlay?):
            OptionsOverlay.RemoveWidget(HasOverlay)
        
        CreateSelectionScreenOptions()

        OptionsOverlay.SetVisibility(widget_visibility.Visible)
        
    
    OnClickButtonUILayoutSwitcher(Message: widget_message): void =
        Logger.Print("Clicked UI Layout Switcher")
        if (CurrentUILayout = ui_layout.Horizontal):
            set CurrentUILayout = ui_layout.Vertical
            ButtonUILayoutSwitcher.SetText(MessageSwitchLayoutHorizontal_EN)
            SetStreamerModeVerticalUI(Message.Player, true)
        else:
            set CurrentUILayout = ui_layout.Horizontal
            ButtonUILayoutSwitcher.SetText(MessageSwitchLayoutVertical_EN)
            SetStreamerModeVerticalUI(Message.Player, false)
    
        # Reset the current page when layout changes
        set CurrentPage = 0
    
        # Refresh the selection screen options
        RefreshSelectionScreenOptions()

        RefreshVoteUI()

        
    
    RefreshVoteUI():void=
        # hide text if vertical ui
        if(CurrentUILayout = ui_layout.Vertical):
            PlayerOneVoteTextureBlock.SetVisibility(widget_visibility.Collapsed)
            PlayerTwoVoteTextureBlock.SetVisibility(widget_visibility.Collapsed)
        else:
            PlayerOneVoteTextureBlock.SetVisibility(widget_visibility.Visible)
            PlayerTwoVoteTextureBlock.SetVisibility(widget_visibility.Visible)
        
    
    OnClickButtonChooseCharacterMethod(Message:widget_message):void=
        Logger.Print("Clicked Choose Character Method")
        # ButtonText := ButtonChooseCharacterMethod.GetText()
        if:
            TheGlobals := Globals[GetSession()]
            CPlayer := TheGlobals.CustomPlayerMap[Message.Player]
            GameBoard := CPlayer.GameBoard
        then:
            if:
                not GameBoard.IsManualCharacterSelection?
            then:
                Print("Choose Character Method")
                ButtonChooseCharacterMethod.SetText(MessageRandomCharacterIcon_EN)
                TextBlockChooseCharacterUpdated.SetText(MessageManualCharacterChoice_EN)
                SetManuallyChooseCharacter(Message.Player, true)
                set GameBoard.IsManualCharacterSelection = true 


                # spawn . ClearCharacterSelectionText()
            else:
                Print("Random Character Method")
                ButtonChooseCharacterMethod.SetText(MessageChooseCharacterIcon_EN)
                TextBlockChooseCharacterUpdated.SetText(MessageRandomCharacterChoice_EN)
                SetManuallyChooseCharacter(Message.Player, false)
                set GameBoard.IsManualCharacterSelection = false

                # spawn . ClearCharacterSelectionText()
        else:
            Logger.Print("OnClickButtonChooseCharacterMethod Failed: Player Not Found")


    ClearCharacterSelectionText()<suspends>:void=
        Sleep(2.0)
        TextBlockChooseCharacterUpdated.SetText(StringToMessage(""))


    
    CreateStyleOption(Data:style_selection_data):selection_screen_option=
        # Logger.Print("Creating Style Option")
        Option := selection_screen_option{}
        Option.Create(Data)
        set AllSelectionScreenOptions += array{Option}
        Option
        

    OnClickButtonStyleNew(Message:widget_message, StyleIndex:int):void=
        Logger.Print("Style Selected: {StyleIndex}")
        Agent := Message.Player
        if:
            Player := player[Agent]
            TheGlobals := Globals[GetSession()]
            CPlayer := TheGlobals.CustomPlayerMap[Player]
        then:
            if:
                set TheGlobals.PlayerVoteIndex[Player] = StyleIndex
            then:
                set CPlayer.HasVoted = true
                Logger.Print("Player Voted For: {StyleIndex}")
                YourVoteTexture.SetImage(GetTextureFromStyleIndex(StyleIndex))
                if:
                    Opponent := CPlayer.GetOpponent()?
                    OpponentCP := TheGlobals.CustomPlayerMap[Opponent]
                then:
                    OpponentCP.GuessUI.TheirVoteTexture.SetImage(GetTextureFromStyleIndex(StyleIndex))
        else:
            Logger.Print("OnClickButtonStyle Failed: Player Not Found")
    
    

    CreateSelectionScreenUI_NEW(Player:player):canvas=

        ButtonPageLeft.SetEnabled(false)
        ButtonPageLeft.SetText(MessageArrowLeft_EN)
        ButtonPageRight.SetText(MessageArrowRight_EN)

        if(not ButtonPageLeftSubscription?):
            LeftSub := ButtonPageLeft.OnClick().Subscribe(OnClickButtonPageLeft)
            set ButtonPageLeftSubscription = option{LeftSub}

        if(not ButtonPageRightSubscription?):
            RightSub := ButtonPageRight.OnClick().Subscribe(OnClickButtonPageRight)
            set ButtonPageRightSubscription = option{RightSub}

        if(not ButtonUILayoutSwitcherSubscription?):
            UISub := ButtonUILayoutSwitcher.OnClick().Subscribe(OnClickButtonUILayoutSwitcher)
            set ButtonUILayoutSwitcherSubscription = option{UISub}

        if(not ButtonChooseCharacterMethodSubscription?):
            RandomSub := ButtonChooseCharacterMethod.OnClick().Subscribe(OnClickButtonChooseCharacterMethod)
            set ButtonChooseCharacterMethodSubscription = option{RandomSub}


        if(CurrentUILayout = ui_layout.Horizontal):
            ButtonUILayoutSwitcher.SetText(MessageSwitchLayoutVertical_EN)
            PlayerOneVoteTextureBlock.SetVisibility(widget_visibility.Visible)
            PlayerTwoVoteTextureBlock.SetVisibility(widget_visibility.Visible)
        else:
            ButtonUILayoutSwitcher.SetText(MessageSwitchLayoutHorizontal_EN)
            PlayerOneVoteTextureBlock.SetVisibility(widget_visibility.Collapsed)
            PlayerTwoVoteTextureBlock.SetVisibility(widget_visibility.Collapsed)


        if:
        #    GetManuallyChooseCharacter[Player]
            TheGlobals := Globals[GetSession()]
            CPlayer := TheGlobals.CustomPlayerMap[Player]
            GameBoard := CPlayer.GameBoard
            GameBoard.IsManualCharacterSelection?
        then:
            TextBlockChooseCharacterUpdated.SetText(MessageManualCharacterChoice_EN) 
            ButtonChooseCharacterMethod.SetText(MessageRandomCharacterIcon_EN)
        else:
            TextBlockChooseCharacterUpdated.SetText(MessageRandomCharacterChoice_EN)
            ButtonChooseCharacterMethod.SetText(MessageChooseCharacterIcon_EN)


        

        # stack_box:
        #     Orientation := orientation.Vertical
        #     Slots := array:
        #         stack_box_slot:
        #             Distribution := option{2.0}
        #             HorizontalAlignment := horizontal_alignment.Fill
        #             VerticalAlignment := vertical_alignment.Fill
        #             Padding := margin{Top := 0.0, Left := 0.0, Right := 0.0, Bottom := 20.0}
        #             Widget := ButtonPageLeft
                # stack_box_slot:
                #     HorizontalAlignment := horizontal_alignment.Fill
                #     VerticalAlignment := vertical_alignment.Bottom
                #     Padding := margin{Top := 0.0, Left := 0.0, Right := 0.0, Bottom := 0.0}
                #     Widget := ButtonUILayoutSwitcher



    
        # LeftUISwitcherButtonSlot : stack_box_slot = stack_box_slot:
        #     HorizontalAlignment := horizontal_alignment.Fill
        #     VerticalAlignment := vertical_alignment.Bottom
        #     Padding := margin{Top := 0.0, Left := 0.0, Right := 0.0, Bottom := 0.0}
        #     Widget := ButtonUILayoutSwitcher


        Logger.Print("Creating Selection Screen UI NEW")

        var LayoutCanvas : canvas = canvas{}

        set LayoutCanvas = canvas:
            Slots := array:
                canvas_slot:
                    Anchors := anchors{Minimum := vector2{X := 0.0, Y := 0.0}, Maximum := vector2{X := 1.0, Y := 1.0} }
                    Offsets := margin{Top := 0.0, Left := 0.0, Right := 0.0, Bottom := 0.0}
                    Alignment := vector2{X := 0.5, Y := 0.5}
                    ZOrder := 8000
                    Widget := overlay:
                        Slots := array:
                            overlay_slot:
                                HorizontalAlignment := horizontal_alignment.Fill
                                VerticalAlignment := vertical_alignment.Fill
                                Padding := margin{Top := 0.0, Left := 0.0, Right := 0.0, Bottom := 0.0}
                                Widget := texture_block:
                                    DefaultImage := TextureBackground
                                    DefaultTint := MakeColorFromSRGB(0.0, 0.0, 1.0)
                                    DefaultDesiredSize := vector2{X := 1920.0, Y := 1080.0}
                            overlay_slot:
                                HorizontalAlignment := horizontal_alignment.Fill
                                VerticalAlignment := vertical_alignment.Fill
                                Padding := margin{Top := 0.0, Left := 0.0, Right := 0.0, Bottom := 0.0}
                                Widget := stack_box:
                                    Orientation := orientation.Vertical
                                    Slots := array:

                                        # Selection Screen Title
                                        stack_box_slot:
                                            HorizontalAlignment := horizontal_alignment.Center
                                            VerticalAlignment := vertical_alignment.Center
                                            Padding := margin{Top := 0.0, Left := 0.0, Right := 0.0, Bottom := 0.0}
                                            Widget := texture_block:
                                                DefaultImage := TextureInstructions_SelectionScreen
                                                DefaultDesiredSize := vector2{X := 1920.0/1.5, Y := 293.0/1.5}

                                        # Selection Screen Options (Pagination)
                                        stack_box_slot:
                                            Distribution := option{2.0}
                                            HorizontalAlignment := horizontal_alignment.Center
                                            VerticalAlignment := vertical_alignment.Fill
                                            Padding := margin{Top := 0.0, Left := 0.0, Right := 0.0, Bottom := 0.0}
                                            Widget := stack_box:
                                                Orientation := orientation.Horizontal
                                                Slots := array:                                                        
                                                    # stack_box_slot:
                                                    #     HorizontalAlignment := horizontal_alignment.Fill
                                                    #     VerticalAlignment := vertical_alignment.Fill
                                                    #     Padding := margin{Top := 0.0, Left := 0.0, Right := 40.0, Bottom := 30.0}
                                                    #     Widget := stack_box:
                                                    #         Orientation := orientation.Vertical
                                                    #         Slots := array:
                                                    #             stack_box_slot:
                                                    #                 Distribution := option{2.0}
                                                    #                 HorizontalAlignment := horizontal_alignment.Fill
                                                    #                 VerticalAlignment := vertical_alignment.Fill
                                                    #                 Padding := margin{Top := 0.0, Left := 0.0, Right := 0.0, Bottom := 20.0}
                                                    #                 Widget := ButtonPageLeft
                                                    #             stack_box_slot:
                                                    #                 HorizontalAlignment := horizontal_alignment.Fill
                                                    #                 VerticalAlignment := vertical_alignment.Bottom
                                                    #                 Padding := margin{Top := 10.0, Left := 0.0, Right := 0.0, Bottom := 0.0}
                                                    #                 Widget := ButtonUILayoutSwitcher

                                                    stack_box_slot:
                                                        Distribution := option{2.0}
                                                        HorizontalAlignment := horizontal_alignment.Fill
                                                        VerticalAlignment := vertical_alignment.Fill
                                                        Padding := margin{Top := 0.0, Left := 0.0, Right := 0.0, Bottom := 0.0}
                                                        Widget := OptionsOverlay

                                                    # stack_box_slot:
                                                    #     HorizontalAlignment := horizontal_alignment.Fill
                                                    #     VerticalAlignment := vertical_alignment.Fill
                                                    #     Padding := margin{Top := 0.0, Left := 40.0, Right := 0.0, Bottom := 30.0}
                                                    #     Widget := stack_box:
                                                    #         Orientation := orientation.Vertical
                                                    #         Slots := array:
                                                    #             stack_box_slot:
                                                    #                 Distribution := option{2.0}
                                                    #                 HorizontalAlignment := horizontal_alignment.Fill
                                                    #                 VerticalAlignment := vertical_alignment.Fill
                                                    #                 Padding := margin{Top := 0.0, Left := 0.0, Right := 0.0, Bottom := 20.0}
                                                    #                 Widget := ButtonPageRight
                                                    #             stack_box_slot:
                                                    #                 HorizontalAlignment := horizontal_alignment.Fill
                                                    #                 VerticalAlignment := vertical_alignment.Bottom
                                                    #                 Padding := margin{Top := 0.0, Left := 0.0, Right := 0.0, Bottom := 0.0}
                                                    #                 Widget := ButtonChooseCharacterMethod
                                        # Vote Display
                                        CreateVoteDisplay()

                            overlay_slot:
                                HorizontalAlignment := horizontal_alignment.Right
                                VerticalAlignment := vertical_alignment.Top
                                Padding := margin{Top := 100.0, Left := 0.0, Right := 100.0, Bottom := 0.0}
                                Widget := TextBlockChooseCharacterUpdated

        return LayoutCanvas


    

    CreateVoteDisplay():stack_box_slot=

        stack_box_slot:
            HorizontalAlignment := horizontal_alignment.Center
            VerticalAlignment := vertical_alignment.Bottom
            Padding := margin{Top := 0.0, Left := 0.0, Right := 0.0, Bottom := 20.0}
            Widget := stack_box:
                Orientation := orientation.Horizontal
                Slots := array:


                    # Left Buttons 
                    stack_box_slot:
                        HorizontalAlignment := horizontal_alignment.Fill
                        VerticalAlignment := vertical_alignment.Bottom
                        Padding := margin{Top := 0.0, Left := 0.0, Right := 40.0, Bottom := 0.0}
                        Widget := stack_box:
                            Orientation := orientation.Vertical
                            Slots := array:
                                stack_box_slot:
                                    HorizontalAlignment := horizontal_alignment.Fill
                                    VerticalAlignment := vertical_alignment.Fill
                                    Padding := margin{Top := 0.0, Left := 0.0, Right := 0.0, Bottom := 20.0}
                                    Widget := ButtonPageLeft
                                stack_box_slot:
                                    HorizontalAlignment := horizontal_alignment.Fill
                                    VerticalAlignment := vertical_alignment.Bottom
                                    Padding := margin{Top := 10.0, Left := 0.0, Right := 0.0, Bottom := 0.0}
                                    Widget := ButtonUILayoutSwitcher

                    stack_box_slot:
                        HorizontalAlignment := horizontal_alignment.Fill
                        VerticalAlignment := vertical_alignment.Fill
                        Padding := margin{Top := 0.0, Left := 0.0, Right := 0.0, Bottom := 0.0}
                        Widget := stack_box:
                            Orientation := orientation.Vertical
                            Slots := array:
                                stack_box_slot:
                                    Distribution := option{50.0}
                                    HorizontalAlignment := horizontal_alignment.Fill
                                    VerticalAlignment := vertical_alignment.Fill
                                    Padding := margin{Top := 0.0, Left := 0.0, Right := 0.0, Bottom := 0.0}
                                    Widget := stack_box:
                                        Orientation := orientation.Horizontal
                                        Slots := array:
                                            # Player One Vote
                                            stack_box_slot:
                                                HorizontalAlignment := horizontal_alignment.Center
                                                VerticalAlignment := vertical_alignment.Center
                                                Padding := margin{Top := 0.0, Left := 0.0, Right := 0.0, Bottom := 0.0}
                                                Widget := overlay:
                                                    Slots := array:
                                                        # Your Vote Background
                                                        overlay_slot:
                                                            HorizontalAlignment := horizontal_alignment.Fill
                                                            VerticalAlignment := vertical_alignment.Fill
                                                            Padding := margin{Top := 0.0, Left := 0.0, Right := 0.0, Bottom := 0.0}
                                                            Widget := MakeRoundedOverlayOuter(MakeColorFromHex("#007FFF"))
                                                        overlay_slot:
                                                            HorizontalAlignment := horizontal_alignment.Fill
                                                            VerticalAlignment := vertical_alignment.Fill
                                                            Padding := margin{Top := 8.0, Left := 8.0, Right := 8.0, Bottom := 8.0}
                                                            Widget := overlay:
                                                                Slots := array:
                                                                    overlay_slot:
                                                                        HorizontalAlignment := horizontal_alignment.Fill
                                                                        VerticalAlignment := vertical_alignment.Fill
                                                                        Padding := margin{Top := 0.0, Left := 0.0, Right := 0.0, Bottom := 0.0}
                                                                        Widget := MakeRoundedOverlayInner(MakeColorFromHex("#003B8C"))
                                                                    # Your Vote Content
                                                                    overlay_slot:
                                                                        HorizontalAlignment := horizontal_alignment.Fill
                                                                        VerticalAlignment := vertical_alignment.Fill
                                                                        Padding := margin{Top := 0.0, Left := 0.0, Right := 0.0, Bottom := 0.0}
                                                                        Widget := stack_box:
                                                                            Orientation := orientation.Horizontal
                                                                            Slots := array:
                                                                                stack_box_slot:
                                                                                    HorizontalAlignment := horizontal_alignment.Center
                                                                                    VerticalAlignment := vertical_alignment.Center
                                                                                    Padding := margin{Top := 0.0, Left := 0.0, Right := 0.0, Bottom := 0.0}
                                                                                    Widget := PlayerOneVoteTextureBlock
                                                                                stack_box_slot:
                                                                                    HorizontalAlignment := horizontal_alignment.Center
                                                                                    VerticalAlignment := vertical_alignment.Center
                                                                                    Padding := margin{Top := 0.0, Left := 20.0, Right := 20.0, Bottom := 0.0}
                                                                                    Widget := YourVoteTexture
                                            # Timer
                                            stack_box_slot:
                                                HorizontalAlignment := horizontal_alignment.Center
                                                VerticalAlignment := vertical_alignment.Center
                                                Padding := margin{Top := 8.0, Left := 20.0, Right := 20.0, Bottom := 8.0}
                                                Widget := overlay:
                                                    Slots := array:
                                                        overlay_slot:
                                                            HorizontalAlignment := horizontal_alignment.Center
                                                            VerticalAlignment := vertical_alignment.Center
                                                            Padding := margin{Top := 0.0, Left := 0.0, Right := 0.0, Bottom := 0.0}
                                                            Widget := TimerBGTexture
                                                        overlay_slot:
                                                            HorizontalAlignment := horizontal_alignment.Center
                                                            VerticalAlignment := vertical_alignment.Center
                                                            Padding := margin{Top := 0.0, Left := 0.0, Right := 0.0, Bottom := 0.0}
                                                            Widget := TimerTimeTexture

                                            # Player Two Vote
                                            stack_box_slot:
                                                HorizontalAlignment := horizontal_alignment.Center
                                                VerticalAlignment := vertical_alignment.Center
                                                Padding := margin{Top := 0.0, Left := 0.0, Right := 0.0, Bottom := 0.0}
                                                Widget := overlay:
                                                    Slots := array:
                                                        # Their Vote Background
                                                        overlay_slot:
                                                            HorizontalAlignment := horizontal_alignment.Fill
                                                            VerticalAlignment := vertical_alignment.Fill      
                                                            Padding := margin{Top := 0.0, Left := 0.0, Right := 0.0, Bottom := 0.0}
                                                            Widget := MakeRoundedOverlayOuter(MakeColorFromHex("#007FFF"))
                                                        overlay_slot:
                                                            HorizontalAlignment := horizontal_alignment.Fill
                                                            VerticalAlignment := vertical_alignment.Fill
                                                            Padding := margin{Top := 8.0, Left := 8.0, Right := 8.0, Bottom := 8.0}
                                                            Widget := overlay:
                                                                Slots := array:
                                                                    overlay_slot:
                                                                        HorizontalAlignment := horizontal_alignment.Fill
                                                                        VerticalAlignment := vertical_alignment.Fill
                                                                        Padding := margin{Top := 0.0, Left := 0.0, Right := 0.0, Bottom := 0.0}
                                                                        Widget := MakeRoundedOverlayInner(MakeColorFromHex("#003B8C"))
                                                                    # Their Vote Content
                                                                    overlay_slot:
                                                                        HorizontalAlignment := horizontal_alignment.Fill
                                                                        VerticalAlignment := vertical_alignment.Fill
                                                                        Padding := margin{Top := 0.0, Left := 0.0, Right := 0.0, Bottom := 0.0}
                                                                        Widget := stack_box:
                                                                            Orientation := orientation.Horizontal
                                                                            Slots := array:
                                                                                stack_box_slot:
                                                                                    HorizontalAlignment := horizontal_alignment.Left
                                                                                    VerticalAlignment := vertical_alignment.Center
                                                                                    Padding := margin{Top := 0.0, Left := 20.0, Right := 20.0, Bottom := 0.0}
                                                                                    Widget := TheirVoteTexture
                                                                                stack_box_slot:
                                                                                    HorizontalAlignment := horizontal_alignment.Center
                                                                                    VerticalAlignment := vertical_alignment.Center
                                                                                    Padding := margin{Top := 0.0, Left := 0.0, Right := 0.0, Bottom := 0.0}
                                                                                    Widget := PlayerTwoVoteTextureBlock

                                stack_box_slot:
                                    Distribution := option{1.0}
                                    HorizontalAlignment := horizontal_alignment.Fill
                                    VerticalAlignment := vertical_alignment.Bottom
                                    Padding := margin{Top := 0.0, Left := 0.0, Right := 0.0, Bottom := 0.0}
                                    Widget := button_quiet{}


                    stack_box_slot:
                        HorizontalAlignment := horizontal_alignment.Fill
                        VerticalAlignment := vertical_alignment.Bottom
                        Padding := margin{Top := 0.0, Left := 40.0, Right := 0.0, Bottom := 0.0}
                        Widget := stack_box:
                            Orientation := orientation.Vertical
                            Slots := array:
                                stack_box_slot:
                                    HorizontalAlignment := horizontal_alignment.Fill
                                    VerticalAlignment := vertical_alignment.Fill
                                    Padding := margin{Top := 0.0, Left := 0.0, Right := 0.0, Bottom := 20.0}
                                    Widget := ButtonPageRight
                                stack_box_slot:
                                    HorizontalAlignment := horizontal_alignment.Fill
                                    VerticalAlignment := vertical_alignment.Bottom
                                    Padding := margin{Top := 0.0, Left := 0.0, Right := 0.0, Bottom := 0.0}
                                    Widget := ButtonChooseCharacterMethod


    RemoveSelectionScreen(Player:player):void=
        Logger.Print("Attempting to Remove Selection Screen")

        # spawn { SendVotedNotification() }

        if (MyUI := MaybeStyleSelectionUI?, PlayerUI := GetPlayerUI[Player]):
            PlayerUI.RemoveWidget(MyUI)
            set MaybeStyleSelectionUI = false
        else:
            Logger.Print("RemoveSelectionScreen Failed: No UI Found")

        # remove preloaded textures
        if(PreLoaded := MaybePreloadedCanvas?, PlayerUI := GetPlayerUI[Player]):
            PlayerUI.RemoveWidget(PreLoaded)
            set MaybePreloadedCanvas = false
        else:
            Logger.Print("RemovePreloadedCanvas Failed: No UI Found")


        # remove subsciptions
        if(HasLeftSub := ButtonPageLeftSubscription?):
            HasLeftSub.Cancel()
            set ButtonPageLeftSubscription = false

        if(HasRightSub := ButtonPageRightSubscription?):
            HasRightSub.Cancel()
            set ButtonPageRightSubscription = false

        if(HasUISub := ButtonUILayoutSwitcherSubscription?):
            HasUISub.Cancel()
            set ButtonUILayoutSwitcherSubscription = false


    RemoveQuestionUI(Player:player)<suspends>:void=
        Logger.Print("Attempting to Remove Question UI")
        for(I := 0..3):
            if (MyUI := MaybeQuestionUI?, PlayerUI := GetPlayerUI[Player]):
                PlayerUI.RemoveWidget(MyUI)
                set MaybeQuestionUI = false
            else:
                Logger.Print("RemoveQuestionUI Failed: No UI Found")
            Sleep(0.33)

   

   
    ##################################################################################################### NEEDS UPDATE WHEN NEW STYLES ARE ADDED
    GetTextureFromStyleIndex(Index:int):texture=
        case(Index):
            0 => TextureFNStyle
            1 => TextureClassicStyle
            2 => TextureIconStyle
            3 => TextureAnimalStyle
            4 => TextureUtilStyle
            5 => TextureMarvelStyle
            6 => TexturePOIStyle
            7 => TextureLegendStyle
            8 => TextureFoodStyle
            9 => TextureStarStyle
            10 => TextureSpookyStyle
            11 => TextureOGStyle
            12 => TextureAnimeStyle
            13 => TextureWinterStyle
            14 => TextureWeaponsStyle
            15 => TexturePickaxesStyle
            16 => TextureVehiclesStyle
            17 => TextureMoviesTVStyle
            18 => TextureBlocksStyle
            19 => TextureChapter5Style
            20 => TextureLockersStyle
            21 => TextureBackblingsStyle
            22 => TextureEmotesStyle
            23 => TextureMarvelStyleSet1
            24 => TextureMarvelStyleSet2
            25 => TextureUmbrellasStyle
            26 => TexturePeelyFishstickStyle
            27 => TextureRemixStyle
            28 => TextureFireIceStyle
            29 => TextureDCStyle
            30 => TextureChapter6Style
            _ => TextureComingSoon  # Default to coming soon texture for undefined indices


    ShowYourTurnUI(Player:player):void=
        if:
            MyUI := MaybeYourTurnUI?
        then:
            Logger.Print("Showing Your Turn UI")
            ShowWidget(MyUI)
            # spawn{DisplayUpdatingTimer(TimerTextBlock2)}
            spawn { UpdateTimerTextureLoop(TimerBGTexture, TimerTimeTexture) }
        else:
            Logger.Print("ShowYourTurnUI Failed: No UI Found")
            

    HideYourTurnUI(Player:player):void=
        if:
            MyUI := MaybeYourTurnUI?
        then:
            Logger.Print("Hiding Your Turn UI")
            HideWidget(MyUI)
            set IsTimerActive = false
        else:
            Logger.Print("HideYourTurnUI Failed: No UI Found")

    RemoveYourTurnUI(Player:player):void=
        if:
            MyUI := MaybeYourTurnUI?
            PlayerUI := GetPlayerUI[Player]
        then:
            PlayerUI.RemoveWidget(MyUI)
            set MaybeYourTurnUI = false
            set IsTimerActive = false

    HideWidget(Widget:widget):void=
        Widget.SetVisibility(widget_visibility.Collapsed)

    ShowWidget(Widget:widget):void=
        Widget.SetVisibility(widget_visibility.Visible)


    HideShootToGuessUI(Player:player):void=
        if(MyUI := MaybeShootToGuessUI?):
            HideWidget(MyUI)
        else:
            Logger.Print("HideShootToGuessUI Failed: No UI Found")

    ShowShootToGuessUI(Player:player):void=
        if(MyUI := MaybeShootToGuessUI?):
            ShowWidget(MyUI)
        else:
            Logger.Print("ShowShootToGuessUI Failed: No UI Found")


    InitCorrectIncorrectUI(Player:player):void=
        if(MyUI := MaybeCorrectIncorrectUI?):
            ShowWidget(MyUI)
        else:
            CorrectIncorrectUI := CreateCorrectIncorrectUIWidget(Player)
            AddCanvasToUI(Player, CorrectIncorrectUI, ui_input_mode.None, 2147483646)
            set MaybeCorrectIncorrectUI = option{CorrectIncorrectUI}

    InitReticleUI(Player:player):void=
        if(MyUI := MaybeReticleUI?):
            Print("Reticle UI Already Exists - Showing")
            ShowWidget(MyUI)
        else:
            Print("Creating Reticle UI - For First Time")
            ReticleUI := CreateReticleUI()
            AddCanvasToUI(Player, ReticleUI, ui_input_mode.None, 1000)
            set MaybeReticleUI = option{ReticleUI}


    CreateCorrectIncorrectUIWidget(Player:player):canvas=
        var ui : canvas = canvas{}

        set ui = canvas:
            Slots := array:
                canvas_slot:
                    Anchors := anchors{Minimum := vector2{X := 0.5, Y := 0.5}, Maximum := vector2{X := 0.5, Y := 0.5}}
                    Offsets := margin{Top := 100.0, Left := 0.0, Right := 0.0, Bottom := 0.0}
                    Alignment := vector2{X := 0.5, Y := 0.5}
                    ZOrder := 121
                    Widget := stack_box:
                        Orientation := orientation.Vertical
                        Slots := array:
                            stack_box_slot:
                                HorizontalAlignment := horizontal_alignment.Center
                                VerticalAlignment := vertical_alignment.Center
                                Padding := margin{Top := 0.0, Left := 0.0, Right := 0.0, Bottom := 5.0}
                                Widget := CorrectIncorrectGuesserText
                            stack_box_slot:
                                HorizontalAlignment := horizontal_alignment.Center
                                VerticalAlignment := vertical_alignment.Center
                                Padding := margin{Top := 0.0, Left := 0.0, Right := 0.0, Bottom := 0.0}
                                Widget := CorrectIncorrectTextureBlock

        CorrectIncorrectGuesserText.SetShadowOpacity(0.75)
        return ui

    
    HideCorrectIncorrectUI(Player:player):void=
        if(MyUI := MaybeCorrectIncorrectUI?):
            HideWidget(MyUI)
        else:
            Logger.Print("HideCorrectIncorrectUI Failed: No UI Found")
    
    UpdateCorrectIncorrectImage(Guesser:player, Player:player, IsCorrect:logic):void=
        DownScaleAmount:= 0.5
        if:
            MyUI := MaybeCorrectIncorrectUI?
        then:
            if(IsCorrect?):
                CorrectIncorrectTextureBlock.SetImage(TextureCorrectText)
                CorrectIncorrectGuesserText.SetText(MakeGuesserMessage(Guesser))
                CorrectIncorrectTextureBlock.SetDesiredSize(vector2{X:= 1200.00 * DownScaleAmount, Y:= 293.00 * DownScaleAmount})
            else:
                CorrectIncorrectTextureBlock.SetImage(TextureIncorrectText)
                CorrectIncorrectGuesserText.SetText(MakeGuesserMessage(Guesser))
                CorrectIncorrectTextureBlock.SetDesiredSize(vector2{X:= 1466.00 * DownScaleAmount, Y:= 293.00 * DownScaleAmount})

        ShowCorrectIncorrectUI(Player)
        
    ShowCorrectIncorrectUI(Player:player):void=
        if(MyUI := MaybeCorrectIncorrectUI?):
            ShowWidget(MyUI)
        else:
            Logger.Print("ShowCorrectIncorrectUI Failed: No UI Found")

   

   

    AnimateTextureBlock(ImgSlot:texture_block, LerpDuration:float, StartSize:vector2, EndSize:vector2 )<suspends>: void =
        var time : float= GetSimulationElapsedTime()
        var timeElapsed : float = 0.0
        Logger.Print("{time} Button Interacted")
        Logger.Print("{StartSize} StartSize, {EndSize} EndSize")
        loop:
    
            Sleep(0.0)
            CurrentTime := GetSimulationElapsedTime()
            DeltaTime := (CurrentTime - time)
            set time = CurrentTime 
            if(timeElapsed < LerpDuration):
                var t :float= timeElapsed / LerpDuration
                set t = InOut(t)
                LerpedValueX := Lerp(StartSize.X, EndSize.X, t)
                LerpedValueY:= Lerp(StartSize.Y, EndSize.Y, t)
                ImgSlot.SetDesiredSize(vector2{X := LerpedValueX, Y := LerpedValueY})
                
                set timeElapsed +=  DeltaTime
            else:
                break
                
        ImgSlot.SetDesiredSize(vector2{X := EndSize.X, Y := EndSize.Y})


    

    InOut(k:float):float=
        var tempK : float = k
        if(k = 0.0) {return 0.0}
        if(k = 1.0) {return 1.0}

        if((set tempK *= 2.0) < 1.0):
            return -0.5*Pow(2.0, 10.0*(set tempK -= 1.0))*Sin((k - 0.1)*(2.0*PiFloat)/0.4)
        else:
            return Pow(2.0, -10.0*(set tempK -= 1.0))*Sin((k - 0.1)*(2.0*PiFloat)/0.4)*0.5 + 1.0;
