using { /Fortnite.com/Devices }
using { /Verse.org/Simulation }
using { /UnrealEngine.com/Temporary/Diagnostics }
using { /UnrealEngine.com/Temporary/SpatialMath }
using { /Fortnite.com/Devices/CreativeAnimation }
# using { /Fortnite.com/Devices/Animation }
using { /Verse.org/Random }
using { /Fortnite.com/Characters }
using { /Verse.org/Colors/NamedColors }
using { /UnrealEngine.com/Temporary/UI }
using { /Fortnite.com/UI }
using { /Verse.org/Assets }

pose :=enum{Idle, Flex_1, Flex_2, Hands_on_Hips_1, Hands_on_Hips_2, Hands_on_Hips_3, Jazz_Hands_1, Jazz_Hands_2, Anguish, Leap, Thumbs_Up, Stand_Tall, Finger_Guns, Run, Run_Determined, Kick_the_Ball, Salute, Love_Hands, Kneeling_Crouch, Survey, Lift_Off, Ready_to_Fight, Strut, Floating, Floating_Look_Back, Wave, Hands_Stop, Peace_Sign}

hologram_pose := class<concrete>():

    @editable PlayerReference : player_reference_device = player_reference_device{}
    @editable Pose : pose = pose.Kneeling_Crouch
    @editable Position : vector3 = vector3{X:=0.0, Y:=0.0, Z:=0.0}
    @editable HatLocation : vector3 = vector3{X:=0.0, Y:=0.0, Z:=0.0}

    @editable var IsActive : logic = false

    Enable(Player:agent):void=
        Print("Attempting to enable pose")
        PlayerReference.Register(Player)
        set IsActive = true
        # if:
        #     PlayerReference.TeleportTo[Position, PlayerReference.GetTransform().Rotation]
        # then:
        #     Print("Teleported to position")
        # else:
        #     Print("Failed to teleport to position")

    Disable(Player:agent):void=
        PlayerReference.Clear()
        set IsActive = false
        # if:
            # PlayerReference.TeleportTo[PlayerReference.GetTransform().Translation + vector3{X:=0.0, Y:=0.0, Z:=-16384.0}, IdentityRotation()]
    
# (X=-3586.497782,Y=7677.627004,Z=1705.111745)

pose_selector := class(creative_device):
    # var CurrentPose : pose = pose.Kneeling_Crouch
    var CurrentPoseIndex : int = 0

    @editable AvailPoses : []hologram_pose = array{}
    @editable MaybeHat : ?creative_prop = false

    @editable ButtonNext : button_device = button_device{}
    @editable ButtonPrev : button_device = button_device{}

    @editable CameraPOV : gameplay_camera_fixed_point_device = gameplay_camera_fixed_point_device{}
    
    var MaybeSpawnSub : ?cancelable = false

    OnBegin<override>()<suspends>:void=
        if(Hat := MaybeHat?):
            Hat.Hide()
        ButtonNext.InteractedWithEvent.Subscribe(OnNextPressed)
        ButtonPrev.InteractedWithEvent.Subscribe(OnPrevPressed)

        if:
            FirstPose := AvailPoses[0]
        then:
            Sub := FirstPose.PlayerReference.AgentUpdatedEvent.Subscribe(OnPlayerRegistered)
            set MaybeSpawnSub = option{Sub}

    OnPlayerRegistered(Agent:agent):void=
        if(Hat := MaybeHat?):
            Hat.Show()
        if(Sub := MaybeSpawnSub?):
            Sub.Cancel()
            set MaybeSpawnSub = false

    OnNextPressed(Agent:agent):void=
        CameraPOV.AddTo(Agent)
        CyclePose(1, Agent)

    OnPrevPressed(Agent:agent):void=
        CameraPOV.AddTo(Agent)
        CyclePose(-1, Agent)

    CyclePose(Direction:int, Agent:agent):void=
        if:
            NextPoseIndex := Mod[(CurrentPoseIndex + Direction + AvailPoses.Length), AvailPoses.Length]
        then:
            Print("CurrentPose: {Self.CurrentPoseIndex}")
            Print("NextPose: {NextPoseIndex}")
            # Print("PlayerReferences Length: {AvailPoses.Length}")

            if:
                CurrentPose := AvailPoses[CurrentPoseIndex]
                NextPose := AvailPoses[NextPoseIndex]
                HatLocation := NextPose.HatLocation
            then:
                CurrentPose.Disable(Agent)
                NextPose.Enable(Agent)
                if:
                    Hat := MaybeHat?
                    Hat.TeleportTo[HatLocation, Hat.GetTransform().Rotation]
                else:
                    Print("No hat to teleport")

                set CurrentPoseIndex = NextPoseIndex

        else:
            Print("Failed to cycle pose")

        spawn {DelayRemoveCamera(Agent)}
            
    DelayRemoveCamera(Agent:agent)<suspends>:void=
        Sleep(1.0)
        CameraPOV.RemoveFrom(Agent)
            