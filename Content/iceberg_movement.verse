using { /Verse.org/Simulation }
using { /Fortnite.com/Devices }
using { /UnrealEngine.com/Temporary/SpatialMath }



iceberg_movement := class<concrete>():
    @editable Iceberg : creative_prop = creative_prop{}

    @editable TrackerSnowmen : tracker_device = tracker_device{}
    @editable Collectible : collectible_object_device = collectible_object_device{}
    @editable StickerPropAsset : creative_prop_asset = DefaultCreativePropAsset

    @editable SnowmanProps : []creative_prop = array{}
    # @editable SnowmanAssetOne : creative_prop_asset = DefaultCreativePropAsset
    # @editable SnowmanAssetTwo : creative_prop_asset = DefaultCreativePropAsset

    StartingLocationX : float = 70190.0
    FinalLocationX : float = -71065.0
    

    StartingDate : time_and_date = time_and_date:
        Day := 1
        Month := 11
        Year := 2024
        Hour := 0
        Minute := 0
        Second := 0

    FinalLocationArrivalDate : time_and_date = time_and_date:
        Day := 25
        Month := 12
        Year := 2024
        Hour := 0
        Minute := 0
        Second := 0

    Dec24th : time_and_date = time_and_date:
        Day := 24
        Month := 12
        Year := 2024
        Hour := 0
        Minute := 0
        Second := 0


    Init()<suspends>:void=
        # Calculate how many steps should have been completed based on current date
        CurrentDay := GetCurrentDateTime()
        # CurrentDay := FinalLocationArrivalDate # For testing purposes only
        # CurrentDay := Dec24th # For testing purposes only
        DaysElapsed := DaysBetweenDates(StartingDate, CurrentDay)

        StartingLocationY := Iceberg.GetTransform().Translation.Y
        FinalLocationY := Iceberg.GetTransform().Translation.Y
        FinalLocationZ := Iceberg.GetTransform().Translation.Z

        # Calculate the total duration in days
        TotalDurationDays := DaysBetweenDates(StartingDate, FinalLocationArrivalDate)

        # Define the total number of steps
        TotalSteps := 100

        if(DaysElapsed <= 0):
            # Before movement starts
            SetIcebergPosition(StartingLocationX, StartingLocationY, FinalLocationZ)
            return
        

        if(DaysElapsed >= TotalDurationDays):
            # After arrival date
            SetIcebergPosition(FinalLocationX, FinalLocationY, FinalLocationZ)

            Collectible.CollectedEvent.Subscribe(HandleCollectibleCollected)

            # SpawnSnowman()
            ShowSnowmanProps()
            return

        HideSnowmanProps()
        
        TargetPosition := vector3{
            X := StartingLocationX + (FinalLocationX - StartingLocationX) * ((DaysElapsed * 1.0) / (TotalDurationDays * 1.0)),
            Y := StartingLocationY,
            Z := FinalLocationZ
        }

        # Teleport the iceberg to the target position
        SetIcebergPosition(TargetPosition.X, TargetPosition.Y, TargetPosition.Z)

        # Optional: Print debug information
        Print("Days Elapsed: {DaysElapsed}")
        Print("Iceberg Teleported to: {TargetPosition}")


    # Function to set the iceberg's position
    SetIcebergPosition(LocationX: float, LocationY: float, LocationZ: float):void=
        PropTransform := Iceberg.GetTransform()
        NewTransform := transform:
            Translation := vector3{
                X := LocationX
                Y := LocationY
                Z := LocationZ
            }
            Rotation := PropTransform.Rotation
            Scale := PropTransform.Scale
        option{ Iceberg.TeleportTo[NewTransform] }
        Print("Iceberg Teleported to X: {LocationX}, Y: {LocationY}, Z: {LocationZ}")




    HandleCollectibleCollected(Agent:agent):void=
        Print("Collectible Collected")
        SetCompletedWinter2024Quest(Agent)
        
        if:
            TheGlobals := Globals[GetSession()]
            Player := player[Agent]
            CPlayer := TheGlobals.CustomPlayerMap[Player]
        then:
            if:
                AvailStickerSlot := CPlayer.GameBoard.GetAvailableStickerSlot[]
            then:
                NewStickerResult := SpawnProp(StickerPropAsset, vector3{}, rotation{})
                Winter2024StickerMaterial := Materials.Stickers.MI_Sticker_Base_Inst_material{}
                set Winter2024StickerMaterial.Sticker_Texture = Materials.Stickers.winter_2024
                set Winter2024StickerMaterial.UseSnowBase = 1.0
                if:
                    TheSticker := NewStickerResult(0)?
                    StickerID := TheGlobals.GetStickerID["Winter2024"]
                then:
                    TheSticker.SetMaterial(Winter2024StickerMaterial)
                    # CPlayer.GameBoard.AddStickerCam.AddTo(Agent)
                    spawn . CPlayer.GameBoard.AddSticker(Agent, CPlayer, AvailStickerSlot, StickerID, TheSticker)                        
            else:
                Print("No available sticker slot")


    HideSnowmanProps():void=
        for(Index -> Prop : SnowmanProps):
            if(Prop.IsValid[]):
                Prop.Hide()

    ShowSnowmanProps():void=
        for(Index -> Prop : SnowmanProps):
            if(Prop.IsValid[]):
                Prop.Show()

    # SpawnSnowman():void=
        # for(Index -> Slot : SnowmanSlots):
        #     if:
        #         #even
        #         Mod[Index, 2] = 0
        #     then:
        #         SpawnedProp := SpawnProp(SnowmanAssetOne, Slot.GetTransform())
        #         if(TheProp := SpawnedProp(0)?):
        #             spawn . AwaitDestroyed(TheProp)
        #     else:
        #         SpawnedProp := SpawnProp(SnowmanAssetTwo, Slot.GetTransform())
        #         if(TheProp := SpawnedProp(0)?):
        #             spawn . AwaitDestroyed(TheProp)

    AwaitDestroyed(Prop:creative_prop)<suspends>:void=
        loop:
            if(not Prop.IsValid[]):
                break

            Sleep(0.25)

        CurrentVal := TrackerSnowmen.GetValue()
        TrackerSnowmen.SetValue(CurrentVal + 1)