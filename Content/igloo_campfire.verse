using { /Verse.org/Simulation }
using { /Fortnite.com/Devices }
using { /UnrealEngine.com/Temporary/SpatialMath }

#(X=-13005.000000,Y=-1736.000000,Z=2149.000000)



igloo_campfire := class<concrete>():

    @editable PlayerCounter : player_counter_device = player_counter_device{}
    @editable AccoladeStandard : accolades_device = accolades_device{}
    @editable AccoladeSpecial : accolades_device = accolades_device{}
    @editable PropStandardXP : creative_prop_asset = DefaultCreativePropAsset
    @editable PropSpecialXP : creative_prop_asset = DefaultCreativePropAsset
    @editable TimeBewteenXP : float = 10.0
    @editable LoopsBeforeSpecialXP : int = 5

    var CurrentXPType : int = -1

    Init()<suspends>:void=

        PlayerCounter.CountedEvent.Subscribe(OnPlayerCounted)

        var LoopCount : int = 0

        loop:
            Sleep(TimeBewteenXP)
            set LoopCount += 1
            if:
                Mod[LoopCount, LoopsBeforeSpecialXP] = 0
            then:
                set CurrentXPType = 1
                SpawnedProp := SpawnProp(PropSpecialXP, vector3{X:=-13005.000000,Y:=-1736.000000,Z:=2149.000000}, IdentityRotation())
                if(TheProp := SpawnedProp(0)?):
                    spawn . TheProp.DelayedDispose(2.0)
            else:
                set CurrentXPType = 0
                SpawnedProp := SpawnProp(PropStandardXP, vector3{X:=-13005.000000,Y:=-1736.000000,Z:=2149.000000}, IdentityRotation())
                if(TheProp := SpawnedProp(0)?):
                    spawn . TheProp.DelayedDispose(2.0)

            PlayerCounter.TransmitForAllCounted()
            

    OnPlayerCounted(Agent:agent):void=
        case(CurrentXPType):
            0 => AccoladeStandard.Award(Agent)
            1 => AccoladeSpecial.Award(Agent)
            _ => block:
        

