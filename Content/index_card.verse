using { /Verse.org/Simulation }
using { /Fortnite.com/Devices }
using { /Fortnite.com/Devices/CreativeAnimation }
using { /UnrealEngine.com/Temporary/SpatialMath }
using { /Verse.org/Assets }

index_card := class<concrete>():

    @editable Card : creative_prop = creative_prop{}

    var IsHidden : logic = true
    var <private> IsHovered : logic = false
    var <private> IsAnimating : logic = false
    
    var SlideDistance : float = 576.0

    Hover():void=
        if (not IsHovered?):
            # Print("Hovering Index")
            Card.SetMaterial(Materials.BoardDesigner.CardBacks.MI_Cardback_Highlight, ?Index:=0)
            set IsHovered = true            
        else:
            # Print("Already Hovering")

    Unhover(SavedFrameMaterial:material):void=
        # Print("Unhovering Index")
        Card.SetMaterial(SavedFrameMaterial, ?Index:=0)
        set IsHovered = false

    SlideOut(BoardLocalForward:float)<suspends>:void=
        if:
            IsHidden?
            not IsAnimating?
        then:
            # Print("Sliding Out")
            set IsAnimating = true
            PerformSlideAnimation(true)
            set IsHidden = false
            set IsAnimating = false
        else:
            # Print("Already Out or Animating")

    SlideIn(BoardLocalForward:float)<suspends>:void=
        if:
            not IsHidden?
            not IsAnimating?
        then:
            # Print("Sliding In")
            set IsAnimating = true
            PerformSlideAnimation(false)
            set IsHidden = true
            set IsAnimating = false
        else:
            # Print("Already In or Animating")

    ToggleSlide(BoardLocalForward:float)<suspends>:void=
        if:
            IsHidden?
        then:
            SlideOut(BoardLocalForward)
        else:
            SlideIn(BoardLocalForward)

    PerformSlideAnimation(SlidesOut:logic)<suspends>:void=
        if:
            AnimController := Self.Card.GetAnimationController[]
        then:
            # Get the card prop's local right direction to determine the slide direction.
            LocalRightDirection := Self.Card.GetTransform().Rotation.GetLocalForward()
    
            # Determine the direction multiplier based on the slide out or in action.
            DirectionMultiplier := if(SlidesOut?) then -1.0 else 1.0
    
            # Calculate the delta location for the slide animation using the local right direction and direction multiplier.
            # This ensures the slide direction is aligned with the card prop's orientation.
            DeltaLocation := vector3{X:=LocalRightDirection.X * SlideDistance * DirectionMultiplier, Y:=LocalRightDirection.Y * SlideDistance * DirectionMultiplier, Z:=0.0}
    
            # Prepare and play the animation with the calculated delta location.
            AnimKeyFrames := array:
                keyframe_delta:
                    DeltaLocation := DeltaLocation
                    DeltaRotation := IdentityRotation()
                    Time := 0.5
                    Interpolation := InterpolationTypes.EaseInOut
            AnimController.SetAnimation(AnimKeyFrames, ?Mode:=animation_mode.OneShot)
            AnimController.Play()
            AnimController.MovementCompleteEvent.Await()
    
            # Toggle the hidden state based on the animation's completion.
            if(IsHidden?) then set IsHidden = false else set IsHidden = true
            set IsAnimating = false
            
            

            
            



    # MakeSlideKeyFrames(DirectionMultiplier:float):[]keyframe_delta=
    #     var NewDistance : float = 0.0
    #     if:
    #         IsHidden?
    #     then:
    #         set NewDistance = SlideDistance * DirectionMultiplier
    #     else:
    #         set NewDistance = SlideDistance * -DirectionMultiplier

    #     array:
    #         keyframe_delta:
    #             DeltaLocation := vector3{X:=NewDistance, Y:=0.0, Z:=0.0}
    #             DeltaRotation := IdentityRotation()
    #             Time := 0.5
    #             Interpolation := InterpolationTypes.EaseInOut