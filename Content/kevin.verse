
using { /Fortnite.com/Devices }
using { /Verse.org/Simulation }
using { /UnrealEngine.com/Temporary/Diagnostics }
using { /UnrealEngine.com/Temporary/SpatialMath }
using { /Verse.org/Random }
using { /UnrealEngine.com/Temporary/UI }
using { /Fortnite.com/FortPlayerUtilities }
using { /Fortnite.com/Characters }
using { /Verse.org/Assets }
using { /Fortnite.com/Devices/CreativeAnimation }

kevin := class<concrete>():
    @editable TheCube : creative_prop = creative_prop{}

    CubeSize : float = 3000.0 #full width of the cube
    # var RotationAxis : vector3 = vector3{X:=0.0,Y:=-1.0,Z:=0.0} #default forwards
    # var RotationAngle : float = 90.0
    var RotationDuration : float = 1.0
    var IsWalking : logic = false
    var CurrentMovePoint : int = 0
    @editable var MovePoints : []creative_prop =array{}

    Init()<suspends>:void=

        Sleep(10.0)

        StartWalking()

    StartWalking()<suspends>:void=
        if(not IsWalking?):
            Print("Starting Walk")
            set IsWalking = true
            WalkCycle()

    WalkCycle()<suspends>:void=
        loop:
            if(not IsWalking?):
                break

            PerformWalkStep()
            Sleep(RotationDuration)


    PerformWalkStep()<suspends>:void=
        if:
            TheCube.IsValid[]
            MovePoints.Length > 0
            CurrentPoint := MovePoints[CurrentMovePoint]
            NextIndex := Mod[CurrentMovePoint + 1, MovePoints.Length]
            NextPoint := MovePoints[NextIndex]
            CurrentPosition := TheCube.GetTransform().Translation
            NextTransform := NextPoint.GetTransform()
            MovementDirection := NextTransform.Translation - CurrentPosition
            NormalizedDirection := MovementDirection.MakeUnitVector[]
        then:
            if(NextIndex = 0):
                return

            Print("Attempting to walk")

            # Determine the direction to roll the cube
            # Directions := 
                # 0, if(NormalizedDirection.X > 0.0). 1 else. -1

            Directions := DetermineDirections(NormalizedDirection)


            Print("Current Point: {CurrentMovePoint} - {CurrentPoint.GetTransform().Translation}")
            Print("Next Point: {NextIndex} - {NextPoint.GetTransform().Translation}")
            Print("Directions: {Directions(0)}, {Directions(1)}")
            

            AnimateCubeRoll(Directions)

            if(NewMovePoint := Mod[CurrentMovePoint + 1, MovePoints.Length]):
                set CurrentMovePoint = NewMovePoint
            else:
                Print("No steps to walk or error")
                set IsWalking = false
                return

        else:
            Print("No steps to walk or error")
            set IsWalking = false
            return



    AnimateCubeRoll(Directions:tuple(int, int))<suspends> : void =

        RotationAxis := vector3:
            X := if (Directions(0) = 0). Directions(1) * 1.0 else. 0.0
            Y := if (Directions(0) = 1). Directions(1) * 1.0 else. 0.0
            Z := 0.0
        

        AnchorPointOffset := vector3:
            X := if (Directions(0) = 1). Directions(1) * (CubeSize / 2.0) else. 0.0
            Y := if (Directions(0) = 0). -Directions(1) * (CubeSize / 2.0) else. 0.0
            Z := -(CubeSize / 2.0)
        

        StartTransform := TheCube.GetTransform()

        AnimDurationTicks := RotationDuration * 30

        for (X := 0..Int[AnimDurationTicks]):
            InterpolationProgress := (X * 1.0)  / (AnimDurationTicks * 1.0)

            RotationAngle := Lerp(0.0, PiFloat / 2.0, EaseInCubic(InterpolationProgress))

            TargetPos := StartTransform.Translation.RotateAroundAnchor(
                StartTransform.Translation + AnchorPointOffset,
                RotationAxis,
                RotationAngle
            )
            TargetRotation := StartTransform.Rotation.RotateBy(MakeRotation(RotationAxis, RotationAngle))
            
            option{TheCube.TeleportTo[TargetPos, TargetRotation]}

            Sleep(0.0)


    EaseInCubic(t:float) : float =
        t * t * t


    DetermineDirections(MovementDirection:vector3) : tuple(int, int) =
        var Directions : tuple(int, int) = (0, 0)

        var Direction : int = 0
        var DominantAxis : string = ""

        # Determine dominant movement axis and direction
        if(Abs(MovementDirection.X) > Abs(MovementDirection.Y)):
            set DominantAxis = "X"
            set Direction = if(MovementDirection.X > 0.0). 1 else. -1
        else:
            set DominantAxis = "Y"
            set Direction = if(MovementDirection.Y > 0.0). 1 else. -1

        Print("Dominant Axis: {DominantAxis}, Direction: {Direction}")

        # Set RotationAxis and AnchorPointOffset based on movement
        if(DominantAxis = "X"):
            RotationAxis := vector3{X:=0.0, Y:=(Direction * 1.0), Z:=0.0}
            AnchorPointOffset := vector3{
                X := 0.0,
                Y := Direction * (CubeSize / 2.0),
                Z := -(CubeSize / 2.0)
            }
        else if(DominantAxis = "Y"):
            RotationAxis := vector3{X:=(Direction * 1.0), Y:=0.0, Z:=0.0}
            AnchorPointOffset := vector3{
                X := Direction * (CubeSize / 2.0),
                Y := 0.0,
                Z := -(CubeSize / 2.0)
            }

        set Directions = (if(DominantAxis = "X"). 0 else. 1, Direction)

        return Directions
