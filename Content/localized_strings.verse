MessageBlank_EN<localizes>:message = ""
MessageSelectSUP<localizes>:message = "ˢᵉˡᵉᶜᵗ"
MessageThinWideArrowDown<localizes>:message = "﹀"
MessageThinWideArrowUp<localizes>:message = "︿"

# Board Designer Strings
MessageBoardDesignerTitle_EN<localizes>:message = "BOARD DESIGNER"
MessageBoard_EN<localizes>:message = "BOARD"
MessageCards_EN<localizes>:message = "CARDS"
MessageStand_EN<localizes>:message = "STAND"
MessagePlatform_EN<localizes>:message = "PLATFORM"
MessageSearch_EN<localizes>:message = "SEARCH"
MessageSelect_EN<localizes>:message = "SELECT"
MessageExit_EN<localizes>:message = "EXIT"
# MessageArrowDown_EN<localizes>:message = "↓"
# MessageArrowUp_EN<localizes>:message = "↑"
# MessageArrowLeft_EN<localizes>:message = "←"
# MessageArrowRight_EN<localizes>:message = "→"
MessageArrowLeft_EN<localizes>:message = "◀"
MessageArrowRight_EN<localizes>:message = "▶"
MessageArrowDown_EN<localizes>:message = "▼"
MessageArrowUp_EN<localizes>:message = "▲"
# MessageArrowDownRight_EN<localizes>:message = "↘"
# MessageArrowDownLeft_EN<localizes>:message = "↙"
# MessageArrowUpRight_EN<localizes>:message = "↗"
# MessageArrowUpLeft_EN<localizes>:message = "↖"
MessageAccept_EN<localizes>:message = "ACCEPT"
MessageCancel_EN<localizes>:message = "CANCEL"
MessageLocked_EN<localizes>:message = "LOCKED"
MessageBuy_EN<localizes>:message = "BUY"
MessageFree_EN<localizes>:message = "FREE!"
MessageClose_EN<localizes>:message = "CLOSE"
MessageActive_EN<localizes>:message = "ACTIVE"
# MessageElpisis_EN<localizes>:message = "…"
# MessageElpisis_EN<localizes>:message = "•••"
# MessageElpisis_EN<localizes>:message = "●●●"
# MessageElpisis_EN<localizes>:message = "⋯"
# MessageElpisis_EN<localizes>:message = "…"
# MessageElpisis_EN<localizes>:message = "‧‧‧" # good
MessageElpisis_EN<localizes>:message = "" # good
MessageTrackQuest_EN<localizes>:message = "TRACK QUEST"
MessageTrack_EN<localizes>:message = "TRACK"
MessageMaterialSlot1_EN<localizes>:message = "MATERIAL SLOT 1"
MessageMaterialSlot2_EN<localizes>:message = "MATERIAL SLOT 2"
MessageMaterialSlot3_EN<localizes>:message = "MATERIAL SLOT 3"
MessageMaterialSlot4_EN<localizes>:message = "MATERIAL SLOT 4"
MessageMaterialSlot5_EN<localizes>:message = "MATERIAL SLOT 5"
MessageMaterialTypePlastic_EN<localizes>:message = "PLASTIC"
MessageMaterialTypeMetallic_EN<localizes>:message = "METALLIC"
MessageMaterialTypePutty_EN<localizes>:message = "PUTTY"
MessageMaterialTypeGalaxy_EN<localizes>:message = "GALAXY"
MessageMaterialTypeNeon_EN<localizes>:message = "NEON"
MessageMaterialTypeWood_EN<localizes>:message = "WOOD"
MessageClearFilters_EN<localizes>:message = "✖"
MessageDollarSign_EN<localizes>:message = "$"
MessageAutoApplyPaint_EN<localizes>:message = "Auto Equip"

MessageCheckBoxChecked_EN<localizes>:message = "✅"
MessageCheckBoxUnchecked_EN<localizes>:message = "✖"

# Selection Screen UI Strings
MessageSwitchLayoutVertical_EN<localizes>:message = "📱"
MessageSwitchLayoutHorizontal_EN<localizes>:message = "🖥️"
MessageChooseCharacterIcon_EN<localizes>:message = "👤"
MessageRandomCharacterIcon_EN<localizes>:message = "✨"
# MessageWaitingForVote_EN<localizes>:message = "WAITING FOR OTHER PLAYER'S VOTE..."
# MessageVotesMatch_EN<localizes>:message = "VOTES MATCH! 😄"
# MessageVotesDontMatch_EN<localizes>:message = "VOTES DON'T MATCH! 😢 - RANDOMIZING... 🎲"
MessageManualCharacterChoice_EN<localizes>:message = "MANUAL CHARACTER CHOICE"
MessageRandomCharacterChoice_EN<localizes>:message = "RANDOM CHARACTER"


# Admin Panel UI Strings
MessageIncorrectRole_EN<localizes>:message = "INC ROLE"
MessageIncorrectCode_EN<localizes>:message = "INC"

# Game Strings
MessageRematchBonusGold_EN<localizes>:message = "Rematch Bonus: 10 GOLD!"
# MessageTGAIslandCode_EN<localizes>:message = "0853-1358-8532"

# Question Bank UI Strings
MessageYesButton_EN<localizes>:message = "YES ✔"
MessageNoButton_EN<localizes>:message = "NO ❌"
MessageMaybeButton_EN<localizes>:message = "🤷‍♂️"
MessagePrevButton_EN<localizes>:message = "◀ PREV"
MessageNextButton_EN<localizes>:message = "NEXT ▶"
MessageCloseButton_EN<localizes>:message = "❌"


TestGoldMessage_EN<localizes>:message = "5,000"