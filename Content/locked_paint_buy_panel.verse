using { /UnrealEngine.com/Temporary/UI }
using { /Verse.org/Assets }
using { /Fortnite.com/UI }
using { /Verse.org/Colors/NamedColors }
using { /UnrealEngine.com/Temporary/SpatialMath }
using { /Verse.org/Simulation }
using { /Verse.org/Colors }

locked_paint_buy_panel := class:

    var Modal : canvas_slot = canvas_slot{Widget:=overlay{}}
    
    var PaintBGColor : color = MakeColorFromHSV(0.0, 0.0, 1.0)
	var PaintBGTexture : texture = _UI.UI_Textures.board_locker_paint_option_bg
    var PaintStyleTexture : texture = _UI.UI_Textures.board_locker_paint_style_plastic

    var PropGameBoard : board_locker_prop = board_locker_prop{}

    NameTextBlock : text_block = text_block:
        DefaultText := MessageMaterialSlot1_EN
        DefaultTextColor := White
        DefaultJustification := text_justification.Center

    CostTextureBlock : texture_block = texture_block:
        DefaultImage := _UI.UI_Textures.ui_gold_coin
        DefaultDesiredSize := vector2{X:=96.0, Y:=96.0}

    ButtonBuy : button_loud = button_loud:
        DefaultText := MessageBuy_EN

    ButtonClose : button_quiet = button_quiet:
        DefaultText := MessageClose_EN

    PaintBGTextureBlock : texture_block = texture_block:
        DefaultImage := _UI.UI_Textures.board_locker_paint_option_bg
        DefaultDesiredSize := vector2{X:=200.0, Y:=90.0}

    PaintStyleTextureBlock : texture_block = texture_block:
        DefaultImage := _UI.UI_Textures.transparent_square
        DefaultDesiredSize := vector2{X:=200.0, Y:=90.0}


    # check box for auto applying paint after purchase
    ButtonAutoApplyToggle : button_quiet = button_quiet:
        DefaultText := MessageCheckBoxChecked_EN

    var AutoApply : logic = true



    Create(Agent:agent, PaintStyle:paint_ui_option, GameBoard:board_locker_prop):void=
        set PropGameBoard = GameBoard
        set PaintBGColor = PaintStyle.PaintBGColor
        set PaintBGTexture = PaintStyle.PaintBGTexture
        set PaintStyleTexture = PaintStyle.PaintStyleTexture

        NameTextBlock.SetText(StringToMessage(PaintStyle.PaintName))

        PaintBGTextureBlock.SetImage(PaintBGTexture)
        PaintStyleTextureBlock.SetImage(PaintStyleTexture)

        PaintBGTextureBlock.SetTint(PaintBGColor)
        PaintStyleTextureBlock.SetTint(PaintBGColor + MakeColorFromHSV(0.0, 0.0, 0.5))

        if(SecondaryColor := PaintStyle.SecondaryBGColor?):
            PaintBGTextureBlock.SetTint(PaintBGColor)
            PaintStyleTextureBlock.SetTint(SecondaryColor)

        ButtonBuy.SetText(MakeBuyPaintButtonMessage(PaintStyle.Cost))


        ButtonAutoApplyToggle.OnClick().Subscribe(OnAutoApplyToggleClicked)

        case(PaintStyle.PaintStyleIndex):
            0 => block:
                PaintMatData := plastic_material_data:
                    ColorTint := PaintStyle.PaintBGColor
                    EmissiveFlickerTint := PaintStyle.PaintBGColor
                NewMI := CreatePlasticMaterialInstance(PaintMatData)
                PropGameBoard.Prop.SetMaterial(NewMI, ?Index:=1)
                PropGameBoard.Prop.SetMaterial(NewMI, ?Index:=2)
                PropGameBoard.Prop.SetMaterial(NewMI, ?Index:=3)
            1 => block:
                PaintMatData := plastic_material_data:
                    ColorTint := PaintStyle.PaintBGColor
                    MetallicAmount := 1.0
                    EmissiveFlickerTint := PaintStyle.PaintBGColor
                NewMI := CreateMetallicMaterialInstance(PaintMatData)
                PropGameBoard.Prop.SetMaterial(NewMI, ?Index:=1)
                PropGameBoard.Prop.SetMaterial(NewMI, ?Index:=2)
                PropGameBoard.Prop.SetMaterial(NewMI, ?Index:=3)
            2 => block:
                if(SecondaryColor := PaintStyle.SecondaryBGColor?):
                    PaintMatData := putty_material_data:
                        ColorClay1 := PaintStyle.PaintBGColor
                        ColorClay2 := PaintStyle.PaintBGColor.HueShift(30.0)
                        EmissiveFlickerTint := PaintStyle.PaintBGColor
                    NewMI := CreatePuttyMaterialInstance(PaintMatData)
                    PropGameBoard.Prop.SetMaterial(NewMI, ?Index:=1)
                    PropGameBoard.Prop.SetMaterial(NewMI, ?Index:=2)
                    PropGameBoard.Prop.SetMaterial(NewMI, ?Index:=3)
            _ => block:
                PaintMatData := plastic_material_data:
                    ColorTint := PaintStyle.PaintBGColor
                    EmissiveFlickerTint := PaintStyle.PaintBGColor
                NewMI := CreatePlasticMaterialInstance(PaintMatData)
                PropGameBoard.Prop.SetMaterial(NewMI, ?Index:=1)
                PropGameBoard.Prop.SetMaterial(NewMI, ?Index:=2)
                PropGameBoard.Prop.SetMaterial(NewMI, ?Index:=3)

        set Modal = canvas_slot:
            Anchors := anchors{Minimum:=vector2{X:=0.0, Y:=0.0}, Maximum:=vector2{X:=1.0, Y:=1.0}}
            Offsets := margin{Top:=0.0, Bottom:=0.0, Left:=0.0, Right:=0.0}
            SizeToContent := true
            Alignment := vector2{X:=0.5, Y:=0.5}
            ZOrder := 9000
            Widget := overlay:
                Slots := array:

                    # Invis to block click through
                    overlay_slot:
                        HorizontalAlignment := horizontal_alignment.Fill
                        VerticalAlignment := vertical_alignment.Fill
                        Padding := margin{Top:=0.0, Bottom:=0.0, Left:=0.0, Right:=0.0}
                        Widget := color_block:
                            DefaultColor := Red
                            DefaultOpacity := 0.0

                    # Modal Container Alignment
                    overlay_slot:
                        HorizontalAlignment := horizontal_alignment.Right
                        VerticalAlignment := vertical_alignment.Bottom
                        Padding := margin{Top:=128.0, Bottom:=0.0, Left:=0.0, Right:=28.0}
                        Widget := overlay:
                            Slots := array:
                                # Modal Drop Shadow
                                overlay_slot:
                                    HorizontalAlignment := horizontal_alignment.Fill
                                    VerticalAlignment := vertical_alignment.Fill
                                    Padding := margin{Top:=0.0, Bottom:=0.0, Left:=0.0, Right:=0.0}
                                    Widget := MakeDropShadowBackground(MakeColorFromHSV(240.0,1.0,0.35), ?TextureSize:=512.0)

                                overlay_slot:
                                    HorizontalAlignment := horizontal_alignment.Center
                                    VerticalAlignment := vertical_alignment.Center
                                    Padding := margin{Top:=80.0, Bottom:=80.0, Left:=80.0, Right:=80.0}
                                    Widget := overlay:
                                        Slots := array:
                                            # Modal BG
                                            overlay_slot:
                                                HorizontalAlignment := horizontal_alignment.Fill
                                                VerticalAlignment := vertical_alignment.Fill
                                                Padding := margin{Top:=0.0, Bottom:=0.0, Left:=0.0, Right:=0.0}
                                                Widget := MakeRoundedOverlayQuietButton(BGColor:=MakeColorFromHSV(240.0,1.0,0.35), ?TextureSize:=40.0)
                                            
                                            # Modal Content
                                            overlay_slot:
                                                HorizontalAlignment := horizontal_alignment.Fill
                                                VerticalAlignment := vertical_alignment.Fill
                                                Padding := margin{Top:=60.0, Bottom:=60.0, Left:=60.0, Right:=60.0}
                                                Widget := stack_box:
                                                    Orientation := orientation.Vertical
                                                    Slots := array:
                                                        stack_box_slot:
                                                            HorizontalAlignment := horizontal_alignment.Fill
                                                            VerticalAlignment := vertical_alignment.Top
                                                            Padding := margin{Top:=0.0, Bottom:=0.0, Left:=0.0, Right:=0.0}
                                                            Widget := overlay:
                                                                Slots := array:
                                                                    # overlay_slot:
                                                                    #     HorizontalAlignment := horizontal_alignment.Fill
                                                                    #     VerticalAlignment := vertical_alignment.Fill
                                                                    #     Padding := margin{Top:=0.0, Bottom:=0.0, Left:=0.0, Right:=0.0}
                                                                    #     Widget := MakeRoundedOverlayInner(BGColor:=MakeColorFromHSV(240.0,1.0,0.25), ?TextureSize:=40.0)
                                                                    overlay_slot:
                                                                        HorizontalAlignment := horizontal_alignment.Center
                                                                        VerticalAlignment := vertical_alignment.Center
                                                                        Padding := margin{Top:=20.0, Bottom:=20.0, Left:=20.0, Right:=20.0}
                                                                        Widget := stack_box:
                                                                            Orientation := orientation.Vertical
                                                                            Slots := array:
                                                                                stack_box_slot:
                                                                                    HorizontalAlignment := horizontal_alignment.Center
                                                                                    VerticalAlignment := vertical_alignment.Center
                                                                                    Padding := margin{Top:=10.0, Bottom:=10.0, Left:=10.0, Right:=10.0}
                                                                                    Widget := NameTextBlock
                                                                                stack_box_slot:
                                                                                    HorizontalAlignment := horizontal_alignment.Center
                                                                                    VerticalAlignment := vertical_alignment.Center
                                                                                    Padding := margin{Top:=10.0, Bottom:=10.0, Left:=10.0, Right:=10.0}
                                                                                    Widget := overlay:
                                                                                        Slots := array:
                                                                                            overlay_slot:
                                                                                                HorizontalAlignment := horizontal_alignment.Center
                                                                                                VerticalAlignment := vertical_alignment.Center
                                                                                                Padding := margin{Top:=0.0, Bottom:=0.0, Left:=0.0, Right:=0.0}
                                                                                                Widget := PaintBGTextureBlock
                                                                                            overlay_slot:
                                                                                                HorizontalAlignment := horizontal_alignment.Center
                                                                                                VerticalAlignment := vertical_alignment.Center
                                                                                                Padding := margin{Top:=0.0, Bottom:=0.0, Left:=0.0, Right:=0.0}
                                                                                                Widget := PaintStyleTextureBlock
                                                        stack_box_slot:
                                                            HorizontalAlignment := horizontal_alignment.Center
                                                            VerticalAlignment := vertical_alignment.Bottom
                                                            Padding := margin{Top:=20.0, Bottom:=0.0, Left:=0.0, Right:=0.0}
                                                            Widget := overlay:
                                                                Slots := array:
                                                                    overlay_slot:
                                                                        HorizontalAlignment := horizontal_alignment.Fill
                                                                        VerticalAlignment := vertical_alignment.Fill
                                                                        Padding := margin{Top:=0.0, Bottom:=0.0, Left:=0.0, Right:=0.0}
                                                                        Widget := MakeRoundedOverlayInner(BGColor:=MakeColorFromHSV(240.0,1.0,0.25), ?TextureSize:=40.0)
                                                                    overlay_slot:
                                                                        HorizontalAlignment := horizontal_alignment.Fill
                                                                        VerticalAlignment := vertical_alignment.Fill
                                                                        Padding := margin{Top:=0.0, Bottom:=0.0, Left:=20.0, Right:=20.0}
                                                                        Widget := stack_box:
                                                                            Orientation := orientation.Horizontal
                                                                            Slots := array:
                                                                                stack_box_slot:
                                                                                    HorizontalAlignment := horizontal_alignment.Left
                                                                                    VerticalAlignment := vertical_alignment.Center
                                                                                    Padding := margin{Top:=10.0, Bottom:=10.0, Left:=10.0, Right:=10.0}
                                                                                    Widget := CostTextureBlock
                                                                                stack_box_slot:
                                                                                    HorizontalAlignment := horizontal_alignment.Right
                                                                                    VerticalAlignment := vertical_alignment.Center
                                                                                    Padding := margin{Top:=10.0, Bottom:=10.0, Left:=10.0, Right:=10.0}
                                                                                    Widget := ButtonBuy                                            

                                                        # Auto Apply Toggle
                                                        stack_box_slot:
                                                            HorizontalAlignment := horizontal_alignment.Center
                                                            VerticalAlignment := vertical_alignment.Center
                                                            Padding := margin{Top:=20.0, Bottom:=10.0, Left:=10.0, Right:=10.0}
                                                            Widget := stack_box:
                                                                Orientation := orientation.Horizontal
                                                                Slots := array:
                                                                    stack_box_slot:
                                                                        HorizontalAlignment := horizontal_alignment.Left
                                                                        VerticalAlignment := vertical_alignment.Center
                                                                        Padding := margin{Top:=0.0, Bottom:=0.0, Left:=0.0, Right:=20.0}
                                                                        Widget := text_block:
                                                                            DefaultText := MessageAutoApplyPaint_EN
                                                                            DefaultTextColor := White
                                                                            DefaultJustification := text_justification.Left
                                                                    stack_box_slot:
                                                                        HorizontalAlignment := horizontal_alignment.Left
                                                                        VerticalAlignment := vertical_alignment.Center
                                                                        Padding := margin{Top:=0.0, Bottom:=0.0, Left:=0.0, Right:=0.0}
                                                                        Widget := ButtonAutoApplyToggle

                                            overlay_slot:
                                                HorizontalAlignment := horizontal_alignment.Right
                                                VerticalAlignment := vertical_alignment.Top
                                                Padding := margin{Top:=20.0, Bottom:=0.0, Left:=0.0, Right:=20.0}
                                                Widget := ButtonClose

    OnAutoApplyToggleClicked(Message:widget_message):void=
        if(AutoApply?):
            ButtonAutoApplyToggle.SetText(MessageCheckBoxUnchecked_EN)
            set AutoApply = false
        else:
            ButtonAutoApplyToggle.SetText(MessageCheckBoxChecked_EN)
            set AutoApply = true