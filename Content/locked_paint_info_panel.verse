using { /UnrealEngine.com/Temporary/UI }
using { /Verse.org/Assets }
using { /Fortnite.com/UI }
using { /Verse.org/Colors/NamedColors }
using { /UnrealEngine.com/Temporary/SpatialMath }
using { /Verse.org/Simulation }
using { /Verse.org/Colors }

locked_paint_info_panel := class:

    var Modal : canvas_slot = canvas_slot{Widget:=overlay{}}

    var PaintBGColor : color = MakeColorFromHSV(0.0, 0.0, 1.0)
	var PaintBGTexture : texture = _UI.UI_Textures.board_locker_paint_option_bg
    var PaintStyleTexture : texture = _UI.UI_Textures.board_locker_paint_style_plastic
    
    var QuestProgressBG : overlay = overlay{}

    var PropGameBoard : board_locker_prop = board_locker_prop{}
    
    NameTextBlock : text_block = text_block:
        DefaultText := MessageMaterialSlot1_EN
        DefaultTextColor := White
        DefaultJustification := text_justification.Center

    QuestTitleTextBlock : text_block = text_block:
        DefaultText := MessageBlank_EN
        DefaultTextColor := White
        DefaultJustification := text_justification.Left

    QuestDescTextBlock : text_block = text_block:
        DefaultText := MessageBlank_EN
        DefaultTextColor := White
        DefaultJustification := text_justification.Left


    QuestProgressTextBlock : text_block = text_block:
        DefaultText := MessageBlank_EN
        DefaultTextColor := White
        DefaultJustification := text_justification.Center

    QuestProgressTextureBlock : texture_block = texture_block:
        DefaultImage := _UI.UI_Textures.ui_rounded_question_icon
        DefaultDesiredSize := vector2{X:=48.0, Y:=48.0}

    ButtonTrackQuest : button_quiet = button_quiet{DefaultText := MessageTrack_EN}
    ButtonClose : button_quiet = button_quiet:
        DefaultText := MessageClose_EN

    PaintBGTextureBlock : texture_block = texture_block:
        DefaultImage := _UI.UI_Textures.board_locker_paint_option_bg
        DefaultDesiredSize := vector2{X:=200.0, Y:=90.0}

    PaintStyleTextureBlock : texture_block = texture_block:
        DefaultImage := _UI.UI_Textures.transparent_square
        DefaultDesiredSize := vector2{X:=200.0, Y:=90.0}


    Create(Agent:agent, PaintStyle:paint_ui_option, GameBoard:board_locker_prop):void=
        set PropGameBoard = GameBoard

        set PaintBGColor = PaintStyle.PaintBGColor
        set PaintBGTexture = PaintStyle.PaintBGTexture
        set PaintStyleTexture = PaintStyle.PaintStyleTexture

        NameTextBlock.SetText(StringToMessage(PaintStyle.PaintName))

        PaintBGTextureBlock.SetImage(PaintBGTexture)
        PaintStyleTextureBlock.SetImage(PaintStyleTexture)

        PaintBGTextureBlock.SetTint(PaintBGColor)
        PaintStyleTextureBlock.SetTint(PaintBGColor + MakeColorFromHSV(0.0, 0.0, 0.5))

        if(SecondaryColor := PaintStyle.SecondaryBGColor?):
            PaintBGTextureBlock.SetTint(PaintBGColor)
            PaintStyleTextureBlock.SetTint(SecondaryColor)


        spawn . DelayApplyPaintPreview(PaintStyle)



        # ButtonClose.OnClick().Subscribe(HandleCloseButtonClick)

        set QuestProgressBG = overlay:
            Slots:= array:
                overlay_slot:
                    HorizontalAlignment := horizontal_alignment.Fill
                    VerticalAlignment := vertical_alignment.Fill
                    Padding := margin{Top:=0.0, Bottom:=0.0, Left:=0.0, Right:=0.0}
                    Widget := MakeRoundedOverlayQuietButton(MakeColorFromHSV(240.0,1.0,0.25), ?TextureSize:=24.0)
                overlay_slot:
                    HorizontalAlignment := horizontal_alignment.Fill
                    VerticalAlignment := vertical_alignment.Fill
                    Padding := margin{Top:=20.0, Bottom:=20.0, Left:=20.0, Right:=20.0}
                    Widget := stack_box:
                        Orientation := orientation.Horizontal
                        Slots := array:
                            stack_box_slot:
                                HorizontalAlignment := horizontal_alignment.Fill
                                VerticalAlignment := vertical_alignment.Fill
                                Padding := margin{Top:=0.0, Bottom:=0.0, Left:=0.0, Right:=0.0}
                                Widget := overlay:
                                    Slots := array:
                                        overlay_slot:
                                            HorizontalAlignment := horizontal_alignment.Fill
                                            VerticalAlignment := vertical_alignment.Fill
                                            Padding := margin{Top:=0.0, Bottom:=0.0, Left:=0.0, Right:=20.0}
                                            Widget := stack_box:
                                                Orientation := orientation.Horizontal
                                                Slots := array:
                                                    stack_box_slot:
                                                        HorizontalAlignment := horizontal_alignment.Left
                                                        VerticalAlignment := vertical_alignment.Center
                                                        Padding := margin{Top:=10.0, Bottom:=10.0, Left:=0.0, Right:=10.0}
                                                        Widget := QuestProgressTextureBlock

                                                    stack_box_slot:
                                                        Distribution := option{2.0}
                                                        HorizontalAlignment := horizontal_alignment.Fill
                                                        VerticalAlignment := vertical_alignment.Center
                                                        Padding := margin{Top:=10.0, Bottom:=10.0, Left:=10.0, Right:=10.0}
                                                        Widget := QuestProgressTextBlock
                            stack_box_slot:
                                HorizontalAlignment := horizontal_alignment.Fill
                                VerticalAlignment := vertical_alignment.Fill
                                Padding := margin{Top:=0.0, Bottom:=0.0, Left:=10.0, Right:=0.0}
                                Widget := ButtonTrackQuest
                

        if:
            PaintStyle.QuestID  <> -1
            QuestData := GetQuestFromID[PaintStyle.QuestID]
            QuestData.QuestType = 5 #Board Designer Quests
        then:
            # ButtonTrackQuest.OnClick().Subscribe(HandleTrackQuestButtonClick)
            QuestTitleTextBlock.SetText(StringToMessage(QuestData.Name))
            QuestDescTextBlock.SetText(StringToMessage(QuestData.Description))
            QuestProgressTextBlock.SetText(StringToMessage("{QuestData.Progress}/{QuestData.Target}"))

        else if:
            PaintStyle.QuestID  <> -1
            QuestData := GetQuestFromID[PaintStyle.QuestID]
        then:
            ButtonTrackQuest.SetVisibility(widget_visibility.Collapsed) # Hide the Track Quest button if the quest is not a Board Designer Quest
            QuestProgressTextBlock.SetVisibility(widget_visibility.Collapsed)
            QuestProgressBG.SetVisibility(widget_visibility.Collapsed)
            QuestProgressTextureBlock.SetVisibility(widget_visibility.Collapsed)
            QuestTitleTextBlock.SetText(StringToMessage(QuestData.Name))
            QuestDescTextBlock.SetText(StringToMessage(QuestData.Description))
        else:            
            ButtonTrackQuest.SetVisibility(widget_visibility.Collapsed)
            QuestTitleTextBlock.SetVisibility(widget_visibility.Collapsed)
            QuestDescTextBlock.SetVisibility(widget_visibility.Collapsed)
            QuestProgressTextBlock.SetVisibility(widget_visibility.Collapsed)


        set Modal = canvas_slot:
            Anchors := anchors{Minimum:=vector2{X:=0.0, Y:=0.0}, Maximum:=vector2{X:=1.0, Y:=1.0}}
            Offsets := margin{Top:=0.0, Bottom:=0.0, Left:=0.0, Right:=0.0}
            SizeToContent := true
            Alignment := vector2{X:=0.5, Y:=0.5}
            ZOrder := 9000
            Widget := overlay:
                Slots := array:

                    # Invis to block click through
                    overlay_slot:
                        HorizontalAlignment := horizontal_alignment.Fill
                        VerticalAlignment := vertical_alignment.Fill
                        Padding := margin{Top:=0.0, Bottom:=0.0, Left:=0.0, Right:=0.0}
                        Widget := color_block:
                            DefaultColor := Red
                            DefaultOpacity := 0.0

                    # Modal Container Alignment
                    overlay_slot:
                        HorizontalAlignment := horizontal_alignment.Right
                        VerticalAlignment := vertical_alignment.Bottom
                        Padding := margin{Top:=128.0, Bottom:=0.0, Left:=0.0, Right:=28.0}
                        Widget := overlay:
                            Slots := array:
                                # Modal Drop Shadow
                                overlay_slot:
                                    HorizontalAlignment := horizontal_alignment.Fill
                                    VerticalAlignment := vertical_alignment.Fill
                                    Padding := margin{Top:=0.0, Bottom:=0.0, Left:=0.0, Right:=0.0}
                                    Widget := MakeDropShadowBackground(MakeColorFromHSV(240.0,1.0,0.35), ?TextureSize:=512.0)

                                overlay_slot:
                                    HorizontalAlignment := horizontal_alignment.Center
                                    VerticalAlignment := vertical_alignment.Center
                                    Padding := margin{Top:=80.0, Bottom:=80.0, Left:=80.0, Right:=80.0}
                                    Widget := overlay:
                                        Slots := array:
                                            # Modal BG
                                            overlay_slot:
                                                HorizontalAlignment := horizontal_alignment.Fill
                                                VerticalAlignment := vertical_alignment.Fill
                                                Padding := margin{Top:=0.0, Bottom:=0.0, Left:=0.0, Right:=0.0}
                                                Widget := MakeRoundedOverlayQuietButton(MakeColorFromHSV(240.0,1.0,0.35), ?TextureSize:=40.0)
                                            
                                            # Modal Content
                                            overlay_slot:
                                                HorizontalAlignment := horizontal_alignment.Fill
                                                VerticalAlignment := vertical_alignment.Fill
                                                Padding := margin{Top:=60.0, Bottom:=60.0, Left:=60.0, Right:=60.0}
                                                Widget := stack_box:
                                                    Orientation := orientation.Vertical
                                                    Slots := array:
                                                        stack_box_slot:
                                                            HorizontalAlignment := horizontal_alignment.Fill
                                                            VerticalAlignment := vertical_alignment.Top
                                                            Padding := margin{Top:=30.0, Bottom:=0.0, Left:=0.0, Right:=0.0}
                                                            Widget := NameTextBlock
                                                        stack_box_slot:
                                                            HorizontalAlignment := horizontal_alignment.Fill
                                                            VerticalAlignment := vertical_alignment.Bottom
                                                            Padding := margin{Top:=30.0, Bottom:=0.0, Left:=0.0, Right:=0.0}
                                                            Widget := overlay:
                                                                Slots := array:
                                                                    overlay_slot:
                                                                        HorizontalAlignment := horizontal_alignment.Center
                                                                        VerticalAlignment := vertical_alignment.Center
                                                                        Padding := margin{Top:=0.0, Bottom:=0.0, Left:=0.0, Right:=0.0}
                                                                        Widget := PaintBGTextureBlock
                                                                    overlay_slot:
                                                                        HorizontalAlignment := horizontal_alignment.Center
                                                                        VerticalAlignment := vertical_alignment.Center
                                                                        Padding := margin{Top:=0.0, Bottom:=0.0, Left:=0.0, Right:=0.0}
                                                                        Widget := PaintStyleTextureBlock
                                                        stack_box_slot:
                                                            Distribution := option{2.0}
                                                            HorizontalAlignment := horizontal_alignment.Fill
                                                            VerticalAlignment := vertical_alignment.Top
                                                            Padding := margin{Top:=30.0, Bottom:=0.0, Left:=0.0, Right:=0.0}
                                                            Widget := overlay:
                                                                Slots := array:
                                                                    overlay_slot:
                                                                        HorizontalAlignment := horizontal_alignment.Fill
                                                                        VerticalAlignment := vertical_alignment.Fill
                                                                        Padding := margin{Top:=0.0, Bottom:=0.0, Left:=0.0, Right:=0.0}
                                                                        Widget :=  MakeRoundedOverlayQuietButton(MakeColorFromHSV(240.0,1.0,0.25), ?TextureSize:=24.0)
                                                                    overlay_slot:
                                                                        HorizontalAlignment := horizontal_alignment.Fill
                                                                        VerticalAlignment := vertical_alignment.Fill
                                                                        Padding := margin{Top:=20.0, Bottom:=20.0, Left:=20.0, Right:=20.0}
                                                                        Widget := stack_box:
                                                                            Orientation := orientation.Vertical
                                                                            Slots := array:
                                                                                stack_box_slot:
                                                                                    HorizontalAlignment := horizontal_alignment.Left
                                                                                    VerticalAlignment := vertical_alignment.Top
                                                                                    Padding := margin{Top:=10.0, Bottom:=10.0, Left:=0.0, Right:=0.0}
                                                                                    Widget := QuestTitleTextBlock
                                                                                stack_box_slot:
                                                                                    HorizontalAlignment := horizontal_alignment.Left
                                                                                    VerticalAlignment := vertical_alignment.Top
                                                                                    Padding := margin{Top:=10.0, Bottom:=0.0, Left:=0.0, Right:=0.0}
                                                                                    Widget := QuestDescTextBlock
                                                                                stack_box_slot:
                                                                                    HorizontalAlignment := horizontal_alignment.Left
                                                                                    VerticalAlignment := vertical_alignment.Bottom
                                                                                    Padding := margin{Top:=50.0, Bottom:=0.0, Left:=0.0, Right:=0.0}
                                                                                    Widget := QuestProgressBG
                                            overlay_slot:
                                                HorizontalAlignment := horizontal_alignment.Right
                                                VerticalAlignment := vertical_alignment.Top
                                                Padding := margin{Top:=20.0, Bottom:=0.0, Left:=0.0, Right:=20.0}
                                                Widget := ButtonClose


    # HandleCloseButtonClick(Message:widget_message):void=
    #     if:
    #         PlayerUI := GetPlayerUI[Message.Player]
    #     then:
    #         PlayerUI.RemoveWidget(Modal.Widget)

    DelayApplyPaintPreview(PaintStyle:paint_ui_option)<suspends>:void=
        Sleep(1.0)
        case(PaintStyle.PaintStyleIndex):
            # 0 = Plastic
            # 1 = Metallic
            # 2 = Putty
            # 3 = Galaxy
            0 => block:
                PlasticMaterialData:=plastic_material_data:
                    ColorTint := PaintBGColor
                    EmissiveFlickerTint := PaintStyle.PaintBGColor
                NewMI := CreatePlasticMaterialInstance(PlasticMaterialData)
                PropGameBoard.Prop.SetMaterial(NewMI, ?Index:=1)
                PropGameBoard.Prop.SetMaterial(NewMI, ?Index:=2)
                PropGameBoard.Prop.SetMaterial(NewMI, ?Index:=3)
            1 => block:
                PlasticMaterialData:=plastic_material_data:
                    ColorTint := PaintBGColor
                    MetallicAmount := 1.0
                    EmissiveFlickerTint := PaintStyle.PaintBGColor
                NewMI := CreateMetallicMaterialInstance(PlasticMaterialData)
                PropGameBoard.Prop.SetMaterial(NewMI, ?Index:=1)
                PropGameBoard.Prop.SetMaterial(NewMI, ?Index:=2)
                PropGameBoard.Prop.SetMaterial(NewMI, ?Index:=3)

            2 => block:
                if(SecondaryColor := PaintStyle.SecondaryBGColor?):
                    PlasticMaterialData:=putty_material_data:
                        ColorClay1 := PaintBGColor
                        ColorClay2 := PaintBGColor.HueShift(45.0)
                        EmissiveFlickerTint := PaintStyle.PaintBGColor
                    NewMI := CreatePuttyMaterialInstance(PlasticMaterialData)
                    PropGameBoard.Prop.SetMaterial(NewMI, ?Index:=1)
                    PropGameBoard.Prop.SetMaterial(NewMI, ?Index:=2)
                    PropGameBoard.Prop.SetMaterial(NewMI, ?Index:=3)

            _ => block:
                PlasticMaterialData:=plastic_material_data:
                    ColorTint := PaintBGColor
                    EmissiveFlickerTint := PaintStyle.PaintBGColor
                NewMI := CreatePlasticMaterialInstance(PlasticMaterialData)
                PropGameBoard.Prop.SetMaterial(NewMI, ?Index:=1)
                PropGameBoard.Prop.SetMaterial(NewMI, ?Index:=2)
                PropGameBoard.Prop.SetMaterial(NewMI, ?Index:=3)