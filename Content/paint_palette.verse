using { /Verse.org/Simulation }
using { /Fortnite.com/Devices }
using { /Verse.org/Assets }
using { /Verse.org/Colors }
using { /Fortnite.com/UI }
using { /UnrealEngine.com/Temporary/UI }
using { /UnrealEngine.com/Temporary/SpatialMath }
using { /Verse.org/Colors/NamedColors }


global_paint := struct:
    PaintBaseMaterial <public>: material 
    PaintBGTexture <public>: texture # what the paint looks like
    PaintStyleColor <public>: color # the color of the paint
    PaintStyleTexture <public>: texture # style of the paint (plain, metallic, putty, etc.)


# AllAvailablePaints : []global_paint = array:
#     # Pink
#     global_paint{PaintBGTexture := _UI.UI_Textures.ui_square_blank, PaintStyleColor := MakeColorFromHSV(324.38913, 0.948145, 0.947917), PaintStyleTexture := _UI.UI_Textures.transparent_square}
#     # Yellow
#     global_paint{PaintBGTexture := _UI.UI_Textures.ui_square_blank, PaintStyleColor := MakeColorFromHSV(45.33017, 1.0, 0.947307), PaintStyleTexture := _UI.UI_Textures.transparent_square}
#     # Orange
#     global_paint{PaintBGTexture := _UI.UI_Textures.ui_square_blank, PaintStyleColor := MakeColorFromHSV(30.0, 1.0, 0.947307), PaintStyleTexture := _UI.UI_Textures.transparent_square}
#     # Red
#     global_paint{PaintBGTexture := _UI.UI_Textures.ui_square_blank, PaintStyleColor := MakeColorFromHSV(0.0, 1.0, 0.947307), PaintStyleTexture := _UI.UI_Textures.transparent_square}
#     # Purple
#     global_paint{PaintBGTexture := _UI.UI_Textures.ui_square_blank, PaintStyleColor := MakeColorFromHSV(300.0, 1.0, 0.947307), PaintStyleTexture := _UI.UI_Textures.transparent_square}
#     # Dark Blue
#     global_paint{PaintBGTexture := _UI.UI_Textures.ui_square_blank, PaintStyleColor := MakeColorFromHSV(240.0, 1.0, 0.947307), PaintStyleTexture := _UI.UI_Textures.transparent_square}
#     # Light Blue
#     global_paint{PaintBGTexture := _UI.UI_Textures.ui_square_blank, PaintStyleColor := MakeColorFromHSV(210.0, 1.0, 0.947307), PaintStyleTexture := _UI.UI_Textures.transparent_square}
#     # Green
#     global_paint{PaintBGTexture := _UI.UI_Textures.ui_square_blank, PaintStyleColor := MakeColorFromHSV(120.0, 1.0, 0.947307), PaintStyleTexture := _UI.UI_Textures.transparent_square}
#     # White
#     global_paint{PaintBGTexture := _UI.UI_Textures.ui_square_blank, PaintStyleColor := MakeColorFromHSV(0.0, 0.0, 0.947307), PaintStyleTexture := _UI.UI_Textures.transparent_square}
#     # Grey
#     global_paint{PaintBGTexture := _UI.UI_Textures.ui_square_blank, PaintStyleColor := MakeColorFromHSV(0.0, 0.0, 0.5), PaintStyleTexture := _UI.UI_Textures.transparent_square}
#     # Black
#     global_paint{PaintBGTexture := _UI.UI_Textures.ui_square_blank, PaintStyleColor := MakeColorFromHSV(0.0, 0.0, 0.0), PaintStyleTexture := _UI.UI_Textures.transparent_square}
#     # Peach
#     global_paint{PaintBGTexture := _UI.UI_Textures.ui_square_blank, PaintStyleColor := MakeColorFromHSV(5.0, 0.6, 0.95), PaintStyleTexture := _UI.UI_Textures.transparent_square}
#     # Bronze
#     global_paint{PaintBGTexture := _UI.UI_Textures.ui_square_blank, PaintStyleColor := MakeColorFromHSV(30.0, 0.5, 0.5), PaintStyleTexture := _UI.UI_Textures.transparent_square}
#     # Silver
#     global_paint{PaintBGTexture := _UI.UI_Textures.ui_square_blank, PaintStyleColor := MakeColorFromHSV(0.0, 0.0, 0.75), PaintStyleTexture := _UI.UI_Textures.transparent_square}
#     # Gold
#     global_paint{PaintBGTexture := _UI.UI_Textures.ui_square_blank, PaintStyleColor := MakeColorFromHSV(45.0, 0.5, 0.5), PaintStyleTexture := _UI.UI_Textures.transparent_square}
#     # Pink Clay
#     global_paint{PaintBGTexture := _UI.UI_Textures.ui_square_blank, PaintStyleColor := MakeColorFromHSV(324.38913, 0.648145, 0.947917), PaintStyleTexture := _UI.UI_Textures.transparent_square}
#     # Yellow Clay
#     global_paint{PaintBGTexture := _UI.UI_Textures.ui_square_blank, PaintStyleColor := MakeColorFromHSV(45.33017, 1.0, 0.947307), PaintStyleTexture := _UI.UI_Textures.transparent_square}
#     # Space Clay
#     global_paint{PaintBGTexture := _UI.UI_Textures.ui_square_blank, PaintStyleColor := MakeColorFromHSV(270.0, 0.5, 0.5), PaintStyleTexture := _UI.UI_Textures.transparent_square}

    
# PremadePaintSets : []premade_paint_set = array:
#     #Default, Pink Board, Yellow Cards
#     premade_paint_set:
#         BoardPaint := board_paint{Rollers := 1, Rim := 0, Base := 0, Underside := 0, Mesh := 0}
#         CardBackPaint := card_back_paint{CardBack := 1}
#         YouStandPaint := you_stand_paint{Stand := 0, CardHolder := 1, Mesh := 0}
#         PlatformPaint := platform_paint{Base := 0, Trim := 0, Main := 0, Mesh := 0}





UI_BGColor : color = MakeColorFromHSV(208.0, 1.0, 1.0)
UI_BGColor2 : color = MakeColorFromHSV(224.0, 1.0, 0.54)

board_paint := struct:
    Rollers<public> : material
    Rim <public>: material
    Base<public> : material
    Underside <public>: material
    Mesh <public>: mesh

board_paint_data:= struct<persistable>:
    Rollers <public>: int
    Rim <public>: int
    Base<public> : int
    Underside <public>: int
    Mesh <public>: int

card_back_paint := struct:
    CardBack <public>: material

card_back_paint_data:= struct<persistable>:
    CardBack <public>: int

you_stand_paint := struct:
    Stand <public>: material
    CardHolder <public>: material
    Mesh <public>: mesh

you_stand_paint_data:= struct<persistable>:
    Stand <public>: int
    CardHolder <public>: int
    Mesh <public>: int

platform_paint := struct:
    Base <public>: material
    Trim <public>: material
    Main <public>: material
    Mesh <public> : mesh

platform_paint_data:= struct<persistable>:
    Base <public>: int
    Trim <public>: int
    Main <public> : int
    Mesh <public> : int

premade_paint_set := struct:
    BoardPaint <public>: board_paint
    CardBackPaint <public> : card_back_paint
    YouStandPaint <public> : you_stand_paint
    PlatformPaint <public> : platform_paint

custom_paint_set := struct<persistable>:
    Name <public>: string
    BoardPaint <public>: board_paint_data
    CardBackPaint <public>: card_back_paint_data
    YouStandPaint <public>: you_stand_paint_data
    PlatformPaint <public>: platform_paint_data

paint_palette := class<concrete>():

    var MaybeGameBoard : ?game_board = false

    @editable Button : button_device = button_device{}
    @editable LockerCamera : gameplay_camera_fixed_point_device = gameplay_camera_fixed_point_device{}

    @editable ConditionalButton : conditional_button_device = conditional_button_device{}

    @editable PaintCanProp : creative_prop = creative_prop{}
    @editable UnlockablePaints : []unlockable_paint = array{}
    @editable HUD_NoMorePaintsToUnlock : hud_message_device = hud_message_device{}

    BoardColorss : []color = array:
        MakeColorFromHSV(324.38913, 0.948145, 0.947917) # Pink
        MakeColorFromHSV(45.33017, 1.0, 0.947307) # Yellow
        MakeColorFromHSV(30.0, 1.0, 0.947307) # Orange
        MakeColorFromHSV(0.0, 1.0, 0.947307) # Red
        MakeColorFromHSV(300.0, 1.0, 0.947307) # Purple
        MakeColorFromHSV(240.0, 1.0, 0.947307) # Dark Blue
        MakeColorFromHSV(210.0, 1.0, 0.947307) # Light Blue
        MakeColorFromHSV(120.0, 1.0, 0.947307) # Green
        MakeColorFromHSV(0.0, 0.0, 0.947307) # White
        MakeColorFromHSV(0.0, 0.0, 0.5) # Grey
        MakeColorFromHSV(0.0, 0.0, 0.0) # Black
        MakeColorFromHSV(5.0, 0.6, 0.95) # Peach
        MakeColorFromHSV(30.0, 0.5, 0.5) # Bronze
        MakeColorFromHSV(0.0, 0.0, 0.75) # Silver
        MakeColorFromHSV(45.0, 0.5, 0.5) # Gold
        MakeColorFromHSV(324.38913, 0.648145, 0.947917) # Pink Clay
        MakeColorFromHSV(45.33017, 1.0, 0.947307) # Yellow Clay
        MakeColorFromHSV(270.0, 0.5, 0.5) # Space Clay        
        MakeColorFromHSV(30.0, 1.0, 0.947307) # Orange Clay

    

    # var TotalBoardColors : int = 11 # 11 default colors
    var TotalBoardColors : int = 12 # 11 default colors + Peach
    


    FrameColorss : []color = array:
        MakeColorFromHSV(324.38913, 0.948145, 0.947917) # Pink
        MakeColorFromHSV(45.33017, 1.0, 0.947307) # Yellow
        MakeColorFromHSV(30.0, 1.0, 0.947307) # Orange
        MakeColorFromHSV(0.0, 1.0, 0.947307) # Red
        MakeColorFromHSV(300.0, 1.0, 0.947307) # Purple
        MakeColorFromHSV(240.0, 1.0, 0.947307) # Dark Blue
        MakeColorFromHSV(210.0, 1.0, 0.947307) # Light Blue
        MakeColorFromHSV(120.0, 1.0, 0.947307) # Green
        MakeColorFromHSV(0.0, 0.0, 0.947307) # White
        MakeColorFromHSV(0.0, 0.0, 0.5) # Grey
        MakeColorFromHSV(0.0, 0.0, 0.0) # Black
        MakeColorFromHSV(5.0, 0.6, 0.95) # Peach
        MakeColorFromHSV(30.0, 0.5, 0.5) # Bronze
        MakeColorFromHSV(0.0, 0.0, 0.75) # Silver
        MakeColorFromHSV(45.0, 0.5, 0.5) # Gold
        MakeColorFromHSV(324.38913, 0.648145, 0.947917) # Pink Clay
        MakeColorFromHSV(45.33017, 1.0, 0.947307) # Yellow Clay
        MakeColorFromHSV(270.0, 0.5, 0.5) # Space Clay
        MakeColorFromHSV(30.0, 1.0, 0.947307) # Orange Clay

    # var TotalFrameColors : int = 11 # 11 default colors
    var TotalFrameColors : int = 12 # 11 default colors + Peach

    var TotalUnlockedPaints : int = 0

    # @editable AnaltyicsBoardColor : []analytics_device = array{}
    @editable TrackerBoardColorIndex : tracker_device = tracker_device{}
    var BoardColorIndex : int = 0
    
    # @editable AnaltyicsFrameColor : []analytics_device = array{}
    @editable TrackerFrameColorIndex : tracker_device = tracker_device{}
    var FrameColorIndex : int = 1

    @editable TrackerStandColorIndex : tracker_device = tracker_device{}
    var StandColorIndex : int = 0

    @editable TrackerPlatformColorIndex : tracker_device = tracker_device{}
    var PlatformColorIndex : int = 0
        
    TextureBoard : texture_block = texture_block:
        DefaultImage := _UI.UI_Textures.paint_palette_board
        DefaultTint := MakeColorFromHSV(324.38913, 0.948145, 0.947917)
        DefaultDesiredSize := vector2{X := 300.0, Y := 300.0}

    var BoardColor : color = MakeColorFromHSV(324.38913, 0.948145, 0.947917)

    TextureFrame : texture_block = texture_block:
        DefaultImage := _UI.UI_Textures.paint_palette_cards
        DefaultTint := MakeColorFromHSV(45.33017, 1.0, 0.947307)
        DefaultDesiredSize := vector2{X := 300.0, Y := 300.0}


    var FrameColor : color = MakeColorFromHSV(45.33017, 1.0, 0.947307)

    ButtonBoardColorLeft : button_quiet = button_quiet{}
    ButtonBoardColorRight : button_quiet = button_quiet{}

    ButtonFrameColorLeft : button_quiet = button_quiet{}
    ButtonFrameColorRight : button_quiet = button_quiet{}

    ButtonApply : button_loud = button_loud{}
    MessageApply<localizes>:message = "APPLY"

    TexturePaintCan : texture_block = texture_block:
        DefaultImage := _UI.UI_Textures.ui_paintcan
        DefaultDesiredSize := vector2{X := 48.0, Y := 48.0}


    TextBoardDesignerTitle : text_block = text_block:
        DefaultJustification:= text_justification.Center
        DefaultTextColor := White

    

    TextBoardLabel : text_block = text_block:
        DefaultJustification:= text_justification.Center
        DefaultTextColor := White
    MessageBoardLabel<localizes>:message = "BOARD"
    TextFrameLabel : text_block = text_block:
        DefaultJustification:= text_justification.Center
        DefaultTextColor := White
    MessageFrameLabel<localizes>:message = "CARDS"



    # NEW LOCKER #################################################################
    var GoldCountOverlay : overlay = overlay{}
    var TurnTimerOverlay : overlay = overlay{}
    var DisplayedPaints : overlay = overlay{}
    
    TextBlockTitle : text_block = text_block{DefaultText := MessageBoardDesignerTitle_EN, DefaultTextColor := White}
    
    ButtonPageSelect0 : button_quiet = button_quiet{DefaultText := MessageBoard_EN}
    ButtonPageSelect1 : button_quiet = button_quiet{DefaultText := MessageCards_EN}
    ButtonPageSelect2 : button_quiet = button_quiet{DefaultText := MessageStand_EN}
    ButtonPageSelect3 : button_quiet = button_quiet{DefaultText := MessagePlatform_EN}
    
    
    TextFieldSearch : text_block = text_block{}
    ButtonSearch : button_quiet = button_quiet{ DefaultText := MessageSearch_EN }
    
    ButtonExit : button_quiet = button_quiet{ DefaultText := MessageExit_EN }

    var CurrentPage : int = 0
    var CurrentInnerPage : int = 0

    ################################################################################################

    SaveCustomPaintSet(Player:player, SetName:string, BoardPaint:board_paint_data, CardBackPaint:card_back_paint_data, YouStandPaint:you_stand_paint_data, PlatformPaint:platform_paint_data):void=
        NewPaintSet := custom_paint_set:
            Name := SetName
            BoardPaint := BoardPaint
            CardBackPaint := CardBackPaint
            YouStandPaint := YouStandPaint
            PlatformPaint := PlatformPaint
            
        if:
            CheckForPlayerSavedData[Player]
            SourceData := PlayerSavedDataMap[Player]
            var NewCustomPaintSets : []custom_paint_set = SourceData.CustomPaintSets
            set NewCustomPaintSets += array{NewPaintSet}
            set PlayerSavedDataMap[Player] = player_saved_data:
                MakePlayerSavedData<constructor>(SourceData)
                CustomPaintSets := NewCustomPaintSets
        else:
            Print("No player data found for Player\n")
    
    
    

    ApplyCustomPaintSet(Player:player, SetIndex:int):void=
        if:
            CheckForPlayerSavedData[Player]
            SourceData := PlayerSavedDataMap[Player]
            PaintSet := SourceData.CustomPaintSets[SetIndex]

            set PlayerSavedDataMap[Player] = player_saved_data:
                MakePlayerSavedData<constructor>(SourceData)
                CurrentBoardPaint := PaintSet.BoardPaint
                CurrentCardBackPaint := PaintSet.CardBackPaint
                CurrentYouStandPaint := PaintSet.YouStandPaint
                CurrentPlatformPaint := PaintSet.PlatformPaint
        then:
            ApplyChangesFromSavedData(Player)
        else:
            Print("No player data found for Player\n")

    ApplyChangesFromSavedData(Player:player):void=
        if:
            CheckForPlayerSavedData[Player]
            SourceData := PlayerSavedDataMap[Player]
            GameBoard := MaybeGameBoard?
        then:
            BoardPaint := SourceData.CurrentBoardPaint
            CardBackPaint := SourceData.CurrentCardBackPaint
            YouStandPaint := SourceData.CurrentYouStandPaint
            PlatformPaint := SourceData.CurrentPlatformPaint
    
            GameBoard.BoardProp.SetMesh(GetMeshFromIndexBoard(BoardPaint.Mesh))
            GameBoard.BoardProp.SetMaterial(GetMaterialFromIndexBoard(BoardPaint.Rollers), ?Index:=0)
            GameBoard.BoardProp.SetMaterial(GetMaterialFromIndexBoard(BoardPaint.Rim), ?Index:=1)
            GameBoard.BoardProp.SetMaterial(GetMaterialFromIndexBoard(BoardPaint.Base), ?Index:=2)
            GameBoard.BoardProp.SetMaterial(GetMaterialFromIndexBoard(BoardPaint.Underside), ?Index:=3)
    
            GameBoard.YouStandProp.SetMesh(GetMeshFromIndexStand(YouStandPaint.Mesh))
            GameBoard.YouStandProp.SetMaterial(GetMaterialFromIndexStand(YouStandPaint.Stand), ?Index:=0)
            GameBoard.YouStandProp.SetMaterial(GetMaterialFromIndexStand(YouStandPaint.CardHolder), ?Index:=1)
    
            GameBoard.PlatformProp.SetMesh(GetMeshFromIndexPlatform(PlatformPaint.Mesh))
            GameBoard.PlatformProp.SetMaterial(GetMaterialFromIndexPlatform(PlatformPaint.Base), ?Index:=0)
            GameBoard.PlatformProp.SetMaterial(GetMaterialFromIndexPlatform(PlatformPaint.Trim), ?Index:=1)
            GameBoard.PlatformProp.SetMaterial(GetMaterialFromIndexPlatform(PlatformPaint.Main), ?Index:=2)
    
            for(RowIndex -> Row : GameBoard.Rows):
                for(CardIndex -> Prop : Row.Tiles):
                    Prop.Prop.SetMaterial(GetMaterialFromIndexFrame(CardBackPaint.CardBack), ?Index:=2)
            
        
    GetMaterialFromBoardPaintData(PaintData:board_paint_data):board_paint=
        return board_paint{
            Rollers := GetMaterialFromIndexBoard(PaintData.Rollers),
            Rim := GetMaterialFromIndexBoard(PaintData.Rim),
            Base := GetMaterialFromIndexBoard(PaintData.Base),
            Underside := GetMaterialFromIndexBoard(PaintData.Underside),
            Mesh := GetMeshFromIndexBoard(PaintData.Mesh)
        }

    GetMaterialFromCardBackPaintData(PaintData:card_back_paint_data):card_back_paint=
        return card_back_paint{
            CardBack := GetMaterialFromIndexFrame(PaintData.CardBack)
        }

    GetMaterialFromYouStandPaintData(PaintData:you_stand_paint_data):you_stand_paint=
        return you_stand_paint{
            Stand := GetMaterialFromIndexStand(PaintData.Stand),
            CardHolder := GetMaterialFromIndexStand(PaintData.CardHolder),
            Mesh := GetMeshFromIndexStand(PaintData.Mesh)
        }

    GetMaterialFromPlatformPaintData(PaintData:platform_paint_data):platform_paint=
        return platform_paint{
            Base := GetMaterialFromIndexPlatform(PaintData.Base),
            Trim := GetMaterialFromIndexPlatform(PaintData.Trim),
            Main := GetMaterialFromIndexPlatform(PaintData.Main),
            Mesh := GetMeshFromIndexPlatform(PaintData.Mesh)
        }

    ConvertOldData(Player:player):void=
        if:
            CheckForPlayerSavedData[Player]
            SourceData := PlayerSavedDataMap[Player]
            not SourceData.ConvertedUnlockedPaints? # if not already converted
            SavedBoardColorIndex := TrackerBoardColorIndex.GetValue(Player)
            SavedFrameColorIndex := TrackerFrameColorIndex.GetValue(Player)
            SavedStandColorIndex := TrackerStandColorIndex.GetValue(Player)
            SavedPlatformColorIndex := TrackerPlatformColorIndex.GetValue(Player)
    
            BoardPaintData := board_paint_data{Rollers := SavedBoardColorIndex, Rim := SavedBoardColorIndex, Base := SavedBoardColorIndex, Underside := SavedBoardColorIndex, Mesh := 0}
            CardBackPaintData := card_back_paint_data{CardBack := SavedFrameColorIndex}
            YouStandPaintData := you_stand_paint_data{Stand := SavedStandColorIndex, CardHolder := SavedStandColorIndex, Mesh := 0}
            PlatformPaintData := platform_paint_data{Base := SavedPlatformColorIndex, Trim := SavedPlatformColorIndex, Main := SavedPlatformColorIndex, Mesh := 0}
    
            set PlayerSavedDataMap[Player] = player_saved_data:
                MakePlayerSavedData<constructor>(SourceData)
                ConvertedUnlockedPaints := true
                CurrentBoardPaint := BoardPaintData
                CurrentCardBackPaint := CardBackPaintData
                CurrentYouStandPaint := YouStandPaintData
                CurrentPlatformPaint := PlatformPaintData
        then:
            UpdateSavedData(Player, SourceData)
        else:
            Print("No player data found for Player\n")
            

    Init(CPlayer:custom_player, GameBoard:game_board):void=
        set MaybeGameBoard = option{GameBoard}

        Player := CPlayer.Player

        SavedBoardColorIndex := TrackerBoardColorIndex.GetValue(Player)
        SavedFrameColorIndex := TrackerFrameColorIndex.GetValue(Player)
        # SavedStandColorIndex := TrackerStandColorIndex.GetValue(Player)
        # SavedPlatformColorIndex := TrackerPlatformColorIndex.GetValue(Player)

        set BoardColorIndex = SavedBoardColorIndex
        set FrameColorIndex = SavedFrameColorIndex

        Print("Saved board color index: {SavedBoardColorIndex}")
        Print("Saved frame color index: {SavedFrameColorIndex}")


        BoardMaterial := GetMaterialFromIndexBoard(SavedBoardColorIndex)
        FrameMaterial := GetMaterialFromIndexFrame(SavedFrameColorIndex)
        # StandMaterial := GetMaterialFromIndexStand(SavedStandColorIndex)
        # PlatformMaterial := GetMaterialFromIndexPlatform(SavedPlatformColorIndex)

        GameBoard.BoardProp.SetMaterial(GetMaterialFromIndexBoard(SavedFrameColorIndex), ?Index:=0) # rods
        GameBoard.BoardProp.SetMaterial(BoardMaterial, ?Index:=1) # rim
        GameBoard.BoardProp.SetMaterial(BoardMaterial, ?Index:=2) # inside
        GameBoard.BoardProp.SetMaterial(BoardMaterial, ?Index:=3) #bottom

        # for(Prop : GameBoard.CylinderProps):
        #     Prop.SetMaterial(FrameMaterial)

        for(Row : GameBoard.Rows):
            for(Prop : Row.Tiles):
                # Prop.Prop.SetMaterial(FrameMaterial, ?Index:=0) # Card Style
                # Prop.Prop.SetMaterial(FrameMaterial, ?Index:=1) # Transparent
                Prop.Prop.SetMaterial(FrameMaterial, ?Index:=2)


        # GameBoard.YouStandProp.SetMaterial(BoardMaterial, ?Index:=0) # like & fav sign
        GameBoard.YouStandProp.SetMaterial(GetMaterialFromIndexStand(BoardColorIndex), ?Index:=1)
        GameBoard.YouStandProp.SetMaterial(GetMaterialFromIndexStand(FrameColorIndex), ?Index:=2)

        GameBoard.YouTile.Prop.SetMaterial(FrameMaterial, ?Index:=2)

        # GameBoard.PlatformProp.SetMaterial(BoardMaterial, ?Index:=0) # white floor
        # GameBoard.PlatformProp.SetMaterial(BoardMaterial, ?Index:=1) # black trim
        GameBoard.PlatformProp.SetMaterial(GetMaterialFromIndexPlatform(BoardColorIndex), ?Index:=2) 

        GameBoard.IndexCard.Card.SetMaterial(FrameMaterial, ?Index:=0)
        # GameBoard.IndexCard.Card.SetMaterial(FrameMaterial, ?Index:=1) # index card image


        for(Unlockable : UnlockablePaints):
            if(Unlockable.IsUnlocked[Player]):
                set TotalBoardColors += 1
                set TotalFrameColors += 1
                set TotalUnlockedPaints += 1
                Print("Unlocked Paint: {Unlockable.Name}")
            # else:
            #     Unlockable.Switch.TurnedOnEvent.SubscribeAgent(OnUnlockableSwitchedOn, Unlockable)

        
        UpdateNextUnlockablePaint(Player)

        ConditionalButton.ActivatedEvent.Subscribe(UnlockNextPaint)


        Button.InteractedWithEvent.Subscribe(HandleOpenPopup)
        ButtonBoardColorLeft.OnClick().Subscribe(PrevBoardColor)
        ButtonBoardColorRight.OnClick().Subscribe(NextBoardColor)
        ButtonFrameColorLeft.OnClick().Subscribe(PrevFrameColor)
        ButtonFrameColorRight.OnClick().Subscribe(NextFrameColor)
        ButtonApply.OnClick().Subscribe(ApplyChanges)


    UpdateNextUnlockablePaint(Player:player):void=
        if:
            # TotalUnlockedPaints <> UnlockablePaints.Length
            NextUnlockable := UnlockablePaints[TotalUnlockedPaints]
        then:
            Print("Next unlockable paint: {NextUnlockable.Name}")
            ConditionalButton.SetItemCountRequired(0, NextUnlockable.Cost)
            PaintCanProp.SetMaterial(GetMaterialFromIndexFrame(TotalBoardColors), ?Index:=1)
            # PaintCanProp.SetMaterial(GetMaterialFromIndexPlatform(TotalBoardColors), ?Index:=1)
            ConditionalButton.SetInteractionText(TwoLineMessage("Purchase {NextUnlockable.Name} Paint", "💰 {NextUnlockable.Cost}"))
        else:
            ConditionalButton.SetItemCountRequired(0, 0)
            PaintCanProp.SetMaterial(Materials.M_Clear_Transparent, ?Index:=1)
            ConditionalButton.SetInteractionText(TwoLineMessage("No more paints to unlock", "Check back later!"))
        


    UnlockNextPaint(Agent:agent):void=
        Print("Unlocking next paint")
        if:
            UnlockedPaint := UnlockablePaints[TotalUnlockedPaints]
            Player := player[Agent]
        then:
            UnlockedPaint.Unlock(Agent)

            set TotalBoardColors += 1
            set TotalFrameColors += 1
            set TotalUnlockedPaints += 1

            UpdateNextUnlockablePaint(Player)
        else:
            Print("No more paints to unlock")
            HUD_NoMorePaintsToUnlock.Show(Agent)

    HandleOpenPopup(Agent:agent, ?BetaTest:logic=false):void=
        Print("Opening paint palette popup")

        if:
            Player := player[Agent]
            PlayerUI := GetPlayerUI[Player]
            TheGlobals := Globals[GetSession()]
            CPlayer := TheGlobals.CustomPlayerMap[Player]
        then:
            
            if(not BetaTest?):
                NewCanvas := CreatePaintPaletteCanvas()
                set CPlayer.GuessUI.MaybePaintPaletteUI = option{NewCanvas}
                PlayerUI.AddWidget(NewCanvas, player_ui_slot{ZOrder:= 9999, InputMode := ui_input_mode.All})
            else:
                NewCanvas := CreateLockerCanvas(Player)
                set CPlayer.GuessUI.MaybePaintPaletteUI = option{NewCanvas}
                PlayerUI.AddWidget(NewCanvas, player_ui_slot{ZOrder:= 9999, InputMode := ui_input_mode.All})
                LockerCamera.AddTo(Agent)



    CreatePaintPaletteCanvas():canvas=

        var LayoutCanvas : canvas = canvas{}

        PaddingBorder := 8.0


        TextBoardDesignerTitle.SetText(MessageBoardDesignerTitle_EN)
        TextBoardLabel.SetText(MessageBoardLabel)
        TextFrameLabel.SetText(MessageFrameLabel)

        ButtonBoardColorLeft.SetText(MessageArrowLeft_EN)
        ButtonBoardColorRight.SetText(MessageArrowRight_EN)
        ButtonFrameColorLeft.SetText(MessageArrowLeft_EN)
        ButtonFrameColorRight.SetText(MessageArrowRight_EN)
        ButtonApply.SetText(MessageApply)



        set LayoutCanvas = canvas:
            Slots := array:

                canvas_slot:
                    Anchors := anchors{Minimum := vector2{X := 0.0, Y := 0.5}, Maximum := vector2{X := 0.0, Y := 0.5} }
                        Offsets := margin{Top := 0.0, Left := 250.0, Right := 0.0, Bottom := 0.0}
                        Alignment := vector2{X := 0.0, Y := 0.5}
                        ZOrder := 9999
                        Widget := overlay:
                            Slots := array:
                                overlay_slot:
                                    HorizontalAlignment := horizontal_alignment.Fill
                                    VerticalAlignment := vertical_alignment.Fill
                                    Padding := margin{Top := 0.0, Left := 0.0, Right := 0.0, Bottom := 0.0}
                                    Widget := MakeRoundedOverlayOuter(UI_BGColor)
                                overlay_slot:
                                    HorizontalAlignment := horizontal_alignment.Center
                                    VerticalAlignment := vertical_alignment.Center
                                    Padding := margin{Top := PaddingBorder, Left := PaddingBorder, Right := PaddingBorder, Bottom := PaddingBorder}
                                    Widget := stack_box:
                                        Orientation := orientation.Vertical
                                        Slots := array:
                                            stack_box_slot:
                                                HorizontalAlignment := horizontal_alignment.Fill
                                                VerticalAlignment := vertical_alignment.Fill
                                                Padding := margin{Top:= 0.0, Left:= 0.0, Right:= 0.0, Bottom:= 0.0}
                                                Widget := overlay:
                                                    Slots := array:
                                                        overlay_slot:
                                                            HorizontalAlignment := horizontal_alignment.Fill
                                                            VerticalAlignment := vertical_alignment.Fill
                                                            Padding := margin{Top:= 0.0, Left:= 0.0, Right:= 0.0, Bottom:= 0.0}
                                                            Widget := MakeRoundedOverlayInner(UI_BGColor2)
                                                        overlay_slot:
                                                            HorizontalAlignment := horizontal_alignment.Center
                                                            VerticalAlignment := vertical_alignment.Fill
                                                            Padding := margin{Top:= 40.0, Left:= 100.0, Right:= 100.0, Bottom:= 60.0}
                                                            Widget := stack_box:
                                                                Orientation := orientation.Vertical
                                                                Slots := array:
                                                                    stack_box_slot:
                                                                        HorizontalAlignment := horizontal_alignment.Center
                                                                        VerticalAlignment := vertical_alignment.Center
                                                                        Padding := margin{Top:= 0.0, Left:= 0.0, Right:= 0.0, Bottom:= 0.0}
                                                                        Widget := stack_box:
                                                                            Orientation := orientation.Horizontal
                                                                            Slots := array:
                                                                                stack_box_slot:
                                                                                    HorizontalAlignment := horizontal_alignment.Center
                                                                                    VerticalAlignment := vertical_alignment.Center
                                                                                    Padding := margin{Top:= 0.0, Left:= 0.0, Right:= 0.0, Bottom:= 0.0}
                                                                                    Widget := TexturePaintCan
                                                                                stack_box_slot:
                                                                                    HorizontalAlignment := horizontal_alignment.Fill
                                                                                    VerticalAlignment := vertical_alignment.Center
                                                                                    Padding := margin{Top:= 0.0, Left:= 10.0, Right:= 0.0, Bottom:= 0.0}
                                                                                    Widget := TextBoardDesignerTitle
                                                                    stack_box_slot:
                                                                        HorizontalAlignment := horizontal_alignment.Fill
                                                                        VerticalAlignment := vertical_alignment.Fill
                                                                        Padding := margin{Top:= 20.0, Left:= 20.0, Right:= 20.0, Bottom:= 20.0}
                                                                        Widget := overlay:
                                                                            Slots := array:
                                                                                overlay_slot:
                                                                                    HorizontalAlignment := horizontal_alignment.Center
                                                                                    VerticalAlignment := vertical_alignment.Center
                                                                                    Padding := margin{Top:= 0.0, Left:= 0.0, Right:= 0.0, Bottom:= 0.0}
                                                                                    Widget := TextureBoard
                                                                                overlay_slot:
                                                                                    HorizontalAlignment := horizontal_alignment.Center
                                                                                    VerticalAlignment := vertical_alignment.Center
                                                                                    Padding := margin{Top:= 0.0, Left:= 0.0, Right:= 0.0, Bottom:= 0.0}
                                                                                    Widget := TextureFrame
                                                                    stack_box_slot:
                                                                        HorizontalAlignment := horizontal_alignment.Center
                                                                        VerticalAlignment := vertical_alignment.Center
                                                                        Padding := margin{Top:= 0.0, Left:= 0.0, Right:= 0.0, Bottom:= 10.0}
                                                                        Widget := stack_box:
                                                                            Orientation := orientation.Horizontal
                                                                            Slots := array:
                                                                                stack_box_slot:
                                                                                    HorizontalAlignment := horizontal_alignment.Center
                                                                                    VerticalAlignment := vertical_alignment.Center
                                                                                    Padding := margin{Top:= 0.0, Left:= 0.0, Right:= 0.0, Bottom:= 0.0}
                                                                                    Widget := ButtonBoardColorLeft
                                                                                stack_box_slot:
                                                                                    HorizontalAlignment := horizontal_alignment.Center
                                                                                    VerticalAlignment := vertical_alignment.Center
                                                                                    Padding := margin{Top:= 0.0, Left:= 10.0, Right:= 10.0, Bottom:= 0.0}
                                                                                    Widget := TextBoardLabel
                                                                                stack_box_slot:
                                                                                    HorizontalAlignment := horizontal_alignment.Center
                                                                                    VerticalAlignment := vertical_alignment.Center
                                                                                    Padding := margin{Top:= 0.0, Left:= 0.0, Right:= 0.0, Bottom:= 0.0}
                                                                                    Widget := ButtonBoardColorRight
                                                                    stack_box_slot:
                                                                        HorizontalAlignment := horizontal_alignment.Center
                                                                        VerticalAlignment := vertical_alignment.Center
                                                                        Padding := margin{Top:= 0.0, Left:= 0.0, Right:= 0.0, Bottom:= 0.0}
                                                                        Widget := stack_box:
                                                                            Orientation := orientation.Horizontal
                                                                            Slots := array:
                                                                                stack_box_slot:
                                                                                    HorizontalAlignment := horizontal_alignment.Center
                                                                                    VerticalAlignment := vertical_alignment.Center
                                                                                    Padding := margin{Top:= 0.0, Left:= 0.0, Right:= 0.0, Bottom:= 0.0}
                                                                                    Widget := ButtonFrameColorLeft
                                                                                stack_box_slot:
                                                                                    HorizontalAlignment := horizontal_alignment.Center
                                                                                    VerticalAlignment := vertical_alignment.Center
                                                                                    Padding := margin{Top:= 0.0, Left:= 10.0, Right:= 10.0, Bottom:= 0.0}
                                                                                    Widget := TextFrameLabel
                                                                                stack_box_slot:
                                                                                    HorizontalAlignment := horizontal_alignment.Center
                                                                                    VerticalAlignment := vertical_alignment.Center
                                                                                    Padding := margin{Top:= 0.0, Left:= 0.0, Right:= 0.0, Bottom:= 0.0}
                                                                                    Widget := ButtonFrameColorRight
                                            stack_box_slot:
                                                HorizontalAlignment := horizontal_alignment.Fill
                                                VerticalAlignment := vertical_alignment.Fill
                                                Padding := margin{Top:= -30.0, Left:= 60.0, Right:=60.0, Bottom:= 20.0}
                                                Widget := ButtonApply

        if:
            SavedBoardTint := BoardColorss[BoardColorIndex]
            SavedFrameTint := FrameColorss[FrameColorIndex]
        then:

            if:
                BoardColorIndex > 11
                CustomColorIndex := BoardColorIndex - 11
                CustomPaint := UnlockablePaints[CustomColorIndex]
            then:
                CustomPaintTexture := CustomPaint.GetTextures()
                TextureBoard.SetImage(CustomPaintTexture(0))
            else:
                TextureBoard.SetImage(_UI.UI_Textures.paint_palette_board)

            if:
                FrameColorIndex > 11
                CustomColorIndex := FrameColorIndex - 11
                CustomPaint := UnlockablePaints[CustomColorIndex]
            then:
                CustomPaintTexture := CustomPaint.GetTextures()
                TextureFrame.SetImage(CustomPaintTexture(1))
            else:
                TextureFrame.SetImage(_UI.UI_Textures.paint_palette_cards)

            TextureBoard.SetTint(SavedBoardTint)
            TextureFrame.SetTint(SavedFrameTint)

        return LayoutCanvas

    PrevBoardColor(Response:widget_message):void=
        ChangeBoardColor(Response.Player, -1)

    NextBoardColor(Response:widget_message):void=
        ChangeBoardColor(Response.Player, 1)

    PrevFrameColor(Response:widget_message):void=
        ChangeFrameColor(Response.Player, -1)

    NextFrameColor(Response:widget_message):void=
        ChangeFrameColor(Response.Player, 1)

    ApplyChanges(Response:widget_message):void=
        Print("Applying changes to board and frame colors")

        if:
            PlayerUI := GetPlayerUI[Response.Player]
            TheGlobals := Globals[GetSession()]
            CPlayer := TheGlobals.CustomPlayerMap[Response.Player]
            ExistingCanvas := CPlayer.GuessUI.MaybePaintPaletteUI?
        then:
            TrackerBoardColorIndex.SetValue(Response.Player, BoardColorIndex)
            TrackerFrameColorIndex.SetValue(Response.Player, FrameColorIndex)

            spawn . ClosePopup(Response.Player)
            LockerCamera.RemoveFrom(Response.Player)



    ClosePopup(Player:player)<suspends>:void=
        if:
            PlayerUI := GetPlayerUI[Player]
            TheGlobals := Globals[GetSession()]
            CPlayer := TheGlobals.CustomPlayerMap[Player]
        then:
            for(I := 0..3):
                if(ExistingCanvas := CPlayer.GuessUI.MaybePaintPaletteUI?):
                    PlayerUI.RemoveWidget(ExistingCanvas)
                    set CPlayer.GuessUI.MaybePaintPaletteUI = false
                Sleep(0.33)
                

    ChangeBoardColor(Agent:agent, Direction:int):void=
        Print("Changing board color")
        if:
            AtLeastOneBoardColor := BoardColorss[0]
            NextColorIndex := Mod[(BoardColorIndex + Direction + TotalBoardColors), TotalBoardColors]
            GameBoard := MaybeGameBoard?
        then:
            set BoardColorIndex = NextColorIndex
            Print("Changing board color to index: {NextColorIndex}")

            # for(Prop : GameBoard.BoardProps):
            #     Prop.SetMaterial(GetMaterialFromIndexBoard(NextColorIndex))

            # GameBoard.BoardProp.SetMaterial(GetMaterialFromIndexBoard(NextColorIndex), ?Index:=0)
            GameBoard.BoardProp.SetMaterial(GetMaterialFromIndexBoard(NextColorIndex), ?Index:=1)
            GameBoard.BoardProp.SetMaterial(GetMaterialFromIndexBoard(NextColorIndex), ?Index:=2)
            GameBoard.BoardProp.SetMaterial(GetMaterialFromIndexBoard(NextColorIndex), ?Index:=3)

            GameBoard.YouStandProp.SetMaterial(GetMaterialFromIndexStand(NextColorIndex), ?Index:=1)

            # GameBoard.PlatformProp.SetMaterial(GetMaterialFromIndexPlatform(NextColorIndex), ?Index:=0)
            # GameBoard.PlatformProp.SetMaterial(GetMaterialFromIndexPlatform(NextColorIndex), ?Index:=1)
            GameBoard.PlatformProp.SetMaterial(GetMaterialFromIndexPlatform(NextColorIndex), ?Index:=2)
            
            if:
                NextColor := BoardColorss[BoardColorIndex]                
            then:

                if:
                    BoardColorIndex > 11
                    CustomColorIndex := BoardColorIndex - 11
                    CustomPaint := UnlockablePaints[CustomColorIndex]
                then:
                    CustomPaintTexture := CustomPaint.GetTextures()
                    TextureBoard.SetImage(CustomPaintTexture(0))
                else:
                    TextureBoard.SetImage(_UI.UI_Textures.paint_palette_board)

                TextureBoard.SetTint(NextColor)
        else:
            Print("BoardColors array is empty.")
    
    ChangeFrameColor(Agent:agent, Direction:int):void=
        Print("Changing frame color")
        if:
            # FrameColorss.Length > 0
            AtLeastOneFrameColor := FrameColorss[0]
            NextColorIndex := Mod[(FrameColorIndex + Direction + TotalFrameColors), TotalFrameColors]
            NextColor := FrameColorss[NextColorIndex]
            GameBoard := MaybeGameBoard?
        then:
            set FrameColorIndex = NextColorIndex
            Print("Changing frame color to index: {NextColorIndex}")
            # for(Prop : GameBoard.CylinderProps):
            #     Prop.SetMaterial(GetMaterialFromIndexFrame(NextColorIndex))

            set GameBoard.SavedFrameColor = GetMaterialFromIndexFrame(NextColorIndex)

            for(Row : GameBoard.Rows):
                for(Prop : Row.Tiles):
                    Prop.Prop.SetMaterial(GetMaterialFromIndexFrame(NextColorIndex), ?Index:=2)


            GameBoard.BoardProp.SetMaterial(GetMaterialFromIndexBoard(NextColorIndex), ?Index:=0) # rods

            GameBoard.YouStandProp.SetMaterial(GetMaterialFromIndexStand(NextColorIndex), ?Index:=2) # card holder

            GameBoard.YouTile.Prop.SetMaterial(GetMaterialFromIndexFrame(NextColorIndex), ?Index:=2)

            GameBoard.IndexCard.Card.SetMaterial(GetMaterialFromIndexFrame(NextColorIndex), ?Index:=0)

            if:
                FrameColorIndex > 11
                CustomColorIndex := FrameColorIndex - 11
                CustomPaint := UnlockablePaints[CustomColorIndex]
            then:
                CustomPaintTexture := CustomPaint.GetTextures()
                TextureFrame.SetImage(CustomPaintTexture(1))
            else:
                TextureFrame.SetImage(_UI.UI_Textures.paint_palette_cards)

            TextureFrame.SetTint(NextColor)
        else:
            Print("FrameColors array is empty.")
    

    GetMaterialFromIndexBoard(Index:int):material=
        case(Index):
            0 => return Materials.BoardDesigner.Boards.MI_Board_Pink
            1 => return Materials.BoardDesigner.Boards.MI_Board_Yellow
            2 => return Materials.BoardDesigner.Boards.MI_Board_Orange
            3 => return Materials.BoardDesigner.Boards.MI_Board_Red
            4 => return Materials.BoardDesigner.Boards.MI_Board_Purple
            5 => return Materials.BoardDesigner.Boards.MI_Board_DarkBlue
            6 => return Materials.BoardDesigner.Boards.MI_Board_LightBlue
            7 => return Materials.BoardDesigner.Boards.MI_Board_Green
            8 => return Materials.BoardDesigner.Boards.MI_Board_White
            9 => return Materials.BoardDesigner.Boards.MI_Board_Grey
            10 => return Materials.BoardDesigner.Boards.MI_Board_Black
            11 => return Materials.BoardDesigner.Boards.MI_Board_Peach
            12 => return Materials.BoardDesigner.Boards.MI_Board_Bronze
            13 => return Materials.BoardDesigner.Boards.MI_Board_Silver
            14 => return Materials.BoardDesigner.Boards.MI_Board_Gold
            15 => return Materials.BoardDesigner.Boards.MI_Clay_Pink
            16 => return Materials.BoardDesigner.Boards.MI_Clay_Yellow
            17 => return Materials.BoardDesigner.Boards.MI_Clay_Space
            18 => return Materials.BoardDesigner.Boards.MI_Clay_Orange

            _ => return Materials.BoardDesigner.Boards.MI_Board_Pink

    GetMeshFromIndexBoard(Index:int):mesh=
        case(Index):
            _ => return NEW_ASSETS.Meshess.SM_PlayBoard


    GetMaterialFromIndexFrame(Index:int):material=
        case(Index):
            0 => return Materials.BoardDesigner.CardBacks.MI_Cardback_Pink
            1 => return Materials.BoardDesigner.CardBacks.MI_Cardback_Yellow
            2 => return Materials.BoardDesigner.CardBacks.MI_Cardback_Orange
            3 => return Materials.BoardDesigner.CardBacks.MI_Cardback_Red
            4 => return Materials.BoardDesigner.CardBacks.MI_Cardback_Purple
            5 => return Materials.BoardDesigner.CardBacks.MI_Cardback_DarkBlue
            6 => return Materials.BoardDesigner.CardBacks.MI_Cardback_LightBlue
            7 => return Materials.BoardDesigner.CardBacks.MI_Cardback_Green
            8 => return Materials.BoardDesigner.CardBacks.MI_Cardback_White
            9 => return Materials.BoardDesigner.CardBacks.MI_Cardback_Grey
            10 => return Materials.BoardDesigner.CardBacks.MI_Cardback_Black
            11 => return Materials.BoardDesigner.CardBacks.MI_Cardback_Peach
            12 => return Materials.BoardDesigner.CardBacks.MI_Cardback_Bronze
            13 => return Materials.BoardDesigner.CardBacks.MI_Cardback_Silver
            14 => return Materials.BoardDesigner.CardBacks.MI_Cardback_Gold
            15 => return Materials.BoardDesigner.CardBacks.MI_CardBack_Clay_Pink
            16 => return Materials.BoardDesigner.CardBacks.MI_CardBack_Clay_Yellow
            17 => return Materials.BoardDesigner.CardBacks.MI_CardBack_Clay_Space
            18 => return Materials.BoardDesigner.CardBacks.MI_CardBack_Clay_Orange

            _ => return Materials.BoardDesigner.CardBacks.MI_Cardback_Yellow


    GetMaterialFromIndexStand(Index:int):material=

        case(Index):
            0 => return Materials.BoardDesigner.YouStands.MI_YouStand_Pink
            1 => return Materials.BoardDesigner.YouStands.MI_YouStand_Yellow
            2 => return Materials.BoardDesigner.YouStands.MI_YouStand_Orange
            3 => return Materials.BoardDesigner.YouStands.MI_YouStand_Red
            4 => return Materials.BoardDesigner.YouStands.MI_YouStand_Purple
            5 => return Materials.BoardDesigner.YouStands.MI_YouStand_DarkBlue
            6 => return Materials.BoardDesigner.YouStands.MI_YouStand_LightBlue
            7 => return Materials.BoardDesigner.YouStands.MI_YouStand_Green
            8 => return Materials.BoardDesigner.YouStands.MI_YouStand_White
            9 => return Materials.BoardDesigner.YouStands.MI_YouStand_Grey
            10 => return Materials.BoardDesigner.YouStands.MI_YouStand_Black
            11 => return Materials.BoardDesigner.YouStands.MI_YouStand_Peach
            12 => return Materials.BoardDesigner.YouStands.MI_YouStand_Bronze
            13 => return Materials.BoardDesigner.YouStands.MI_YouStand_Silver
            14 => return Materials.BoardDesigner.YouStands.MI_YouStand_Gold
            15 => return Materials.BoardDesigner.YouStands.MI_YouStand_Clay_Pink
            16 => return Materials.BoardDesigner.YouStands.MI_YouStand_Clay_Yellow
            17 => return Materials.BoardDesigner.YouStands.MI_YouStand_Clay_Space
            18 => return Materials.BoardDesigner.YouStands.MI_YouStand_Clay_Orange

            _ => return Materials.BoardDesigner.YouStands.MI_YouStand_Pink

    GetMeshFromIndexStand(Index:int):mesh=
        case(Index):
            _ => return NEW_ASSETS.Meshess.SM_YouStand

    GetMaterialFromIndexPlatform(Index:int):material=

        case(Index):
            0 => return Materials.BoardDesigner.Platforms.MI_Platform_Pink
            1 => return Materials.BoardDesigner.Platforms.MI_Platform_Yellow
            2 => return Materials.BoardDesigner.Platforms.MI_Platform_Orange
            3 => return Materials.BoardDesigner.Platforms.MI_Platform_Red
            4 => return Materials.BoardDesigner.Platforms.MI_Platform_Purple
            5 => return Materials.BoardDesigner.Platforms.MI_Platform_DarkBlue
            6 => return Materials.BoardDesigner.Platforms.MI_Platform_LightBlue
            7 => return Materials.BoardDesigner.Platforms.MI_Platform_Green
            8 => return Materials.BoardDesigner.Platforms.MI_Platform_White
            9 => return Materials.BoardDesigner.Platforms.MI_Platform_Grey
            10 => return Materials.BoardDesigner.Platforms.MI_Platform_Black
            11 => return Materials.BoardDesigner.Platforms.MI_Platform_Peach
            12 => return Materials.BoardDesigner.Platforms.MI_Platform_Bronze
            13 => return Materials.BoardDesigner.Platforms.MI_Platform_Silver
            14 => return Materials.BoardDesigner.Platforms.MI_Platform_Gold
            15 => return Materials.BoardDesigner.Platforms.MI_Platform_Clay_Pink
            16 => return Materials.BoardDesigner.Platforms.MI_Platform_Clay_Yellow
            17 => return Materials.BoardDesigner.Platforms.MI_Platform_Clay_Space
            18 => return Materials.BoardDesigner.Platforms.MI_Platform_Clay_Orange

            _ => return Materials.BoardDesigner.Platforms.MI_Platform_Pink

    GetMeshFromIndexPlatform(Index:int):mesh=
        case(Index):
            _ => return NEW_ASSETS.Meshess.SM_PlayerPlatform





    CreateLockerCanvas(Agent:agent):canvas=

        var LayoutCanvas : canvas = canvas{}

        PageMargin := margin{Top := 20.0, Left := 20.0, Right := 20.0, Bottom := 20.0}
        ButtonPadding := 16.0

        set LayoutCanvas = canvas:
            Slots := array:
                canvas_slot:
                    Anchors := anchors{Minimum := vector2{X := 0.0, Y := 1.0}, Maximum := vector2{X := 0.0, Y := 1.0} }
                        Offsets := margin{Top := 0.0, Left := 0.0, Right := 0.0, Bottom := 0.0}
                        Alignment := vector2{X := 0.0, Y := 0.0}
                        ZOrder := 9999
                        Widget := overlay:
                            Slots := array:
                                # Background, but we don't need to see it in final version
                                overlay_slot:
                                    HorizontalAlignment := horizontal_alignment.Fill
                                    VerticalAlignment := vertical_alignment.Fill
                                    Padding := PageMargin
                                    Widget := color_block{DefaultColor:=Blue, DefaultOpacity:=0.0}

                                # Gold Count
                                overlay_slot:
                                    HorizontalAlignment := horizontal_alignment.Left
                                    VerticalAlignment := vertical_alignment.Top
                                    Padding := margin{Top:=0.0, Left:=0.0, Right:=0.0, Bottom:=0.0}
                                    Widget := GoldCountOverlay

                                # Turn Timer
                                overlay_slot:
                                    HorizontalAlignment := horizontal_alignment.Left
                                    VerticalAlignment := vertical_alignment.Bottom
                                    Padding := margin{Top:=0.0, Left:=0.0, Right:=0.0, Bottom:=0.0}
                                    Widget := TurnTimerOverlay

                                # Locker UI
                                overlay_slot:
                                    HorizontalAlignment := horizontal_alignment.Right
                                    VerticalAlignment := vertical_alignment.Fill
                                    Padding := margin{Top:=0.0, Left:=0.0, Right:=0.0, Bottom:=0.0}
                                    Widget := stack_box:
                                        Orientation := orientation.Vertical
                                        Slots := array:

                                            # Title
                                            stack_box_slot:
                                                HorizontalAlignment := horizontal_alignment.Left
                                                VerticalAlignment := vertical_alignment.Center
                                                Padding := margin{Top:=0.0, Left:=0.0, Right:=0.0, Bottom:=0.0}
                                                Widget := TextBlockTitle

                                            # Page Buttons
                                            stack_box_slot:
                                                HorizontalAlignment := horizontal_alignment.Fill
                                                VerticalAlignment := vertical_alignment.Fill
                                                Padding := margin{Top:=ButtonPadding, Left:=0.0, Right:=0.0, Bottom:=ButtonPadding}
                                                Widget := stack_box:
                                                    Orientation := orientation.Horizontal
                                                    Slots := array:

                                                        #Board Page
                                                        stack_box_slot:
                                                            HorizontalAlignment := horizontal_alignment.Left
                                                            VerticalAlignment := vertical_alignment.Fill
                                                            Padding := margin{Top:=0.0, Left:=0.0, Right:=ButtonPadding, Bottom:=0.0}
                                                            Widget := ButtonPageSelect0

                                                        #CardBack Page
                                                        stack_box_slot:
                                                            HorizontalAlignment := horizontal_alignment.Left
                                                            VerticalAlignment := vertical_alignment.Fill
                                                            Padding := margin{Top:=0.0, Left:=0.0, Right:=ButtonPadding, Bottom:=0.0}
                                                            Widget := ButtonPageSelect1

                                                        #YouStand Page
                                                        stack_box_slot:
                                                            HorizontalAlignment := horizontal_alignment.Left
                                                            VerticalAlignment := vertical_alignment.Fill
                                                            Padding := margin{Top:=0.0, Left:=0.0, Right:=ButtonPadding, Bottom:=0.0}
                                                            Widget := ButtonPageSelect2

                                                        #Platform Page
                                                        stack_box_slot:
                                                            HorizontalAlignment := horizontal_alignment.Left
                                                            VerticalAlignment := vertical_alignment.Fill
                                                            Padding := margin{Top:=0.0, Left:=0.0, Right:=ButtonPadding, Bottom:=0.0}
                                                            Widget := ButtonPageSelect3

                                            # Future Search
                                            stack_box_slot:
                                                HorizontalAlignment := horizontal_alignment.Fill
                                                VerticalAlignment := vertical_alignment.Fill
                                                Padding := margin{Top:=ButtonPadding, Left:=0.0, Right:=0.0, Bottom:=ButtonPadding}
                                                Widget := stack_box:
                                                    Orientation := orientation.Horizontal
                                                    Slots := array:
                                                        # Search Field
                                                        stack_box_slot:
                                                            Distribution := option{2.0}
                                                            HorizontalAlignment := horizontal_alignment.Fill
                                                            VerticalAlignment := vertical_alignment.Fill
                                                            Padding := margin{Top:=0.0, Left:=0.0, Right:=ButtonPadding, Bottom:=0.0}
                                                            Widget := TextFieldSearch
                                                        # Search Clear
                                                        stack_box_slot:
                                                            HorizontalAlignment := horizontal_alignment.Right
                                                            VerticalAlignment := vertical_alignment.Fill
                                                            Padding := margin{Top:=0.0, Left:=0.0, Right:=ButtonPadding, Bottom:=0.0}
                                                            Widget := ButtonSearch

                                            # Paints List
                                            stack_box_slot:
                                                Distribution := option{2.0}
                                                HorizontalAlignment := horizontal_alignment.Fill
                                                VerticalAlignment := vertical_alignment.Fill
                                                Padding := margin{Top:=ButtonPadding, Left:=0.0, Right:=0.0, Bottom:=ButtonPadding}
                                                Widget := DisplayPaintOptions(Agent, CurrentPage, CurrentInnerPage)

                                            # Exit Button
                                            stack_box_slot:
                                                HorizontalAlignment := horizontal_alignment.Right
                                                VerticalAlignment := vertical_alignment.Center
                                                Padding := margin{Top:=ButtonPadding, Left:=0.0, Right:=0.0, Bottom:=0.0}
                                                Widget := ButtonExit




        return LayoutCanvas


    DisplayPaintOptions(Agent:agent, Page:int, InnerPage:int, ?SearchText:string=""):canvas=

        var PaintCanvas : canvas = canvas{}

        if:
            CheckForPlayerSavedData[Agent] # auto creates new player data if none exists
            SourceData := PlayerSavedDataMap[Agent]
            UnlockedBoardPaints := SourceData.UnlockedBoardPaints
            # UnlockedCardBackPaints := SourceData.UnlockedCardBackPaints
            # UnlockedYouStandPaints := SourceData.UnlockedYouStandPaints
            # UnlockedPlatformPaints := SourceData.UnlockedPlatformPaints
            CurrentBoardPaint := SourceData.CurrentBoardPaint
            CurrentCardBackPaint := SourceData.CurrentCardBackPaint
            CurrentYouStandPaint := SourceData.CurrentYouStandPaint
            CurrentPlatformPaint := SourceData.CurrentPlatformPaint
            # CustomPaintSets := SourceData.CustomPaintSets
        then:




            set PaintCanvas = canvas:
                Slots := array:
                    canvas_slot:
                        Anchors := anchors{Minimum := vector2{X := 0.0, Y := 0.0}, Maximum := vector2{X := 1.0, Y := 1.0} }
                        Offsets := margin{Top := 0.0, Left := 0.0, Right := 0.0, Bottom := 0.0}
                        Alignment := vector2{X := 0.0, Y := 0.0}
                        ZOrder := 9999
                        Widget := overlay:
                            Slots := array:
                                overlay_slot:
                                    HorizontalAlignment := horizontal_alignment.Fill
                                    VerticalAlignment := vertical_alignment.Fill
                                    Padding := margin{Top:=0.0, Left:=0.0, Right:=0.0, Bottom:=0.0}
                                    Widget := color_block{DefaultColor:=Blue, DefaultOpacity:=1.0}

        return PaintCanvas