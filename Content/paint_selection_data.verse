using { /Verse.org/Colors }
using { /Verse.org/Assets }

paint_selection_data := class():
    PaintIndex : int
    PaintName : string
    PaintStyleIndex : int
    Cost : int
    PaintBGColor : color
    
    PaintBGTexture : texture
    PaintStyleTexture : texture
    FilterTags : []string
    IsNew : logic
    IsLocked : logic
    IsOwned : logic
    HasUsed : logic
    IsActive : logic
    ?QuestID : int = -1
    ?SecondaryBGColor : ?color = false
