using { /Fortnite.com/UI }
using { /UnrealEngine.com/Temporary/UI }
using { /UnrealEngine.com/Temporary/SpatialMath }
using { /Verse.org/Assets }
using { /Verse.org/Colors/NamedColors }
using { /Verse.org/Colors }
using { /Verse.org/Simulation }

paint_ui_option := class:

    var MaybePaintStyle : ?paint_selection_data = false
    var PaintIndex : int = 0
    var PaintName : string = "[Paint Name]"
    var PaintStyleIndex : int = 0
    var Cost : int = 0
    var QuestID : int = -1 

    var Widget : overlay = overlay{}
    ButtonContainer : overlay = overlay{}
    Button : button_quiet = button_quiet{DefaultText := MessageLocked_EN}

    # IconContainer : overlay = overlay{}
    # ButtonUnlockInfo : button_quiet = button_quiet{DefaultText := MessageBlank_EN}

    var PaintBGColor : color = MakeColorFromHSV(0.0, 0.0, 1.0)
    var SecondaryBGColor : ?color = false
    var PaintBGTexture : texture = _UI.UI_Textures.board_locker_paint_option_bg
    var PaintStyleTexture : texture = _UI.UI_Textures.board_locker_paint_style_plastic
    var NewTexture : texture = _UI.UI_Textures.Notification_One
    var ActiveTexture : texture = _UI.UI_Textures.board_locker_paint_active_overlay
    var TrackedQuestTexture : texture = _UI.UI_Textures.ui_rounded_question_icon

    var FilterTags : []string = array{}

    var Active : logic = false
    var Owned : logic = false
    var Unlocked : logic = false
    var BeenUsed : logic = false

    PaintBGTextureBlock : texture_block = texture_block:
        DefaultImage := _UI.UI_Textures.board_locker_paint_option_bg
        DefaultDesiredSize := vector2{X:=200.0, Y:=90.0}

    PaintStyleTextureBlock : texture_block = texture_block:
        DefaultImage := _UI.UI_Textures.transparent_square
        DefaultDesiredSize := vector2{X:=200.0, Y:=90.0}

    NewTextureBlock : texture_block = texture_block:
        DefaultImage := _UI.UI_Textures.transparent_square
        DefaultDesiredSize := vector2{X:=32.0, Y:=32.0}

    FreeTextureBlock : texture_block = texture_block:
        DefaultImage := _UI.UI_Textures.transparent_square
        DefaultDesiredSize := vector2{X:=32.0, Y:=32.0}

    TrackedQuestTextureBlock : texture_block = texture_block:
        DefaultImage := _UI.UI_Textures.transparent_square
        DefaultDesiredSize := vector2{X:=32.0, Y:=32.0}

    ActiveTextureBlock : texture_block = texture_block:
        DefaultImage := _UI.UI_Textures.transparent_square
        DefaultDesiredSize := vector2{X:=210.0, Y:=100.0}

    # New variables for the preview button and texture
    # PreviewButton : button_quiet = button_quiet{DefaultText := MessageBlank_EN}

    Create(Agent:agent, PaintStyle:paint_selection_data, ActiveQuestID:int):void=

        set MaybePaintStyle = option{PaintStyle}

        set PaintIndex = PaintStyle.PaintIndex
        set PaintBGColor = PaintStyle.PaintBGColor
        set PaintBGTexture = PaintStyle.PaintBGTexture
        set PaintStyleTexture = PaintStyle.PaintStyleTexture
        set FilterTags = PaintStyle.FilterTags
        set PaintName = PaintStyle.PaintName
        set Cost = PaintStyle.Cost
        set QuestID = PaintStyle.QuestID
        set PaintStyleIndex = PaintStyle.PaintStyleIndex

        if(SecondColor := PaintStyle.SecondaryBGColor?):
            set SecondaryBGColor = option{SecondColor}
    

        PaintBGTextureBlock.SetImage(PaintBGTexture)
        PaintStyleTextureBlock.SetImage(PaintStyleTexture)

        PaintBGTextureBlock.SetTint(PaintBGColor)
        PaintStyleTextureBlock.SetTint(PaintBGColor + MakeColorFromHSV(0.0, 0.0, 0.5))

        if(SecondaryColor := PaintStyle.SecondaryBGColor?):
            PaintBGTextureBlock.SetTint(PaintBGColor)
            PaintStyleTextureBlock.SetTint(SecondaryColor)

        # Set textures to transparent by default
        NewTextureBlock.SetImage(_UI.UI_Textures.transparent_square)
        TrackedQuestTextureBlock.SetImage(_UI.UI_Textures.transparent_square)
        ActiveTextureBlock.SetImage(_UI.UI_Textures.transparent_square)

        # Determine if the paint is unlocked
        if(PaintStyle.IsLocked?):
            set Unlocked = false
        else:
            set Unlocked = true

        # Initialize logic variables
        set Owned = if(PaintStyle.IsOwned?) then true else false
        set Active = if(PaintStyle.IsActive?) then true else false

        # Proceed with the logic
        if (Active?):
            # Active Paint
            # Button.SetEnabled(false)
            Button.SetText(MessageActive_EN)
            set Active = true
            ActiveTextureBlock.SetImage(ActiveTexture)
            # NewTextureBlock remains transparent
            # TrackedQuestTextureBlock remains transparent

        else if (Owned?):
            # Owned Paint (Not Active)
            # Button.SetEnabled(true)
            Button.SetText(MessageSelect_EN)
            set Active = false
            # ActiveTextureBlock remains transparent
            # TrackedQuestTextureBlock remains transparent

            if (PaintStyle.IsNew?):
                NewTextureBlock.SetImage(NewTexture)
            # else NewTextureBlock remains transparent

        else if (Unlocked?):
            # Unlocked Paint (Not Owned)
            if(PaintStyle.Cost > 0):
                # Button.SetEnabled(true)
                Button.SetText(MakePaintCostMessage(MessageBuy_EN, PaintStyle.Cost))
                set Active = false
                
                # ButtonUnlockContainer.AddWidget(overlay_slot{Widget := PreviewButton})

            else:
                # Free
                # Button.SetEnabled(true)
                Button.SetText(MessageFree_EN)
                # NewTextureBlock.SetImage(_UI.UI_Textures.ui_free_icon)
                FreeTextureBlock.SetImage(_UI.UI_Textures.ui_free_icon)
                set Active = false

           

        else:
            # Locked Paint
            # Button.SetEnabled(false)
            Button.SetText(MessageLocked_EN)
            set Active = false
            if(ActiveQuestID = QuestID):
                TrackedQuestTextureBlock.SetImage(TrackedQuestTexture)

            # ButtonUnlockContainer.AddWidget(overlay_slot{Widget := ButtonUnlockInfo})

        # Build the UI Widget
        set Widget = overlay:
            Slots := array:
                overlay_slot:
                    HorizontalAlignment := horizontal_alignment.Fill
                    VerticalAlignment := vertical_alignment.Fill
                    Padding := margin{Top := 0.0, Left := 0.0, Right := 0.0, Bottom := 10.0}
                    Widget := stack_box:
                        Orientation := orientation.Vertical
                        Slots := array:
                            stack_box_slot:
                                HorizontalAlignment := horizontal_alignment.Fill
                                VerticalAlignment := vertical_alignment.Fill
                                Padding := margin{Top := 0.0, Left := 8.0, Right := 8.0, Bottom := 0.0}
                                Widget := overlay:
                                    Slots := array:
                                        overlay_slot:
                                            HorizontalAlignment := horizontal_alignment.Center
                                            VerticalAlignment := vertical_alignment.Center
                                            Padding := margin{Top := 0.0, Left := 0.0, Right := 0.0, Bottom := 0.0}
                                            Widget := PaintBGTextureBlock
                                        overlay_slot:
                                            HorizontalAlignment := horizontal_alignment.Center
                                            VerticalAlignment := vertical_alignment.Center
                                            Padding := margin{Top := 0.0, Left := 0.0, Right := 0.0, Bottom := 0.0}
                                            Widget := PaintStyleTextureBlock
                                        overlay_slot:
                                            HorizontalAlignment := horizontal_alignment.Fill
                                            VerticalAlignment := vertical_alignment.Fill
                                            Padding := margin{Top := 0.0, Left := 0.0, Right := 0.0, Bottom := 0.0}
                                            Widget := ActiveTextureBlock
                                        overlay_slot:
                                            HorizontalAlignment := horizontal_alignment.Right
                                            VerticalAlignment := vertical_alignment.Top
                                            Padding := margin{Top := -10.0, Left := 0.0, Right := -10.0, Bottom := 0.0}
                                            Widget := NewTextureBlock
                                        overlay_slot:
                                            HorizontalAlignment := horizontal_alignment.Left
                                            VerticalAlignment := vertical_alignment.Top
                                            Padding := margin{Top := 12.0, Left := 12.0, Right := 0.0, Bottom := 0.0}
                                            Widget := TrackedQuestTextureBlock
                                        # # Makes the button smaller using stack boxes
                                        # overlay_slot:
                                        #     HorizontalAlignment := horizontal_alignment.Left
                                        #     VerticalAlignment := vertical_alignment.Top
                                        #     Padding := margin{Top := 12.0, Left := 12.0, Right := 12.0, Bottom := 12.0}
                                        #     Widget := stack_box:
                                        #         Orientation := orientation.Horizontal
                                        #         Slots := array:                                                    
                                        #             stack_box_slot:
                                        #                 Distribution := option{1.0}
                                        #                 HorizontalAlignment := horizontal_alignment.Left
                                        #                 VerticalAlignment := vertical_alignment.Top
                                        #                 Padding := margin{Top := 0.0, Left := 0.0, Right := 0.0, Bottom := 0.0}
                                        #                 Widget := stack_box:
                                        #                     Orientation := orientation.Vertical
                                        #                     Slots := array:
                                        #                         stack_box_slot:
                                        #                             Distribution := option{1.0}
                                        #                             HorizontalAlignment := horizontal_alignment.Left
                                        #                             VerticalAlignment := vertical_alignment.Top
                                        #                             Padding := margin{Top := 0.0, Left := 0.0, Right := 0.0, Bottom := 0.0}
                                        #                             Widget := ButtonUnlockContainer
                                        #                         stack_box_slot:
                                        #                             Distribution := option{1.0}
                                        #                             HorizontalAlignment := horizontal_alignment.Fill
                                        #                             VerticalAlignment := vertical_alignment.Fill
                                        #                             Padding := margin{Top := 0.0, Left := 0.0, Right := 0.0, Bottom := 0.0}
                                        #                             Widget := color_block:
                                        #                                 DefaultColor := Red
                                        #                                 DefaultDesiredSize := vector2{X:=8.0, Y:=8.0}
                                        #                                 DefaultOpacity := 0.0
                                        #             stack_box_slot:
                                        #                 Distribution := option{1.0}
                                        #                 HorizontalAlignment := horizontal_alignment.Fill
                                        #                 VerticalAlignment := vertical_alignment.Fill
                                        #                 Padding := margin{Top := 0.0, Left := 0.0, Right := 0.0, Bottom := 0.0}
                                        #                 Widget := color_block:
                                        #                     DefaultColor := Red
                                        #                     DefaultDesiredSize := vector2{X:=16.0, Y:=16.0}
                                        #                     DefaultOpacity := 0.0

                            stack_box_slot:
                                HorizontalAlignment := horizontal_alignment.Fill
                                VerticalAlignment := vertical_alignment.Bottom
                                Padding := margin{Top := 6.0, Left := 10.0, Right := 10.0, Bottom := 0.0}
                                Widget := overlay:
                                    Slots := array:
                                        overlay_slot:
                                            HorizontalAlignment := horizontal_alignment.Fill
                                            VerticalAlignment := vertical_alignment.Fill
                                            Padding := margin{Top := 0.0, Left := 0.0, Right := 0.0, Bottom := 0.0}
                                            Widget := Button
                                    
                                        overlay_slot:
                                            HorizontalAlignment := horizontal_alignment.Right
                                            VerticalAlignment := vertical_alignment.Top
                                            Padding := margin{Top := -10.0, Left := 0.0, Right := -10.0, Bottom := 0.0}
                                            Widget := FreeTextureBlock

    Buy():void=
        # Button.SetEnabled(true)
        Button.SetText(MessageSelect_EN)
        set Owned = true
        set Unlocked = true

    Unlock():void=
        if (Cost > 0):
            # Button.SetEnabled(true)
            Button.SetText(MessageBuy_EN)
            set Owned = true
            set Unlocked = false
        else:
            # Button.SetEnabled(true)
            Button.SetText(MessageSelect_EN)
            set Owned = true
            set Unlocked = true

    Lock():void=
        # Button.SetEnabled(false)
        Button.SetText(MessageLocked_EN)
        set Owned = false
        set Unlocked = false

    SetNew():void=
        NewTextureBlock.SetImage(NewTexture)

    ClearNew():void=
        NewTextureBlock.SetImage(_UI.UI_Textures.transparent_square)

    SetActive(aLogic:logic):void=
        set Active = aLogic

        if(Active?):
            # Button.SetEnabled(false)
            Button.SetText(MessageActive_EN)
            ActiveTextureBlock.SetImage(ActiveTexture)

            if(not BeenUsed?):
                NewTextureBlock.SetImage(_UI.UI_Textures.transparent_square)
                set BeenUsed = true

        else:
            # Button.SetEnabled(true)
            Button.SetText(MessageSelect_EN)
            ActiveTextureBlock.SetImage(_UI.UI_Textures.transparent_square)
    
    SetOwned(aLogic:logic):void=
        set Owned = aLogic
        # update the button text
        if (Owned?):
            Button.SetText(MessageSelect_EN)
            FreeTextureBlock.SetImage(_UI.UI_Textures.transparent_square)
            TrackedQuestTextureBlock.SetImage(_UI.UI_Textures.transparent_square)            
            NewTextureBlock.SetImage(NewTexture)
        else:
            if (Cost > 0):
                Button.SetText(MessageBuy_EN)
            else:
                Button.SetText(MessageFree_EN)
                FreeTextureBlock.SetImage(_UI.UI_Textures.ui_free_icon)

    SetTrackedQuest(aLogic:logic):void=
        if (aLogic?):
            TrackedQuestTextureBlock.SetImage(TrackedQuestTexture)
        else:
            TrackedQuestTextureBlock.SetImage(_UI.UI_Textures.transparent_square)

