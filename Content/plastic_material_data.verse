using { /Verse.org/Colors }
using { /Verse.org/Colors/NamedColors }
using { /Verse.org/Assets }




plastic_material_data := class():
    BaseColor:texture=T_CardBack_C 
    BaseHueShift:float=0.0 
    ColorDesaturation:float=0.0 
    ColorTint:color=White 
    SecondaryColorMask:texture=T_PlayBoard_SecondaryColor_M 
    SecondaryColorTint:color=White 
    SecondaryHueShift:float=0.752 
    EmissiveFlickerSpeed:float=1.0 
    FrameEmissiveStrength:float=1.0 
    GlobalEmissiveStrength:float=50.0 
    EmissiveMask2:texture=T_Lights_EM 
    EmissiveFlickerTint:color=Pink 
    EmissiveTint:color=White 
    FrameEmissiveTint:color=White 
    BaseMaskedTexture:texture=T_PlayBoard_Base_ORM 
    MetallicAmount:float=0.0 
    RoughnessMultiplier:float=1.0 
    BaseNormal:texture=T_PlayBoard_N