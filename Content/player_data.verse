using { /Verse.org/Simulation }
using { /UnrealEngine.com/Temporary/SpatialMath }

player_role := enum:
    Player
    Admin
    Moderator
    Developer
    BetaTester
    None


(PR:player_role).ToString():string=
    case(PR):
        player_role.Player => "Player"
        player_role.Admin => "Admin"
        player_role.Moderator => "Moderator"
        player_role.Developer => "Developer"
        player_role.BetaTester => "BetaTester"
        player_role.None => "None"

AssignRole(Agent:agent, Role:player_role, Enable:logic):void=
    case(Role):
        player_role.Player => block:
            AssignAdmin(Agent, false)
            AssignModerator(Agent, false)
            AssignDeveloper(Agent, false)
            AssignBetaTester(Agent, false)
        player_role.Admin => AssignAdmin(Agent, Enable)
        player_role.Moderator => AssignModerator(Agent, Enable)
        player_role.Developer => AssignDeveloper(Agent, Enable)
        player_role.BetaTester => AssignBetaTester(Agent, Enable)
        player_role.None => Print("Do nothing")


            

player_saved_data := class<final><persistable>:
    DataVersion:int = 3

    # Admin Things
    IsAdmin : logic
    IsModerator : logic
    IsDeveloper : logic
    IsBetaTester : logic

    # Moderation Things
    IsBanned : logic
    BanReason : ?string
    BanExpiration : ?string
    BanIssuedBy : ?string
    BansIssued : int
    WarningsIssued : int
    PastWarnings : []string

    # Game Stats
    XP:int
    Rank:int
    Score:int #currently wins
    Level:int

    # Currency
    Gold:int
    OverrideThemeTokens:int

    # Game Stats
    ConvertedScoreToCorrectGuesses:logic
    CorrectGuesses:int
    IncorrectGuesses:int
    CorrectYoloGuesses:int # Correct guesses made with only one card flipped
    ConsecutiveCorrectGuesses:int
    ConsecutiveIncorrectGuesses:int
    ConsecutiveCorrectGuessesRecord:int
    ConsecutiveIncorrectGuessesRecord:int #using to store First Daily Reward

    GamesPlayed:int
    GamesPlayedWithNewFortWhoPlayers:int
    ConsecutiveDaysPlayed:int
    ConsecutiveDaysPlayedRecord:int
    LastPlayedDate:string
    FirstPlayedDate:string
    LastDayWon:string
    
    WinStreak:int
    WinStreakRecord:int
    LastStyleChoice:int
    PrevousStyleChoices:[]int

    # Board Paints
    ConvertedUnlockedPaints:logic
    UnlockedBoardPaints:[]int
    UnlockedCardBackPaints:[]int
    UnlockedYouStandPaints:[]int
    UnlockedPlatformPaints:[]int

    CurrentBoardPaint:board_paint_data
    CurrentCardBackPaint:card_back_paint_data
    CurrentYouStandPaint:you_stand_paint_data
    CurrentPlatformPaint:platform_paint_data

    CustomPaintSets:[]custom_paint_set

    UnlockedBoardMeshes:[]int
    CurrentBoardMesh:int

    UnlockedCardBackMeshes:[]int
    CurrentCardBackMesh:int

    UnlockedYouStandMeshes:[]int
    CurrentYouStandMesh:int

    UnlockedPlatformMeshes:[]int
    CurrentPlatformMesh:int



    # Cosmetics

    UnlockedHats:[]int
    CurrentHat:int

    UnlockedPoses:[]int
    CurrentPose:int

    UnlockedPointers:[]int
    CurrentPointer:int

    UnlockedTrails:[]int
    CurrentTrail:int

    UnlockedTitles:[]int
    CurrentTitle:int

    UnlockedBackgrounds:[]int
    CurrentBackground:int

    # Audio

    UnlockedMusic:[]int
    CurrentMusic:int

    UnlockedCardKnockSFX:[]int
    CurrentCardKnockSFX:int

    UnlockedUndoSFX:[]int
    CurrentUndoSFX:int

    UnlockedCorrectSFX:[]int
    CurrentCorrectSFX:int

    UnlockedIncorrectSFX:[]int
    CurrentIncorrectSFX:int

    # Quests

    CompletedQuests:[]int
    CurrentQuests:[]int
    CurrentPinnedQuests:[]int


    #added in v2

    FirstDayPlayed : time_and_date = time_and_date{}
    LastDayPlayed : time_and_date = time_and_date{}
    TotalDaysPlayed : int = 0
    TotalMinutesPlayed : int = 0
    NewPlayersPlayedWith : int = 0
    LastGuessCorrect : logic = false

    FriendStreaks : []friend_streak = array{}
    CardsKnocked : int = 0
    PreviousStyleChoices : []int = array{}


    #added in v3
    CardHighlightColor : tuple(float,float,float) = (0.0,0.302997,1.0)
    LastSessionTime : time_and_date = time_and_date{}

    StreamerModeHideCharacter : logic = false
    StreamerModeVerticalUI : logic = false

    ManuallyChooseCharacter : logic = false

    #added in v4
    CompletedWinter2024Quest : logic = false

    #added in v5
    TotalFlyingTargetsHit : int = 0

    #added in v6
    BeenAssignedPointer : logic = false

    #added in v7
    UnlockedStickers : []int = array{}
    CurrentStickers : []sticker_slot = array{} # tuple(StickerID, Slot Transform)

    

sticker_slot := struct<persistable>:
    StickerID:int
    Translation:vector3
    Rotation:float #yaw

DebugPrintPlayerSavedData(Agent:agent):void=
    if:
        CheckForPlayerSavedData[Agent]
        Player := player[Agent]
        SourceData := PlayerSavedDataMap[Player]
    then:
        Print("DataVersion: {SourceData.DataVersion}\n IsAdmin: {SourceData.IsAdmin.ToString()}\n IsModerator: {SourceData.IsModerator.ToString()}\n IsDeveloper: {SourceData.IsDeveloper.ToString()}\n IsBanned: {SourceData.IsBanned.ToString()}\n BanReason: [Cannot Print]\n BanExpiration: [Cannot Print]\n BanIssuedBy: [Cannot Print]\n BansIssued: {SourceData.BansIssued}\n WarningsIssued: {SourceData.WarningsIssued}\n PastWarnings: {SourceData.PastWarnings.ArrPrint()}\n XP: {SourceData.XP}\n Rank: {SourceData.Rank}\n Score: {SourceData.Score}\n Level: {SourceData.Level}\n Gold: {SourceData.Gold}\n OverrideThemeTokens: {SourceData.OverrideThemeTokens}\n CorrectGuesses: {SourceData.CorrectGuesses}\n IncorrectGuesses: {SourceData.IncorrectGuesses}\n CorrectYoloGuesses: {SourceData.CorrectYoloGuesses}\n ConsecutiveCorrectGuesses: {SourceData.ConsecutiveCorrectGuesses}\n ConsecutiveIncorrectGuesses: {SourceData.ConsecutiveIncorrectGuesses}\n ConsecutiveCorrectGuessesRecord: {SourceData.ConsecutiveCorrectGuessesRecord}\n ConsecutiveIncorrectGuessesRecord: {SourceData.ConsecutiveIncorrectGuessesRecord}\n GamesPlayed: {SourceData.GamesPlayed}\n GamesPlayedWithNewFortWhoPlayers: {SourceData.GamesPlayedWithNewFortWhoPlayers}\n ConsecutiveDaysPlayed: {SourceData.ConsecutiveDaysPlayed}\n ConsecutiveDaysPlayedRecord: {SourceData.ConsecutiveDaysPlayedRecord}\n LastPlayedDate: {SourceData.LastPlayedDate}\n FirstPlayedDate: {SourceData.FirstPlayedDate}\n LastDayWon: {SourceData.LastDayWon}\n WinStreak: {SourceData.WinStreak}\n WinStreakRecord: {SourceData.WinStreakRecord}\n LastStyleChoice: {SourceData.LastStyleChoice}\n PrevousStyleChoices: {SourceData.PrevousStyleChoices.ToPrint()}\n UnlockedBoardPaints: {SourceData.UnlockedBoardPaints.ToPrint()}\n UnlockedCardBackPaints: {SourceData.UnlockedCardBackPaints.ToPrint()}\n UnlockedYouStandPaints: {SourceData.UnlockedYouStandPaints.ToPrint()}\n UnlockedPlatformPaints: {SourceData.UnlockedPlatformPaints.ToPrint()}\n CurrentBoardPaint: [Cannot Print]\n CurrentCardBackPaint: [Cannot Print]\n CurrentYouStandPaint: [Cannot Print]\n CurrentPlatformPaint: [Cannot Print]\n UnlockedBoardMeshes: {SourceData.UnlockedBoardMeshes.ToPrint()}\n CurrentBoardMesh: {SourceData.CurrentBoardMesh}\n UnlockedCardBackMeshes: {SourceData.UnlockedCardBackMeshes.ToPrint()}\n CurrentCardBackMesh: {SourceData.CurrentCardBackMesh}\n UnlockedYouStandMeshes: {SourceData.UnlockedYouStandMeshes.ToPrint()}\n CurrentYouStandMesh: {SourceData.CurrentYouStandMesh}\n UnlockedPlatformMeshes: {SourceData.UnlockedPlatformMeshes.ToPrint()}\n CurrentPlatformMesh: {SourceData.CurrentPlatformMesh}\n UnlockedHats: {SourceData.UnlockedHats.ToPrint()}\n CurrentHat: {SourceData.CurrentHat}\n UnlockedPoses: {SourceData.UnlockedPoses.ToPrint()}\n CurrentPose: {SourceData.CurrentPose}\n UnlockedPointers: {SourceData.UnlockedPointers.ToPrint()}\n CurrentPointer: {SourceData.CurrentPointer}\n UnlockedTrails: {SourceData.UnlockedTrails.ToPrint()}\n CurrentTrail: {SourceData.CurrentTrail}\n UnlockedTitles: {SourceData.UnlockedTitles.ToPrint()}\n CurrentTitle: {SourceData.CurrentTitle}\n UnlockedBackgrounds: {SourceData.UnlockedBackgrounds.ToPrint()}\n CurrentBackground: {SourceData.CurrentBackground}\n UnlockedMusic: {SourceData.UnlockedMusic.ToPrint()}\n CurrentMusic: {SourceData.CurrentMusic}\n UnlockedCardKnockSFX: {SourceData.UnlockedCardKnockSFX.ToPrint()}\n CurrentCardKnockSFX: {SourceData.CurrentCardKnockSFX}\n UnlockedUndoSFX: {SourceData.UnlockedUndoSFX.ToPrint()}\n CurrentUndoSFX: {SourceData.CurrentUndoSFX}\n UnlockedCorrectSFX: {SourceData.UnlockedCorrectSFX.ToPrint()}\n CurrentCorrectSFX: {SourceData.CurrentCorrectSFX}\n UnlockedIncorrectSFX: {SourceData.UnlockedIncorrectSFX.ToPrint()}\n CurrentIncorrectSFX: {SourceData.CurrentIncorrectSFX}\n") 
    else:
        Print("No player data found for Player\n")


MakePlayerSavedData<constructor>(Src:player_saved_data)<transacts> := player_saved_data:
    DataVersion := Src.DataVersion
    IsAdmin := Src.IsAdmin 
    IsModerator := Src.IsModerator
    IsDeveloper := Src.IsDeveloper
    IsBetaTester := Src.IsBetaTester
    IsBanned := Src.IsBanned
    BanReason := Src.BanReason
    BanExpiration := Src.BanExpiration
    BanIssuedBy := Src.BanIssuedBy
    BansIssued := Src.BansIssued
    WarningsIssued := Src.WarningsIssued
    PastWarnings := Src.PastWarnings
    XP := Src.XP
    Rank := Src.Rank
    Score := Src.Score
    Level := Src.Level
    Gold := Src.Gold
    OverrideThemeTokens := Src.OverrideThemeTokens
    ConvertedScoreToCorrectGuesses := Src.ConvertedScoreToCorrectGuesses
    CorrectGuesses := Src.CorrectGuesses
    IncorrectGuesses := Src.IncorrectGuesses
    CorrectYoloGuesses := Src.CorrectYoloGuesses
    ConsecutiveCorrectGuesses := Src.ConsecutiveCorrectGuesses
    ConsecutiveIncorrectGuesses := Src.ConsecutiveIncorrectGuesses
    ConsecutiveCorrectGuessesRecord := Src.ConsecutiveCorrectGuessesRecord
    ConsecutiveIncorrectGuessesRecord := Src.ConsecutiveIncorrectGuessesRecord
    GamesPlayed := Src.GamesPlayed
    GamesPlayedWithNewFortWhoPlayers := Src.GamesPlayedWithNewFortWhoPlayers
    ConsecutiveDaysPlayed := Src.ConsecutiveDaysPlayed
    ConsecutiveDaysPlayedRecord := Src.ConsecutiveDaysPlayedRecord
    LastPlayedDate := Src.LastPlayedDate
    FirstPlayedDate := Src.FirstPlayedDate
    LastDayWon := Src.LastDayWon
    WinStreak := Src.WinStreak
    WinStreakRecord := Src.WinStreakRecord
    LastStyleChoice := Src.LastStyleChoice
    PrevousStyleChoices := Src.PrevousStyleChoices
    CustomPaintSets := Src.CustomPaintSets
    ConvertedUnlockedPaints := Src.ConvertedUnlockedPaints
    UnlockedBoardPaints := Src.UnlockedBoardPaints
    UnlockedCardBackPaints := Src.UnlockedCardBackPaints
    UnlockedYouStandPaints := Src.UnlockedYouStandPaints
    UnlockedPlatformPaints := Src.UnlockedPlatformPaints
    CurrentBoardPaint := Src.CurrentBoardPaint
    CurrentCardBackPaint := Src.CurrentCardBackPaint
    CurrentYouStandPaint := Src.CurrentYouStandPaint
    CurrentPlatformPaint := Src.CurrentPlatformPaint
    UnlockedBoardMeshes := Src.UnlockedBoardMeshes
    CurrentBoardMesh := Src.CurrentBoardMesh
    UnlockedCardBackMeshes := Src.UnlockedCardBackMeshes
    CurrentCardBackMesh := Src.CurrentCardBackMesh
    UnlockedYouStandMeshes := Src.UnlockedYouStandMeshes
    CurrentYouStandMesh := Src.CurrentYouStandMesh
    UnlockedPlatformMeshes := Src.UnlockedPlatformMeshes
    CurrentPlatformMesh := Src.CurrentPlatformMesh
    UnlockedHats := Src.UnlockedHats
    CurrentHat := Src.CurrentHat
    UnlockedPoses := Src.UnlockedPoses
    CurrentPose := Src.CurrentPose
    UnlockedPointers := Src.UnlockedPointers
    CurrentPointer := Src.CurrentPointer
    UnlockedTrails := Src.UnlockedTrails
    CurrentTrail := Src.CurrentTrail
    UnlockedTitles := Src.UnlockedTitles
    CurrentTitle := Src.CurrentTitle
    UnlockedBackgrounds := Src.UnlockedBackgrounds
    CurrentBackground := Src.CurrentBackground
    UnlockedMusic := Src.UnlockedMusic
    CurrentMusic := Src.CurrentMusic
    UnlockedCardKnockSFX := Src.UnlockedCardKnockSFX
    CurrentCardKnockSFX := Src.CurrentCardKnockSFX
    UnlockedUndoSFX := Src.UnlockedUndoSFX
    CurrentUndoSFX := Src.CurrentUndoSFX
    UnlockedCorrectSFX := Src.UnlockedCorrectSFX
    CurrentCorrectSFX := Src.CurrentCorrectSFX
    UnlockedIncorrectSFX := Src.UnlockedIncorrectSFX
    CurrentIncorrectSFX := Src.CurrentIncorrectSFX
    CompletedQuests := Src.CompletedQuests
    CurrentQuests := Src.CurrentQuests
    CurrentPinnedQuests := Src.CurrentPinnedQuests
    
    #added in v2
    FirstDayPlayed := Src.FirstDayPlayed
    LastDayPlayed := Src.LastDayPlayed
    TotalDaysPlayed := Src.TotalDaysPlayed
    TotalMinutesPlayed := Src.TotalMinutesPlayed
    NewPlayersPlayedWith := Src.NewPlayersPlayedWith
    LastGuessCorrect := Src.LastGuessCorrect
    FriendStreaks := Src.FriendStreaks
    CardsKnocked := Src.CardsKnocked
    PreviousStyleChoices := Src.PreviousStyleChoices

    #added in v3
    CardHighlightColor := Src.CardHighlightColor
    LastSessionTime := Src.LastSessionTime
    StreamerModeHideCharacter := Src.StreamerModeHideCharacter
    StreamerModeVerticalUI := Src.StreamerModeVerticalUI
    ManuallyChooseCharacter := Src.ManuallyChooseCharacter

    #added in v4
    CompletedWinter2024Quest := Src.CompletedWinter2024Quest

    #added in v5
    TotalFlyingTargetsHit := Src.TotalFlyingTargetsHit

    #added in v6
    BeenAssignedPointer := Src.BeenAssignedPointer

    #added in v7
    UnlockedStickers := Src.UnlockedStickers
    CurrentStickers := Src.CurrentStickers
    


UpdateSavedData(Agent:agent, UpdatedData:player_saved_data):void=
    if:
        CheckForPlayerSavedData[Agent]
        Player := player[Agent]
        SourceData := PlayerSavedDataMap[Player]

        set PlayerSavedDataMap[Player] = player_saved_data:
            MakePlayerSavedData<constructor>(SourceData)
            DataVersion := UpdatedData.DataVersion
            IsAdmin := UpdatedData.IsAdmin
            IsModerator := UpdatedData.IsModerator
            IsDeveloper := UpdatedData.IsDeveloper
            IsBanned := UpdatedData.IsBanned
            BanReason := UpdatedData.BanReason
            BanExpiration := UpdatedData.BanExpiration
            BanIssuedBy := UpdatedData.BanIssuedBy
            BansIssued := UpdatedData.BansIssued
            WarningsIssued := UpdatedData.WarningsIssued
            PastWarnings := UpdatedData.PastWarnings
            XP := UpdatedData.XP
            Rank := UpdatedData.Rank
            Score := UpdatedData.Score
            Level := UpdatedData.Level
            Gold := UpdatedData.Gold
            OverrideThemeTokens := UpdatedData.OverrideThemeTokens
            CorrectGuesses := UpdatedData.CorrectGuesses
            IncorrectGuesses := UpdatedData.IncorrectGuesses
            CorrectYoloGuesses := UpdatedData.CorrectYoloGuesses
            ConsecutiveCorrectGuesses := UpdatedData.ConsecutiveCorrectGuesses
            ConsecutiveIncorrectGuesses := UpdatedData.ConsecutiveIncorrectGuesses
            ConsecutiveCorrectGuessesRecord := UpdatedData.ConsecutiveCorrectGuessesRecord
            ConsecutiveIncorrectGuessesRecord := UpdatedData.ConsecutiveIncorrectGuessesRecord
            GamesPlayed := UpdatedData.GamesPlayed
            GamesPlayedWithNewFortWhoPlayers := UpdatedData.GamesPlayedWithNewFortWhoPlayers
            ConsecutiveDaysPlayed := UpdatedData.ConsecutiveDaysPlayed
            ConsecutiveDaysPlayedRecord := UpdatedData.ConsecutiveDaysPlayedRecord
            LastPlayedDate := UpdatedData.LastPlayedDate
            FirstPlayedDate := UpdatedData.FirstPlayedDate
            LastDayWon := UpdatedData.LastDayWon
            WinStreak := UpdatedData.WinStreak
            WinStreakRecord := UpdatedData.WinStreakRecord
            LastStyleChoice := UpdatedData.LastStyleChoice
            PrevousStyleChoices := UpdatedData.PrevousStyleChoices
            UnlockedBoardPaints := UpdatedData.UnlockedBoardPaints
            UnlockedCardBackPaints := UpdatedData.UnlockedCardBackPaints
            UnlockedYouStandPaints := UpdatedData.UnlockedYouStandPaints
            UnlockedPlatformPaints := UpdatedData.UnlockedPlatformPaints
            CurrentBoardPaint := UpdatedData.CurrentBoardPaint
            CurrentCardBackPaint := UpdatedData.CurrentCardBackPaint
            CurrentYouStandPaint := UpdatedData.CurrentYouStandPaint
            CurrentPlatformPaint := UpdatedData.CurrentPlatformPaint
            UnlockedBoardMeshes := UpdatedData.UnlockedBoardMeshes
            CurrentBoardMesh := UpdatedData.CurrentBoardMesh
            UnlockedCardBackMeshes := UpdatedData.UnlockedCardBackMeshes
            CurrentCardBackMesh := UpdatedData.CurrentCardBackMesh
            UnlockedYouStandMeshes := UpdatedData.UnlockedYouStandMeshes
            CurrentYouStandMesh := UpdatedData.CurrentYouStandMesh
            UnlockedPlatformMeshes := UpdatedData.UnlockedPlatformMeshes
            CurrentPlatformMesh := UpdatedData.CurrentPlatformMesh
            UnlockedHats := UpdatedData.UnlockedHats
            CurrentHat := UpdatedData.CurrentHat
            UnlockedPoses := UpdatedData.UnlockedPoses
            CurrentPose := UpdatedData.CurrentPose
            UnlockedPointers := UpdatedData.UnlockedPointers
            CurrentPointer := UpdatedData.CurrentPointer
            UnlockedTrails := UpdatedData.UnlockedTrails
            CurrentTrail := UpdatedData.CurrentTrail
            UnlockedTitles := UpdatedData.UnlockedTitles
            CurrentTitle := UpdatedData.CurrentTitle
            UnlockedBackgrounds := UpdatedData.UnlockedBackgrounds
            CurrentBackground := UpdatedData.CurrentBackground
            UnlockedMusic := UpdatedData.UnlockedMusic
            CurrentMusic := UpdatedData.CurrentMusic
            UnlockedCardKnockSFX := UpdatedData.UnlockedCardKnockSFX
            CurrentCardKnockSFX := UpdatedData.CurrentCardKnockSFX
            UnlockedUndoSFX := UpdatedData.UnlockedUndoSFX
            CurrentUndoSFX := UpdatedData.CurrentUndoSFX
            UnlockedCorrectSFX := UpdatedData.UnlockedCorrectSFX
            CurrentCorrectSFX := UpdatedData.CurrentCorrectSFX
            UnlockedIncorrectSFX := UpdatedData.UnlockedIncorrectSFX
            CurrentIncorrectSFX := UpdatedData.CurrentIncorrectSFX
            CompletedQuests := UpdatedData.CompletedQuests
            CurrentQuests := UpdatedData.CurrentQuests
            CurrentPinnedQuests := UpdatedData.CurrentPinnedQuests
            FirstDayPlayed := UpdatedData.FirstDayPlayed
            LastDayPlayed := UpdatedData.LastDayPlayed
            TotalDaysPlayed := UpdatedData.TotalDaysPlayed
            TotalMinutesPlayed := UpdatedData.TotalMinutesPlayed
            NewPlayersPlayedWith := UpdatedData.NewPlayersPlayedWith
            LastGuessCorrect := UpdatedData.LastGuessCorrect
            FriendStreaks := UpdatedData.FriendStreaks
            CardsKnocked := UpdatedData.CardsKnocked
            PreviousStyleChoices := UpdatedData.PreviousStyleChoices
            CardHighlightColor := UpdatedData.CardHighlightColor
            LastSessionTime := UpdatedData.LastSessionTime
            StreamerModeHideCharacter := UpdatedData.StreamerModeHideCharacter
            StreamerModeVerticalUI := UpdatedData.StreamerModeVerticalUI
            ManuallyChooseCharacter := UpdatedData.ManuallyChooseCharacter            
            CompletedWinter2024Quest := UpdatedData.CompletedWinter2024Quest
            TotalFlyingTargetsHit := UpdatedData.TotalFlyingTargetsHit
            BeenAssignedPointer := UpdatedData.BeenAssignedPointer
            UnlockedStickers := UpdatedData.UnlockedStickers
            CurrentStickers := UpdatedData.CurrentStickers


    else:
        Print("No player data found for Player\n")


AddKnockedCard(Agent:agent):void=
    if:
        CheckForPlayerSavedData[Agent]
        Player := player[Agent]
        SourceData := PlayerSavedDataMap[Player]

        set PlayerSavedDataMap[Player] = player_saved_data:
            MakePlayerSavedData<constructor>(SourceData)
            CardsKnocked := SourceData.CardsKnocked + 1
    else:
        Print("No player data found for Player\n")

AddPreviousStyleChoice(Agent:agent, StyleChoice:int):void=
    if:
        CheckForPlayerSavedData[Agent]
        Player := player[Agent]
        SourceData := PlayerSavedDataMap[Player]

        set PlayerSavedDataMap[Player] = player_saved_data:
            MakePlayerSavedData<constructor>(SourceData)
            PreviousStyleChoices := SourceData.PreviousStyleChoices + array{StyleChoice}
    else:
        Print("No player data found for Player\n")

AssignAdmin(Agent:agent, IsAdmin:logic):void=
    if:
        CheckForPlayerSavedData[Agent]
        Player := player[Agent]
        SourceData := PlayerSavedDataMap[Player]

        set PlayerSavedDataMap[Player] = player_saved_data:
            MakePlayerSavedData<constructor>(SourceData)
            IsAdmin := IsAdmin
    then:
        Print("Admin assigned to Player\n")
    else:
        Print("No player data found for Player\n")

# IsAtLeast(Agent:agent, Role:player_role)<transacts><decides>:void=
#     CheckForPlayerSavedData[Agent]
#     Player := player[Agent]
#     SourceData := PlayerSavedDataMap[Player]
#     IsAdmin := SourceData.IsAdmin
#     IsModerator := SourceData.IsModerator
#     IsDeveloper := SourceData.IsDeveloper
#     IsBetaTester := SourceData.IsBetaTester
#     ValidToProceed := if (Role = player_role.Admin) then IsAdmin else if (Role = player_role.Moderator) then IsModerator else if (Role = player_role.Developer) then IsDeveloper else if (Role = player_role.BetaTester) then IsBetaTester else false
    
IsAtLeast(Agent:agent, Role:player_role)<transacts><decides>:void=

    var Passes : logic = false

    CheckForPlayerSavedData[Agent]
    Player := player[Agent]
    SourceData := PlayerSavedDataMap[Player]
    IsAdmin := SourceData.IsAdmin
    IsModerator := SourceData.IsModerator
    IsDeveloper := SourceData.IsDeveloper
    IsBetaTester := SourceData.IsBetaTester

    # Determine if the player has the required role or higher
    if(Role = player_role.Admin):
        set Passes = IsAdmin
    else if(Role = player_role.Moderator):
        set Passes = if(IsAdmin? or IsModerator?) then true else false
    else if(Role = player_role.Developer):
        set Passes = if(IsAdmin? or IsModerator? or IsDeveloper?) then true else false
    else if(Role = player_role.BetaTester):
        set Passes = if(IsAdmin? or IsModerator? or IsDeveloper? or IsBetaTester?) then true else false
    else:
        set Passes = true # Default case, if no specific role is required

    Passes?


AssignModerator(Agent:agent, IsModerator:logic):void=
    if:
        CheckForPlayerSavedData[Agent]
        Player := player[Agent]
        SourceData := PlayerSavedDataMap[Player]

        set PlayerSavedDataMap[Player] = player_saved_data:
            MakePlayerSavedData<constructor>(SourceData)
            IsModerator := IsModerator
    then:
        Print("Moderator assigned to Player\n")
    else:
        Print("No player data found for Player\n")

AssignDeveloper(Agent:agent, IsDeveloper:logic):void=
    if:
        CheckForPlayerSavedData[Agent]
        Player := player[Agent]
        SourceData := PlayerSavedDataMap[Player]

        set PlayerSavedDataMap[Player] = player_saved_data:
            MakePlayerSavedData<constructor>(SourceData)
            IsDeveloper := IsDeveloper
    then:
        Print("Developer assigned to Player\n")
    else:
        Print("No player data found for Player\n")


AssignBetaTester(Agent:agent, IsBetaTester:logic):void=
    if:
        CheckForPlayerSavedData[Agent]
        Player := player[Agent]
        SourceData := PlayerSavedDataMap[Player]

        set PlayerSavedDataMap[Player] = player_saved_data:
            MakePlayerSavedData<constructor>(SourceData)
            IsBetaTester := IsBetaTester
    then:
        Print("BetaTester assigned to Player\n")
    else:
        Print("No player data found for Player\n")


AddBan(Agent:agent, BanReason:string, BanExpiration:string, BanIssuedBy:string):void=
    if:
        CheckForPlayerSavedData[Agent]
        Player := player[Agent]
        SourceData := PlayerSavedDataMap[Player]

        set PlayerSavedDataMap[Player] = player_saved_data:
            MakePlayerSavedData<constructor>(SourceData)
            IsBanned := true
            BanReason := option{BanReason}
            BanExpiration := option{BanExpiration}
            BanIssuedBy := option{BanIssuedBy}
            BansIssued := SourceData.BansIssued + 1
    else:
        Print("No player data found for Player\n")

RemoveBan(Agent:agent):void=
    if:
        CheckForPlayerSavedData[Agent]
        Player := player[Agent]
        SourceData := PlayerSavedDataMap[Player]

        set PlayerSavedDataMap[Player] = player_saved_data:
            MakePlayerSavedData<constructor>(SourceData)
            IsBanned := false
    else:
        Print("No player data found for Player\n")

AddWarning(Agent:agent, WarningReason:string):void=
    if:
        CheckForPlayerSavedData[Agent]
        Player := player[Agent]
        SourceData := PlayerSavedDataMap[Player]

        set PlayerSavedDataMap[Player] = player_saved_data:
            MakePlayerSavedData<constructor>(SourceData)
            WarningsIssued := SourceData.WarningsIssued + 1
            PastWarnings := SourceData.PastWarnings + array{WarningReason}
    else:
        Print("No player data found for Player\n")

AddXP(Agent:agent, XP:int):void=
    if:
        CheckForPlayerSavedData[Agent]
        Player := player[Agent]
        SourceData := PlayerSavedDataMap[Player]

        set PlayerSavedDataMap[Player] = player_saved_data:
            MakePlayerSavedData<constructor>(SourceData)
            XP := SourceData.XP + XP
    else:
        Print("No player data found for Player\n")


AddRank(Agent:agent, Rank:int):void=
    if:
        CheckForPlayerSavedData[Agent]
        Player := player[Agent]
        SourceData := PlayerSavedDataMap[Player]

        set PlayerSavedDataMap[Player] = player_saved_data:
            MakePlayerSavedData<constructor>(SourceData)
            Rank := SourceData.Rank + Rank
    else:
        Print("No player data found for Player\n")


AddScore(Agent:agent, Score:int):void=
    if:
        CheckForPlayerSavedData[Agent]
        Player := player[Agent]
        SourceData := PlayerSavedDataMap[Player]

        set PlayerSavedDataMap[Player] = player_saved_data:
            MakePlayerSavedData<constructor>(SourceData)
            Score := SourceData.Score + Score
    else:
        Print("No player data found for Player\n")


AddLevel(Agent:agent, Level:int):void=
    if:
        CheckForPlayerSavedData[Agent]
        Player := player[Agent]
        SourceData := PlayerSavedDataMap[Player]

        set PlayerSavedDataMap[Player] = player_saved_data:
            MakePlayerSavedData<constructor>(SourceData)
            Level := SourceData.Level + Level
    else:
        Print("No player data found for Player\n")


AddGold(Agent:agent, Gold:int)<transacts>:void=
    if:
        CheckForPlayerSavedData[Agent]
        Player := player[Agent]
        SourceData := PlayerSavedDataMap[Player]

        set PlayerSavedDataMap[Player] = player_saved_data:
            MakePlayerSavedData<constructor>(SourceData)
            Gold := SourceData.Gold + Gold
    else:
        Print("No player data found for Player\n")


DeductGold(Agent:agent, Gold:int)<transacts>:void=
    if:
        CheckForPlayerSavedData[Agent]
        Player := player[Agent]
        SourceData := PlayerSavedDataMap[Player]

        if:
            SourceData.Gold >= Gold
            set PlayerSavedDataMap[Player] = player_saved_data:
                MakePlayerSavedData<constructor>(SourceData)
                Gold := SourceData.Gold - Gold
        else:
            Print("Not enough gold to deduct\n")
    else:
        Print("No player data found for Player\n")


AddOverrideThemeTokens(Agent:agent, OverrideThemeTokens:int):void=
    if:
        CheckForPlayerSavedData[Agent]
        Player := player[Agent]
        SourceData := PlayerSavedDataMap[Player]

        set PlayerSavedDataMap[Player] = player_saved_data:
            MakePlayerSavedData<constructor>(SourceData)
            OverrideThemeTokens := SourceData.OverrideThemeTokens + OverrideThemeTokens
    else:
        Print("No player data found for Player\n")


AddCorrectGuess(Agent:agent):void=
    if:
        CheckForPlayerSavedData[Agent]
        Player := player[Agent]
        SourceData := PlayerSavedDataMap[Player]

        set PlayerSavedDataMap[Player] = player_saved_data:
            MakePlayerSavedData<constructor>(SourceData)
            CorrectGuesses := SourceData.CorrectGuesses + 1
    then:
        SetLastGuess(Agent, true)
        CheckForConsecutiveCorrectGuesses(Agent)
    else:
        Print("No player data found for Player\n")


AddIncorrectGuess(Agent:agent):void=
    if:
        CheckForPlayerSavedData[Agent]
        Player := player[Agent]
        SourceData := PlayerSavedDataMap[Player]

        set PlayerSavedDataMap[Player] = player_saved_data:
            MakePlayerSavedData<constructor>(SourceData)
            IncorrectGuesses := SourceData.IncorrectGuesses + 1
    then:
        SetLastGuess(Agent, false)
    else:
        Print("No player data found for Player\n")


AddCorrectYoloGuess(Agent:agent):void=
    if:
        CheckForPlayerSavedData[Agent]
        Player := player[Agent]
        SourceData := PlayerSavedDataMap[Player]

        set PlayerSavedDataMap[Player] = player_saved_data:
            MakePlayerSavedData<constructor>(SourceData)
            CorrectYoloGuesses := SourceData.CorrectYoloGuesses + 1
    else:
        Print("No player data found for Player\n")


#must retrieve the current consecutive correct guesses and add 1 to it
AddConsecutiveCorrectGuess(Agent:agent):void=
    # get existing consecutive correct guesses
    if:
        CheckForPlayerSavedData[Agent]
        Player := player[Agent]
        SourceData := PlayerSavedDataMap[Player]
        var ConsecutiveCorrectGuesses : int = SourceData.ConsecutiveCorrectGuesses
        ConsecutiveCorrectGuessesRecord := SourceData.ConsecutiveCorrectGuessesRecord

        # if the current consecutive correct guesses is greater than the record, update the record
        if:
            ConsecutiveCorrectGuesses > ConsecutiveCorrectGuessesRecord
            set PlayerSavedDataMap[Player] = player_saved_data:
                MakePlayerSavedData<constructor>(SourceData)
                ConsecutiveCorrectGuessesRecord := ConsecutiveCorrectGuesses
        else:
            set PlayerSavedDataMap[Player] = player_saved_data:
                MakePlayerSavedData<constructor>(SourceData)
        set ConsecutiveCorrectGuesses = ConsecutiveCorrectGuesses + 1

        #update saved data
        set PlayerSavedDataMap[Player] = player_saved_data:
            MakePlayerSavedData<constructor>(SourceData)
            ConsecutiveCorrectGuesses := ConsecutiveCorrectGuesses
    else:
        Print("No player data found for Player\n")


#must retrieve the current consecutive incorrect guesses and add 1 to it
AddConsecutiveIncorrectGuess(Agent:agent):void=
    # get existing consecutive incorrect guesses
    if:
        CheckForPlayerSavedData[Agent]
        Player := player[Agent]
        SourceData := PlayerSavedDataMap[Player]
        var ConsecutiveIncorrectGuesses : int = SourceData.ConsecutiveIncorrectGuesses
        ConsecutiveIncorrectGuessesRecord := SourceData.ConsecutiveIncorrectGuessesRecord

        # if the current consecutive incorrect guesses is greater than the record, update the record
        if:
            ConsecutiveIncorrectGuesses > ConsecutiveIncorrectGuessesRecord
            set PlayerSavedDataMap[Player] = player_saved_data:
                MakePlayerSavedData<constructor>(SourceData)
                ConsecutiveIncorrectGuessesRecord := ConsecutiveIncorrectGuesses
        else:
            set PlayerSavedDataMap[Player] = player_saved_data:
                MakePlayerSavedData<constructor>(SourceData)
        set ConsecutiveIncorrectGuesses = ConsecutiveIncorrectGuesses + 1

        #update saved data
        set PlayerSavedDataMap[Player] = player_saved_data:
            MakePlayerSavedData<constructor>(SourceData)
            ConsecutiveIncorrectGuesses := ConsecutiveIncorrectGuesses
    else:
        Print("No player data found for Player\n")


AddGamesPlayed(Agent:agent):void=
    if:
        CheckForPlayerSavedData[Agent]
        Player := player[Agent]
        SourceData := PlayerSavedDataMap[Player]

        set PlayerSavedDataMap[Player] = player_saved_data:
            MakePlayerSavedData<constructor>(SourceData)
            GamesPlayed := SourceData.GamesPlayed + 1
    else:
        Print("No player data found for Player\n")


AddGamesPlayedWithNewFortWhoPlayers(Agent:agent):void=
    if:
        CheckForPlayerSavedData[Agent]
        Player := player[Agent]
        SourceData := PlayerSavedDataMap[Player]

        set PlayerSavedDataMap[Player] = player_saved_data:
            MakePlayerSavedData<constructor>(SourceData)
            GamesPlayedWithNewFortWhoPlayers := SourceData.GamesPlayedWithNewFortWhoPlayers + 1
    else:
        Print("No player data found for Player\n")

    
AddConsecutiveDaysPlayed(Agent:agent):void=
    if:
        CheckForPlayerSavedData[Agent]
        Player := player[Agent]
        SourceData := PlayerSavedDataMap[Player]

        set PlayerSavedDataMap[Player] = player_saved_data:
            MakePlayerSavedData<constructor>(SourceData)
            ConsecutiveDaysPlayed := SourceData.ConsecutiveDaysPlayed + 1
    else:
        Print("No player data found for Player\n")


AddWinStreak(Agent:agent):void=
    if:
        CheckForPlayerSavedData[Agent]
        Player := player[Agent]
        SourceData := PlayerSavedDataMap[Player]

        set PlayerSavedDataMap[Player] = player_saved_data:
            MakePlayerSavedData<constructor>(SourceData)
            WinStreak := SourceData.WinStreak + 1
    else:
        Print("No player data found for Player\n")


AddLastStyleChoice(Agent:agent, StyleChoice:int):void=
    if:
        CheckForPlayerSavedData[Agent]
        Player := player[Agent]
        SourceData := PlayerSavedDataMap[Player]

        set PlayerSavedDataMap[Player] = player_saved_data:
            MakePlayerSavedData<constructor>(SourceData)
            LastStyleChoice := StyleChoice
            PrevousStyleChoices := SourceData.PrevousStyleChoices + array{StyleChoice}
    else:
        Print("No player data found for Player\n")


AddUnlockedBoardPaint(Agent:agent, BoardPaintID:int):void=
    if:
        CheckForPlayerSavedData[Agent]
        Player := player[Agent]
        SourceData := PlayerSavedDataMap[Player]

        set PlayerSavedDataMap[Player] = player_saved_data:
            MakePlayerSavedData<constructor>(SourceData)
            UnlockedBoardPaints := SourceData.UnlockedBoardPaints + array{BoardPaintID}
    else:
        Print("No player data found for Player\n")


AddUnlockedCardBackPaint(Agent:agent, CardBackPaintID:int):void=
    if:
        CheckForPlayerSavedData[Agent]
        Player := player[Agent]
        SourceData := PlayerSavedDataMap[Player]

        set PlayerSavedDataMap[Player] = player_saved_data:
            MakePlayerSavedData<constructor>(SourceData)
            UnlockedCardBackPaints := SourceData.UnlockedCardBackPaints + array{CardBackPaintID}
    else:
        Print("No player data found for Player\n")


AddUnlockedYouStandPaint(Agent:agent, YouStandPaintID:int):void=
    if:
        CheckForPlayerSavedData[Agent]
        Player := player[Agent]
        SourceData := PlayerSavedDataMap[Player]

        set PlayerSavedDataMap[Player] = player_saved_data:
            MakePlayerSavedData<constructor>(SourceData)
            UnlockedYouStandPaints := SourceData.UnlockedYouStandPaints + array{YouStandPaintID}
    else:
        Print("No player data found for Player\n")


AddUnlockedPlatformPaint(Agent:agent, PlatformPaintID:int):void=
    if:
        CheckForPlayerSavedData[Agent]
        Player := player[Agent]
        SourceData := PlayerSavedDataMap[Player]

        set PlayerSavedDataMap[Player] = player_saved_data:
            MakePlayerSavedData<constructor>(SourceData)
            UnlockedPlatformPaints := SourceData.UnlockedPlatformPaints + array{PlatformPaintID}
    else:
        Print("No player data found for Player\n")


AddCustomPaintSet(Agent:agent, CustomPaintSet:custom_paint_set):void=
    if:
        CheckForPlayerSavedData[Agent]
        Player := player[Agent]
        SourceData := PlayerSavedDataMap[Player]

        set PlayerSavedDataMap[Player] = player_saved_data:
            MakePlayerSavedData<constructor>(SourceData)
            CustomPaintSets := SourceData.CustomPaintSets + array{CustomPaintSet}
    else:
        Print("No player data found for Player\n")


AddUnlockedBoardMesh(Agent:agent, BoardMeshID:int):void=
    if:
        CheckForPlayerSavedData[Agent]
        Player := player[Agent]
        SourceData := PlayerSavedDataMap[Player]

        set PlayerSavedDataMap[Player] = player_saved_data:
            MakePlayerSavedData<constructor>(SourceData)
            UnlockedBoardMeshes := SourceData.UnlockedBoardMeshes + array{BoardMeshID}
    else:
        Print("No player data found for Player\n")


AddUnlockedCardBackMesh(Agent:agent, CardBackMeshID:int):void=
    if:
        CheckForPlayerSavedData[Agent]
        Player := player[Agent]
        SourceData := PlayerSavedDataMap[Player]

        set PlayerSavedDataMap[Player] = player_saved_data:
            MakePlayerSavedData<constructor>(SourceData)
            UnlockedCardBackMeshes := SourceData.UnlockedCardBackMeshes + array{CardBackMeshID}
    else:
        Print("No player data found for Player\n")


AddUnlockedYouStandMesh(Agent:agent, YouStandMeshID:int):void=
    if:
        CheckForPlayerSavedData[Agent]
        Player := player[Agent]
        SourceData := PlayerSavedDataMap[Player]

        set PlayerSavedDataMap[Player] = player_saved_data:
            MakePlayerSavedData<constructor>(SourceData)
            UnlockedYouStandMeshes := SourceData.UnlockedYouStandMeshes + array{YouStandMeshID}
    else:
        Print("No player data found for Player\n")


AddUnlockedPlatformMesh(Agent:agent, PlatformMeshID:int):void=
    if:
        CheckForPlayerSavedData[Agent]
        Player := player[Agent]
        SourceData := PlayerSavedDataMap[Player]

        set PlayerSavedDataMap[Player] = player_saved_data:
            MakePlayerSavedData<constructor>(SourceData)
            UnlockedPlatformMeshes := SourceData.UnlockedPlatformMeshes + array{PlatformMeshID}
    else:
        Print("No player data found for Player\n")


AddUnlockedHat(Agent:agent, HatID:int):void=
    if:
        CheckForPlayerSavedData[Agent]
        Player := player[Agent]
        SourceData := PlayerSavedDataMap[Player]

        set PlayerSavedDataMap[Player] = player_saved_data:
            MakePlayerSavedData<constructor>(SourceData)
            UnlockedHats := SourceData.UnlockedHats + array{HatID}
    else:
        Print("No player data found for Player\n")


AddUnlockedPose(Agent:agent, PoseID:int):void=
    if:
        CheckForPlayerSavedData[Agent]
        Player := player[Agent]
        SourceData := PlayerSavedDataMap[Player]

        set PlayerSavedDataMap[Player] = player_saved_data:
            MakePlayerSavedData<constructor>(SourceData)
            UnlockedPoses := SourceData.UnlockedPoses + array{PoseID}
    else:
        Print("No player data found for Player\n")


AddUnlockedPointer(Agent:agent, PointerID:int)<transacts>:void=
    if:
        CheckForPlayerSavedData[Agent]
        Player := player[Agent]
        SourceData := PlayerSavedDataMap[Player]

        not SourceData.UnlockedPointers.Find[PointerID] # ensure the pointer is not already unlocked

        set PlayerSavedDataMap[Player] = player_saved_data:
            MakePlayerSavedData<constructor>(SourceData)
            UnlockedPointers := SourceData.UnlockedPointers + array{PointerID}
    else:
        Print("No player data found for Player\n")

SetCurrentPointer(Agent:agent, PointerID:int):void=
    if:
        CheckForPlayerSavedData[Agent]
        Player := player[Agent]
        SourceData := PlayerSavedDataMap[Player]

        PointerID <> SourceData.CurrentPointer

        set PlayerSavedDataMap[Player] = player_saved_data:
            MakePlayerSavedData<constructor>(SourceData)
            CurrentPointer := PointerID
            BeenAssignedPointer := true
    else:
        Print("No player data found for Player, or already using that pointer\n")

GetCurrentPointer(Agent:agent)<transacts><decides>:int=
    CheckForPlayerSavedData[Agent]
    Player := player[Agent]
    SourceData := PlayerSavedDataMap[Player]
    ThePointer := SourceData.CurrentPointer

HasPointerBeenAssigned(Agent:agent)<transacts><decides>:void=
    CheckForPlayerSavedData[Agent]
    Player := player[Agent]
    SourceData := PlayerSavedDataMap[Player]
    SourceData.BeenAssignedPointer?

AddUnlockedTrail(Agent:agent, TrailID:int):void=
    if:
        CheckForPlayerSavedData[Agent]
        Player := player[Agent]
        SourceData := PlayerSavedDataMap[Player]

        set PlayerSavedDataMap[Player] = player_saved_data:
            MakePlayerSavedData<constructor>(SourceData)
            UnlockedTrails := SourceData.UnlockedTrails + array{TrailID}
    else:
        Print("No player data found for Player\n")


AddUnlockedTitle(Agent:agent, TitleID:int):void=
    if:
        CheckForPlayerSavedData[Agent]
        Player := player[Agent]
        SourceData := PlayerSavedDataMap[Player]

        set PlayerSavedDataMap[Player] = player_saved_data:
            MakePlayerSavedData<constructor>(SourceData)
            UnlockedTitles := SourceData.UnlockedTitles + array{TitleID}
    else:
        Print("No player data found for Player\n")


AddUnlockedBackground(Agent:agent, BackgroundID:int):void=
    if:
        CheckForPlayerSavedData[Agent]
        Player := player[Agent]
        SourceData := PlayerSavedDataMap[Player]

        set PlayerSavedDataMap[Player] = player_saved_data:
            MakePlayerSavedData<constructor>(SourceData)
            UnlockedBackgrounds := SourceData.UnlockedBackgrounds + array{BackgroundID}
    else:
        Print("No player data found for Player\n")


AddUnlockedMusic(Agent:agent, MusicID:int):void=
    if:
        CheckForPlayerSavedData[Agent]
        Player := player[Agent]
        SourceData := PlayerSavedDataMap[Player]

        set PlayerSavedDataMap[Player] = player_saved_data:
            MakePlayerSavedData<constructor>(SourceData)
            UnlockedMusic := SourceData.UnlockedMusic + array{MusicID}
    else:
        Print("No player data found for Player\n")


AddUnlockedCardKnockSFX(Agent:agent, CardKnockSFXID:int):void=
    if:
        CheckForPlayerSavedData[Agent]
        Player := player[Agent]
        SourceData := PlayerSavedDataMap[Player]

        set PlayerSavedDataMap[Player] = player_saved_data:
            MakePlayerSavedData<constructor>(SourceData)
            UnlockedCardKnockSFX := SourceData.UnlockedCardKnockSFX + array{CardKnockSFXID}
    else:
        Print("No player data found for Player\n")


AddUnlockedUndoSFX(Agent:agent, UndoSFXID:int):void=
    if:
        CheckForPlayerSavedData[Agent]
        Player := player[Agent]
        SourceData := PlayerSavedDataMap[Player]

        set PlayerSavedDataMap[Player] = player_saved_data:
            MakePlayerSavedData<constructor>(SourceData)
            UnlockedUndoSFX := SourceData.UnlockedUndoSFX + array{UndoSFXID}
    else:
        Print("No player data found for Player\n")


AddUnlockedCorrectSFX(Agent:agent, CorrectSFXID:int):void=
    if:
        CheckForPlayerSavedData[Agent]
        Player := player[Agent]
        SourceData := PlayerSavedDataMap[Player]

        set PlayerSavedDataMap[Player] = player_saved_data:
            MakePlayerSavedData<constructor>(SourceData)
            UnlockedCorrectSFX := SourceData.UnlockedCorrectSFX + array{CorrectSFXID}
    else:
        Print("No player data found for Player\n")


AddUnlockedIncorrectSFX(Agent:agent, IncorrectSFXID:int):void=
    if:
        CheckForPlayerSavedData[Agent]
        Player := player[Agent]
        SourceData := PlayerSavedDataMap[Player]

        set PlayerSavedDataMap[Player] = player_saved_data:
            MakePlayerSavedData<constructor>(SourceData)
            UnlockedIncorrectSFX := SourceData.UnlockedIncorrectSFX + array{IncorrectSFXID}
    else:
        Print("No player data found for Player\n")


AddCompletedQuest(Agent:agent, QuestID:int):void=
    if:
        CheckForPlayerSavedData[Agent]
        Player := player[Agent]
        SourceData := PlayerSavedDataMap[Player]

        set PlayerSavedDataMap[Player] = player_saved_data:
            MakePlayerSavedData<constructor>(SourceData)
            CompletedQuests := SourceData.CompletedQuests + array{QuestID}
    else:
        Print("No player data found for Player\n")


AddCurrentQuest(Agent:agent, QuestID:int):void=
    if:
        CheckForPlayerSavedData[Agent]
        Player := player[Agent]
        SourceData := PlayerSavedDataMap[Player]

        set PlayerSavedDataMap[Player] = player_saved_data:
            MakePlayerSavedData<constructor>(SourceData)
            CurrentQuests := SourceData.CurrentQuests + array{QuestID}
    else:
        Print("No player data found for Player\n")


AddCurrentPinnedQuest(Agent:agent, QuestID:int):void=
    if:
        CheckForPlayerSavedData[Agent]
        Player := player[Agent]
        SourceData := PlayerSavedDataMap[Player]

        set PlayerSavedDataMap[Player] = player_saved_data:
            MakePlayerSavedData<constructor>(SourceData)
            CurrentPinnedQuests := SourceData.CurrentPinnedQuests + array{QuestID}
    else:
        Print("No player data found for Player\n")


RemoveCurrentPinnedQuest(Agent:agent, QuestID:int):void=
    if:
        CheckForPlayerSavedData[Agent]
        Player := player[Agent]
        SourceData := PlayerSavedDataMap[Player]

        FoundQuest := SourceData.CurrentPinnedQuests.Find[QuestID]
        NewQuestArr := SourceData.CurrentPinnedQuests.RemoveElement[FoundQuest]

        if:
            set PlayerSavedDataMap[Player] = player_saved_data:
                MakePlayerSavedData<constructor>(SourceData)
                CurrentPinnedQuests := NewQuestArr
        else:
            Print("Quest not found in CurrentPinnedQuests\n")


CompleteQuest(Agent:agent, QuestID:int):void=
    if:
        CheckForPlayerSavedData[Agent]
        Player := player[Agent]
        SourceData := PlayerSavedDataMap[Player]

        FoundQuest := SourceData.CurrentQuests.Find[QuestID]
        NewQuestArr := SourceData.CurrentQuests.RemoveElement[FoundQuest]

        if:
            set PlayerSavedDataMap[Player] = player_saved_data:
                MakePlayerSavedData<constructor>(SourceData)
                CurrentQuests := NewQuestArr
                CompletedQuests := SourceData.CompletedQuests + array{QuestID}

        #if pinned, remove from pinned quests
        if:
            FoundPinnedQuest := SourceData.CurrentPinnedQuests.Find[QuestID]
            NewPinnedQuestArr := SourceData.CurrentPinnedQuests.RemoveElement[FoundPinnedQuest]

            if:
                set PlayerSavedDataMap[Player] = player_saved_data:
                    MakePlayerSavedData<constructor>(SourceData)
                    CurrentPinnedQuests := NewPinnedQuestArr
            else:
                Print("Quest not found in CurrentPinnedQuests\n")

            
        else:
            Print("Quest not found in CurrentQuests\n")

SetFirstDayPlayed(Agent:agent, FirstDay:time_and_date)<transacts>:void=
    if:
        CheckForPlayerSavedData[Agent]
        Player := player[Agent]
        SourceData := PlayerSavedDataMap[Player]

        set PlayerSavedDataMap[Player] = player_saved_data:
            MakePlayerSavedData<constructor>(SourceData)
            FirstDayPlayed := FirstDay
    else:
        Print("PlayerSavedData not found for agent")


GetFirstDayPlayed(Agent:agent)<transacts><decides>:time_and_date=
    CheckForPlayerSavedData[Agent]
    Player := player[Agent]
    SourceData := PlayerSavedDataMap[Player]
    FirstDay := SourceData.FirstDayPlayed

SetLastDayPlayed(Agent:agent, LastDay:time_and_date)<transacts>:void=
    if:
        CheckForPlayerSavedData[Agent]
        Player := player[Agent]
        SourceData := PlayerSavedDataMap[Player]

        set PlayerSavedDataMap[Player] = player_saved_data:
            MakePlayerSavedData<constructor>(SourceData)
            LastDayPlayed := LastDay
    else:
        Print("PlayerSavedData not found for agent")

GetLastDayPlayed(Agent:agent)<transacts><decides>:time_and_date=
    CheckForPlayerSavedData[Agent]
    Player := player[Agent]
    SourceData := PlayerSavedDataMap[Player]
    LastDay := SourceData.LastDayPlayed

SetLastSessionTime(Agent:agent, LastSessionTime:time_and_date)<transacts>:void=
    if:
        CheckForPlayerSavedData[Agent]
        Player := player[Agent]
        SourceData := PlayerSavedDataMap[Player]

        set PlayerSavedDataMap[Player] = player_saved_data:
            MakePlayerSavedData<constructor>(SourceData)
            LastSessionTime := LastSessionTime
    else:
        Print("PlayerSavedData not found for agent")

GetLastSessionTime(Agent:agent)<transacts><decides>:time_and_date=
    CheckForPlayerSavedData[Agent]
    Player := player[Agent]
    SourceData := PlayerSavedDataMap[Player]
    LastSessionTime := SourceData.LastSessionTime



AddTotalDayPlayed(Agent:agent)<transacts>:void=
    var CurrentDaysPlayed : int = 1
    if:
        DaysPlayed := GetTotalDaysPlayed[Agent]
    then:
        set CurrentDaysPlayed = DaysPlayed
        set CurrentDaysPlayed += 1

        SetTotalDaysPlayed(Agent, CurrentDaysPlayed)

SetTotalDaysPlayed(Agent:agent, TotalDays:int)<transacts>:void=
    if:
        CheckForPlayerSavedData[Agent]
        Player := player[Agent]
        SourceData := PlayerSavedDataMap[Player]

        set PlayerSavedDataMap[Player] = player_saved_data:
            MakePlayerSavedData<constructor>(SourceData)
            TotalDaysPlayed := TotalDays
    else:
        Print("PlayerSavedData not found for agent")

GetTotalDaysPlayed(Agent:agent)<transacts><decides>:int=
    CheckForPlayerSavedData[Agent]
    Player := player[Agent]
    SourceData := PlayerSavedDataMap[Player]
    TotalDays := SourceData.TotalDaysPlayed

SetMostConsecutiveDaysPlayed(Agent:agent, MostConsecutiveDays:int)<transacts>:void=
    if:
        CheckForPlayerSavedData[Agent]
        Player := player[Agent]
        SourceData := PlayerSavedDataMap[Player]

        set PlayerSavedDataMap[Player] = player_saved_data:
            MakePlayerSavedData<constructor>(SourceData)
            ConsecutiveDaysPlayedRecord := MostConsecutiveDays
    else:
        Print("PlayerSavedData not found for agent")

GetMostConsecutiveDaysPlayed(Agent:agent)<transacts>:int=
    if:
        CheckForPlayerSavedData[Agent]
        Player := player[Agent]
        SourceData := PlayerSavedDataMap[Player]
    then:
        return MostConsecutiveDays := SourceData.ConsecutiveDaysPlayedRecord

    return 0

SetCurrentConsecutiveDaysPlayed(Agent:agent, CurrentConsecutiveDays:int)<transacts>:void=
    if:
        CheckForPlayerSavedData[Agent]
        Player := player[Agent]
        SourceData := PlayerSavedDataMap[Player]

        set PlayerSavedDataMap[Player] = player_saved_data:
            MakePlayerSavedData<constructor>(SourceData)
            ConsecutiveDaysPlayed := CurrentConsecutiveDays
    else:
        Print("PlayerSavedData not found for agent")

GetCurrentConsecutiveDaysPlayed(Agent:agent)<transacts>:int=
    if:
        CheckForPlayerSavedData[Agent]
        Player := player[Agent]
        SourceData := PlayerSavedDataMap[Player]
    then:
        return CurrentConsecutiveDays := SourceData.ConsecutiveDaysPlayed
    
    return 0


CheckForConsecutiveDaysPlayed(Agent:agent, CurrentDate:time_and_date):void=
    Print("Checking for consecutive days played")
    if:
        CheckForPlayerSavedData[Agent]
        Player := player[Agent]
        SourceData := PlayerSavedDataMap[Player]
    then:
        # Yesterday := time_and_date:
        #     Day := CurrentDate.Day - 1
        #     Month := CurrentDate.Month
        #     Year := CurrentDate.Year
        #     Hour := 0
        #     Minute := 0
        #     Second := 0

        DaysBetween := DaysBetweenDates(SourceData.LastDayPlayed, CurrentDate)
        Print("Check between {SourceData.LastDayPlayed.String()} and {CurrentDate.String()} :  {DaysBetween}")
        CurrentConsecutiveDaysPlayed := SourceData.ConsecutiveDaysPlayed
        Print("Current Consecutive Days Played: {CurrentConsecutiveDaysPlayed}")

        # If the player played yesterday, increment the current consecutive days played
        Print("DaysBetween: {DaysBetween}")
        if(DaysBetween = 1):
            Print("Incrementing Current Consecutive Days Played")
            SetCurrentConsecutiveDaysPlayed(Player, CurrentConsecutiveDaysPlayed + 1)

        else:
            Print("Resetting Current Consecutive Days Played")
            SetCurrentConsecutiveDaysPlayed(Player, 1)

        # check record
        if(CurrentConsecutiveDaysPlayed + 1 > SourceData.ConsecutiveDaysPlayedRecord):
            Print("Updating Consecutive Days Played Record")
            SetMostConsecutiveDaysPlayed(Player, CurrentConsecutiveDaysPlayed + 1)
        else:
            Print("Consecutive Days Played Record Not Updated")
    else:
        Print("Failed to check/update for consecutive days played")

SetLastGuess(Agent:agent, LastGuess:logic):void=
    if:
        CheckForPlayerSavedData[Agent]
        Player := player[Agent]
        SourceData := PlayerSavedDataMap[Player]

        set PlayerSavedDataMap[Player] = player_saved_data:
            MakePlayerSavedData<constructor>(SourceData)
            LastGuessCorrect := LastGuess
    else:
        Print("PlayerSavedData not found for agent")

CheckForConsecutiveCorrectGuesses(Agent:agent):void=
    if:
        CheckForPlayerSavedData[Agent]
        Player := player[Agent]
        SourceData := PlayerSavedDataMap[Player]
    then:
        CurrentConsecutiveCorrectGuesses := SourceData.ConsecutiveCorrectGuesses
        NewConsecutiveCorrectGuesses := CurrentConsecutiveCorrectGuesses + 1

        # If the player guessed correctly, increment the current consecutive correct guesses
        if(SourceData.LastGuessCorrect?):
            SetConsecutiveCorrectGuesses(Player, NewConsecutiveCorrectGuesses)

        else:
            SetConsecutiveCorrectGuesses(Player, 1)

        # check record
        if(NewConsecutiveCorrectGuesses > SourceData.ConsecutiveCorrectGuessesRecord):
            SetConsecutiveCorrectGuessesRecord(Player, NewConsecutiveCorrectGuesses)


SetConsecutiveCorrectGuesses(Agent:agent, ConsecutiveCorrectGuesses:int)<transacts>:void=
    if:
        CheckForPlayerSavedData[Agent]
        Player := player[Agent]
        SourceData := PlayerSavedDataMap[Player]

        set PlayerSavedDataMap[Player] = player_saved_data:
            MakePlayerSavedData<constructor>(SourceData)
            ConsecutiveCorrectGuesses := ConsecutiveCorrectGuesses
    else:
        Print("PlayerSavedData not found for agent")

SetConsecutiveCorrectGuessesRecord(Agent:agent, ConsecutiveCorrectGuesses:int)<transacts>:void=
    if:
        CheckForPlayerSavedData[Agent]
        Player := player[Agent]
        SourceData := PlayerSavedDataMap[Player]

        set PlayerSavedDataMap[Player] = player_saved_data:
            MakePlayerSavedData<constructor>(SourceData)
            ConsecutiveCorrectGuessesRecord := ConsecutiveCorrectGuesses
    else:
        Print("PlayerSavedData not found for agent")

SetTotalMinutesPlayed(Agent:agent, TotalMinutes:int)<transacts>:void=
    if:
        CheckForPlayerSavedData[Agent]
        Player := player[Agent]
        SourceData := PlayerSavedDataMap[Player]

        set PlayerSavedDataMap[Player] = player_saved_data:
            MakePlayerSavedData<constructor>(SourceData)
            TotalMinutesPlayed := TotalMinutes
    else:
        Print("PlayerSavedData not found for agent")

AddMinutePlayed(Agent:agent)<transacts>:void=
    var CurrentMinutesPlayed : int = 1
    if:
        MinutesPlayed := GetTotalMinutesPlayed[Agent]
    then:
        set CurrentMinutesPlayed = MinutesPlayed
        set CurrentMinutesPlayed += 1

        SetTotalMinutesPlayed(Agent, CurrentMinutesPlayed)

GetTotalMinutesPlayed(Agent:agent)<transacts><decides>:int=
    CheckForPlayerSavedData[Agent]
    Player := player[Agent]
    SourceData := PlayerSavedDataMap[Player]
    TotalMinutes := SourceData.TotalMinutesPlayed


IsNewPlayer(Agent:agent)<transacts><decides>:void=
    CheckForPlayerSavedData[Agent]
    Player := player[Agent]
    SourceData := PlayerSavedDataMap[Player]
    NewPlayer := CompareEqualDateIgnoreTime[SourceData.FirstDayPlayed, SourceData.LastDayPlayed]

AllowDailyReward(Agent:agent)<transacts><decides>:void=
    CheckForPlayerSavedData[Agent]
    Player := player[Agent]
    SourceData := PlayerSavedDataMap[Player]
    LastDay := SourceData.LastDayPlayed
    CurrentDay := GetCurrentDateTime()
    DaysDiff := DaysBetweenDates(LastDay, CurrentDay)
    # Allow If First Daily Ever as well, using unused var to track
    AllowReward := DaysDiff > 0 or SourceData.ConsecutiveIncorrectGuessesRecord = 0

ClaimDailyReward(Agent:agent)<transacts>:void=
    if:
        CheckForPlayerSavedData[Agent]
        Player := player[Agent]
        SourceData := PlayerSavedDataMap[Player]
        LastDay := SourceData.LastDayPlayed
        CurrentDay := GetCurrentDateTime()
        DaysDiff := DaysBetweenDates(LastDay, CurrentDay)
        
        set PlayerSavedDataMap[Player] = player_saved_data:
            MakePlayerSavedData<constructor>(SourceData)
            ConsecutiveIncorrectGuessesRecord := -1
            LastDayPlayed := CurrentDay


# TimeBeforeNextDailyReward(Agent:agent)<transacts>:tuple(int, int)=
#     if:
#         CheckForPlayerSavedData[Agent]
#         Player := player[Agent]
#         SourceData := PlayerSavedDataMap[Player]
#         LastDay := SourceData.LastDayPlayed
#         CurrentDay := GetCurrentDateTime()
#         NextRewardDay := AddDayToDateTime(CurrentDay, 1)
#         Minutes := MinutesBetweenDates(CurrentDay, NextRewardDay)
#         Hours := Floor(Minutes / 60)
#         RemainingMinutes := Mod[Minutes,60]
#     then:
#         return (Hours, RemainingMinutes)

#     return (0, 0)

# TimeUntilUTCMidnight()<transacts>:tuple(int, int)=
#     CurrentDay := GetCurrentDateTime()
#     NextDay := AddDayToDateTime(CurrentDay, 1)
#     Minutes := MinutesBetweenDates(CurrentDay, NextDay)
#     Print("Current Day: {CurrentDay.String()} Next Day: {NextDay.String()} Minutes: {Minutes}")
    

#     if:
#         Hours := Floor(Minutes / 60)
#         RemainingMinutes := Mod[Minutes,60]
#     then:
#         return (Hours, RemainingMinutes)
    
#     return (-1, -1)

TimeUntilNextUTCMidnight(CurrentDay:time_and_date)<transacts>:tuple(int, int)=
    NextDay := AddDayToDateTime(CurrentDay, 1)
    NextMidnight := NextDay.GetMidnight()
    Minutes := MinutesBetweenDates(CurrentDay, NextMidnight)
    Print("Current Day: {CurrentDay.String()} Next Day: {NextDay.String()} Midnight: {NextMidnight.String()} Minutes: {Minutes}")
    

    if:
        Hours := Floor(Minutes / 60)
        RemainingMinutes := Mod[Minutes,60]
    then:
        return (Hours, RemainingMinutes)
    
    return (-1, -1)



SetStreamerModeHideCharacter(Agent:agent, Hide:logic)<transacts>:void=
    if:
        CheckForPlayerSavedData[Agent]
        Player := player[Agent]
        SourceData := PlayerSavedDataMap[Player]

        set PlayerSavedDataMap[Player] = player_saved_data:
            MakePlayerSavedData<constructor>(SourceData)
            StreamerModeHideCharacter := Hide
    else:
        Print("PlayerSavedData not found for agent")

GetStreamerModeHideCharacter(Agent:agent)<transacts><decides>:void=
    CheckForPlayerSavedData[Agent]
    Player := player[Agent]
    SourceData := PlayerSavedDataMap[Player]
    SourceData.StreamerModeHideCharacter?

SetStreamerModeVerticalUI(Agent:agent, Vertical:logic)<transacts>:void=
    if:
        CheckForPlayerSavedData[Agent]
        Player := player[Agent]
        SourceData := PlayerSavedDataMap[Player]

        set PlayerSavedDataMap[Player] = player_saved_data:
            MakePlayerSavedData<constructor>(SourceData)
            StreamerModeVerticalUI := Vertical
    else:
        Print("PlayerSavedData not found for agent")

GetStreamerModeVerticalUI(Agent:agent)<transacts><decides>:void=
    CheckForPlayerSavedData[Agent]
    Player := player[Agent]
    SourceData := PlayerSavedDataMap[Player]
    SourceData.StreamerModeVerticalUI?

SetManuallyChooseCharacter(Agent:agent, ManuallyChoose:logic)<transacts>:void=
    if:
        CheckForPlayerSavedData[Agent]
        Player := player[Agent]
        SourceData := PlayerSavedDataMap[Player]

        set PlayerSavedDataMap[Player] = player_saved_data:
            MakePlayerSavedData<constructor>(SourceData)
            ManuallyChooseCharacter := ManuallyChoose
    else:
        Print("PlayerSavedData not found for agent")

GetManuallyChooseCharacter(Agent:agent)<transacts><decides>:void=
    CheckForPlayerSavedData[Agent]
    Player := player[Agent]
    SourceData := PlayerSavedDataMap[Player]
    SourceData.ManuallyChooseCharacter?



SetCompletedWinter2024Quest(Agent:agent)<transacts>:void=
    if:
        CheckForPlayerSavedData[Agent]
        Player := player[Agent]
        SourceData := PlayerSavedDataMap[Player]

        set PlayerSavedDataMap[Player] = player_saved_data:
            MakePlayerSavedData<constructor>(SourceData)
            CompletedWinter2024Quest := true
    else:
        Print("PlayerSavedData not found for agent")


HasCompletedWinter2024Quest(Agent:agent)<transacts><decides>:void=
    CheckForPlayerSavedData[Agent]
    Player := player[Agent]
    SourceData := PlayerSavedDataMap[Player]
    SourceData.CompletedWinter2024Quest?


AddFlyingTargetHit(Agent:agent)<transacts>:void=
    if:
        CheckForPlayerSavedData[Agent]
        Player := player[Agent]
        SourceData := PlayerSavedDataMap[Player]

        set PlayerSavedDataMap[Player] = player_saved_data:
            MakePlayerSavedData<constructor>(SourceData)
            TotalFlyingTargetsHit := SourceData.TotalFlyingTargetsHit + 1
    else:
        Print("PlayerSavedData not found for agent")

GetTotalFlyingTargetsHit(Agent:agent)<transacts><decides>:int=
    CheckForPlayerSavedData[Agent]
    Player := player[Agent]
    SourceData := PlayerSavedDataMap[Player]
    SourceData.TotalFlyingTargetsHit


UnlockSticker(Agent:agent, StickerID:int)<transacts>:void=
    if:
        CheckForPlayerSavedData[Agent]
        Player := player[Agent]
        SourceData := PlayerSavedDataMap[Player]

        not SourceData.UnlockedStickers.Find[StickerID] # ensure the sticker is not already unlocked

        set PlayerSavedDataMap[Player] = player_saved_data:
            MakePlayerSavedData<constructor>(SourceData)
            UnlockedStickers := SourceData.UnlockedStickers + array{StickerID}
    else:
        Print("PlayerSavedData not found for agent")

AlreadyUnlockedSticker(Agent:agent, StickerID:int)<transacts><decides>:void=
    CheckForPlayerSavedData[Agent]
    Player := player[Agent]
    SourceData := PlayerSavedDataMap[Player]
    SourceData.UnlockedStickers.Find[StickerID]

SaveCurrentSticker(Agent:agent, StickerID:int, Transform:transform)<transacts>:void=
    if:
        CheckForPlayerSavedData[Agent]
        Player := player[Agent]
        SourceData := PlayerSavedDataMap[Player]
        YawRotation := Transform.Rotation.GetYawPitchRollDegrees()[0]

        set PlayerSavedDataMap[Player] = player_saved_data:
            MakePlayerSavedData<constructor>(SourceData)
            CurrentStickers := SourceData.CurrentStickers + array{sticker_slot{StickerID:=StickerID, Translation:=Transform.Translation, Rotation:=YawRotation}}
            
GetCurrentStickers(Agent:agent)<transacts><decides>:[]sticker_slot=
    CheckForPlayerSavedData[Agent]
    Player := player[Agent]
    SourceData := PlayerSavedDataMap[Player]
    SourceData.CurrentStickers

RemoveSticker(Agent:agent, StickerID:int)<transacts>:void=
    if:
        CheckForPlayerSavedData[Agent]
        Player := player[Agent]
        SourceData := PlayerSavedDataMap[Player]

        FoundSticker := SourceData.CurrentStickers.Find[StickerID]
        NewStickerArr := SourceData.CurrentStickers.RemoveElement[FoundSticker]

        if:
            set PlayerSavedDataMap[Player] = player_saved_data:
                MakePlayerSavedData<constructor>(SourceData)
                CurrentStickers := NewStickerArr
        else:
            Print("Sticker not found in CurrentStickers\n")