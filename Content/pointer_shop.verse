
using { /Verse.org/Simulation }
using { /Fortnite.com/Devices }



pointer_shop := class<concrete>():

    @editable Button_Equip_Point_Plank : button_device = button_device{}
    @editable Button_Equip_WhoShooter : button_device = button_device{}
    @editable Button_Equip_FoamDiscLauncher : button_device = button_device{}

    @editable Dialog : popup_dialog_device = popup_dialog_device{}
    @editable Remover : item_remover_device = item_remover_device{}

    # Pointer Things
    @editable Granters : []item_granter_device = array{}
    @editable ConditionalButtons : []conditional_button_device = array{}
    @editable PointersVFX : []visual_effect_powerup_device = array{}
    @editable HUDInventoryIcons : []hud_message_device = array{}
    @editable Analytics : []analytics_device = array{}
    @editable AmmoProps : custom_ammo_assets = custom_ammo_assets{}


    Init()<suspends>:void=
        spawn . AwaitPointPlankEquip()
        spawn . AwaitWhoShooterEquip()
        spawn . AwaitFoamDiscLauncherEquip()

        Dialog.RespondingButtonEvent.Subscribe(HandleEquipPointerResponse)
        Dialog.TimeOutEvent.Subscribe(HandleEquipPointerTimeOut)

    HandleEquipPointerTimeOut(Agent:agent):void=
        HandleEquipPointerResponse(Agent, 0)

    HandleEquipPointerResponse(Agent:agent, Index:int):void=
        #equip and assign their fisrt pointer
        if:
            TheGlobals := Globals[GetSession()]
            ThePointer := TheGlobals.GetPointerData(Index)
            PointerVFX := PointersVFX[Index]
            PointerGranter := Granters[Index]
            Conditional := ConditionalButtons[Index]
            InvIcon := HUDInventoryIcons[Index]
            Analytic := Analytics[Index]
        then:
            ThePointer.InstantUnlock(Agent)
            ThePointer.Equip(Agent, PointerGranter, PointerVFX, Conditional, InvIcon, Remover)
            Analytic.Submit(Agent)

    AwaitPointPlankEquip()<suspends>:void=
        loop:
            Agent := Button_Equip_Point_Plank.InteractedWithEvent.Await()
            Print("Purchasing PointPlank")
            # spawn . PurchasePointPlank(Agent)
            spawn . PurchasePointer(0, Agent)

    AwaitWhoShooterEquip()<suspends>:void=
        loop:
            Agent := Button_Equip_WhoShooter.InteractedWithEvent.Await()
            Print("Purchasing WhoShooter")
            # spawn . PurchaseWhoShooter(Agent)
            spawn . PurchasePointer(1, Agent)

    AwaitFoamDiscLauncherEquip()<suspends>:void=
        loop:
            Agent := Button_Equip_FoamDiscLauncher.InteractedWithEvent.Await()
            Print("Purchasing FoamDiscLauncher")
            # spawn . PurchaseFoamDiscLauncher(Agent)
            spawn . PurchasePointer(2, Agent)


    PurchasePointer(PointerID:int, Agent:agent)<suspends>:void=
        if:
            TheGlobals := Globals[GetSession()]
            Player := player[Agent]
            SourceData := PlayerSavedDataMap[Player]
            PointerData := TheGlobals.GetPointerData(PointerID)
            IsOwned := PointerData.IsOwned[Agent]
            PointerVFX := PointersVFX[PointerID]
            PointerGranter := Granters[PointerID]
            Conditional := ConditionalButtons[PointerID]
            InvIcon := HUDInventoryIcons[PointerID]
            Analytic := Analytics[PointerID]
        then:
            # Owns, equip
            PointerData.Equip(Agent, PointerGranter, PointerVFX, Conditional, InvIcon, Remover)
            Analytic.Submit(Agent)
        else if:
            # allow purchase
            TheGlobals := Globals[GetSession()]
            Player := player[Agent]
            SourceData := PlayerSavedDataMap[Player]
            PointerData := TheGlobals.GetPointerData(PointerID)
            IsPurchasedAllowed := PointerData.IsPurchasedAllowed[Agent]
            PointerData.Purchase(Agent)
            PointerGranter := Granters[PointerID]
            PointerVFX := PointersVFX[PointerID]
            Conditional := ConditionalButtons[PointerID]
            InvIcon := HUDInventoryIcons[PointerID]
        then:
            # bought
            PointerData.Equip(Agent, PointerGranter, PointerVFX, Conditional, InvIcon, Remover)
        else:
            # cannot afford
            Print("Cannot afford")
        Sleep(0.0)



        
    # PurchaseWhoShooter(Agent:agent)<suspends>:void=
        
    #     if:
    #         TheGlobals := Globals[GetSession()]
    #         Player := player[Agent]
    #         SourceData := PlayerSavedDataMap[Player]
    #         PointerData := TheGlobals.GetPointerData(1)
    #         IsOwned := PointerData.IsOwned[Agent]
    #         PointerVFX := PointersVFX[1]
    #         PointerGranter := Granters[1]
    #         Analytic := Analytics[Index]
    #     then:
    #         # Owns, equip
    #         PointerData.Equip(Agent, PointerGranter, PointerVFX, Remover)
    #         Analytic.Submit(Agent)
    #     else if:
    #         # allow purchase
    #         TheGlobals := Globals[GetSession()]
    #         Player := player[Agent]
    #         SourceData := PlayerSavedDataMap[Player]
    #         PointerData := TheGlobals.GetPointerData(1)
    #         PointerVFX := PointersVFX[1]
    #         PointerGranter := Granters[1]
    #         IsPurchasedAllowed := PointerData.IsPurchasedAllowed[Agent]
    #         PointerData.Purchase(Agent)

    #     then:
    #     # bought
    #         PointerData.Equip(Agent, PointerGranter, PointerVFX, Remover)
    #     else:
    #         # cannot afford
    #         Print("Cannot afford")
    #     Sleep(0.0)
    
     # PurchasePointPlank(Agent:agent)<suspends>:void=
    #     if:
    #         TheGlobals := Globals[GetSession()]
    #         Player := player[Agent]
    #         SourceData := PlayerSavedDataMap[Player]
    #         PointerData := TheGlobals.GetPointerData(0)
    #         IsOwned := PointerData.IsOwned[Agent]
    #         PointerVFX := PointersVFX[0]
    #         PointerGranter := Granters[0]
    #         Analytic := Analytics[Index]            
    #     then:
    #         # Owns, equip
    #         PointerData.Equip(Agent, PointerGranter, PointerVFX, Remover)
    #         Analytic.Submit(Agent)
    #     else if:
    #         # allow purchase
    #         TheGlobals := Globals[GetSession()]
    #         Player := player[Agent]
    #         SourceData := PlayerSavedDataMap[Player]
    #         PointerData := TheGlobals.GetPointerData(0)
    #         IsPurchasedAllowed := PointerData.IsPurchasedAllowed[Agent]
    #         PointerData.Purchase(Agent)
    #     else:
    #         # cannot afford
    #         Print("Cannot afford")
    #     Sleep(0.0)

