using { /Verse.org/Colors }
using { /Verse.org/Colors/NamedColors }
using { /Verse.org/Assets }




putty_material_data := class():
    GlobalTilling:float=1.0
    Metallic:float=0.0
    Roughness:float=0.6 
    ColorClay1:color=White 
    ColorClay2:color=White 
    ColorClayDot:color=MakeColorFromHSV(0.881317, 0.161458, 0.708376) 
    Clay2GreyScaleMap:texture=T_Clay_Pattern_01_GreySacle 
    Clay2NormalMap:texture=T_Clay_Pattern_01_Normal 
    TilingUClaySplite:float=1.0 
    TilingVClaySplite:float=1.0 
    MapClayDot:texture=T_Clay_Dot 
    NumberClayDot:float=0.066667 
    TilingUClayDot:float=5.0 
    TilingVClayDot:float=5.0 
    ClayNormalMap:texture=T_Clay_Normal 
    IntensityClayNormal:float=1.0 
    TilingUClayNormal:float=1.0 
    TilingVClayNormal:float=1.0 
    EmissiveStrength:float=1.0 
    BaseReflectFractionInFresnel:float=0.0 
    ExponentInFresnel:float=4.0 
    IntensityFresnel:float=5.0 
    EmissiveFlickerSpeed:float=1.0 
    FrameEmissiveStrength:float=1.0 
    GlobalEmissiveStrength:float=50.0 
    BoardNormal:texture=T_PlayBoard_N 
    BoardORM:texture=T_PlayBoard_Base_ORM 
    EmissiveMask2:texture=T_Lights_EM 
    EmissiveFlickerTint:color=Pink 
    EmissiveTint:color=White 
    FrameEmissiveTint:color=White