using { /Fortnite.com/Devices }
using { /Verse.org/Simulation }
using { /UnrealEngine.com/Temporary/Diagnostics }
# using { /Verse.org/Random }
using { /Fortnite.com/UI}
using { /UnrealEngine.com/Temporary/UI }
using { /UnrealEngine.com/Temporary/SpatialMath}
using { /Verse.org/Colors}
# using { /Fortnite.com/Characters}
# using { /Fortnite.com/Teams } 
using { /Verse.org/Colors/NamedColors }
# using { /Fortnite.com/FortPlayerUtilities }
using { /Verse.org/Assets }

using { QuestionBank }

# See https://dev.epicgames.com/documentation/en-us/uefn/create-your-own-device-in-verse for how to create a verse device.

my_log_channel := class(log_channel){}

# NEEDS TO BE UPDATED WHEN NEW STYLES (SETS) ARE ADDED ######################################################################################################
card_style := enum:
    FN         # 0
    Classic    # 1
    Icon       # 2
    Animal     # 3
    Utilities  # 4
    Marvel     # 5 - Soon to be deprecated
    POI        # 6 - Vaulted
    Legends    # 7
    Food       # 8
    Star       # 9
    Spooky     # 10
    OG         # 11
    Anime      # 12
    Winter     # 13
    Weapons    # 14
    Pickaxes   # 15
    Vehicles   # 16
    MoviesTV   # 17
    Blocks     # 18 - never used
    Chapter5   # 19
    Lockers    # 20
    Backblings # 21
    Emotes     # 22
    Marvel1    # 23
    Marvel2    # 24
    Umbrellas  # 25
    PeelyFishstick # 26
    Remix      # 27
    FireIce     # 28
    DC         # 29
    Chapter6   # 30




category := class<concrete>():
    Style : card_style = card_style.FN
    Name : string = "Category"
    NameMessage : message = MessageBlank_EN
    Questions : []question = array{}
    # ToMessage<localizes>():message=
    #     "{TheQuestion}"

question := class<concrete>():
    # TheQuestion : string = "Default Question"
    TheQuestion : message = MessageBlank_EN
    # ToMessage<localizes>():message=
    #     "{TheQuestion}"

# A Verse-authored creative device that can be placed in a level
question_bank_device := class():

    Logger:log = log{ Channel := my_log_channel }

    #################################### 
    # Variables
    ####################################

    var SelectedQuestion : message = MessageBlank_EN

    # var Categories : []category = array{}

    var HUDResponseNo : hud_message_device =hud_message_device{}
    var HUDResponseYes: hud_message_device =hud_message_device{}
    var HUDResponseMaybe: hud_message_device =hud_message_device{}

    var AnalyticQuestionAsked : analytics_device = analytics_device{}
    var AnalyticQuestionAnswered : analytics_device = analytics_device{}
    

    var ButtonYes : button_regular = button_regular{}
    var ButtonNo : button_regular = button_regular{}
    var ButtonMaybe : button_regular = button_regular{}

    var LastRespondedTime : float = 0.0

    OpponentResponseEvent:event()=event():
    QuestionAskedEvent:event()=event():
    # var CurrentlyAskingQuestion : logic = false

    var ButtonPrev : button_loud = button_loud{}
    var ButtonNext : button_loud = button_loud{}
    var ButtonExit : button_quiet = button_quiet{}

    var FNQuestions : []category = array{}

    # var FinishedLoadingPage : logic = true

    InitFNQuestions()<suspends>:void=

        ButtonYes.OnClick().Subscribe(OnClickButtonYes) 
        ButtonNo.OnClick().Subscribe(OnClickButtonNo) 
        ButtonMaybe.OnClick().Subscribe(OnClickButtonMaybe)

        ButtonPrev.OnClick().Subscribe(OnClickButtonClassLeft) 
        ButtonNext.OnClick().Subscribe(OnClickButtonClassRight) 
        ButtonExit.OnClick().Subscribe(OnClickButtonExit) 

        set FNQuestions = array{category{
            Style := card_style.FN
            Name := "Colors [1/2]"
            NameMessage := QuestionBank.AllCategories.MessageColors_EN
            Questions := array{
                question{TheQuestion := QuestionBank.AllQuestions.CharacterQuestions.MessageCharacterColorYellow_EN},
                question{TheQuestion := QuestionBank.AllQuestions.CharacterQuestions.MessageCharacterColorOrange_EN},
                question{TheQuestion := QuestionBank.AllQuestions.CharacterQuestions.MessageCharacterColorRed_EN},
                question{TheQuestion := QuestionBank.AllQuestions.CharacterQuestions.MessageCharacterColorPink_EN},
                question{TheQuestion := QuestionBank.AllQuestions.CharacterQuestions.MessageCharacterColorGreen_EN},
                question{TheQuestion := QuestionBank.AllQuestions.CharacterQuestions.MessageCharacterColorBlue_EN},
                question{TheQuestion := QuestionBank.AllQuestions.CharacterQuestions.MessageCharacterColorPurple_EN},
                question{TheQuestion := QuestionBank.AllQuestions.CharacterQuestions.MessageCharacterColorBlack_EN},
                question{TheQuestion := QuestionBank.AllQuestions.CharacterQuestions.MessageCharacterColorWhite_EN},
                question{TheQuestion := QuestionBank.AllQuestions.CharacterQuestions.MessageCharacterColorTeal_EN}
            }
        }}

        set FNQuestions += array{category{
            Style:=card_style.FN,
            Name:="Accessories",
            NameMessage := QuestionBank.AllCategories.MessageAccessories_EN,
            Questions:=array{
                question{TheQuestion := QuestionBank.AllQuestions.CharacterQuestions.MessageCharacterHelmet_EN},
                question{TheQuestion := QuestionBank.AllQuestions.CharacterQuestions.MessageCharacterGlasses_EN},
                question{TheQuestion := QuestionBank.AllQuestions.CharacterQuestions.MessageCharacterMask_EN},
                question{TheQuestion := QuestionBank.AllQuestions.CharacterQuestions.MessageCharacterNecklace_EN},
                question{TheQuestion := QuestionBank.AllQuestions.CharacterQuestions.MessageCharacterHat_EN},
                question{TheQuestion := QuestionBank.AllQuestions.CharacterQuestions.MessageCharacterTie_EN},
                question{TheQuestion := QuestionBank.AllQuestions.WinterQuestions.MessageWinterAttireScarf_EN},
                question{TheQuestion := QuestionBank.AllQuestions.CharacterQuestions.MessageCharacterHair_EN}
            }
        }}
        set FNQuestions += array{category{
            Style := card_style.FN
            Name := "Species"
            NameMessage := QuestionBank.AllCategories.MessageSpecies_EN
            Questions := array{
                question{TheQuestion := QuestionBank.AllQuestions.CharacterQuestions.MessageCharacterSpeciesHuman_EN},
                question{TheQuestion := QuestionBank.AllQuestions.CharacterQuestions.MessageCharacterSpeciesAnimal_EN},
                question{TheQuestion := QuestionBank.AllQuestions.CharacterQuestions.MessageCharacterSpeciesFood_EN},
                question{TheQuestion := QuestionBank.AllQuestions.CharacterQuestions.MessageCharacterSpeciesInflatable_EN},
                question{TheQuestion := QuestionBank.AllQuestions.CharacterQuestions.MessageCharacterSpeciesRobot_EN}
            }
        }}
        set FNQuestions += array{category{
            Style := card_style.FN
            Name := "Other"
            NameMessage := QuestionBank.AllCategories.MessageOther_EN
            Questions := array{
                question{TheQuestion := QuestionBank.AllQuestions.CharacterQuestions.MessageCharacterFacePaint_EN},
                question{TheQuestion := QuestionBank.AllQuestions.CharacterQuestions.MessageCharacterEyeColorGlowing_EN},
                question{TheQuestion := QuestionBank.AllQuestions.CharacterQuestions.MessageCharacterSmiling_EN},
                question{TheQuestion := QuestionBank.AllQuestions.CharacterQuestions.MessageCharacterDogTags_EN},
                question{TheQuestion := QuestionBank.AllQuestions.CharacterQuestions.MessageCharacterChainmail_EN},
                question{TheQuestion := QuestionBank.AllQuestions.CharacterQuestions.MessageCharacterEyeColorBlack_EN},
                question{TheQuestion := QuestionBank.AllQuestions.CharacterQuestions.MessageCharacterEyeColorWhite_EN}
            }
        }}
    


    
    

    var IconQuestions : []category = array{}

    InitIconQuestions()<suspends>:void=

        set IconQuestions = array{category{
            Style := card_style.Icon
            Name := "Occupation"
            NameMessage := QuestionBank.AllCategories.MessageOccupations_EN
            Questions := array{
                question{TheQuestion := QuestionBank.AllQuestions.CharacterQuestions.MessageCharacterMuscian_EN},
                question{TheQuestion := QuestionBank.AllQuestions.CharacterQuestions.MessageCharacterAthlete_EN},
                question{TheQuestion := QuestionBank.AllQuestions.CharacterQuestions.MessageCharacterYouTuber_EN},
                question{TheQuestion := QuestionBank.AllQuestions.CharacterQuestions.MessageCharacterStreamer_EN},
                question{TheQuestion := QuestionBank.AllQuestions.CharacterQuestions.MessageCharacterProGamer_EN},
                question{TheQuestion := QuestionBank.AllQuestions.CharacterQuestions.MessageCharacterSinger_EN},
                question{TheQuestion := QuestionBank.AllQuestions.CharacterQuestions.MessageCharacterActor_EN}
            }
        }}
        set IconQuestions += array{category{
            Style := card_style.Icon
            Name := "Colors [1/2]"
            NameMessage := QuestionBank.AllCategories.MessageColors_EN
            Questions := array{
                question{TheQuestion := QuestionBank.AllQuestions.CharacterQuestions.MessageCharacterColorYellow_EN},
                question{TheQuestion := QuestionBank.AllQuestions.CharacterQuestions.MessageCharacterColorOrange_EN},
                question{TheQuestion := QuestionBank.AllQuestions.CharacterQuestions.MessageCharacterColorRed_EN},
                question{TheQuestion := QuestionBank.AllQuestions.CharacterQuestions.MessageCharacterColorPink_EN},
                question{TheQuestion := QuestionBank.AllQuestions.CharacterQuestions.MessageCharacterColorGreen_EN},
                question{TheQuestion := QuestionBank.AllQuestions.CharacterQuestions.MessageCharacterColorBlue_EN},
                question{TheQuestion := QuestionBank.AllQuestions.CharacterQuestions.MessageCharacterColorPurple_EN},
                question{TheQuestion := QuestionBank.AllQuestions.CharacterQuestions.MessageCharacterColorBlack_EN},
                question{TheQuestion := QuestionBank.AllQuestions.CharacterQuestions.MessageCharacterColorWhite_EN},
                question{TheQuestion := QuestionBank.AllQuestions.CharacterQuestions.MessageCharacterColorGold_EN},
                question{TheQuestion := QuestionBank.AllQuestions.CharacterQuestions.MessageCharacterColorBrown_EN},
                question{TheQuestion := QuestionBank.AllQuestions.CharacterQuestions.MessageCharacterColorTeal_EN}
            }
        }}

        set IconQuestions += array{category{
            Style := card_style.Icon
            Name := "Accessories"
            NameMessage := QuestionBank.AllCategories.MessageAccessories_EN
            Questions := array{
                question{TheQuestion := QuestionBank.AllQuestions.CharacterQuestions.MessageCharacterHat_EN},
                question{TheQuestion := QuestionBank.AllQuestions.CharacterQuestions.MessageCharacterGlasses_EN},
                question{TheQuestion := QuestionBank.AllQuestions.CharacterQuestions.MessageCharacterMask_EN},
                question{TheQuestion := QuestionBank.AllQuestions.CharacterQuestions.MessageCharacterNecklace_EN},
                question{TheQuestion := QuestionBank.AllQuestions.CharacterQuestions.MessageCharacterHoodie_EN},
                question{TheQuestion := QuestionBank.AllQuestions.CharacterQuestions.MessageCharacterMusicalItem_EN},
                question{TheQuestion := QuestionBank.AllQuestions.CharacterQuestions.MessageCharacterShirt_EN},
                question{TheQuestion := QuestionBank.AllQuestions.CharacterQuestions.MessageCharacterSmiling_EN}
            }
        }}
        set IconQuestions += array{category{
            Style := card_style.Icon
            Name := "Hair Colors"
            NameMessage := QuestionBank.AllCategories.MessageHairColors_EN
            Questions := array{
                question{TheQuestion := QuestionBank.AllQuestions.CharacterQuestions.MessageCharacterHairColorBlack_EN},
                question{TheQuestion := QuestionBank.AllQuestions.CharacterQuestions.MessageCharacterHairColorRed_EN},
                question{TheQuestion := QuestionBank.AllQuestions.CharacterQuestions.MessageCharacterHairColorPink_EN},
                question{TheQuestion := QuestionBank.AllQuestions.CharacterQuestions.MessageCharacterHairColorBrown_EN},
                question{TheQuestion := QuestionBank.AllQuestions.CharacterQuestions.MessageCharacterHairColorBlue_EN},
                question{TheQuestion := QuestionBank.AllQuestions.CharacterQuestions.MessageCharacterHairColorBlonde_EN},
                question{TheQuestion := QuestionBank.AllQuestions.CharacterQuestions.MessageCharacterHairColorDyed_EN}
            }
        }}
        set IconQuestions += array{category{
            Style := card_style.Icon
            Name := "Nationality"
            NameMessage := QuestionBank.AllCategories.MessageNationalities_EN
            Questions := array{
                question{TheQuestion := QuestionBank.AllQuestions.CharacterQuestions.MessageCharacterUSA_EN},
                question{TheQuestion := QuestionBank.AllQuestions.CharacterQuestions.MessageCharacterCanada_EN},
                question{TheQuestion := QuestionBank.AllQuestions.CharacterQuestions.MessageCharacterAustralia_EN},
                question{TheQuestion := QuestionBank.AllQuestions.CharacterQuestions.MessageCharacterSpain_EN},
                question{TheQuestion := QuestionBank.AllQuestions.CharacterQuestions.MessageCharacterBrazil_EN},
                question{TheQuestion := QuestionBank.AllQuestions.CharacterQuestions.MessageCharacterJapan_EN},
                question{TheQuestion := QuestionBank.AllQuestions.CharacterQuestions.MessageCharacterColombia_EN}
            }
        }}
        set IconQuestions += array{category{
            Style := card_style.Icon
            Name := "Other"
            NameMessage := QuestionBank.AllCategories.MessageOther_EN
            Questions := array{
                question{TheQuestion := QuestionBank.AllQuestions.CharacterQuestions.MessageCharacterGenderMale_EN},
                question{TheQuestion := QuestionBank.AllQuestions.CharacterQuestions.MessageCharacterGenderFemale_EN},
                question{TheQuestion := QuestionBank.AllQuestions.CharacterQuestions.MessageCharacterZipper_EN},
                question{TheQuestion := QuestionBank.AllQuestions.CharacterQuestions.MessageCharacterTattoo_EN},
                question{TheQuestion := QuestionBank.AllQuestions.CharacterQuestions.MessageCharacterEyeColorBlue_EN},
                question{TheQuestion := QuestionBank.AllQuestions.CharacterQuestions.MessageCharacterEyeColorBrown_EN},
                question{TheQuestion := QuestionBank.AllQuestions.CharacterQuestions.MessageCharacterFacialHair_EN}
            }
        }}
    

    var AnimalQuestions : []category = array{}

    InitAnimalQuestions()<suspends>:void=

        set AnimalQuestions = array{category{
            Style := card_style.Animal
            Name := "Color"
            NameMessage := QuestionBank.AllCategories.MessageColors_EN
            Questions := array{
                question{TheQuestion := QuestionBank.AllQuestions.AnimalQuestions.MessageAnimalBlue_EN},
                question{TheQuestion := QuestionBank.AllQuestions.AnimalQuestions.MessageAnimalGreen_EN},
                question{TheQuestion := QuestionBank.AllQuestions.AnimalQuestions.MessageAnimalPurple_EN},
                question{TheQuestion := QuestionBank.AllQuestions.AnimalQuestions.MessageAnimalYellow_EN},
                question{TheQuestion := QuestionBank.AllQuestions.AnimalQuestions.MessageAnimalRed_EN},
                question{TheQuestion := QuestionBank.AllQuestions.AnimalQuestions.MessageAnimalWhite_EN},
                question{TheQuestion := QuestionBank.AllQuestions.AnimalQuestions.MessageAnimalBrown_EN},
                question{TheQuestion := QuestionBank.AllQuestions.AnimalQuestions.MessageAnimalBlack_EN}
            }
        }}
        set AnimalQuestions += array{category{
            Style := card_style.Animal
            Name := "Size"
            NameMessage := QuestionBank.AllCategories.MessageSize_EN
            Questions := array{
                question{TheQuestion := QuestionBank.AllQuestions.AnimalQuestions.MessageAnimalTiny_EN},
                question{TheQuestion := QuestionBank.AllQuestions.AnimalQuestions.MessageAnimalSmall_EN},
                question{TheQuestion := QuestionBank.AllQuestions.AnimalQuestions.MessageAnimalMedium_EN},
                question{TheQuestion := QuestionBank.AllQuestions.AnimalQuestions.MessageAnimalHumanSized_EN},
                question{TheQuestion := QuestionBank.AllQuestions.AnimalQuestions.MessageAnimalLarge_EN},
                question{TheQuestion := QuestionBank.AllQuestions.AnimalQuestions.MessageAnimalHuge_EN}
            }
        }}
        set AnimalQuestions += array{category{
            Style := card_style.Animal
            Name := "Attributes"
            NameMessage := QuestionBank.AllCategories.MessageAttributes_EN
            Questions := array{
                question{TheQuestion := QuestionBank.AllQuestions.AnimalQuestions.MessageAnimalSharpTeeth_EN},
                question{TheQuestion := QuestionBank.AllQuestions.AnimalQuestions.MessageAnimalYellowEyes_EN},
                question{TheQuestion := QuestionBank.AllQuestions.AnimalQuestions.MessageAnimalEyes_EN},
                question{TheQuestion := QuestionBank.AllQuestions.AnimalQuestions.MessageAnimalScales_EN},
                question{TheQuestion := QuestionBank.AllQuestions.AnimalQuestions.MessageAnimalFeathers_EN},
                question{TheQuestion := QuestionBank.AllQuestions.AnimalQuestions.MessageAnimalFur_EN},
                question{TheQuestion := QuestionBank.AllQuestions.AnimalQuestions.MessageAnimalWings_EN},
                question{TheQuestion := QuestionBank.AllQuestions.AnimalQuestions.MessageAnimalExists_EN}
            }
        }}
        set AnimalQuestions += array{category{
            Style := card_style.Animal
            Name := "Mobility"
            NameMessage := QuestionBank.AllCategories.MessageMobility_EN
            Questions := array{
                question{TheQuestion := QuestionBank.AllQuestions.AnimalQuestions.MessageAnimalRidable_EN},
                question{TheQuestion := QuestionBank.AllQuestions.AnimalQuestions.MessageAnimalFly_EN},
                question{TheQuestion := QuestionBank.AllQuestions.AnimalQuestions.MessageAnimalRunFast_EN},
                question{TheQuestion := QuestionBank.AllQuestions.AnimalQuestions.MessageAnimalJumpHigh_EN},
                question{TheQuestion := QuestionBank.AllQuestions.AnimalQuestions.MessageAnimalSwim_EN},
                question{TheQuestion := QuestionBank.AllQuestions.AnimalQuestions.MessageAnimalGlow_EN}
            }
        }}
        set AnimalQuestions += array{category{
            Style := card_style.Animal
            Name := "Combat"
            NameMessage := QuestionBank.AllCategories.MessageCombat_EN
            Questions := array{
                question{TheQuestion := QuestionBank.AllQuestions.AnimalQuestions.MessageAnimalAttacks_EN},
                question{TheQuestion := QuestionBank.AllQuestions.AnimalQuestions.MessageAnimalSpecialAttack_EN},
                question{TheQuestion := QuestionBank.AllQuestions.AnimalQuestions.MessageAnimalTamed_EN},
                question{TheQuestion := QuestionBank.AllQuestions.AnimalQuestions.MessageAnimalHostile_EN},
                question{TheQuestion := QuestionBank.AllQuestions.AnimalQuestions.MessageAnimalHealAbility_EN},
                question{TheQuestion := QuestionBank.AllQuestions.AnimalQuestions.MessageAnimalRunAway_EN}
            }
        }}
        set AnimalQuestions += array{category{
            Style := card_style.Animal
            Name := "Other"
            NameMessage := QuestionBank.AllCategories.MessageOther_EN
            Questions := array{
                question{TheQuestion := QuestionBank.AllQuestions.AnimalQuestions.MessageAnimalSpawnRegion_EN},
                question{TheQuestion := QuestionBank.AllQuestions.AnimalQuestions.MessageAnimalSpawnGroups_EN},
                question{TheQuestion := QuestionBank.AllQuestions.AnimalQuestions.MessageAnimalAttackAnimals_EN},
                question{TheQuestion := QuestionBank.AllQuestions.AnimalQuestions.MessageAnimalEggs_EN},
                question{TheQuestion := QuestionBank.AllQuestions.AnimalQuestions.MessageAnimalCanBePickedUp_EN},
                question{TheQuestion := QuestionBank.AllQuestions.AnimalQuestions.MessageAnimalConsideredFood_EN}
            }
        }}
    
    


    var UtilitiesQuestions : []category = array{}

    InitUtilitiesQuestions()<suspends>:void=
        

        set UtilitiesQuestions = array{category{
            Style := card_style.Utilities
            Name := "Functionality 1/2"
            NameMessage := QuestionBank.AllCategories.MessageFunctionalityOneOfTwo_EN
            Questions := array{
                question{TheQuestion := QuestionBank.AllQuestions.ItemQuestions.MessageItemThowable_EN},
                question{TheQuestion := QuestionBank.AllQuestions.ItemQuestions.MessageItemDamage_EN},
                question{TheQuestion := QuestionBank.AllQuestions.ItemQuestions.MessageItemMobility_EN},
                question{TheQuestion := QuestionBank.AllQuestions.ItemQuestions.MessageItemHealthShield_EN},
                question{TheQuestion := QuestionBank.AllQuestions.ItemQuestions.MessageItemStructures_EN},
                question{TheQuestion := QuestionBank.AllQuestions.ItemQuestions.MessageItemAdvantage_EN}
            }
        }}

        set UtilitiesQuestions += array{category{
            Style := card_style.Utilities
            Name := "Functionality 2/2"
            NameMessage := QuestionBank.AllCategories.MessageFunctionalityTwoOfTwo_EN
            Questions := array{
                question{TheQuestion := QuestionBank.AllQuestions.ItemQuestions.MessageItemBlockDamage_EN},
                question{TheQuestion := QuestionBank.AllQuestions.ItemQuestions.MessageItemHideInBehind_EN},
                question{TheQuestion := QuestionBank.AllQuestions.ItemQuestions.MessageItemFreezeEnemies_EN},
                question{TheQuestion := QuestionBank.AllQuestions.ItemQuestions.MessageItemRift_EN},
                question{TheQuestion := QuestionBank.AllQuestions.ItemQuestions.MessageItemUsedWithVehicle_EN},
                question{TheQuestion := QuestionBank.AllQuestions.ItemQuestions.MessageItemLaunchFlight_EN},
                question{TheQuestion := QuestionBank.AllQuestions.ItemQuestions.MessageItemFire_EN},
                question{TheQuestion := QuestionBank.AllQuestions.ItemQuestions.MessageItemMusic_EN}


            }
        }}

        set UtilitiesQuestions += array{category{
            Style := card_style.Utilities
            Name := "Usage"
            NameMessage := QuestionBank.AllCategories.MessageUsage_EN
            Questions := array{
                question{TheQuestion := QuestionBank.AllQuestions.ItemQuestions.MessageItemInstantly_EN},
                question{TheQuestion := QuestionBank.AllQuestions.ItemQuestions.MessageItemCooldown_EN},
                question{TheQuestion := QuestionBank.AllQuestions.ItemQuestions.MessageItemMultipleUses_EN},
                question{TheQuestion := QuestionBank.AllQuestions.ItemQuestions.MessageItemAimed_EN},
                question{TheQuestion := QuestionBank.AllQuestions.ItemQuestions.MessageItemAreaRadius_EN},
                question{TheQuestion := QuestionBank.AllQuestions.ItemQuestions.MessageItemActivated_EN},
                question{TheQuestion := QuestionBank.AllQuestions.ItemQuestions.MessageItemLimitedDuration_EN}
            }
        }}
        set UtilitiesQuestions += array{category{
            Style := card_style.Utilities
            Name := "Appearance"
            NameMessage := QuestionBank.AllCategories.MessageAppearance_EN
            Questions := array{
                question{TheQuestion := QuestionBank.AllQuestions.ItemQuestions.MessageItemCanister_EN},
                question{TheQuestion := QuestionBank.AllQuestions.ItemQuestions.MessageItemLights_EN},
                question{TheQuestion := QuestionBank.AllQuestions.ItemQuestions.MessageItemHighTech_EN},
                question{TheQuestion := QuestionBank.AllQuestions.ItemQuestions.MessageItemMilitary_EN},
                question{TheQuestion := QuestionBank.AllQuestions.ItemQuestions.MessageItemFutruristic_EN},
                question{TheQuestion := QuestionBank.AllQuestions.ItemQuestions.MessageItemDuctTape_EN},
                question{TheQuestion := QuestionBank.AllQuestions.ItemQuestions.MessageItemGlass_EN},
                question{TheQuestion := QuestionBank.AllQuestions.ItemQuestions.MessageItemHomemade_EN},
                question{TheQuestion := QuestionBank.AllQuestions.ItemQuestions.MessageItemScreen_EN}
            }
        }}


        set UtilitiesQuestions += array{category{
            Style := card_style.Utilities
            Name := "Stack Size"
            NameMessage := QuestionBank.AllCategories.MessageStackSize_EN
            Questions := array{
                question{TheQuestion := QuestionBank.AllQuestions.ItemQuestions.MessageStackSizeOne_EN},
                question{TheQuestion := QuestionBank.AllQuestions.ItemQuestions.MessageStackSizeTwo_EN},
                question{TheQuestion := QuestionBank.AllQuestions.ItemQuestions.MessageStackSizeFour_EN},
                question{TheQuestion := QuestionBank.AllQuestions.ItemQuestions.MessageStackSizeFive_EN},
                question{TheQuestion := QuestionBank.AllQuestions.ItemQuestions.MessageStackSizeSix_EN},
                question{TheQuestion := QuestionBank.AllQuestions.ItemQuestions.MessageStackSizeNine_EN},
                question{TheQuestion := QuestionBank.AllQuestions.ItemQuestions.MessageStackSizeTen_EN},
                question{TheQuestion := QuestionBank.AllQuestions.ItemQuestions.MessageStackSizeNineHundredNinetyNine_EN}
            }
        }}

        set UtilitiesQuestions += array{category{
            Style := card_style.Utilities
            Name := "Rarity"
            NameMessage := QuestionBank.AllCategories.MessageRarity_EN
            Questions := array{
                question{TheQuestion := QuestionBank.AllQuestions.ItemQuestions.MessageItemCommonRarity_EN},
                question{TheQuestion := QuestionBank.AllQuestions.ItemQuestions.MessageItemUncommonRarity_EN},
                question{TheQuestion := QuestionBank.AllQuestions.ItemQuestions.MessageItemRareRarity_EN},
                question{TheQuestion := QuestionBank.AllQuestions.ItemQuestions.MessageItemEpicRarity_EN},
                question{TheQuestion := QuestionBank.AllQuestions.ItemQuestions.MessageItemLegendaryRarity_EN}
            }
        }}
        set UtilitiesQuestions += array{category{
            Style := card_style.Utilities
            Name := "Color"
            NameMessage := QuestionBank.AllCategories.MessageColors_EN
            Questions := array{
                question{TheQuestion := QuestionBank.AllQuestions.ItemQuestions.MessageItemBlue_EN},
                question{TheQuestion := QuestionBank.AllQuestions.ItemQuestions.MessageItemGreen_EN},
                question{TheQuestion := QuestionBank.AllQuestions.ItemQuestions.MessageItemPurple_EN},
                question{TheQuestion := QuestionBank.AllQuestions.ItemQuestions.MessageItemYellow_EN},
                question{TheQuestion := QuestionBank.AllQuestions.ItemQuestions.MessageItemRed_EN},
                question{TheQuestion := QuestionBank.AllQuestions.ItemQuestions.MessageItemWhite_EN},
                question{TheQuestion := QuestionBank.AllQuestions.ItemQuestions.MessageItemSilver_EN},
                question{TheQuestion := QuestionBank.AllQuestions.ItemQuestions.MessageItemOrange_EN}
            }
        }}


    var Marvel1Questions : []category = array{}

    InitMarvel1Questions()<suspends>:void=
        set Marvel1Questions = array{
            category{
                Style:=card_style.Marvel1,
                Name:="Super Powers",
                NameMessage := QuestionBank.AllCategories.MessageSuperPowers_EN,
                Questions:=array{
                    question{TheQuestion := QuestionBank.AllQuestions.SuperHeroQuestions.MessageMarvelSuperPowersStrength_EN},
                    question{TheQuestion := QuestionBank.AllQuestions.SuperHeroQuestions.MessageMarvelSuperPowersFly_EN},
                    question{TheQuestion := QuestionBank.AllQuestions.SuperHeroQuestions.MessageMarvelSuperPowersMindRead_EN},
                    question{TheQuestion := QuestionBank.AllQuestions.SuperHeroQuestions.MessageMarvelSuperPowersHeal_EN},
                    question{TheQuestion := QuestionBank.AllQuestions.SuperHeroQuestions.MessageMarvelSuperPowersShapeChange_EN},
                    question{TheQuestion := QuestionBank.AllQuestions.SuperHeroQuestions.MessageMarvelSuperPowersControlElements_EN},
                    question{TheQuestion := QuestionBank.AllQuestions.SuperHeroQuestions.MessageMarvelSuperPowersUseGadgets_EN},
                    question{TheQuestion := QuestionBank.AllQuestions.SuperHeroQuestions.MessageMarvelSuperPowersInvisibility_EN},
                    question{TheQuestion := QuestionBank.AllQuestions.SuperHeroQuestions.MessageMarvelSuperPowersMagic_EN}
                }
            }
        }
    
        set Marvel1Questions += array{
            category{
                Style:=card_style.Marvel1,
                Name:="Origins",
                NameMessage := QuestionBank.AllCategories.MessageOrigin_EN,
                Questions:=array{
                    question{TheQuestion := QuestionBank.AllQuestions.SuperHeroQuestions.MessageMarvelOriginsFromEarth_EN},
                    question{TheQuestion := QuestionBank.AllQuestions.SuperHeroQuestions.MessageMarvelOriginsFromSpace_EN},
                    question{TheQuestion := QuestionBank.AllQuestions.SuperHeroQuestions.MessageMarvelOriginsMutant_EN},
                    question{TheQuestion := QuestionBank.AllQuestions.SuperHeroQuestions.MessageMarvelOriginsAccidentPowers_EN},
                    question{TheQuestion := QuestionBank.AllQuestions.SuperHeroQuestions.MessageMarvelOriginsExperimentPowers_EN},
                    question{TheQuestion := QuestionBank.AllQuestions.SuperHeroQuestions.MessageMarvelOriginsGod_EN},
                    question{TheQuestion := QuestionBank.AllQuestions.CharacterQuestions.MessageCharacterSpeciesAlien_EN},
                    question{TheQuestion := QuestionBank.AllQuestions.CharacterQuestions.MessageCharacterSpeciesRobot_EN}
                }
            }
        }
    
        set Marvel1Questions += array{
            category{
                Style:=card_style.Marvel1,
                Name:="Teams and Enemies",
                NameMessage := QuestionBank.AllCategories.MessageTeamsEnemies_EN,
                Questions:=array{
                    question{TheQuestion := QuestionBank.AllQuestions.SuperHeroQuestions.MessageMarvelTeamAvenger_EN},
                    question{TheQuestion := QuestionBank.AllQuestions.SuperHeroQuestions.MessageMarvelTeamXMen_EN},
                    question{TheQuestion := QuestionBank.AllQuestions.SuperHeroQuestions.MessageMarvelTeamGuardians_EN},
                    question{TheQuestion := QuestionBank.AllQuestions.CharacterQuestions.MessageCharacterVillain_EN},
                    question{TheQuestion := QuestionBank.AllQuestions.SuperHeroQuestions.MessageMarvelAlignmentHero_EN},
                    question{TheQuestion := QuestionBank.AllQuestions.CharacterQuestions.MessageCharacterSwitchedSides_EN},
                    question{TheQuestion := QuestionBank.AllQuestions.SuperHeroQuestions.MessageMarvelEnemyOfSpiderMan_EN},
                    question{TheQuestion := QuestionBank.AllQuestions.SuperHeroQuestions.MessageMarvelEnemyOfAvengers_EN}
                }
            }
        }
    
        set Marvel1Questions += array{
            category{
                Style:=card_style.Marvel1,
                Name:="Costumes",
                NameMessage := QuestionBank.AllCategories.MessageCostumes_EN,
                Questions:=array{
                    question{TheQuestion := QuestionBank.AllQuestions.SuperHeroQuestions.MessageMarvelCostumeHelmet_EN},
                    question{TheQuestion := QuestionBank.AllQuestions.CharacterQuestions.MessageCharacterMask_EN},
                    question{TheQuestion := QuestionBank.AllQuestions.SuperHeroQuestions.MessageMarvelCostumeHood_EN},
                    question{TheQuestion := QuestionBank.AllQuestions.CharacterQuestions.MessageCharacterCape_EN},
                    question{TheQuestion := QuestionBank.AllQuestions.SuperHeroQuestions.MessageMarvelCostumeVisibleSkin_EN},
                    question{TheQuestion := QuestionBank.AllQuestions.SuperHeroQuestions.MessageMarvelCostumeMetal_EN},
                    question{TheQuestion := QuestionBank.AllQuestions.SuperHeroQuestions.MessageMarvelCostumeSymbol_EN},
                    question{TheQuestion := QuestionBank.AllQuestions.SuperHeroQuestions.MessageMarvelCostumeWings_EN}
                }
            }
        }
    
        set Marvel1Questions += array{
            category{
                Style:=card_style.Marvel1,
                Name:="Colors",
                NameMessage := QuestionBank.AllCategories.MessageColors_EN,
                Questions:=array{
                    question{TheQuestion := QuestionBank.AllQuestions.CharacterQuestions.MessageCharacterColorRed_EN},
                    question{TheQuestion := QuestionBank.AllQuestions.CharacterQuestions.MessageCharacterColorBlue_EN},
                    question{TheQuestion := QuestionBank.AllQuestions.CharacterQuestions.MessageCharacterColorGreen_EN},
                    question{TheQuestion := QuestionBank.AllQuestions.CharacterQuestions.MessageCharacterColorBlack_EN},
                    question{TheQuestion := QuestionBank.AllQuestions.CharacterQuestions.MessageCharacterColorWhite_EN},
                    question{TheQuestion := QuestionBank.AllQuestions.CharacterQuestions.MessageCharacterColorYellow_EN},
                    question{TheQuestion := QuestionBank.AllQuestions.CharacterQuestions.MessageCharacterColorOrange_EN},
                    question{TheQuestion := QuestionBank.AllQuestions.CharacterQuestions.MessageCharacterColorPurple_EN}
                }
            }
        }
    
        set Marvel1Questions += array{
            category{
                Style:=card_style.Marvel1,
                Name:="Fun Facts",
                NameMessage := QuestionBank.AllCategories.MessageFunFacts_EN,
                Questions:=array{
                    question{TheQuestion := QuestionBank.AllQuestions.SuperHeroQuestions.MessageMarvelFactBillionaire_EN},
                    question{TheQuestion := QuestionBank.AllQuestions.MovieTVQuestions.MessageMovieCharacterCareerScientist_EN},
                    question{TheQuestion := QuestionBank.AllQuestions.SuperHeroQuestions.MessageMarvelFactLeader_EN},
                    question{TheQuestion := QuestionBank.AllQuestions.CharacterQuestions.MessageCharacterSidekickOrPartner_EN},
                    question{TheQuestion := QuestionBank.AllQuestions.SuperHeroQuestions.MessageMarvelFactFromWakanda_EN},
                    question{TheQuestion := QuestionBank.AllQuestions.SuperHeroQuestions.MessageMarvelFactFoughtThanos_EN},
                    question{TheQuestion := QuestionBank.AllQuestions.SuperHeroQuestions.MessageMarvelFactFriendsWithGroot_EN},
                    question{TheQuestion := QuestionBank.AllQuestions.SuperHeroQuestions.MessageMarvelFactTimeTraveler_EN},
                    question{TheQuestion := QuestionBank.AllQuestions.SuperHeroQuestions.MessageMarvelFactWorksWithShield_EN},
                    question{TheQuestion := QuestionBank.AllQuestions.CharacterQuestions.MessageCharacterVillain_EN}
                }
            }
        }

        set Marvel1Questions += array{
            category{
                Style:=card_style.Marvel1,
                Name:="Weapons",
                NameMessage := QuestionBank.AllCategories.MessageWeapons_EN,
                Questions:=array{
                    question{TheQuestion := QuestionBank.AllQuestions.SuperHeroQuestions.MessageMarvelWeaponShield_EN},
                    question{TheQuestion := QuestionBank.AllQuestions.MovieTVQuestions.MessageMovieWeaponBow_EN},
                    question{TheQuestion := QuestionBank.AllQuestions.SuperHeroQuestions.MessageMarvelWeaponHammer_EN},
                    question{TheQuestion := QuestionBank.AllQuestions.MovieTVQuestions.MessageMovieWeaponSword_EN},
                    question{TheQuestion := QuestionBank.AllQuestions.MovieTVQuestions.MessageMovieWeaponGun_EN},
                    question{TheQuestion := QuestionBank.AllQuestions.SuperHeroQuestions.MessageMarvelWeaponStaff_EN},
                    question{TheQuestion := QuestionBank.AllQuestions.SuperHeroQuestions.MessageMarvelWeaponWhip_EN},
                    question{TheQuestion := QuestionBank.AllQuestions.MovieTVQuestions.MessageMovieWeaponLaser_EN},
                    question{TheQuestion := QuestionBank.AllQuestions.SuperHeroQuestions.MessageMarvelWeaponGauntlet_EN},
                    question{TheQuestion := QuestionBank.AllQuestions.MovieTVQuestions.MessageMovieWeaponBow_EN}
                }
            }
        }

        set Marvel1Questions += array{
            category{
                Style:=card_style.Marvel1,
                Name:="Other",
                NameMessage := QuestionBank.AllCategories.MessageOther_EN,
                Questions:=array{
                    question{TheQuestion := QuestionBank.AllQuestions.CharacterQuestions.MessageCharacterBeard_EN},
                    question{TheQuestion := QuestionBank.AllQuestions.SuperHeroQuestions.MessageMarvelOtherMoustache_EN},
                    question{TheQuestion := QuestionBank.AllQuestions.CharacterQuestions.MessageCharacterGoatee_EN},
                    question{TheQuestion := QuestionBank.AllQuestions.LockerBundleQuestions.MessageLockerOtherScars_EN},
                    question{TheQuestion := QuestionBank.AllQuestions.CharacterQuestions.MessageCharacterGenderMale_EN},
                    question{TheQuestion := QuestionBank.AllQuestions.CharacterQuestions.MessageCharacterGenderFemale_EN}
                }
            }
        }
        

    var Marvel2Questions : []category = array{}

    InitMarvel2Questions()<suspends>:void=
        set Marvel2Questions = array{
            category{
                Style:=card_style.Marvel2,
                Name:="Super Powers",
                NameMessage := QuestionBank.AllCategories.MessageSuperPowers_EN,
                Questions:=array{
                    question{TheQuestion := QuestionBank.AllQuestions.SuperHeroQuestions.MessageMarvelSuperPowersStrength_EN},
                    question{TheQuestion := QuestionBank.AllQuestions.SuperHeroQuestions.MessageMarvelSuperPowersFly_EN},
                    question{TheQuestion := QuestionBank.AllQuestions.SuperHeroQuestions.MessageMarvelSuperPowersMindRead_EN},
                    question{TheQuestion := QuestionBank.AllQuestions.SuperHeroQuestions.MessageMarvelSuperPowersHeal_EN},
                    question{TheQuestion := QuestionBank.AllQuestions.SuperHeroQuestions.MessageMarvelSuperPowersControlFire_EN},
                    question{TheQuestion := QuestionBank.AllQuestions.SuperHeroQuestions.MessageMarvelSuperPowersControlIce_EN},
                    question{TheQuestion := QuestionBank.AllQuestions.SuperHeroQuestions.MessageMarvelSuperPowersMagic_EN},
                    question{TheQuestion := QuestionBank.AllQuestions.SuperHeroQuestions.MessageMarvelSuperPowersStretch_EN},
                    question{TheQuestion := QuestionBank.AllQuestions.SuperHeroQuestions.MessageMarvelSuperPowersShrinkGrow_EN},
                    question{TheQuestion := QuestionBank.AllQuestions.SuperHeroQuestions.MessageMarvelSuperPowersControlMetal_EN}
                }
            }
        }
    
        set Marvel2Questions += array{
            category{
                Style:=card_style.Marvel2,
                Name:="Origins",
                NameMessage := QuestionBank.AllCategories.MessageOrigin_EN,
                Questions:=array{
                    question{TheQuestion := QuestionBank.AllQuestions.SuperHeroQuestions.MessageMarvelOriginsFromEarth_EN},
                    question{TheQuestion := QuestionBank.AllQuestions.SuperHeroQuestions.MessageMarvelOriginsFromSpace_EN},
                    question{TheQuestion := QuestionBank.AllQuestions.SuperHeroQuestions.MessageMarvelOriginsMutant_EN},
                    question{TheQuestion := QuestionBank.AllQuestions.SuperHeroQuestions.MessageMarvelOriginsAccidentPowers_EN},
                    question{TheQuestion := QuestionBank.AllQuestions.SuperHeroQuestions.MessageMarvelOriginsExperimentPowers_EN},
                    question{TheQuestion := QuestionBank.AllQuestions.SuperHeroQuestions.MessageMarvelOriginsGod_EN},
                    question{TheQuestion := QuestionBank.AllQuestions.CharacterQuestions.MessageCharacterSpeciesAlien_EN},
                    question{TheQuestion := QuestionBank.AllQuestions.SuperHeroQuestions.MessageMarvelOriginsFromAsgard_EN},
                    question{TheQuestion := QuestionBank.AllQuestions.SuperHeroQuestions.MessageMarvelOriginsDemigod_EN}
                }
            }
        }
    
        set Marvel2Questions += array{
            category{
                Style:=card_style.Marvel2,
                Name:="Teams and Enemies",
                NameMessage := QuestionBank.AllCategories.MessageTeamsEnemies_EN,
                Questions:=array{
                    question{TheQuestion := QuestionBank.AllQuestions.SuperHeroQuestions.MessageMarvelTeamAvenger_EN},
                    question{TheQuestion := QuestionBank.AllQuestions.SuperHeroQuestions.MessageMarvelTeamXMen_EN},
                    question{TheQuestion := QuestionBank.AllQuestions.SuperHeroQuestions.MessageMarvelTeamGuardians_EN},
                    question{TheQuestion := QuestionBank.AllQuestions.SuperHeroQuestions.MessageMarvelTeamFantasticFour_EN},
                    question{TheQuestion := QuestionBank.AllQuestions.CharacterQuestions.MessageCharacterVillain_EN},
                    question{TheQuestion := QuestionBank.AllQuestions.SuperHeroQuestions.MessageMarvelAlignmentHero_EN},
                    question{TheQuestion := QuestionBank.AllQuestions.CharacterQuestions.MessageCharacterSwitchedSides_EN},
                    question{TheQuestion := QuestionBank.AllQuestions.SuperHeroQuestions.MessageMarvelEnemyOfXMen_EN},
                    question{TheQuestion := QuestionBank.AllQuestions.SuperHeroQuestions.MessageMarvelEnemyOfSpiderMan_EN},
                    question{TheQuestion := QuestionBank.AllQuestions.SuperHeroQuestions.MessageMarvelEnemyOfAvengers_EN},
                    question{TheQuestion := QuestionBank.AllQuestions.SuperHeroQuestions.MessageMarvelEnemyOfFantasticFour_EN}
                }
            }
        }
    
        set Marvel2Questions += array{
            category{
                Style:=card_style.Marvel2,
                Name:="Costume",
                NameMessage := QuestionBank.AllCategories.MessageCostumes_EN,
                Questions:=array{
                    question{TheQuestion := QuestionBank.AllQuestions.SuperHeroQuestions.MessageMarvelCostumeHelmet_EN},
                    question{TheQuestion := QuestionBank.AllQuestions.CharacterQuestions.MessageCharacterMask_EN},
                    question{TheQuestion := QuestionBank.AllQuestions.SuperHeroQuestions.MessageMarvelCostumeHood_EN},
                    question{TheQuestion := QuestionBank.AllQuestions.CharacterQuestions.MessageCharacterCape_EN},
                    question{TheQuestion := QuestionBank.AllQuestions.SuperHeroQuestions.MessageMarvelCostumeVisibleSkin_EN},
                    question{TheQuestion := QuestionBank.AllQuestions.SuperHeroQuestions.MessageMarvelCostumeMetal_EN},
                    question{TheQuestion := QuestionBank.AllQuestions.SuperHeroQuestions.MessageMarvelCostumeSymbol_EN},
                    question{TheQuestion := QuestionBank.AllQuestions.CharacterQuestions.MessageCharacterArmor_EN}
                }
            }
        }
    
        set Marvel2Questions += array{
            category{
                Style:=card_style.Marvel2,
                Name:="Colors",
                NameMessage := QuestionBank.AllCategories.MessageColors_EN,
                Questions:=array{
                    question{TheQuestion := QuestionBank.AllQuestions.CharacterQuestions.MessageCharacterColorRed_EN},
                    question{TheQuestion := QuestionBank.AllQuestions.CharacterQuestions.MessageCharacterColorBlue_EN},
                    question{TheQuestion := QuestionBank.AllQuestions.CharacterQuestions.MessageCharacterColorGreen_EN},
                    question{TheQuestion := QuestionBank.AllQuestions.CharacterQuestions.MessageCharacterColorBlack_EN},
                    question{TheQuestion := QuestionBank.AllQuestions.CharacterQuestions.MessageCharacterColorWhite_EN},
                    question{TheQuestion := QuestionBank.AllQuestions.CharacterQuestions.MessageCharacterColorYellow_EN},
                    question{TheQuestion := QuestionBank.AllQuestions.CharacterQuestions.MessageCharacterColorOrange_EN},
                    question{TheQuestion := QuestionBank.AllQuestions.CharacterQuestions.MessageCharacterColorPurple_EN}
                }
            }
        }
    
        set Marvel2Questions += array{
            category{
                Style:=card_style.Marvel2,
                Name:="Fun Facts",
                NameMessage := QuestionBank.AllCategories.MessageFunFacts_EN,
                Questions:=array{
                    question{TheQuestion := QuestionBank.AllQuestions.SuperHeroQuestions.MessageMarvelFactBillionaire_EN},
                    question{TheQuestion := QuestionBank.AllQuestions.MovieTVQuestions.MessageMovieCharacterCareerScientist_EN},
                    question{TheQuestion := QuestionBank.AllQuestions.SuperHeroQuestions.MessageMarvelFactLeader_EN},
                    question{TheQuestion := QuestionBank.AllQuestions.CharacterQuestions.MessageCharacterSidekickOrPartner_EN},
                    question{TheQuestion := QuestionBank.AllQuestions.SuperHeroQuestions.MessageMarvelFactFromWakanda_EN},
                    question{TheQuestion := QuestionBank.AllQuestions.SuperHeroQuestions.MessageMarvelFactFoughtThanos_EN},
                    question{TheQuestion := QuestionBank.AllQuestions.SuperHeroQuestions.MessageMarvelFactFriendsWithGroot_EN},
                    question{TheQuestion := QuestionBank.AllQuestions.SuperHeroQuestions.MessageMarvelFactTimeTraveler_EN},
                    question{TheQuestion := QuestionBank.AllQuestions.SuperHeroQuestions.MessageMarvelFactWorksWithShield_EN},
                    question{TheQuestion := QuestionBank.AllQuestions.CharacterQuestions.MessageCharacterVillain_EN}
                }
            }
        }

        set Marvel2Questions += array{
            category{
                Style:=card_style.Marvel2,
                Name:="Weapons",
                NameMessage := QuestionBank.AllCategories.MessageWeapons_EN,
                Questions:=array{
                    question{TheQuestion := QuestionBank.AllQuestions.SuperHeroQuestions.MessageMarvelWeaponShield_EN},
                    question{TheQuestion := QuestionBank.AllQuestions.MovieTVQuestions.MessageMovieWeaponBow_EN},
                    question{TheQuestion := QuestionBank.AllQuestions.SuperHeroQuestions.MessageMarvelWeaponHammer_EN},
                    question{TheQuestion := QuestionBank.AllQuestions.MovieTVQuestions.MessageMovieWeaponSword_EN},
                    question{TheQuestion := QuestionBank.AllQuestions.MovieTVQuestions.MessageMovieWeaponGun_EN},
                    question{TheQuestion := QuestionBank.AllQuestions.SuperHeroQuestions.MessageMarvelWeaponStaff_EN},
                    question{TheQuestion := QuestionBank.AllQuestions.SuperHeroQuestions.MessageMarvelWeaponWhip_EN},
                    question{TheQuestion := QuestionBank.AllQuestions.MovieTVQuestions.MessageMovieWeaponLaser_EN},
                    question{TheQuestion := QuestionBank.AllQuestions.SuperHeroQuestions.MessageMarvelWeaponGauntlet_EN},
                    question{TheQuestion := QuestionBank.AllQuestions.MovieTVQuestions.MessageMovieWeaponBow_EN}
                }
            }
        }

        set Marvel2Questions += array{
            category{
                Style:=card_style.Marvel2,
                Name:="Other",
                NameMessage := QuestionBank.AllCategories.MessageOther_EN,
                Questions:=array{
                    question{TheQuestion := QuestionBank.AllQuestions.CharacterQuestions.MessageCharacterBeard_EN},
                    question{TheQuestion := QuestionBank.AllQuestions.SuperHeroQuestions.MessageMarvelOtherMoustache_EN},
                    question{TheQuestion := QuestionBank.AllQuestions.CharacterQuestions.MessageCharacterGoatee_EN},
                    question{TheQuestion := QuestionBank.AllQuestions.CharacterQuestions.MessageCharacterBeard_EN},
                    question{TheQuestion := QuestionBank.AllQuestions.CharacterQuestions.MessageCharacterGenderMale_EN},
                    question{TheQuestion := QuestionBank.AllQuestions.CharacterQuestions.MessageCharacterGenderFemale_EN}
                }
            }
        }
    
    
    
    var POIQuestions : []category = array {}

    InitPOIQuestions()<suspends>:void=
        block:

    #     set POIQuestions = array{category{
    #         Style:=card_style.POI,
    #         Name:="Location 🌍",
    #         NameMessage := QuestionBank.AllCategories.MessageLocation_EN,
    #         Questions:=array{
    #             question{TheQuestion:= "Is your POI located in a suburban area? 🏘"},
    #             question{TheQuestion:= "Is your POI located in an urban area? 🏙"},
    #             question{TheQuestion:= "Is your POI located in a rural or natural area? 🌳"},
    #             question{TheQuestion:= "Does your POI have a significant amount of water? 🌊"},
    #             question{TheQuestion:= "Does your POI have high elevation or mountains? 🗻"},
    #             question{TheQuestion:= "Does your POI have industrial elements? 🏭"},
    #             question{TheQuestion:= "Is your POI known for its loot? 💰"},
    #             question{TheQuestion:= "Has your POI been present in multiple seasons? 🔄"}
    #         }
    #     }}
    #     set POIQuestions += array{category{
    #         Style:=card_style.POI,
    #         Name:="Historical Significance 📜",
    #         NameMessage := QuestionBank.AllCategories.MessageHistoricalSignificance_EN,
    #         Questions:=array{
    #             question{TheQuestion:= "Has your POI undergone significant changes during its existence? 🔄"},
    #             question{TheQuestion:= "Has your POI been a focal point for in-game events? 🎆"},
    #             question{TheQuestion:= "Has your POI been removed and then returned in a later season? 🔄"},
    #             question{TheQuestion:= "Is your POI one of the original locations from Fortnite's launch? 🚀"},
    #             question{TheQuestion:= "Has your POI been featured in promotional materials or trailers? 🎥"},
    #             question{TheQuestion:= "Has your POI been a popular landing spot for players? 🎯"},
    #             question{TheQuestion:= "Has your POI been associated with a specific Fortnite character or storyline? 📖"},
    #             question{TheQuestion:= "Has your POI been a part of a crossover event with another franchise? 🤝"}
    #         }
    #     }}
    #     set POIQuestions += array{category{
    #         Style:=card_style.POI,
    #         Name:="Chapter and Season 📅",
    #         NameMessage := QuestionBank.AllCategories.MessageChapterSeason_EN,
    #         Questions:=array{
    #             question{TheQuestion:= "Did your POI first appear in Chapter 1? 📖"},
    #             question{TheQuestion:= "Did your POI first appear in Chapter 2? 📖"},
    #             question{TheQuestion:= "Did your POI first appear in Chapter 3? 📖"},
    #             question{TheQuestion:= "Did your POI first appear in Chapter 4? 📖"},
    #             question{TheQuestion:= "Has your POI been present in all seasons of its debut chapter? 🔄"},
    #             question{TheQuestion:= "Has your POI been removed or replaced in a later season? ❌"},
    #             question{TheQuestion:= "Has your POI been significantly altered by a new season's theme? 🔄"},
    #             question{TheQuestion:= "Has your POI been a part of a season's main storyline or questline? 📜"},
    #             question{TheQuestion:= "Has your POI been a part of a season's secret or bonus content? 🤫"}
    #         }
    #     }}
    #     set POIQuestions += array{category{
    #         Style:=card_style.POI,
    #         Name:="Gameplay and Strategy 🎮",
    #         NameMessage := QuestionBank.AllCategories.MessageGameplayStrategy_EN,
    #         Questions:=array{
    #             question{TheQuestion:= "Is your POI known for its high-risk, high-reward gameplay? 🎲"},
    #             question{TheQuestion:= "Is your POI a popular spot for competitive or tournament play? 🏆"},
    #             question{TheQuestion:= "Does your POI offer unique gameplay mechanics or features? 🕹"},
    #             question{TheQuestion:= "Is your POI a common spot for completing challenges or quests? 🎯"},
    #             question{TheQuestion:= "Is your POI a preferred landing spot for solo, duo, or squad play? 👥"},
    #             question{TheQuestion:= "Does your POI offer good opportunities for gathering materials? ⛏"},
    #             question{TheQuestion:= "Does your POI have good defensive or strategic positions? 🛡"},
    #             question{TheQuestion:= "Is your POI a common spot for end-game scenarios or final circles? 🎯"}
    #         }
    #     }}
    #    set POIQuestions += array{category{
    #         Style:=card_style.POI,
    #         Name:="Lore and Theme 📚",
    #         NameMessage := QuestionBank.AllCategories.MessageLoreTheme_EN,
    #         Questions:=array{
    #             question{TheQuestion:= "Does your POI have a distinct theme or aesthetic? 🎨"},
    #             question{TheQuestion:= "Does your POI have a backstory or lore associated with it? 📖"},
    #             question{TheQuestion:= "Has your POI been a part of Fortnite's overarching storyline? 📚"},
    #             question{TheQuestion:= "Has your POI been associated with a specific Fortnite character or group? 👥"},
    #             question{TheQuestion:= "Has your POI been a part of a Fortnite mystery or secret? 🕵️"},
    #             question{TheQuestion:= "Does your POI have a unique or iconic structure or landmark? 🏰"},
    #             question{TheQuestion:= "Has your POI been influenced by a real-world location or culture? 🌍"},
    #             question{TheQuestion:= "Does your POI have a unique or memorable name? 📝"}
    #         }
    #     }}
                    

    var LegendsQuestions : []category = array {}

    InitLegendsQuestions()<suspends>:void=
    
        set LegendsQuestions = array{category{
            Style:=card_style.Legends,
            Name:="Game Genre 🎮",
            NameMessage := QuestionBank.AllCategories.MessageGameGenre_EN,
            Questions:=array{
                question{TheQuestion := QuestionBank.AllQuestions.GamingLegendsQuestions.MessageLegendFightingGame_EN},
                question{TheQuestion := QuestionBank.AllQuestions.GamingLegendsQuestions.MessageLegendRPGGame_EN},
                question{TheQuestion := QuestionBank.AllQuestions.GamingLegendsQuestions.MessageLegendSurvialHorrorGame_EN},
                question{TheQuestion := QuestionBank.AllQuestions.GamingLegendsQuestions.MessageLegendActionAdventureGame_EN},
                question{TheQuestion := QuestionBank.AllQuestions.GamingLegendsQuestions.MessageLegendFPSGame_EN},
                question{TheQuestion := QuestionBank.AllQuestions.GamingLegendsQuestions.MessageLegendSimulationGame_EN},
                question{TheQuestion := QuestionBank.AllQuestions.GamingLegendsQuestions.MessageLegendPlatformerGame_EN},
                question{TheQuestion := QuestionBank.AllQuestions.GamingLegendsQuestions.MessageLegendMOBA_EN}
            }
        }}
        set LegendsQuestions += array{category{
            Style:=card_style.Legends,
            Name:="Character Traits 🦸‍♂️",
            NameMessage := QuestionBank.AllCategories.MessageCharacterTraits_EN,
            Questions:=array{
                question{TheQuestion := QuestionBank.AllQuestions.CharacterQuestions.MessageCharacterGenderMale_EN},
                question{TheQuestion := QuestionBank.AllQuestions.CharacterQuestions.MessageCharacterGenderFemale_EN},
                question{TheQuestion := QuestionBank.AllQuestions.GamingLegendsQuestions.MessageLegendStrength_EN},
                question{TheQuestion := QuestionBank.AllQuestions.GamingLegendsQuestions.MessageLegendIntelligence_EN},
                question{TheQuestion := QuestionBank.AllQuestions.GamingLegendsQuestions.MessageLegendSupernatural_EN},
                question{TheQuestion := QuestionBank.AllQuestions.GamingLegendsQuestions.MessageLegendProtagonist_EN},
                question{TheQuestion := QuestionBank.AllQuestions.GamingLegendsQuestions.MessageLegendIconicWeaponTool_EN},
                question{TheQuestion := QuestionBank.AllQuestions.MovieTVQuestions.MessageMovieTraitCatchphrase_EN},
                question{TheQuestion := QuestionBank.AllQuestions.GamingLegendsQuestions.MessageLegendNotableRival_EN}
            }
        }}
        set LegendsQuestions += array{category{
            Style:=card_style.Legends,
            Name:="Game Setting and World 🌍",
            NameMessage := QuestionBank.AllCategories.MessageGameSettingWorld_EN,
            Questions:=array{
                question{TheQuestion := QuestionBank.AllQuestions.GamingLegendsQuestions.MessageLegendPostApocalyptic_EN},
                question{TheQuestion := QuestionBank.AllQuestions.GamingLegendsQuestions.MessageLegendHistorical_EN},
                question{TheQuestion := QuestionBank.AllQuestions.GamingLegendsQuestions.MessageLegendMagicMythical_EN},
                question{TheQuestion := QuestionBank.AllQuestions.GamingLegendsQuestions.MessageLegendSpace_EN},
                question{TheQuestion := QuestionBank.AllQuestions.GamingLegendsQuestions.MessageLegendEarthLike_EN},
                question{TheQuestion := QuestionBank.AllQuestions.GamingLegendsQuestions.MessageLegendMultipleRealmsDimensions_EN},
                question{TheQuestion := QuestionBank.AllQuestions.GamingLegendsQuestions.MessageLegendGameKnownForLores_EN}
            }
        }}
        set LegendsQuestions += array{category{
            Style:=card_style.Legends,
            Name:="Game Mechanics and Features",
            NameMessage := QuestionBank.AllCategories.MessageGameMechanicsFeatures_EN,
            Questions:=array{
                question{TheQuestion := QuestionBank.AllQuestions.GamingLegendsQuestions.MessageLegendGameCraftingOrBuilding_EN},
                question{TheQuestion := QuestionBank.AllQuestions.GamingLegendsQuestions.MessageLegendGameStealth_EN},
                question{TheQuestion := QuestionBank.AllQuestions.GamingLegendsQuestions.MessageLegendGameCoop_EN},
                question{TheQuestion := QuestionBank.AllQuestions.GamingLegendsQuestions.MessageLegendGameMultipleEndings_EN},
                question{TheQuestion := QuestionBank.AllQuestions.GamingLegendsQuestions.MessageLegendGameSkillTree_EN},
                question{TheQuestion := QuestionBank.AllQuestions.GamingLegendsQuestions.MessageLegendGameUniqueMovement_EN},
                question{TheQuestion := QuestionBank.AllQuestions.GamingLegendsQuestions.MessageLegendGameMoralitySystem_EN},
                question{TheQuestion := QuestionBank.AllQuestions.GamingLegendsQuestions.MessageLegendGameDifficult_EN}
            }
        }}
        set LegendsQuestions += array{category{
            Style:=card_style.Legends,
            Name:="Game Legacy and Popularity 🌟",
            NameMessage := QuestionBank.AllCategories.MessageGameLegacyPopularity_EN,
            Questions:=array{
                question{TheQuestion := QuestionBank.AllQuestions.GamingLegendsQuestions.MessageLegendGameSequel_EN},
                question{TheQuestion := QuestionBank.AllQuestions.GamingLegendsQuestions.MessageLegendGameTurnedMovie_EN},
                question{TheQuestion := QuestionBank.AllQuestions.GamingLegendsQuestions.MessageLegendGameTurnedShow_EN},
                question{TheQuestion := QuestionBank.AllQuestions.GamingLegendsQuestions.MessageLegendEsportsScene_EN},
                question{TheQuestion := QuestionBank.AllQuestions.GamingLegendsQuestions.MessageLegendGameAwards_EN},
                question{TheQuestion := QuestionBank.AllQuestions.GamingLegendsQuestions.MessageLegendTenYearsPlus_EN},
                question{TheQuestion := QuestionBank.AllQuestions.GamingLegendsQuestions.MessageLegendGameMusic_EN},
                question{TheQuestion := QuestionBank.AllQuestions.GamingLegendsQuestions.MessageLegendInGameEvents_EN}
            }
        }}
        set LegendsQuestions += array{category{
            Style:=card_style.Legends,
            Name:="Color",
            NameMessage := QuestionBank.AllCategories.MessageColors_EN,
            Questions:=array{
                question{TheQuestion := QuestionBank.AllQuestions.CharacterQuestions.MessageCharacterColorRed_EN},
                question{TheQuestion := QuestionBank.AllQuestions.CharacterQuestions.MessageCharacterColorBlue_EN},
                question{TheQuestion := QuestionBank.AllQuestions.CharacterQuestions.MessageCharacterColorGreen_EN},
                question{TheQuestion := QuestionBank.AllQuestions.CharacterQuestions.MessageCharacterColorYellow_EN},
                question{TheQuestion := QuestionBank.AllQuestions.CharacterQuestions.MessageCharacterColorOrange_EN},
                question{TheQuestion := QuestionBank.AllQuestions.CharacterQuestions.MessageCharacterColorBrown_EN},
                question{TheQuestion := QuestionBank.AllQuestions.CharacterQuestions.MessageCharacterColorBlack_EN},
                question{TheQuestion := QuestionBank.AllQuestions.CharacterQuestions.MessageCharacterColorWhite_EN}
            }
        }}
    
    

    var FoodQuestions : []category = array{}

    InitFoodQuestions()<suspends>:void=
        set FoodQuestions = array{category{
            Style:=card_style.Food,
            Name:="Item Effects",
            NameMessage := QuestionBank.AllCategories.MessageItemEffects_EN,
            Questions:=array{
                question{TheQuestion := QuestionBank.AllQuestions.ItemQuestions.MessageItemHealth_EN},
                question{TheQuestion := QuestionBank.AllQuestions.ItemQuestions.MessageItemShield_EN},
                question{TheQuestion := QuestionBank.AllQuestions.ItemQuestions.MessageItemHealthShield_EN},
                question{TheQuestion := QuestionBank.AllQuestions.ItemQuestions.MessageItemSpeedBoost_EN},
                question{TheQuestion := QuestionBank.AllQuestions.ItemQuestions.MessageItemJumpHigher_EN},
                question{TheQuestion := QuestionBank.AllQuestions.ItemQuestions.MessageItemVision_EN}

            }
        }}

        set FoodQuestions += array{category{
            Style:=card_style.Food,
            Name:="Item Appearance and Type 🍎",
            NameMessage := QuestionBank.AllCategories.MessageItemAppearanceType_EN,
            Questions:=array{
                question{TheQuestion := QuestionBank.AllQuestions.ItemQuestions.MessageItemFruit_EN},
                question{TheQuestion := QuestionBank.AllQuestions.ItemQuestions.MessageItemVegetable_EN},
                question{TheQuestion := QuestionBank.AllQuestions.ItemQuestions.MessageItemCandy_EN},
                question{TheQuestion := QuestionBank.AllQuestions.ItemQuestions.MessageItemRealLifeFood_EN},
                question{TheQuestion := QuestionBank.AllQuestions.ItemQuestions.MessageItemSpecialEventFood_EN},
                question{TheQuestion := QuestionBank.AllQuestions.ItemQuestions.MessageItemBushPlantForaged_EN},
                question{TheQuestion := QuestionBank.AllQuestions.ItemQuestions.MessageItemMushroom_EN}
            }
        }}

        set FoodQuestions += array{category{
            Style:=card_style.Food,
            Name:="Location and Availability 🌍",
            NameMessage := QuestionBank.AllCategories.MessageLocations_EN,
            Questions:=array{
                question{TheQuestion := QuestionBank.AllQuestions.ItemQuestions.MessageItemForestedAreas_EN},
                question{TheQuestion := QuestionBank.AllQuestions.ItemQuestions.MessageItemWaterNearby_EN},
                question{TheQuestion := QuestionBank.AllQuestions.ItemQuestions.MessageItemSeasonEventExclusive_EN},
                question{TheQuestion := QuestionBank.AllQuestions.ItemQuestions.MessageItemUrbanAreas_EN},
                question{TheQuestion := QuestionBank.AllQuestions.ItemQuestions.MessageItemSpecificPOI_EN},
                question{TheQuestion := QuestionBank.AllQuestions.ItemQuestions.MessageItemAllOverMap_EN},
                question{TheQuestion := QuestionBank.AllQuestions.ItemQuestions.MessageItemGroupsClusters_EN},
                question{TheQuestion := QuestionBank.AllQuestions.ItemQuestions.MessageItemRareHardToFind_EN}
            }
        }}
        set FoodQuestions += array{category{
            Style:=card_style.Food,
            Name:="General Knowledge 🧠",
            NameMessage := QuestionBank.AllCategories.MessageGeneralKnowledge_EN,
            Questions:=array{
                question{TheQuestion := QuestionBank.AllQuestions.ItemQuestions.MessageItemEdibleInRealLife_EN},
                question{TheQuestion := QuestionBank.AllQuestions.ItemQuestions.MessageItemMultipleSeasons_EN},
                question{TheQuestion := QuestionBank.AllQuestions.ItemQuestions.MessageItemVaulted_EN},
                question{TheQuestion := QuestionBank.AllQuestions.ItemQuestions.MessageItemIconic_EN},
                question{TheQuestion := QuestionBank.AllQuestions.ItemQuestions.MessageItemCompetitive_EN},
                question{TheQuestion := QuestionBank.AllQuestions.ItemQuestions.MessageItemOverlooked_EN},
                question{TheQuestion := QuestionBank.AllQuestions.ItemQuestions.MessageItemRecentlyAdded_EN}
            }
        }}
        set FoodQuestions += array{category{
            Style:=card_style.Food,
            Name:="Colors 🌈",
            NameMessage := QuestionBank.AllCategories.MessageColors_EN,
            Questions:=array{
                question{TheQuestion := QuestionBank.AllQuestions.ItemQuestions.MessageItemRed_EN},
                question{TheQuestion := QuestionBank.AllQuestions.ItemQuestions.MessageItemBlue_EN},
                question{TheQuestion := QuestionBank.AllQuestions.ItemQuestions.MessageItemGreen_EN},
                question{TheQuestion := QuestionBank.AllQuestions.ItemQuestions.MessageItemYellow_EN},
                question{TheQuestion := QuestionBank.AllQuestions.ItemQuestions.MessageItemPurple_EN},
                question{TheQuestion := QuestionBank.AllQuestions.ItemQuestions.MessageItemOrange_EN},
                question{TheQuestion := QuestionBank.AllQuestions.ItemQuestions.MessageItemPink_EN},
                question{TheQuestion := QuestionBank.AllQuestions.ItemQuestions.MessageItemBlack_EN},
                question{TheQuestion := QuestionBank.AllQuestions.ItemQuestions.MessageItemWhite_EN}
            }
        }}
    
    

    var StarQuestions : []category = array{}

    InitStarQuestions()<suspends>:void=
        set StarQuestions = array{category{
            Style:=card_style.Star,
            Name:="Galactic Allegiances 🌌",
            NameMessage := QuestionBank.AllCategories.MessageAlleigance_EN,
            Questions:=array{
                question{TheQuestion := QuestionBank.AllQuestions.StarWarsQuestions.MessageStarAllegianceJediOrder_EN},
                question{TheQuestion := QuestionBank.AllQuestions.StarWarsQuestions.MessageStarAllegianceSith_EN},
                question{TheQuestion := QuestionBank.AllQuestions.StarWarsQuestions.MessageStarAllegianceRebelAlliance_EN},
                question{TheQuestion := QuestionBank.AllQuestions.StarWarsQuestions.MessageStarAllegianceGalacticEmpire_EN},
                question{TheQuestion := QuestionBank.AllQuestions.StarWarsQuestions.MessageStarAllegianceMandalorian_EN},
                question{TheQuestion := QuestionBank.AllQuestions.StarWarsQuestions.MessageStarPrequelTrilogy_EN},
                question{TheQuestion := QuestionBank.AllQuestions.StarWarsQuestions.MessageStarOriginalTrilogy_EN},
                question{TheQuestion := QuestionBank.AllQuestions.StarWarsQuestions.MessageStarSequelTrilogy_EN}
            }
        }}
        set StarQuestions += array{category{
            Style:=card_style.Star,
            Name:="Physical Traits 🌠",
            NameMessage := QuestionBank.AllCategories.MessagePhysicalTraits_EN,
            Questions:=array{
                question{TheQuestion := QuestionBank.AllQuestions.CharacterQuestions.MessageCharacterSpeciesHuman_EN},
                question{TheQuestion := QuestionBank.AllQuestions.StarWarsQuestions.MessageStarPhysicalAppearanceClone_EN},
                question{TheQuestion := QuestionBank.AllQuestions.StarWarsQuestions.MessageStarPhysicalAppearanceCybernetic_EN},
                question{TheQuestion := QuestionBank.AllQuestions.CharacterQuestions.MessageCharacterGenderMale_EN},
                question{TheQuestion := QuestionBank.AllQuestions.CharacterQuestions.MessageCharacterGenderFemale_EN},
                question{TheQuestion := QuestionBank.AllQuestions.StarWarsQuestions.MessageStarPhysicalAppearanceWookiee_EN},
                question{TheQuestion := QuestionBank.AllQuestions.StarWarsQuestions.MessageStarPhysicalLightsaber_EN},
                question{TheQuestion := QuestionBank.AllQuestions.CharacterQuestions.MessageCharacterHelmet_EN}
            }
        }}
        set StarQuestions += array{category{
            Style:=card_style.Star,
            Name:="Facial Features 🌟",
            NameMessage := QuestionBank.AllCategories.MessageFacialAttributes_EN,
            Questions:=array{
                question{TheQuestion := QuestionBank.AllQuestions.CharacterQuestions.MessageCharacterFacialHair_EN},
                question{TheQuestion := QuestionBank.AllQuestions.CharacterQuestions.MessageCharacterMask_EN},
                question{TheQuestion := QuestionBank.AllQuestions.WinterQuestions.MessageWinterFacialVisible_EN},
                question{TheQuestion := QuestionBank.AllQuestions.StarWarsQuestions.MessageStarFacialMarkings_EN},
                question{TheQuestion := QuestionBank.AllQuestions.CharacterQuestions.MessageCharacterTattoo_EN},
                question{TheQuestion := QuestionBank.AllQuestions.StarWarsQuestions.MessageStarPhysicalAppearanceFur_EN}
            }
        }}


        set StarQuestions += array{category{
            Style:=card_style.Star,
            Name:="Galactic Adventures 🚀",
            NameMessage := QuestionBank.AllCategories.MessageGalacticAdventures_EN,
            Questions:=array{
                question{TheQuestion := QuestionBank.AllQuestions.StarWarsQuestions.MessageStarGalacticAdventurePilotedMillenniumFalcon_EN},
                question{TheQuestion := QuestionBank.AllQuestions.StarWarsQuestions.MessageStarGalacticAdventureDagobah_EN},
                question{TheQuestion := QuestionBank.AllQuestions.StarWarsQuestions.MessageStarGalacticAdventureCarbonite_EN},
                question{TheQuestion := QuestionBank.AllQuestions.StarWarsQuestions.MessageStarGalacticAdventureTrainedJedi_EN},
                question{TheQuestion := QuestionBank.AllQuestions.StarWarsQuestions.MessageStarGalacticAdventureBountyHunter_EN}
            }
        }}
        set StarQuestions += array{category{
            Style:=card_style.Star,
            Name:="Mystical Forces ✨",
            NameMessage := QuestionBank.AllCategories.MessageMysticalForces_EN,
            Questions:=array{
                question{TheQuestion := QuestionBank.AllQuestions.StarWarsQuestions.MessageStarForceSensitive_EN},
                question{TheQuestion := QuestionBank.AllQuestions.StarWarsQuestions.MessageStarForceYoda_EN},
                question{TheQuestion := QuestionBank.AllQuestions.StarWarsQuestions.MessageStarForceDarkSide_EN},
                question{TheQuestion := QuestionBank.AllQuestions.StarWarsQuestions.MessageStarForceJediCouncil_EN},
                question{TheQuestion := QuestionBank.AllQuestions.StarWarsQuestions.MessageStarForceBuildLightsaber_EN},
                question{TheQuestion := QuestionBank.AllQuestions.StarWarsQuestions.MessageStarForceVisions_EN},
                question{TheQuestion := QuestionBank.AllQuestions.StarWarsQuestions.MessageStarForceMindTricks_EN},
                question{TheQuestion := QuestionBank.AllQuestions.StarWarsQuestions.MessageStarForceFoughtSithLord_EN}
            }
        }}
        
        set StarQuestions += array{category{
            Style:=card_style.Star,
            Name:="Galactic Relationships 💫",
            NameMessage := QuestionBank.AllCategories.MessageRelationships_EN,
            Questions:=array{
                question{TheQuestion := QuestionBank.AllQuestions.StarWarsQuestions.MessageStarRelationshipLukeSkywalker_EN},
                question{TheQuestion := QuestionBank.AllQuestions.StarWarsQuestions.MessageStarRelationshipApprentice_EN},
                question{TheQuestion := QuestionBank.AllQuestions.StarWarsQuestions.MessageStarRelationshipHadApprentice_EN},
                question{TheQuestion := QuestionBank.AllQuestions.StarWarsQuestions.MessageStarRelationshipDroidCompanion_EN},
                question{TheQuestion := QuestionBank.AllQuestions.StarWarsQuestions.MessageStarRelationshipBetrayed_EN},
                question{TheQuestion := QuestionBank.AllQuestions.StarWarsQuestions.MessageStarRelationshipBeast_EN},
                question{TheQuestion := QuestionBank.AllQuestions.MovieTVQuestions.MessageMovieAffiliationsTeam_EN},
                question{TheQuestion := QuestionBank.AllQuestions.StarWarsQuestions.MessageStarRelationshipFriendMando_EN}
            }
        }}

        set StarQuestions += array{category{
            Style:=card_style.Star,
            Name:="Color",
            NameMessage := QuestionBank.AllCategories.MessageColors_EN,
            Questions:=array{
                question{TheQuestion := QuestionBank.AllQuestions.CharacterQuestions.MessageCharacterColorRed_EN},
                question{TheQuestion := QuestionBank.AllQuestions.CharacterQuestions.MessageCharacterColorBlue_EN},
                question{TheQuestion := QuestionBank.AllQuestions.CharacterQuestions.MessageCharacterColorGreen_EN},
                question{TheQuestion := QuestionBank.AllQuestions.CharacterQuestions.MessageCharacterColorBlack_EN},
                question{TheQuestion := QuestionBank.AllQuestions.CharacterQuestions.MessageCharacterColorOrange_EN},
                question{TheQuestion := QuestionBank.AllQuestions.CharacterQuestions.MessageCharacterColorWhite_EN},
                question{TheQuestion := QuestionBank.AllQuestions.CharacterQuestions.MessageCharacterColorYellow_EN},
                question{TheQuestion := QuestionBank.AllQuestions.CharacterQuestions.MessageCharacterColorBrown_EN}
            }
        }}

        set StarQuestions += array{category{
            Style:=card_style.Star,
            Name:="Other",
            NameMessage := QuestionBank.AllCategories.MessageOther_EN,
            Questions:=array{
                question{TheQuestion := QuestionBank.AllQuestions.CharacterQuestions.MessageCharacterGenderMale_EN},
                question{TheQuestion := QuestionBank.AllQuestions.CharacterQuestions.MessageCharacterGenderFemale_EN},
                question{TheQuestion := QuestionBank.AllQuestions.StarWarsQuestions.MessageStarPhysicalAppearanceDroid_EN}

            }
        }}
    

    var SpookyQuestions : []category = array{}

    InitSpookyQuestions()<suspends>:void=
    
        set SpookyQuestions = array{
            category{
                Style:=card_style.Spooky,
                Name:="Character Type",
                NameMessage := QuestionBank.AllCategories.MessageSpecies_EN,
                Questions:=array{
                    question{TheQuestion := QuestionBank.AllQuestions.SpookyQuestions.MessageSpookyTypeZombie_EN},
                    question{TheQuestion := QuestionBank.AllQuestions.SpookyQuestions.MessageSpookyTypeSkeleton_EN},
                    question{TheQuestion := QuestionBank.AllQuestions.SpookyQuestions.MessageSpookyTypeVampire_EN},
                    question{TheQuestion := QuestionBank.AllQuestions.SpookyQuestions.MessageSpookyTypeWerewolf_EN},
                    question{TheQuestion := QuestionBank.AllQuestions.SpookyQuestions.MessageSpookyTypeWitch_EN},
                    question{TheQuestion := QuestionBank.AllQuestions.CharacterQuestions.MessageCharacterSpeciesMonster_EN},
                    question{TheQuestion := QuestionBank.AllQuestions.SpookyQuestions.MessageSpookyTypePumpkin_EN},
                    question{TheQuestion := QuestionBank.AllQuestions.SpookyQuestions.MessageSpookyTypeSlime_EN},
                    question{TheQuestion := QuestionBank.AllQuestions.SpookyQuestions.MessageSpookyTypeHay_EN},
                    question{TheQuestion := QuestionBank.AllQuestions.SpookyQuestions.MessageSpookyTypeCostume_EN}
                }
            }
        }
    
        set SpookyQuestions += array{
            category{
                Style:=card_style.Spooky,
                Name:="Release Year (Top Corner of Card)",
                NameMessage := QuestionBank.AllCategories.MessageReleaseYear_EN,
                Questions:=array{
                    question{TheQuestion := QuestionBank.AllQuestions.SpookyQuestions.MessageSpookyRelease2017_EN},
                    question{TheQuestion := QuestionBank.AllQuestions.SpookyQuestions.MessageSpookyRelease2018_EN},
                    question{TheQuestion := QuestionBank.AllQuestions.SpookyQuestions.MessageSpookyRelease2019_EN},
                    question{TheQuestion := QuestionBank.AllQuestions.SpookyQuestions.MessageSpookyRelease2020_EN},
                    question{TheQuestion := QuestionBank.AllQuestions.SpookyQuestions.MessageSpookyRelease2021_EN},
                    question{TheQuestion := QuestionBank.AllQuestions.SpookyQuestions.MessageSpookyRelease2022_EN},
                    question{TheQuestion := QuestionBank.AllQuestions.SpookyQuestions.MessageSpookyRelease2023_EN},
                    question{TheQuestion := QuestionBank.AllQuestions.SpookyQuestions.MessageSpookyRelease2024_EN}
                }
            }
        }
    
        set SpookyQuestions += array{
            category{
                Style:=card_style.Spooky,
                Name:="Appearance",
                NameMessage := QuestionBank.AllCategories.MessageAppearance_EN,
                Questions:=array{
                    question{TheQuestion := QuestionBank.AllQuestions.CharacterQuestions.MessageCharacterMask_EN},
                    question{TheQuestion := QuestionBank.AllQuestions.CharacterQuestions.MessageCharacterEyeColorGlowing_EN},
                    question{TheQuestion := QuestionBank.AllQuestions.SpookyQuestions.MessageSpookyAppearanceSharpTeeth_EN},
                    question{TheQuestion := QuestionBank.AllQuestions.SpookyQuestions.MessageSpookyAppearanceBones_EN},
                    question{TheQuestion := QuestionBank.AllQuestions.SpookyQuestions.MessageSpookyAppearanceBigHead_EN},
                    question{TheQuestion := QuestionBank.AllQuestions.CharacterQuestions.MessageCharacterFacePaint_EN},
                    question{TheQuestion := QuestionBank.AllQuestions.SpookyQuestions.MessageSpookyAppearanceFancyClothes_EN},
                    question{TheQuestion := QuestionBank.AllQuestions.CharacterQuestions.MessageCharacterHair_EN}
                }
            }
        }

        set SpookyQuestions += array{
            category{
                Style:=card_style.Spooky,
                Name:="Accessories",
                NameMessage := QuestionBank.AllCategories.MessageAccessories_EN,
                Questions:=array{
                    question{TheQuestion := QuestionBank.AllQuestions.CharacterQuestions.MessageCharacterHat_EN},
                    question{TheQuestion := QuestionBank.AllQuestions.CharacterQuestions.MessageCharacterGlasses_EN},
                    question{TheQuestion := QuestionBank.AllQuestions.CharacterQuestions.MessageCharacterBandana_EN},
                    question{TheQuestion := QuestionBank.AllQuestions.CharacterQuestions.MessageCharacterNecklace_EN},
                    question{TheQuestion := QuestionBank.AllQuestions.CharacterQuestions.MessageCharacterDogTags_EN},
                    question{TheQuestion := QuestionBank.AllQuestions.CharacterQuestions.MessageCharacterTie_EN}
                }
            }
        }

        set SpookyQuestions += array{
            category{
                Style:=card_style.Spooky,
                Name:="Colors",
                NameMessage := QuestionBank.AllCategories.MessageColors_EN,
                Questions:=array{
                    question{TheQuestion := QuestionBank.AllQuestions.CharacterQuestions.MessageCharacterColorRed_EN},
                    question{TheQuestion := QuestionBank.AllQuestions.CharacterQuestions.MessageCharacterColorBlue_EN},
                    question{TheQuestion := QuestionBank.AllQuestions.CharacterQuestions.MessageCharacterColorGreen_EN},
                    question{TheQuestion := QuestionBank.AllQuestions.CharacterQuestions.MessageCharacterColorYellow_EN},
                    question{TheQuestion := QuestionBank.AllQuestions.CharacterQuestions.MessageCharacterColorPurple_EN},
                    question{TheQuestion := QuestionBank.AllQuestions.CharacterQuestions.MessageCharacterColorOrange_EN},
                    question{TheQuestion := QuestionBank.AllQuestions.CharacterQuestions.MessageCharacterColorBlack_EN},
                    question{TheQuestion := QuestionBank.AllQuestions.CharacterQuestions.MessageCharacterColorWhite_EN}
                }
            }
        }

        set SpookyQuestions += array{
            category{
                Style:=card_style.Spooky,
                Name:="Other",
                NameMessage := QuestionBank.AllCategories.MessageOther_EN,
                Questions:=array{
                    question{TheQuestion := QuestionBank.AllQuestions.CharacterQuestions.MessageCharacterGenderMale_EN},
                    question{TheQuestion := QuestionBank.AllQuestions.CharacterQuestions.MessageCharacterGenderFemale_EN},
                    question{TheQuestion := QuestionBank.AllQuestions.CharacterQuestions.MessageCharacterTattoo_EN},
                    question{TheQuestion := QuestionBank.AllQuestions.CharacterQuestions.MessageCharacterEyeColorBlue_EN},
                    question{TheQuestion := QuestionBank.AllQuestions.CharacterQuestions.MessageCharacterEyeColorBrown_EN},
                    question{TheQuestion := QuestionBank.AllQuestions.CharacterQuestions.MessageCharacterEyeColorGreen_EN},
                    question{TheQuestion := QuestionBank.AllQuestions.CharacterQuestions.MessageCharacterEyeColorRed_EN},
                    question{TheQuestion := QuestionBank.AllQuestions.CharacterQuestions.MessageCharacterEyeColorBlack_EN},
                    question{TheQuestion := QuestionBank.AllQuestions.CharacterQuestions.MessageCharacterEyeColorWhite_EN},
                    question{TheQuestion := QuestionBank.AllQuestions.SpookyQuestions.MessageSpookyOtherNoEyes_EN}
                }
            }
        }
        
    
    
        
    var OGQuestions : []category = array{}

    InitOGQuestions()<suspends>:void=
        set OGQuestions = array{category{
            Style:=card_style.OG,
            Name:="Accessories",
            NameMessage := QuestionBank.AllCategories.MessageAccessories_EN,
            Questions:=array{
                question{TheQuestion := QuestionBank.AllQuestions.CharacterQuestions.MessageCharacterHelmet_EN},
                question{TheQuestion := QuestionBank.AllQuestions.CharacterQuestions.MessageCharacterHat_EN},
                question{TheQuestion := QuestionBank.AllQuestions.CharacterQuestions.MessageCharacterGlasses_EN},
                question{TheQuestion := QuestionBank.AllQuestions.CharacterQuestions.MessageCharacterDogTags_EN},
                question{TheQuestion := QuestionBank.AllQuestions.CharacterQuestions.MessageCharacterBandana_EN},
                question{TheQuestion := QuestionBank.AllQuestions.CharacterQuestions.MessageCharacterFacePaint_EN},
                question{TheQuestion := QuestionBank.AllQuestions.CharacterQuestions.MessageCharacterFacialHair_EN}
            }
        }}
        set OGQuestions += array{category{
            Style:=card_style.OG,
            Name:="Hair Colors",
            NameMessage := QuestionBank.AllCategories.MessageHairColors_EN,
            Questions:=array{
                question{TheQuestion := QuestionBank.AllQuestions.CharacterQuestions.MessageCharacterHairColorOrange_EN},
                question{TheQuestion := QuestionBank.AllQuestions.CharacterQuestions.MessageCharacterHairColorWhite_EN},
                question{TheQuestion := QuestionBank.AllQuestions.CharacterQuestions.MessageCharacterHairColorBlack_EN},
                question{TheQuestion := QuestionBank.AllQuestions.CharacterQuestions.MessageCharacterHairColorBlonde_EN},
                question{TheQuestion := QuestionBank.AllQuestions.CharacterQuestions.MessageCharacterHairColorGreen_EN},
                question{TheQuestion := QuestionBank.AllQuestions.CharacterQuestions.MessageCharacterHairColorBrown_EN}
            }
        }}
        set OGQuestions += array{category{
            Style:=card_style.OG,
            Name:="Hair Styles",
            NameMessage := QuestionBank.AllCategories.MessageHairStyles_EN,
            Questions:=array{
                question{TheQuestion := QuestionBank.AllQuestions.CharacterQuestions.MessageCharacterHairCurly_EN},
                question{TheQuestion := QuestionBank.AllQuestions.CharacterQuestions.MessageCharacterHairBald_EN},
                question{TheQuestion := QuestionBank.AllQuestions.CharacterQuestions.MessageCharacterHairShort_EN},
                question{TheQuestion := QuestionBank.AllQuestions.CharacterQuestions.MessageCharacterHairLong_EN},
                question{TheQuestion := QuestionBank.AllQuestions.CharacterQuestions.MessageCharacterHairPonytail_EN},
                question{TheQuestion := QuestionBank.AllQuestions.CharacterQuestions.MessageCharacterHairFade_EN},
                question{TheQuestion := QuestionBank.AllQuestions.CharacterQuestions.MessageCharacterHairAfro_EN}
            }
        }}
        set OGQuestions += array{category{
            Style:=card_style.OG,
            Name:="Clothing",
            NameMessage := QuestionBank.AllCategories.MessageClothing_EN,
            Questions:=array{
                question{TheQuestion := QuestionBank.AllQuestions.CharacterQuestions.MessageCharacterJacket_EN},
                question{TheQuestion := QuestionBank.AllQuestions.CharacterQuestions.MessageCharacterTankTop_EN},
                question{TheQuestion := QuestionBank.AllQuestions.CharacterQuestions.MessageCharacterChainmail_EN},
                question{TheQuestion := QuestionBank.AllQuestions.CharacterQuestions.MessageCharacterLongSleeve_EN}
            }
        }}
        set OGQuestions += array{category{
            Style:=card_style.OG,
            Name:="Colors",
            NameMessage := QuestionBank.AllCategories.MessageColors_EN,
            Questions:=array{
                question{TheQuestion := QuestionBank.AllQuestions.CharacterQuestions.MessageCharacterColorRed_EN},
                question{TheQuestion := QuestionBank.AllQuestions.CharacterQuestions.MessageCharacterColorBlue_EN},
                question{TheQuestion := QuestionBank.AllQuestions.CharacterQuestions.MessageCharacterColorGreen_EN},
                question{TheQuestion := QuestionBank.AllQuestions.CharacterQuestions.MessageCharacterColorYellow_EN},
                question{TheQuestion := QuestionBank.AllQuestions.CharacterQuestions.MessageCharacterColorPurple_EN},
                question{TheQuestion := QuestionBank.AllQuestions.CharacterQuestions.MessageCharacterColorOrange_EN},
                question{TheQuestion := QuestionBank.AllQuestions.CharacterQuestions.MessageCharacterColorPink_EN},
                question{TheQuestion := QuestionBank.AllQuestions.CharacterQuestions.MessageCharacterColorBlack_EN},
                question{TheQuestion := QuestionBank.AllQuestions.CharacterQuestions.MessageCharacterColorWhite_EN}

            }
        }}
        set OGQuestions += array{category{
            Style:=card_style.OG,
            Name:= "Other",
            NameMessage := QuestionBank.AllCategories.MessageOther_EN,
            Questions:=array{
                question{TheQuestion := QuestionBank.AllQuestions.CharacterQuestions.MessageCharacterGenderMale_EN},
                question{TheQuestion := QuestionBank.AllQuestions.CharacterQuestions.MessageCharacterGenderFemale_EN},
                question{TheQuestion := QuestionBank.AllQuestions.CharacterQuestions.MessageCharacterTannedSkinTone_EN},
                question{TheQuestion := QuestionBank.AllQuestions.CharacterQuestions.MessageCharacterDarkSkinTone_EN},
                question{TheQuestion := QuestionBank.AllQuestions.CharacterQuestions.MessageCharacterLightSkinTone_EN},
                question{TheQuestion := QuestionBank.AllQuestions.CharacterQuestions.MessageCharacterEyeColorBlue_EN},
                question{TheQuestion := QuestionBank.AllQuestions.CharacterQuestions.MessageCharacterEyeColorBrown_EN},
                question{TheQuestion := QuestionBank.AllQuestions.CharacterQuestions.MessageCharacterEyeColorGreen_EN}
            }
        }}
    
    
    var AnimeQuestions : []category = array{}

    InitAnimeQuestions()<suspends>:void=
        set AnimeQuestions = array{category{
            Style:=card_style.Anime,
            Name:="Outfits and Abilities",
            NameMessage := QuestionBank.AllCategories.MessageOutfitsAbilities_EN,
            Questions:=array{
                question{TheQuestion := QuestionBank.AllQuestions.AnimeQuestions.MessageAnimeOutfitBattleSuit_EN},
                question{TheQuestion := QuestionBank.AllQuestions.AnimeQuestions.MessageAnimeOutfitProtectiveGear_EN},
                question{TheQuestion := QuestionBank.AllQuestions.AnimeQuestions.MessageAnimeOutfitCasual_EN},
                question{TheQuestion := QuestionBank.AllQuestions.AnimeQuestions.MessageAnimeAbilityElemental_EN},
                question{TheQuestion := QuestionBank.AllQuestions.AnimeQuestions.MessageAnimeOutfitDrasticChange_EN},
                question{TheQuestion := QuestionBank.AllQuestions.AnimeQuestions.MessageAnimeAbilityIconic_EN}
            }
        }}
        set AnimeQuestions += array{category{
            Style:=card_style.Anime,
            Name:="Role & Allegiance",
            NameMessage := QuestionBank.AllCategories.MessageRoleAllegiance_EN,
            Questions:=array{
                question{TheQuestion := QuestionBank.AllQuestions.AnimeQuestions.MessageAnimeAllegianceProtagonist_EN},
                question{TheQuestion := QuestionBank.AllQuestions.AnimeQuestions.MessageAnimeAllegianceAntagonist_EN},
                question{TheQuestion := QuestionBank.AllQuestions.AnimeQuestions.MessageAnimeAllegianceMentor_EN}
            }
        }}
        set AnimeQuestions += array{category{
            Style:=card_style.Anime,
            Name:="Physical Traits",
            NameMessage := QuestionBank.AllCategories.MessagePhysicalTraits_EN,
            Questions:=array{
                question{TheQuestion := QuestionBank.AllQuestions.StarWarsQuestions.MessageStarFacialMarkings_EN},
                question{TheQuestion := QuestionBank.AllQuestions.CharacterQuestions.MessageCharacterTattoo_EN},
                question{TheQuestion := QuestionBank.AllQuestions.LockerBundleQuestions.MessageLockerOtherScars_EN},
                question{TheQuestion := QuestionBank.AllQuestions.AnimeQuestions.MessageAnimePhysicalHuman_EN},
                question{TheQuestion := QuestionBank.AllQuestions.AnimeQuestions.MessageAnimePhysicalHeadgear_EN},
                question{TheQuestion := QuestionBank.AllQuestions.CharacterQuestions.MessageCharacterCape_EN}
            }
        }}
        set AnimeQuestions += array{category{
            Style:=card_style.Anime,
            Name:="Series",
            NameMessage := QuestionBank.AllCategories.MessageSeries_EN,
            Questions:=array{
                question{TheQuestion := QuestionBank.AllQuestions.AnimeQuestions.MessageAnimeSeriesAttackOnTitan_EN},
                question{TheQuestion := QuestionBank.AllQuestions.AnimeQuestions.MessageAnimeSeriesDragonBall_EN},
                question{TheQuestion := QuestionBank.AllQuestions.AnimeQuestions.MessageAnimeSeriesJujutsuKaisen_EN},
                question{TheQuestion := QuestionBank.AllQuestions.AnimeQuestions.MessageAnimeSeriesMyHeroAcademia_EN},
                question{TheQuestion := QuestionBank.AllQuestions.AnimeQuestions.MessageAnimeSeriesNaruto_EN}
            }
        }}
        set AnimeQuestions += array{category{
            Style:=card_style.Anime,
            Name:="Hair Colors",
            NameMessage := QuestionBank.AllCategories.MessageHairColors_EN,
            Questions:=array{
                question{TheQuestion := QuestionBank.AllQuestions.CharacterQuestions.MessageCharacterHairColorBlack_EN},
                question{TheQuestion := QuestionBank.AllQuestions.CharacterQuestions.MessageCharacterHairColorBlonde_EN},
                question{TheQuestion := QuestionBank.AllQuestions.CharacterQuestions.MessageCharacterHairColorOrangeRed_EN},
                question{TheQuestion := QuestionBank.AllQuestions.CharacterQuestions.MessageCharacterHairColorWhite_EN},
                question{TheQuestion := QuestionBank.AllQuestions.CharacterQuestions.MessageCharacterHairColorDyed_EN}  
            }
        }}
        set AnimeQuestions += array{category{
            Style:=card_style.Anime,
            Name:="Hair Styles",
            NameMessage := QuestionBank.AllCategories.MessageHairStyles_EN,
            Questions:=array{
                question{TheQuestion := QuestionBank.AllQuestions.CharacterQuestions.MessageCharacterHairSpiky_EN},
                question{TheQuestion := QuestionBank.AllQuestions.CharacterQuestions.MessageCharacterHairLong_EN},
                question{TheQuestion := QuestionBank.AllQuestions.CharacterQuestions.MessageCharacterHairShort_EN}
            }
        }}
        set AnimeQuestions += array{category{
            Style:=card_style.Anime,
            Name:="Colors",
            NameMessage := QuestionBank.AllCategories.MessageColors_EN,
            Questions:=array{
                question{TheQuestion := QuestionBank.AllQuestions.CharacterQuestions.MessageCharacterColorRed_EN},
                question{TheQuestion := QuestionBank.AllQuestions.CharacterQuestions.MessageCharacterColorBlue_EN},
                question{TheQuestion := QuestionBank.AllQuestions.CharacterQuestions.MessageCharacterColorGreen_EN},
                question{TheQuestion := QuestionBank.AllQuestions.CharacterQuestions.MessageCharacterColorYellow_EN},
                question{TheQuestion := QuestionBank.AllQuestions.CharacterQuestions.MessageCharacterColorPurple_EN},
                question{TheQuestion := QuestionBank.AllQuestions.CharacterQuestions.MessageCharacterColorOrange_EN},
                question{TheQuestion := QuestionBank.AllQuestions.CharacterQuestions.MessageCharacterColorPink_EN},
                question{TheQuestion := QuestionBank.AllQuestions.CharacterQuestions.MessageCharacterColorBlack_EN},
                question{TheQuestion := QuestionBank.AllQuestions.CharacterQuestions.MessageCharacterColorWhite_EN}
            }
        }}
        set AnimeQuestions += array{category{
            Style:=card_style.Anime,
            Name:= "Other",
            NameMessage := QuestionBank.AllCategories.MessageOther_EN,
            Questions:=array{
                question{TheQuestion := QuestionBank.AllQuestions.CharacterQuestions.MessageCharacterGenderMale_EN},
                question{TheQuestion := QuestionBank.AllQuestions.CharacterQuestions.MessageCharacterGenderFemale_EN},
                question{TheQuestion := QuestionBank.AllQuestions.CharacterQuestions.MessageCharacterTannedSkinTone_EN},
                question{TheQuestion := QuestionBank.AllQuestions.CharacterQuestions.MessageCharacterDarkSkinTone_EN},
                question{TheQuestion := QuestionBank.AllQuestions.CharacterQuestions.MessageCharacterLightSkinTone_EN},
                question{TheQuestion := QuestionBank.AllQuestions.CharacterQuestions.MessageCharacterEyeColorBlue_EN},
                question{TheQuestion := QuestionBank.AllQuestions.CharacterQuestions.MessageCharacterEyeColorBrown_EN},
                question{TheQuestion := QuestionBank.AllQuestions.CharacterQuestions.MessageCharacterEyeColorGreen_EN}
            }
        }}
    

    var WinterQuestions : []category = array{}

    InitWinterQuestions()<suspends>:void=
        set WinterQuestions = array{category{
            Style:=card_style.Winter,
            Name:="Festive Attire",
            NameMessage := QuestionBank.AllCategories.MessageFestiveAttire_EN,
            Questions:=array{
                question{TheQuestion := QuestionBank.AllQuestions.WinterQuestions.MessageWinterAttireSantaHat_EN},
                question{TheQuestion := QuestionBank.AllQuestions.WinterQuestions.MessageWinterAttireCozySweater_EN},
                question{TheQuestion := QuestionBank.AllQuestions.WinterQuestions.MessageWinterAttireLights_EN},
                question{TheQuestion := QuestionBank.AllQuestions.WinterQuestions.MessageWinterAttireSnowflakes_EN},
                question{TheQuestion := QuestionBank.AllQuestions.WinterQuestions.MessageWinterAttireLlama_EN},
                question{TheQuestion := QuestionBank.AllQuestions.WinterQuestions.MessageWinterAttireCandy_EN},
                question{TheQuestion := QuestionBank.AllQuestions.WinterQuestions.MessageWinterAttireScarf_EN},
                question{TheQuestion := QuestionBank.AllQuestions.WinterQuestions.MessageWinterAttireBowtie_EN}
            }
        }}
            
        set WinterQuestions += array{category{
            Style:=card_style.Winter,
            Name:="Species",
            NameMessage := QuestionBank.AllCategories.MessageSpecies_EN,
            Questions:=array{
                question{TheQuestion := QuestionBank.AllQuestions.CharacterQuestions.MessageCharacterSpeciesHuman_EN},
                question{TheQuestion := QuestionBank.AllQuestions.CharacterQuestions.MessageCharacterSpeciesAnimal_EN},
                question{TheQuestion := QuestionBank.AllQuestions.WinterQuestions.MessageWinterSpeciesGlass_EN},
                question{TheQuestion := QuestionBank.AllQuestions.WinterQuestions.MessageWinterSpeciesSnowman_EN},
                question{TheQuestion := QuestionBank.AllQuestions.WinterQuestions.MessageWinterSpeciesGingerbread_EN},
                question{TheQuestion := QuestionBank.AllQuestions.WinterQuestions.MessageWinterSpeciesNutcracker_EN},
                question{TheQuestion := QuestionBank.AllQuestions.WinterQuestions.MessageWinterSpeciesIce_EN},
                question{TheQuestion := QuestionBank.AllQuestions.CharacterQuestions.MessageCharacterSpeciesFood_EN}
            }
        }}

        set WinterQuestions += array{category{
            Style:=card_style.Winter,
            Name:="Facial Features",
            NameMessage := QuestionBank.AllCategories.MessageFacialAttributes_EN,
            Questions:=array{
                question{TheQuestion := QuestionBank.AllQuestions.CharacterQuestions.MessageCharacterFacePaint_EN},
                question{TheQuestion := QuestionBank.AllQuestions.CharacterQuestions.MessageCharacterFacialHair_EN},
                question{TheQuestion := QuestionBank.AllQuestions.CharacterQuestions.MessageCharacterBeard_EN},
                question{TheQuestion := QuestionBank.AllQuestions.CharacterQuestions.MessageCharacterGlasses_EN},
                question{TheQuestion := QuestionBank.AllQuestions.WinterQuestions.MessageWinterFacialVisibleNose_EN},
                question{TheQuestion := QuestionBank.AllQuestions.WinterQuestions.MessageWinterFacialVisible_EN}
            }
        }}

        set WinterQuestions += array{category{
            Style:=card_style.Winter,
            Name:="Accessories",
            NameMessage := QuestionBank.AllCategories.MessageAccessories_EN,
            Questions:=array{
                question{TheQuestion := QuestionBank.AllQuestions.CharacterQuestions.MessageCharacterDogTags_EN},
                question{TheQuestion := QuestionBank.AllQuestions.CharacterQuestions.MessageCharacterHat_EN},
                question{TheQuestion := QuestionBank.AllQuestions.WinterQuestions.MessageWinterAccessoriesCarrot_EN},
                question{TheQuestion := QuestionBank.AllQuestions.PeelyFishstickQuestions.MessagePFAccessoriesStripes_EN}
            }
        }}

        set WinterQuestions += array{category{
            Style:=card_style.Winter,
            Name:="Colors",
            NameMessage := QuestionBank.AllCategories.MessageColors_EN,
            Questions:=array{
                question{TheQuestion := QuestionBank.AllQuestions.CharacterQuestions.MessageCharacterColorRed_EN},
                question{TheQuestion := QuestionBank.AllQuestions.CharacterQuestions.MessageCharacterColorBlue_EN},
                question{TheQuestion := QuestionBank.AllQuestions.CharacterQuestions.MessageCharacterColorGreen_EN},
                question{TheQuestion := QuestionBank.AllQuestions.CharacterQuestions.MessageCharacterColorYellow_EN},
                question{TheQuestion := QuestionBank.AllQuestions.CharacterQuestions.MessageCharacterColorPurple_EN},
                question{TheQuestion := QuestionBank.AllQuestions.CharacterQuestions.MessageCharacterColorOrange_EN},
                question{TheQuestion := QuestionBank.AllQuestions.CharacterQuestions.MessageCharacterColorGold_EN},
                question{TheQuestion := QuestionBank.AllQuestions.CharacterQuestions.MessageCharacterColorBlack_EN}
            }
        }}



    var WeaponsQuestions : []category = array{}

    InitWeaponsQuestions()<suspends>:void=

        set WeaponsQuestions = array{category{
            Style:=card_style.Weapons,
            Name:="Card Rarity",
            NameMessage := QuestionBank.AllCategories.MessageRarity_EN,
            Questions:=array{
                question{TheQuestion := QuestionBank.AllQuestions.ItemQuestions.MessageItemCommonRarity_EN},
                question{TheQuestion := QuestionBank.AllQuestions.ItemQuestions.MessageItemUncommonRarity_EN},
                question{TheQuestion := QuestionBank.AllQuestions.ItemQuestions.MessageItemRareRarity_EN},
                question{TheQuestion := QuestionBank.AllQuestions.ItemQuestions.MessageItemEpicRarity_EN},
                question{TheQuestion := QuestionBank.AllQuestions.ItemQuestions.MessageItemLegendaryRarity_EN}
            }
        }}

        set WeaponsQuestions += array{category{
            Style:=card_style.Weapons,
            Name:="Physical Traits",
            NameMessage := QuestionBank.AllCategories.MessagePhysicalTraits_EN,
            Questions:=array{
                question{TheQuestion := QuestionBank.AllQuestions.ItemQuestions.MessageItemScope_EN},
                question{TheQuestion := QuestionBank.AllQuestions.ItemQuestions.MessageItemSilencer_EN},                
                question{TheQuestion := QuestionBank.AllQuestions.ItemQuestions.MessageItemMagazine_EN},
                question{TheQuestion := QuestionBank.AllQuestions.ItemQuestions.MessageItemStockShoulderRest_EN},
                question{TheQuestion := QuestionBank.AllQuestions.ItemQuestions.MessageItemSupportLegs_EN},
                question{TheQuestion := QuestionBank.AllQuestions.ItemQuestions.MessageItemBarrelLoaded_EN}
            }
        }}

        set WeaponsQuestions += array{category{
            Style:=card_style.Weapons,
            Name:="Origin",
            NameMessage := QuestionBank.AllCategories.MessageOrigin_EN,
            Questions:=array{
                question{TheQuestion := QuestionBank.AllQuestions.ItemQuestions.MessageItemChapterOne_EN},
                question{TheQuestion := QuestionBank.AllQuestions.ItemQuestions.MessageItemChapterTwo_EN},
                question{TheQuestion := QuestionBank.AllQuestions.ItemQuestions.MessageItemChapterThree_EN},
                question{TheQuestion := QuestionBank.AllQuestions.ItemQuestions.MessageItemChapterFour_EN},
                question{TheQuestion := QuestionBank.AllQuestions.ItemQuestions.MessageItemChapterFive_EN},
                question{TheQuestion := QuestionBank.AllQuestions.ItemQuestions.MessageItemImportantCharacterOrBoss_EN}
            }
        }}

        set WeaponsQuestions += array{category{
            Style:=card_style.Weapons,
            Name:="Other",
            NameMessage := QuestionBank.AllCategories.MessageOther_EN,
            Questions:=array{
                question{TheQuestion := QuestionBank.AllQuestions.WeaponQuestions.MessageWeaponOtherSpecialAbility_EN},
                question{TheQuestion := QuestionBank.AllQuestions.WeaponQuestions.MessageWeaponOtherOverheats_EN},            
                question{TheQuestion := QuestionBank.AllQuestions.WeaponQuestions.MessageWeaponOtherCurrentlyInGame_EN},
                question{TheQuestion := QuestionBank.AllQuestions.WeaponQuestions.MessageWeaponOtherReloadedEachShot_EN},
                question{TheQuestion := QuestionBank.AllQuestions.WeaponQuestions.MessageWeaponOtherChargedShot_EN}
            }
        }}
        set WeaponsQuestions += array{category{
            Style:=card_style.Weapons,
            Name:="Ammo Type",
            NameMessage := QuestionBank.AllCategories.MessageAmmoType_EN,
            Questions:=array{
                question{TheQuestion := QuestionBank.AllQuestions.WeaponQuestions.MessageWeaponAmmoLightBullets_EN},
                question{TheQuestion := QuestionBank.AllQuestions.WeaponQuestions.MessageWeaponAmmoMediumBullets_EN},                
                question{TheQuestion := QuestionBank.AllQuestions.WeaponQuestions.MessageWeaponAmmoHeavyBullets_EN},
                question{TheQuestion := QuestionBank.AllQuestions.WeaponQuestions.MessageWeaponAmmoShotgunShells_EN},
                question{TheQuestion := QuestionBank.AllQuestions.WeaponQuestions.MessageWeaponAmmoRockets_EN},
                question{TheQuestion := QuestionBank.AllQuestions.WeaponQuestions.MessageWeaponAmmoArrows_EN}
            }
        }}
        
        set WeaponsQuestions += array{category{
            Style:=card_style.Weapons,
            Name:="Color",
            NameMessage := QuestionBank.AllCategories.MessageColors_EN,
            Questions:=array{
                question{TheQuestion := QuestionBank.AllQuestions.ItemQuestions.MessageItemBlue_EN},
                question{TheQuestion := QuestionBank.AllQuestions.ItemQuestions.MessageItemGreen_EN},
                question{TheQuestion := QuestionBank.AllQuestions.ItemQuestions.MessageItemTan_EN},
                question{TheQuestion := QuestionBank.AllQuestions.ItemQuestions.MessageItemRed_EN},
                question{TheQuestion := QuestionBank.AllQuestions.ItemQuestions.MessageItemWhite_EN},
                question{TheQuestion := QuestionBank.AllQuestions.ItemQuestions.MessageItemSilver_EN},
                question{TheQuestion := QuestionBank.AllQuestions.ItemQuestions.MessageItemBrown_EN}
            }
        }}


    var PickaxesQuestions : []category = array{}

    InitPickaxeQuestions()<suspends>:void=

        set PickaxesQuestions = array{
            category{
                Style:=card_style.Pickaxes,
                Name:="Theme",
                NameMessage := QuestionBank.AllCategories.MessageTheme_EN,
                Questions:=array{
                    question{TheQuestion := QuestionBank.AllQuestions.PickaxeQuestions.MessagePickaxeHolidayThemed_EN},
                    question{TheQuestion := QuestionBank.AllQuestions.PickaxeQuestions.MessagePickaxeGlowLights_EN},
                    question{TheQuestion := QuestionBank.AllQuestions.PickaxeQuestions.MessagePickaxeMetallic_EN},
                    question{TheQuestion := QuestionBank.AllQuestions.PickaxeQuestions.MessagePickaxeSportsTheme_EN},
                    question{TheQuestion := QuestionBank.AllQuestions.PickaxeQuestions.MessagePickaxeEdibleInRealLife_EN},
                    question{TheQuestion := QuestionBank.AllQuestions.PickaxeQuestions.MessagePickaxeToy_EN},
                    question{TheQuestion := QuestionBank.AllQuestions.PickaxeQuestions.MessagePickaxeAnimalTheme_EN},
                    question{TheQuestion := QuestionBank.AllQuestions.PickaxeQuestions.MessagePickaxeRainbow_EN}
                }
            }
        }
    
        set PickaxesQuestions += array{
            category{
                Style:=card_style.Pickaxes,
                Name:="Functionalality",
                NameMessage := QuestionBank.AllCategories.MessageFunctionality_EN,
                Questions:=array{
                    question{TheQuestion := QuestionBank.AllQuestions.PickaxeQuestions.MessagePickaxeSharpEdges_EN},
                    question{TheQuestion := QuestionBank.AllQuestions.PickaxeQuestions.MessagePickaxeBladedDesign_EN},
                    question{TheQuestion := QuestionBank.AllQuestions.PickaxeQuestions.MessagePickaxeTool_EN},
                    question{TheQuestion := QuestionBank.AllQuestions.PickaxeQuestions.MessagePickaxeUsedEatDrink_EN},
                    question{TheQuestion := QuestionBank.AllQuestions.PickaxeQuestions.MessagePickaxeDualWield_EN}
                }
            }
        }
    
        set PickaxesQuestions += array{category{
            Style:=card_style.Pickaxes,
            Name:="Color",
            NameMessage := QuestionBank.AllCategories.MessageColors_EN,
            Questions:=array{
                question{TheQuestion := QuestionBank.AllQuestions.ItemQuestions.MessageItemBlue_EN},
                question{TheQuestion := QuestionBank.AllQuestions.ItemQuestions.MessageItemGreen_EN},
                question{TheQuestion := QuestionBank.AllQuestions.ItemQuestions.MessageItemYellow_EN},
                question{TheQuestion := QuestionBank.AllQuestions.ItemQuestions.MessageItemRed_EN},
                question{TheQuestion := QuestionBank.AllQuestions.ItemQuestions.MessageItemWhite_EN},
                question{TheQuestion := QuestionBank.AllQuestions.ItemQuestions.MessageItemSilver_EN},
                question{TheQuestion := QuestionBank.AllQuestions.ItemQuestions.MessageItemPink_EN},
                question{TheQuestion := QuestionBank.AllQuestions.ItemQuestions.MessageItemPurple_EN},
                question{TheQuestion := QuestionBank.AllQuestions.ItemQuestions.MessageItemBrown_EN}
            }
        }}

        set PickaxesQuestions += array{category{
            Style:=card_style.Pickaxes,
            Name:="Card Rarity",
            NameMessage := QuestionBank.AllCategories.MessageRarity_EN,
            Questions:=array{
                question{TheQuestion := QuestionBank.AllQuestions.ItemQuestions.MessageItemCommonRarity_EN},
                question{TheQuestion := QuestionBank.AllQuestions.ItemQuestions.MessageItemUncommonRarity_EN},
                question{TheQuestion := QuestionBank.AllQuestions.ItemQuestions.MessageItemRareRarity_EN},
                question{TheQuestion := QuestionBank.AllQuestions.ItemQuestions.MessageItemEpicRarity_EN},
                question{TheQuestion := QuestionBank.AllQuestions.ItemQuestions.MessageItemLegendaryRarity_EN}
            }
        }}

        set PickaxesQuestions += array{category{
            Style:=card_style.Pickaxes,
            Name:="Other",
            NameMessage := QuestionBank.AllCategories.MessageOther_EN,
            Questions:=array{
                question{TheQuestion := QuestionBank.AllQuestions.PickaxeQuestions.MessagePickaxeRareHardToFind_EN},
                question{TheQuestion := QuestionBank.AllQuestions.PickaxeQuestions.MessagePickaxeBattlePass_EN},
                question{TheQuestion := QuestionBank.AllQuestions.PickaxeQuestions.MessagePickaxeChampionship_EN},
                question{TheQuestion := QuestionBank.AllQuestions.PickaxeQuestions.MessagePickaxeWouldItHurt_EN},
                question{TheQuestion := QuestionBank.AllQuestions.PickaxeQuestions.MessagePickaxeRealWorldInfluence_EN}
            }
        }}


    var VehiclesQuestions : []category = array{}

    InitVehicleQuestions()<suspends>:void=

        set VehiclesQuestions = array{
            category{
                Style:=card_style.Vehicles,
                Name:="Vehicle Type",
                NameMessage := QuestionBank.AllCategories.MessageVehicleType_EN,
                Questions:=array{
                    question{TheQuestion := QuestionBank.AllQuestions.VehicleQuestions.MessageVehicleTypeLand_EN},
                    question{TheQuestion := QuestionBank.AllQuestions.VehicleQuestions.MessageVehicleTypeWater_EN},
                    question{TheQuestion := QuestionBank.AllQuestions.VehicleQuestions.MessageVehicleTypeAir_EN},
                    question{TheQuestion := QuestionBank.AllQuestions.VehicleQuestions.MessageVehicleTypeSpace_EN},
                    question{TheQuestion := QuestionBank.AllQuestions.VehicleQuestions.MessageVehicleTypeSnow_EN},
                    question{TheQuestion := QuestionBank.AllQuestions.VehicleQuestions.MessageVehicleTypeCombat_EN}
                }
            }
        }
    
        set VehiclesQuestions += array{
            category{
                Style:=card_style.Vehicles,
                Name:="Passenger Capacity",
                NameMessage := QuestionBank.AllCategories.MessagePassengerCapacity_EN,
                Questions:=array{
                    question{TheQuestion := QuestionBank.AllQuestions.VehicleQuestions.MessageVehicleCapacityOne_EN},
                    question{TheQuestion := QuestionBank.AllQuestions.VehicleQuestions.MessageVehicleCapacityTwo_EN},
                    question{TheQuestion := QuestionBank.AllQuestions.VehicleQuestions.MessageVehicleCapacityThree_EN},
                    question{TheQuestion := QuestionBank.AllQuestions.VehicleQuestions.MessageVehicleCapacityFour_EN},
                    question{TheQuestion := QuestionBank.AllQuestions.VehicleQuestions.MessageVehicleCapacityFive_EN},
                    question{TheQuestion := QuestionBank.AllQuestions.VehicleQuestions.MessageVehicleCapacityFivePlus_EN}
                }
            }
        }
    
        set VehiclesQuestions += array{
            category{
                Style:=card_style.Vehicles,
                Name:="Capabilities",
                NameMessage := QuestionBank.AllCategories.MessageCapabilities_EN,
                Questions:=array{
                    question{TheQuestion := QuestionBank.AllQuestions.VehicleQuestions.MessageVehicleCapabilitiesWeapons_EN},
                    question{TheQuestion := QuestionBank.AllQuestions.VehicleQuestions.MessageVehicleCapabilitiesBoost_EN},
                    question{TheQuestion := QuestionBank.AllQuestions.VehicleQuestions.MessageVehicleCapabilitiesFly_EN},
                    question{TheQuestion := QuestionBank.AllQuestions.VehicleQuestions.MessageVehicleCapabilitiesMods_EN},
                    question{TheQuestion := QuestionBank.AllQuestions.VehicleQuestions.MessageVehicleCapabilitiesHorn_EN},
                    question{TheQuestion := QuestionBank.AllQuestions.VehicleQuestions.MessageVehicleCapabilitiesRadio_EN},
                    question{TheQuestion := QuestionBank.AllQuestions.VehicleQuestions.MessageVehicleCapabilitiesJump_EN},
                    question{TheQuestion := QuestionBank.AllQuestions.VehicleQuestions.MessageVehicleCapabilitiesShopping_EN},
                    question{TheQuestion := QuestionBank.AllQuestions.VehicleQuestions.MessageVehicleCapabilitiesMissiles_EN}
                }
            }
        }
    
        set VehiclesQuestions += array{
            category{
                Style:=card_style.Vehicles,
                Name:="Appearance",
                NameMessage := QuestionBank.AllCategories.MessageAppearance_EN,
                Questions:=array{
                    question{TheQuestion := QuestionBank.AllQuestions.VehicleQuestions.MessageVehicleWheels_EN},
                    question{TheQuestion := QuestionBank.AllQuestions.VehicleQuestions.MessageVehicleWheelsTwo_EN},
                    question{TheQuestion := QuestionBank.AllQuestions.VehicleQuestions.MessageVehicleWheelsFour_EN},
                    question{TheQuestion := QuestionBank.AllQuestions.VehicleQuestions.MessageVehicleWheelsMoreThanFour_EN},
                    question{TheQuestion := QuestionBank.AllQuestions.VehicleQuestions.MessageVehicleAppearanceWindshield_EN},
                    question{TheQuestion := QuestionBank.AllQuestions.VehicleQuestions.MessageVehicleAppearanceRoof_EN},
                    question{TheQuestion := QuestionBank.AllQuestions.VehicleQuestions.MessageVehicleAppearanceTrunk_EN},
                    question{TheQuestion := QuestionBank.AllQuestions.VehicleQuestions.MessageVehicleAppearanceTreads_EN},
                    question{TheQuestion := QuestionBank.AllQuestions.VehicleQuestions.MessageVehicleAppearanceTurret_EN}
                }
            }
        }

        set VehiclesQuestions += array{
            category{
                Style:=card_style.Vehicles,
                Name:="Other",
                NameMessage := QuestionBank.AllCategories.MessageOther_EN,
                Questions:=array{
                    question{TheQuestion := QuestionBank.AllQuestions.VehicleQuestions.MessageVehicleOtherStoodOn_EN},
                    question{TheQuestion := QuestionBank.AllQuestions.VehicleQuestions.MessageVehicleOtherRealWorld_EN},
                    question{TheQuestion := QuestionBank.AllQuestions.VehicleQuestions.MessageVehicleOtherConsideredWeapon_EN},
                    question{TheQuestion := QuestionBank.AllQuestions.VehicleQuestions.MessageVehicleOtherContainsGlass_EN},
                    question{TheQuestion := QuestionBank.AllQuestions.VehicleQuestions.MessageVehicleOtherPropeller_EN},
                    question{TheQuestion := QuestionBank.AllQuestions.VehicleQuestions.MessageVehicleOtherHover_EN},
                    question{TheQuestion := QuestionBank.AllQuestions.VehicleQuestions.MessageVehicleOtherUnlimitedFuel_EN}, 
                    question{TheQuestion := QuestionBank.AllQuestions.VehicleQuestions.MessageVehicleOtherFullyCovered_EN},
                    question{TheQuestion := QuestionBank.AllQuestions.VehicleQuestions.MessageVehicleOtherCarryItems_EN}
                }
            }
        }

        set VehiclesQuestions += array{
            category{
                Style:=card_style.Vehicles,
                Name:="Colors",
                NameMessage := QuestionBank.AllCategories.MessageColors_EN,
                Questions:=array{
                    question{TheQuestion := QuestionBank.AllQuestions.ItemQuestions.MessageItemBlue_EN},
                    question{TheQuestion := QuestionBank.AllQuestions.ItemQuestions.MessageItemGreen_EN},
                    question{TheQuestion := QuestionBank.AllQuestions.ItemQuestions.MessageItemYellow_EN},
                    question{TheQuestion := QuestionBank.AllQuestions.ItemQuestions.MessageItemRed_EN},
                    question{TheQuestion := QuestionBank.AllQuestions.ItemQuestions.MessageItemWhite_EN},
                    question{TheQuestion := QuestionBank.AllQuestions.ItemQuestions.MessageItemBlack_EN},
                    question{TheQuestion := QuestionBank.AllQuestions.ItemQuestions.MessageItemSilver_EN},
                    question{TheQuestion := QuestionBank.AllQuestions.ItemQuestions.MessageItemOrange_EN},
                    question{TheQuestion := QuestionBank.AllQuestions.ItemQuestions.MessageItemPink_EN},
                    question{TheQuestion := QuestionBank.AllQuestions.ItemQuestions.MessageItemPurple_EN},
                    question{TheQuestion := QuestionBank.AllQuestions.VehicleQuestions.MessageVehicleFlashyPaint_EN}
                }
            }
        }


        
    
    var MoviesTVQuestions : []category = array{}

    InitMoviesTVQuestions()<suspends>:void=

        set MoviesTVQuestions = array{
            category{
                Style:=card_style.MoviesTV,
                Name:="Franchise (1/2)",
                NameMessage := QuestionBank.AllCategories.MessageFranchise_EN,
                Questions:=array{
                    question{TheQuestion := QuestionBank.AllQuestions.MovieTVQuestions.MessageMovieFranchisePOTC_EN},
                    question{TheQuestion := QuestionBank.AllQuestions.MovieTVQuestions.MessageMovieFranchiseTWD_EN},
                    question{TheQuestion := QuestionBank.AllQuestions.MovieTVQuestions.MessageMovieFranchiseStrangerThings_EN},
                    question{TheQuestion := QuestionBank.AllQuestions.MovieTVQuestions.MessageMovieFranchiseTerminator_EN},
                    question{TheQuestion := QuestionBank.AllQuestions.MovieTVQuestions.MessageMovieFranchiseAlien_EN},
                    # question{TheQuestion := QuestionBank.AllQuestions.MovieTVQuestions.MessageMovieFranchiseTransformers_EN},
                    question{TheQuestion := QuestionBank.AllQuestions.MovieTVQuestions.MessageMovieFranchiseFuturama_EN},
                    question{TheQuestion := QuestionBank.AllQuestions.MovieTVQuestions.MessageMovieFranchiseJohnWick_EN},
                    question{TheQuestion := QuestionBank.AllQuestions.MovieTVQuestions.MessageMovieFranchiseFamilyGuy_EN},
                    question{TheQuestion := QuestionBank.AllQuestions.MovieTVQuestions.MessageMovieFranchiseRickMorty_EN},
                    question{TheQuestion := QuestionBank.AllQuestions.MovieTVQuestions.MessageMovieFranchiseIndianaJones_EN},
                    question{TheQuestion := QuestionBank.AllQuestions.MovieTVQuestions.MessageMovieFranchiseAvatar_EN},
                    question{TheQuestion := QuestionBank.AllQuestions.MovieTVQuestions.MessageMovieFranchiseRobocop_EN}
            }
        }}


    
        set MoviesTVQuestions += array{
            category{
                Style:=card_style.MoviesTV,
                Name:="Character Nature",
                NameMessage := QuestionBank.AllCategories.MessageCharacterNature_EN,
                Questions:=array{
                    question{TheQuestion := QuestionBank.AllQuestions.CharacterQuestions.MessageCharacterSpeciesHuman_EN},
                    question{TheQuestion := QuestionBank.AllQuestions.CharacterQuestions.MessageCharacterSpeciesAlien_EN},
                    question{TheQuestion := QuestionBank.AllQuestions.CharacterQuestions.MessageCharacterSpeciesRobot_EN},
                    question{TheQuestion := QuestionBank.AllQuestions.CharacterQuestions.MessageCharacterSpeciesMonster_EN},
                    question{TheQuestion := QuestionBank.AllQuestions.MovieTVQuestions.MessageMovieCharacterNatureCartoon_EN},
                    question{TheQuestion := QuestionBank.AllQuestions.MovieTVQuestions.MessageMovieCharacterNatureMainCharacter_EN},
                    question{TheQuestion := QuestionBank.AllQuestions.MovieTVQuestions.MessageMovieCharacterNatureMainVillain_EN}
                }
            }
        }

        set MoviesTVQuestions += array{
            category{
                Style:=card_style.MoviesTV,
                Name:="Career",
                NameMessage := QuestionBank.AllCategories.MessageOccupations_EN,
                Questions:=array{
                    question{TheQuestion := QuestionBank.AllQuestions.MovieTVQuestions.MessageMovieCharacterCareerDetective_EN},
                    question{TheQuestion := QuestionBank.AllQuestions.MovieTVQuestions.MessageMovieCharacterCareerScientist_EN},
                    question{TheQuestion := QuestionBank.AllQuestions.MovieTVQuestions.MessageMovieCharacterCareerSuperhero_EN},
                    question{TheQuestion := QuestionBank.AllQuestions.CharacterQuestions.MessageCharacterVillain_EN},
                    question{TheQuestion := QuestionBank.AllQuestions.MovieTVQuestions.MessageMovieCharacterCareerPirate_EN},
                    question{TheQuestion := QuestionBank.AllQuestions.MovieTVQuestions.MessageMovieCharacterCareerSoldier_EN},
                    question{TheQuestion := QuestionBank.AllQuestions.RemixQuestions.MessageRemixTypeSpyAgent_EN},
                    question{TheQuestion := QuestionBank.AllQuestions.MovieTVQuestions.MessageMovieCharacterCareerWizard_EN},
                    question{TheQuestion := QuestionBank.AllQuestions.MovieTVQuestions.MessageMovieCharacterCareerBiker_EN}
                }
            }
        }

        set MoviesTVQuestions += array{
            category{
                Style:=card_style.MoviesTV,
                Name:="Weapons",
                NameMessage := QuestionBank.AllCategories.MessageWeapons_EN,
                Questions:=array{
                    question{TheQuestion := QuestionBank.AllQuestions.MovieTVQuestions.MessageMovieWeaponBow_EN},
                    question{TheQuestion := QuestionBank.AllQuestions.MovieTVQuestions.MessageMovieWeaponGun_EN},
                    question{TheQuestion := QuestionBank.AllQuestions.MovieTVQuestions.MessageMovieWeaponSword_EN},
                    question{TheQuestion := QuestionBank.AllQuestions.MovieTVQuestions.MessageMovieWeaponChainsaw_EN},
                    question{TheQuestion := QuestionBank.AllQuestions.MovieTVQuestions.MessageMovieWeaponLaser_EN},
                    question{TheQuestion := QuestionBank.AllQuestions.MovieTVQuestions.MessageMovieWeaponBow_EN},
                    question{TheQuestion := QuestionBank.AllQuestions.MovieTVQuestions.MessageMovieWeaponExplosives_EN}
                }
            }
        }
                
                
        set MoviesTVQuestions += array{
            category{
                Style:=card_style.MoviesTV,
                Name:="Traits / Accomplishments",
                NameMessage := QuestionBank.AllCategories.MessageTraitsAccomplishments_EN,
                Questions:=array{
                    question{TheQuestion := QuestionBank.AllQuestions.CharacterQuestions.MessageCharacterPsychic_EN},
                    question{TheQuestion := QuestionBank.AllQuestions.MovieTVQuestions.MessageMovieTraitSuper_EN},
                    question{TheQuestion := QuestionBank.AllQuestions.MovieTVQuestions.MessageMovieTraitCatchphrase_EN},
                    question{TheQuestion := QuestionBank.AllQuestions.GamingLegendsQuestions.MessageLegendIntelligence_EN},
                    question{TheQuestion := QuestionBank.AllQuestions.GamingLegendsQuestions.MessageLegendStrength_EN},
                    question{TheQuestion := QuestionBank.AllQuestions.MovieTVQuestions.MessageMovieAccomplishmentSavedWorld_EN}
                }
            }
        }

        set MoviesTVQuestions += array{
            category{
                Style:=card_style.MoviesTV,
                Name:="Appearance",
                NameMessage := QuestionBank.AllCategories.MessageAppearance_EN,
                Questions:=array{
                    question{TheQuestion := QuestionBank.AllQuestions.CharacterQuestions.MessageCharacterHat_EN},
                    question{TheQuestion := QuestionBank.AllQuestions.CharacterQuestions.MessageCharacterGlasses_EN},
                    question{TheQuestion := QuestionBank.AllQuestions.CharacterQuestions.MessageCharacterMask_EN},
                    question{TheQuestion := QuestionBank.AllQuestions.CharacterQuestions.MessageCharacterHelmet_EN},
                    question{TheQuestion := QuestionBank.AllQuestions.MovieTVQuestions.MessageMovieAppearanceCollaredShirt_EN},
                    question{TheQuestion := QuestionBank.AllQuestions.CharacterQuestions.MessageCharacterEyeColorGlowing_EN},
                    question{TheQuestion := QuestionBank.AllQuestions.CharacterQuestions.MessageCharacterBeard_EN},
                    question{TheQuestion := QuestionBank.AllQuestions.CharacterQuestions.MessageCharacterShowingTeeth_EN},
                    question{TheQuestion := QuestionBank.AllQuestions.CharacterQuestions.MessageCharacterTankTop_EN},
                    question{TheQuestion := QuestionBank.AllQuestions.CharacterQuestions.MessageCharacterJacket_EN}
                }
            }
        }

        set MoviesTVQuestions += array{
            category{
                Style:=card_style.MoviesTV,
                Name:="Affiliations",
                NameMessage := QuestionBank.AllCategories.MessageAffiliations_EN,
                Questions:=array{
                    question{TheQuestion := QuestionBank.AllQuestions.MovieTVQuestions.MessageMovieAffiliationsTeam_EN},
                    question{TheQuestion := QuestionBank.AllQuestions.MovieTVQuestions.MessageMovieAffiliationsLoneWolf_EN},
                    question{TheQuestion := QuestionBank.AllQuestions.MovieTVQuestions.MessageMovieAffiliationsFamily_EN},
                    question{TheQuestion := QuestionBank.AllQuestions.MovieTVQuestions.MessageMovieAffiliationsOrganization_EN},
                    question{TheQuestion := QuestionBank.AllQuestions.MovieTVQuestions.MessageMovieAffiliationsMilitary_EN}
                }
            }
        }

        set MoviesTVQuestions += array{
            category{
                Style:=card_style.MoviesTV,
                Name:="Hair Colors",
                NameMessage := QuestionBank.AllCategories.MessageHairColors_EN,
                Questions:=array{
                    question{TheQuestion := QuestionBank.AllQuestions.CharacterQuestions.MessageCharacterHairColorBlack_EN},
                    question{TheQuestion := QuestionBank.AllQuestions.CharacterQuestions.MessageCharacterHairColorBlonde_EN},
                    question{TheQuestion := QuestionBank.AllQuestions.CharacterQuestions.MessageCharacterHairColorOrangeRed_EN},
                    question{TheQuestion := QuestionBank.AllQuestions.CharacterQuestions.MessageCharacterHairColorWhite_EN},
                    question{TheQuestion := QuestionBank.AllQuestions.CharacterQuestions.MessageCharacterHairColorPurple_EN}
                    
                }
            }
        }

        set MoviesTVQuestions += array{category{
                Style := card_style.MoviesTV
                Name := "Colors"
                NameMessage := QuestionBank.AllCategories.MessageColors_EN
                Questions := array{
                    question{TheQuestion := QuestionBank.AllQuestions.CharacterQuestions.MessageCharacterColorYellow_EN},
                    question{TheQuestion := QuestionBank.AllQuestions.CharacterQuestions.MessageCharacterColorRed_EN},
                    question{TheQuestion := QuestionBank.AllQuestions.CharacterQuestions.MessageCharacterColorBlue_EN},
                    question{TheQuestion := QuestionBank.AllQuestions.CharacterQuestions.MessageCharacterColorPurple_EN},
                    question{TheQuestion := QuestionBank.AllQuestions.CharacterQuestions.MessageCharacterColorBlack_EN},
                    question{TheQuestion := QuestionBank.AllQuestions.CharacterQuestions.MessageCharacterColorWhite_EN},
                    question{TheQuestion := QuestionBank.AllQuestions.CharacterQuestions.MessageCharacterColorBrown_EN},
                    question{TheQuestion := QuestionBank.AllQuestions.CharacterQuestions.MessageCharacterColorBeige_EN}
                }
            }
        }

        set MoviesTVQuestions += array{category{
                Style := card_style.MoviesTV
                Name := "Other"
                NameMessage := QuestionBank.AllCategories.MessageOther_EN
                Questions := array{
                    question{TheQuestion := QuestionBank.AllQuestions.CharacterQuestions.MessageCharacterGenderMale_EN},
                    question{TheQuestion := QuestionBank.AllQuestions.CharacterQuestions.MessageCharacterGenderFemale_EN},
                    question{TheQuestion := QuestionBank.AllQuestions.CharacterQuestions.MessageCharacterDogTags_EN},
                    question{TheQuestion := QuestionBank.AllQuestions.CharacterQuestions.MessageCharacterArmor_EN},
                    question{TheQuestion := QuestionBank.AllQuestions.CharacterQuestions.MessageCharacterTattoo_EN},
                    question{TheQuestion := QuestionBank.AllQuestions.LockerBundleQuestions.MessageLockerOtherScars_EN},
                    question{TheQuestion := QuestionBank.AllQuestions.CharacterQuestions.MessageCharacterFacePaint_EN}
                }
            }
        }
    
            
                        
    var BlocksQuestions : []category = array{}

    var Chapter5Questions : []category = array{}

    InitChapter5Questions()<suspends>:void=

        set Chapter5Questions = array{
            category{
                Style:=card_style.Chapter5,
                Name:="Colors",
                NameMessage := QuestionBank.AllCategories.MessageColors_EN,
                Questions:=array{
                    question{TheQuestion := QuestionBank.AllQuestions.CharacterQuestions.MessageCharacterColorYellow_EN},
                    question{TheQuestion := QuestionBank.AllQuestions.CharacterQuestions.MessageCharacterColorOrange_EN},
                    question{TheQuestion := QuestionBank.AllQuestions.CharacterQuestions.MessageCharacterColorRed_EN},
                    question{TheQuestion := QuestionBank.AllQuestions.CharacterQuestions.MessageCharacterColorPink_EN},
                    question{TheQuestion := QuestionBank.AllQuestions.CharacterQuestions.MessageCharacterColorBlue_EN},
                    question{TheQuestion := QuestionBank.AllQuestions.CharacterQuestions.MessageCharacterColorPurple_EN},
                    question{TheQuestion := QuestionBank.AllQuestions.CharacterQuestions.MessageCharacterColorBlack_EN},
                    question{TheQuestion := QuestionBank.AllQuestions.CharacterQuestions.MessageCharacterColorWhite_EN},
                    question{TheQuestion := QuestionBank.AllQuestions.CharacterQuestions.MessageCharacterColorGold_EN},
                    question{TheQuestion := QuestionBank.AllQuestions.CharacterQuestions.MessageCharacterColorSilver_EN}
                }
            }
        }
    
        set Chapter5Questions += array{
            category{
                Style:=Style:=card_style.Chapter5,
                Name:="How Obtained?",
                NameMessage := QuestionBank.AllCategories.MessageHowObtained_EN,
                Questions:=array{
                    question{TheQuestion := QuestionBank.AllQuestions.CharacterQuestions.MessageCharacterFromBattlePass_EN},
                    question{TheQuestion := QuestionBank.AllQuestions.CharacterQuestions.MessageCharacterFromItemShop_EN},
                    question{TheQuestion := QuestionBank.AllQuestions.CharacterQuestions.MessageCharacterFromPack_EN},
                    question{TheQuestion := QuestionBank.AllQuestions.CharacterQuestions.MessageCharacterFromCrew_EN},
                    question{TheQuestion := QuestionBank.AllQuestions.CharacterQuestions.MessageCharacterFromQuest_EN}
                }
            }
        }
    
        set Chapter5Questions += array{
            category{
                Style:=Style:=card_style.Chapter5,
                Name:="Species",
                NameMessage := QuestionBank.AllCategories.MessageSpecies_EN,
                Questions:=array{
                    question{TheQuestion := QuestionBank.AllQuestions.CharacterQuestions.MessageCharacterSpeciesHuman_EN},
                    question{TheQuestion := QuestionBank.AllQuestions.CharacterQuestions.MessageCharacterSpeciesAnimal_EN},
                    question{TheQuestion := QuestionBank.AllQuestions.CharacterQuestions.MessageCharacterSpeciesFood_EN}
                }
            }
        }
    
        set Chapter5Questions += array{
            category{
                Style:=Style:=card_style.Chapter5,
                Name:="Accessories",
                NameMessage := QuestionBank.AllCategories.MessageAccessories_EN,
                Questions:=array{
                    question{TheQuestion := QuestionBank.AllQuestions.CharacterQuestions.MessageCharacterHat_EN},
                    question{TheQuestion := QuestionBank.AllQuestions.CharacterQuestions.MessageCharacterMask_EN},
                    question{TheQuestion := QuestionBank.AllQuestions.CharacterQuestions.MessageCharacterHelmet_EN},
                    question{TheQuestion := QuestionBank.AllQuestions.CharacterQuestions.MessageCharacterGlasses_EN},
                    question{TheQuestion := QuestionBank.AllQuestions.CharacterQuestions.MessageCharacterNecklace_EN},
                    question{TheQuestion := QuestionBank.AllQuestions.CharacterQuestions.MessageCharacterShowingTeeth_EN},
                    question{TheQuestion := QuestionBank.AllQuestions.CharacterQuestions.MessageCharacterJacket_EN}
                }
            }
        }

        set Chapter5Questions += array{
            category{
                Style:=Style:=card_style.Chapter5,
                Name:= "Other",
                NameMessage := QuestionBank.AllCategories.MessageOther_EN,
                Questions:=array{
                    question{TheQuestion := QuestionBank.AllQuestions.CharacterQuestions.MessageCharacterGenderMale_EN},
                    question{TheQuestion := QuestionBank.AllQuestions.CharacterQuestions.MessageCharacterGenderFemale_EN},
                    question{TheQuestion := QuestionBank.AllQuestions.CharacterQuestions.MessageCharacterDogTags_EN},
                    question{TheQuestion := QuestionBank.AllQuestions.CharacterQuestions.MessageCharacterArmor_EN},
                    question{TheQuestion := QuestionBank.AllQuestions.CharacterQuestions.MessageCharacterTattoo_EN},
                    question{TheQuestion := QuestionBank.AllQuestions.LockerBundleQuestions.MessageLockerOtherScars_EN},
                    question{TheQuestion := QuestionBank.AllQuestions.CharacterQuestions.MessageCharacterFacePaint_EN}
                }
            }
        }

        set Chapter5Questions += array{
            category{
                Style:=Style:=card_style.Chapter5,
                Name:= "Facial Attributes",
                NameMessage := QuestionBank.AllCategories.MessageFacialAttributes_EN,
                Questions := array{
                    question{TheQuestion := QuestionBank.AllQuestions.CharacterQuestions.MessageCharacterFreckles_EN},
                    question{TheQuestion := QuestionBank.AllQuestions.CharacterQuestions.MessageCharacterTannedSkinTone_EN},
                    question{TheQuestion := QuestionBank.AllQuestions.CharacterQuestions.MessageCharacterDarkSkinTone_EN},
                    question{TheQuestion := QuestionBank.AllQuestions.CharacterQuestions.MessageCharacterLightSkinTone_EN},
                    question{TheQuestion := QuestionBank.AllQuestions.CharacterQuestions.MessageCharacterEyeColorBlue_EN},
                    question{TheQuestion := QuestionBank.AllQuestions.CharacterQuestions.MessageCharacterEyeColorBrown_EN},
                    question{TheQuestion := QuestionBank.AllQuestions.CharacterQuestions.MessageCharacterEyeColorGreen_EN}
                }
            }
        }

        set Chapter5Questions += array{category{
            Style := card_style.Chapter5
            Name := "Hair Colors"
            NameMessage := QuestionBank.AllCategories.MessageHairColors_EN
            Questions := array{
                question{TheQuestion := QuestionBank.AllQuestions.CharacterQuestions.MessageCharacterHairColorOrange_EN},
                question{TheQuestion := QuestionBank.AllQuestions.CharacterQuestions.MessageCharacterHairColorWhite_EN},
                question{TheQuestion := QuestionBank.AllQuestions.CharacterQuestions.MessageCharacterHairColorBlack_EN},
                question{TheQuestion := QuestionBank.AllQuestions.CharacterQuestions.MessageCharacterHairColorBlonde_EN},
                question{TheQuestion := QuestionBank.AllQuestions.CharacterQuestions.MessageCharacterHairColorGreen_EN},
                question{TheQuestion := QuestionBank.AllQuestions.CharacterQuestions.MessageCharacterHairColorBrown_EN}
            }
        }}

        set Chapter5Questions += array{category{
            Style := card_style.Chapter5
            Name := "Hair Styles"
            NameMessage := QuestionBank.AllCategories.MessageHairStyles_EN
            Questions := array{
                question{TheQuestion := QuestionBank.AllQuestions.CharacterQuestions.MessageCharacterHairCurly_EN},
                question{TheQuestion := QuestionBank.AllQuestions.CharacterQuestions.MessageCharacterHairBald_EN},
                question{TheQuestion := QuestionBank.AllQuestions.CharacterQuestions.MessageCharacterHairShort_EN},
                question{TheQuestion := QuestionBank.AllQuestions.CharacterQuestions.MessageCharacterHairLong_EN},
                question{TheQuestion := QuestionBank.AllQuestions.CharacterQuestions.MessageCharacterHairPonytail_EN},
                question{TheQuestion := QuestionBank.AllQuestions.CharacterQuestions.MessageCharacterHairFade_EN}
            }
        }}



    var LockersQuestions : []category = array{}

    InitLockersQuestions()<suspends>:void=

        set LockersQuestions = array{
            category{
                Style:=card_style.Lockers,
                Name:="Character Features",
                NameMessage := QuestionBank.AllCategories.MessageFeatures_EN,
                Questions:=array{
                    question{TheQuestion := QuestionBank.AllQuestions.CharacterQuestions.MessageCharacterHat_EN},
                    question{TheQuestion := QuestionBank.AllQuestions.CharacterQuestions.MessageCharacterGlasses_EN},
                    question{TheQuestion := QuestionBank.AllQuestions.CharacterQuestions.MessageCharacterHeadphones_EN},
                    question{TheQuestion := QuestionBank.AllQuestions.CharacterQuestions.MessageCharacterMask_EN},
                    question{TheQuestion := QuestionBank.AllQuestions.CharacterQuestions.MessageCharacterHelmet_EN},
                    question{TheQuestion := QuestionBank.AllQuestions.WinterQuestions.MessageWinterFacialVisible_EN},
                    question{TheQuestion := QuestionBank.AllQuestions.CharacterQuestions.MessageCharacterSpeciesHuman_EN}

                }
            }
        }

        set LockersQuestions += array{
            category{
                Style:=card_style.Lockers,
                Name:="Bundle Themes",
                NameMessage := QuestionBank.AllCategories.MessageTheme_EN,
                Questions:=array{
                    question{TheQuestion := QuestionBank.AllQuestions.LockerBundleQuestions.MessageLockerThemeWinter_EN},
                    question{TheQuestion := QuestionBank.AllQuestions.LockerBundleQuestions.MessageLockerThemeSummer_EN},
                    question{TheQuestion := QuestionBank.AllQuestions.LockerBundleQuestions.MessageLockerThemeMusical_EN},
                    question{TheQuestion := QuestionBank.AllQuestions.LockerBundleQuestions.MessageLockerThemeFuture_EN},
                    question{TheQuestion := QuestionBank.AllQuestions.LockerBundleQuestions.MessageLockerThemeKnight_EN}
                }
            }
        }
    
        set LockersQuestions += array{
            category{
                Style:=card_style.Lockers,
                Name:="Bundle Accessories",
                NameMessage := QuestionBank.AllCategories.MessageAccessories_EN,
                Questions:=array{
                    question{TheQuestion := QuestionBank.AllQuestions.LockerBundleQuestions.MessageLockerAccessoriesGlider_EN},
                    question{TheQuestion := QuestionBank.AllQuestions.LockerBundleQuestions.MessageLockerAccessoriesBackBling_EN},
                    question{TheQuestion := QuestionBank.AllQuestions.LockerBundleQuestions.MessageLockerAccessoriesWrap_EN},
                    question{TheQuestion := QuestionBank.AllQuestions.LockerBundleQuestions.MessageLockerAccessoriesPickaxe_EN},
                    question{TheQuestion := QuestionBank.AllQuestions.LockerBundleQuestions.MessageLockerAccessoriesEmote_EN},
                    question{TheQuestion := QuestionBank.AllQuestions.LockerBundleQuestions.MessageLockerAccessoriesMultipleEmotes_EN}
                }
            }
        }
    
        set LockersQuestions += array{
            category{
                Style:=card_style.Lockers,
                Name:="Uniqueness",
                NameMessage := QuestionBank.AllCategories.MessageUniqueness_EN,
                Questions:=array{
                    question{TheQuestion := QuestionBank.AllQuestions.LockerBundleQuestions.MessageLockerUniqueCharacter_EN},
                    question{TheQuestion := QuestionBank.AllQuestions.LockerBundleQuestions.MessageLockerUniqueGliders_EN},
                    question{TheQuestion := QuestionBank.AllQuestions.LockerBundleQuestions.MessageLockerUniqueBackBling_EN},
                    question{TheQuestion := QuestionBank.AllQuestions.LockerBundleQuestions.MessageLockerUniqueWrap_EN},
                    question{TheQuestion := QuestionBank.AllQuestions.LockerBundleQuestions.MessageLockerUniquePickaxe_EN},
                    question{TheQuestion := QuestionBank.AllQuestions.LockerBundleQuestions.MessageLockerUniqueEmote_EN}
                }
            }
        }
    
        set LockersQuestions += array{category{
            Style := card_style.Lockers
            Name := "Hair Styles"
            NameMessage := QuestionBank.AllCategories.MessageHairStyles_EN
            Questions := array{
                question{TheQuestion := QuestionBank.AllQuestions.CharacterQuestions.MessageCharacterHairCurly_EN},
                question{TheQuestion := QuestionBank.AllQuestions.CharacterQuestions.MessageCharacterHairBald_EN},
                question{TheQuestion := QuestionBank.AllQuestions.CharacterQuestions.MessageCharacterHairShort_EN},
                question{TheQuestion := QuestionBank.AllQuestions.CharacterQuestions.MessageCharacterHairLong_EN},
                question{TheQuestion := QuestionBank.AllQuestions.CharacterQuestions.MessageCharacterHairPonytail_EN},
                question{TheQuestion := QuestionBank.AllQuestions.CharacterQuestions.MessageCharacterHairFade_EN}
            }
        }}

        set LockersQuestions += array{category{
            Style := card_style.Lockers
            Name := "Hair Colors"
            NameMessage := QuestionBank.AllCategories.MessageHairColors_EN
            Questions := array{
                question{TheQuestion := QuestionBank.AllQuestions.CharacterQuestions.MessageCharacterHairColorOrange_EN},
                question{TheQuestion := QuestionBank.AllQuestions.CharacterQuestions.MessageCharacterHairColorWhite_EN},
                question{TheQuestion := QuestionBank.AllQuestions.CharacterQuestions.MessageCharacterHairColorBlack_EN},
                question{TheQuestion := QuestionBank.AllQuestions.CharacterQuestions.MessageCharacterHairColorBlonde_EN},
                question{TheQuestion := QuestionBank.AllQuestions.CharacterQuestions.MessageCharacterHairColorGreen_EN},
                question{TheQuestion := QuestionBank.AllQuestions.CharacterQuestions.MessageCharacterHairColorBrown_EN}
            }
        }}


        set LockersQuestions += array{
            category{
                Style:=Style:=card_style.Lockers,
                Name:= "Facial Attributes",
                NameMessage := QuestionBank.AllCategories.MessageFacialAttributes_EN,
                Questions := array{
                    question{TheQuestion := QuestionBank.AllQuestions.CharacterQuestions.MessageCharacterFreckles_EN},
                    question{TheQuestion := QuestionBank.AllQuestions.CharacterQuestions.MessageCharacterTannedSkinTone_EN},
                    question{TheQuestion := QuestionBank.AllQuestions.CharacterQuestions.MessageCharacterDarkSkinTone_EN},
                    question{TheQuestion := QuestionBank.AllQuestions.CharacterQuestions.MessageCharacterLightSkinTone_EN},
                    question{TheQuestion := QuestionBank.AllQuestions.CharacterQuestions.MessageCharacterEyeColorBlue_EN},
                    question{TheQuestion := QuestionBank.AllQuestions.CharacterQuestions.MessageCharacterEyeColorBrown_EN},
                    question{TheQuestion := QuestionBank.AllQuestions.CharacterQuestions.MessageCharacterEyeColorGreen_EN}
                }
            }
        }

        set LockersQuestions += array{
            category{
                Style:=card_style.Lockers,
                Name:="Bundle Colors",
                NameMessage := QuestionBank.AllCategories.MessageColors_EN,
                Questions:=array{
                    question{TheQuestion := QuestionBank.AllQuestions.ItemQuestions.MessageItemYellow_EN},
                    question{TheQuestion := QuestionBank.AllQuestions.ItemQuestions.MessageItemOrange_EN},
                    question{TheQuestion := QuestionBank.AllQuestions.ItemQuestions.MessageItemRed_EN},
                    question{TheQuestion := QuestionBank.AllQuestions.ItemQuestions.MessageItemPink_EN},
                    question{TheQuestion := QuestionBank.AllQuestions.ItemQuestions.MessageItemBlue_EN},
                    question{TheQuestion := QuestionBank.AllQuestions.ItemQuestions.MessageItemPurple_EN},
                    question{TheQuestion := QuestionBank.AllQuestions.ItemQuestions.MessageItemBlack_EN},
                    question{TheQuestion := QuestionBank.AllQuestions.ItemQuestions.MessageItemWhite_EN},
                    question{TheQuestion := QuestionBank.AllQuestions.ItemQuestions.MessageItemGold_EN},
                    question{TheQuestion := QuestionBank.AllQuestions.ItemQuestions.MessageItemSilver_EN}
                }
            }
        }


        set LockersQuestions += array{
            category{
                Style:=Style:=card_style.Lockers,
                Name:= "Other",
                NameMessage := QuestionBank.AllCategories.MessageOther_EN,
                Questions:=array{
                    question{TheQuestion := QuestionBank.AllQuestions.CharacterQuestions.MessageCharacterDogTags_EN},
                    question{TheQuestion := QuestionBank.AllQuestions.CharacterQuestions.MessageCharacterArmor_EN},
                    question{TheQuestion := QuestionBank.AllQuestions.CharacterQuestions.MessageCharacterTattoo_EN},
                    question{TheQuestion := QuestionBank.AllQuestions.LockerBundleQuestions.MessageLockerOtherScars_EN},
                    question{TheQuestion := QuestionBank.AllQuestions.CharacterQuestions.MessageCharacterFacePaint_EN}
                }
            }
        }


    var BackblingQuestions : []category = array{}

    InitBackblingQuestions()<suspends>:void=
        set BackblingQuestions = array{
            category{
                Style:=card_style.Backblings,
                Name:="Backbling Type 🎒",
                NameMessage := QuestionBank.AllCategories.MessageBackblingType_EN,
                Questions:=array{
                    question{TheQuestion := QuestionBank.AllQuestions.BackBlingQuestions.MessageBackblingTypeShield_EN},
                    question{TheQuestion := QuestionBank.AllQuestions.BackBlingQuestions.MessageBackblingTypeAnimal_EN},
                    question{TheQuestion := QuestionBank.AllQuestions.BackBlingQuestions.MessageBackblingTypeStuffedAnimal_EN},
                    question{TheQuestion := QuestionBank.AllQuestions.BackBlingQuestions.MessageBackblingTypeFood_EN},
                    question{TheQuestion := QuestionBank.AllQuestions.BackBlingQuestions.MessageBackblingTypeSpace_EN},
                    question{TheQuestion := QuestionBank.AllQuestions.BackBlingQuestions.MessageBackblingTypeCape_EN},
                    question{TheQuestion := QuestionBank.AllQuestions.BackBlingQuestions.MessageBackblingTypeBackpack_EN}                    
                }
            }
        }
    
        set BackblingQuestions += array{
            category{
                Style:=card_style.Backblings,
                Name:="Backbling Rarity 🌟",
                NameMessage := QuestionBank.AllCategories.MessageRarity_EN,
                Questions:=array{
                    question{TheQuestion := QuestionBank.AllQuestions.ItemQuestions.MessageItemUncommonRarity_EN},
                    question{TheQuestion := QuestionBank.AllQuestions.ItemQuestions.MessageItemRareRarity_EN},
                    question{TheQuestion := QuestionBank.AllQuestions.ItemQuestions.MessageItemEpicRarity_EN},
                    question{TheQuestion := QuestionBank.AllQuestions.ItemQuestions.MessageItemLegendaryRarity_EN}
                }
            }
        }
    
        set BackblingQuestions += array{
            category{
                Style:=card_style.Backblings,
                Name:="Backbling Features ✨",
                NameMessage := QuestionBank.AllCategories.MessageFeatures_EN,
                Questions:=array{
                    question{TheQuestion := QuestionBank.AllQuestions.BackBlingQuestions.MessageBackblingFeaturesStyles_EN},
                    question{TheQuestion := QuestionBank.AllQuestions.BackBlingQuestions.MessageBackblingFeaturesLights_EN},
                    question{TheQuestion := QuestionBank.AllQuestions.BackBlingQuestions.MessageBackblingFeaturesAnimated_EN},
                    question{TheQuestion := QuestionBank.AllQuestions.BackBlingQuestions.MessageBackblingFeaturesSound_EN},
                    question{TheQuestion := QuestionBank.AllQuestions.BackBlingQuestions.MessageBackblingFeaturesInteractive_EN}
                }
            }
        }
    
        set BackblingQuestions += array{
            category{
                Style:=card_style.Backblings,
                Name:="Backbling Colors",
                NameMessage := QuestionBank.AllCategories.MessageColors_EN,
                Questions:=array{
                    question{TheQuestion := QuestionBank.AllQuestions.ItemQuestions.MessageItemBlue_EN},
                    question{TheQuestion := QuestionBank.AllQuestions.ItemQuestions.MessageItemGreen_EN},
                    question{TheQuestion := QuestionBank.AllQuestions.ItemQuestions.MessageItemYellow_EN},
                    question{TheQuestion := QuestionBank.AllQuestions.ItemQuestions.MessageItemRed_EN},
                    question{TheQuestion := QuestionBank.AllQuestions.ItemQuestions.MessageItemWhite_EN},
                    question{TheQuestion := QuestionBank.AllQuestions.ItemQuestions.MessageItemBlack_EN},
                    question{TheQuestion := QuestionBank.AllQuestions.ItemQuestions.MessageItemSilver_EN},
                    question{TheQuestion := QuestionBank.AllQuestions.ItemQuestions.MessageItemGold_EN},
                    question{TheQuestion := QuestionBank.AllQuestions.ItemQuestions.MessageItemPink_EN},
                    question{TheQuestion := QuestionBank.AllQuestions.ItemQuestions.MessageItemPurple_EN}
                }
            }
        }

        set BackblingQuestions += array{
            category{
                Style:=card_style.Backblings,
                Name:= "Other",
                NameMessage := QuestionBank.AllCategories.MessageOther_EN,
                Questions:=array{
                    question{TheQuestion := QuestionBank.AllQuestions.BackBlingQuestions.MessageBackblingOtherInGameEvent_EN},
                    question{TheQuestion := QuestionBank.AllQuestions.BackBlingQuestions.MessageBackblingOtherEyes_EN},
                    question{TheQuestion := QuestionBank.AllQuestions.BackBlingQuestions.MessageBackblingOtherFace_EN},
                    question{TheQuestion := QuestionBank.AllQuestions.BackBlingQuestions.MessageBackblingOtherMouth_EN},
                    question{TheQuestion := QuestionBank.AllQuestions.BackBlingQuestions.MessageBackblingOtherTail_EN},
                    question{TheQuestion := QuestionBank.AllQuestions.BackBlingQuestions.MessageBackblingOtherPatchesBadgesPins_EN},
                    question{TheQuestion := QuestionBank.AllQuestions.BackBlingQuestions.MessageBackblingOtherLogo_EN}
                }
            }
        }

        set BackblingQuestions += array{
            category{
                Style:=card_style.Backblings,
                Name:= "Community Made!",
                NameMessage := QuestionBank.AllCategories.MessageCommunityMade_EN,
                Questions:=array{
                    question{TheQuestion := QuestionBank.AllQuestions.BackBlingQuestions.MessageBackblingCommunityCreated_EN},
                    question{TheQuestion := QuestionBank.AllQuestions.BackBlingQuestions.MessageBackblingCommunityCreatedD3NNI_EN},
                    question{TheQuestion := QuestionBank.AllQuestions.BackBlingQuestions.MessageBackblingCommunityCreatedSharktoofs_EN},
                    question{TheQuestion := QuestionBank.AllQuestions.BackBlingQuestions.MessageBackblingCommunityCreatedHidden_EN},
                    question{TheQuestion := QuestionBank.AllQuestions.BackBlingQuestions.MessageBackblingCommunityCreatedDahja_EN}
                }
            }
        }
    

    var EmoteQuestions : []category = array{}

    InitEmoteQuestions()<suspends>:void=
        set EmoteQuestions = array{
            category{
                Style:=card_style.Emotes,
                Name:="Emote Type 💃",
                NameMessage := QuestionBank.AllCategories.MessageEmoteType_EN,
                Questions:=array{
                    question{TheQuestion := QuestionBank.AllQuestions.EmoteQuestions.MessageEmoteTypeDanceMove_EN},
                    question{TheQuestion := QuestionBank.AllQuestions.EmoteQuestions.MessageEmoteTypeGesturePose_EN},
                    question{TheQuestion := QuestionBank.AllQuestions.EmoteQuestions.MessageEmoteTypeInteractiveRandom_EN},
                    question{TheQuestion := QuestionBank.AllQuestions.EmoteQuestions.MessageEmoteTypeSyncedOthers_EN},
                    question{TheQuestion := QuestionBank.AllQuestions.EmoteQuestions.MessageEmoteTypeTraversal_EN},
                    question{TheQuestion := QuestionBank.AllQuestions.EmoteQuestions.MessageEmoteTypeProp_EN},
                    question{TheQuestion := QuestionBank.AllQuestions.EmoteQuestions.MessageEmoteTypeLooping_EN}
                }
            }
        }
    
        set EmoteQuestions += array{
            category{
                Style:=card_style.Emotes,
                Name:="Emote Theme 🎭",
                NameMessage := QuestionBank.AllCategories.MessageTheme_EN,
                Questions:=array{
                    question{TheQuestion := QuestionBank.AllQuestions.EmoteQuestions.MessageEmoteThemeScaryFunnySilly_EN},
                    question{TheQuestion := QuestionBank.AllQuestions.EmoteQuestions.MessageEmoteThemeScaryCreepy_EN},                    
                    question{TheQuestion := QuestionBank.AllQuestions.EmoteQuestions.MessageEmoteThemeFoodDrink_EN},
                    question{TheQuestion := QuestionBank.AllQuestions.EmoteQuestions.MessageEmoteUseVictoryRoyale_EN},
                    question{TheQuestion := QuestionBank.AllQuestions.EmoteQuestions.MessageEmoteThemeMilitary_EN},
                    question{TheQuestion := QuestionBank.AllQuestions.EmoteQuestions.MessageEmoteThemeInGameEvent_EN},
                    question{TheQuestion := QuestionBank.AllQuestions.EmoteQuestions.MessageEmoteThemeMeme_EN},
                    question{TheQuestion := QuestionBank.AllQuestions.EmoteQuestions.MessageEmoteThemeSong_EN}
                }
            }
        }
        
        set EmoteQuestions += array{
            category{
                Style:=card_style.Emotes,
                Name:="Used For ...",
                NameMessage := QuestionBank.AllCategories.MessageEmoteUse_EN,
                Questions:=array{
                    question{TheQuestion := QuestionBank.AllQuestions.EmoteQuestions.MessageEmoteUseVictoryRoyale_EN},
                    question{TheQuestion := QuestionBank.AllQuestions.EmoteQuestions.MessageEmoteUseElimination_EN},
                    question{TheQuestion := QuestionBank.AllQuestions.EmoteQuestions.MessageEmoteUseTrickShot_EN},
                    question{TheQuestion := QuestionBank.AllQuestions.EmoteQuestions.MessageEmoteUseFun},
                    question{TheQuestion := QuestionBank.AllQuestions.EmoteQuestions.MessageEmoteUseShowOff_EN},
                    question{TheQuestion := QuestionBank.AllQuestions.EmoteQuestions.MessageEmoteUseTaunt_EN},
                    question{TheQuestion := QuestionBank.AllQuestions.EmoteQuestions.MessageEmoteUseCommunication_EN}
                }
            }
        }

        set EmoteQuestions += array{
            category{
                Style:=card_style.Emotes,
                Name:="Emote Duration ⏱️",
                NameMessage := QuestionBank.AllCategories.MessageEmoteDuration_EN,
                Questions:=array{
                    question{TheQuestion := QuestionBank.AllQuestions.EmoteQuestions.MessageEmoteLengthVeryShort_EN},
                    question{TheQuestion := QuestionBank.AllQuestions.EmoteQuestions.MessageEmoteLengthShort_EN},
                    question{TheQuestion := QuestionBank.AllQuestions.EmoteQuestions.MessageEmoteLengthMedium_EN},
                    question{TheQuestion := QuestionBank.AllQuestions.EmoteQuestions.MessageEmoteLengthLong_EN},
                    question{TheQuestion := QuestionBank.AllQuestions.EmoteQuestions.MessageEmoteLengthVeryLong_EN}
                }
            }
        }

        set EmoteQuestions += array{
            category{
                Style:=card_style.Emotes,
                Name:= "Other",
                NameMessage := QuestionBank.AllCategories.MessageOther_EN,
                Questions:=array{
                    question{TheQuestion := QuestionBank.AllQuestions.EmoteQuestions.MessageEmoteOtherRealLife_EN},
                    question{TheQuestion := QuestionBank.AllQuestions.EmoteQuestions.MessageEmoteOtherPublicPlace_EN},
                    question{TheQuestion := QuestionBank.AllQuestions.EmoteQuestions.MessageEmoteOtherGrandma_EN},
                    question{TheQuestion := QuestionBank.AllQuestions.EmoteQuestions.MessageEmoteOtherBasedRealLife_EN},
                    question{TheQuestion := QuestionBank.AllQuestions.EmoteQuestions.MessageEmoteOtherMovieTV_EN},
                    question{TheQuestion := QuestionBank.AllQuestions.EmoteQuestions.MessageEmoteOtherVideoGame_EN},
                    question{TheQuestion := QuestionBank.AllQuestions.EmoteQuestions.MessageEmoteThemeMeme_EN}
                }
            }
        }

        set EmoteQuestions += array{
            category{
                Style:=card_style.Emotes,
                Name:="Holiday and Seasons 🎄",
                NameMessage := QuestionBank.AllCategories.MessageEmoteHoliday_EN,
                Questions:=array{
                    question{TheQuestion := QuestionBank.AllQuestions.EmoteQuestions.MessageEmoteThemeChristmas_EN},
                    question{TheQuestion := QuestionBank.AllQuestions.EmoteQuestions.MessageEmoteThemeHalloween_EN},
                    question{TheQuestion := QuestionBank.AllQuestions.EmoteQuestions.MessageEmoteHolidaySummer_EN},
                    question{TheQuestion := QuestionBank.AllQuestions.EmoteQuestions.MessageEmoteHolidayWinter_EN},
                    question{TheQuestion := QuestionBank.AllQuestions.EmoteQuestions.MessageEmoteHolidayAny_EN}
                }
            }
        }


    #Umberlla Questions
    var UmbrellasQuestions : []category = array{}

    InitUmbrellasQuestions()<suspends>:void=

        set UmbrellasQuestions = array{
            category{
                Style:=card_style.Umbrellas,
                Name:="Umbrella Material",
                NameMessage := QuestionBank.AllCategories.MessageMaterial_EN,
                Questions:=array{
                    question{TheQuestion := QuestionBank.AllQuestions.UmbrellaQuestions.MessageUmbrellaMaterialMetal_EN},
                    question{TheQuestion := QuestionBank.AllQuestions.UmbrellaQuestions.MessageUmbrellaMaterialWood_EN},
                    question{TheQuestion := QuestionBank.AllQuestions.UmbrellaQuestions.MessageUmbrellaMaterialFabric_EN},
                    question{TheQuestion := QuestionBank.AllQuestions.UmbrellaQuestions.MessageUmbrellaMaterialPaper_EN},
                    question{TheQuestion := QuestionBank.AllQuestions.UmbrellaQuestions.MessageUmbrellaMaterialPlastic_EN},
                    question{TheQuestion := QuestionBank.AllQuestions.UmbrellaQuestions.MessageUmbrellaMaterialPlantMaterial_EN},
                    question{TheQuestion := QuestionBank.AllQuestions.UmbrellaQuestions.MessageUmbrellaMaterialRealLife_EN}
                }
            }
        }

        set UmbrellasQuestions += array{
            category{
                Style:=card_style.Umbrellas,
                Name:="Umbrella Appearance",
                NameMessage := QuestionBank.AllCategories.MessageAppearance_EN,
                Questions:=array{
                    question{TheQuestion := QuestionBank.AllQuestions.UmbrellaQuestions.MessageUmbrellaAppearancePattern_EN},
                    question{TheQuestion := QuestionBank.AllQuestions.UmbrellaQuestions.MessageUmbrellaAppearanceLogo_EN},
                    question{TheQuestion := QuestionBank.AllQuestions.UmbrellaQuestions.MessageUmbrellaAppearanceDesign_EN},
                    question{TheQuestion := QuestionBank.AllQuestions.UmbrellaQuestions.MessageUmbrellaAppearancePicture_EN},
                    question{TheQuestion := QuestionBank.AllQuestions.UmbrellaQuestions.MessageUmbrellaAppearanceColorGradient_EN},
                    question{TheQuestion := QuestionBank.AllQuestions.UmbrellaQuestions.MessageUmbrellaAppearanceSolidColor_EN},
                    question{TheQuestion := QuestionBank.AllQuestions.UmbrellaQuestions.MessageUmbrellaAppearanceSeeThrough_EN}, 
                    question{TheQuestion := QuestionBank.AllQuestions.UmbrellaQuestions.MessageUmbrellaSpecialVFX_EN},
                    question{TheQuestion := QuestionBank.AllQuestions.UmbrellaQuestions.MessageUmbrellaAppearanceUniqueShape_EN}
                }
            }
        }

        set UmbrellasQuestions += array{
            category{
                Style:=card_style.Umbrellas,
                Name:="Umbrella Theme",
                NameMessage := QuestionBank.AllCategories.MessageTheme_EN,
                Questions:=array{
                    question{TheQuestion := QuestionBank.AllQuestions.UmbrellaQuestions.MessageUmbrellaThemeHoliday_EN},
                    question{TheQuestion := QuestionBank.AllQuestions.UmbrellaQuestions.MessageUmbrellaThemeEvent_EN},
                    question{TheQuestion := QuestionBank.AllQuestions.UmbrellaQuestions.MessageUmbrellaThemeMusic_EN},
                    question{TheQuestion := QuestionBank.AllQuestions.UmbrellaQuestions.MessageUmbrellaThemeSports_EN},
                    question{TheQuestion := QuestionBank.AllQuestions.UmbrellaQuestions.MessageUmbrellaThemeFood_EN},
                    question{TheQuestion := QuestionBank.AllQuestions.UmbrellaQuestions.MessageUmbrellaThemeAnimal_EN},
                    question{TheQuestion := QuestionBank.AllQuestions.UmbrellaQuestions.MessageUmbrellaThemeNature_EN}
                }
            }
        }

        set UmbrellasQuestions += array{
            category{
                Style:=card_style.Umbrellas,
                Name:="Umbrella Colors",
                NameMessage := QuestionBank.AllCategories.MessageColors_EN,
                Questions:=array{
                    question{TheQuestion := QuestionBank.AllQuestions.ItemQuestions.MessageItemYellow_EN},
                    question{TheQuestion := QuestionBank.AllQuestions.ItemQuestions.MessageItemOrange_EN},
                    question{TheQuestion := QuestionBank.AllQuestions.ItemQuestions.MessageItemRed_EN},
                    question{TheQuestion := QuestionBank.AllQuestions.ItemQuestions.MessageItemPink_EN},
                    question{TheQuestion := QuestionBank.AllQuestions.ItemQuestions.MessageItemBlue_EN},
                    question{TheQuestion := QuestionBank.AllQuestions.ItemQuestions.MessageItemPurple_EN},
                    question{TheQuestion := QuestionBank.AllQuestions.ItemQuestions.MessageItemBlack_EN},
                    question{TheQuestion := QuestionBank.AllQuestions.ItemQuestions.MessageItemWhite_EN},
                    question{TheQuestion := QuestionBank.AllQuestions.ItemQuestions.MessageItemGold_EN},
                    question{TheQuestion := QuestionBank.AllQuestions.ItemQuestions.MessageItemSilver_EN}
                }
            }
        }

        set UmbrellasQuestions += array{
            category{
                Style:=card_style.Umbrellas,
                Name:= "Chapters",
                NameMessage := QuestionBank.AllCategories.MessageChapters_EN,
                Questions:=array{
                    question{TheQuestion := QuestionBank.AllQuestions.ItemQuestions.MessageItemChapterOne_EN},
                    question{TheQuestion := QuestionBank.AllQuestions.ItemQuestions.MessageItemChapterTwo_EN},
                    question{TheQuestion := QuestionBank.AllQuestions.ItemQuestions.MessageItemChapterThree_EN},
                    question{TheQuestion := QuestionBank.AllQuestions.ItemQuestions.MessageItemChapterFour_EN},
                    question{TheQuestion := QuestionBank.AllQuestions.ItemQuestions.MessageItemChapterFive_EN}
                }
            }
        }

        set UmbrellasQuestions += array{
            category{
                Style:=card_style.Umbrellas,
                Name:= "Seasons",
                NameMessage := QuestionBank.AllCategories.MessageSeasons_EN,
                Questions:=array{
                    question{TheQuestion := QuestionBank.AllQuestions.UmbrellaQuestions.MessageUmbreallaAnySeason1_EN},
                    question{TheQuestion := QuestionBank.AllQuestions.UmbrellaQuestions.MessageUmbrellaAnySeason2_EN},
                    question{TheQuestion := QuestionBank.AllQuestions.UmbrellaQuestions.MessageUmbrellaAnySeason3_EN},
                    question{TheQuestion := QuestionBank.AllQuestions.UmbrellaQuestions.MessageUmbrellaAnySeason4_EN},
                    question{TheQuestion := QuestionBank.AllQuestions.UmbrellaQuestions.MessageUmbrellaAnySeason5_EN},
                    question{TheQuestion := QuestionBank.AllQuestions.UmbrellaQuestions.MessageUmbrellaAnySeason6_EN},
                    question{TheQuestion := QuestionBank.AllQuestions.UmbrellaQuestions.MessageUmbrellaAnySeason7_EN},
                    question{TheQuestion := QuestionBank.AllQuestions.UmbrellaQuestions.MessageUmbrellaAnySeason8_EN},
                    question{TheQuestion := QuestionBank.AllQuestions.UmbrellaQuestions.MessageUmbrellaAnySeason9_EN},
                    question{TheQuestion := QuestionBank.AllQuestions.UmbrellaQuestions.MessageUmbrellaAnySeason10_EN}
                }
            }
        }

        set UmbrellasQuestions += array{
            category{
                Style:=card_style.Umbrellas,
                Name:= "Other",
                NameMessage := QuestionBank.AllCategories.MessageOther_EN,
                Questions:=array{
                    question{TheQuestion := QuestionBank.AllQuestions.CharacterQuestions.MessageCharacterGenderMale_EN},
                    question{TheQuestion := QuestionBank.AllQuestions.CharacterQuestions.MessageCharacterGenderFemale_EN},
                    question{TheQuestion := QuestionBank.AllQuestions.CharacterQuestions.MessageCharacterSpeciesHuman_EN},
                    question{TheQuestion := QuestionBank.AllQuestions.UmbrellaQuestions.MessageUmbrellaOtherLootLake_EN},
                    question{TheQuestion := QuestionBank.AllQuestions.UmbrellaQuestions.MessageUmbrellaOtherSnow_EN},
                    question{TheQuestion := QuestionBank.AllQuestions.UmbrellaQuestions.MessageUmbrellaOtherOcean_EN},
                    question{TheQuestion := QuestionBank.AllQuestions.UmbrellaQuestions.MessageUmbrellaOtherDesert_EN}
                }
            }
        }

    var PeelyFishstickQuestions: []category = array{}

    InitPeelyFishstickQuestions()<suspends>:void=
            set PeelyFishstickQuestions += array{
                category{
                    Style:=card_style.PeelyFishstick,
                    Name:="Outfit Styles",
                    NameMessage := QuestionBank.AllCategories.MessageOutfitStyles_EN,
                    Questions:=array{
                        question{TheQuestion := QuestionBank.AllQuestions.PeelyFishstickQuestions.MessagePFStylePeely_EN},
                        question{TheQuestion := QuestionBank.AllQuestions.PeelyFishstickQuestions.MessagePFStyleFishstick_EN},
                        question{TheQuestion := QuestionBank.AllQuestions.SpookyQuestions.MessageSpookyTypeSkeleton_EN},
                        question{TheQuestion := QuestionBank.AllQuestions.CharacterQuestions.MessageCharacterArmor_EN},
                        question{TheQuestion := QuestionBank.AllQuestions.CharacterQuestions.MessageCharacterEyeColorGlowing_EN},
                        question{TheQuestion := QuestionBank.AllQuestions.PeelyFishstickQuestions.MessagePFStyleToon_EN},
                        question{TheQuestion := QuestionBank.AllQuestions.PeelyFishstickQuestions.MessagePFStyleThicc_EN},
                        question{TheQuestion := QuestionBank.AllQuestions.PeelyFishstickQuestions.MessagePFStylePeeled_EN},
                        question{TheQuestion := QuestionBank.AllQuestions.PeelyFishstickQuestions.MessagePFStyleEdible_EN}
                    }
                }
            }
        
            set PeelyFishstickQuestions += array{
                category{
                    Style:=card_style.PeelyFishstick,
                    Name:="Appearance",
                    NameMessage := QuestionBank.AllCategories.MessageAppearance_EN,
                    Questions:=array{
                        question{TheQuestion := QuestionBank.AllQuestions.PeelyFishstickQuestions.MessagePFAppearanceFrozen_EN},
                        question{TheQuestion := QuestionBank.AllQuestions.WinterQuestions.MessageWinterAttireCandy_EN},
                        question{TheQuestion := QuestionBank.AllQuestions.CharacterQuestions.MessageCharacterSuit_EN},
                        question{TheQuestion := QuestionBank.AllQuestions.PeelyFishstickQuestions.MessagePFAppearanceShorts_EN},
                        question{TheQuestion := QuestionBank.AllQuestions.CharacterQuestions.MessageCharacterSpeciesRobot_EN},
                        question{TheQuestion := QuestionBank.AllQuestions.PeelyFishstickQuestions.MessagePFAppearanceMythicFantasy_EN},
                        question{TheQuestion := QuestionBank.AllQuestions.GamingLegendsQuestions.MessageLegendHistorical_EN},
                        question{TheQuestion := QuestionBank.AllQuestions.PeelyFishstickQuestions.MessagePFAppearanceAnotherSkin_EN},
                        question{TheQuestion := QuestionBank.AllQuestions.CharacterQuestions.MessageCharacterFacePaint_EN}

                    }
                }
            }
        
            set PeelyFishstickQuestions += array{
                category{
                    Style:=card_style.PeelyFishstick,
                    Name:="Accessories",
                    NameMessage := QuestionBank.AllCategories.MessageAccessories_EN,
                    Questions:=array{
                        question{TheQuestion := QuestionBank.AllQuestions.CharacterQuestions.MessageCharacterBandana_EN},
                        question{TheQuestion := QuestionBank.AllQuestions.CharacterQuestions.MessageCharacterGlasses_EN},
                        question{TheQuestion := QuestionBank.AllQuestions.CharacterQuestions.MessageCharacterHat_EN},
                        question{TheQuestion := QuestionBank.AllQuestions.CharacterQuestions.MessageCharacterMask_EN},
                        question{TheQuestion := QuestionBank.AllQuestions.PeelyFishstickQuestions.MessagePFAccessoriesGloves_EN},
                        question{TheQuestion := QuestionBank.AllQuestions.CharacterQuestions.MessageCharacterShirt_EN},
                        question{TheQuestion := QuestionBank.AllQuestions.PeelyFishstickQuestions.MessagePFAccessoriesBeltPouch_EN},
                        question{TheQuestion := QuestionBank.AllQuestions.PeelyFishstickQuestions.MessagePFAccessoriesStripes_EN}
                    }
                }
            }

            set PeelyFishstickQuestions += array{
                category{
                    Style:=card_style.PeelyFishstick,
                    Name:="Other",
                    NameMessage := QuestionBank.AllCategories.MessageOther_EN,
                    Questions:=array{
                        question{TheQuestion := QuestionBank.AllQuestions.PeelyFishstickQuestions.MessagePFOtherLookAngry_EN},
                        question{TheQuestion := QuestionBank.AllQuestions.PeelyFishstickQuestions.MessagePFOtherLookHappy_EN},
                        question{TheQuestion := QuestionBank.AllQuestions.CharacterQuestions.MessageCharacterBeard_EN},
                        question{TheQuestion := QuestionBank.AllQuestions.PeelyFishstickQuestions.MessagePFOtherOutfitLogo_EN},
                        question{TheQuestion := QuestionBank.AllQuestions.PeelyFishstickQuestions.MessagePFOtherOutfitWords_EN},
                        question{TheQuestion := QuestionBank.AllQuestions.CharacterQuestions.MessageCharacterGenderMale_EN},
                        question{TheQuestion := QuestionBank.AllQuestions.CharacterQuestions.MessageCharacterGenderFemale_EN},
                        question{TheQuestion := QuestionBank.AllQuestions.PeelyFishstickQuestions.MessagePFOtherBlender_EN}
                    }
                }
            }
        
            set PeelyFishstickQuestions += array{
                category{
                    Style:=card_style.PeelyFishstick,
                    Name:="Colors",
                    NameMessage := QuestionBank.AllCategories.MessageColors_EN,
                    Questions:=array{
                        question{TheQuestion := QuestionBank.AllQuestions.CharacterQuestions.MessageCharacterColorBlue_EN},
                        question{TheQuestion := QuestionBank.AllQuestions.CharacterQuestions.MessageCharacterColorGreen_EN},
                        question{TheQuestion := QuestionBank.AllQuestions.CharacterQuestions.MessageCharacterColorYellow_EN},
                        question{TheQuestion := QuestionBank.AllQuestions.CharacterQuestions.MessageCharacterColorRed_EN},
                        question{TheQuestion := QuestionBank.AllQuestions.CharacterQuestions.MessageCharacterColorWhite_EN},
                        question{TheQuestion := QuestionBank.AllQuestions.CharacterQuestions.MessageCharacterColorBlack_EN},
                        question{TheQuestion := QuestionBank.AllQuestions.CharacterQuestions.MessageCharacterColorOrange_EN},
                        question{TheQuestion := QuestionBank.AllQuestions.CharacterQuestions.MessageCharacterColorPink_EN},
                        question{TheQuestion := QuestionBank.AllQuestions.CharacterQuestions.MessageCharacterColorPurple_EN},
                        question{TheQuestion := QuestionBank.AllQuestions.CharacterQuestions.MessageCharacterColorGold_EN},
                        question{TheQuestion := QuestionBank.AllQuestions.CharacterQuestions.MessageCharacterColorSilver_EN},
                        question{TheQuestion := QuestionBank.AllQuestions.CharacterQuestions.MessageCharacterColorBronze_EN}
                    }
                }
            }


    var RemixQuestions: []category = array{}

    InitRemixQuestions()<suspends>:void=
        set RemixQuestions = array{
            category{
                Style:=card_style.Remix,
                Name:="Character Type",
                NameMessage := QuestionBank.AllCategories.MessageCharacterType_EN,
                Questions:=array{
                    question{TheQuestion := QuestionBank.AllQuestions.RemixQuestions.MessageRemixTypeRemix_EN},
                    question{TheQuestion := QuestionBank.AllQuestions.RemixQuestions.MessageRemixTypeMashup_EN},
                    question{TheQuestion := QuestionBank.AllQuestions.RemixQuestions.MessageRemixTypeDuo_EN},
                    question{TheQuestion := QuestionBank.AllQuestions.RemixQuestions.MessageRemixTypeAltStyle_EN},
                    question{TheQuestion := QuestionBank.AllQuestions.RemixQuestions.MessageRemixTypeSpyAgent_EN},
                    question{TheQuestion := QuestionBank.AllQuestions.CharacterQuestions.MessageCharacterSpeciesHuman_EN},
                    question{TheQuestion := QuestionBank.AllQuestions.CharacterQuestions.MessageCharacterSpeciesRobot_EN},
                    question{TheQuestion := QuestionBank.AllQuestions.CharacterQuestions.MessageCharacterSpeciesAnimal_EN},
                    question{TheQuestion := QuestionBank.AllQuestions.CharacterQuestions.MessageCharacterSpeciesFood_EN},
                    question{TheQuestion := QuestionBank.AllQuestions.RemixQuestions.MessageRemixTypeStuffedAnimal_EN}
                }
            }
        }
    
        set RemixQuestions += array{
            category{
                Style:=card_style.Remix,
                Name:="Appearance",
                NameMessage := QuestionBank.AllCategories.MessageAppearance_EN,
                Questions:=array{
                    question{TheQuestion := QuestionBank.AllQuestions.RemixQuestions.MessageRemixAppearanceHorns_EN},
                    question{TheQuestion := QuestionBank.AllQuestions.RemixQuestions.MessageRemixAppearanceMuscular_EN},
                    question{TheQuestion := QuestionBank.AllQuestions.CharacterQuestions.MessageCharacterSuit_EN},
                    question{TheQuestion := QuestionBank.AllQuestions.CharacterQuestions.MessageCharacterGlasses_EN},
                    question{TheQuestion := QuestionBank.AllQuestions.CharacterQuestions.MessageCharacterMask_EN},
                    question{TheQuestion := QuestionBank.AllQuestions.CharacterQuestions.MessageCharacterHat_EN},
                    question{TheQuestion := QuestionBank.AllQuestions.RemixQuestions.MessageRemixAppearanceAnimalPattern_EN}
                }
            }
        }

        set RemixQuestions += array{
            category{
                Style:=card_style.Remix,
                Name:="Hair Styles",
                NameMessage := QuestionBank.AllCategories.MessageHairStyles_EN,
                Questions:=array{
                    question{TheQuestion := QuestionBank.AllQuestions.CharacterQuestions.MessageCharacterHairCurly_EN},
                    question{TheQuestion := QuestionBank.AllQuestions.CharacterQuestions.MessageCharacterHairBald_EN},
                    question{TheQuestion := QuestionBank.AllQuestions.CharacterQuestions.MessageCharacterHairShort_EN},
                    question{TheQuestion := QuestionBank.AllQuestions.CharacterQuestions.MessageCharacterHairLong_EN},
                    question{TheQuestion := QuestionBank.AllQuestions.CharacterQuestions.MessageCharacterHairPonytail_EN},
                    question{TheQuestion := QuestionBank.AllQuestions.CharacterQuestions.MessageCharacterHairFade_EN}
                }
            }
        }
    
        set RemixQuestions += array{
            category{
                Style:=card_style.Remix,
                Name:="Outfit Features",
                NameMessage := QuestionBank.AllCategories.MessageOutfitFeatures_EN,
                Questions:=array{
                    question{TheQuestion := QuestionBank.AllQuestions.RemixQuestions.MessageRemixOutfitSuspenders_EN},
                    question{TheQuestion := QuestionBank.AllQuestions.CharacterQuestions.MessageCharacterShirt_EN},
                    question{TheQuestion := QuestionBank.AllQuestions.CharacterQuestions.MessageCharacterTie_EN},
                    question{TheQuestion := QuestionBank.AllQuestions.PeelyFishstickQuestions.MessagePFAccessoriesGloves_EN},
                    question{TheQuestion := QuestionBank.AllQuestions.RemixQuestions.MessageRemixOutfitBelt_EN},
                    question{TheQuestion := QuestionBank.AllQuestions.RemixQuestions.MessageRemixOutfitVest_EN},
                    question{TheQuestion := QuestionBank.AllQuestions.CharacterQuestions.MessageCharacterNecklace_EN},
                    question{TheQuestion := QuestionBank.AllQuestions.RemixQuestions.MessageRemixOutfitAnyAccessories_EN}
                }
            }
        }
    
        set RemixQuestions += array{
            category{
                Style:=card_style.Remix,
                Name:="Other",
                NameMessage := QuestionBank.AllCategories.MessageOther_EN,
                Questions:=array{
                    question{TheQuestion := QuestionBank.AllQuestions.CharacterQuestions.MessageCharacterGenderMale_EN},
                    question{TheQuestion := QuestionBank.AllQuestions.CharacterQuestions.MessageCharacterGenderFemale_EN},
                    question{TheQuestion := QuestionBank.AllQuestions.RemixQuestions.MessageRemixOtherMainCharacter_EN},
                    question{TheQuestion := QuestionBank.AllQuestions.RemixQuestions.MessageRemixOtherBilliards_EN},
                    question{TheQuestion := QuestionBank.AllQuestions.PeelyFishstickQuestions.MessagePFOtherOutfitLogo_EN},
                    question{TheQuestion := QuestionBank.AllQuestions.PeelyFishstickQuestions.MessagePFOtherOutfitWords_EN},
                    question{TheQuestion := QuestionBank.AllQuestions.LockerBundleQuestions.MessageLockerOtherScars_EN},
                    question{TheQuestion := QuestionBank.AllQuestions.CharacterQuestions.MessageCharacterTattoo_EN},
                    question{TheQuestion := QuestionBank.AllQuestions.CharacterQuestions.MessageCharacterFacePaint_EN},
                    question{TheQuestion := QuestionBank.AllQuestions.CharacterQuestions.MessageCharacterEyeColorBlack_EN},
                    question{TheQuestion := QuestionBank.AllQuestions.RemixQuestions.MessageRemixOtherSlurpJuice_EN}
                }
            }
        }
    
        set RemixQuestions += array{
            category{
                Style:=card_style.Remix,
                Name:="Colors",
                NameMessage := QuestionBank.AllCategories.MessageColors_EN,
                    Questions:=array{
                        question{TheQuestion := QuestionBank.AllQuestions.CharacterQuestions.MessageCharacterColorBlue_EN},
                        question{TheQuestion := QuestionBank.AllQuestions.CharacterQuestions.MessageCharacterColorGreen_EN},
                        question{TheQuestion := QuestionBank.AllQuestions.CharacterQuestions.MessageCharacterColorYellow_EN},
                        question{TheQuestion := QuestionBank.AllQuestions.CharacterQuestions.MessageCharacterColorRed_EN},
                        question{TheQuestion := QuestionBank.AllQuestions.CharacterQuestions.MessageCharacterColorWhite_EN},
                        question{TheQuestion := QuestionBank.AllQuestions.CharacterQuestions.MessageCharacterColorBlack_EN},
                        question{TheQuestion := QuestionBank.AllQuestions.CharacterQuestions.MessageCharacterColorOrange_EN},
                        question{TheQuestion := QuestionBank.AllQuestions.CharacterQuestions.MessageCharacterColorPink_EN},
                        question{TheQuestion := QuestionBank.AllQuestions.CharacterQuestions.MessageCharacterColorPurple_EN},
                        question{TheQuestion := QuestionBank.AllQuestions.CharacterQuestions.MessageCharacterColorGold_EN},
                        question{TheQuestion := QuestionBank.AllQuestions.CharacterQuestions.MessageCharacterColorSilver_EN},
                        question{TheQuestion := QuestionBank.AllQuestions.CharacterQuestions.MessageCharacterColorBronze_EN}
                    }
                }
            }
    

    var FireIceQuestions: []category = array{}

    InitFireIceQuestions()<suspends>:void=
        set FireIceQuestions = array{
            category{
                Style:=card_style.FireIce,
                Name:="Appearance",
                NameMessage := QuestionBank.AllCategories.MessageAppearance_EN,
                Questions:=array{
                    question{TheQuestion := QuestionBank.AllQuestions.FireIceQuestions.MessageFire_EN},
                    question{TheQuestion := QuestionBank.AllQuestions.FireIceQuestions.MessageIce_EN},
                    question{TheQuestion := QuestionBank.AllQuestions.FireIceQuestions.MessageMadeOfIce_EN},
                    question{TheQuestion := QuestionBank.AllQuestions.CharacterQuestions.MessageCharacterSpeciesHuman_EN},
                    question{TheQuestion := QuestionBank.AllQuestions.CharacterQuestions.MessageCharacterSpeciesMonster_EN},
                    question{TheQuestion := QuestionBank.AllQuestions.CharacterQuestions.MessageCharacterSpeciesFood_EN},
                    question{TheQuestion := QuestionBank.AllQuestions.CharacterQuestions.MessageCharacterSpeciesAnimal_EN},            
                    question{TheQuestion := QuestionBank.AllQuestions.CharacterQuestions.MessageCharacterChainmail_EN},
                    question{TheQuestion := QuestionBank.AllQuestions.CharacterQuestions.MessageCharacterSuit_EN},
                    question{TheQuestion := QuestionBank.AllQuestions.CharacterQuestions.MessageCharacterOutfitFur_EN},
                    question{TheQuestion := QuestionBank.AllQuestions.SpookyQuestions.MessageSpookyAppearanceSharpTeeth_EN},
                    question{TheQuestion := QuestionBank.AllQuestions.WinterQuestions.MessageWinterFacialVisible_EN}
                }
            }
        }

        set FireIceQuestions += array{
            category{
                Style:=card_style.FireIce,
                Name:="Accessories",
                NameMessage := QuestionBank.AllCategories.MessageAccessories_EN,
                Questions:=array{
                    question{TheQuestion := QuestionBank.AllQuestions.CharacterQuestions.MessageCharacterGlasses_EN},
                    question{TheQuestion := QuestionBank.AllQuestions.CharacterQuestions.MessageCharacterHat_EN},
                    question{TheQuestion := QuestionBank.AllQuestions.CharacterQuestions.MessageCharacterMask_EN},
                    question{TheQuestion := QuestionBank.AllQuestions.CharacterQuestions.MessageCharacterHelmet_EN},
                    question{TheQuestion := QuestionBank.AllQuestions.CharacterQuestions.MessageCharacterNecklace_EN},
                    question{TheQuestion := QuestionBank.AllQuestions.CharacterQuestions.MessageCharacterShirt_EN},
                    question{TheQuestion := QuestionBank.AllQuestions.CharacterQuestions.MessageCharacterTie_EN},
                    question{TheQuestion := QuestionBank.AllQuestions.CharacterQuestions.MessageCharacterBandana_EN},
                    question{TheQuestion := QuestionBank.AllQuestions.CharacterQuestions.MessageCharacterDogTags_EN},
                    question{TheQuestion := QuestionBank.AllQuestions.PeelyFishstickQuestions.MessagePFAccessoriesBeltPouch_EN},
                    question{TheQuestion := QuestionBank.AllQuestions.RemixQuestions.MessageRemixOutfitBelt_EN}
                }
            }
        }

        set FireIceQuestions += array{
            category{
                Style:=card_style.FireIce,
                Name:="Hair Styles",
                NameMessage := QuestionBank.AllCategories.MessageHairStyles_EN,
                Questions:=array{
                    question{TheQuestion := QuestionBank.AllQuestions.CharacterQuestions.MessageCharacterHairCurly_EN},
                    question{TheQuestion := QuestionBank.AllQuestions.CharacterQuestions.MessageCharacterHairBald_EN},
                    question{TheQuestion := QuestionBank.AllQuestions.CharacterQuestions.MessageCharacterHairShort_EN},
                    question{TheQuestion := QuestionBank.AllQuestions.CharacterQuestions.MessageCharacterHairLong_EN},
                    question{TheQuestion := QuestionBank.AllQuestions.CharacterQuestions.MessageCharacterHairPonytail_EN},
                    question{TheQuestion := QuestionBank.AllQuestions.CharacterQuestions.MessageCharacterHairPigtails_EN},
                    question{TheQuestion := QuestionBank.AllQuestions.CharacterQuestions.MessageCharacterHairMohawk_EN},
                    question{TheQuestion := QuestionBank.AllQuestions.CharacterQuestions.MessageCharacterHairBraids_EN},
                    question{TheQuestion := QuestionBank.AllQuestions.CharacterQuestions.MessageCharacterHair_EN}
                }
            }
        }


        set FireIceQuestions += array{
            category{
                Style:=card_style.FireIce,
                Name:="Other",
                NameMessage := QuestionBank.AllCategories.MessageOther_EN,
                Questions:=array{
                    question{TheQuestion := QuestionBank.AllQuestions.CharacterQuestions.MessageCharacterEyeColorBlue_EN},
                    question{TheQuestion := QuestionBank.AllQuestions.CharacterQuestions.MessageCharacterEyeColorPurple_EN},
                    question{TheQuestion := QuestionBank.AllQuestions.CharacterQuestions.MessageCharacterEyeColorRed_EN},
                    question{TheQuestion := QuestionBank.AllQuestions.CharacterQuestions.MessageCharacterEyeColorOrange_EN},
                    question{TheQuestion := QuestionBank.AllQuestions.CharacterQuestions.MessageCharacterEyeColorBlack_EN},
                    question{TheQuestion := QuestionBank.AllQuestions.CharacterQuestions.MessageCharacterEyeColorWhite_EN},
                    question{TheQuestion := QuestionBank.AllQuestions.CharacterQuestions.MessageCharacterEyeColorGlowing_EN},
                    question{TheQuestion := QuestionBank.AllQuestions.CharacterQuestions.MessageCharacterVisibleEyes_EN},
                    question{TheQuestion := QuestionBank.AllQuestions.CharacterQuestions.MessageCharacterFacePaint_EN},
                    question{TheQuestion := QuestionBank.AllQuestions.CharacterQuestions.MessageCharacterTattoo_EN}
                }
            }
        }

        set FireIceQuestions += array{
            category{
                Style:=card_style.FireIce,
                Name:="Colors",
                NameMessage := QuestionBank.AllCategories.MessageColors_EN,
                Questions:=array{
                    question{TheQuestion := QuestionBank.AllQuestions.CharacterQuestions.MessageCharacterColorBlue_EN},
                    question{TheQuestion := QuestionBank.AllQuestions.CharacterQuestions.MessageCharacterColorGreen_EN},
                    question{TheQuestion := QuestionBank.AllQuestions.CharacterQuestions.MessageCharacterColorYellow_EN},
                    question{TheQuestion := QuestionBank.AllQuestions.CharacterQuestions.MessageCharacterColorRed_EN},
                    question{TheQuestion := QuestionBank.AllQuestions.CharacterQuestions.MessageCharacterColorWhite_EN},
                    question{TheQuestion := QuestionBank.AllQuestions.CharacterQuestions.MessageCharacterColorBlack_EN},
                    question{TheQuestion := QuestionBank.AllQuestions.CharacterQuestions.MessageCharacterColorOrange_EN},
                    question{TheQuestion := QuestionBank.AllQuestions.CharacterQuestions.MessageCharacterColorPink_EN},
                    question{TheQuestion := QuestionBank.AllQuestions.CharacterQuestions.MessageCharacterColorPurple_EN},
                    question{TheQuestion := QuestionBank.AllQuestions.CharacterQuestions.MessageCharacterColorGold_EN},
                    question{TheQuestion := QuestionBank.AllQuestions.CharacterQuestions.MessageCharacterColorSilver_EN},
                    question{TheQuestion := QuestionBank.AllQuestions.CharacterQuestions.MessageCharacterColorBronze_EN}
                }
            }
        }

    var DCQuestions: []category = array{}

    InitDCQuestions()<suspends>:void=
        set DCQuestions = array{
            category{
                Style:=card_style.DC,
                Name:="Super Powers",
                NameMessage := QuestionBank.AllCategories.MessageSuperPowers_EN,
                Questions:=array{
                    question{TheQuestion := QuestionBank.AllQuestions.DCQuestions.MessageDCXRayVision_EN},
                    question{TheQuestion := QuestionBank.AllQuestions.DCQuestions.MessageDCBreatheUnderwater_EN},
                    question{TheQuestion := QuestionBank.AllQuestions.DCQuestions.MessageDCSuperSpeed_EN},
                    question{TheQuestion := QuestionBank.AllQuestions.DCQuestions.MessageDCInvisibility_EN},
                    question{TheQuestion := QuestionBank.AllQuestions.DCQuestions.MessageDCSizeChange_EN},
                    question{TheQuestion := QuestionBank.AllQuestions.DCQuestions.MessageDCTalkToAnimals_EN},
                    question{TheQuestion := QuestionBank.AllQuestions.DCQuestions.MessageDCBatman_EN}
                }
            }
        }

        set DCQuestions += array{
            category{
                Style:=card_style.DC,
                Name:="Origins & Background",
                NameMessage := QuestionBank.AllCategories.MessageOrigin_EN,
                Questions:=array{
                    question{TheQuestion := QuestionBank.AllQuestions.DCQuestions.MessageDCKryptonian_EN},
                    question{TheQuestion := QuestionBank.AllQuestions.DCQuestions.MessageDCAmazon_EN},
                    question{TheQuestion := QuestionBank.AllQuestions.DCQuestions.MessageDCDivinePower_EN},
                    question{TheQuestion := QuestionBank.AllQuestions.DCQuestions.MessageDCLeagueOfAssassins_EN},
                    question{TheQuestion := QuestionBank.AllQuestions.DCQuestions.MessageDCBornGotham_EN},
                    question{TheQuestion := QuestionBank.AllQuestions.DCQuestions.MessageDCLabExperiment_EN}
                }
            }
        }

        set DCQuestions += array{
            category{
                Style:=card_style.DC,
                Name:="Teams & Allegiances",
                NameMessage := QuestionBank.AllCategories.MessageTeamsEnemies_EN,
                Questions:=array{
                    question{TheQuestion := QuestionBank.AllQuestions.DCQuestions.MessageDCJusticeLeagueFounder_EN},
                    question{TheQuestion := QuestionBank.AllQuestions.DCQuestions.MessageDCTeenTitans_EN},
                    question{TheQuestion := QuestionBank.AllQuestions.DCQuestions.MessageDCSuicideSquad_EN},
                    question{TheQuestion := QuestionBank.AllQuestions.DCQuestions.MessageDCBatmanEnemies_EN},
                    question{TheQuestion := QuestionBank.AllQuestions.DCQuestions.MessageDCLexCorp_EN},
                    question{TheQuestion := QuestionBank.AllQuestions.DCQuestions.MessageDCFoughtDarkseid_EN}
                }
            }
        }

        set DCQuestions += array{
            category{
                Style:=card_style.DC,
                Name:="Appearance & Costume",
                NameMessage := QuestionBank.AllCategories.MessageAppearance_EN,
                Questions:=array{
                    question{TheQuestion := QuestionBank.AllQuestions.DCQuestions.MessageDCBatSymbol_EN},
                    question{TheQuestion := QuestionBank.AllQuestions.DCQuestions.MessageDCUtilityBelt_EN},
                    question{TheQuestion := QuestionBank.AllQuestions.CharacterQuestions.MessageCharacterSuit_EN},
                    question{TheQuestion := QuestionBank.AllQuestions.CharacterQuestions.MessageCharacterHelmet_EN},
                    question{TheQuestion := QuestionBank.AllQuestions.CharacterQuestions.MessageCharacterGlasses_EN},
                    question{TheQuestion := QuestionBank.AllQuestions.CharacterQuestions.MessageCharacterMask_EN},
                    question{TheQuestion := QuestionBank.AllQuestions.CharacterQuestions.MessageCharacterCape_EN},
                    question{TheQuestion := QuestionBank.AllQuestions.CharacterQuestions.MessageCharacterFacePaint_EN},
                    question{TheQuestion := QuestionBank.AllQuestions.CharacterQuestions.MessageCharacterVisibleSkin_EN},
                    question{TheQuestion := QuestionBank.AllQuestions.CharacterQuestions.MessageCharacterArmor_EN}
                }
            }
        }

        set DCQuestions += array{
            category{
                Style:=card_style.DC,
                Name:="Weapons & Gadgets",
                NameMessage := QuestionBank.AllCategories.MessageWeapons_EN,
                Questions:=array{
                    question{TheQuestion := QuestionBank.AllQuestions.DCQuestions.MessageDCLasso_EN},
                    question{TheQuestion := QuestionBank.AllQuestions.DCQuestions.MessageDCBatarang_EN},
                    question{TheQuestion := QuestionBank.AllQuestions.DCQuestions.MessageDCDualPistols_EN},
                    question{TheQuestion := QuestionBank.AllQuestions.DCQuestions.MessageDCEnergyBeams_EN},
                    question{TheQuestion := QuestionBank.AllQuestions.DCQuestions.MessageDCArmoredSuit_EN}
                }
            }
        }

        set DCQuestions += array{
            category{
                Style:=card_style.DC,
                Name:="Vehicles",
                NameMessage := QuestionBank.AllCategories.MessageVehicleType_EN,
                Questions:=array{
                    question{TheQuestion := QuestionBank.AllQuestions.DCQuestions.MessageDCBatmobile_EN},
                    question{TheQuestion := QuestionBank.AllQuestions.DCQuestions.MessageDCJet_EN},
                    question{TheQuestion := QuestionBank.AllQuestions.DCQuestions.MessageDCMotorcycle_EN}
                }
            }
        }

        set DCQuestions += array{
            category{
                Style:=card_style.DC,
                Name:="Facts",
                NameMessage := QuestionBank.AllCategories.MessageFunFacts_EN,
                Questions:=array{
                    question{TheQuestion := QuestionBank.AllQuestions.DCQuestions.MessageDCBillionaire_EN},
                    question{TheQuestion := QuestionBank.AllQuestions.DCQuestions.MessageDCSidekick_EN},
                    question{TheQuestion := QuestionBank.AllQuestions.DCQuestions.MessageDCPetCompanion_EN},
                    question{TheQuestion := QuestionBank.AllQuestions.DCQuestions.MessageDCDiedCameBack_EN},
                    question{TheQuestion := QuestionBank.AllQuestions.DCQuestions.MessageDCControlPlants_EN},
                    question{TheQuestion := QuestionBank.AllQuestions.DCQuestions.MessageDCAlterEgo_EN}
                }
            }
        }

        set DCQuestions += array{
            category{
                Style:=card_style.DC,
                Name:="Colors",
                NameMessage := QuestionBank.AllCategories.MessageColors_EN,
                Questions:=array{
                    question{TheQuestion := QuestionBank.AllQuestions.CharacterQuestions.MessageCharacterColorYellow_EN},
                    question{TheQuestion := QuestionBank.AllQuestions.CharacterQuestions.MessageCharacterColorOrange_EN},
                    question{TheQuestion := QuestionBank.AllQuestions.CharacterQuestions.MessageCharacterColorRed_EN},
                    question{TheQuestion := QuestionBank.AllQuestions.CharacterQuestions.MessageCharacterColorPink_EN},
                    question{TheQuestion := QuestionBank.AllQuestions.CharacterQuestions.MessageCharacterColorBlue_EN},
                    question{TheQuestion := QuestionBank.AllQuestions.CharacterQuestions.MessageCharacterColorPurple_EN},
                    question{TheQuestion := QuestionBank.AllQuestions.CharacterQuestions.MessageCharacterColorWhite_EN},
                    question{TheQuestion := QuestionBank.AllQuestions.CharacterQuestions.MessageCharacterColorBlack_EN},
                    question{TheQuestion := QuestionBank.AllQuestions.CharacterQuestions.MessageCharacterColorGold_EN},
                    question{TheQuestion := QuestionBank.AllQuestions.CharacterQuestions.MessageCharacterColorSilver_EN},
                    question{TheQuestion := QuestionBank.AllQuestions.CharacterQuestions.MessageCharacterColorGreen_EN},
                    question{TheQuestion := QuestionBank.AllQuestions.CharacterQuestions.MessageCharacterColorBrown_EN}
                }
            }
        }

        set DCQuestions += array{
            category{
                Style := card_style.DC,
                Name:="Other",
                NameMessage := QuestionBank.AllCategories.MessageOther_EN,
                Questions := array{
                    question{TheQuestion := QuestionBank.AllQuestions.CharacterQuestions.MessageCharacterGenderMale_EN},
                    question{TheQuestion := QuestionBank.AllQuestions.CharacterQuestions.MessageCharacterGenderFemale_EN},
                    question{TheQuestion := QuestionBank.AllQuestions.CharacterQuestions.MessageCharacterHairColorRed_EN},
                    question{TheQuestion := QuestionBank.AllQuestions.CharacterQuestions.MessageCharacterHairColorGreen_EN},
                    question{TheQuestion := QuestionBank.AllQuestions.CharacterQuestions.MessageCharacterHairColorBrown_EN},
                    question{TheQuestion := QuestionBank.AllQuestions.CharacterQuestions.MessageCharacterHairColorDyed_EN},
                    question{TheQuestion := QuestionBank.AllQuestions.CharacterQuestions.MessageCharacterHairColorBlack_EN},
                    question{TheQuestion := QuestionBank.AllQuestions.CharacterQuestions.MessageCharacterHairColorPink_EN},
                    question{TheQuestion := QuestionBank.AllQuestions.CharacterQuestions.MessageCharacterHairColorBlue_EN},
                    question{TheQuestion := QuestionBank.AllQuestions.CharacterQuestions.MessageCharacterHairBald_EN}
                }
            }
        }


    var Chapter6Questions: []category = array{}

    InitChapter6Questions()<suspends>:void=

        set Chapter6Questions = array{
            category{
                Style:=card_style.Chapter6,
                Name:="Colors",
                NameMessage := QuestionBank.AllCategories.MessageColors_EN,
                Questions:=array{
                    question{TheQuestion := QuestionBank.AllQuestions.CharacterQuestions.MessageCharacterColorYellow_EN},
                    question{TheQuestion := QuestionBank.AllQuestions.CharacterQuestions.MessageCharacterColorOrange_EN},
                    question{TheQuestion := QuestionBank.AllQuestions.CharacterQuestions.MessageCharacterColorRed_EN},
                    question{TheQuestion := QuestionBank.AllQuestions.CharacterQuestions.MessageCharacterColorPink_EN},
                    question{TheQuestion := QuestionBank.AllQuestions.CharacterQuestions.MessageCharacterColorBlue_EN},
                    question{TheQuestion := QuestionBank.AllQuestions.CharacterQuestions.MessageCharacterColorPurple_EN},
                    question{TheQuestion := QuestionBank.AllQuestions.CharacterQuestions.MessageCharacterColorBlack_EN},
                    question{TheQuestion := QuestionBank.AllQuestions.CharacterQuestions.MessageCharacterColorWhite_EN},
                    question{TheQuestion := QuestionBank.AllQuestions.CharacterQuestions.MessageCharacterColorGold_EN},
                    question{TheQuestion := QuestionBank.AllQuestions.CharacterQuestions.MessageCharacterColorSilver_EN}
                }
            }
        }
    
        set Chapter6Questions += array{
            category{
                Style:=card_style.Chapter6,
                Name:="How Obtained?",
                NameMessage := QuestionBank.AllCategories.MessageHowObtained_EN,
                Questions:=array{
                    question{TheQuestion := QuestionBank.AllQuestions.CharacterQuestions.MessageCharacterFromBattlePass_EN},
                    question{TheQuestion := QuestionBank.AllQuestions.CharacterQuestions.MessageCharacterFromItemShop_EN},
                    question{TheQuestion := QuestionBank.AllQuestions.CharacterQuestions.MessageCharacterFromPack_EN},
                    question{TheQuestion := QuestionBank.AllQuestions.CharacterQuestions.MessageCharacterFromCrew_EN},
                    question{TheQuestion := QuestionBank.AllQuestions.CharacterQuestions.MessageCharacterFromQuest_EN},
                    question{TheQuestion := QuestionBank.AllQuestions.FireIceQuestions.MessageFireAndIceReskin_EN},
                    question{TheQuestion := QuestionBank.AllQuestions.FireIceQuestions.MessageFireAndIceSpecialEvent_EN},
                    question{TheQuestion := QuestionBank.AllQuestions.FireIceQuestions.MessageFireAndIceBundle_EN}
                }
            }
        }
    
        set Chapter6Questions += array{
            category{
                Style:=card_style.Chapter6,
                Name:="Species",
                NameMessage := QuestionBank.AllCategories.MessageSpecies_EN,
                Questions:=array{
                    question{TheQuestion := QuestionBank.AllQuestions.CharacterQuestions.MessageCharacterSpeciesHuman_EN},
                    question{TheQuestion := QuestionBank.AllQuestions.CharacterQuestions.MessageCharacterSpeciesAnimal_EN},
                    question{TheQuestion := QuestionBank.AllQuestions.CharacterQuestions.MessageCharacterSpeciesFood_EN},
                    question{TheQuestion := QuestionBank.AllQuestions.CharacterQuestions.MessageCharacterSpeciesRobot_EN},
                    question{TheQuestion := QuestionBank.AllQuestions.CharacterQuestions.MessageCharacterSpeciesMonster_EN}
                }
            }
        }
    
        set Chapter6Questions += array{
            category{
                Style:=card_style.Chapter6,
                Name:="Accessories",
                NameMessage := QuestionBank.AllCategories.MessageAccessories_EN,
                Questions:=array{
                    question{TheQuestion := QuestionBank.AllQuestions.CharacterQuestions.MessageCharacterHat_EN},
                    question{TheQuestion := QuestionBank.AllQuestions.CharacterQuestions.MessageCharacterMask_EN},
                    question{TheQuestion := QuestionBank.AllQuestions.CharacterQuestions.MessageCharacterHelmet_EN},
                    question{TheQuestion := QuestionBank.AllQuestions.CharacterQuestions.MessageCharacterGlasses_EN},
                    question{TheQuestion := QuestionBank.AllQuestions.CharacterQuestions.MessageCharacterNecklace_EN},
                    question{TheQuestion := QuestionBank.AllQuestions.CharacterQuestions.MessageCharacterShowingTeeth_EN},
                    question{TheQuestion := QuestionBank.AllQuestions.CharacterQuestions.MessageCharacterJacket_EN}
                }
            }
        }
    
        set Chapter6Questions += array{
            category{
                Style:=card_style.Chapter6,
                Name:= "Other",
                NameMessage := QuestionBank.AllCategories.MessageOther_EN,
                Questions:=array{
                    question{TheQuestion := QuestionBank.AllQuestions.CharacterQuestions.MessageCharacterGenderMale_EN},
                    question{TheQuestion := QuestionBank.AllQuestions.CharacterQuestions.MessageCharacterGenderFemale_EN},
                    question{TheQuestion := QuestionBank.AllQuestions.CharacterQuestions.MessageCharacterDogTags_EN},
                    question{TheQuestion := QuestionBank.AllQuestions.CharacterQuestions.MessageCharacterArmor_EN},
                    question{TheQuestion := QuestionBank.AllQuestions.CharacterQuestions.MessageCharacterTattoo_EN},
                    question{TheQuestion := QuestionBank.AllQuestions.LockerBundleQuestions.MessageLockerOtherScars_EN},
                    question{TheQuestion := QuestionBank.AllQuestions.CharacterQuestions.MessageCharacterFacePaint_EN},
                    question{TheQuestion := QuestionBank.AllQuestions.CharacterQuestions.MessageCharacterVillain_EN},
                    question{TheQuestion := QuestionBank.AllQuestions.CharacterQuestions.MessageCharacterHero_EN},
                    question{TheQuestion := QuestionBank.AllQuestions.RemixQuestions.MessageRemixOtherMainCharacter_EN}
                    
                }
            }
        }
    
        set Chapter6Questions += array{
            category{
                Style:=card_style.Chapter6,
                Name:= "Facial Attributes",
                NameMessage := QuestionBank.AllCategories.MessageFacialAttributes_EN,
                Questions := array{
                    question{TheQuestion := QuestionBank.AllQuestions.CharacterQuestions.MessageCharacterFreckles_EN},
                    question{TheQuestion := QuestionBank.AllQuestions.CharacterQuestions.MessageCharacterTannedSkinTone_EN},
                    question{TheQuestion := QuestionBank.AllQuestions.CharacterQuestions.MessageCharacterDarkSkinTone_EN},
                    question{TheQuestion := QuestionBank.AllQuestions.CharacterQuestions.MessageCharacterLightSkinTone_EN},
                    question{TheQuestion := QuestionBank.AllQuestions.CharacterQuestions.MessageCharacterEyeColorBlue_EN},
                    question{TheQuestion := QuestionBank.AllQuestions.CharacterQuestions.MessageCharacterEyeColorBrown_EN},
                    question{TheQuestion := QuestionBank.AllQuestions.CharacterQuestions.MessageCharacterEyeColorGreen_EN}
                }
            }
        }
    
        set Chapter6Questions += array{
            category{
                Style := card_style.Chapter6,
                Name := "Hair Colors",
                NameMessage := QuestionBank.AllCategories.MessageHairColors_EN,
                Questions := array{
                    question{TheQuestion := QuestionBank.AllQuestions.CharacterQuestions.MessageCharacterHairColorOrange_EN},
                    question{TheQuestion := QuestionBank.AllQuestions.CharacterQuestions.MessageCharacterHairColorWhite_EN},
                    question{TheQuestion := QuestionBank.AllQuestions.CharacterQuestions.MessageCharacterHairColorBlack_EN},
                    question{TheQuestion := QuestionBank.AllQuestions.CharacterQuestions.MessageCharacterHairColorBlonde_EN},
                    question{TheQuestion := QuestionBank.AllQuestions.CharacterQuestions.MessageCharacterHairColorGreen_EN},
                    question{TheQuestion := QuestionBank.AllQuestions.CharacterQuestions.MessageCharacterHairColorBrown_EN}
                }
            }
        }
    
        set Chapter6Questions += array{
            category{
                Style := card_style.Chapter6,
                Name := "Hair Styles",
                NameMessage := QuestionBank.AllCategories.MessageHairStyles_EN,
                Questions := array{
                    question{TheQuestion := QuestionBank.AllQuestions.CharacterQuestions.MessageCharacterHairCurly_EN},
                    question{TheQuestion := QuestionBank.AllQuestions.CharacterQuestions.MessageCharacterHairBald_EN},
                    question{TheQuestion := QuestionBank.AllQuestions.CharacterQuestions.MessageCharacterHairShort_EN},
                    question{TheQuestion := QuestionBank.AllQuestions.CharacterQuestions.MessageCharacterHairLong_EN},
                    question{TheQuestion := QuestionBank.AllQuestions.CharacterQuestions.MessageCharacterHairPonytail_EN},
                    question{TheQuestion := QuestionBank.AllQuestions.CharacterQuestions.MessageCharacterHairFade_EN}
                }
            }
        }
    
        set Chapter6Questions += array{
            category{
                Style:=card_style.Chapter6,
                Name:="Chapter & Season",
                NameMessage := QuestionBank.AllCategories.MessageChapterSeason_EN,
                Questions:=array{
                    question{TheQuestion := QuestionBank.AllQuestions.Chapter6Questions.MessageChapter6Season1_EN},
                    question{TheQuestion := QuestionBank.AllQuestions.Chapter6Questions.MessageChapter6Season2_EN}
                }
            }
        }
    

        


    # NEEDS TO BE UPDATED WHEN NEW STYLES (SETS) ARE ADDED ######################################################################################################
    GetQuestionsFromCategory(Category:card_style)<suspends>:[]category=
        case(Category):
            card_style.FN => 
                # Logger.Print("Length of Requested Questions: {FNQuestions.Length}")
                return FNQuestions
            card_style.Classic =>
                Logger.Print("Depreciated Style (Classic)")
                return array{}
            card_style.Animal =>
                # Logger.Print("Length of Requested Questions: {AnimalQuestions.Length}")
                return AnimalQuestions
            card_style.Icon =>
                # Logger.Print("Length of Requested Questions: {IconQuestions.Length}")
                return IconQuestions
            card_style.Utilities =>
                # Logger.Print("Length of Requested Questions: {UtilitiesQuestions.Length}")
                return UtilitiesQuestions
            card_style.Marvel =>
                Logger.Print("Depreciated Style (Marvel)")
                return array{}
            card_style.POI =>
                # Logger.Print("Length of Requested Questions: {POIQuestions.Length}")
                return POIQuestions
            card_style.Legends =>
                # Logger.Print("Length of Requested Questions: {LegendsQuestions.Length}")
                return LegendsQuestions
            card_style.Food =>
                # Logger.Print("Length of Requested Questions: {FoodQuestions.Length}")
                return FoodQuestions
            card_style.Star =>
                # Logger.Print("Length of Requested Questions: {StarQuestions.Length}")
                return StarQuestions
            card_style.Spooky =>
                # Logger.Print("Length of Requested Questions: {SpookyQuestions.Length}")
                return SpookyQuestions
            card_style.OG =>
                # Logger.Print("Length of Requested Questions: {OGQuestions.Length}")
                return OGQuestions
            card_style.Anime =>
                # Logger.Print("Length of Requested Questions: {AnimeQuestions.Length}")
                return AnimeQuestions
            card_style.Winter =>
                # Logger.Print("Length of Requested Questions: {WinterQuestions.Length}")
                return WinterQuestions
            card_style.Weapons =>
                # Logger.Print("Length of Requested Questions: {WeaponsQuestions.Length}")
                return WeaponsQuestions
            card_style.Pickaxes =>
                # Logger.Print("Length of Requested Questions: {PickaxesQuestions.Length}")
                return PickaxesQuestions
            card_style.Vehicles =>
                # Logger.Print("Length of Requested Questions: {VehiclesQuestions.Length}")
                return VehiclesQuestions
            card_style.MoviesTV =>
                # Logger.Print("Length of Requested Questions: {MoviesTVQuestions.Length}")
                return MoviesTVQuestions
            card_style.Blocks =>
                # Logger.Print("Length of Requested Questions: {BlocksQuestions.Length}")
                return BlocksQuestions
            card_style.Chapter5 =>
                # Logger.Print("Length of Requested Questions: {Chapter5Questions.Length}")
                return Chapter5Questions
            card_style.Lockers =>
                # Logger.Print("Length of Requested Questions: {LockersQuestions.Length}")
                return LockersQuestions
            card_style.Backblings =>
                # Logger.Print("Length of Requested Questions: {BackblingQuestions.Length}")
                return BackblingQuestions
            card_style.Emotes =>
                # Logger.Print("Length of Requested Questions: {EmoteQuestions.Length}")
                return EmoteQuestions
            card_style.Marvel1 =>
                # Logger.Print("Length of Requested Questions: {Marvel1Questions.Length}")
                return Marvel1Questions
            card_style.Marvel2 =>
                # Logger.Print("Length of Requested Questions: {Marvel2Questions.Length}")
                return Marvel2Questions
            card_style.Umbrellas =>
                # Logger.Print("Length of Requested Questions: {UmbrellasQuestions.Length}")
                return UmbrellasQuestions
            card_style.PeelyFishstick =>
                # Logger.Print("Length of Requested Questions: {PeelyFishstickQuestions.Length}")
                return PeelyFishstickQuestions
            card_style.Remix =>
                # Logger.Print("Length of Requested Questions: {RemixQuestions.Length}")
                return RemixQuestions
            card_style.FireIce =>
                # Logger.Print("Length of Requested Questions: {FireIceQuestions.Length}")
                return FireIceQuestions
            card_style.DC =>
                # Logger.Print("Length of Requested Questions: {DCQuestions.Length}")
                return DCQuestions
            card_style.Chapter6 =>
                return Chapter6Questions



    HideEveryoneUI():void=
        Logger.Print("HideEveryoneUI")

        if:
            TheGlobals := Globals[GetSession()]
            AllPlayers := TheGlobals.AllPlayers
        then:

            if:
                PlayerOne := AllPlayers[0]
                CPlayerOne := TheGlobals.CustomPlayerMap[PlayerOne]
                PlayerOneUI := GetPlayerUI[PlayerOne]
            then:
                if:
                    HasUI := CPlayerOne.GuessUI.MaybeQuestionUI?
                then:
                    PlayerOneUI.RemoveWidget(HasUI)
                    set CPlayerOne.GuessUI.MaybeQuestionUI = false


                    ReshowPaintPanelIfPresent(PlayerOne)

                else:
                    Logger.Print("Failed to get Player One UI")
            else:
                Logger.Print("Failed to get Player One")

            if:
                PlayerTwo := AllPlayers[1]
                CPlayerTwo := TheGlobals.CustomPlayerMap[PlayerTwo]
                PlayerTwoUI := GetPlayerUI[PlayerTwo]
            then:
                if:
                    HasUI := CPlayerTwo.GuessUI.MaybeQuestionUI?
                then:
                    PlayerTwoUI.RemoveWidget(HasUI)
                    set CPlayerTwo.GuessUI.MaybeQuestionUI = false

                    ReshowPaintPanelIfPresent(PlayerTwo)
                else:
                    Logger.Print("Failed to get Player Two UI")
            else:
                Logger.Print("Failed to get Player Two")

    
    var CurrentStyle : card_style = card_style.FN
    var CurrentCategories : []category = array{}
    

    Init(Style:card_style)<suspends>:void=
        #clear catagories for fake rounds
        set CurrentCategories = array{}
        Logger.Print("Should init qustion bank with style {Style.ToString()}")
        set CurrentCategories = GetQuestionsFromCategory(Style)
        # Logger.Print("Number of Cats: {CurrentCategories.Length}")


    HandleTriggerAskSignal(Agent:?agent):void=
        if(TheAgent:= Agent?):
            spawn{ShowMenu(TheAgent)}

    HandleRemoteSignal(Agent:agent):void =
        spawn{ShowMenu(Agent)}



    ShowMenu<public>(Agent : agent)<suspends> : void =
        if:
            TheGlobals := Globals[GetSession()]
            Player := player[Agent]
            CPlayer := TheGlobals.CustomPlayerMap[Player]
            PlayerUI := GetPlayerUI[Player]
        then:
            HideMenuIfPresent(Player)
            HidePaintPanelIfPresent(Player)
    
            MenuIndex := GetMenuIndex(Player)
            MainUI := CreateMainUI(Player, MenuIndex)
    
            set CPlayer.GuessUI.MaybeQuestionUI = option{MainUI} 
            PlayerUI.AddWidget(MainUI, player_ui_slot{InputMode := ui_input_mode.All, ZOrder:= 99})
        else:
            Logger.Print("Failed to show menu")
    


    UpdateQuestionsOnly(Agent:agent)<suspends>:void=
        Logger.Print("Update Questions - Attempting to update questions")
        if:
            TheGlobals := Globals[GetSession()]
            Player := player[Agent]
            CPlayer := TheGlobals.CustomPlayerMap[Player]
            PlayerUI := GetPlayerUI[Player]
        then:
            if(IsMenuLoaded := CPlayer.GuessUI.MaybeQuestionUI?, ExistingQuestions := CPlayer.GuessUI.MaybeExistingQuestionButtons?):
                MenuIndex := GetMenuIndex(Player)
                NewQuestionButtons := MakeAllButtons_New(GetAllButtonsForQuestions(MenuIndex), 80.0)
                
                # Hide and remove existing questions
                Logger.Print("Hide Existing Questions")
                ExistingQuestions.Widget.SetVisibility(widget_visibility.Hidden)
                Sleep(0.1)  # Ensure the visibility change takes effect
                CPlayer.GuessUI.AllQuestionButtonsOverlay.RemoveWidget(ExistingQuestions.Widget)
                Sleep(0.1)  # Ensure enough time for removal to take effect
                
                # Set the new questions
                set CPlayer.GuessUI.MaybeExistingQuestionButtons = option{NewQuestionButtons}
                Logger.Print("Add New Questions")
                CPlayer.GuessUI.AllQuestionButtonsOverlay.AddWidget(NewQuestionButtons)
    
                TitleText := GetCategoryTitleMessage(MenuIndex)
                CPlayer.GuessUI.QuestionCategoryText.SetText(TitleText)
            else:
                Logger.Print("Update Questions - No Menu Loaded")
        else:
            Logger.Print("Update Questions - Failed to update questions")
    

    GetOpponentAgent(Agent:agent)<transacts><decides>:agent=
        var Opponent: agent = Agent
        if:
            AgentPlayer := player[Agent]
            TheGlobals :=  Globals[GetSession()]
            CPlayer := TheGlobals.CustomPlayerMap[player[Agent]]
        then:
            for(APlayer : TheGlobals.AllPlayers, AgentPlayer <> APlayer):
                if(ActualAgent := agent[APlayer]):
                    set Opponent = ActualAgent
        return Opponent

    ShowOpponentMenu<public>(Agent : agent)<suspends> : void =
        Sleep(0.1)
        
        var IndexWasHidden : logic = false
        var CardWasHidden : logic = false

        if:
            TheOpponent := GetOpponentAgent[Agent]
            OpponentPlayer := player[TheOpponent]
            PlayerUI := GetPlayerUI[OpponentPlayer]
            TheGlobals :=  Globals[GetSession()]
            CPlayer := TheGlobals.CustomPlayerMap[player[Agent]]
            OpponentCPlayer := TheGlobals.CustomPlayerMap[OpponentPlayer]
            AskingPlayer := player[Agent]
            not CPlayer.GameBoard.AwaitingResponse?
        then:
            HidePaintPanelIfPresent(OpponentPlayer)
            Sleep(0.1)
               
            OpponentUI := CreateOpponentUI(OpponentPlayer)
            set OpponentCPlayer.GuessUI.MaybeQuestionUI = option{OpponentUI} # store reference to question answering ui
            PlayerUI.AddWidget(OpponentUI, player_ui_slot{InputMode := ui_input_mode.All,ZOrder:= 1})

            Logger.Print("attempt play cine")

            OpponentCPlayer.GameBoard.CameraZoom1.RemoveFrom(TheOpponent)
            OpponentCPlayer.GameBoard.CameraZoom2.RemoveFrom(TheOpponent)
            OpponentCPlayer.GameBoard.YesNoCamera.AddTo(TheOpponent)

            CPlayer.GameBoard.EndTurnButton.Disable()
            CPlayer.GameBoard.MakeGuessButton.Disable()

            set CPlayer.GameBoard.AwaitingResponse = true
            set OpponentCPlayer.GameBoard.AwaitingResponse = true

            if:
                OpponentCPlayer.GameBoard.YouTile.CardHidden?
            then:
                OpponentCPlayer.GameBoard.YouTile.ShowCharacterCard(OpponentPlayer)
                set CardWasHidden = true

            if:
                OpponentCPlayer.GameBoard.IndexCard.IsHidden?
                OpponentCPlayer.GameBoard.StyleHasIndexCard?
            then:
                set IndexWasHidden = true
                OpponentCPlayer.GameBoard.IndexCard.SlideOut(OpponentCPlayer.GameBoard.BoardLocalForward)

            race:
                block:
                    OpponentResponseEvent.Await()
                block:
                    Sleep(25.0)

                    
            OpponentCPlayer.GameBoard.YesNoCamera.RemoveFrom(TheOpponent)
            HideEveryoneUI()

            set CPlayer.GameBoard.AwaitingResponse = false
            set OpponentCPlayer.GameBoard.AwaitingResponse = false

            if:
                IndexWasHidden?
                OpponentCPlayer.GameBoard.StyleHasIndexCard?
            then:
                OpponentCPlayer.GameBoard.IndexCard.SlideIn(OpponentCPlayer.GameBoard.BoardLocalForward)

            if:
                CardWasHidden?
            then:
                OpponentCPlayer.GameBoard.YouTile.HideCharacterCard(OpponentPlayer)

            CPlayer.GameBoard.EndTurnButton.Enable()
            CPlayer.GameBoard.MakeGuessButton.Enable()


        else:
            Logger.Print("Failed to show menu")



    CreateUIStyleOption(FirstOfRow:logic, Padding:margin, Texture:texture, Button:button_loud):stack_box_slot=

        var MyPadding : margin = Padding

        if(FirstOfRow?):
            set MyPadding = margin{Top := 0.0, Left := 0.0, Right := 0.0, Bottom := 0.0}

        TheStyle := stack_box_slot:
            Padding:= MyPadding
            HorizontalAlignment := horizontal_alignment.Fill
            VerticalAlignment := vertical_alignment.Fill
            Widget := overlay:
                
                Slots:= array:
                    
                    overlay_slot:
                        
                        HorizontalAlignment := horizontal_alignment.Fill
                        VerticalAlignment := vertical_alignment.Fill
                        Widget := color_block{DefaultColor:=Blue, DefaultOpacity:=0.75}
                    overlay_slot:
                        HorizontalAlignment := horizontal_alignment.Fill
                        VerticalAlignment := vertical_alignment.Fill
                        Widget:= 
                            stack_box:     
                                                           
                                Orientation := orientation.Vertical
                                Slots := array:
                                    stack_box_slot:
                                        Padding := margin{Top := 10.0, Left := 10.0, Right := 10.0, Bottom := 10.0}
                                        HorizontalAlignment := horizontal_alignment.Fill
                                        VerticalAlignment := vertical_alignment.Fill
                                        Widget := stack_box:
                                            Orientation := orientation.Vertical
                                            Slots := array:
                                                stack_box_slot:
                                                    HorizontalAlignment := horizontal_alignment.Fill
                                                    VerticalAlignment := vertical_alignment.Fill
                                                    Padding := margin{Top := 0.0, Left := 0.0, Right := 0.0, Bottom := 0.0}
                                                    Widget := texture_block{DefaultImage := Texture, DefaultDesiredSize:=vector2{X:= 200.0, Y:= 200.0}}
                                                stack_box_slot:
                                                    HorizontalAlignment := horizontal_alignment.Fill
                                                    VerticalAlignment := vertical_alignment.Fill
                                                    Padding := margin{Top := 0.0, Left := 0.0, Right := 0.0, Bottom := 0.0}
                                                    Widget := overlay:
                                                        Slots:= array:
                                                            overlay_slot:
                                                                HorizontalAlignment := horizontal_alignment.Fill
                                                                VerticalAlignment := vertical_alignment.Fill
                                                                Padding := margin{Top := 0.0, Left := 0.0, Right := 0.0, Bottom := 0.0}
                                                                Widget := color_block{DefaultColor:=Blue, DefaultOpacity:=0.75}
                                                            overlay_slot:
                                                                HorizontalAlignment := horizontal_alignment.Fill
                                                                VerticalAlignment := vertical_alignment.Fill
                                                                Padding := margin{Top := 0.0, Left := 0.0, Right := 0.0, Bottom := 0.0}
                                                                Widget := Button


        return TheStyle

     
    CreateOpponentUI<private>(Player : player) : widget =

        PaddingBorder := 8.0

        ButtonYes.SetText(MessageYesButton_EN)        

        ButtonNo.SetText(MessageNoButton_EN)        

        ButtonMaybe.SetText(MessageMaybeButton_EN)        

        TittleText := SelectedQuestion

        ColorBlockTitle := color_block:
            DefaultColor:= MakeColorFromSRGBValues(0,85,254)
            DefaultDesiredSize:= vector2{X:= 1000.0, Y:=100.0}
            DefaultOpacity:=1.0

        ColorBlock := color_block {DefaultColor:= MakeColorFromSRGBValues(20,31,83),DefaultDesiredSize:= vector2{X:= 1000.0, Y:=200.0},DefaultOpacity:=1.0}
         
        Title := text_block{DefaultText:=TittleText,DefaultTextColor:= MakeColorFromSRGBValues(255,255,255),DefaultShadowColor:=MakeColorFromHex("#000000"),DefaultJustification:= text_justification.Center }

        LayoutCanvas : canvas = canvas:
            Slots := array:

                canvas_slot:
                    Anchors := anchors{Minimum := vector2{X := 0.0, Y := 0.0}, Maximum := vector2{X := 1.0, Y := 0.0}}
                    Offsets := margin{Top := 0.0, Left := 0.0, Right := 0.0, Bottom := 0.0}
                    Alignment := vector2{X := 0.5, Y := 0.0}
                    ZOrder := 999
                    Widget := overlay:
                        Slots := array:
                            overlay_slot:
                                HorizontalAlignment := horizontal_alignment.Fill
                                VerticalAlignment := vertical_alignment.Fill
                                Padding := margin{Top := 0.0, Left := 0.0, Right := 0.0, Bottom := 0.0}
                                Widget := MakeRoundedOverlayOuterTopCorners(UI_BGColor)
                            overlay_slot:
                                HorizontalAlignment := horizontal_alignment.Fill
                                VerticalAlignment := vertical_alignment.Fill
                                Padding := margin{Top := PaddingBorder, Left := PaddingBorder, Right := PaddingBorder, Bottom := PaddingBorder}
                                Widget := stack_box:
                                    Orientation := orientation.Vertical
                                    Slots := array:
                                        stack_box_slot:
                                            HorizontalAlignment := horizontal_alignment.Fill
                                            VerticalAlignment := vertical_alignment.Fill
                                            Padding := margin{Top := 0.0, Left := 0.0, Right := 0.0, Bottom := 0.0}
                                            Widget := overlay:
                                                Slots := array:
                                                    overlay_slot:
                                                        HorizontalAlignment := horizontal_alignment.Fill
                                                        VerticalAlignment := vertical_alignment.Fill
                                                        Widget := MakeRoundedOverlayInnerTopCorners(MakeColorFromSRGBValues(0,85,254))
                                                    overlay_slot:
                                                        HorizontalAlignment := horizontal_alignment.Fill
                                                        VerticalAlignment := vertical_alignment.Fill
                                                        Padding := margin{Top := PaddingBorder, Left := PaddingBorder, Right := PaddingBorder, Bottom := PaddingBorder}
                                                        Widget := overlay:
                                                            Slots:= array:
                                                                overlay_slot:
                                                                    HorizontalAlignment := horizontal_alignment.Fill
                                                                    VerticalAlignment := vertical_alignment.Fill
                                                                    Padding := margin{Top := 0.0, Left := 0.0, Right := 0.0, Bottom := 0.0}
                                                                    Widget := ColorBlock
                                                                overlay_slot:
                                                                    HorizontalAlignment := horizontal_alignment.Fill
                                                                    VerticalAlignment := vertical_alignment.Fill
                                                                    Padding := margin{Top := 0.0, Left := 0.0, Right := 0.0, Bottom := 0.0}
                                                                    Widget := stack_box:
                                                                        Orientation := orientation.Vertical
                                                                        Slots := array:
                                                                            stack_box_slot:
                                                                                HorizontalAlignment := horizontal_alignment.Fill
                                                                                VerticalAlignment := vertical_alignment.Fill
                                                                                Padding := margin{Top := 0.0, Left := 0.0, Right := 0.0, Bottom := 0.0}
                                                                                Widget := overlay:
                                                                                    Slots := array:
                                                                                        overlay_slot:
                                                                                            HorizontalAlignment := horizontal_alignment.Fill
                                                                                            VerticalAlignment := vertical_alignment.Fill
                                                                                            Padding := margin{Top := 0.0, Left := 0.0, Right := 0.0, Bottom := 0.0}
                                                                                            Widget := ColorBlockTitle
                                                                                        overlay_slot:
                                                                                            HorizontalAlignment := horizontal_alignment.Center
                                                                                            VerticalAlignment := vertical_alignment.Center
                                                                                            Padding := margin{Top := 0.0, Left := 0.0, Right := 0.0, Bottom := 0.0}
                                                                                            Widget := Title
                                        stack_box_slot:
                                            HorizontalAlignment := horizontal_alignment.Fill
                                            VerticalAlignment := vertical_alignment.Fill
                                            Padding := margin{Top := 0.0, Left := 0.0, Right := 0.0, Bottom := 0.0}
                                            Widget := overlay:
                                                Slots := array:
                                                    overlay_slot:
                                                        HorizontalAlignment := horizontal_alignment.Fill
                                                        VerticalAlignment := vertical_alignment.Fill
                                                        Padding := margin{Top := 0.0, Left := 0.0, Right := 0.0, Bottom := 0.0}
                                                        Widget := MakeRoundedOverlayInnerBottomCorners(MakeColorFromSRGBValues(20,31,83))
                                                    overlay_slot:
                                                        HorizontalAlignment := horizontal_alignment.Fill
                                                        VerticalAlignment := vertical_alignment.Fill
                                                        Padding := margin{Top := 0.0, Left := 0.0, Right := 0.0, Bottom := 0.0}
                                                        Widget := ColorBlock
                                                    overlay_slot:
                                                        HorizontalAlignment := horizontal_alignment.Fill
                                                        VerticalAlignment := vertical_alignment.Fill
                                                        Padding := margin{Top := 0.0, Left := 0.0, Right := 0.0, Bottom := 0.0}
                                                        Widget := overlay:
                                                            Slots:= array:
                                                                overlay_slot:
                                                                    HorizontalAlignment := horizontal_alignment.Center
                                                                    VerticalAlignment := vertical_alignment.Fill
                                                                    Widget:=stack_box:
                                                                        Orientation := orientation.Horizontal
                                                                        Slots := array:
                                                                            stack_box_slot:
                                                                                HorizontalAlignment := horizontal_alignment.Left
                                                                                VerticalAlignment := vertical_alignment.Center
                                                                                Padding:=margin{Top := 0.0, Left := 60.0, Right := 60.0, Bottom := 0.0}
                                                                                Widget := overlay:
                                                                                    Slots := array:
                                                                                        overlay_slot:
                                                                                            HorizontalAlignment := horizontal_alignment.Fill
                                                                                            VerticalAlignment := vertical_alignment.Fill
                                                                                            Widget := MakeRoundedOverlayQuietButton(White, ?TextureSize:=18.0)
                                                                                        overlay_slot:                                                
                                                                                            HorizontalAlignment := horizontal_alignment.Fill
                                                                                            VerticalAlignment := vertical_alignment.Fill
                                                                                            Widget := ButtonYes
                                                                            stack_box_slot:
                                                                                HorizontalAlignment := horizontal_alignment.Center
                                                                                VerticalAlignment := vertical_alignment.Center
                                                                                Padding:=margin{Top := 0.0, Left := 60.0, Right := 60.0, Bottom := 0.0}
                                                                                Widget := overlay:
                                                                                    Slots := array:
                                                                                        overlay_slot:
                                                                                            HorizontalAlignment := horizontal_alignment.Fill
                                                                                            VerticalAlignment := vertical_alignment.Fill
                                                                                            Widget := MakeRoundedOverlayQuietButton(White, ?TextureSize:=18.0)
                                                                                        overlay_slot:                                                
                                                                                            HorizontalAlignment := horizontal_alignment.Fill
                                                                                            VerticalAlignment := vertical_alignment.Fill
                                                                                            Widget := ButtonNo
                                                                            stack_box_slot:
                                                                                HorizontalAlignment := horizontal_alignment.Right
                                                                                VerticalAlignment := vertical_alignment.Center
                                                                                Padding:=margin{Top := 0.0, Left := 60.0, Right := 60.0, Bottom := 0.0}
                                                                                Widget := overlay:
                                                                                    Slots := array:
                                                                                        overlay_slot:
                                                                                            HorizontalAlignment := horizontal_alignment.Fill
                                                                                            VerticalAlignment := vertical_alignment.Fill
                                                                                            Widget := MakeRoundedOverlayQuietButton(White, ?TextureSize:=18.0)
                                                                                        overlay_slot:                                                
                                                                                            HorizontalAlignment := horizontal_alignment.Center
                                                                                            VerticalAlignment := vertical_alignment.Center
                                                                                            Widget := ButtonMaybe
                                                            

                                                                                                        

        # AddYesNoButtonsToWidget(LayoutCanvas)
        
        return CreateMenuCanvasBottom(LayoutCanvas)


    CreateMainUI<private>(Player : player, MenuIndex : int) : widget =

        TittleText := GetCategoryTitleMessage(MenuIndex)
        # Questions := GetCategoryQuestions(MenuIndex)

        ColorBlockTitle := color_block {DefaultColor:= MakeColorFromSRGBValues(0,85,254),DefaultDesiredSize:= vector2{X:= 1000.0, Y:=100.0},DefaultOpacity:=0.0}
        ColorBlock := color_block {DefaultColor:= MakeColorFromSRGBValues(20,31,83),DefaultDesiredSize:= vector2{X:= 1000.0, Y:=800.0},DefaultOpacity:=0.0}
         
        ButtonPrev.SetText(MessagePrevButton_EN)

        # Title := text_block{DefaultText:=StringToMessage("Category: " + TittleText),DefaultTextColor:= MakeColorFromSRGBValues(255,255,255),DefaultShadowColor:=MakeColorFromHex("#000000"),DefaultJustification:= text_justification.Center }

        ButtonNext.SetText(MessageNextButton_EN)

        ButtonExit.SetText(MessageCloseButton_EN)

        

        
        # ButtonStackbox := stack_box:
        #     Orientation := orientation.Horizontal
        #     Slots := array{}

        # ButtonStackbox.AddWidget(stack_box_slot{Padding:=margin{Top := 0.0,Left := 20.0, Right := 20.0, Bottom := 0.0}, Widget := ButtonPrev, HorizontalAlignment := horizontal_alignment.Fill, VerticalAlignment := vertical_alignment.Fill})
        # ButtonStackbox.AddWidget(stack_box_slot{Padding:=margin{Top := 0.0, Left := 20.0, Right := 20.0, Bottom := 0.0},Widget := Title, HorizontalAlignment := horizontal_alignment.Fill, VerticalAlignment := vertical_alignment.Fill})
        # ButtonStackbox.AddWidget(stack_box_slot{Padding:=margin{Top := 0.0, Left := 20.0, Right := 20.0, Bottom := 0.0},Widget := ButtonNext, HorizontalAlignment := horizontal_alignment.Fill, VerticalAlignment := vertical_alignment.Fill})
        # ButtonStackbox.AddWidget(stack_box_slot{Padding:=margin{Top := 0.0, Left := 0.0, Right := 20.0, Bottom := 0.0},Widget := ButtonExit, HorizontalAlignment := horizontal_alignment.Fill, VerticalAlignment := vertical_alignment.Fill})

        var LayoutCanvas : canvas = canvas{}

        if:
            CPlayer := Globals[GetSession()].CustomPlayerMap[Player]
        then:
            NewQuestionButtons := MakeAllButtons_New(GetAllButtonsForQuestions(MenuIndex), 80.0)
            set CPlayer.GuessUI.MaybeExistingQuestionButtons = option{NewQuestionButtons}
            CPlayer.GuessUI.AllQuestionButtonsOverlay.AddWidget(NewQuestionButtons)

            CPlayer.GuessUI.QuestionCategoryText.SetText(TittleText)

            PaddingBorder := 8.0

            set LayoutCanvas = canvas:
                Slots := array:
                    canvas_slot:
                        Anchors := anchors{Minimum := vector2{X := 0.1, Y := 0.0}, Maximum := vector2{X := 0.9, Y := 0.0}}
                        Offsets := margin{Top := 0.0, Left := 0.0, Right := 0.0, Bottom := 0.0}
                        Alignment := vector2{X := 0.5, Y := 0.0}
                        ZOrder := 999
                        SizeToContent := true
                        Widget := overlay:
                            Slots := array:
                                overlay_slot:
                                    HorizontalAlignment := horizontal_alignment.Fill
                                    VerticalAlignment := vertical_alignment.Fill
                                    Padding := margin{Top := 0.0, Left := 0.0, Right := 0.0, Bottom := 0.0}
                                    Widget := MakeRoundedOverlayOuter(UI_BGColor)
                                overlay_slot:
                                    HorizontalAlignment := horizontal_alignment.Fill
                                    VerticalAlignment := vertical_alignment.Fill
                                    Padding := margin{Top := 0.0, Left := 0.0, Right := 0.0, Bottom := 0.0}
                                    Widget := ColorBlockTitle
                                overlay_slot:
                                    HorizontalAlignment := horizontal_alignment.Fill
                                    VerticalAlignment := vertical_alignment.Fill
                                    Padding := margin{Top := PaddingBorder, Left := PaddingBorder, Right := PaddingBorder, Bottom := PaddingBorder}
                                    Widget:= stack_box:
                                        Orientation := orientation.Vertical
                                        Slots := array:
                                            stack_box_slot:
                                                HorizontalAlignment := horizontal_alignment.Fill
                                                VerticalAlignment := vertical_alignment.Fill
                                                Padding := margin{Top := 0.0, Left := 0.0, Right := 0.0, Bottom := 0.0}
                                                Widget := overlay:
                                                    Slots := array:
                                                        overlay_slot:
                                                            HorizontalAlignment := horizontal_alignment.Fill
                                                            VerticalAlignment := vertical_alignment.Fill
                                                            Widget := MakeRoundedOverlayInnerTopCorners(MakeColorFromSRGBValues(0,85,254))
                                                        overlay_slot:
                                                            HorizontalAlignment := horizontal_alignment.Fill
                                                            VerticalAlignment := vertical_alignment.Fill
                                                            Padding := margin{Top := PaddingBorder, Left := PaddingBorder, Right := PaddingBorder, Bottom := PaddingBorder}
                                                            Widget := overlay:
                                                                Slots:= array:
                                                                    overlay_slot:
                                                                        HorizontalAlignment := horizontal_alignment.Fill
                                                                        VerticalAlignment := vertical_alignment.Fill
                                                                        Widget:= stack_box:
                                                                            Orientation := orientation.Horizontal
                                                                            Slots := array:
                                                                                stack_box_slot:
                                                                                    Distribution := option{1.0}
                                                                                    HorizontalAlignment := horizontal_alignment.Left
                                                                                    VerticalAlignment := vertical_alignment.Center
                                                                                    Padding:=margin{Top := 0.0, Left := 20.0, Right := 20.0, Bottom := 0.0}
                                                                                    Widget := ButtonPrev
                                                                                stack_box_slot:
                                                                                    Distribution := option{1.0}
                                                                                    HorizontalAlignment := horizontal_alignment.Fill
                                                                                    VerticalAlignment := vertical_alignment.Center
                                                                                    Padding:=margin{Top := 0.0, Left := 0.0, Right := 0.0, Bottom := 0.0}
                                                                                    Widget := CPlayer.GuessUI.QuestionCategoryText
                                                                                stack_box_slot:
                                                                                    Distribution := option{1.0}
                                                                                    HorizontalAlignment := horizontal_alignment.Right
                                                                                    VerticalAlignment := vertical_alignment.Center
                                                                                    Padding:=margin{Top := 0.0, Left := 20.0, Right := 0.0, Bottom := 0.0}
                                                                                    Widget := ButtonNext
                                                                                stack_box_slot:
                                                                                    # Distribution := option{1.0}
                                                                                    HorizontalAlignment := horizontal_alignment.Right
                                                                                    VerticalAlignment := vertical_alignment.Center
                                                                                    Padding:=margin{Top := 0.0, Left := 20.0, Right := 20.0, Bottom := 0.0}
                                                                                    Widget := overlay:
                                                                                        Slots := array:
                                                                                            overlay_slot:
                                                                                                HorizontalAlignment := horizontal_alignment.Fill
                                                                                                VerticalAlignment := vertical_alignment.Fill
                                                                                                Widget := MakeRoundedOverlayQuietButton(Red, ?TextureSize:=18.0)
                                                                                            overlay_slot:
                                                                                                HorizontalAlignment := horizontal_alignment.Right
                                                                                                VerticalAlignment := vertical_alignment.Center
                                                                                                Widget := ButtonExit
                                            stack_box_slot:                
                                                HorizontalAlignment := horizontal_alignment.Fill
                                                VerticalAlignment := vertical_alignment.Fill
                                                Padding := margin{Top := 0.0, Left := 0.0, Right := 0.0, Bottom := 0.0}
                                                Widget := overlay:
                                                    Slots := array:
                                                        overlay_slot:
                                                            HorizontalAlignment := horizontal_alignment.Fill
                                                            VerticalAlignment := vertical_alignment.Fill
                                                            Widget := MakeRoundedOverlayInnerBottomCorners(MakeColorFromSRGBValues(20,31,83))

                                                        overlay_slot:
                                                            HorizontalAlignment := horizontal_alignment.Fill
                                                            VerticalAlignment := vertical_alignment.Fill
                                                            Widget := ColorBlock

                                                        overlay_slot:
                                                            HorizontalAlignment := horizontal_alignment.Fill
                                                            VerticalAlignment := vertical_alignment.Fill
                                                            Padding := margin{Top := PaddingBorder, Left := PaddingBorder, Right := PaddingBorder, Bottom := PaddingBorder}
                                                            Widget := CPlayer.GuessUI.AllQuestionButtonsOverlay
        return CreateMenuCanvas(LayoutCanvas)



    AddYesNoButtonsToWidget(LayoutCanvas: canvas): void =
        ButtonYes.SetText(MessageYesButton_EN)        

        ButtonNo.SetText(MessageNoButton_EN)        

        ButtonMaybe.SetText(MessageMaybeButton_EN)        

        Overlay := overlay:
            Slots:= array:
                overlay_slot:
                    HorizontalAlignment := horizontal_alignment.Center
                    VerticalAlignment := vertical_alignment.Fill
                    Widget:=stack_box:
                        Orientation := orientation.Horizontal
                        Slots := array:
                            stack_box_slot:
                                HorizontalAlignment := horizontal_alignment.Left
                                VerticalAlignment := vertical_alignment.Fill
                                Padding:=margin{Top := 0.0, Left := 60.0, Right := 60.0, Bottom := 0.0}
                                Widget := overlay:
                                    Slots := array:
                                        overlay_slot:
                                            HorizontalAlignment := horizontal_alignment.Fill
                                            VerticalAlignment := vertical_alignment.Fill
                                            Widget := MakeRoundedOverlayQuietButton(White, ?TextureSize:=18.0)
                                        overlay_slot:                                                
                                            HorizontalAlignment := horizontal_alignment.Fill
                                            VerticalAlignment := vertical_alignment.Fill
                                            Widget := ButtonYes
                            stack_box_slot:
                                HorizontalAlignment := horizontal_alignment.Center
                                VerticalAlignment := vertical_alignment.Fill
                                Padding:=margin{Top := 0.0, Left := 60.0, Right := 60.0, Bottom := 0.0}
                                Widget := overlay:
                                    Slots := array:
                                        overlay_slot:
                                            HorizontalAlignment := horizontal_alignment.Fill
                                            VerticalAlignment := vertical_alignment.Fill
                                            Widget := MakeRoundedOverlayQuietButton(White, ?TextureSize:=18.0)
                                        overlay_slot:                                                
                                            HorizontalAlignment := horizontal_alignment.Fill
                                            VerticalAlignment := vertical_alignment.Fill
                                            Widget := ButtonNo
                            stack_box_slot:
                                HorizontalAlignment := horizontal_alignment.Right
                                VerticalAlignment := vertical_alignment.Fill
                                Padding:=margin{Top := 0.0, Left := 60.0, Right := 60.0, Bottom := 0.0}
                                Widget := overlay:
                                    Slots := array:
                                        overlay_slot:
                                            HorizontalAlignment := horizontal_alignment.Fill
                                            VerticalAlignment := vertical_alignment.Fill
                                            Widget := MakeRoundedOverlayQuietButton(White, ?TextureSize:=18.0)
                                        overlay_slot:                                                
                                            HorizontalAlignment := horizontal_alignment.Fill
                                            VerticalAlignment := vertical_alignment.Fill
                                            Widget := ButtonMaybe

        LayoutCanvas.AddWidget(MakeButtonSlot(Overlay,180.0))


    GetAllButtonsForQuestions(MenuIndex:int):[]widget=
        Logger.Print("GetAllButtonsForQuestions")

        var AllWidgets : []widget = array{}
        var TempButton : button_quiet = button_quiet{}

        if(QuestionExists := CurrentCategories[MenuIndex].Questions[0]):
            # Button :=  button_quiet{DefaultText := StringToMessage(GetCategoryQuestion(MenuIndex,0))}
            Button :=  button_quiet{DefaultText := GetCategoryQuestionMessage(MenuIndex,0)}
            Button.OnClick().Subscribe(OnClickQuestion1) 
            set AllWidgets += array{Button}
        if(QuestionExists := CurrentCategories[MenuIndex].Questions[1]):
            # Button :=  button_quiet{DefaultText := StringToMessage(GetCategoryQuestion(MenuIndex,1))}
            Button :=  button_quiet{DefaultText := GetCategoryQuestionMessage(MenuIndex,1)}
            Button.OnClick().Subscribe(OnClickQuestion2) 
            set AllWidgets += array{Button}
        if(QuestionExists := CurrentCategories[MenuIndex].Questions[2]):
            # Button :=  button_quiet{DefaultText := StringToMessage(GetCategoryQuestion(MenuIndex,2))}
            Button :=  button_quiet{DefaultText := GetCategoryQuestionMessage(MenuIndex,2)}
            Button.OnClick().Subscribe(OnClickQuestion3) 
            set AllWidgets += array{Button}
        if(QuestionExists := CurrentCategories[MenuIndex].Questions[3]):
            # Button :=  button_quiet{DefaultText := StringToMessage(GetCategoryQuestion(MenuIndex,3))}
            Button :=  button_quiet{DefaultText := GetCategoryQuestionMessage(MenuIndex,3)}
            Button.OnClick().Subscribe(OnClickQuestion4) 
            set AllWidgets += array{Button}
        if(QuestionExists := CurrentCategories[MenuIndex].Questions[4]):
            # Button :=  button_quiet{DefaultText := StringToMessage(GetCategoryQuestion(MenuIndex,4))}
            Button :=  button_quiet{DefaultText := GetCategoryQuestionMessage(MenuIndex,4)}
            Button.OnClick().Subscribe(OnClickQuestion5) 
            set AllWidgets += array{Button}
        if(QuestionExists := CurrentCategories[MenuIndex].Questions[5]):
            # Button :=  button_quiet{DefaultText := StringToMessage(GetCategoryQuestion(MenuIndex,5))}
            Button :=  button_quiet{DefaultText := GetCategoryQuestionMessage(MenuIndex,5)}
            Button.OnClick().Subscribe(OnClickQuestion6) 
            set AllWidgets += array{Button}
        if(QuestionExists := CurrentCategories[MenuIndex].Questions[6]):
            # Button :=  button_quiet{DefaultText := StringToMessage(GetCategoryQuestion(MenuIndex,6))}
            Button :=  button_quiet{DefaultText := GetCategoryQuestionMessage(MenuIndex,6)}
            Button.OnClick().Subscribe(OnClickQuestion7) 
            set AllWidgets += array{Button}
        if(QuestionExists := CurrentCategories[MenuIndex].Questions[7]):
            # Button :=  button_quiet{DefaultText := StringToMessage(GetCategoryQuestion(MenuIndex,7))}
            Button :=  button_quiet{DefaultText := GetCategoryQuestionMessage(MenuIndex,7)}
            Button.OnClick().Subscribe(OnClickQuestion8) 
            set AllWidgets += array{Button}
        if(QuestionExists := CurrentCategories[MenuIndex].Questions[8]):
            # Button :=  button_quiet{DefaultText := StringToMessage(GetCategoryQuestion(MenuIndex,8))}
            Button :=  button_quiet{DefaultText := GetCategoryQuestionMessage(MenuIndex,8)}
            Button.OnClick().Subscribe(OnClickQuestion9) 
            set AllWidgets += array{Button}
        if(QuestionExists := CurrentCategories[MenuIndex].Questions[9]):
            # Button :=  button_quiet{DefaultText := StringToMessage(GetCategoryQuestion(MenuIndex,9))}
            Button :=  button_quiet{DefaultText := GetCategoryQuestionMessage(MenuIndex,9)}
            Button.OnClick().Subscribe(OnClickQuestion10) 
            set AllWidgets += array{Button}
        if(QuestionExists := CurrentCategories[MenuIndex].Questions[10]):
            # Button :=  button_quiet{DefaultText := StringToMessage(GetCategoryQuestion(MenuIndex,10))}
            Button :=  button_quiet{DefaultText := GetCategoryQuestionMessage(MenuIndex,10)}
            Button.OnClick().Subscribe(OnClickQuestion11) 
            set AllWidgets += array{Button}
        if(QuestionExists := CurrentCategories[MenuIndex].Questions[11]):
            # Button :=  button_quiet{DefaultText := StringToMessage(GetCategoryQuestion(MenuIndex,11))}
            Button :=  button_quiet{DefaultText := GetCategoryQuestionMessage(MenuIndex,11)}
            Button.OnClick().Subscribe(OnClickQuestion12) 
            set AllWidgets += array{Button}
        if(QuestionExists := CurrentCategories[MenuIndex].Questions[12]):
            # Button :=  button_quiet{DefaultText := StringToMessage(GetCategoryQuestion(MenuIndex,12))}
            Button :=  button_quiet{DefaultText := GetCategoryQuestionMessage(MenuIndex,12)}
            Button.OnClick().Subscribe(OnClickQuestion13) 
            set AllWidgets += array{Button}
        
        return AllWidgets


    # AddButtonsToWidget_NEW(LayoutCanvas: canvas , MenuIndex :int): void =
    #     Print("AddButtonsToWidget")    
    #     LayoutCanvas.AddWidget(MakeAllButtons(GetAllButtonsForQuestions(MenuIndex), 80.0))


    MakeAllButtons_New(AllWidgets:[]widget, MarginTop:float):overlay_slot=
        # Logger.Print("Making All Buttons - Length: {AllWidgets.Length}")
        
        return overlay_slot:
            HorizontalAlignment := horizontal_alignment.Fill
            VerticalAlignment := vertical_alignment.Fill
            Padding := margin{Top := 0.0, Left := 0.0, Right := 0.0, Bottom := 20.0}
            Widget := stack_box:
                Orientation := orientation.Vertical 
                Slots := for (Index -> aWidget : AllWidgets):
                    stack_box_slot:
                        HorizontalAlignment := horizontal_alignment.Fill
                        VerticalAlignment := vertical_alignment.Center
                        Padding := margin{Top := 20.0, Left := 20.0, Right := 20.0, Bottom := 0.0}
                        Widget := overlay:
                            Slots := array:
                                overlay_slot:
                                    HorizontalAlignment := horizontal_alignment.Fill
                                    VerticalAlignment := vertical_alignment.Center
                                    Padding := margin{Top := 0.0, Left := 20.0, Right := 20.0, Bottom := 0.0}
                                    Widget := MakeRoundedOverlayQuietButton(White, ?TextureSize:=18.0)
                                overlay_slot:
                                    HorizontalAlignment := horizontal_alignment.Fill
                                    VerticalAlignment := vertical_alignment.Center
                                    Padding := margin{Top := 0.0, Left := 20.0, Right := 20.0, Bottom := 0.0}
                                    Widget := aWidget


    OnClickButtonExit<private>(Message: widget_message):void=
        Logger.Print("Player clicked exit")
        spawn . HideMenuIfPresent(Message.Player)

    OnClickQuestion1<private>(Message : widget_message) : void =
        UpdateDialouge(GetCategoryQuestionMessage(GetMenuIndex(Message.Player),0),Message.Player)
        # HideMenuIfPresent(Message.Player)

    OnClickQuestion2<private>(Message : widget_message) : void =
        UpdateDialouge(GetCategoryQuestionMessage(GetMenuIndex(Message.Player),1),Message.Player)
        # HideMenuIfPresent(Message.Player)

    OnClickQuestion3<private>(Message : widget_message) : void =
        UpdateDialouge(GetCategoryQuestionMessage(GetMenuIndex(Message.Player),2),Message.Player)
        # HideMenuIfPresent(Message.Player)

    OnClickQuestion4<private>(Message : widget_message) : void =
        UpdateDialouge(GetCategoryQuestionMessage(GetMenuIndex(Message.Player),3),Message.Player)
        # HideMenuIfPresent(Message.Player)

    OnClickQuestion5<private>(Message : widget_message) : void =
        UpdateDialouge(GetCategoryQuestionMessage(GetMenuIndex(Message.Player),4),Message.Player)
        # HideMenuIfPresent(Message.Player)

    OnClickQuestion6<private>(Message : widget_message) : void =
        UpdateDialouge(GetCategoryQuestionMessage(GetMenuIndex(Message.Player),5),Message.Player)
        # HideMenuIfPresent(Message.Player)

    OnClickQuestion7<private>(Message : widget_message) : void =
        UpdateDialouge(GetCategoryQuestionMessage(GetMenuIndex(Message.Player),6),Message.Player)
        # HideMenuIfPresent(Message.Player)

    OnClickQuestion8<private>(Message : widget_message) : void =
        UpdateDialouge(GetCategoryQuestionMessage(GetMenuIndex(Message.Player),7),Message.Player)
        # HideMenuIfPresent(Message.Player)

    OnClickQuestion9<private>(Message : widget_message) : void =
        UpdateDialouge(GetCategoryQuestionMessage(GetMenuIndex(Message.Player),8),Message.Player)
        # HideMenuIfPresent(Message.Player)

    OnClickQuestion10<private>(Message : widget_message) : void =
        UpdateDialouge(GetCategoryQuestionMessage(GetMenuIndex(Message.Player),9),Message.Player)
        # HideMenuIfPresent(Message.Player)

    OnClickQuestion11<private>(Message : widget_message) : void =
        UpdateDialouge(GetCategoryQuestionMessage(GetMenuIndex(Message.Player),10),Message.Player)
        # HideMenuIfPresent(Message.Player)
    
    OnClickQuestion12<private>(Message : widget_message) : void =
        UpdateDialouge(GetCategoryQuestionMessage(GetMenuIndex(Message.Player),11),Message.Player)
        # HideMenuIfPresent(Message.Player)

    OnClickQuestion13<private>(Message : widget_message) : void =
        UpdateDialouge(GetCategoryQuestionMessage(GetMenuIndex(Message.Player),12),Message.Player)
        # HideMenuIfPresent(Message.Player)

    OnClickButtonYes<private>(Message : widget_message) : void =
        Logger.Print("Player clicked yes")

        if(GetSimulationElapsedTime() - LastRespondedTime < 1.0):
            Logger.Print("Player clicked too fast")

            #remove the ui
            if:
                TheGlobals := Globals[GetSession()]
                CPlayer := TheGlobals.CustomPlayerMap[Message.Player]
                PlayerOneUI := GetPlayerUI[Message.Player]
            then:
                if:
                    ExistingQuestionUI := CPlayer.GuessUI.MaybeQuestionUI?
                then:
                    PlayerOneUI.RemoveWidget(ExistingQuestionUI)
                    set CPlayer.GuessUI.MaybeQuestionUI = false

            return

        # HideMenuIfPresent(Message.Player)
        HideEveryoneUI()
        HUDResponseYes.SetText(SelectedQuestion)
        HUDResponseYes.Show()
        AnalyticQuestionAnswered.Submit(Message.Player)

        if:
            TheGlobals := Globals[GetSession()]
            CPlayer := TheGlobals.CustomPlayerMap[Message.Player]
            PlayerOneUI := GetPlayerUI[Message.Player]
        then:
            if:
                ExistingQuestionUI := CPlayer.GuessUI.MaybeQuestionUI?
            then:
                PlayerOneUI.RemoveWidget(ExistingQuestionUI)
                set CPlayer.GuessUI.MaybeQuestionUI = false
                
        set LastRespondedTime = GetSimulationElapsedTime()
        OpponentResponseEvent.Signal()

        if:
            TheGlobals :=  Globals[GetSession()]
        then:
            set TheGlobals.OpponentAnswered = (true,false)

    OnClickButtonNo<private>(Message : widget_message) : void =
        Logger.Print("Player clicked no")

        if(GetSimulationElapsedTime() - LastRespondedTime < 1.0):
            Logger.Print("Player clicked too fast")

            #remove the ui
            if:
                TheGlobals := Globals[GetSession()]
                CPlayer := TheGlobals.CustomPlayerMap[Message.Player]
                PlayerOneUI := GetPlayerUI[Message.Player]
            then:
                if:
                    ExistingQuestionUI := CPlayer.GuessUI.MaybeQuestionUI?
                then:
                    PlayerOneUI.RemoveWidget(ExistingQuestionUI)
                    set CPlayer.GuessUI.MaybeQuestionUI = false

            return

        # HideMenuIfPresent(Message.Player)
        HideEveryoneUI()
        HUDResponseNo.SetText(SelectedQuestion)
        HUDResponseNo.Show()
        AnalyticQuestionAnswered.Submit(Message.Player)

        if:
            TheGlobals := Globals[GetSession()]
            CPlayer := TheGlobals.CustomPlayerMap[Message.Player]
            PlayerOneUI := GetPlayerUI[Message.Player]
        then:
            if:
                ExistingQuestionUI := CPlayer.GuessUI.MaybeQuestionUI?
            then:
                PlayerOneUI.RemoveWidget(ExistingQuestionUI)
                set CPlayer.GuessUI.MaybeQuestionUI = false

        set LastRespondedTime = GetSimulationElapsedTime()
        OpponentResponseEvent.Signal()

        if:
            TheGlobals :=  Globals[GetSession()]
        then:
            set TheGlobals.OpponentAnswered = (false,true)

    OnClickButtonMaybe<private>(Message : widget_message) : void =
        Logger.Print("Player clicked maybe")

        if(GetSimulationElapsedTime() - LastRespondedTime < 1.0):
            Logger.Print("Player clicked too fast")

            #remove the ui
            if:
                TheGlobals := Globals[GetSession()]
                CPlayer := TheGlobals.CustomPlayerMap[Message.Player]
                PlayerOneUI := GetPlayerUI[Message.Player]
            then:
                if:
                    ExistingQuestionUI := CPlayer.GuessUI.MaybeQuestionUI?
                then:
                    PlayerOneUI.RemoveWidget(ExistingQuestionUI)
                    set CPlayer.GuessUI.MaybeQuestionUI = false

            return

        # HideMenuIfPresent(Message.Player)
        HideEveryoneUI()
        HUDResponseMaybe.SetText(SelectedQuestion)
        HUDResponseMaybe.Show()

        if:
            TheGlobals := Globals[GetSession()]
            CPlayer := TheGlobals.CustomPlayerMap[Message.Player]
            PlayerOneUI := GetPlayerUI[Message.Player]
        then:
            if:
                ExistingQuestionUI := CPlayer.GuessUI.MaybeQuestionUI?
            then:
                PlayerOneUI.RemoveWidget(ExistingQuestionUI)
                set CPlayer.GuessUI.MaybeQuestionUI = false


        AnalyticQuestionAnswered.Submit(Message.Player)
        set LastRespondedTime = GetSimulationElapsedTime()
        OpponentResponseEvent.Signal()

    # UpdateDialouge(Question : string, Player: player):void=
    #     # if(not CurrentlyAskingQuestion?):
    #         # set CurrentlyAskingQuestion = true
    #     set SelectedQuestion = Question
    #     QuestionAskedEvent.Signal()
    #     AnalyticQuestionAsked.Submit(Player)
    #     spawn . HideMenuIfPresent(Player)
    #     spawn . ShowOpponentMenu(Player)

    UpdateDialouge(Question : message, Player: player):void=
        # if(not CurrentlyAskingQuestion?):
            # set CurrentlyAskingQuestion = true
        set SelectedQuestion = Question
        QuestionAskedEvent.Signal()
        AnalyticQuestionAsked.Submit(Player)
        spawn . HideMenuIfPresent(Player)
        spawn . ShowOpponentMenu(Player)

    MakeAllButtons(AllWidgets:[]widget, MarginTop:float):canvas_slot=
        Logger.Print("Making All Buttons")
        
        return canvas_slot:
            Anchors := anchors{Minimum := vector2{X := 0.2, Y := 0.0}, Maximum := vector2{X := 0.8, Y := 0.0}}
            Offsets := margin{Top := 120.0, Left := 0.0, Right := 0.0, Bottom := 0.0}
            Alignment := vector2{X := 0.5, Y := 0.0}
            ZOrder := 5
            
            Widget := stack_box:
                Orientation := orientation.Vertical 
                Slots := for (Index -> aWidget : AllWidgets):
                    stack_box_slot:
                        HorizontalAlignment := horizontal_alignment.Fill
                        VerticalAlignment := vertical_alignment.Fill
                        Widget := aWidget
                        Padding := margin{Top := 20.0, Left := 20.0, Right := 20.0, Bottom := 0.0}

    MakeButtonSlot(Widget: widget, MarginTop : float) : canvas_slot =
        Logger.Print("Making a Button Slot")

        return canvas_slot:
            Anchors := anchors{Minimum := vector2{X := 0.0, Y := 0.0}, Maximum := vector2{X := 1.0, Y := 0.0}}
            Offsets := margin{Top := MarginTop, Left := 0.0, Right := 0.0, Bottom := 0.0}
            Alignment := vector2{X := 0.5, Y := 0.0}
            ZOrder := 5
            Widget := Widget

    GetMenuIndex<private>(Player : player) : int =
        var MenuIndex : int = 0
        if:
            TheGlobals :=  Globals[GetSession()]
            CPlayer := TheGlobals.CustomPlayerMap[Player]
        then:
                set MenuIndex = CPlayer.MenuCategoryIndex
                return MenuIndex
        else:
            Logger.Print("Info: Player did not have any menu index, initalizing with 0")
            NewMenuIndex :int = 0
            if:
                TheGlobals :=  Globals[GetSession()]
                CPlayer := TheGlobals.CustomPlayerMap[Player]
            then:
                set CPlayer.MenuCategoryIndex = NewMenuIndex
            return NewMenuIndex
 
    GetCategoryTitle(MenuIndex : int) : string =
        Logger.Print("GetCategoryTitle - Index: {MenuIndex}")
        if(Category := CurrentCategories[MenuIndex]):
            Logger.Print("Should be title from Cat: {Category.Name}")
            return Category.Name
        else:
            Logger.Print("Info: Category could not be loaded")
            return ""

        
    GetCategoryTitleMessage(MenuIndex : int) : message =
        Logger.Print("GetCategoryTitle - Index: {MenuIndex}")
        if(Category := CurrentCategories[MenuIndex]):
            Logger.Print("Should be title from Cat: {Category.Name}")
            return Category.NameMessage
        else:
            Logger.Print("Info: Category could not be loaded")
            return Err("Err: Unable to get category title")

    GetCategoryQuestions(MenuIndex : int) : []question =
        Logger.Print("GetCategoryQuestions - Index: {MenuIndex}")
        if(Category := CurrentCategories[MenuIndex]):
            Logger.Print("Should be questions from Cat: {Category.Name}")
            return Category.Questions
        else:
            Logger.Print("Info: Category could not be loaded")
            return array{}

    # GetCategoryQuestion(MenuIndex : int, Index: int) : string =
    #     Logger.Print("GetCategoryQuestion: MenuIndex: {MenuIndex} --- Index: {Index}")
    #     if(Category := CurrentCategories[MenuIndex]):
    #         if(Question := Category.Questions[Index]):
    #             return Question.TheQuestion
    #         Logger.Print("Info: Question {Index} of category {MenuIndex} is not set")
    #         return ""
           
    #     else:
    #         Logger.Print("Info: Category could not be loaded")
    #         return ""

    GetCategoryQuestionMessages(MenuIndex : int) : []message =
        Logger.Print("GetCategoryQuestions - Index: {MenuIndex}")
        var TheMessages : []message = array{}
        if:
            Category := CurrentCategories[MenuIndex]
            Questions := Category.Questions
        then:
            for(Question : Questions):
                set TheMessages += array{Question.TheQuestion}

        return TheMessages

    GetCategoryQuestionMessage(MenuIndex : int, Index: int) : message =
        Logger.Print("GetCategoryQuestion: MenuIndex: {MenuIndex} --- Index: {Index}")
        if(Category := CurrentCategories[MenuIndex]):
            if(Question := Category.Questions[Index]):
                return Question.TheQuestion
            Logger.Print("Info: Question {Index} of category {MenuIndex} is not set")

        return Err("Err: Unable to get question message")


    OnClickButtonClassLeft<private>(Message : widget_message) : void =
        Logger.Print("OnClickButtonClassLeft")
        if:
            TheGlobals :=  Globals[GetSession()]
            CPlayer := TheGlobals.CustomPlayerMap[Message.Player]
        then:
            NewMenuIndex := CalculateOverflowMenuIndex(GetMenuIndex(Message.Player) - 1)
            set CPlayer.MenuCategoryIndex = NewMenuIndex
            # spawn{ShowMenu(Message.Player)}
            spawn . UpdateQuestionsOnly(Message.Player)

    OnClickButtonClassRight<private>(Message : widget_message) : void =
        Logger.Print("OnClickButtonClassRight")
        if:
            TheGlobals :=  Globals[GetSession()]
            CPlayer := TheGlobals.CustomPlayerMap[Message.Player]
        then:
            NewMenuIndex : int = CalculateOverflowMenuIndex(GetMenuIndex(Message.Player) + 1)
            set CPlayer.MenuCategoryIndex = NewMenuIndex
            # spawn{ShowMenu(Message.Player)}
            spawn . UpdateQuestionsOnly(Message.Player)

    CalculateOverflowMenuIndex<private>(MenuIndex : int) : int=
        if(MenuIndex < 0):
            return CurrentCategories.Length - 1
        else if(MenuIndex > CurrentCategories.Length - 1):
            return 0
        else return MenuIndex

    CreateEmptyCanvas() : canvas =
        MenuCanvas : canvas = canvas:
            Slots := array{}
        return MenuCanvas

    # HideMenuIfPresent(Player : player) <suspends>: void =
    #     if:
    #         TheGlobals :=  Globals[GetSession()]
    #     then:
    #         if:
    #             CPlayer := TheGlobals.CustomPlayerMap[Player]
    #         then:
    #             if:       
    #                 PlayerUI := GetPlayerUI[Player]
    #                 # RootWidget := ExistingQuestionButtons.Widget.GetRootWidget[]
    #             then:
                    

    #                 # QuestionUI.SetVisibility(widget_visibility.Hidden)
    #                 # ExistingQuestionButtons.Widget.SetVisibility(widget_visibility.Hidden)
    #                 # CPlayer.GuessUI.AllQuestionButtonsOverlay.RemoveWidget(ExistingQuestionButtons.Widget)
    #                 # CPlayer.GuessUI.AllQuestionButtonsOverlay.RemoveWidget(RootWidget)
    #                 # set CPlayer.GuessUI.MaybeExistingQuestionButtons = false

                    
    #                 # if:
    #                 #     ExistingQuestionButtons := CPlayer.GuessUI.MaybeExistingQuestionButtons?
    #                 # then:
    #                 #     CPlayer.GuessUI.AllQuestionButtonsOverlay.RemoveWidget(ExistingQuestionButtons.Widget)
    #                 # else:
    #                 #     Print("Failed To Remove UI Question Buttons")

    #                 # set CPlayer.GuessUI.MaybeExistingQuestionButtons = false

    #                 # if:
    #                 #     QuestionUI := CPlayer.GuessUI.MaybeQuestionUI?
    #                 # then:
    #                 #     PlayerUI.RemoveWidget(QuestionUI)
    #                 # else:
    #                 #     Print("Failed To Remove UI Question UI")

    #                 # set CPlayer.GuessUI.MaybeQuestionUI = false


    #                 if:
    #                     ExistingQuestionButtons := CPlayer.GuessUI.MaybeExistingQuestionButtons?
    #                     # QuestionUI := CPlayer.GuessUI.MaybeQuestionUI?
    #                     # RootWidget := ExistingQuestionButtons.Widget.GetRootWidget[]
    #                     # ParentWidget := ExistingQuestionButtons.Widget.GetParentWidget[]
    #                 then:
    #                     PlayerUI.RemoveWidget(ExistingQuestionButtons.Widget)
    #                     # PlayerUI.RemoveWidget(ParentWidget)
    #                     # PlayerUI.RemoveWidget(RootWidget)
    #                     set CPlayer.GuessUI.MaybeExistingQuestionButtons = false
    #                     # set CPlayer.GuessUI.MaybeQuestionUI = false

    #                 else:
    #                     Print("Failed To Remove Existing Question Buttons")

    #                 Sleep(0.1)

    #                 if:
    #                     RootWidget := CPlayer.GuessUI.MaybeQuestionUI?
    #                 then:
    #                     PlayerUI.RemoveWidget(RootWidget)
    #                     set CPlayer.GuessUI.MaybeQuestionUI = false
    #                 else:
    #                     Print("Failed To Remove UI Question UI")


    #         else:
    #             Print("HideMenuIfPresent - Failed To Get CPlayer")

    #     else:
    #         Print("HideMenuIfPresent - Failed To Get Globals")

    HideMenuIfPresent(Player : player) <suspends>: void =
        if:
            TheGlobals :=  Globals[GetSession()]
        then:
            if:
                CPlayer := TheGlobals.CustomPlayerMap[Player]
            then:
                if:
                    PlayerUI := GetPlayerUI[Player]
                then:
                    if:
                        ExistingQuestionButtons := CPlayer.GuessUI.MaybeExistingQuestionButtons?
                    then:
                        ExistingQuestionButtons.Widget.SetVisibility(widget_visibility.Hidden)
                        Sleep(0.1)  # Ensure the visibility change takes effect
                        PlayerUI.RemoveWidget(ExistingQuestionButtons.Widget)
                        set CPlayer.GuessUI.MaybeExistingQuestionButtons = false
                    else:
                        Logger.Print("Failed To Remove Existing Question Buttons")
    
                    Sleep(0.1)  # Ensure enough time for removal to take effect
    
                    if:
                        RootWidget := CPlayer.GuessUI.MaybeQuestionUI?
                    then:
                        RootWidget.SetVisibility(widget_visibility.Hidden)
                        Sleep(0.1)  # Ensure the visibility change takes effect
                        PlayerUI.RemoveWidget(RootWidget)
                        set CPlayer.GuessUI.MaybeQuestionUI = false
                        set CPlayer.GameBoard.QuestionBankOpen = false
                    else:
                        Logger.Print("Failed To Remove UI Question UI")
                else:
                    Logger.Print("HideMenuIfPresent - Failed To Get CPlayer")
            else:
                Logger.Print("HideMenuIfPresent - Failed To Get Globals")
    

    HidePaintPanelIfPresent(Player : player) : void =
        if:
            TheGlobals :=  Globals[GetSession()]
            CPlayer := TheGlobals.CustomPlayerMap[Player]
            PaintPanel := CPlayer.GuessUI.MaybePaintPaletteUI?
            PlayerUI := GetPlayerUI[Player]
        then:
            PlayerUI.RemoveWidget(PaintPanel)
            # set CPlayer.GuessUI.MaybePaintPaletteUI = false
        else:
            Logger.Print("Could not find Paint UI to remove")

    ReshowPaintPanelIfPresent(Player : player) : void =
        if:
            TheGlobals :=  Globals[GetSession()]
            CPlayer := TheGlobals.CustomPlayerMap[Player]
            PaintPanel := CPlayer.GuessUI.MaybePaintPaletteUI?
            PlayerUI := GetPlayerUI[Player]
        then:
            PlayerUI.AddWidget(PaintPanel, player_ui_slot{ZOrder := 101, InputMode := ui_input_mode.All})
        else:
            Logger.Print("Could not find Paint UI to reshow")

    CreateMenuCanvas(Widget : widget) : canvas =
        MenuCanvas : canvas = canvas:
            Slots := array:
                canvas_slot:
                    Anchors := anchors{Minimum := vector2{X := 0.0, Y := 0.5}, Maximum := vector2{X := 1.0, Y := 0.5}}
                    Offsets := margin{Top := 0.0, Left := 0.0, Right := 0.0, Bottom := 0.0}
                    Alignment := vector2{X := 0.5, Y := 0.5}
                    ZOrder := 100
                    Widget := Widget
        return MenuCanvas

    CreateMenuCanvasBottom(Widget : widget) : canvas =
        MenuCanvas : canvas = canvas:
            Slots := array:
                canvas_slot:
                    Anchors := anchors{Minimum := vector2{X := 0.5, Y := 1.0}, Maximum := vector2{X := 0.5, Y := 1.0}}
                    Offsets := margin{Top := 0.0, Left := 0.0, Right := 0.0, Bottom := 50.0}
                    Alignment := vector2{X := 0.5, Y := 1.0}
                    ZOrder := 100
                    Widget := Widget
        return MenuCanvas