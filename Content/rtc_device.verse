using { /Fortnite.com/Devices }
using { /Verse.org/Simulation }
using { /UnrealEngine.com/Temporary/Diagnostics }
using { /Verse.org/Simulation/Tags }

time_and_date <public> := class<persistable><final><concrete>:
    @editable Day : int = 0
    @editable Month : int = 0
    @editable Year : int = 0
    @editable Hour : int = 0
    @editable Minute : int = 0
    @editable Second : int = 0

    Print():void=
        Print("{Day}/{Month}/{Year} {Hour}:{Minute}:{Second}")

    String()<transacts>:string=
        return "{Day}/{Month}/{Year} {Hour}:{Minute}:{Second}"


date_string := class<concrete>():
    @editable . ToolTip := DateToolTip
    Date : string = "01/01/2024"
    DateToolTip<localizes>:message = "Format: DD/MM/YYYY"

    @editable . ToolTip := TimeToolTip
    Time : string = "00:00:00"
    TimeToolTip<localizes>:message = "Must use UTC Time!\nFormat: HH:MM:SS"


MakeTimeAndDate(Day:int, Month:int, Year:int, Hour:int, Minute:int, Second:int)<transacts>:time_and_date=
    time_and_date:
        Day := Day
        Month := Month
        Year := Year
        Hour := Hour
        Minute := Minute
        Second := Second

# MakeTimeAndDate(Prop:creative_prop)<transacts>:time_and_date=

#     Transform := Prop.GetTransform()

#     if:
#         Prop.IsValid[]
#         Day := Int[Transform.Translation.X]
#         Month := Int[Transform.Translation.Y]
#         Year := Int[Transform.Translation.Z]
#         Hour := Int[Transform.Scale.X]-1
#         Minute := Int[Transform.Scale.Y]-1
#         Second := Int[Transform.Scale.Z]-1
#     then:
#         return MakeTimeAndDate(Day, Month, Year, Hour, Minute, Second)

#     return time_and_date{} #fail safe


    # Returns the number of seconds since January 1, 1970 UTC, ignoring leap seconds. I.e, this function implements Unix time. This function always returns the same value within the same transaction.

MakeDateTimeNowUtc<constructor><public>()<decides><transacts>:=time_and_date:
    let:
        DateTime := MakeDateTimeFromEpochSeconds[GetSecondsSinceEpoch()]
    Year := DateTime.Year
    Month := DateTime.Month
    Day := DateTime.Day
    Hour := DateTime.Hour
    Minute := DateTime.Minute
    Second := DateTime.Second



# MakeDateTimeFromEpochSeconds<public>(EpochSeconds:float)<decides><transacts>:time_and_date=
#     IntEpochSeconds := Int[EpochSeconds]
#     Days := Quotient[IntEpochSeconds, 86400]
#     CivilDate := MakeCivilFromDays[Days]
#     var Seconds:int = IntEpochSeconds - Days * 86400
#     Hour := Quotient[Seconds, 3600]
#     set Seconds -= Hour * 3600
#     Minute := Quotient[Seconds, 60]
#     set Seconds -= Minute * 60
#     time_and_date:
#         Year := CivilDate.Year
#         Month := CivilDate.Month
#         Day := CivilDate.Day
#         Hour := Hour
#         Minute := Minute
#         Second := Seconds


MakeDateTimeFromEpochSeconds<public>(EpochSeconds:float)<decides><transacts>:time_and_date=
    # Constants
    SecondsPerDay := 86400
    DaysInYear := 365

    # Convert seconds to days and calculate the remaining seconds in the day
    IntEpochSeconds := Int[EpochSeconds]
    TotalDays := Quotient[IntEpochSeconds, SecondsPerDay]
    var RemainingSeconds : int = IntEpochSeconds - TotalDays * SecondsPerDay

    # Calculate the hour, minute, and second of the day
    Hour := Quotient[RemainingSeconds, 3600]
    set RemainingSeconds -= Hour * 3600
    Minute := Quotient[RemainingSeconds, 60]
    Second := RemainingSeconds - Minute * 60

    # Calculate the date using the new algorithm
    CivilDate := CalculateCivilDateFromDays[TotalDays]

    # Return the full time_and_date structure
    time_and_date:
        Year := CivilDate.Year
        Month := CivilDate.Month
        Day := CivilDate.Day
        Hour := Hour
        Minute := Minute
        Second := Second


CalculateCivilDateFromDays<public>(TotalDays:int)<decides><transacts>:civil_date=
    # Constants
    DaysInCommonYear := 365
    DaysInLeapYear := 366
    var DaysInMonth : []int = {31, 28, 31, 30, 31, 30, 31, 31, 30, 31, 30, 31}

    # Epoch starts on January 1, 1970
    var Year : int = 1970
    var DaysRemaining : int = TotalDays
    var DaysInCurrentYear : int = 0

    # Calculate the year
    loop:
        if (IsLeapYear[Year]):
             set DaysInCurrentYear = DaysInLeapYear
        else:
            set DaysInCurrentYear = DaysInCommonYear

        if (DaysRemaining < DaysInCurrentYear):
            break

        set DaysRemaining -= DaysInCurrentYear
        set Year += 1

    # Adjust February for leap years
    if (IsLeapYear[Year]):
        if . set DaysInMonth[1] = 29  # February has 29 days in a leap year

    # Calculate the month
    var Month : int = 1
    loop:
        if:
            DaysInCurrentMonth := DaysInMonth[Month - 1]
        then:

            if (DaysRemaining < DaysInCurrentMonth):
                break

            set DaysRemaining -= DaysInCurrentMonth
            set Month += 1

    # The remaining days represent the day of the month
    Day := DaysRemaining + 1

    # Return the civil_date structure
    civil_date:
        Year := Year
        Month := Month
        Day := Day


civil_date<public> := struct:
    Year<public>:   int = 0
    Month<public>:  int = 0
    Day<public>:    int = 0

MakeCivilFromDays<public>(Days:int)<decides><transacts>:time_and_date=
    AdjustedDays := Days + 719468
    Era := if (AdjustedDays >= 0) then Quotient[AdjustedDays, 146097] else Quotient[AdjustedDays - 146096, 146097]
    DayOfEra := AdjustedDays - Era * 146097
    YearOfEra := Quotient[DayOfEra - Quotient[DayOfEra, 1460] + Quotient[DayOfEra, 36524] - Quotient[DayOfEra, 146096], 365]
    Year := YearOfEra + Era * 400
    DayOfYear := DayOfEra - (365 * YearOfEra + Quotient[YearOfEra, 4] - Quotient[YearOfEra, 100])
    # The algorithm uses March as the first month of the year.
    MarchBasedMonth := Quotient[5 * DayOfYear + 2, 153]
    Day := DayOfYear - Quotient[153 * MarchBasedMonth + 2, 5] + 1
    Month := if (MarchBasedMonth < 10) then MarchBasedMonth + 3 else MarchBasedMonth - 9
    time_and_date{Year := Year, Month := Month, Day := Day}






CompareEqualDateAndTime(FirstDate:time_and_date, SecondDate:time_and_date)<transacts><decides>:void=
    FirstDate.Day = SecondDate.Day
    FirstDate.Month = SecondDate.Month
    FirstDate.Year = SecondDate.Year
    FirstDate.Hour = SecondDate.Hour
    FirstDate.Minute = SecondDate.Minute
    FirstDate.Second = SecondDate.Second


CompareEqualDateIgnoreTime(FirstDate:time_and_date, SecondDate:time_and_date)<transacts><decides>:void=
    FirstDate.Day = SecondDate.Day
    FirstDate.Month = SecondDate.Month
    FirstDate.Year = SecondDate.Year

PassedDateAndTime(CurrentTime:time_and_date, DateToCheck:time_and_date)<transacts><decides>:void=
    # Print("CurrentTime: {CurrentTime.Year}/{CurrentTime.Month}/{CurrentTime.Day} {CurrentTime.Hour}:{CurrentTime.Minute}:{CurrentTime.Second}")
    # Print("DateToCheck: {DateToCheck.Year}/{DateToCheck.Month}/{DateToCheck.Day} {DateToCheck.Hour}:{DateToCheck.Minute}:{DateToCheck.Second}")
    
    var HasPassed : logic = false

    if (CurrentTime.Year > DateToCheck.Year):
        set HasPassed = true
    else if (CurrentTime.Year < DateToCheck.Year):
        set HasPassed = false
    else:
        if (CurrentTime.Month > DateToCheck.Month):
            set HasPassed = true
        else if (CurrentTime.Month < DateToCheck.Month):
            set HasPassed = false
        else:
            if (CurrentTime.Day > DateToCheck.Day):
                set HasPassed = true
            else if (CurrentTime.Day < DateToCheck.Day):
                set HasPassed = false
            else:
                if (CurrentTime.Hour > DateToCheck.Hour):
                    set HasPassed = true
                else if (CurrentTime.Hour < DateToCheck.Hour):
                    set HasPassed = false
                else:
                    if (CurrentTime.Minute > DateToCheck.Minute):
                        set HasPassed = true
                    else if (CurrentTime.Minute < DateToCheck.Minute):
                        set HasPassed = false
                    else:
                        if (CurrentTime.Second >= DateToCheck.Second):
                            set HasPassed = true
                        else:
                            set HasPassed = false

    HasPassed?



IsLeapYear(Year:int)<transacts><decides>:void=
    Mod[Year, 4] = 0 and (Mod[Year, 100] <> 0 or Mod[Year, 400] = 0)
    

IsLeapYear(Date:time_and_date)<transacts><decides>:void=
    IsLeapYear[Date.Year]


IsWeekend(Date:time_and_date)<transacts><decides>:void=
    DayOfWeek := Mod[CalculatedTotalDaysSinceEpoch(Date),7]
    DayOfWeek = 0 or DayOfWeek = 6

IsWeekday(Date:time_and_date)<transacts><decides>:void=
    not IsWeekend[Date]

GetDaysInMonth(Year:int, Month:int)<transacts>:int=
    if (Month = 1 or Month = 3 or Month = 5 or Month = 7 or Month = 8 or Month = 10 or Month = 12):
        return 31
    else if (Month = 4 or Month = 6 or Month = 9 or Month = 11):
        return 30
    else if (Month = 2):
        if (IsLeapYear[Year]):  # Adjust for February in leap years
            return 29
        else:
            return 28
    return 0  # Invalid month
    
    

GetDaysInMonth(Date:time_and_date)<transacts>:int=
    GetDaysInMonth(Date.Year, Date.Month)

CalculatedTotalDaysSinceEpoch(Year:int, Month:int, Day:int)<transacts>:int=
    if (Day <= 0 or Month <= 0 or Month > 12 or Year < 1970):
        Print("Invalid date: {Day}/{Month}/{Year}")
        return 0  # Return a default value for invalid dates.

    var TotalDays : int = 0

    # Count days for complete years since epoch (1970)
    for (SomeYear := 1970 .. Year - 1):
        if (IsLeapYear[SomeYear]):  # Leap year adjustment
            set TotalDays += 366
        else:
            set TotalDays += 365

    # Count days for complete months in the current year
    for (SomeMonth := 1 .. Month - 1):
        set TotalDays += GetDaysInMonth(Year, SomeMonth)

    # Add the days in the current month
    set TotalDays += Day - 1  # Subtract 1 because the epoch starts at the beginning of a day

    Print("Date: {Day}/{Month}/{Year} - Total Days Since Epoch: {TotalDays}")
    return TotalDays
    

    
    

CalculatedTotalDaysSinceEpoch(Date:time_and_date)<transacts>:int=
    CalculatedTotalDaysSinceEpoch(Date.Year, Date.Month, Date.Day)
    

CalculatedTotalSecondsSinceEpoch(Year:int, Month:int, Day:int, Hour:int, Minute:int, Second:int)<transacts>:int=
    if:
        YearFloorDivision := Floor((Year - 1969) / 4)
        DaysSinceEpoch := ((Year - 1970) * 365 + YearFloorDivision) + GetDayOfYear(Year, Month, Day) - 1
        SecondsInDay := Hour * 3600 + Minute * 60 + Second
    then:
        return DaysSinceEpoch * 86400 + SecondsInDay
    -1

    
GetDayOfYear<internal>(Year: int, Month: int, Day: int)<transacts>:int=
    if:
        N1 := Floor(275 * Month / 9)
        N2 := Floor((Month + 9) / 12)
        N3 := 1 + Floor((Year - 4 * Floor(Year / 4) + 2) / 3)
    then:
        return N1 - (N2 * N3) + Day - 30
    -1
    

CalculatedTotalSecondsSinceEpoch(Date:time_and_date)<transacts>:int=
    CalculatedTotalSecondsSinceEpoch(Date.Year, Date.Month, Date.Day, Date.Hour, Date.Minute, Date.Second)




DaysBetweenDates(FirstDate:time_and_date, SecondDate:time_and_date)<transacts>:int=
    FirstDateTotalDays := CalculatedTotalDaysSinceEpoch(FirstDate)
    SecondDateTotalDays := CalculatedTotalDaysSinceEpoch(SecondDate)
    Days := Abs(SecondDateTotalDays - FirstDateTotalDays)
    return Days
    


SecondsBetweenDates(FirstDate:time_and_date, SecondDate:time_and_date)<transacts>:int=
    FirstDateTotalSeconds := CalculatedTotalSecondsSinceEpoch(FirstDate)
    SecondDateTotalSeconds := CalculatedTotalSecondsSinceEpoch(SecondDate)
    TotalSeconds := SecondDateTotalSeconds - FirstDateTotalSeconds
    return TotalSeconds
    

MinutesBetweenDates(FirstDate:time_and_date, SecondDate:time_and_date)<transacts>:int=
    TotalSeconds := SecondsBetweenDates(FirstDate, SecondDate)
    var TotalMinutes: int = 0
    if:
        aTotalMinutes := Floor(TotalSeconds / 60)
    then:
        set TotalMinutes = aTotalMinutes

    return TotalMinutes
    

HoursBetweenDates(FirstDate:time_and_date, SecondDate:time_and_date)<transacts>:int=
    TotalMinutes := MinutesBetweenDates(FirstDate, SecondDate)
    var TotalHours: int = 0
    if:
        aTotalHours := Floor(TotalMinutes / 60)
    then:
        set TotalHours = aTotalHours

    return TotalHours
    

MonthsBetweenDates(FirstDate:time_and_date, SecondDate:time_and_date)<transacts>:int=
    FirstDateTotalMonths := FirstDate.Year * 12 + FirstDate.Month
    SecondDateTotalMonths := SecondDate.Year * 12 + SecondDate.Month
    Months := SecondDateTotalMonths - FirstDateTotalMonths

YearsBetweenDates(FirstDate:time_and_date, SecondDate:time_and_date)<transacts>:int=
    Years := SecondDate.Year - FirstDate.Year


AddDayToDateTime(CurrentTime: time_and_date, Days: int)<transacts>: time_and_date =
    var NewDay: int = CurrentTime.Day
    var NewMonth: int = CurrentTime.Month
    var NewYear: int = CurrentTime.Year
    var RemainingDays: int = Days

    loop:
        DaysInCurrentMonth := GetDaysInMonth(NewYear, NewMonth)

        if (RemainingDays + NewDay <= DaysInCurrentMonth):
            set NewDay += RemainingDays
            break
        else:
            set RemainingDays -= (DaysInCurrentMonth - NewDay + 1)
            set NewDay = 1
            set NewMonth += 1

            if (NewMonth > 12):
                set NewMonth = 1
                set NewYear += 1

    return MakeTimeAndDate(NewDay, NewMonth, NewYear, CurrentTime.Hour, CurrentTime.Minute, CurrentTime.Second)


AddMonthToDateTime(CurrentTime:time_and_date, Months:int)<transacts>:time_and_date=
    var NewMonth : int = CurrentTime.Month + Months
    var NewYear : int = CurrentTime.Year

    loop:
        if (NewMonth > 12):
            set NewMonth -= 12
            set NewYear += 1
        else:
            break

    return MakeTimeAndDate(CurrentTime.Day, NewMonth, NewYear, CurrentTime.Hour, CurrentTime.Minute, CurrentTime.Second)

AddYearToDateTime(CurrentTime:time_and_date, Years:int)<transacts>:time_and_date=
    var NewYear : int = CurrentTime.Year + Years
    return MakeTimeAndDate(CurrentTime.Day, CurrentTime.Month, NewYear, CurrentTime.Hour, CurrentTime.Minute, CurrentTime.Second)


AddHourToDateTime(CurrentTime:time_and_date, Hours:int)<transacts>:time_and_date=
    var NewHour : int = CurrentTime.Hour + Hours
    var NewDay : int = CurrentTime.Day
    var NewMonth : int = CurrentTime.Month
    var NewYear : int = CurrentTime.Year

    loop:
        DaysInCurrentMonth := GetDaysInMonth(NewYear, NewMonth)

        if (NewHour >= 24):
            set NewHour -= 24
            set NewDay += 1

            if (NewDay > DaysInCurrentMonth):
                set NewDay = 1
                set NewMonth += 1

                if (NewMonth > 12):
                    set NewMonth = 1
                    set NewYear += 1
        else:
            break

    return MakeTimeAndDate(NewDay, NewMonth, NewYear, NewHour, CurrentTime.Minute, CurrentTime.Second)


AddMinuteToDateTime(CurrentTime:time_and_date, Minutes:int)<transacts>:time_and_date=
    var NewMinute : int = CurrentTime.Minute + Minutes
    var NewHour : int = CurrentTime.Hour
    var NewDay : int = CurrentTime.Day
    var NewMonth : int = CurrentTime.Month
    var NewYear : int = CurrentTime.Year

    loop:
        if (NewMinute >= 60):
            set NewMinute -= 60
            set NewHour += 1

            if (NewHour >= 24):
                set NewHour -= 24
                set NewDay += 1

                if (NewDay > GetDaysInMonth(NewYear, NewMonth)):
                    set NewDay = 1
                    set NewMonth += 1

                    if (NewMonth > 12):
                        set NewMonth = 1
                        set NewYear += 1
        else:
            break

    return MakeTimeAndDate(NewDay, NewMonth, NewYear, NewHour, NewMinute, CurrentTime.Second)


AddSecondToDateTime(CurrentTime:time_and_date, Seconds:int)<transacts>:time_and_date=
    var NewSecond : int = CurrentTime.Second + Seconds
    var NewMinute : int = CurrentTime.Minute
    var NewHour : int = CurrentTime.Hour
    var NewDay : int = CurrentTime.Day
    var NewMonth : int = CurrentTime.Month
    var NewYear : int = CurrentTime.Year

    loop:
        if (NewSecond >= 60):
            set NewSecond -= 60
            set NewMinute += 1

            if (NewMinute >= 60):
                set NewMinute -= 60
                set NewHour += 1

                if (NewHour >= 24):
                    set NewHour -= 24
                    set NewDay += 1

                    if (NewDay > GetDaysInMonth(NewYear, NewMonth)):
                        set NewDay = 1
                        set NewMonth += 1

                        if (NewMonth > 12):
                            set NewMonth = 1
                            set NewYear += 1
        else:
            break

    return MakeTimeAndDate(NewDay, NewMonth, NewYear, NewHour, NewMinute, NewSecond)


AddDates(FirstDate:time_and_date, SecondDate:time_and_date)<transacts>:time_and_date=
    var NewDay : int = FirstDate.Day + SecondDate.Day
    var NewMonth : int = FirstDate.Month + SecondDate.Month
    var NewYear : int = FirstDate.Year + SecondDate.Year

    loop:
        if (NewDay > GetDaysInMonth(NewYear, NewMonth)):
            set NewDay -= GetDaysInMonth(NewYear, NewMonth)
            set NewMonth += 1

            if (NewMonth > 12):
                set NewMonth -= 12
                set NewYear += 1
        else:
            break

    return MakeTimeAndDate(NewDay, NewMonth, NewYear, FirstDate.Hour, FirstDate.Minute, FirstDate.Second)



SetHoursOfDateTime(CurrentTime:time_and_date, Hours:int)<transacts>:time_and_date=
    return MakeTimeAndDate(CurrentTime.Day, CurrentTime.Month, CurrentTime.Year, Hours, CurrentTime.Minute, CurrentTime.Second)

SetMinutesOfDateTime(CurrentTime:time_and_date, Minutes:int)<transacts>:time_and_date=
    return MakeTimeAndDate(CurrentTime.Day, CurrentTime.Month, CurrentTime.Year, CurrentTime.Hour, Minutes, CurrentTime.Second)

SetSecondsOfDateTime(CurrentTime:time_and_date, Seconds:int)<transacts>:time_and_date=
    return MakeTimeAndDate(CurrentTime.Day, CurrentTime.Month, CurrentTime.Year, CurrentTime.Hour, CurrentTime.Minute, Seconds)

(Date:time_and_date).GetMidnight()<transacts>:time_and_date=
    return MakeTimeAndDate(Date.Day, Date.Month, Date.Year, 0, 0, 0)

time_and_date_trigger := class<concrete>:
    var DateTimeToTrigger : time_and_date = time_and_date{}
    var DateTimeToDeactivate : time_and_date = time_and_date{}

    @editable . ToolTip := TitleToolTip
    Title : string = "Default Event Name"
    TitleToolTip<localizes>:message = "Name Your Event"

    @editable . ToolTip := EventDateTimeToolTip
    EventDateTime : date_string = date_string{}
    EventDateTimeToolTip<localizes>:message = "Date and Time of Event"

    @editable . ToolTip := OutputTriggerActivatedToolTip
    OutputTriggerActivated : trigger_device = trigger_device{}
    OutputTriggerActivatedToolTip<localizes>:message = "Trigger when event is activated"

    @editable . ToolTip := ActivateToolTip
    ActivateIfPassed : logic = true
    ActivateToolTip<localizes>:message = "If unchecked, one time event"

    @editable . ToolTip := OutputTriggerDeactivatedToolTip
    OutputTriggerDeactivated : trigger_device = trigger_device{}
    OutputTriggerDeactivatedToolTip<localizes>:message = "Trigger when event is deactivated"

    @editable . ToolTip := DeactivateToolTip
    DeactivateDate : ?date_string = false
    DeactivateToolTip<localizes>:message = "Date and Time to Deactivate Event"


    var ClockProp : creative_prop = creative_prop{}

    Init(TheClockProp:creative_prop)<suspends>:void=

        set ClockProp = TheClockProp
        set DateTimeToTrigger = ParseDateTimes(EventDateTime.Date, EventDateTime.Time)
        DateTimeToTrigger.Print()

        if(HasDeactivateDate := DeactivateDate?):
            set DateTimeToDeactivate = ParseDateTimes(HasDeactivateDate.Date, HasDeactivateDate.Time)
            DateTimeToDeactivate.Print()

        # CurrentTime := MakeTimeAndDate(ClockProp)
        CurrentTime := GetCurrentDateTime()


        if:
            ActivateIfPassed?            
            PassedDateAndTime[CurrentTime, DateTimeToTrigger]
        then:
            if:
                HasDeactivateDate := DeactivateDate?
                PassedDateAndTime[CurrentTime, DateTimeToDeactivate]
            then:
                # do nothing because the event has already passed
                Print("{Title} - Event has already passed")
            else:
                Print("{Title} - Activated, time passed")
                OutputTriggerActivated.Trigger()
        else:
            Print("{Title} - Time and Date has not passed yet")
            AwaitTimeAndDate(DateTimeToTrigger, OutputTriggerActivated)


        if:
            HasDeactivateDate := DeactivateDate?
            not PassedDateAndTime[CurrentTime, DateTimeToDeactivate]
        then:
            Print("{Title} - Awaiting deactivation")
            AwaitTimeAndDate(DateTimeToDeactivate, OutputTriggerDeactivated)
            Print("{Title} - Deactivated, time passed")



    AwaitTimeAndDate(DateTime:time_and_date, Trigger:trigger_device)<suspends>:void=
        loop:
            # CurrentTime := MakeTimeAndDate(ClockProp)
            CurrentTime := GetCurrentDateTime()

            if:
                CurrentTime.Day = DateTime.Day
                CurrentTime.Month = DateTime.Month
                CurrentTime.Year = DateTime.Year
                CurrentTime.Hour = DateTime.Hour
                CurrentTime.Minute = DateTime.Minute
                CurrentTime.Second = DateTime.Second
            then:
                Trigger.Trigger()
                break

            Sleep(1.0)


rtc_device := class<concrete>():
    @editable ClockProp : creative_prop = creative_prop{}
    @editable TestButton : button_device = button_device{}
    @editable DebugBillboard : billboard_device = billboard_device{}

    @editable EventTriggers : []time_and_date_trigger = array{}

    var ButtonPresses : int = 0

    Init()<suspends>:void=

        TestButton.InteractedWithEvent.Subscribe(ReadTimeFromProp)

        for (RTCTrigger : EventTriggers):
            spawn . RTCTrigger.Init(ClockProp)

        GlobalMinuteHeartbeat()

    GlobalMinuteHeartbeat()<suspends>:void=
        Print("Starting Global Heartbeat")
        loop:
            AllPlayers := ClockProp.GetPlayspace().GetPlayers()            
            for(Player:AllPlayers):
                AddMinutePlayed(Player)                
            Sleep(60.0)
            Print("❤")

    ReadTimeFromProp(Agent:agent):void=
        # if(ButtonPresses = 0):
        DebugDateFunction(Agent, ClockProp)

        # else if(ButtonPresses = 1):
        #     if(TomorrowClockProp := DebugClockProps[0]):
        #         DebugDateFunction(Agent, TomorrowClockProp)
        # else if(ButtonPresses = 2):
        #     if(DayAfterTomorrowClockProp := DebugClockProps[1]):
        #         DebugDateFunction(Agent, DayAfterTomorrowClockProp)
        # else if(ButtonPresses = 3):
        #     if(DayAfterDayAfterTomorrowClockProp := DebugClockProps[2]):
        #         DebugDateFunction(Agent, DayAfterDayAfterTomorrowClockProp)
        # else if(ButtonPresses = 4):
        #     if(July3ClockProp := DebugClockProps[3]):
        #         DebugDateFunction (Agent, July3ClockProp)
        # else if(ButtonPresses = 5):
        #     if(July4ClockProp := DebugClockProps[4]):
        #         DebugDateFunction (Agent, July4ClockProp)



    DebugDateFunction(Agent:agent, DateProp:creative_prop):void=
        # CurrentTimeAndDate := MakeTimeAndDate(DateProp)
        CurrentTimeAndDate := GetCurrentDateTime()
        CurrentTimeAndDate.Print()

        if:
            CheckForPlayerSavedData[Agent]
            Player := player[Agent]
            SourceData := PlayerSavedDataMap[Player]
            FirstDayPlayed := GetFirstDayPlayed[Agent]
            TheGlobals := Globals[GetSession()]
            CPlayer := TheGlobals.CustomPlayerMap[Player]
            Opponent := CPlayer.GetOpponent()?
        then:
            Print("Player has saved data")
            AddTotalDayPlayed(Agent)
            CheckForConsecutiveDaysPlayed(Agent, CurrentTimeAndDate)
            SetLastDayPlayed(Agent, CurrentTimeAndDate)

            if(CompareEqualDateIgnoreTime[FirstDayPlayed,CurrentTimeAndDate]):
                Print("First day played is today")

            if:
                TotalDays := GetTotalDaysPlayed[Agent]
                TotalMinutes := GetTotalMinutesPlayed[Agent]
                LastDayPlayed := GetLastDayPlayed[Agent]
                FriendStreaks := SourceData.FriendStreaks
                ExistingStreak := CheckForExistingStreak(Agent, Opponent)?
            then:
                MostConsecutiveDaysPlayed := GetMostConsecutiveDaysPlayed(Agent)
                CurrentConsecutiveDaysPlayed := GetCurrentConsecutiveDaysPlayed(Agent)
                StreakWins := GetWinsFromStreakData(ExistingStreak)

                Print("Total Days Played: {TotalDays} - ")
                Print("Total")
                DebugBillboard.SetText(EightLineMessage(
                    "Total Days: {TotalDays} Minutes: {TotalMinutes} First: {FirstDayPlayed.String()}", 
                    "Last: {LastDayPlayed.String()} ^Consecutive: {MostConsecutiveDaysPlayed} - Current: {CurrentConsecutiveDaysPlayed}",
                    "Correct Guesses: {SourceData.CorrectGuesses} - Consecutive: {SourceData.ConsecutiveCorrectGuesses} Record: {SourceData.ConsecutiveCorrectGuessesRecord}",
                    "GamesPlayed: {SourceData.GamesPlayed} - GamesPlayedWithNewPlayers: {SourceData.GamesPlayedWithNewFortWhoPlayers}",
                    "FriendStreaks: {FriendStreaks.Length} - CurrentStreak: {ExistingStreak.CurrentStreak}",
                    "ExistingStreakID: {ExistingStreak.ID} - ExistingStreakLength: {ExistingStreak.CurrentStreak}",
                    "StreakTimesPlayed: {ExistingStreak.TimesPlayed} - BestStreak: {ExistingStreak.HighestStreak}",
                    "P1Wins: {StreakWins(0)} - P2Wins: {StreakWins(1)}"
                ))

        else:
            Print("Player has no saved data, creating new data")
    
        set ButtonPresses += 1
