using { /Fortnite.com/UI }
using { /UnrealEngine.com/Temporary/UI }
using { /UnrealEngine.com/Temporary/SpatialMath }
using { /Verse.org/Assets }
using { /Verse.org/Colors/NamedColors }

# Selection Screen Option

SelectText<localizes>:message = "select"

selection_screen_option := class:

    var StyleIndex : int = 0

    var Widget : overlay = overlay:

    StyleImage : texture_block = texture_block:
        DefaultImage := Cards.Textures.UI.verse_ui_coming_soon
        DefaultDesiredSize := vector2{X:=140.0, Y:=140.0}

    var StyleTexture : texture = Cards.Textures.UI.verse_ui_coming_soon

    NewIconTexture : texture = _UI.UI_Textures.New_Badge_EN
    UpdatedIconTexture : texture = _UI.UI_Textures.Updated_Badge_EN
    ClearIconTexture : texture = _UI.UI_Textures.transparent_square

    NewTextureBlock : texture_block = texture_block:
        DefaultImage := _UI.UI_Textures.transparent_square
        DefaultDesiredSize := vector2{X:=360.0/3.0, Y:=90.0/3.0}

    Button : button_quiet = button_quiet:
        DefaultText := MessageSelectSUP

    var TitleText : text_block = text_block:
        DefaultText := MessageBlank_EN
        DefaultTextColor := White
        DefaultJustification := text_justification.Center
        DefaultShadowColor := Black
        DefaultShadowOffset := option{vector2{X:=4.0, Y:=4.0}}
        DefaultShadowOpacity := 0.75


    Create(StyleData:style_selection_data):void=
        set StyleIndex = StyleData.StyleIndex
        set StyleTexture = StyleData.StyleTexture
        StyleImage.SetImage(StyleTexture)

        TitleText.SetText(StyleData.StyleName)

        if(StyleData.IsNew?):
            NewTextureBlock.SetImage(NewIconTexture)
        else if(StyleData.IsUpdated?):
            NewTextureBlock.SetImage(UpdatedIconTexture)
        else if(StyleData.IsComingSoon?):
            NewTextureBlock.SetImage(ClearIconTexture)
        else:
            NewTextureBlock.SetImage(ClearIconTexture)

        set Widget = overlay:
            Slots := array:
                overlay_slot:
                    HorizontalAlignment := horizontal_alignment.Fill
                    VerticalAlignment := vertical_alignment.Fill
                    Padding := margin{Top := 10.0, Left := 10.0, Right := 10.0, Bottom := 10.0}
                    Widget := stack_box:
                        Orientation := orientation.Vertical
                        Slots := array:
                            stack_box_slot:
                                Widget := overlay:
                                    Slots := array:
                                        overlay_slot:
                                            HorizontalAlignment := horizontal_alignment.Center
                                            VerticalAlignment := vertical_alignment.Center
                                            Padding := margin{Top := 4.0, Left := 4.0, Right := 4.0, Bottom := 4.0}
                                            Widget := StyleImage
                                        overlay_slot:
                                            HorizontalAlignment := horizontal_alignment.Left
                                            VerticalAlignment := vertical_alignment.Top
                                            Padding := margin{Top := 7.0, Left := 7.0, Right := 0.0, Bottom := 0.0}
                                            Widget := NewTextureBlock
                                        overlay_slot:
                                            HorizontalAlignment := horizontal_alignment.Center
                                            VerticalAlignment := vertical_alignment.Bottom
                                            Padding := margin{Top := 0.0, Left := 0.0, Right := 0.0, Bottom := 4.0}
                                            Widget := TitleText
                            stack_box_slot:
                                HorizontalAlignment := horizontal_alignment.Fill
                                VerticalAlignment := vertical_alignment.Bottom
                                Padding := margin{Top := 5.0, Left := 0.0, Right := 0.0, Bottom := 0.0}
                                Widget := stack_box:
                                    Orientation := orientation.Vertical
                                    Slots := array:
                                        # stack_box_slot:
                                        #     Distribution := option{1.0}
                                        #     HorizontalAlignment := horizontal_alignment.Center
                                        #     VerticalAlignment := vertical_alignment.Fill
                                        #     Padding := margin{Top := 0.0, Left := 0.0, Right := 0.0, Bottom := 4.0}
                                        #     Widget := TitleText
                                        stack_box_slot:
                                            Distribution := option{1.0}
                                            HorizontalAlignment := horizontal_alignment.Fill
                                            VerticalAlignment := vertical_alignment.Fill
                                            Padding := margin{Top := 0.0, Left := 0.0, Right := 0.0, Bottom := 0.0}
                                            Widget := Button
                                        stack_box_slot:
                                            Distribution := option{1.75}
                                            HorizontalAlignment := horizontal_alignment.Fill
                                            VerticalAlignment := vertical_alignment.Fill
                                            Padding := margin{Top := 0.0, Left := 0.0, Right := 0.0, Bottom := 0.0}
                                            Widget := color_block:
                                                DefaultColor := Red
                                                DefaultOpacity := 0.0
                                                DefaultDesiredSize := vector2{X:=10.0, Y:=10.0}


                                

        

        # if(StyleData.IsComingSoon?):
        #     Button.SetEnabled(false)
        #     Button.SetText(StringToMessage("SOON!!!"))