using { /Verse.org/Assets }

style_selection_data := class():
    StyleIndex: int
    StyleTexture: texture
    StyleName: message
    CreationDate: string
    LastUpdatedDate: string
    SortOrder: int
    IsPinned: logic = false
    IsNew: logic = false
    IsUpdated: logic = false
    IsHidden: logic = false
    IsBeta: logic = false
    IsUnlockable: logic = false
    IsComingSoon: logic = false
    IsAd : logic = false
    AdStartDate: string
    AdEndDate: string
    