using { /Fortnite.com/Devices }
using { /UnrealEngine.com/Temporary/SpatialMath }
using { /Verse.org }
using { /Verse.org/Random }
using { /Verse.org/Simulation }
using { /UnrealEngine.com/Assets }

target_area := class<concrete>():

    @editable ZoneBounds : []creative_prop = array{}
    @editable Trigger : trigger_device = trigger_device{}
    @editable TargetAsset : creative_prop_asset = DefaultCreativePropAsset
    @editable DestoryedAsset : creative_prop_asset = DefaultCreativePropAsset

    @editable AccoladeHit : accolades_device = accolades_device{}

    @editable TIME_BETWEEN_SPAWNS : float = 5.0
    @editable PlayerCounter : player_counter_device = player_counter_device{}

    var ActualTimeBetweenSpawns : float = 0.0

    Init(Platform:creative_prop)<suspends>:void=
        Print("Init target_area")
        set ActualTimeBetweenSpawns = TIME_BETWEEN_SPAWNS
        loop:
            if:
                NewSpawnPoint := GetRandomPoint[]
                Rotation := LookAtLocked[NewSpawnPoint, Platform.GetTransform().Translation, vector3{Z:=1.0}]
                SpawnedPropResult := SpawnProp(TargetAsset, NewSpawnPoint, Rotation)
                TheProp := SpawnedPropResult(0)?
                Trigger.TeleportTo[NewSpawnPoint, Rotation]
            then:
                Trigger.Enable()
                MaybeAgent := Trigger.TriggeredEvent.Await()
                TheProp.SetMaterial(FlyingTargets.Materials.MI_FlyingTarget_Hit, ?Index:=0)
                Trigger.Disable()
                Sleep(0.1)
                TheProp.Dispose()
                # continuesly spawns particles...so we spawn a prop instead with the vfx inside
                # DestroyVFX := SpawnParticleSystem(FlyingTargets.VFX.NS_Flying_Target_Destroyed, NewSpawnPoint, Rotation) 
                DestroyedPropResult := SpawnProp(DestoryedAsset, NewSpawnPoint, Rotation)
                if:
                    DestroyedProp := DestroyedPropResult(0)?
                then:
                    spawn . DestroyedProp.DelayedDispose(8.0)

                if(Agent := MaybeAgent?):
                    AddFlyingTargetHit(Agent)
                    AccoladeHit.Award(Agent)

                    if(TotalHits := GetTotalFlyingTargetsHit[Agent]):
                        Print("Total Flying Targets Hit: {TotalHits}")

                        if:
                            TotalHits >= 10000
                            TheGlobals := Globals[GetSession()]
                            Player := player[Agent]
                            CPlayer := TheGlobals.CustomPlayerMap[Player]
                            StickerID := TheGlobals.GetStickerID["Targets10000"]
                            not AlreadyUnlockedSticker[Agent, StickerID]
                            AvailStickerSlot := CPlayer.GameBoard.GetAvailableStickerSlot[]
                        then:
                            NewStickerResult := SpawnProp(TheGlobals.DefaultStickerPropAsset, vector3{}, rotation{})
                            Targets10000Material := Materials.Stickers.MI_Sticker_Base_Inst_material{}
                            set Targets10000Material.Sticker_Texture = Materials.Stickers.targets_10000
                            if:
                                TheSticker := NewStickerResult(0)?
                            then:
                                TheSticker.SetMaterial(Targets10000Material)
                                spawn . CPlayer.GameBoard.AddSticker(Agent, CPlayer, AvailStickerSlot, StickerID, TheSticker)

                            # Unlock the sticker
                            UnlockSticker(Agent, StickerID)
                            SaveCurrentSticker(Agent, StickerID, AvailStickerSlot.GetTransform())
                        else:
                            Print("Have not yet unlocked the 10000 targets sticker, only {TotalHits} hits so far")


                if(PlayerCounter.GetCount() = 2):
                    set ActualTimeBetweenSpawns = 0.03
                else:
                    set ActualTimeBetweenSpawns = TIME_BETWEEN_SPAWNS

            Sleep(ActualTimeBetweenSpawns)

    GetRandomPoint()<transacts><decides>:vector3=
        CoinFlip := GetRandomInt(0, 1)
        Zone := ZoneBounds[CoinFlip]
        BoundingBox := GetExtentsOfBoundingProp(Zone)
        aRandomPoint := BoundingBox.GetRandomPointWithin()
      
