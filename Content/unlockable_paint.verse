using { /Verse.org/Simulation }
using { /Fortnite.com/Devices }
using { /Verse.org/Assets }
using { /Verse.org/Colors }
using { /Fortnite.com/UI }
using { /UnrealEngine.com/Temporary/UI }
using { /UnrealEngine.com/Temporary/SpatialMath }
using { /Verse.org/Colors/NamedColors }

board_texture := enum{Plain, Clay}

unlockable_paint := class<concrete>():

    @editable Name : string = ""
    @editable Cost : int = 50
    @editable Switch : switch_device = switch_device{}
    @editable BoardTexture : board_texture = board_texture.Plain
    @editable HudMessage : hud_message_device = hud_message_device{}
    @editable Accolade : accolades_device = accolades_device{}

    IsUnlocked(Agent:agent)<transacts><decides>:void=
        Switch.GetCurrentState[Agent]

    Unlock(Agent:agent):void=
        Switch.TurnOn(Agent)
        # HudMessage.Show(Agent)
        Accolade.Award(Agent)


    GetTextures():tuple(texture, texture)=
        case(BoardTexture):
            board_texture.Plain => (_UI.UI_Textures.paint_palette_board, _UI.UI_Textures.paint_palette_cards)
            board_texture.Clay => (_UI.UI_Textures.paint_palette_board_clay, _UI.UI_Textures.paint_palette_cards_clay)