using { /Fortnite.com/Characters }
using { /Fortnite.com/Devices }
using { /Fortnite.com/Devices/CreativeAnimation }
using { /Fortnite.com/Teams }
using { /UnrealEngine.com/Temporary }
using { /UnrealEngine.com/Temporary/Diagnostics }
using { /UnrealEngine.com/Temporary/SpatialMath }
using { /UnrealEngine.com/Temporary/UI }
using { /Verse.org/Assets }
using { /Verse.org/Colors }
using { /Verse.org/Colors/NamedColors }
using { /Verse.org/Random }
using { /Verse.org/Simulation }
using { /Verse.org/Simulation/Tags }

log_device:=class(log_channel){}
debug_drawing := class(debug_draw_channel){}

spawnpad := class(tag){}
ballon_landing_spot := class(tag){}
light_orb := class(tag){}
clock_prop := class(tag){}



StringToMessage<localizes>(A812A22:string) : message = "{A812A22}"
AgentToMessage<localizes>(A812A22:agent):message = "{A812A22}"
IntToMessage<localizes>(A812A22:int) : message = "{A812A22}"
TwoLineMessage<localizes>(A812A221:string, A812A222:string):message = "{A812A221}\n{A812A222}"
ThreeLineMessage<localizes>(A812A221:string, A812A222:string, A812A2223:string):message = "{A812A221}\n{A812A222}\n{A812A2223}"
FiveLineMessage<localizes>(A812A221:string, A812A222:string, A812A2223:string, A812A22234:string, A812A222345:string):message = "{A812A221}\n{A812A222}\n{A812A2223}\n{A812A22234}\n{A812A222345}"
SixLineMessage<localizes>(A812A221:string, A812A222:string, A812A2223:string, A812A22234:string, A812A222345:string, A812A222346:string):message = "{A812A221}\n{A812A222}\n{A812A2223}\n{A812A22234}\n{A812A222345}\n{A812A222346}"
EightLineMessage<localizes>(A812A221:string, A812A222:string, A812A2223:string, A812A22234:string, A812A222345:string, A812A222346:string, A812A222347:string, A812A222348:string):message = "{A812A221}\n{A812A222}\n{A812A2223}\n{A812A22234}\n{A812A222345}\n{A812A222346}\n{A812A222347}\n{A812A222348}"
NineLineMessage<localizes>(A812A221:string, A812A222:string, A812A2223:string, A812A22234:string, A812A222345:string, A812A222346:string, A812A222347:string, A812A222348:string, A812A222349:string):message = "{A812A221}\n{A812A222}\n{A812A2223}\n{A812A22234}\n{A812A222345}\n{A812A222346}\n{A812A222347}\n{A812A222348}\n{A812A222349}"
TenLineMessage<localizes>(A812A221:string, A812A222:string, A812A2223:string, A812A22234:string, A812A222345:string, A812A222346:string, A812A222347:string, A812A222348:string, A812A222349:string, A812A2223410:string):message = "{A812A221}\n{A812A222}\n{A812A2223}\n{A812A22234}\n{A812A222345}\n{A812A222346}\n{A812A222347}\n{A812A222348}\n{A812A222349}\n{A812A2223410}"
MakeGuesserMessage<localizes>(A812A22:player):message = "{A812A22} Guessed,"

MakeHrsMinsMessage<localizes>(A812A221:int, A812A222:int):message = "{A812A221}hrs {A812A222}mins"

MakePaintCostMessage<localizes>(A812A221:message, A812A222:int):message = "{A812A221} ({A812A222})"
MakeBuyPaintButtonMessage<localizes>(A812A22:int):message = "BUY ({A812A22})"


card_state := enum{Down, Up, Animating}

UpAngle : float = 120.0 # X axis
FlipSpeed : float = 0.25

TileDimensions : vector3 = vector3{X:=175.0, Y:=288.0, Z:=16.0}
TilePivotOffset : vector3 = vector3{X:=0.0, Y:=0.0, Z:=64.0}

# NEEDS TO BE UPDATED WHEN NEW STYLES (SETS) ARE ADDED ######################################################################################################
MakeMaterialForCardStyle(CardStyle:card_style):material=
    NewMI := Cards.Materials.CardStyles.MI_Card_Master_NEW_material{}
    case(CardStyle):
        card_style.FN => block:
            set NewMI.CardImageGrid = GetCardGridTextureForCardStyle(card_style.FN)
            return NewMI
        card_style.Classic => block:
            set NewMI.CardImageGrid = GetCardGridTextureForCardStyle(card_style.Classic)
            return NewMI
        card_style.Icon => block:
            set NewMI.CardImageGrid = GetCardGridTextureForCardStyle(card_style.Icon)
            return NewMI
        card_style.Animal => block:
            set NewMI.CardImageGrid = GetCardGridTextureForCardStyle(card_style.Animal)
            return NewMI
        card_style.Utilities => block:
            set NewMI.CardImageGrid = GetCardGridTextureForCardStyle(card_style.Utilities)
            return NewMI
        card_style.Marvel => block:
            set NewMI.CardImageGrid = GetCardGridTextureForCardStyle(card_style.Marvel)
            return NewMI
        card_style.POI => block:
            set NewMI.CardImageGrid = GetCardGridTextureForCardStyle(card_style.POI)
            return NewMI
        card_style.Legends => block:
            set NewMI.CardImageGrid = GetCardGridTextureForCardStyle(card_style.Legends)
            return NewMI
        card_style.Food => block:
            set NewMI.CardImageGrid = GetCardGridTextureForCardStyle(card_style.Food)
            return NewMI
        card_style.Star => block:
            set NewMI.CardImageGrid = GetCardGridTextureForCardStyle(card_style.Star)
            return NewMI
        card_style.Spooky => block:
            set NewMI.CardImageGrid = GetCardGridTextureForCardStyle(card_style.Spooky)
            return NewMI
        card_style.OG => block:
            set NewMI.CardImageGrid = GetCardGridTextureForCardStyle(card_style.OG)
            return NewMI
        card_style.Anime => block:
            set NewMI.CardImageGrid = GetCardGridTextureForCardStyle(card_style.Anime)
            return NewMI
        card_style.Winter => block:
            set NewMI.CardImageGrid = GetCardGridTextureForCardStyle(card_style.Winter)
            return NewMI
        card_style.Weapons => block:
            set NewMI.CardImageGrid = GetCardGridTextureForCardStyle(card_style.Weapons)
            return NewMI
        card_style.Pickaxes => block:
            set NewMI.CardImageGrid = GetCardGridTextureForCardStyle(card_style.Pickaxes)
            return NewMI
        card_style.Vehicles => block:
            set NewMI.CardImageGrid = GetCardGridTextureForCardStyle(card_style.Vehicles)
            return NewMI
        card_style.MoviesTV => block:
            set NewMI.CardImageGrid = GetCardGridTextureForCardStyle(card_style.MoviesTV)
            return NewMI
        card_style.Blocks => block:
            set NewMI.CardImageGrid = GetCardGridTextureForCardStyle(card_style.Blocks)
            return NewMI
        card_style.Chapter5 => block:
            set NewMI.CardImageGrid = GetCardGridTextureForCardStyle(card_style.Chapter5)
            return NewMI
        card_style.Lockers => block:
            set NewMI.CardImageGrid = GetCardGridTextureForCardStyle(card_style.Lockers)
            return NewMI
        card_style.Backblings => block:
            set NewMI.CardImageGrid = GetCardGridTextureForCardStyle(card_style.Backblings)
            return NewMI
        card_style.Emotes => block:
            set NewMI.CardImageGrid = GetCardGridTextureForCardStyle(card_style.Emotes)
            return NewMI
        card_style.Marvel1 => block:
            set NewMI.CardImageGrid = GetCardGridTextureForCardStyle(card_style.Marvel1)
            return NewMI
        card_style.Marvel2 => block:
            set NewMI.CardImageGrid = GetCardGridTextureForCardStyle(card_style.Marvel2)
            return NewMI
        card_style.Umbrellas => block:
            set NewMI.CardImageGrid = GetCardGridTextureForCardStyle(card_style.Umbrellas)
            return NewMI
        card_style.PeelyFishstick => block:
            set NewMI.CardImageGrid = GetCardGridTextureForCardStyle(card_style.PeelyFishstick)
            return NewMI
        card_style.Remix => block:
            set NewMI.CardImageGrid = GetCardGridTextureForCardStyle(card_style.Remix)
            return NewMI
        card_style.FireIce => block:
            set NewMI.CardImageGrid = GetCardGridTextureForCardStyle(card_style.FireIce)
            return NewMI
        card_style.DC => block:
            set NewMI.CardImageGrid = GetCardGridTextureForCardStyle(card_style.DC)
            return NewMI
        card_style.Chapter6 => block:
            set NewMI.CardImageGrid = GetCardGridTextureForCardStyle(card_style.Chapter6)
            return NewMI


# GetMaterialForCardStyle(CardStyle:card_style):material=
#     case(CardStyle):
#         card_style.FN => return Cards.Materials.CardStyles.MI_Card_Style_New_FN
#         card_style.Classic => return Cards.Materials.CardStyles.MI_Card_Style_New_Classic
#         card_style.Icon => return Cards.Materials.CardStyles.MI_Card_Style_New_Icon
#         card_style.Animal => return Cards.Materials.CardStyles.MI_Card_Style_New_Animal
#         card_style.Utilities => return Cards.Materials.CardStyles.MI_Card_Style_New_Utilities
#         card_style.Marvel => return Cards.Materials.CardStyles.MI_Card_Style_New_Marvel
#         card_style.POI => return Cards.Materials.CardStyles.MI_Card_Style_New_POI
#         card_style.Legends => return Cards.Materials.CardStyles.MI_Card_Style_New_Legends
#         card_style.Food => return Cards.Materials.CardStyles.MI_Card_Style_New_Food
#         card_style.Star => return Cards.Materials.CardStyles.MI_Card_Style_New_Star
#         card_style.Spooky => return Cards.Materials.CardStyles.MI_Card_Style_New_Spooky
#         card_style.OG => return Cards.Materials.CardStyles.MI_Card_Style_New_OG
#         card_style.Anime => return Cards.Materials.CardStyles.MI_Card_Style_New_Anime
#         card_style.Winter => return Cards.Materials.CardStyles.MI_Card_Style_New_Winter
#         card_style.Weapons => return Cards.Materials.CardStyles.MI_Card_Style_New_Weapons
#         card_style.Pickaxes => return Cards.Materials.CardStyles.MI_Card_Style_New_Pickaxes
#         card_style.Vehicles => return Cards.Materials.CardStyles.MI_Card_Style_New_Vehicles
#         card_style.MoviesTV => return Cards.Materials.CardStyles.MI_Card_Style_New_Movies_TV
#         card_style.Blocks => return Cards.Materials.CardStyles.MI_Card_Style_New_FN
#         card_style.Chapter5 => return Cards.Materials.CardStyles.MI_Card_Style_New_Chapter5
#         card_style.Lockers => return Cards.Materials.CardStyles.MI_Card_Style_New_Lockers
#         card_style.Backblings => return Cards.Materials.CardStyles.MI_Card_Style_New_Backblings
#         card_style.Emotes => return Cards.Materials.CardStyles.MI_Card_Style_New_Emotes
#         card_style.Marvel1 => return Cards.Materials.CardStyles.MI_Card_Style_New_Marvel_Set_1
#         card_style.Marvel2 => return Cards.Materials.CardStyles.MI_Card_Style_New_Marvel_Set_2
#         card_style.Umbrellas => return Cards.Materials.CardStyles.MI_Card_Style_New_Umbrellas
#         card_style.PeelyFishstick => return Cards.Materials.CardStyles.MI_Card_Style_New_PeelyFishstick
#         card_style.Remix => return Cards.Materials.CardStyles.MI_Card_Style_New_Remix
#         card_style.FireIce => return Cards.Materials.CardStyles.MI_Card_Style_New_Fire_Ice
#         card_style.DC => return Cards.Materials.CardStyles.MI_Card_Style_New_DC

# NEEDS TO BE UPDATED WHEN NEW STYLES (SETS) ARE ADDED ######################################################################################################
GetCardGridTextureForCardStyle(CardStyle:card_style):texture=
    case(CardStyle):
        card_style.FN => return Cards.Textures.CardGrids.fn_card_grid
        card_style.Classic => return Cards.Textures.CardGrids.classic_card_grid
        card_style.Icon => return Cards.Textures.CardGrids.icon_card_grid
        card_style.Animal => return Cards.Textures.CardGrids.animal_card_grid
        card_style.Utilities => return Cards.Textures.CardGrids.util_card_grid
        card_style.Marvel => return Cards.Textures.CardGrids.marvel_card_grid
        card_style.POI => return Cards.Textures.CardGrids.poi_card_grid
        card_style.Legends => return Cards.Textures.CardGrids.legends_card_grid
        card_style.Food => return Cards.Textures.CardGrids.food_card_grid
        card_style.Star => return Cards.Textures.CardGrids.star_card_grid
        card_style.Spooky => return Cards.Textures.CardGrids.spooky_card_grid
        card_style.OG => return Cards.Textures.CardGrids.og_card_grid
        card_style.Anime => return Cards.Textures.CardGrids.anime_card_grid
        card_style.Winter => return Cards.Textures.CardGrids.winter_card_grid
        card_style.Weapons => return Cards.Textures.CardGrids.weapons_card_grid
        card_style.Pickaxes => return Cards.Textures.CardGrids.pickaxes_card_grid
        card_style.Vehicles => return Cards.Textures.CardGrids.vehicles_card_grid
        card_style.MoviesTV => return Cards.Textures.CardGrids.movies_tv_card_grid
        card_style.Blocks => return Cards.Textures.CardGrids.fn_card_grid
        card_style.Chapter5 => return Cards.Textures.CardGrids.chapter5_card_grid
        card_style.Lockers => return Cards.Textures.CardGrids.lockers_card_grid
        card_style.Backblings => return Cards.Textures.CardGrids.uefn_card_grid_backblings
        card_style.Emotes => return Cards.Textures.CardGrids.emotes_card_grid
        card_style.Marvel1 => return Cards.Textures.CardGrids.marvel_set_1_card_grid
        card_style.Marvel2 => return Cards.Textures.CardGrids.marvel_set_2_card_grid
        card_style.Umbrellas => return Cards.Textures.CardGrids.umbrellas_card_grid
        card_style.PeelyFishstick => return Cards.Textures.CardGrids.peely_fishstick_card_grid
        card_style.Remix => return Cards.Textures.CardGrids.remix_card_grid
        card_style.FireIce => return Cards.Textures.CardGrids.fire_ice_card_grid
        card_style.DC => return Cards.Textures.CardGrids.dc_card_grid
        card_style.Chapter6 => return Cards.Textures.CardGrids.chapter6_card_grid


# NEEDS TO BE UPDATED WHEN NEW STYLES (SETS) ARE ADDED ######################################################################################################
GetCardStyleForIndex(Index:int):card_style=
    case(Index):
        0 => return card_style.FN
        1 => return card_style.Classic
        2 => return card_style.Icon
        3 => return card_style.Animal
        4 => return card_style.Utilities
        5 => return card_style.Marvel
        6 => return card_style.POI
        7 => return card_style.Legends
        8 => return card_style.Food
        9 => return card_style.Star
        10 => return card_style.Spooky
        11 => return card_style.OG
        12 => return card_style.Anime
        13 => return card_style.Winter
        14 => return card_style.Weapons
        15 => return card_style.Pickaxes
        16 => return card_style.Vehicles
        17 => return card_style.MoviesTV
        18 => return card_style.Blocks
        19 => return card_style.Chapter5
        20 => return card_style.Lockers
        21 => return card_style.Backblings
        22 => return card_style.Emotes
        23 => return card_style.Marvel1
        24 => return card_style.Marvel2
        25 => return card_style.Umbrellas
        26 => return card_style.PeelyFishstick
        27 => return card_style.Remix
        28 => return card_style.FireIce
        29 => return card_style.DC
        30 => return card_style.Chapter6
        _ => return card_style.FN


# NEEDS TO BE UPDATED WHEN NEW STYLES (SETS) ARE ADDED ######################################################################################################
GetIndexForCardStyle(CS:card_style):int=
    case (CS):
        card_style.FN => return 0
        card_style.Classic => return 1
        card_style.Icon => return 2
        card_style.Animal => return 3
        card_style.Utilities => return 4
        card_style.Marvel => return 5
        card_style.POI => return 6
        card_style.Legends => return 7
        card_style.Food => return 8
        card_style.Star => return 9
        card_style.Spooky => return 10
        card_style.OG => return 11
        card_style.Anime => return 12
        card_style.Winter => return 13
        card_style.Weapons => return 14
        card_style.Pickaxes => return 15
        card_style.Vehicles => return 16
        card_style.MoviesTV => return 17
        card_style.Blocks => return 18
        card_style.Chapter5 => return 19
        card_style.Lockers => return 20
        card_style.Backblings => return 21
        card_style.Emotes => return 22
        card_style.Marvel1 => return 23
        card_style.Marvel2 => return 24
        card_style.Umbrellas => return 25
        card_style.PeelyFishstick => return 26
        card_style.Remix => return 27
        card_style.FireIce => return 28
        card_style.DC => return 29
        card_style.Chapter6 => return 30

GetIndexCardMaterialForStyleIndex(StyleIndex:int):material=
    case (StyleIndex):
        4 => return Cards.Materials.IndexCards.MI_IndexCard_Util
        8 => return Cards.Materials.IndexCards.MI_IndexCard_Food
        16 => return Cards.Materials.IndexCards.MI_IndexCard_Vehicles
        _ => return Cards.Materials.IndexCards.MI_IndexCard_Util


GetCardMeshForIndex(CardIndex:int):mesh=
    case (CardIndex):
        0 => return NEW_ASSETS.Meshess.Cards.SM_Card_1
        1 => return NEW_ASSETS.Meshess.Cards.SM_Card_2
        2 => return NEW_ASSETS.Meshess.Cards.SM_Card_3
        3 => return NEW_ASSETS.Meshess.Cards.SM_Card_4
        4 => return NEW_ASSETS.Meshess.Cards.SM_Card_5
        5 => return NEW_ASSETS.Meshess.Cards.SM_Card_6
        6 => return NEW_ASSETS.Meshess.Cards.SM_Card_7
        7 => return NEW_ASSETS.Meshess.Cards.SM_Card_8
        8 => return NEW_ASSETS.Meshess.Cards.SM_Card_9
        9 => return NEW_ASSETS.Meshess.Cards.SM_Card_10
        10 => return NEW_ASSETS.Meshess.Cards.SM_Card_11
        11 => return NEW_ASSETS.Meshess.Cards.SM_Card_12
        12 => return NEW_ASSETS.Meshess.Cards.SM_Card_13
        13 => return NEW_ASSETS.Meshess.Cards.SM_Card_14
        14 => return NEW_ASSETS.Meshess.Cards.SM_Card_15
        15 => return NEW_ASSETS.Meshess.Cards.SM_Card_16
        16 => return NEW_ASSETS.Meshess.Cards.SM_Card_17
        17 => return NEW_ASSETS.Meshess.Cards.SM_Card_18
        18 => return NEW_ASSETS.Meshess.Cards.SM_Card_19
        19 => return NEW_ASSETS.Meshess.Cards.SM_Card_20
        20 => return NEW_ASSETS.Meshess.Cards.SM_Card_21
        21 => return NEW_ASSETS.Meshess.Cards.SM_Card_22
        22 => return NEW_ASSETS.Meshess.Cards.SM_Card_23
        23 => return NEW_ASSETS.Meshess.Cards.SM_Card_24
        _ => return NEW_ASSETS.Meshess.Cards.SM_Card_28 # No Cheating Card

# NEEDS TO BE UPDATED WHEN NEW STYLES (SETS) ARE ADDED ######################################################################################################
(CS:card_style).ToString():string=
    case (CS):
        card_style.FN => return "FN"
        card_style.Classic => return "Classic"
        card_style.Icon => return "Icon"
        card_style.Animal => return "Animal"
        card_style.Utilities => return "Utilities"
        card_style.Marvel => return "Marvel"
        card_style.POI => return "POI"
        card_style.Legends => return "Legends"
        card_style.Food => return "Food"
        card_style.Star => return "Star"
        card_style.Spooky => return "Spooky"
        card_style.OG => return "OG"
        card_style.Anime => return "Anime"        
        card_style.Winter => return "Winter"
        card_style.Weapons => return "Weapons"        
        card_style.Pickaxes => return "Pickaxes"
        card_style.Vehicles => return "Vehicles"
        card_style.MoviesTV => return "Movies/TV"
        card_style.Blocks => return "Blocks"
        card_style.Chapter5 => return "Chapter 5"
        card_style.Lockers => return "Lockers"
        card_style.Backblings => return "Backblings"
        card_style.Emotes => return "Emotes"
        card_style.Marvel1 => return "Heroes [1]"
        card_style.Marvel2 => return "Heroes [2]"
        card_style.Umbrellas => return "Umbrellas"
        card_style.PeelyFishstick => return "Peely Fishstick"
        card_style.Remix => return "Remix"
        card_style.FireIce => return "Fire & Ice"
        card_style.DC => return "Heroes [3]"
        card_style.Chapter6 => return "Chapter 6"


GetStickerMaterial(StickerID:int)<transacts>:material=
    case(StickerID):
        0 => block:
            Winter2024StickerMaterial := Materials.Stickers.MI_Sticker_Base_Inst_material{}
            set Winter2024StickerMaterial.Sticker_Texture = Materials.Stickers.winter_2024
            set Winter2024StickerMaterial.UseSnowBase = 1.0
            return Winter2024StickerMaterial
        1 => block:
            Targets10000Material := Materials.Stickers.MI_Sticker_Base_Inst_material{}
            set Targets10000Material.Sticker_Texture = Materials.Stickers.targets_10000
            return Targets10000Material
        _ => return Materials.Stickers.MI_Sticker_Base_Inst_material{}


(Overlay:overlay).ClearSlots():void=
    for:
        Slot : Overlay.Slots
    do:
        Overlay.RemoveWidget(Slot.Widget)
        Print("Removed Widget from Overlay")


RayCastHitButton(HitLocation:transform, Button:creative_prop)<transacts><decides>:void=

    if:
        Button.IsValid[]
    then:
        var Height : float = 80.0 * Button.GetTransform().Scale.Z
        var Width : float = 80.0 * Button.GetTransform().Scale.X
        var Depth : float = 80.0 * Button.GetTransform().Scale.Y

        var TargetTransform : transform = Button.GetTransform()
        var TargetLocation : vector3 = TargetTransform.Translation
        var TargetRotation : rotation = TargetTransform.Rotation
        var ForwardVector : vector3 = TargetRotation.GetLocalForward()
        var RightVector : vector3 = TargetRotation.GetLocalRight()
        var UpVector : vector3 = TargetRotation.GetLocalUp()

        var BoxCenter : vector3 = TargetLocation + UpVector * Height / 2.0
        var BoxHalfHeight : float = Height / 2.0
        var BoxHalfWidth : float = Width / 2.0
        var BoxHalfDepth : float = Depth / 2.0

        var BoxHalfWidthVector : vector3 = RightVector * BoxHalfWidth
        var BoxHalfDepthVector : vector3 = ForwardVector * BoxHalfDepth

        if:
            DotProduct(HitLocation.Translation - BoxCenter, UpVector) < BoxHalfHeight
            DotProduct(HitLocation.Translation - BoxCenter, UpVector) > -BoxHalfHeight
            DotProduct(HitLocation.Translation - BoxCenter, RightVector) < BoxHalfWidth
            DotProduct(HitLocation.Translation - BoxCenter, RightVector) > -BoxHalfWidth
            DotProduct(HitLocation.Translation - BoxCenter, ForwardVector) < BoxHalfDepth
            DotProduct(HitLocation.Translation - BoxCenter, ForwardVector) > -BoxHalfDepth
        


(Input:[]t where t:type).Reverse<public>()<computes>:[]t=
    for (Index -> Unused : Input, Element := Input[Input.Length - Index - 1]). Element


(FortCharacter:fort_character).IsLookingAtSphere(Center : vector3, Radius : float)<transacts><decides>:void=
    FortCharacter.IsActive[]

    U := FortCharacter.GetViewRotation().GetLocalForward()
    V := FortCharacter.GetViewLocation() - Center

    a := DotProduct(U, U) # 1.0 since U is a unit vector
    b := 2.0 * DotProduct(U, V)
    c := DotProduct(V, V) - Radius * Radius
    d := b * b - 4.0 * a * c

    d >= 0.0


(FortCharacter:fort_character).IsLookingAtPlane(PlaneCenter: vector3, PlaneNormal: vector3)<transacts><decides>:void=
    if (FortCharacter.IsActive[]):
        # U is the view direction vector of the character (a unit vector)
        U := FortCharacter.GetViewRotation().GetLocalForward()
        # Origin is the view location of the character
        Origin := FortCharacter.GetViewLocation()
        # V is the vector from the character to the plane center
        V := PlaneCenter - Origin

        # Calculate the distance from the Origin to the point of intersection on the plane
        # This uses the plane equation and dot product properties
        # Note: We assume PlaneNormal is normalized. If not, normalize it first.
        var DistanceToPlane : float = -DotProduct(V, PlaneNormal) / DotProduct(U, PlaneNormal)

        if (DistanceToPlane > 0.0):
            # The character is looking towards the plane (intersection point is in front of the character)
            # Implement the logic for when the character is looking at the plane
            # This could be triggering an event, logging, etc.


(Agent:agent).AwaitNotActive()<suspends>:void=
    loop:
        if:
            FortChar := Agent.GetFortCharacter[]
            not FortChar.IsActive[]
        then:
            break
            
        Sleep(1.0)


(FortCharacter:fort_character).IsLookingAtRectangle(RectCenter:vector3, RectWidth:float, RectHeight:float, RectDir1:vector3, RectDir2:vector3, PlaneNormal:vector3)<transacts><decides>:void=
    var Logger:log := log{Channel:=log_device}
    var LookingAt : logic = false

    # Ensure the character is active
    if (FortCharacter.IsActive[]):
        U := FortCharacter.GetViewRotation().GetLocalForward() # Direction character is looking
        Origin := FortCharacter.GetViewLocation() # Character's location

        # Normalize direction and normal vectors
        NormalizedRectDir1 := RectDir1.MakeUnitVector[]
        NormalizedRectDir2 := RectDir2.MakeUnitVector[]
        NormalizedPlaneNormal := PlaneNormal.MakeUnitVector[]

        # Calculate vector from character's location to the rectangle center
        V := RectCenter - Origin

        # Check for non-parallelism between view direction and plane normal
        DotProd := DotProduct(U, NormalizedPlaneNormal)
        if (DotProd <> 0.0):
            DistanceToPlane := DotProduct(V, NormalizedPlaneNormal) / DotProd

            # Verify intersection point is in front of the character
            if (DistanceToPlane > 0.0):
                IntersectionPoint := Origin + (U * DistanceToPlane)

                # Project intersection point onto the rectangle's plane
                VecFromCenter := IntersectionPoint - RectCenter
                ProjOnDir1 := DotProduct(VecFromCenter, NormalizedRectDir1)
                ProjOnDir2 := DotProduct(VecFromCenter, NormalizedRectDir2)

                # Check if within rectangle bounds
                if (Abs(ProjOnDir1) <= RectWidth / 2.0 and Abs(ProjOnDir2) <= RectHeight / 2.0):
                    # Logger.Print("Character is looking at the rectangle.")
                    set LookingAt = true
                else:
                    # Logger.Print("Character is not looking at the rectangle.")
            else:
                # Logger.Print("The plane is behind the character.")
        else:
            # Logger.Print("View direction is parallel to the plane.")

    LookingAt?

            

(Prop:creative_prop).DelayedDispose(Delay:float)<suspends>:void=
    Sleep(Delay)
    Prop.Dispose()
    


        
GetNearestCard(HitLocation:transform, GameBoard:game_board)<transacts><decides>:card_tile=
    var NearestCard : ?card_tile = false

    for(Row : GameBoard.Rows):
        for(Tile : Row.Tiles):

            TileTransform := Tile.Prop.GetTransform()
            TilePosition := if(Tile.State = card_state.Up) then TileTransform.Translation + TilePivotOffset else TileTransform.Translation + vector3{Y:=32.0}
            LocalUp := TileTransform.Rotation.GetLocalRight()
            UpwardOffset := LocalUp * TileDimensions.Y / 2.0
            BoxCenter := TilePosition + UpwardOffset

            if:
                # Distance(HitLocation.Translation, BoxCenter) < TileDimensions.Y / 2.0
                DistanceSquared(HitLocation.Translation, BoxCenter) < Pow(TileDimensions.Y / 2.0, 2.0)
            then:

                if:
                    HasNearest := NearestCard?
                then:
                    if:
                        # Distance(HitLocation.Translation, Tile.Prop.GetTransform().Translation) < Distance(HitLocation.Translation, HasNearest.Prop.GetTransform().Translation)
                        DistanceSquared(HitLocation.Translation, Tile.Prop.GetTransform().Translation) < DistanceSquared(HitLocation.Translation, HasNearest.Prop.GetTransform().Translation)
                    then:
                        set NearestCard = option{Tile}
                else:
                    set NearestCard = option{Tile}

    return NearestCard?


ConvertOldScoreData(Player:player, OldScore:int):void=
    if:
        CheckForPlayerSavedData[Player]
        SourceData := PlayerSavedDataMap[Player]
        not SourceData.ConvertedScoreToCorrectGuesses? # if not already converted
        
        set PlayerSavedDataMap[Player] = player_saved_data:
            MakePlayerSavedData<constructor>(SourceData)
            ConvertedScoreToCorrectGuesses := true
            Score := OldScore
            CorrectGuesses := OldScore
    else:
        Print("No player data found for Player\n")

(L:logic).ToString():string=
    if (L?):
        return "True"
    else:
        return "False"

(SA:[]string).ArrPrint():string=
    var Result : string = ""
    for (Index -> Element : SA):
        set Result += Element
        if (Index < SA.Length - 1):
            set Result += ", "
    return Result

(IA:[]int).ToPrint():string=
    var Result : string = ""
    for (Index -> Element : IA):
        set Result += "{Element}"
        if (Index < IA.Length - 1):
            set Result += ", "
    return Result

# Customize cubic Bezier parameters for a stadium wave effect
# StadiumWave<public>:cubic_bezier_parameters = cubic_bezier_parameters{
#     X0 := 0.42,
#     Y0 := 0.0,
#     X1 := 0.58,
#     Y1 := 1.0
# }

StadiumWave<public>:cubic_bezier_parameters = cubic_bezier_parameters{
    X0 := 0.22,
    Y0 := 0.0,
    X1 := 0.50,
    Y1 := 1.0
}
StadiumWaveReverse<public>:cubic_bezier_parameters = cubic_bezier_parameters{
    X0 := 0.42,
    Y0 := 0.0,
    X1 := 0.58,
    Y1 := 1.0
}

(TeamColl:fort_team_collection).GetTeamsWithPlayers()<transacts><decides>:[]team=
    TeamsWithPlayers := for (Team: TeamColl.GetTeams(), (for (Player: TeamColl.GetAgents[Team], Player.GetFortCharacter[].IsActive[]) {Team}).Length > 0):
        Team

(TeamCollection:fort_team_collection).GetTeamIndex(Agent:agent)<decides><transacts>:int=
    Team := TeamCollection.GetTeam[Agent]
    TeamCollection.GetTeams().Find[Team]


# SortStylesByOrder(Styles:[]style_selection_data):[]style_selection_data=
#     var TempStyles : []style_selection_data = Styles

#     for (i := 0 .. TempStyles.Length - 1):
#         var MinIndex : int = i
#         for (j := i + 1 .. TempStyles.Length - 1):
#             if:
#                 StyleJ := TempStyles[j]
#                 StyleMin := TempStyles[MinIndex]
#             then:
#                 if(StyleJ.SortOrder < StyleMin.SortOrder):
#                     set MinIndex = j
#         if(MinIndex <> i):
#             if:
#                 Temp := TempStyles[i]
#                 MinStyle := TempStyles[MinIndex]
#             then:
#                 if:
#                     set TempStyles[i] = MinStyle
#                 then:
#                     if . set TempStyles[MinIndex] = Temp
#     return TempStyles


SortStylesByOrder(Styles: []style_selection_data): []style_selection_data =
    var TempStyles: []style_selection_data = Styles

    for (i := 0 .. TempStyles.Length - 1):
        var MinIndex: int = i
        for (j := i + 1 .. TempStyles.Length - 1):
            if:
                StyleJ := TempStyles[j]
                StyleMin := TempStyles[MinIndex]
            then:
                if (StyleJ.SortOrder < StyleMin.SortOrder):
                    set MinIndex = j
        if (MinIndex <> i):
            if:
                Temp := TempStyles[i]
                MinStyle := TempStyles[MinIndex]
                set TempStyles[i] = MinStyle
                set TempStyles[MinIndex] = Temp
    return TempStyles




        

# Removes an element from the given map and returns a new map without that element
# (Map:[keytype]valuetype where keytype:subtype(comparable), valuetype:type).RemoveKey(ElementToRemove:keytype):[keytype]valuetype=
#     var NewMap:[keytype]valuetype = map{}
#     for (Key -> Value : Map, Key <> ElementToRemove):
#         set NewMap = ConcatenateMaps(NewMap, map{Key => Value})
#     return NewMap




(Input:[k]v where k:subtype(comparable), v:type).RemoveKey(KeyToRemove:k):[k]v=
    if (Input[KeyToRemove]):
        var NewMap:[k]v = map{}
        for (Key->Value : Input, Key <> KeyToRemove) do if:
            set NewMap[Key] = Value
        NewMap
    else:
        Input

(Listenable : listenable(agent)).SubscribeAgent(OutputFunc : tuple(agent, t)->void, ExtraData : t where t:type) : cancelable =
    Wrapper := wrapper_agent(t){ExtraData := ExtraData, OutputFunc := OutputFunc}
    Listenable.Subscribe(Wrapper.InputFunc)


(Listenable : listenable(?agent)).SubscribeMaybeAgent(OutputFunc : tuple(?agent, t)->void, ExtraData : t where t:type) : cancelable =
    Wrapper := wrapper_maybe_agent(t){ExtraData := ExtraData, OutputFunc := OutputFunc}
    Listenable.Subscribe(Wrapper.InputFunc)

wrapper_agent(t : type) := class():
    ExtraData : t;
    OutputFunc : tuple(agent, t) -> void
    InputFunc(Agent : agent): void = OutputFunc(Agent, ExtraData)


wrapper_maybe_agent(t : type) := class():
    ExtraData : t;
    OutputFunc : tuple(?agent, t) -> void
    InputFunc(Agent : ?agent): void = OutputFunc(Agent, ExtraData)


# Subscribe to a listenable(widget_message) with extra data
(Listenable : listenable(widget_message)).SubscribeMessage(OutputFunc : tuple(widget_message, t, m)->void, ExtraData : t , MoreData : m where t:type, m : type) : cancelable =
    Wrapper := wrapper_message_extra(t, m){ExtraData := ExtraData, MoreData := MoreData,  OutputFunc := OutputFunc}
    Listenable.Subscribe(Wrapper.InputFunc)

wrapper_message_extra(t : type, m : type) := class():
    ExtraData : t;
    MoreData : m;
    OutputFunc : tuple(widget_message, t, m) -> void
    InputFunc(Message : widget_message):void = OutputFunc(Message, ExtraData, MoreData)

# Subscribe to a listenable(widget_message) with extra data
(Listenable : listenable(widget_message)).SubscribeMessage(OutputFunc : tuple(widget_message, t)->void, ExtraData : t where t:type) : cancelable =
    Wrapper := wrapper_message(t){ExtraData := ExtraData,  OutputFunc := OutputFunc}
    Listenable.Subscribe(Wrapper.InputFunc)

wrapper_message(t : type) := class():
    ExtraData : t;
    OutputFunc : tuple(widget_message, t) -> void
    InputFunc(Message : widget_message):void = OutputFunc(Message, ExtraData)

lerping <public>:=module:
    LinearEasing := class():
        LinearIn(k: float): float =
            var tempK: float = k
            return tempK;

    QuadraticEasing := class():
        In(k: float): float =
            var tempK: float = k
            return tempK * tempK;

        Out(k: float): float =
            var tempK: float = k
            return tempK * (2.0 - tempK);

        InOut(k: float): float =
            var tempK: float = k
            if (tempK = 1.0) { return 0.5 * tempK * tempK; }
            var tempK2: float = k - 1.0
            return -0.5 * (tempK2 * (tempK2 - 2.0) - 1.0);

        QuadraticBezier(k: float, c: float): float =
            return c * 2.0 * k * (1.0 - k) + k * k;

    CubicEasing <public>:= class():
        In<public>(k: float): float =
            var tempK: float = k
            return tempK * tempK * tempK;

        Out<public>(k: float): float =
            var tempK: float = k - 1.0
            return 1.0 + (tempK * tempK * tempK);

        InOut<public>(k: float): float =
            var tempK: float = k
            if (tempK = 1.0) { return 0.5 * tempK * tempK * tempK; }
            var tempK2: float = k - 2.0
            return 0.5 * (tempK2 * tempK2 * tempK2 + 2.0);

    QuarticEasing := class():
        In(k: float): float =
            var tempK: float = k
            return tempK * tempK * tempK * tempK;

        Out(k: float): float =
            var tempK: float = k
            return 1.0 - (tempK * tempK * tempK * tempK);

        InOut(k: float): float =
            var tempK: float = k
            if (tempK = 1.0) { return 0.5 * tempK * tempK * tempK * tempK; }
            var tempK2: float = k - 2.0
            return -0.5 * (tempK2 * tempK2 * tempK2 * tempK2 - 2.0);

    QuinticEasing <public> := class():
        In(k: float): float =
            var tempK: float = k
            return tempK * tempK * tempK * tempK * tempK;

        Out(k: float): float =
            var tempK: float = k - 1.0
            return 1.0 + (tempK * tempK * tempK * tempK * tempK);

        InOut<public>(k: float): float =
            var tempK: float = k
            if (tempK = 1.0) { return 0.5 * tempK * tempK * tempK * tempK * tempK; }
            var tempK2: float = k - 2.0
            return 0.5 * (tempK2 * tempK2 * tempK2 * tempK2 * tempK2 + 2.0);

    SinusoidalEasing := class():
        In(k: float): float =
            return 1.0 - Cos(k * PiFloat / 2.0);

        Out(k: float): float =
            return Sin(k * PiFloat / 2.0);

        InOut(k: float): float =
            return 0.5 * (1.0 - Cos(PiFloat * k));

    ExponentialEasing := class():
        In(k: float): float =
            return if( k = 0.0 ) { 0.0} else Pow(1024.0, k - 1.0);

        Out(k: float): float =
            return if( k = 1.0 ) { 1.0} else 1.0 - Pow(2.0, -10.0 * k);

        InOut(k: float): float =
            var tempK: float = k
            if (tempK = 0.0) { return 0.0; }
            if (tempK = 1.0) { return 1.0; }
            if ((tempK * 2.0) < 1.0) { return 0.5 * Pow(1024.0, tempK - 1.0); }
            var tempK2: float = tempK - 1.0
            return 0.5 * (-Pow(2.0, -10.0 * tempK2) + 2.0);

    CircularEasing := class():
        In(k: float): float =
            return 1.0 - Sqrt(1.0 - k * k);

        Out(k: float): float =
            var tempK: float = k - 1.0
            return Sqrt(1.0 - tempK * tempK);

        InOut(k: float): float =
            var tempK: float = k
            if ((tempK * 2.0) < 1.0) { return -0.5 * (Sqrt(1.0 - tempK * tempK) - 1.0); }
            var tempK2: float = tempK - 2.0
            return 0.5 * (Sqrt(1.0 - tempK2 * tempK2) + 1.0);

    ElasticEasing <public>:= class():
        In(k: float): float =
            if (k = 0.0) { return 0.0; }
            if (k = 1.0) { return 1.0; }
            return -Pow(2.0, 10.0 * (k - 1.0)) * Sin((k - 0.1) * (2.0 * PiFloat) / 0.4);

        Out(k: float): float =
            if (k = 0.0) { return 0.0; }
            if (k = 1.0) { return 1.0; }
            return Pow(2.0, -10.0 * k) * Sin((k - 0.1) * (2.0 * PiFloat) / 0.4) + 1.0;

        InOut<public>(k: float): float =
            var tempK: float = k
            if ((tempK * 2.0) < 1.0) { return -0.5 * Pow(2.0, 10.0 * (tempK - 1.0))} else return 0.5 * Pow(2.0, -10.0 * (tempK - 1.0)) * Sin((tempK - 1.1) * (2.0 * PiFloat) / 0.4) + 1.0;


    BounceEasing <public>:= class():
        In<public>(k: float): float =
            return 1.0 - Out(1.0 - k);

        Out<public>(k: float): float =
            if (k < (1.0 / 2.75)) {
                return 7.5625 * k * k;
            }
            else if (k < (2.0 / 2.75)) {
                return 7.5625 * (k - (1.5 / 2.75)) * k + 0.75;
            }
            else if (k < (2.5 / 2.75)) {
                return 7.5625 * (k - (2.25 / 2.75)) * k + 0.9375;
            }
            else {
                return 7.5625 * (k - (2.625 / 2.75)) * k + 0.984375;
            }

        InOut<public>(k: float): float =
            if (k < 0.5) {
                return In(k * 2.0) * 0.5;
            }
            return Out(k * 2.0 - 1.0) * 0.5 + 0.5;
            

            
(Input: string).Split(Delimeter: string)<transacts>: []string =
    if. Delimeter.Length = 0 then. return for. Char : Input do. ToString(Char)
    var Result: []string = array{}
    var Start: int = 0
    var End: int = 0
    if. Input.Length > 0 then. loop:
        if:
            InputSlice := Input.Slice[End, End + Delimeter.Length]
            InputSlice = Delimeter
            ResultSlice := Input.Slice[Start, End]
        then:
            set Result += array{ResultSlice}
            set End += Delimeter.Length - 1
            set Start = End + 1
        set End += 1
        if. End = Input.Length then. break
    if:
        Start <= Input.Length - 1
        RemainingSlice := Input.Slice[Start]
    then. set Result += array{RemainingSlice}
    return Result

ParseInt(String:string)<decides><transacts>:int=
    Characters:string = "0123456789"
    var Result:int = 0
    Sign:int = String[0] = "-" and -1 or 1
    for (Index->Char : String, Index > 0 or Sign = 1):
        TheChar := Characters.Find[Char]
        set Result = Result * 10 + TheChar
    Result * Sign

ParseDateTimes(Date:string, Time:string):time_and_date=
    DateParts := Date.Split("/")
    TimeParts := Time.Split(":")
    if:
        Day := ParseInt[DateParts[0]]
        Month := ParseInt[DateParts[1]]
        Year := ParseInt[DateParts[2]]
        Hour := ParseInt[TimeParts[0]]
        Minute := ParseInt[TimeParts[1]]
        Second := ParseInt[TimeParts[2]]
    then:
        return MakeTimeAndDate(Day, Month, Year, Hour, Minute, Second)
    return time_and_date{} # fail safe


# (V:vector3).RotateAngleAxis(Angle:float, Axis:vector3)<transacts>:vector3=
#     S := Sin(DegreesToRadians(Angle))
#     C := Cos(DegreesToRadians(Angle))

#     XX := Axis.X * Axis.X
#     YY := Axis.Y * Axis.Y
#     ZZ := Axis.Z * Axis.Z

#     XY := Axis.X * Axis.Y
#     YZ := Axis.Y * Axis.Z
#     ZX := Axis.Z * Axis.X

#     XS := Axis.X * S
#     YS := Axis.Y * S
#     ZS := Axis.Z * S

#     OMC := 1.0 - C

#     return vector3:
#         X := (OMC * XX + C) * V.X + (OMC * XY - ZS) * V.Y + (OMC * ZX + YS) * V.Z
#         Y := (OMC * XY + ZS) * V.X + (OMC * YY + C) * V.Y + (OMC * YZ - XS) * V.Z
#         Z := (OMC * ZX - YS) * V.X + (OMC * YZ  + XS) * V.Y + (OMC * ZZ + C) * V.Z

# (Current:vector3).RotateAround(Anchor:vector3, Axis:vector3, Angle:float)<transacts>:vector3=
#     TranslatedPoint := Current - Anchor

#     return TranslatedPoint.RotateAngleAxis(Angle, Axis) + Anchor



(Point:vector3).RotateAroundAnchor<public>(Anchor:vector3, Axis:vector3, AngleRadians:float)<transacts>:vector3=
    MakeRotation(Axis, AngleRadians).RotateVector(Point - Anchor) + Anchor


# ArrayRace will call the Function argument for each element in the Array and wait for only one of the async functions to complete, canceling all the rest.
# A divide-and-conquer concurrency algorithm that divides an array into two and races between the two subarrays.
# This is a recursive implementation, where the function calls itself to race between the divided arrays.
# The base case (the condition to stop the recursion) is the array has only one element.
# This is a generic implementation using parametric types, so you can provide your own type and your own async function as arguments.
ArrayRace(Array:[]t, Function(Element:t)<suspends>:u where t:type, u:type)<suspends>:u=
    # If more than two elements in the array, then split the array in half,
    # and race between the subarrays recursively.
    if:
        Array.Length > 1
        Mid := Floor(Array.Length / 2)
        LeftArray := Array.Slice[0, Mid]
        RightArray := Array.Slice[Mid]
    then:
        race:
            ArrayRace(LeftArray, Function)
            ArrayRace(RightArray, Function)
    else if:
        First := Array[0]
    then:
        # Grab the only element in the array and call the async function with it as the argument.
        Function(First)
    else:
        # A race with no subtasks never completes, so sleep infinitely and call Err() because we shouldn't return a value here.
        Sleep(Inf)
        Err()

# ArraySync will call the Function argument for each element in the Array and wait for all the async functions to complete.
# A divide-and-conquer concurrency algorithm that divides an array into two and syncs between the two subarrays.
# This is a recursive implementation, where the function calls itself to sync between the divided arrays.
# The base case (the condition to stop the recursion) is the array has only one element.
# This is a generic implementation using parametric types, so you can provide your own type and your own async function as arguments.
ArraySync(Array:[]t, Function(Element:t)<suspends>:u where t:type, u:type)<suspends>:[]u=
    # If more than two elements in the array, then split the array in half,
    # and sync between the subarrays recursively.
    if:
        Array.Length > 1
        Mid := Floor(Array.Length / 2)
        LeftArray := Array.Slice[0, Mid]
        RightArray := Array.Slice[Mid]
    then:
        SyncResult := sync:
            ArraySync(LeftArray, Function)
            ArraySync(RightArray, Function)

        # Return the results of the sync expressions as an array.
        SyncResult(0) + SyncResult(1)
    else if:
        First := Array[0]
    then:
        # Grab the only element in the array and call the async function with it as the argument and add the result to an array.
        array{Function(First)}
    else:
        # No element in the given Array so return an empty array.
        array{}


(OriginalColor: color).HueShift(HueShiftAmount: float): color =
    # Convert the original color to HSV components
    var HSV : tuple(float, float,float) = MakeHSVFromColor(OriginalColor)
    
    # Adjust the hue
    NewHue := HSV(0) + HueShiftAmount
    
    # Wrap the hue value using the FloatMod function
    if:
        aNewHue := FloatMod[NewHue, 360.0]
    then:  
        set HSV = (aNewHue, HSV(1), HSV(2))
        # Create a new color with the shifted hue

    NewColor := MakeColorFromHSV(HSV)
# Assuming a and b are positive
FloatMod(a: float, b: float)<transacts><decides>:float =
    Div := a / b
    var DivInt : int = Int[Div]  # Convert to integer (truncates towards zero)
    if (Div < 0.0 and Div <> DivInt):
        set DivInt -= 1  # Adjust for negative numbers
    var Result : float = a - b * DivInt
    if (Result < 0.0):
        set Result += b
    Result

Abs<public>(Vector3:vector3)<computes>:vector3=
    return vector3{X:=Abs(Vector3.X), Y:=Abs(Vector3.Y), Z:=Abs(Vector3.Z)}

bounding_box<public>:=struct<computes>:
    Center<public>:vector3
    Extents<public>:vector3

(BB:bounding_box).Contains(Position:vector3)<computes><decides>:void=
    Offset := Abs(Position - BB.Center)
    Offset.X < BB.Extents.X
    Offset.Y < BB.Extents.Y
    Offset.Z < BB.Extents.Z

(BB:bounding_box).GetRandomPointWithin()<transacts>:vector3=
    return vector3{X:=GetRandomFloat(-BB.Extents.X, BB.Extents.X), Y:=GetRandomFloat(-BB.Extents.Y, BB.Extents.Y), Z:=GetRandomFloat(-BB.Extents.Z, BB.Extents.Z)} + BB.Center

GetExtentsOfBoundingProp(Prop:creative_prop)<transacts>:bounding_box=
    Transform := Prop.GetTransform()
    Position := Transform.Translation
    Scale := Transform.Scale
    Extents := vector3{X:=100.0, Y:=100.0, Z:=100.0} * Scale / 2.0
    bounding_box{Center := Transform.Translation, Extents := Extents}


bounding_cylinder<public>:=struct<computes>:
    Center<public>:vector3
    Radius<public>:float
    Height<public>:float

(BoundingCylinder:bounding_cylinder).Contains(Position:vector3)<computes><decides>:void=
    Offset := Position - BoundingCylinder.Center
    Offset.X * Offset.X + Offset.Z * Offset.Z < BoundingCylinder.Radius * BoundingCylinder.Radius
    Offset.Y < BoundingCylinder.Height / 2.0

(BoundingCylinder:bounding_cylinder).GetRandomPointWithin()<transacts>:vector3=
    Angle := GetRandomFloat(0.0, 2.0 * PiFloat)
    Radius := GetRandomFloat(0.0, BoundingCylinder.Radius)
    Height := GetRandomFloat(-BoundingCylinder.Height / 2.0, BoundingCylinder.Height / 2.0)
    return vector3{X:=Radius * Cos(Angle), Y:=Height, Z:=Radius * Sin(Angle)} + BoundingCylinder.Center

GetBoundingCylinderOfBoundingProp(Prop:creative_prop):bounding_cylinder=
    Transform := Prop.GetTransform()
    Position := Transform.Translation
    Scale := Transform.Scale.X
    Radius := 100.0 * Scale / 2.0
    Height := 100.0 * Scale
    bounding_cylinder{Center := Position, Radius := Radius, Height := Height}

bounding_sphere<public>:=struct<computes>:
    Center<public>:vector3
    Radius<public>:float

(BoundingSphere:bounding_sphere).Contains(Position:vector3)<computes><decides>:void=
    Offset := Position - BoundingSphere.Center
    Offset.X * Offset.X + Offset.Y * Offset.Y + Offset.Z * Offset.Z < BoundingSphere.Radius * BoundingSphere.Radius

(BoundingSphere:bounding_sphere).GetRandomPointWithin()<transacts>:vector3=
    Angle1 := GetRandomFloat(0.0, 2.0 * PiFloat)
    Angle2 := GetRandomFloat(0.0, 2.0 * PiFloat)
    Radius := GetRandomFloat(0.0, BoundingSphere.Radius)
    return vector3{X:=Radius * Sin(Angle1) * Cos(Angle2), Y:=Radius * Sin(Angle1) * Sin(Angle2), Z:=Radius * Cos(Angle1)} + BoundingSphere.Center

GetBoundingSphereOfBoundingProp(Prop:creative_prop):bounding_sphere=
    Transform := Prop.GetTransform()
    Position := Transform.Translation
    Scale := Transform.Scale.X
    Radius := 100.0 * Scale / 2.0
    bounding_sphere{Center := Position, Radius := Radius}

LookAt(Origin: vector3, Target: vector3)<transacts><decides>:rotation =
    Direction := (Target - Origin).MakeUnitVector[]
    Yaw := RadiansToDegrees(ArcTan(Direction.Y, Direction.X))
    UnrotatedAfterYaw := MakeRotationFromYawPitchRollDegrees(Yaw, 0.0, 0.0).UnrotateVector(Direction)
    Pitch := RadiansToDegrees(ArcTan(UnrotatedAfterYaw.X, -UnrotatedAfterYaw.Z))
    UnrotatedAfterPitch := MakeRotationFromYawPitchRollDegrees(0.0, Pitch, 0.0).UnrotateVector(UnrotatedAfterYaw)
    Roll := RadiansToDegrees(ArcTan(-UnrotatedAfterPitch.Z, UnrotatedAfterPitch.Y))
    MakeRotationFromYawPitchRollDegrees(Yaw, Pitch - 90.0, Roll - 90.0)
    

LookAtLocked(Origin: vector3, Target: vector3, LockAxis: vector3)<transacts><decides>:rotation =
    Direction := (Target - Origin).MakeUnitVector[]
    ProjectedDirection := Direction - (DotProduct(Direction, LockAxis) * LockAxis)
    NormalizedProjectedDirection := ProjectedDirection.MakeUnitVector[]
    Yaw := SpatialMath.RadiansToDegrees(ArcTan(NormalizedProjectedDirection.Y, NormalizedProjectedDirection.X))
    Pitch := SpatialMath.RadiansToDegrees(ArcSin(Direction.Z))
    SpatialMath.MakeRotationFromYawPitchRollDegrees(Yaw, Pitch, 0.0)
    

    


MakeRoundedOverlayOuter(BGColor:color):overlay=
    overlay:
        Slots := array:
            overlay_slot:
                HorizontalAlignment := horizontal_alignment.Fill
                VerticalAlignment := vertical_alignment.Fill
                Padding := margin{Top := 0.0, Left := 0.0, Right := 0.0, Bottom := 0.0}
                Widget := stack_box:
                    Orientation := orientation.Vertical
                    Slots := array:
                        # Top row with corners and horizontal fill
                        stack_box_slot:
                            HorizontalAlignment := horizontal_alignment.Fill
                            VerticalAlignment := vertical_alignment.Fill
                            Widget:= stack_box:
                                Orientation := orientation.Horizontal
                                Slots:= array:
                                    # Top left corner
                                    stack_box_slot:
                                        HorizontalAlignment := horizontal_alignment.Left
                                        Widget:= texture_block:
                                            DefaultTint:=BGColor
                                            DefaultImage := _UI.UI_Textures.ui_rounded_top_left_outer
                                    # Horizontal fill
                                    stack_box_slot:
                                        Distribution := option{2.0}
                                        HorizontalAlignment := horizontal_alignment.Fill
                                        VerticalAlignment := vertical_alignment.Fill
                                        Widget:= texture_block:
                                            DefaultTint:=BGColor
                                            DefaultImage := _UI.UI_Textures.ui_square_blank
                                            DefaultHorizontalTiling:= image_tiling.Stretch
                                            DefaultVerticalTiling:= image_tiling.Stretch
                                    # Top right corner
                                    stack_box_slot:
                                        HorizontalAlignment := horizontal_alignment.Right
                                        Widget:= texture_block:
                                            DefaultTint:=BGColor
                                            DefaultImage := _UI.UI_Textures.ui_rounded_top_right_outer

                        # Middle row (vertical fill, no corners)
                        stack_box_slot:
                            Distribution := option{2.0}
                            HorizontalAlignment := horizontal_alignment.Fill
                            VerticalAlignment := vertical_alignment.Fill
                            Widget:= texture_block:
                                DefaultTint:=BGColor
                                DefaultImage := _UI.UI_Textures.ui_square_blank
                                DefaultHorizontalTiling:= image_tiling.Stretch
                                DefaultVerticalTiling:= image_tiling.Stretch
                        # Bottom row with corners and horizontal fill
                        stack_box_slot:
                            HorizontalAlignment := horizontal_alignment.Fill
                            VerticalAlignment := vertical_alignment.Fill
                            Widget:= stack_box:
                                Orientation := orientation.Horizontal
                                Slots:= array:
                                    # Bottom left corner
                                    stack_box_slot:
                                        HorizontalAlignment := horizontal_alignment.Left
                                        Widget:= texture_block:
                                            DefaultTint:=BGColor
                                            DefaultImage := _UI.UI_Textures.ui_rounded_bottom_left_outer
                                    # Horizontal fill
                                    stack_box_slot:
                                        HorizontalAlignment := horizontal_alignment.Fill
                                        VerticalAlignment := vertical_alignment.Fill
                                        Distribution := option{2.0}
                                        Widget:= texture_block:
                                            DefaultTint:=BGColor
                                            DefaultImage := _UI.UI_Textures.ui_square_blank
                                            DefaultHorizontalTiling:= image_tiling.Stretch
                                            DefaultVerticalTiling:= image_tiling.Stretch
                                    # Bottom right corner
                                    stack_box_slot:
                                        HorizontalAlignment := horizontal_alignment.Right
                                        Widget:= texture_block:
                                            DefaultTint:=BGColor
                                            DefaultImage := _UI.UI_Textures.ui_rounded_bottom_right_outer

MakeRoundedOverlayOuterTopCorners(BGColor:color):overlay=
    overlay:
        Slots := array:
            overlay_slot:
                HorizontalAlignment := horizontal_alignment.Fill
                VerticalAlignment := vertical_alignment.Fill
                Padding := margin{Top := 0.0, Left := 0.0, Right := 0.0, Bottom := 0.0}
                Widget := stack_box:
                    Orientation := orientation.Vertical
                    Slots := array:
                        # Top row with corners and horizontal fill
                        stack_box_slot:
                            HorizontalAlignment := horizontal_alignment.Fill
                            VerticalAlignment := vertical_alignment.Fill
                            Widget:= stack_box:
                                Orientation := orientation.Horizontal
                                Slots:= array:
                                    # Top left corner
                                    stack_box_slot:
                                        HorizontalAlignment := horizontal_alignment.Left
                                        Widget:= texture_block:
                                            DefaultTint:=BGColor
                                            DefaultImage := _UI.UI_Textures.ui_rounded_top_left_outer
                                    # Horizontal fill
                                    stack_box_slot:
                                        Distribution := option{2.0}
                                        HorizontalAlignment := horizontal_alignment.Fill
                                        VerticalAlignment := vertical_alignment.Fill
                                        Widget:= texture_block:
                                            DefaultTint:=BGColor
                                            DefaultImage := _UI.UI_Textures.ui_square_blank
                                            DefaultHorizontalTiling:= image_tiling.Stretch
                                            DefaultVerticalTiling:= image_tiling.Stretch
                                    # Top right corner
                                    stack_box_slot:
                                        HorizontalAlignment := horizontal_alignment.Right
                                        Widget:= texture_block:
                                            DefaultTint:=BGColor
                                            DefaultImage := _UI.UI_Textures.ui_rounded_top_right_outer

                        # Middle row (vertical fill, no corners)
                        stack_box_slot:
                            Distribution := option{2.0}
                            HorizontalAlignment := horizontal_alignment.Fill
                            VerticalAlignment := vertical_alignment.Fill
                            Widget:= texture_block:
                                DefaultTint:=BGColor
                                DefaultImage := _UI.UI_Textures.ui_square_blank
                                DefaultHorizontalTiling:= image_tiling.Stretch
                                DefaultVerticalTiling:= image_tiling.Stretch

MakeRoundedOverlayInner(BGColor:color, ?TextureSize:float=32.0):overlay=
    overlay:
        Slots := array:
            overlay_slot:
                HorizontalAlignment := horizontal_alignment.Fill
                VerticalAlignment := vertical_alignment.Fill
                Padding := margin{Top := 0.0, Left := 0.0, Right := 0.0, Bottom := 0.0}
                Widget := stack_box:
                    Orientation := orientation.Vertical
                    Slots := array:
                        # Top row with corners and horizontal fill
                        stack_box_slot:
                            HorizontalAlignment := horizontal_alignment.Fill
                            VerticalAlignment := vertical_alignment.Fill
                            Widget:= stack_box:
                                Orientation := orientation.Horizontal
                                Slots:= array:
                                    # Top left corner
                                    stack_box_slot:
                                        HorizontalAlignment := horizontal_alignment.Left
                                        Widget:= texture_block:
                                            DefaultTint:=BGColor
                                            DefaultImage := _UI.UI_Textures.ui_rounded_top_left_inner
                                            DefaultDesiredSize := vector2{X:=TextureSize, Y:=TextureSize}
                                    # Horizontal fill
                                    stack_box_slot:
                                        Distribution := option{2.0}
                                        HorizontalAlignment := horizontal_alignment.Fill
                                        VerticalAlignment := vertical_alignment.Fill
                                        Widget:= texture_block:
                                            DefaultTint:=BGColor
                                            DefaultImage := _UI.UI_Textures.ui_square_blank
                                            DefaultHorizontalTiling:= image_tiling.Stretch
                                            DefaultVerticalTiling:= image_tiling.Stretch
                                            DefaultDesiredSize := vector2{X:=TextureSize, Y:=TextureSize}
                                    # Top right corner
                                    stack_box_slot:
                                        HorizontalAlignment := horizontal_alignment.Right
                                        Widget:= texture_block:
                                            DefaultTint:=BGColor
                                            DefaultImage := _UI.UI_Textures.ui_rounded_top_right_inner
                                            DefaultDesiredSize := vector2{X:=TextureSize, Y:=TextureSize}

                        # Middle row (vertical fill, no corners)
                        stack_box_slot:
                            Distribution := option{2.0}
                            HorizontalAlignment := horizontal_alignment.Fill
                            VerticalAlignment := vertical_alignment.Fill
                            Widget:= texture_block:
                                DefaultTint:=BGColor
                                DefaultImage := _UI.UI_Textures.ui_square_blank
                                DefaultHorizontalTiling:= image_tiling.Stretch
                                DefaultVerticalTiling:= image_tiling.Stretch
                                DefaultDesiredSize := vector2{X:=TextureSize, Y:=TextureSize}
                        # Bottom row with corners and horizontal fill
                        stack_box_slot:
                            HorizontalAlignment := horizontal_alignment.Fill
                            VerticalAlignment := vertical_alignment.Fill
                            Widget:= stack_box:
                                Orientation := orientation.Horizontal
                                Slots:= array:
                                    # Bottom left corner
                                    stack_box_slot:
                                        HorizontalAlignment := horizontal_alignment.Left
                                        Widget:= texture_block:
                                            DefaultTint:=BGColor
                                            DefaultImage := _UI.UI_Textures.ui_rounded_bottom_left_inner
                                            DefaultDesiredSize := vector2{X:=TextureSize, Y:=TextureSize}
                                    # Horizontal fill
                                    stack_box_slot:
                                        HorizontalAlignment := horizontal_alignment.Fill
                                        VerticalAlignment := vertical_alignment.Fill
                                        Distribution := option{2.0}
                                        Widget:= texture_block:
                                            DefaultTint:=BGColor
                                            DefaultImage := _UI.UI_Textures.ui_square_blank
                                            DefaultHorizontalTiling:= image_tiling.Stretch
                                            DefaultVerticalTiling:= image_tiling.Stretch
                                            DefaultDesiredSize := vector2{X:=TextureSize, Y:=TextureSize}
                                    # Bottom right corner
                                    stack_box_slot:
                                        HorizontalAlignment := horizontal_alignment.Right
                                        Widget:= texture_block:
                                            DefaultTint:=BGColor
                                            DefaultImage := _UI.UI_Textures.ui_rounded_bottom_right_inner
                                            DefaultDesiredSize := vector2{X:=TextureSize, Y:=TextureSize}


MakeRoundedOverlayInnerTopCorners(BGColor:color):overlay=
    overlay:
        Slots := array:
            overlay_slot:
                HorizontalAlignment := horizontal_alignment.Fill
                VerticalAlignment := vertical_alignment.Fill
                Padding := margin{Top := 0.0, Left := 0.0, Right := 0.0, Bottom := 0.0}
                Widget := stack_box:
                    Orientation := orientation.Vertical
                    Slots := array:
                        # Top row with corners and horizontal fill
                        stack_box_slot:
                            HorizontalAlignment := horizontal_alignment.Fill
                            VerticalAlignment := vertical_alignment.Fill
                            Widget:= stack_box:
                                Orientation := orientation.Horizontal
                                Slots:= array:
                                    # Top left corner
                                    stack_box_slot:
                                        HorizontalAlignment := horizontal_alignment.Left
                                        Widget:= texture_block:
                                            DefaultTint:=BGColor
                                            DefaultImage := _UI.UI_Textures.ui_rounded_top_left_inner
                                    # Horizontal fill
                                    stack_box_slot:
                                        Distribution := option{2.0}
                                        HorizontalAlignment := horizontal_alignment.Fill
                                        VerticalAlignment := vertical_alignment.Fill
                                        Widget:= texture_block:
                                            DefaultTint:=BGColor
                                            DefaultImage := _UI.UI_Textures.ui_square_blank
                                            DefaultHorizontalTiling:= image_tiling.Stretch
                                            DefaultVerticalTiling:= image_tiling.Stretch
                                    # Top right corner
                                    stack_box_slot:
                                        HorizontalAlignment := horizontal_alignment.Right
                                        Widget:= texture_block:
                                            DefaultTint:=BGColor
                                            DefaultImage := _UI.UI_Textures.ui_rounded_top_right_inner

                        # Middle row (vertical fill, no corners)
                        stack_box_slot:
                            Distribution := option{2.0}
                            HorizontalAlignment := horizontal_alignment.Fill
                            VerticalAlignment := vertical_alignment.Fill
                            Widget:= texture_block:
                                DefaultTint:=BGColor
                                DefaultImage := _UI.UI_Textures.ui_square_blank
                                DefaultHorizontalTiling:= image_tiling.Stretch
                                DefaultVerticalTiling:= image_tiling.Stretch

MakeRoundedOverlayInnerBottomCorners(BGColor:color):overlay=
    overlay:
        Slots := array:
            overlay_slot:
                HorizontalAlignment := horizontal_alignment.Fill
                VerticalAlignment := vertical_alignment.Fill
                Padding := margin{Top := 0.0, Left := 0.0, Right := 0.0, Bottom := 0.0}
                Widget := stack_box:
                    Orientation := orientation.Vertical
                    Slots := array:
                        # Middle row (vertical fill, no corners)
                        stack_box_slot:
                            Distribution := option{2.0}
                            HorizontalAlignment := horizontal_alignment.Fill
                            VerticalAlignment := vertical_alignment.Fill
                            Widget:= texture_block:
                                DefaultTint:=BGColor
                                DefaultImage := _UI.UI_Textures.ui_square_blank
                                DefaultHorizontalTiling:= image_tiling.Stretch
                                DefaultVerticalTiling:= image_tiling.Stretch
                        # Bottom row with corners and horizontal fill
                        stack_box_slot:
                            HorizontalAlignment := horizontal_alignment.Fill
                            VerticalAlignment := vertical_alignment.Fill
                            Widget:= stack_box:
                                Orientation := orientation.Horizontal
                                Slots:= array:
                                    # Bottom left corner
                                    stack_box_slot:
                                        HorizontalAlignment := horizontal_alignment.Left
                                        Widget:= texture_block:
                                            DefaultTint:=BGColor
                                            DefaultImage := _UI.UI_Textures.ui_rounded_bottom_left_inner
                                    # Horizontal fill
                                    stack_box_slot:
                                        HorizontalAlignment := horizontal_alignment.Fill
                                        VerticalAlignment := vertical_alignment.Fill
                                        Distribution := option{2.0}
                                        Widget:= texture_block:
                                            DefaultTint:=BGColor
                                            DefaultImage := _UI.UI_Textures.ui_square_blank
                                            DefaultHorizontalTiling:= image_tiling.Stretch
                                            DefaultVerticalTiling:= image_tiling.Stretch
                                    # Bottom right corner
                                    stack_box_slot:
                                        HorizontalAlignment := horizontal_alignment.Right
                                        Widget:= texture_block:
                                            DefaultTint:=BGColor
                                            DefaultImage := _UI.UI_Textures.ui_rounded_bottom_right_inner

MakeRoundedOverlayQuietButton(BGColor:color, ?TextureSize:float = 16.0):overlay=
    overlay:
        Slots := array:
            overlay_slot:
                HorizontalAlignment := horizontal_alignment.Fill
                VerticalAlignment := vertical_alignment.Fill
                Padding := margin{Top := 0.0, Left := 0.0, Right := 0.0, Bottom := 0.0}
                Widget := stack_box:
                    Orientation := orientation.Vertical
                    Slots := array:
                        # Top row with corners and horizontal fill
                        stack_box_slot:
                            HorizontalAlignment := horizontal_alignment.Fill
                            VerticalAlignment := vertical_alignment.Fill
                            Widget:= stack_box:
                                Orientation := orientation.Horizontal
                                Slots:= array:
                                    # Top left corner
                                    stack_box_slot:
                                        HorizontalAlignment := horizontal_alignment.Left
                                        Widget:= texture_block:
                                            DefaultTint:=BGColor
                                            DefaultImage := _UI.UI_Textures.ui_rounded_top_left_quiet
                                            DefaultDesiredSize := vector2{X:=TextureSize, Y:=TextureSize}
                                    # Horizontal fill
                                    stack_box_slot:
                                        Distribution := option{2.0}
                                        HorizontalAlignment := horizontal_alignment.Fill
                                        VerticalAlignment := vertical_alignment.Fill
                                        Widget:= texture_block:
                                            DefaultTint:=BGColor
                                            DefaultImage := _UI.UI_Textures.ui_square_blank
                                            DefaultHorizontalTiling:= image_tiling.Stretch
                                            DefaultVerticalTiling:= image_tiling.Stretch
                                            DefaultDesiredSize := vector2{X:=TextureSize, Y:=TextureSize}
                                    # Top right corner
                                    stack_box_slot:
                                        HorizontalAlignment := horizontal_alignment.Right
                                        Widget:= texture_block:
                                            DefaultTint:=BGColor
                                            DefaultImage := _UI.UI_Textures.ui_rounded_top_right_quiet
                                            DefaultDesiredSize := vector2{X:=TextureSize, Y:=TextureSize}

                        # Middle row (vertical fill, no corners)
                        stack_box_slot:
                            Distribution := option{2.0}
                            HorizontalAlignment := horizontal_alignment.Fill
                            VerticalAlignment := vertical_alignment.Fill
                            Widget:= texture_block:
                                DefaultTint:=BGColor
                                DefaultImage := _UI.UI_Textures.ui_square_blank
                                DefaultHorizontalTiling:= image_tiling.Stretch
                                DefaultVerticalTiling:= image_tiling.Stretch
                                DefaultDesiredSize := vector2{X:=TextureSize, Y:=TextureSize}
                        # Bottom row with corners and horizontal fill
                        stack_box_slot:
                            HorizontalAlignment := horizontal_alignment.Fill
                            VerticalAlignment := vertical_alignment.Fill
                            Widget:= stack_box:
                                Orientation := orientation.Horizontal
                                Slots:= array:
                                    # Bottom left corner
                                    stack_box_slot:
                                        HorizontalAlignment := horizontal_alignment.Left
                                        Widget:= texture_block:
                                            DefaultTint:=BGColor
                                            DefaultImage := _UI.UI_Textures.ui_rounded_bottom_left_quiet
                                            DefaultDesiredSize := vector2{X:=TextureSize, Y:=TextureSize}
                                    # Horizontal fill
                                    stack_box_slot:
                                        HorizontalAlignment := horizontal_alignment.Fill
                                        VerticalAlignment := vertical_alignment.Fill
                                        Distribution := option{2.0}
                                        Widget:= texture_block:
                                            DefaultTint:=BGColor
                                            DefaultImage := _UI.UI_Textures.ui_square_blank
                                            DefaultHorizontalTiling:= image_tiling.Stretch
                                            DefaultVerticalTiling:= image_tiling.Stretch
                                            DefaultDesiredSize := vector2{X:=TextureSize, Y:=TextureSize}
                                    # Bottom right corner
                                    stack_box_slot:
                                        HorizontalAlignment := horizontal_alignment.Right
                                        Widget:= texture_block:
                                            DefaultTint:=BGColor
                                            DefaultImage := _UI.UI_Textures.ui_rounded_bottom_right_quiet
                                            DefaultDesiredSize := vector2{X:=TextureSize, Y:=TextureSize}
                                            

MakeRoundedOverlayNine(Slices:tuple(texture,texture,texture,texture,texture,texture,texture,texture,texture), TextureSize:float, ?BorderTint:color=White):overlay=
    overlay:
        Slots := array:
            overlay_slot:
                HorizontalAlignment := horizontal_alignment.Fill
                VerticalAlignment := vertical_alignment.Fill
                Padding := margin{Top := 0.0, Left := 0.0, Right := 0.0, Bottom := 0.0}
                Widget := stack_box:
                    Orientation := orientation.Vertical
                    Slots := array:
                        # Top row with left, center, and right
                        stack_box_slot:
                            HorizontalAlignment := horizontal_alignment.Fill
                            VerticalAlignment := vertical_alignment.Fill
                            Widget:= stack_box:
                                Orientation := orientation.Horizontal
                                Slots:= array:
                                    # Top left corner
                                    stack_box_slot:
                                        Widget:= texture_block:
                                            DefaultImage := Slices(0)
                                            DefaultDesiredSize := vector2{X:=TextureSize, Y:=TextureSize}
                                            DefaultTint := BorderTint
                                    # Top center
                                    stack_box_slot:
                                        Distribution := option{2.0}
                                        HorizontalAlignment := horizontal_alignment.Fill
                                        VerticalAlignment := vertical_alignment.Fill
                                        Widget:= texture_block:
                                            DefaultImage := Slices(1)
                                            DefaultHorizontalTiling:= image_tiling.Stretch
                                            DefaultVerticalTiling:= image_tiling.Stretch
                                            DefaultDesiredSize := vector2{X:=TextureSize, Y:=TextureSize}
                                            DefaultTint := BorderTint
                                    # Top right corner
                                    stack_box_slot:
                                        Widget:= texture_block:
                                            DefaultImage := Slices(2)
                                            DefaultDesiredSize := vector2{X:=TextureSize, Y:=TextureSize}
                                            DefaultTint := BorderTint

                        # Middle row with left, center, and right
                        stack_box_slot:
                            Distribution := option{2.0}
                            HorizontalAlignment := horizontal_alignment.Fill
                            VerticalAlignment := vertical_alignment.Fill
                            Widget:= stack_box:
                                Orientation := orientation.Horizontal
                                Slots:= array:
                                    # Middle left
                                    stack_box_slot:
                                        HorizontalAlignment := horizontal_alignment.Fill
                                        VerticalAlignment := vertical_alignment.Fill
                                        Widget:= texture_block:
                                            DefaultImage := Slices(3)
                                            DefaultDesiredSize := vector2{X:=TextureSize, Y:=TextureSize}
                                            DefaultTint := BorderTint
                                    # Middle center
                                    stack_box_slot:
                                        Distribution := option{2.0}
                                        HorizontalAlignment := horizontal_alignment.Fill
                                        VerticalAlignment := vertical_alignment.Fill
                                        Widget:= texture_block:
                                            DefaultImage := Slices(4)
                                            DefaultHorizontalTiling:= image_tiling.Stretch
                                            DefaultVerticalTiling:= image_tiling.Stretch
                                            DefaultDesiredSize := vector2{X:=TextureSize, Y:=TextureSize}
                                            DefaultTint := BorderTint
                                    # Middle right
                                    stack_box_slot:
                                        HorizontalAlignment := horizontal_alignment.Fill
                                        VerticalAlignment := vertical_alignment.Fill
                                        Widget:= texture_block:
                                            DefaultImage := Slices(5)
                                            DefaultDesiredSize := vector2{X:=TextureSize, Y:=TextureSize}
                                            DefaultTint := BorderTint

                        # Bottom row with left, center, and right
                        stack_box_slot:
                            HorizontalAlignment := horizontal_alignment.Fill
                            VerticalAlignment := vertical_alignment.Fill
                            Widget:= stack_box:
                                Orientation := orientation.Horizontal
                                Slots:= array:
                                    # Bottom left corner
                                    stack_box_slot:
                                        Widget:= texture_block:
                                            DefaultImage := Slices(6)
                                            DefaultDesiredSize := vector2{X:=TextureSize, Y:=TextureSize}
                                            DefaultTint := BorderTint
                                    # Bottom center
                                    stack_box_slot:
                                        Distribution := option{2.0}
                                        HorizontalAlignment := horizontal_alignment.Fill
                                        VerticalAlignment := vertical_alignment.Fill
                                        Widget:= texture_block:
                                            DefaultImage := Slices(7)
                                            DefaultHorizontalTiling:= image_tiling.Stretch
                                            DefaultVerticalTiling:= image_tiling.Stretch
                                            DefaultDesiredSize := vector2{X:=TextureSize, Y:=TextureSize}
                                            DefaultTint := BorderTint
                                    # Bottom right corner
                                    stack_box_slot:
                                        Widget:= texture_block:
                                            DefaultImage := Slices(8)
                                            DefaultDesiredSize := vector2{X:=TextureSize, Y:=TextureSize}    
                                            DefaultTint := BorderTint


MakeDropShadowOverlay(WidgetToWrap: widget, ?ShadowColor: color := MakeColorFromHSV(0.0, 0.0, 0.0), ?ShadowOffset: vector2 := vector2{X:=4.0, Y:=4.0}, ?TextureSize: float = 16.0): overlay =
    overlay:
        Slots := array:
            # Drop shadow layer
            overlay_slot:
                HorizontalAlignment := horizontal_alignment.Fill,
                VerticalAlignment := vertical_alignment.Fill,
                Padding := margin{Left := ShadowOffset.X, Top := ShadowOffset.Y, Right := -ShadowOffset.X, Bottom := -ShadowOffset.Y},
                Widget := MakeDropShadowBackground(ShadowColor, ?TextureSize:=TextureSize)
            # Original widget layer
            overlay_slot:
                HorizontalAlignment := horizontal_alignment.Fill,
                VerticalAlignment := vertical_alignment.Fill,
                Widget := WidgetToWrap

MakeDropShadowBackground(ShadowColor: color, ?TextureSize:float=256.0): overlay =
    overlay:
        Slots := array:
            overlay_slot:
                HorizontalAlignment := horizontal_alignment.Fill,
                VerticalAlignment := vertical_alignment.Fill,
                Widget := stack_box:
                    Orientation := orientation.Vertical,
                    Slots := array:
                        # Top row with corners and horizontal fill
                        stack_box_slot:
                            HorizontalAlignment := horizontal_alignment.Fill,
                            VerticalAlignment := vertical_alignment.Top,
                            Widget := stack_box:
                                Orientation := orientation.Horizontal,
                                Slots := array:
                                    # Top left corner shadow
                                    stack_box_slot:
                                        Widget := texture_block:
                                            DefaultTint := ShadowColor,
                                            # DefaultOpacity := ShadowOpacity,
                                            DefaultImage := _UI.UI_Textures.ui_shadow_top_left,
                                            DefaultDesiredSize := vector2{X := TextureSize, Y := TextureSize}
                                    # Top edge shadow
                                    stack_box_slot:
                                        Distribution := option{1.0},
                                        Widget := texture_block:
                                            DefaultTint := ShadowColor,
                                            # DefaultOpacity := ShadowOpacity,
                                            DefaultImage := _UI.UI_Textures.ui_shadow_top,
                                            DefaultHorizontalTiling := image_tiling.Stretch,
                                            DefaultDesiredSize := vector2{X := 0.0, Y := TextureSize}
                                    # Top right corner shadow
                                    stack_box_slot:
                                        Widget := texture_block:
                                            DefaultTint := ShadowColor,
                                            # DefaultOpacity := ShadowOpacity,
                                            DefaultImage := _UI.UI_Textures.ui_shadow_top_right,
                                            DefaultDesiredSize := vector2{X := TextureSize, Y := TextureSize}
                        # Middle row (vertical fill)
                        stack_box_slot:
                            Distribution := option{1.0},
                            HorizontalAlignment := horizontal_alignment.Fill,
                            VerticalAlignment := vertical_alignment.Fill,
                            Widget := stack_box:
                                Orientation := orientation.Horizontal,
                                Slots := array:
                                    # Left edge shadow
                                    stack_box_slot:
                                        Widget := texture_block:
                                            DefaultTint := ShadowColor,
                                            # DefaultOpacity := ShadowOpacity,
                                            DefaultImage := _UI.UI_Textures.ui_shadow_left,
                                            DefaultVerticalTiling := image_tiling.Stretch,
                                            DefaultDesiredSize := vector2{X := TextureSize, Y := 0.0}
                                    # Center (transparent)
                                    stack_box_slot:
                                        Distribution := option{1.0},
                                        Widget := overlay{}  # Empty space
                                    # Right edge shadow
                                    stack_box_slot:
                                        Widget := texture_block:
                                            DefaultTint := ShadowColor,
                                            # DefaultOpacity := ShadowOpacity,
                                            DefaultImage := _UI.UI_Textures.ui_shadow_right,
                                            DefaultVerticalTiling := image_tiling.Stretch,
                                            DefaultDesiredSize := vector2{X := TextureSize, Y := 0.0}
                        # Bottom row with corners and horizontal fill
                        stack_box_slot:
                            HorizontalAlignment := horizontal_alignment.Fill,
                            VerticalAlignment := vertical_alignment.Bottom,
                            Widget := stack_box:
                                Orientation := orientation.Horizontal,
                                Slots := array:
                                    # Bottom left corner shadow
                                    stack_box_slot:
                                        Widget := texture_block:
                                            DefaultTint := ShadowColor,
                                            # DefaultOpacity := ShadowOpacity,
                                            DefaultImage := _UI.UI_Textures.ui_shadow_bottom_left,
                                            DefaultDesiredSize := vector2{X := TextureSize, Y := TextureSize}
                                    # Bottom edge shadow
                                    stack_box_slot:
                                        Distribution := option{1.0},
                                        Widget := texture_block:
                                            DefaultTint := ShadowColor,
                                            # DefaultOpacity := ShadowOpacity,
                                            DefaultImage := _UI.UI_Textures.ui_shadow_bottom,
                                            DefaultHorizontalTiling := image_tiling.Stretch,
                                            DefaultDesiredSize := vector2{X := 0.0, Y := TextureSize}
                                    # Bottom right corner shadow
                                    stack_box_slot:
                                        Widget := texture_block:
                                            DefaultTint := ShadowColor,
                                            # DefaultOpacity := ShadowOpacity,
                                            DefaultImage := _UI.UI_Textures.ui_shadow_bottom_right,
                                            DefaultDesiredSize := vector2{X := TextureSize, Y := TextureSize}
