using { /Fortnite.com/Devices }
using { /Verse.org/Simulation }
using { /UnrealEngine.com/Temporary/SpatialMath }
using { /Fortnite.com/Devices/CreativeAnimation }
using { /UnrealEngine.com/Temporary/Diagnostics }

you_tile := class<concrete>():

    Logger:log := log{Channel:=log_device}

    @editable Prop : creative_prop = creative_prop{}
    # @editable Trigger : trigger_device = trigger_device{}
    @editable CinePeak : cinematic_sequence_device = cinematic_sequence_device{}
    @editable CineReveal : cinematic_sequence_device = cinematic_sequence_device{}
    @editable CineHide : cinematic_sequence_device = cinematic_sequence_device{}
    @editable FlipButtonProp : creative_prop = creative_prop{}
    
    var CardHidden : logic = false

    SetStyle(Style:card_style):void=
        Self.Prop.SetMaterial(MakeMaterialForCardStyle(Style), ?Index:=0)
        Self.Prop.SetMaterial(MakeMaterialForCardStyle(Style), ?Index:=1)

    SetCardIndex(Index:int):void=
        Self.Prop.SetMesh(GetCardMeshForIndex(Index))
        if:
            Self.Prop.TeleportTo[transform{Translation:=Self.Prop.GetTransform().Translation, Rotation:=Self.Prop.GetTransform().Rotation, Scale:=vector3{X:=3.5, Y:=3.5, Z:=3.5}}]
        else:
            Logger.Print("Failed to set resize the you tile, trying once more")
            if:
                Self.Prop.TeleportTo[transform{Translation:=Self.Prop.GetTransform().Translation, Rotation:=Self.Prop.GetTransform().Rotation, Scale:=vector3{X:=3.5, Y:=3.5, Z:=3.5}}]
            then:
                Logger.Print("Successfully resized the you tile")
            else:
                Logger.Print("Failed to resize the you tile")


    SetCard(Style:card_style, Index:int, GameBoard:game_board, ?P1Cover :logic = false):void=
        Self.SetStyle(Style)
        Self.SetCardIndex(Index)

        SavedFrameMaterial := GameBoard.PaintPalette.GetMaterialFromIndexFrame(GameBoard.PaintPalette.FrameColorIndex)

        if(P1CoverMat := P1Cover?):
            var CoverMaterial : Cards.Materials.YouCards.MI_You_Card_Cover_P1_material = Cards.Materials.YouCards.MI_You_Card_Cover_P1_material{}
            set CoverMaterial.CardImageGrid = GetCardGridTextureForCardStyle(Style)
            Self.Prop.SetMaterial(CoverMaterial, ?Index:=1)
            Self.Prop.SetMaterial(SavedFrameMaterial, ?Index:=2)

        else:
            CoverMaterial : Cards.Materials.YouCards.MI_You_Card_Cover_P2_material = Cards.Materials.YouCards.MI_You_Card_Cover_P2_material{}
            set CoverMaterial.CardImageGrid = GetCardGridTextureForCardStyle(Style)
            Self.Prop.SetMaterial(CoverMaterial, ?Index:=1)
            Self.Prop.SetMaterial(SavedFrameMaterial, ?Index:=2)

        # set YouTileCoverMaterial.CardImageGrid = GetCardGridTextureForCardStyle(Style)

        # Self.Prop.SetMaterial(Cards.Materials.M_Card_Master_NEW, ?Index:=1) # hides character, use Sequence to show character
        # Self.Prop.SetMaterial(CoverMaterial, ?Index:=1) # hides character, use Sequence to show character
        # Self.Prop.SetMaterial(SavedFrameMaterial, ?Index:=2)

    Peak(Agent:agent):void=
        Self.CinePeak.Play(Agent) # reveals the cover layer for the player

    Reveal()<suspends>:void=
        Self.CineReveal.Play() # removes the cover layer for all players

        if:
            CardHidden?
            Prop.TeleportTo[transform{Translation:=Prop.GetTransform().Translation, Rotation:=Prop.GetTransform().Rotation.ApplyPitch(PiFloat / 2.0 * 2.0), Scale:=vector3{X:=3.5, Y:=3.5, Z:=3.5}}]
        then:
            # set CardHidden = false
            

        if:
            AnimControler := Self.Prop.GetAnimationController[]
            CurrentRotation := Self.Prop.GetTransform().Rotation.GetYawPitchRollDegrees()
            Direction := if (CurrentRotation[0] < 0.0) then 1.0 else -1.0
        then:
            AnimateReveal(AnimControler, Direction)
                    
    Hide():void=
        Self.CineHide.Play() # replaces the cover layer for all players

        

        
    # (Pitch=0.000000,Yaw=-180.000000,Roll=-152.000000)

    AnimateReveal(AnimationController: animation_controller, Direction:float) <suspends>: void =
        # Initial animation setup
        InitialTransform := Self.Prop.GetTransform()
        InitialScale: vector3 = vector3{X:= 3.5, Y:= 3.5, Z:= 3.5}
        ScaleAmount: vector3 = vector3{X:= 2.0, Y:= 2.0, Z:= 2.0}

        # FirstRoll := 0.001 * Direction
        FirstRoll := -62.0
        FirstYaw := -180.0 * Direction

        # FirstYMovement := 1000.0
        FirstYMovement := if(Direction > 0.0) then 1200.0 else -1200.0
        FirstZMovement := 1200.0

        # FirstRotation :=rotation{}.ApplyRoll(FirstRoll)
        
        FirstRotation := MakeRotationFromYawPitchRollDegrees(0.0, FirstYaw, FirstRoll)
    
        # Define keyframes for the animation
        var Keyframes: []keyframe_delta = array:
            keyframe_delta:
                DeltaLocation := vector3{Y:=FirstYMovement, Z:=FirstZMovement}
                DeltaRotation := FirstRotation
                DeltaScale := ScaleAmount
                Time := 1.0
                Interpolation := InterpolationTypes.EaseIn
            # keyframe_delta:
            #     DeltaLocation := vector3{Z:=-50.0}
            #     DeltaRotation := rotation{}
            #     DeltaScale := vector3{X:=0.9, Y:=0.9, Z:=0.9}
            #     Time := 0.25
            #     Interpolation := InterpolationTypes.EaseIn
            # keyframe_delta:
            #     DeltaLocation := vector3{Z:=50.0}
            #     DeltaRotation := rotation{}
            #     DeltaScale := vector3{X:=1.1, Y:=1.1, Z:=1.1}
            #     Time := 0.25
            #     Interpolation := InterpolationTypes.EaseOut




        # Apply the animation
        AnimationController.SetAnimation(Keyframes, ?Mode:=animation_mode.OneShot)
        AnimationController.Play()


        Logger.Print("Current Position {Prop.GetTransform().Translation} Rotation {Prop.GetTransform().Rotation} Scale {Prop.GetTransform().Scale}")

        Sleep(8.0)


        # Reset the animation
        if:
            Self.Prop.TeleportTo[InitialTransform]
        then:
            Logger.Print("Successfully reset the you tile")
        else:
            Logger.Print("Failed to reset the you tile, trying once more")
            if:
                Self.Prop.TeleportTo[InitialTransform]
            then:
                Logger.Print("Successfully reset the you tile")                
            else:
                Logger.Print("Failed to reset the you tile")

        if:
            CardHidden?
            Prop.TeleportTo[transform{Translation:=InitialTransform.Translation, Rotation:=InitialTransform.Rotation.ApplyPitch(PiFloat / 2.0 * 2.0), Scale:=InitialTransform.Scale}]
        then:
            Logger.Print("Reset the you tile to hidden state")

        Self.Hide()


    HideCharacterCard(Agent:agent)<suspends>:void=
        if:
            not CardHidden?
        then:
            Transform := Prop.GetTransform()
            Rotation := Transform.Rotation
            Location := Transform.Translation
            Flip180 :=Rotation.ApplyPitch(PiFloat / 2.0 * 2.0)
            Prop.MoveTo(Location, Flip180, 0.25)

            SetStreamerModeHideCharacter(Agent, true)
            set CardHidden = true
            if. FlipButtonProp.TeleportTo[FlipButtonProp.GetTransform().Translation + vector3{Z:=512.0}, FlipButtonProp.GetTransform().Rotation]
            FlipButtonProp.SetMaterial(NEW_ASSETS.Materials.MI_Button)

    ShowCharacterCard(Agent:agent)<suspends>:void=
        if:
            CardHidden?
        then:
            Transform := Prop.GetTransform()
            Rotation := Transform.Rotation
            Location := Transform.Translation
            Flip180 :=Rotation.ApplyPitch(PiFloat / 2.0 * 2.0)
            Prop.MoveTo(Location, Flip180, 0.25)

            SetStreamerModeHideCharacter(Agent, false)
            set CardHidden = false
            if. FlipButtonProp.TeleportTo[FlipButtonProp.GetTransform().Translation + vector3{Z:=-512.0}, FlipButtonProp.GetTransform().Rotation]
